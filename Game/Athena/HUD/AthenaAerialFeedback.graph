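# HUD widget on the native AthenaAerialFeedback class that shows an aerial action prompt, such as "To Deploy Glider", beside a key binding.
# BP_HandleTextChanged plays the UMG Intro animation, which fades, drops and scale-pops the whole row with a slight overshoot.
# A FortSimpleWidgetAnimationsPanel around the text also defines a curve-driven Intro for alpha, scale and translation, which this graph does not directly trigger.
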WidgetBlueprint /Game/Athena/HUD/AthenaAerialFeedback
animations [Intro {
  binding_count=1
  length_seconds=0.45
  object=&self:Intro
  sequence {bindings [Panel_Content {
     id=#1
     possessed_class=HorizontalBox
     tracks [
      {
       type=MovieScene2DTransformTrack
       sections [{
         type=MovieScene2DTransformSection
         end_frame=9
         start_frame=0
         channels [
          Translation.X {type=MovieSceneScriptingFloatChannel}
          Translation.Y {type=MovieSceneScriptingFloatChannel, keys [[0,-12.7133],[2,0]]}
          Angle {type=MovieSceneScriptingFloatChannel}
          Scale.X {type=MovieSceneScriptingFloatChannel, keys [[0,0.4518],[3,1.06],[4,1.06],[6,0.97],[9,1]]}
          Scale.Y {type=MovieSceneScriptingFloatChannel, keys [[0,0.9162],[3,1]]}
          Shear.X {type=MovieSceneScriptingFloatChannel}
          Shear.Y {type=MovieSceneScriptingFloatChannel}
         ]
        }]
      }
      {
       type=MovieSceneFloatTrack
       display_name=Render Opacity
       sections [{type=MovieSceneFloatSection, end_frame=9, start_frame=0, channels [None {type=MovieSceneScriptingFloatChannel, keys [[0,0],[2,1]]}]}]
      }
     ]
    }]}
 }]
graphs [EventGraph {nodes [
   Event BP_HandleTextChanged() {is=title K2Node_Event}
   > PlayAnimation(InAnimation=Get Intro())
  ]}]
variables []
widgets {
 info {inheritedWidgetCount=3, parentClass=/Script/FortniteUI.AthenaAerialFeedback, rootWidgetClass=&HorizontalBox, widgetCount=6}
 widgets [
  Panel_Content {widgetClassPath=&HorizontalBox, widget=Panel_Content}
  KeyBind {
   widgetClassPath=/FortUILibrary/Input/KeybindWidget
   bInherited=true
   parent=Panel_Content
   widget=KeyBind
   properties {axisScale=1, renderTransform {angle=0, scale=<1 1>, shear=<0 0>, translation=<0 -2>}}
   slot {class=HorizontalBoxSlot, properties {horizontalAlignment=HAlign_Right, verticalAlignment=VAlign_Center}}
  }
  AnimationsPanel {
   widgetClassPath=/Script/FortniteUI.FortSimpleWidgetAnimationsPanel
   bIsVariable=true
   parent=Panel_Content
   widget=AnimationsPanel
   properties {animations [Intro {
      bIsEnabled=true
      onAnimationsFinished=()
      targets [{
        alpha=&/AthenaAerialFeedback_IntroAlpha
        bResetOnFinish=false
        bShouldAnimateAlpha=true
        bShouldAnimateScale=true
        bShouldAnimateTranslation=true
        scale=&/AthenaAerialFeedback_IntroScale
        targetName=Text_Feedback
        translation=&/AthenaAerialFeedback_IntroTranslation
        materialScalarAnimations []
       }]
     }]}
   slot {class=HorizontalBoxSlot, properties {verticalAlignment=VAlign_Center, padding=<ltrb 4 0 0 0>}}
  }
  SizeBox_0 {
   widgetClassPath=&SizeBox
   parent=AnimationsPanel
   widget=SizeBox_0
   properties {bOverride_MinDesiredWidth=true, minDesiredWidth=350}
   slot {class=FortSimpleWidgetAnimationsSlot}
  }
  Text_Feedback {
   widgetClassPath=&CommonTextBlock
   bInherited=true
   parent=SizeBox_0
   widget=Text_Feedback
   properties {
    text=To Deploy Glider
    textTransformPolicy=ToUpper
    font {
     bForceMonospaced=false
     bMaterialIsStencil=false
     fontMaterial=null
     fontObject=&BurbankSmall
     letterSpacing=0
     monospacedWidth=1
     size=15
     skewAmount=0
     typefaceFontName=Black
     outlineSettings {
      bApplyOutlineToDropShadows=true
      bMiteredCorners=false
      bSeparateFillAlpha=true
      outlineMaterial=&M-FontShadow
      outlineSize=3
      outlineColor=<rgba 0 0 0 0.5>
     }
    }
    shadowOffset=<0 0>
   }
   slot {class=SizeBoxSlot}
  }
  TutorialHighlight {widgetClassPath=&TutorialHighlight}
 ]
}
