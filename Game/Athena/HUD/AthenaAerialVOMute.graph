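# Aerial HUD widget showing a hold-to-mute prompt for VO, with keybind, radial progress ring, feedback text and mic icon.
# The native AthenaAerialVOMute parent raises mute and hold events; mute changes swap the icon texture.
# Hold start plays HoldProgress, rate-scaled to the hold duration, driving the radial material's Percent.
# Hold end stops it and plays HoldReset, which also runs on Construct.
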WidgetBlueprint /Game/Athena/HUD/AthenaAerialVOMute
animations [
 HoldProgress {
  binding_count=1
  length_seconds=1
  object=&self:HoldProgress
  sequence {bindings [Image_Progress {
     id=#1
     possessed_class=Image
     tracks [{
       type=MovieSceneWidgetMaterialTrack
       sections [{type=MovieSceneParameterSection, end_frame=20, start_frame=0, channels [Percent {type=MovieSceneScriptingFloatChannel, keys [[0,0],[20,1]]}]}]
      }]
    }]}
 }
 HoldReset {
  binding_count=1
  length_seconds=0.25
  object=&self:HoldReset
  sequence {bindings [Image_Progress {
     id=#2
     possessed_class=Image
     tracks [{
       type=MovieSceneWidgetMaterialTrack
       sections [{type=MovieSceneParameterSection, end_frame=5, start_frame=0, channels [Percent {type=MovieSceneScriptingFloatChannel, const [0,1]}]}]
      }]
    }]}
 }
]
graphs [EventGraph {nodes [
   n1 = Event OnVOMuteStateChanged!() -> (OutputDelegate, then, bIsMuted:bool) {is=title K2Node_Event}
   > SetBrush(
    self=Get Icon() -> (Icon:Image Object Reference)
    InBrush=MakeBrushFromTexture!(
     Texture:Texture 2D Object Reference=Select(
      Option 0:Texture 2D Object Reference=/Game/Athena/HUD/Art/mic-muted
      Option 1:Texture 2D Object Reference=/Game/Athena/HUD/Art/mic-notalk
      Index=@n1.bIsMuted
     ) -> (ReturnValue:Texture 2D Object Reference)
     Width:int=32
     Height:int=32
    ) -> (ReturnValue:Slate Brush Structure)
   )
   n2 = Event OnVOMuteHoldStarted!() -> (OutputDelegate, then, Duration:float) {is=title K2Node_Event}
   n3 = Get HoldProgress() -> (HoldProgress:Widget Animation Object Reference)
   > PlayAnimationForward(InAnimation=@n3, PlaybackSpeed=float / float(A=GetEndTime(self=@n3), B=@n2.Duration))
   Event OnVOMuteHoldEnded() {is=title K2Node_Event}
   > n4 = StopAnimation(InAnimation=Get HoldProgress())
   PlayAnimationForward(execute=[@n4.then, @n5.then], InAnimation=Get HoldReset() -> (HoldReset:Widget Animation Object Reference))
   n5 = Event Construct()
  ]}]
variables []
widgets {
 info {inheritedWidgetCount=3, parentClass=/Script/FortniteUI.AthenaAerialVOMute, rootWidgetClass=&HorizontalBox, widgetCount=8}
 widgets [
  Panel_Content {widgetClassPath=&HorizontalBox, widget=Panel_Content}
  Overlay_0 {
   widgetClassPath=&Overlay
   parent=Panel_Content
   widget=Overlay_0
   slot {class=HorizontalBoxSlot, properties {horizontalAlignment=HAlign_Right, verticalAlignment=VAlign_Center}}
  }
  Image_Progress {
   widgetClassPath=&Image
   bIsVariable=true
   parent=Overlay_0
   widget=Image_Progress
   properties {
    brush {
     bIsDynamicallyLoaded=false
     drawAs=Image
     imageType=NoImage
     mirroring=NoMirror
     resourceName=None
     resourceObject=/Game/Athena/HUD/Materials/M_UI_RadialProgress
     tiling=NoTile
     imageSize=<32 32>
     margin=<ltrb 0 0 0 0>
     outlineSettings {
      bUseBrushTransparency=false
      roundingType=HalfHeightRadius
      width=0
      color {colorUseRule=UseColor_Specified, specifiedColor=<rgba 0 0 0 0>}
      cornerRadii=<0 0 0 0>
     }
     tintColor {colorUseRule=UseColor_Specified, specifiedColor=<rgba 1 1 1 1>}
     uVRegion {bIsValid=false, max=<0 0>, min=<0 0>}
    }
    renderTransform {angle=0, scale=<1.5 1.5>, shear=<0 0>, translation=<0 -2>}
   }
   slot {class=OverlaySlot, properties {horizontalAlignment=HAlign_Fill, verticalAlignment=VAlign_Fill}}
  }
  KeyBind {
   widgetClassPath=/FortUILibrary/Input/KeybindWidget
   bInherited=true
   parent=Overlay_0
   widget=KeyBind
   properties {axisScale=1, renderTransform {angle=0, scale=<1 1>, shear=<0 0>, translation=<0 -2>}}
   slot {class=OverlaySlot}
  }
  AnimationsPanel {
   widgetClassPath=/Script/FortniteUI.FortSimpleWidgetAnimationsPanel
   bIsVariable=true
   parent=Panel_Content
   widget=AnimationsPanel
   properties {animations [Intro {
      bIsEnabled=true
      onAnimationsFinished=()
      targets [{
        alpha=&/AthenaAerialFeedback_IntroAlpha
        bResetOnFinish=false
        bShouldAnimateAlpha=true
        bShouldAnimateScale=true
        bShouldAnimateTranslation=true
        scale=&/AthenaAerialFeedback_IntroScale
        targetName=Text_Feedback
        translation=&/AthenaAerialFeedback_IntroTranslation
        materialScalarAnimations []
       }]
     }]}
   slot {class=HorizontalBoxSlot, properties {verticalAlignment=VAlign_Center, padding=<ltrb 4 0 0 0>}}
  }
  Text_Feedback {
   widgetClassPath=&CommonTextBlock
   bInherited=true
   parent=AnimationsPanel
   widget=Text_Feedback
   properties {
    text="(Hold) To Mute"
    textTransformPolicy=ToUpper
    font {
     bForceMonospaced=false
     bMaterialIsStencil=false
     fontMaterial=null
     fontObject=&BurbankSmall
     letterSpacing=0
     monospacedWidth=1
     size=15
     skewAmount=0
     typefaceFontName=Black
     outlineSettings {
      bApplyOutlineToDropShadows=true
      bMiteredCorners=false
      bSeparateFillAlpha=true
      outlineMaterial=&M-FontShadow
      outlineSize=3
      outlineColor=<rgba 0 0 0 0.5>
     }
    }
    shadowOffset=<0 0>
   }
   slot {class=FortSimpleWidgetAnimationsSlot}
  }
  Icon {
   widgetClassPath=&Image
   bIsVariable=true
   parent=Panel_Content
   widget=Icon
   properties {brush {
     bIsDynamicallyLoaded=false
     drawAs=Image
     imageType=NoImage
     mirroring=NoMirror
     resourceName=None
     resourceObject=&mic-muted
     tiling=NoTile
     imageSize=<32 32>
     margin=<ltrb 0 0 0 0>
     outlineSettings {
      bUseBrushTransparency=false
      roundingType=HalfHeightRadius
      width=0
      color {colorUseRule=UseColor_Specified, specifiedColor=<rgba 0 0 0 0>}
      cornerRadii=<0 0 0 0>
     }
     tintColor {colorUseRule=UseColor_Specified, specifiedColor=<rgba 1 1 1 1>}
     uVRegion {bIsValid=false, max=<0 0>, min=<0 0>}
    }}
   slot {class=HorizontalBoxSlot, properties {padding=<ltrb 10 0 0 0>}}
  }
  TutorialHighlight {widgetClassPath=&TutorialHighlight}
 ]
}
