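# Athena HUD elimination marker widget: a skull image with a downward arrow and a square pulse, built on the native AthenaEliminationIndicator class.
# On Construct it plays Intro, which pops the skull, slides in the arrow and expands the pulse using transform, opacity and material-parameter tracks.
# A retriggerable delay of the runtime-options marker TTL minus twice the Outro end time then triggers Outro, which shrinks, drops and fades skull and arrow.
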WidgetBlueprint /Game/Athena/HUD/AthenaEliminationIndicator
animations [
 Intro {
  binding_count=3
  length_seconds=0.5752
  object=&self:Intro
  sequence {bindings [
    Skull {
     id=#1
     possessed_class=Image
     tracks [
      {
       type=MovieScene2DTransformTrack
       sections [{
         type=MovieScene2DTransformSection
         channels [
          Translation.X {type=MovieSceneScriptingFloatChannel}
          Translation.Y {type=MovieSceneScriptingFloatChannel, const [0,1]}
          Angle {type=MovieSceneScriptingFloatChannel}
          Scale.X {type=MovieSceneScriptingFloatChannel, keys [[0,0],[7.9293,1.5],[9.0023,0.8544],[11.504,1]]}
          Scale.Y {type=MovieSceneScriptingFloatChannel, keys [[0,0],[7.9293,1.5],[9.0023,0.8544],[11.504,1]]}
          Shear.X {type=MovieSceneScriptingFloatChannel}
          Shear.Y {type=MovieSceneScriptingFloatChannel}
         ]
        }]
      }
      {
       type=MovieSceneFloatVectorTrack
       display_name=Pivot
       sections [{
         type=MovieSceneFloatVectorSection
         channels [
          Vector.X {type=MovieSceneScriptingFloatChannel, const [0.5,1]}
          Vector.Y {type=MovieSceneScriptingFloatChannel, const [0.5,1]}
         ]
        }]
      }
      {
       type=MovieSceneFloatTrack
       display_name=Render Opacity
       sections [{type=MovieSceneFloatSection, channels [None {type=MovieSceneScriptingFloatChannel, keys [[0,0],[3.3833,1]]}]}]
      }
      {
       type=MovieSceneWidgetMaterialTrack
       sections [{
         type=MovieSceneParameterSection
         channels [
          Outline {type=MovieSceneScriptingFloatChannel, keys [[0,1.2674],[7.9293,-0.009],[9.0023,-0.009]]}
          Edge {type=MovieSceneScriptingFloatChannel, keys [[0,1.1004],[7.9293,0.3333],[9.0023,0.3333]]}
         ]
        }]
      }
     ]
    }
    Image_Diamondpulse {
     id=#2
     possessed_class=Image
     tracks [
      {
       type=MovieSceneFloatTrack
       display_name=Render Opacity
       sections [{type=MovieSceneFloatSection, channels [None {type=MovieSceneScriptingFloatChannel, keys [[0,1],[7.9293,1],[9.9633,0]]}]}]
      }
      {
       type=MovieSceneWidgetMaterialTrack
       sections [{
         type=MovieSceneParameterSection
         channels [
          OutlineWidth {type=MovieSceneScriptingFloatChannel, keys [[0,0.1],[9.9633,0.01]]}
          Scale {type=MovieSceneScriptingFloatChannel, keys [[0,0],[9.9633,0.6946]]}
         ]
        }]
      }
     ]
    }
    Arrow {
     id=#3
     possessed_class=Image
     tracks [
      {
       type=MovieScene2DTransformTrack
       sections [{
         type=MovieScene2DTransformSection
         channels [
          Translation.X {type=MovieSceneScriptingFloatChannel}
          Translation.Y {type=MovieSceneScriptingFloatChannel, keys [[6.3767,-10.6929],[10.3267,15.1751],[11.504,12]]}
          Scale.X {type=MovieSceneScriptingFloatChannel, keys [[6.3767,0.5],[11.504,1]]}
          Scale.Y {type=MovieSceneScriptingFloatChannel, keys [[6.3767,0.5],[11.504,1]]}
         ]
        }]
      }
      {
       type=MovieSceneFloatTrack
       display_name=Render Opacity
       sections [{type=MovieSceneFloatSection, channels [None {type=MovieSceneScriptingFloatChannel, keys [[6.3767,0],[11.504,1]]}]}]
      }
     ]
    }
   ]}
 }
 Outro {
  binding_count=2
  length_seconds=0.2
  object=&self:Outro
  sequence {bindings [
    Skull {
     id=#4
     possessed_class=Image
     tracks [
      {
       type=MovieSceneFloatTrack
       display_name=Render Opacity
       sections [{type=MovieSceneFloatSection, channels [None {type=MovieSceneScriptingFloatChannel, keys [[0,1],[4,0]]}]}]
      }
      {
       type=MovieScene2DTransformTrack
       sections [{
         type=MovieScene2DTransformSection
         channels [
          Translation.X {type=MovieSceneScriptingFloatChannel}
          Translation.Y {type=MovieSceneScriptingFloatChannel, keys [[0,0],[4,29.5364]]}
          Angle {type=MovieSceneScriptingFloatChannel}
          Scale.X {type=MovieSceneScriptingFloatChannel, keys [[0,1],[4,0.2]]}
          Scale.Y {type=MovieSceneScriptingFloatChannel, keys [[0,1],[4,0.2]]}
          Shear.X {type=MovieSceneScriptingFloatChannel}
          Shear.Y {type=MovieSceneScriptingFloatChannel}
         ]
        }]
      }
     ]
    }
    Arrow {
     id=#5
     possessed_class=Image
     tracks [
      {
       type=MovieScene2DTransformTrack
       sections [{
         type=MovieScene2DTransformSection
         channels [
          Translation.X {type=MovieSceneScriptingFloatChannel}
          Translation.Y {type=MovieSceneScriptingFloatChannel, keys [[0,12],[4,20]]}
          Scale.X {type=MovieSceneScriptingFloatChannel, keys [[0,1],[4,0.2]]}
          Scale.Y {type=MovieSceneScriptingFloatChannel, keys [[0,1],[4,0.2]]}
         ]
        }]
      }
      {
       type=MovieSceneFloatTrack
       display_name=Render Opacity
       sections [{type=MovieSceneFloatSection, channels [None {type=MovieSceneScriptingFloatChannel, keys [[0,1],[4,0]]}]}]
      }
     ]
    }
   ]}
 }
]
graphs [EventGraph {nodes [
   Event Construct()
   > StopAnimation(InAnimation=Get Outro() -> (Outro:Widget Animation Object Reference))
   > PlayAnimation(InAnimation=Get Intro())
   > n1 = GetEliminationMarkerTTL!(execute, self:Fort Runtime Options Object Reference=GetRuntimeOptions()) -> (then, ReturnValue:float)
   n2 = GetEndTime(self=Get Outro())
   > RetriggerableDelay(Duration=float - float(A=@n1, B=float * float(A=@n2, B=2)) {is=title K2Node_CallFunction})
   > StopAnimation(InAnimation=Get Intro())
   > PlayAnimation(InAnimation=Get Outro())
  ]}]
variables []
widgets {
 info {inheritedWidgetCount=5, parentClass=/Script/FortniteUI.AthenaEliminationIndicator, rootWidgetClass=&VerticalBox, widgetCount=11}
 widgets [
  VerticalBox_0 {widgetClassPath=&VerticalBox, widget=VerticalBox_0}
  Overlay_0 {
   widgetClassPath=&Overlay
   parent=VerticalBox_0
   widget=Overlay_0
   slot {class=VerticalBoxSlot, properties {horizontalAlignment=HAlign_Center, verticalAlignment=VAlign_Center}}
  }
  CommonBorder_0 {
   widgetClassPath=&CommonBorder
   parent=Overlay_0
   widget=CommonBorder_0
   properties {visibility=HitTestInvisible, padding=<ltrb 0 0 0 0>}
   slot {class=OverlaySlot, properties {horizontalAlignment=HAlign_Center, verticalAlignment=VAlign_Center}}
  }
  Arrow {
   widgetClassPath=&Image
   bIsVariable=true
   parent=CommonBorder_0
   widget=Arrow
   properties {
    visibility=HitTestInvisible
    brush {
     bIsDynamicallyLoaded=false
     drawAs=Image
     imageType=NoImage
     mirroring=NoMirror
     resourceName=None
     resourceObject=/Game/UI/CodeReferenced/UI_Radar_AthenaPlayerArrow
     tiling=NoTile
     imageSize=<25 10>
     margin=<ltrb 0 0 0 0>
     outlineSettings {
      bUseBrushTransparency=false
      roundingType=HalfHeightRadius
      width=0
      color {colorUseRule=UseColor_Specified, specifiedColor=<rgba 0 0 0 0>}
      cornerRadii=<0 0 0 0>
     }
     tintColor {colorUseRule=UseColor_Specified, specifiedColor=<rgba 1 1 1 1>}
     uVRegion {bIsValid=false, max=<0 0>, min=<0 0>}
    }
    renderTransform {angle=180, scale=<1 1>, shear=<0 0>, translation=<0 12>}
    renderTransformPivot=<0.5 1>
   }
   slot {class=BorderSlot, properties {padding=<ltrb 0 0 0 0>}}
  }
  Skull {
   widgetClassPath=&Image
   parent=Overlay_0
   widget=Skull
   properties {
    visibility=SelfHitTestInvisible
    brush {
     bIsDynamicallyLoaded=false
     drawAs=Image
     imageType=NoImage
     mirroring=NoMirror
     resourceName=None
     resourceObject=/Game/Athena/HUD/Materials/M_UI_EliminationMarker_Inst
     tiling=NoTile
     imageSize=<32 32>
     margin=<ltrb 0 0 0 0>
     outlineSettings {
      bUseBrushTransparency=false
      roundingType=HalfHeightRadius
      width=0
      color {colorUseRule=UseColor_Specified, specifiedColor=<rgba 0 0 0 0>}
      cornerRadii=<0 0 0 0>
     }
     tintColor {colorUseRule=UseColor_Specified, specifiedColor=<rgba 1 1 1 1>}
     uVRegion {bIsValid=false, max=<0 0>, min=<0 0>}
    }
   }
   slot {class=OverlaySlot}
  }
  Image_Diamondpulse {
   widgetClassPath=&Image
   bIsVariable=true
   parent=Overlay_0
   widget=Image_Diamondpulse
   properties {
    renderOpacity=0
    brush {
     bIsDynamicallyLoaded=false
     drawAs=Image
     imageType=NoImage
     mirroring=NoMirror
     resourceName=None
     resourceObject=/Game/Athena/Prototype/Blueprints/Galileo/Materials/M_UI_SquarePulse_Inst
     tiling=NoTile
     imageSize=<150 150>
     margin=<ltrb 0 0 0 0>
     outlineSettings {
      bUseBrushTransparency=false
      roundingType=HalfHeightRadius
      width=0
      color {colorUseRule=UseColor_Specified, specifiedColor=<rgba 0 0 0 0>}
      cornerRadii=<0 0 0 0>
     }
     tintColor {colorUseRule=UseColor_Specified, specifiedColor=<rgba 1 1 1 1>}
     uVRegion {bIsValid=false, max=<0 0>, min=<0 0>}
    }
   }
   slot {class=OverlaySlot, properties {horizontalAlignment=HAlign_Center, verticalAlignment=VAlign_Center, padding=<ltrb -200 -200 -200 -200>}}
  }
  Switcher_IndicatorSize {widgetClassPath=&WidgetSwitcher, bInherited=true, widget=null}
  Panel_SmallIndicator {widgetClassPath=/Script/UMG.PanelWidget, bInherited=true, widget=null}
  Panel_LargeIndicator {widgetClassPath=&PanelWidget, bInherited=true, widget=null}
  Text_Distance {widgetClassPath=&CommonTextBlock, bInherited=true, widget=null}
  TutorialHighlight {widgetClassPath=&TutorialHighlight}
 ]
}
