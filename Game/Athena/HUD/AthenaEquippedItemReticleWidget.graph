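# Reticle ammo-meter widget for the equipped weapon, drawing a radial-progress material for ammo and forwarding overheat values to its native base.
# Weapon events update the gauge and fade it in or out on reload, targeting and secondary fire.
# Reloads drive material scalars through a curve-based simple-animation panel, scaled to reload time, then restore or hide the parameters on finish.
# In compact UI, the meter's offset scales with viewport aspect ratio.
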WidgetBlueprint /Game/Athena/HUD/AthenaEquippedItemReticleWidget
animations [Fadeout {
  binding_count=1
  length_seconds=0.4969
  object=&self:Fadeout
  sequence {
   bindings [Image_AmmoMeter {
     id=#1
     display_name=CommonBorder_AmmoCounter
     possessed_class=Image
     tracks [{
       type=MovieSceneFloatTrack
       display_name=Render Opacity
       sections [{type=MovieSceneFloatSection, channels [None {type=MovieSceneScriptingFloatChannel, keys [[0,1],[9.937,0]]}]}]
      }]
    }]
   tracks [{
     type=MovieSceneEventTrack
     sections [{
       type=MovieSceneEventTriggerSection
       channels [None {type=MovieSceneScriptingEventChannel, const ["<Struct 'MovieSceneEvent' (0x46794523c00) {}>", 1]}]
      }]
    }]
  }
 }]
graphs [
 EventGraph {nodes [
   HideReloadParams() {is=title K2Node_CustomEvent}
   > n1 = Branch(Condition=IsAnimationPlaying!(
     self:Fort Simple Widget Animations Panel Object Reference=Get SimpleAnimationsPanel() -> (SimpleAnimationsPanel:Fort Simple Widget Animations Panel Object Reference)
     Name:"Name (by ref)"=Get SimpleAnimationReload() -> (SimpleAnimationReload:Name=None)
    ) -> (ReturnValue))
   n2 = Get AmmoCounterMaterialInstanceDynamic() -> (AmmoCounterMaterialInstanceDynamic:Material Instance Dynamic Object Reference)
   SetScalarParameterValue(execute=@n1.else, self=@n2, ParameterName=Get AmmoCounterOverridePercentParam() -> (AmmoCounterOverridePercentParam:Name=None), Value=0)
   > SetScalarParameterValue(self=@n2, ParameterName=Get AmmoCounterSegmentOverrideParam() -> (AmmoCounterSegmentOverrideParam:Name=None), Value=0)
   > SetScalarParameterValue(
    self=@n2
    ParameterName=Get AmmoCounterPercentParam() -> (AmmoCounterPercentParam:Name=None)
    Value=Get CurrentAmmoPercentage() -> (CurrentAmmoPercentage:float)
   )
   n3 = OnSimpleAnimationFinished!() -> (OutputDelegate, then, Name:Name) {is=title K2Node_CustomEvent}
   > Branch(Condition="Equal (Name)"(A=@n3.Name, B=Get SimpleAnimationReload() -> (SimpleAnimationReload=None)))
   > FadeAmmoInOut(self:Athena Equipped Item Reticle Widget Base Object Reference)
   > n4 = Branch(Condition=Get CanOverheat() -> (CanOverheat:bool))
   HideReloadParams(execute=@n4.else) {is=title K2Node_CallFunction}
   OnHUDVisibilityRefreshed() {is=title K2Node_CustomEvent}
   > NativeUpdateForWeapon!(
    execute
    self:Athena Equipped Item Reticle Widget Base Object Reference
    InNewWeapon:Fort Weapon Object Reference=Get CurrentEquippedWeapon() -> (CurrentEquippedWeapon:Fort Weapon Object Reference)
    InSimpleAnimationsPanel:Fort Simple Widget Animations Panel Object Reference=Get SimpleAnimationsPanel()
   ) -> (then)
   n5 = Event Handle Weapon Overheat Value Changed!() -> (OutputDelegate, then, CurrentValue:float, MaxValue:float, State:EFortWeaponOverheatState Enum=None, bIsInCooldown:bool) {is=title K2Node_Event}
   > NativeUpdateOverheatValue!(
    execute
    self:Athena Equipped Item Reticle Widget Base Object Reference
    Current:float=@n5.CurrentValue
    Max:float=@n5.MaxValue
    State:EFortWeaponOverheatState Enum=@n5.State
    bIsInCooldown:bool=@n5.bIsInCooldown
   ) -> (then)
   n6 = Event Handle Weapon Targeting Changed!() -> (OutputDelegate, then, bIsTargeting:bool) {is=title K2Node_Event}
   > Branch(Condition="Not Equal (Enum)"!(
     A:ESlateVisibility Enum=Get WeaponEquipVisibility() -> (WeaponEquipVisibility:ESlateVisibility Enum=Visible)
     B:ESlateVisibility Enum=Collapsed
    ) -> (ReturnValue:bool))
   > n7 = Branch(Condition=Get bShouldHideReticleAmmoCounterWhileTargeting() -> (bShouldHideReticleAmmoCounterWhileTargeting:bool))
   > n8 = Branch(Condition=@n6.bIsTargeting)
   > FadeAmmoOut!(execute, self:Athena Equipped Item Reticle Widget Base Object Reference, PlaybackSpeed:float=2) -> (then)
   FadeAmmoInOut(execute=[@n8.else, @n9.then])
   n9 = Branch(execute=@n7.else, Condition=@n6.bIsTargeting)
   n10 = Event Handle Weapon Reload Cancelled!() -> (OutputDelegate, then, Weapon:Fort Weapon Object Reference) {is=title K2Node_Event}
   > StopAnimationByName!(
    execute
    self:Fort Simple Widget Animations Panel Object Reference=Get SimpleAnimationsPanel()
    Name:"Name (by ref)"=Get SimpleAnimationReload() -> (SimpleAnimationReload=None)
   ) -> (then, ReturnValue:bool)
   > OnSimpleAnimationFinished!(execute, self, Name:Name=Get SimpleAnimationReload() -> (SimpleAnimationReload=None)) -> (then) {is=title K2Node_CallFunction}
   > NativeUpdateGaugeValue!(execute, self:Athena Equipped Item Reticle Widget Base Object Reference, InWeapon:Fort Weapon Object Reference=@n10.Weapon) -> (then)
   n11 = Event Handle Weapon Reload Started!() -> (OutputDelegate, then, ReloadTime:float, ReloadType:EFortWeaponReloadType Enum=ReloadWholeClip) {is=title K2Node_Event}
   > n12 = Cast To FortWeaponRanged!(execute, Object=Get CurrentEquippedWeapon()) -> (then, CastFailed, AsFort Weapon Ranged:Fort Weapon Ranged Object Reference) {is=title K2Node_DynamicCast}
   > n13 = Branch(Condition=ShouldDisplayReticleAmmoCounter!(self:Fort Weapon Ranged Object Reference=@n12) -> (ReturnValue:bool))
   n14 = Branch(
    execute=[@n13.then, @n12.CastFailed]
    Condition="Equal (Enum)"(A:EFortWeaponReloadType Enum=@n11.ReloadType, B:EFortWeaponReloadType Enum=ReloadIndividualBullets)
   )
   n15 = Get Scalar Parameter Value!(
    execute=@n14.else
    self:Material Instance Dynamic Object Reference=Get AmmoCounterMaterialInstanceDynamic()
    ParameterName:Name=Get AmmoCounterPercentParam() -> (AmmoCounterPercentParam=None)
   ) -> (then, ReturnValue:float) {is=title K2Node_CallFunction}
   > Set CurrentAmmoPercentage(CurrentAmmoPercentage:float=@n15) -> (Output_Get:float)
   > SetScalarParameterValue(self=Get AmmoCounterMaterialInstanceDynamic(), ParameterName=Get AmmoCounterPercentParam() -> (AmmoCounterPercentParam=None), Value=0)
   > SetScalarParameterValue(
    self=Get AmmoCounterMaterialInstanceDynamic()
    ParameterName=Get AmmoCounterSegmentOverrideParam() -> (AmmoCounterSegmentOverrideParam=None)
    Value=1
   )
   n16 = "float > float"(A=@n11.ReloadTime, B=0) {is=title K2Node_CallFunction}
   n17 = Get CurrentEquippedWeapon()
   > PlayAnimationByName!(
    execute
    self:Fort Simple Widget Animations Panel Object Reference=Get SimpleAnimationsPanel()
    Name:"Name (by ref)"=Get SimpleAnimationReload() -> (SimpleAnimationReload=None)
    Speed:float=float / float(A=1, B=Select(Option 0:double=1, Option 1:double=@n11.ReloadTime, Index=@n16) -> (ReturnValue:double)) {is=title K2Node_CallFunction}
    StartAtTime:float=SafeDivide(A=GetReloadStartTime!(self:Fort Weapon Object Reference=@n17) -> (ReturnValue:float), B=@n11.ReloadTime)
   ) -> (then, ReturnValue:bool)
   > Clear and Invalidate Timer by Handle(Handle=Get FadeOutTimerHandle() -> (FadeOutTimerHandle:Timer Handle Structure))
   > FadeAmmoIn(self:Athena Equipped Item Reticle Widget Base Object Reference)
   Event Handle Secondary Fire Pressed() {is=title K2Node_Event}
   > n18 = Branch(Condition=Get bIsHoldingSecondaryFire() -> (bIsHoldingSecondaryFire:bool))
   Set bIsHoldingSecondaryFire(execute=@n18.else, bIsHoldingSecondaryFire:bool=true) -> (Output_Get:bool)
   > Clear and Invalidate Timer by Handle(Handle=Get FadeOutTimerHandle())
   > FadeAmmoIn()
   Event Handle Secondary Fire Released() {is=title K2Node_Event}
   > Branch(Condition=Get bIsHoldingSecondaryFire())
   > Set bIsHoldingSecondaryFire(bIsHoldingSecondaryFire=false)
   > FadeAmmoInOut()
   Event OnInitialized()
   > Bind Event to On Animations Finished(self:Fort Simple Widget Animations Panel Object Reference=Get SimpleAnimationsPanel(), Delegate=Create Event()) {is=title K2Node_AddDelegate}
   > n19 = Cast To FortPlayerControllerGameplay!(execute, Object=GetOwningPlayer()) -> (then, CastFailed, AsFort Player Controller Gameplay:Fort Player Controller Gameplay Object Reference) {is=title K2Node_DynamicCast}
   > Bind Event to On HUDElement Visibility Refreshed Delegate(self:Fort Player Controller Gameplay Object Reference=@n19, Delegate=Create Event()) {is=title K2Node_AddDelegate}
   Event Handle Ammo Clip Attribute Changed!() -> (OutputDelegate, then, Value:float) {is=title K2Node_Event}
   > NativeUpdateGaugeValue(InWeapon=Get CurrentEquippedWeapon())
   Event Construct()
   > TranslationsAdjustments()
   > n20 = GetDynamicMaterial(self=Get Image_AmmoMeter() -> (Image_AmmoMeter:Image Object Reference))
   > Set AmmoCounterMaterialInstanceDynamic(AmmoCounterMaterialInstanceDynamic:Material Instance Dynamic Object Reference=@n20) -> (Output_Get:Material Instance Dynamic Object Reference)
   > Set AmmoCounter(AmmoCounter:Common Border Object Reference=Get CommonBorder_AmmoCounter() -> (CommonBorder_AmmoCounter:Common Border Object Reference)) -> (Output_Get:Common Border Object Reference)
   > Set FadeoutAnimation(FadeoutAnimation:Widget Animation Object Reference=Get Fadeout() -> (Fadeout:Widget Animation Object Reference)) -> (Output_Get:Widget Animation Object Reference)
   > Set AnimationsPanel(AnimationsPanel:Fort Simple Widget Animations Panel Object Reference=Get SimpleAnimationsPanel()) -> (Output_Get:Fort Simple Widget Animations Panel Object Reference)
   > Set AmmoMeter(AmmoMeter:Image Object Reference=Get Image_AmmoMeter()) -> (Output_Get:Image Object Reference)
  ]}
 TranslationsAdjustments {nodes [
   TranslationsAdjustments() {is=title K2Node_FunctionEntry}
   > Branch(Condition=IsUsingCompactUI())
   n1 = BreakVector2D(InVec=GetViewportSize!() -> (ReturnValue:Vector 2D Structure))
   n2 = SafeDivide(A=@n1, B=@n1.Y)
   n3 = SafeDivide(A=@n1.Y, B=@n1)
   n4 = "float > float"(A=@n1, B=@n1.Y) {is=title K2Node_PromotableOperator}
   n5 = "Clamp (Float)"(Value=float - float(A=SelectFloat(A=@n2, B=@n3, bPickA=@n4), B=Make Literal Float(Value=1.77)), Min=0, Max=2)
   > SetRenderTranslation(
    self=Get CommonBorder_AmmoCounter()
    Translation=vector2d * vector2d(A:Vector 2D Structure=Get DefaultTranslation() -> (DefaultTranslation:Vector 2D Structure), B=float + float(A=@n5, B=1)) -> (ReturnValue:Vector 2D Structure) {is=title K2Node_PromotableOperator}
   )
  ]}
]
variables [
 CanOverheat {type=boolean}
 DefaultTranslation {type=Vector2D, default=<25 25>}
 Icons Size {type=Vector2D, default=<128 128>}
]
widgets {
 info {
  inheritedWidgetCount=21
  parentClass=/Script/FortniteUI.AthenaEquippedItemReticleWidgetBase
  rootWidgetClass=&SizeBox
  widgetCount=25
 }
 widgets [
  SizeBox_0 {
   widgetClassPath=&SizeBox
   widget=SizeBox_0
   properties {bOverride_MinDesiredHeight=true, bOverride_MinDesiredWidth=true, minDesiredHeight=200, minDesiredWidth=200}
  }
  RootScaleBox {
   widgetClassPath=&ScaleBox
   bInherited=true
   bIsVariable=true
   parent=SizeBox_0
   widget=RootScaleBox
   properties {stretch=UserSpecified}
   slot {class=SizeBoxSlot}
  }
  SimpleAnimationsPanel {
   widgetClassPath=/Script/FortniteUI.FortSimpleWidgetAnimationsPanel
   bIsVariable=true
   parent=RootScaleBox
   widget=SimpleAnimationsPanel
   properties {animations [Reload {
      bIsEnabled=true
      onAnimationsFinished=()
      targets [{
        alpha=null
        bResetOnFinish=false
        bShouldAnimateAlpha=false
        bShouldAnimateScale=false
        bShouldAnimateTranslation=false
        scale=null
        targetName=Image_AmmoMeter
        translation=null
        materialScalarAnimations [
         {materialParameter=OverridePercent, scalarCurve=&/AthenaEquippedItemReticleReloadPercentCurve}
         {materialParameter=Percent, scalarCurve=&/AthenaEquippedItemReticleReloadOverrideCurve}
        ]
       }]
     }]}
   slot {class=ScaleBoxSlot}
  }
  CommonBorder_AmmoCounter {
   widgetClassPath=&CommonBorder
   bIsVariable=true
   parent=SimpleAnimationsPanel
   widget=CommonBorder_AmmoCounter
   properties {padding=<ltrb 0 0 0 0>, renderTransform {angle=0, scale=<1 1>, shear=<0 0>, translation=<25 25>}}
   slot {class=FortSimpleWidgetAnimationsSlot}
  }
  Image_AmmoMeter {
   widgetClassPath=&Image
   bIsVariable=true
   parent=CommonBorder_AmmoCounter
   widget=Image_AmmoMeter
   properties {
    visibility=Collapsed
    brush {
     bIsDynamicallyLoaded=false
     drawAs=Image
     imageType=NoImage
     mirroring=NoMirror
     resourceName=None
     resourceObject=/Game/Athena/HUD/Materials/M_UI_RadialProgress_Ammo
     tiling=NoTile
     imageSize=<64 64>
     margin=<ltrb 0 0 0 0>
     outlineSettings {
      bUseBrushTransparency=false
      roundingType=HalfHeightRadius
      width=0
      color {colorUseRule=UseColor_Specified, specifiedColor=<rgba 0 0 0 0>}
      cornerRadii=<0 0 0 0>
     }
     tintColor {colorUseRule=UseColor_Specified, specifiedColor=<rgba 1 1 1 1>}
     uVRegion {bIsValid=false, max=<0 0>, min=<0 0>}
    }
    renderTransform {angle=0, scale=<-1 -1>, shear=<0 0>, translation=<0 0>}
   }
   slot {class=BorderSlot, properties {padding=<ltrb 0 0 0 0>}}
  }
  Switcher_Content {widgetClassPath=&CommonVisibilitySwitcher, bInherited=true, widget=null}
  Panel_ResourceContent {widgetClassPath=/Script/UMG.PanelWidget, bInherited=true, widget=null}
  Panel_AmmoContent {widgetClassPath=&PanelWidget, bInherited=true, widget=null}
  Panel_UtilityContent {widgetClassPath=&PanelWidget, bInherited=true, widget=null}
  Image_AmmoIcon {widgetClassPath=&Image, bInherited=true, widget=null}
  Image_AmmoSplit {widgetClassPath=&Image, bInherited=true, widget=null}
  Image_UtilityIcon {widgetClassPath=&Image, bInherited=true, widget=null}
  Image_ResourceIcon {widgetClassPath=&Image, bInherited=true, widget=null}
  Image_UtilitySplit {widgetClassPath=&Image, bInherited=true, widget=null}
  Image_RechargeableIcon {widgetClassPath=&Image, bInherited=true, widget=null}
  Text_MagazineAmmoCount {widgetClassPath=&CommonTextBlock, bInherited=true, widget=null}
  Text_SurplusAmmoCount {widgetClassPath=&CommonTextBlock, bInherited=true, widget=null}
  Text_ResourceCount {widgetClassPath=&CommonTextBlock, bInherited=true, widget=null}
  Text_SurplusUtilityAmmoCount {widgetClassPath=&CommonTextBlock, bInherited=true, widget=null}
  Text_UtilityAmmoCount {widgetClassPath=&CommonTextBlock, bInherited=true, widget=null}
  Text_OutOfAmmo {widgetClassPath=&CommonTextBlock, bInherited=true, widget=null}
  Overlay_UtilityAmmoIcon {widgetClassPath=&Overlay, bInherited=true, widget=null}
  Image_Background {widgetClassPath=&Image, bInherited=true, widget=null}
  ScaleBox_HUDWidgetScale {widgetClassPath=&ScaleBox, bInherited=true, widget=null}
  TutorialHighlight {widgetClassPath=&TutorialHighlight}
 ]
}
