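# HUD element widget for situational vehicle feedback, built around a vertical box of feedback lines that can be shown or hidden.
# On Construct it rebuilds a pool of feedback line widgets, looping up to MaxLinesNeeded.
# ShowFeedback reveals the box and, with bAutoHide set, uses a retriggerable delay to collapse it.
# IsKeyInUse reports whether any line binds a key while the widget is visible, and UpdateFeedbackStyles passes a text style table to each line.
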WidgetBlueprint /Game/Athena/HUD/WBP_AthenaHUDSituationalVehicleFeedback
graphs [
 EventGraph {nodes [
   n1 = HideFeedback()
   Clear(
    execute=[@n1.then, @n3.then]
    TargetArray:Array of WBP Athena HUDFeedback Line New CD Object References=Get FeedbackLines() -> (FeedbackLines:Array of WBP Athena HUDFeedback Line New CD Object References)
   )
   > For Loop(LastIndex=Get MaxLinesNeeded() -> (MaxLinesNeeded:int=0))
   > n2 = Create Widget!(
    execute:Exec
    Class:User Widget Class Reference=&/WBP_AthenaHUDFeedbackLineNewCD.WBP_AthenaHUDFeedbackLineNewCD_C
    OwningPlayer:Player Controller Object Reference=GetOwningPlayer()
   ) -> (then:Exec, ReturnValue:WBP Athena HUDFeedback Line New CD Object Reference)
   > Add(
    TargetArray:Array of WBP Athena HUDFeedback Line New CD Object References=Get FeedbackLines()
    NewItem:"WBP Athena HUDFeedback Line New CD Object Reference (by ref)"=@n2
   ) {is=title K2Node_CallArrayFunction}
   HideFeedback() {is=title K2Node_CustomEvent}
   > SetVisibility(self=Get Feedback() -> (Feedback:Vertical Box Object Reference), InVisibility=Collapsed)
   ShowFeedback() {is=title K2Node_CustomEvent}
   > SetVisibility(self=Get Feedback(), InVisibility=HitTestInvisible)
   > Branch(Condition=Get bAutoHide() -> (bAutoHide:bool))
   > RetriggerableDelay(Duration=Get AutoHideTime() -> (AutoHideTime:double))
   > HideFeedback() {is=title K2Node_CallFunction}
   n3 = Event Construct()
  ]}
 IsKeyInUse {nodes [
   n1 = IsKeyInUse!() -> (then, Key:Key Structure=None) {is=title K2Node_FunctionEntry}
   > n2 = Branch(Condition=IsVisible())
   > n3 = For Each Loop(Array:Array of WBP Athena HUDFeedback Line New CD Object References=Get FeedbackLines()) -> (Array Element:WBP Athena HUDFeedback Line New CD Object Reference)
   Return Node!(execute=[@n3.Completed, @n2.else], InUse:bool=false)
   > Branch(Condition=IsKeyBound!(self:WBP Athena HUDFeedback Line New CD Object Reference=@n3, Key:Key Structure=@n1) -> (IsBound:bool))
   > Return Node(InUse=true)
  ]}
 UpdateFeedbackStyles {nodes [
   n1 = UpdateFeedbackStyles!() -> (then, Style:Data Table Object Reference) {is=title K2Node_FunctionEntry}
   > n2 = For Each Loop(Array=Get FeedbackLines())
   > Is Valid(InputObject=@n2)
   > SetTextStyleSet!(execute, self:WBP Athena HUDFeedback Line New CD Object Reference=@n2, NewTextStyleSet:Data Table Object Reference=@n1) -> (then)
  ]}
]
variables [
 AutoHideTime {type=number, default=15}
 bAutoHide {type=boolean}
 DefaultTextStyleSet {type=DataTable, default=/Game/Athena/HUD/TextStyles/PlayersLeftStyles}
 FeedbackLines {type="array<WBP_AthenaHUDFeedbackLineNewCD_C>", default []}
 MaxLinesNeeded {type=integer, default=6}
]
widgets {
 info {inheritedWidgetCount=1, parentClass=&FortHUDElementWidget, rootWidgetClass=&VerticalBox, widgetCount=2}
 widgets [
  Feedback {widgetClassPath=&VerticalBox, bIsVariable=true, widget=Feedback}
  TutorialHighlight {widgetClassPath=&TutorialHighlight}
 ]
}
