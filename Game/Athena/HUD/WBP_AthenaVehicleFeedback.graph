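# Child HUD widget that shows input prompts for the local player's vehicle seat actions.
# Enter, seat, slot and keybind events queue one next-tick refresh, which fetches action display data from the vehicle interface.
# Refresh refills inherited, reused feedback lines, supporting legacy and Enhanced Input, multiple bindings and disabled tags.
# It also collapses while local challenges are visible and updates text styles when picture-in-picture blocking changes.
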WidgetBlueprint /Game/Athena/HUD/WBP_AthenaVehicleFeedback
graphs [
 AddFeedbackLine {nodes [
   n1 = AddFeedbackLine!() -> (then, Action:Action Def To Display Structure, ActionIndex:int) {is=title K2Node_FunctionEntry}
   > Branch(Condition="integer < integer"!(A:int=@n1.ActionIndex, B:int=Get MaxLinesNeeded() -> (MaxLinesNeeded:int=0)) -> (ReturnValue:bool) {is=title K2Node_CallFunction})
   > Set ActionWidget(ActionWidget:Athena HUDFeedback Line Object Reference="Get (a copy)"(
     Array:Array of WBP Athena HUDFeedback Line New CD Object References=Get FeedbackLines() -> (FeedbackLines:Array of WBP Athena HUDFeedback Line New CD Object References)
     Dimension 1=Get ActionIndex() -> (ActionIndex:int=0)
    ) -> (Output:WBP Athena HUDFeedback Line New CD Object Reference)) -> (Output_Get:Athena HUDFeedback Line Object Reference)
   > n2 = For Each Loop(Array:Array of Input Action Display Structures=Break Action Def to Display!(ActionDefToDisplay:"Action Def To Display Structure (by ref)"=Get Action() -> (Action:Action Def To Display Structure)) -> (
     LegacyInputAction:Name
     EnhancedInputAction:Input Action Object Reference
     InputActions:Array of Input Action Display Structures
     ActionLabel:Text
     DisabledTags:Gameplay Tag Container Structure
    ) {is=title K2Node_BreakStruct}.InputActions) -> (Array Element:Input Action Display Structure)
   > Set CurrentActionDisplay(CurrentActionDisplay:Input Action Display Structure=@n2) -> (Output_Get:Input Action Display Structure)
   > n3 = Branch(Condition="Equal (Integer)"(A=@n2.Array Index))
   n4 = Break Input Action Display!(InputActionDisplay:"Input Action Display Structure (by ref)"=Get CurrentActionDisplay() -> (CurrentActionDisplay:Input Action Display Structure)) -> (LegacyInputAction:Name, EnhancedInputAction:Input Action Object Reference, bShowSecondaryBoundKey:bool=~) {is=title K2Node_BreakStruct}
   > n5 = Branch(Condition="Not Equal (Name)"(A=@n4) -> (ReturnValue=~))
   n6 = Break Action Def to Display(ActionDefToDisplay=Get Action())
   > n7 = SetFeedback!(
    execute
    self:Athena HUDFeedback Line Object Reference=Get ActionWidget() -> (ActionWidget:Athena HUDFeedback Line Object Reference)
    NewBoundAction:Name=@n4
    FeedbackText:"Text (by ref)"=@n6.ActionLabel
    bShowEnabled:bool=true
    AxisScale:float=0
    bHasMultipleFeedbackLines:bool=false
    bShowSecondaryBoundKey:bool=@n4.bShowSecondaryBoundKey
   ) -> (then)
   SetVisibility(execute=[@n7.then, @n8.then], self=Get ActionWidget(), InVisibility=SelfHitTestInvisible)
   > SetDisabledTags!(
    execute
    self:Athena HUDFeedback Line Base Object Reference=Get ActionWidget()
    InDisabledTags:"Gameplay Tag Container Structure (by ref)"=Break Action Def to Display(ActionDefToDisplay=Get Action()).DisabledTags
   ) -> (then)
   > SetDividerVisible!(execute, self:Athena HUDFeedback Line Object Reference=Get ActionWidget(), bIsVisible:bool=false) -> (then)
   > Branch(Condition=Get bShouldClearFeedbackPanel() -> (bShouldClearFeedbackPanel:bool))
   > AddChild(self=Get Feedback() -> (Feedback:Vertical Box Object Reference), Content=Get ActionWidget())
   Is Valid(exec=@n5.else, InputObject=@n4.EnhancedInputAction)
   > n8 = SetFeedbackFromInputAction!(
    execute
    self:Athena HUDFeedback Line Object Reference=Get ActionWidget()
    NewBoundAction:Input Action Object Reference=@n4.EnhancedInputAction
    FeedbackText:"Text (by ref)"=@n6.ActionLabel
    bShowEnabled:bool=true
    AxisScale:float=0
    bHasMultipleFeedbackLines:bool=false
   ) -> (then)
   n9 = Break Input Action Display(InputActionDisplay=Get CurrentActionDisplay()) -> (bShowSecondaryBoundKey=~)
   n10 = GetKeysBoundToAction!(
    self:Fort Global UIContext Object Reference=Get FortGlobalUIContext!(PlayerController:Player Controller Object Reference=GetOwningPlayer()) -> (ReturnValue:Fort Global UIContext Object Reference) {is=title K2Node_GetSubsystemFromPC}
    Action:Name=@n9
    bForceGamepadKey:bool=false
    PresetNameOverride:Name=None
   ) -> (OutPrimaryKey:Key Structure, OutSecondaryKey:Key Structure, ReturnValue:bool)
   n11 = Get CommonInputSubsystem!(PlayerController:Player Controller Object Reference=GetOwningPlayer()) -> (ReturnValue:Common Input Object Reference) {is=title K2Node_GetSubsystemFromPC}
   n12 = AND Boolean(
    A=@n9.bShowSecondaryBoundKey
    B="Not Equal (Enum)"!(A:ECommonInputType Enum=GetCurrentInputType(self=@n11), B:ECommonInputType Enum=Gamepad) -> (ReturnValue:bool)
   )
   Branch(
    execute=@n3.else
    Condition=OR Boolean(A=Is Valid Key!(Key:"Key Structure (by ref)"=@n10.OutSecondaryKey) -> (ReturnValue:bool), B=NOT Boolean(A=@n12))
   )
   > AddBoundAction!(
    execute
    self:Athena HUDFeedback Line Object Reference=Get ActionWidget()
    NewBoundLegacyAction:Name=@n9
    NewBoundEnhancedAction:Input Action Object Reference=@n9.EnhancedInputAction
    AxisScale:float=0
    bShowSecondaryBoundKey:bool=@n12
   ) -> (then)
  ]}
 AttemptRefreshFeedbackLines {nodes [
   AttemptRefreshFeedbackLines() {is=title K2Node_FunctionEntry}
   > n1 = Branch(Condition=Get bFeedbackRefreshQueued() -> (bFeedbackRefreshQueued:bool))
   Set Timer for Next Tick by Event!(execute=@n1.else, Delegate:Delegate=Create Event()) -> (then, ReturnValue:Timer Handle Structure) {is=title K2Node_CallFunction}
   > Set bFeedbackRefreshQueued(bFeedbackRefreshQueued:bool=true) -> (Output_Get:bool)
  ]}
 Bind To Vehicle Seat Delegate {nodes [
   Bind To Vehicle Seat Delegate() {is=title K2Node_FunctionEntry}
   n1 = GetVehicleInterface!() -> (bIsValid:bool, VehicleInterface:Fort Vehicle Interface Interface, OwningFortPlayerPawn:Fort Player Pawn Object Reference)
   > Branch(Condition=@n1)
   > n2 = GetVehicleSeatComponent!(execute, self:Fort Vehicle Interface Interface=@n1.VehicleInterface) -> (then, ReturnValue:Fort Vehicle Seat Component Object Reference)
   > Is Valid(InputObject=@n2)
   > n3 = Set CachedVehicleSeatComponent(CachedVehicleSeatComponent:Fort Vehicle Seat Component Object Reference=@n2) -> (Output_Get:Fort Vehicle Seat Component Object Reference)
   > Bind Event to On Vehicle Seat Player Slots Changed(self:Fort Vehicle Seat Component Object Reference=@n3, Delegate=Create Event()) {is=title K2Node_AddDelegate}
  ]}
 EventGraph {nodes [
   OnEnterVehiclePassenger_Bind() {is=title K2Node_CustomEvent}
   > AttemptRefreshFeedbackLines() {is=title K2Node_CallFunction}
   > ShowFeedback(self:WBP Athena HUDSituational Vehicle Feedback Object Reference)
   > Bind To Vehicle Seat Delegate() {is=title K2Node_CallFunction}
   OnVehicleSeatChanged_Bind!() -> (OutputDelegate, then, SeatIndex:int) {is=title K2Node_CustomEvent}
   > AttemptRefreshFeedbackLines()
   > ShowFeedback()
   n1 = OnPiPBlockPlayer!() -> (OutputDelegate, then, PIPIsBlocking:bool=~) {is=title K2Node_CustomEvent}
   > AttemptRefreshFeedbackLines()
   > n2 = Branch(Condition=@n1.PIPIsBlocking)
   > UpdateFeedbackStyles!(execute, self:WBP Athena HUDSituational Vehicle Feedback Object Reference, Style:Data Table Object Reference) -> (then)
   UpdateFeedbackStyles(execute=@n2.else, Style=Get DefaultTextStyleSet() -> (DefaultTextStyleSet:Data Table Object Reference))
   OnEnterVehicleDriver_Bind() {is=title K2Node_CustomEvent}
   > AttemptRefreshFeedbackLines()
   > ShowFeedback()
   > Bind To Vehicle Seat Delegate()
   OnExitVehicle_Bind() {is=title K2Node_CustomEvent}
   > HideFeedback(self:WBP Athena HUDSituational Vehicle Feedback Object Reference)
   n3 = Get CachedVehicleSeatComponent() -> (CachedVehicleSeatComponent:Fort Vehicle Seat Component Object Reference)
   > Is Valid(InputObject=@n3)
   > Unbind Event from On Vehicle Seat Player Slots Changed(self:Fort Vehicle Seat Component Object Reference=@n3, Delegate=Create Event()) {is=title K2Node_RemoveDelegate}
   > Set CachedVehicleSeatComponent()
   OnKeybindsChanged_Bind() {is=title K2Node_CustomEvent}
   > AttemptRefreshFeedbackLines()
   > ShowFeedback()
   Event ShowFeedback() {is=title K2Node_Event}
   n4 = GetPiPBlockManager!() -> (ReturnValue:Live Stream Pi PBlock Manager Object Reference)
   > Is Valid(InputObject=@n4)
   > Bind Event to PIPPlayer Blocking Event(self:Live Stream Pi PBlock Manager Object Reference=@n4, Delegate=@n1) {is=title K2Node_AddDelegate}
   > OnPiPBlockPlayer!(execute, self, PIPIsBlocking:bool=IsPlayerBlocked!(self:Live Stream Pi PBlock Manager Object Reference=@n4) -> (ReturnValue:bool)) -> (then) {is=title K2Node_CallFunction}
   > Parent: ShowFeedback() {is=title K2Node_CallParentFunction}
   Event HideFeedback() {is=title K2Node_Event}
   n5 = GetPiPBlockManager()
   > Is Valid(InputObject=@n5)
   > Unbind Event from PIPPlayer Blocking Event(self:Live Stream Pi PBlock Manager Object Reference=@n5, Delegate=Create Event()) {is=title K2Node_RemoveDelegate}
   > UpdateFeedbackStyles(Style=Get DefaultTextStyleSet())
   > Parent: HideFeedback() {is=title K2Node_CallParentFunction}
   Event Construct()
   > Parent: Construct()
   > Set bFeedbackRefreshQueued(bFeedbackRefreshQueued=false)
   > Bind Event to On Local Challenges Visibility Changed(
    self:Fort HUDContext Object Reference=Get FortHUDContext!() -> (ReturnValue:Fort HUDContext Object Reference) {is=title K2Node_GetBlueprintContextEx}
    Delegate=Create Event()
   ) {is=title K2Node_AddDelegate}
   > OnLocalChallengesVisibilityChanged!(execute, self, bIsVisible:bool=IsLocalChallengesVisible!(self:Fort HUDContext Object Reference=Get FortHUDContext()) -> (ReturnValue:bool)) -> (then)
   > AttemptRefreshFeedbackLines()
   n6 = Get FortGlobalUIContext!() -> (ReturnValue:Fort Global UIContext Object Reference) {is=title K2Node_GetBlueprintContextEx}
   > Bind Event to On Keybinds Changed(self:Fort Global UIContext Object Reference=@n6, Delegate=Create Event()) {is=title K2Node_AddDelegate}
   > Bind Event to On Enter Vehicle Driver(self:Fort Global UIContext Object Reference=@n6, Delegate=Create Event()) {is=title K2Node_AddDelegate}
   > Bind Event to On Enter Vehicle Passenger(self:Fort Global UIContext Object Reference=@n6, Delegate=Create Event()) {is=title K2Node_AddDelegate}
   > Bind Event to On Exit Vehicle(self:Fort Global UIContext Object Reference=@n6, Delegate=Create Event()) {is=title K2Node_AddDelegate}
   > Bind Event to On Vehicle Seat Changed(self:Fort Global UIContext Object Reference=@n6, Delegate=Create Event()) {is=title K2Node_AddDelegate}
   > n7 = Branch(Condition=Get bShouldClearFeedbackPanel())
   n8 = For Each Loop(Exec=@n7.else, Array:Array of WBP Athena HUDFeedback Line New CD Object References=Get FeedbackLines()) -> (Array Element:WBP Athena HUDFeedback Line New CD Object Reference)
   > SetVisibility(self=@n8, InVisibility=Collapsed)
   > AddChild(self=Get Feedback(), Content=@n8)
   SetVisibility(InVisibility=Select(
     Option 0=SelfHitTestInvisible
     Option 1=Collapsed
     Index=OnLocalChallengesVisibilityChanged!() -> (OutputDelegate, then, bIsVisible:bool=~) {is=title K2Node_CustomEvent}.bIsVisible
    ))
  ]}
 GetVehicleInterface {nodes [
   GetVehicleInterface() {is=title K2Node_FunctionEntry}
   > n1 = Sequence!(execute) -> (then_0, then_1, then_2:Exec)
   > n2 = Cast To FortPlayerPawn(Object=GetOwningPlayerPawn!(self:User Widget Object Reference) -> (ReturnValue:Pawn Object Reference))
   > n3 = Set Local_OwningFortPlayerPawn(Local_OwningFortPlayerPawn:Fort Player Pawn Object Reference=@n2) -> (Output_Get:Fort Player Pawn Object Reference)
   n4 = GetVehicleInterface!(self:Fort Player Pawn Object Reference=@n3) -> (ReturnValue:Fort Vehicle Interface Interface) {is=title K2Node_CallFunction}
   > Is Valid(InputObject="To Object (Interface)"!(Interface:"Interface Interface (by ref)"=@n4) -> (ReturnValue:Object Reference))
   > Set Local_VehicleInterface(Local_VehicleInterface:Fort Vehicle Interface Interface=@n4) -> (Output_Get:Fort Vehicle Interface Interface)
   > Set bLocal_IsValid(bLocal_IsValid:bool=true) -> (Output_Get:bool)
   n5 = Branch(execute=@n1.then_1, Condition=Get bLocal_IsValid() -> (bLocal_IsValid:bool))
   n6 = Cast To FortVehicleInterface!(execute=@n5.else, Object=Get Local_OwningFortPlayerPawn() -> (Local_OwningFortPlayerPawn:Fort Player Pawn Object Reference)) -> (then, CastFailed, AsFort Vehicle Interface:Fort Vehicle Interface Interface) {is=title K2Node_DynamicCast}
   > Set Local_VehicleInterface(Local_VehicleInterface=@n6)
   > Set bLocal_IsValid(bLocal_IsValid=true)
   Return Node!(
    execute=@n1.then_2
    bIsValid:bool=Get bLocal_IsValid()
    VehicleInterface:Fort Vehicle Interface Interface=Get Local_VehicleInterface() -> (Local_VehicleInterface:Fort Vehicle Interface Interface)
    OwningFortPlayerPawn:Fort Player Pawn Object Reference=Get Local_OwningFortPlayerPawn()
   )
  ]}
 RefreshFeedbackLines {nodes [
   RefreshFeedbackLines() {is=title K2Node_FunctionEntry}
   > n1 = For Each Loop(Array=Get FeedbackLines())
   > SetVisibility(self=@n1, InVisibility=Collapsed)
   Set bFeedbackRefreshQueued(execute=@n1.Completed, bFeedbackRefreshQueued=false)
   > Clear(TargetArray:Array of Names=Get SeatActionNames() -> (SeatActionNames:Array of Names))
   > Clear(TargetArray:Array of Texts=Get SeatActionLabels() -> (SeatActionLabels:Array of Texts))
   > n2 = Branch(Condition=Get bShouldClearFeedbackPanel())
   > n3 = ClearChildren(self=Get Feedback())
   n4 = GetVehicleInterface!() -> (bIsValid:bool, VehicleInterface:Fort Vehicle Interface Interface, OwningFortPlayerPawn:Fort Player Pawn Object Reference)
   Branch(execute=[@n3.then, @n2.else], Condition=@n4)
   > n5 = Set SeatActions(SeatActions:Array of Action Def To Display Structures=GetVehicleActionDisplayData!(self:Fort Vehicle Interface Interface=@n4.VehicleInterface, FortPlayerPawn:Fort Player Pawn Object Reference=@n4.OwningFortPlayerPawn) -> (SeatActions:Array of Action Def To Display Structures)) -> (Output_Get:Array of Action Def To Display Structures)
   > n6 = For Each Loop(Array:Array of Action Def To Display Structures=@n5) -> (Array Element:Action Def To Display Structure)
   > AddFeedbackLine!(execute, self, Action:Action Def To Display Structure=@n6, ActionIndex:int=@n6.Array Index) -> (then) {is=title K2Node_CallFunction}
  ]}
 Vehicle Player Slots Changed {nodes [
   Vehicle Player Slots Changed() {is=title K2Node_FunctionEntry}
   > AttemptRefreshFeedbackLines()
  ]}
]
variables [
 bFeedbackRefreshQueued {type=boolean}
 bShouldClearFeedbackPanel {type=boolean}
 CachedVehicleSeatComponent {type=FortVehicleSeatComponent, default=null}
 SeatActionLabels {type="array<string>", default []}
 SeatActionNames {type="array<string>", default []}
 SeatActions {type="array<ActionDefToDisplay>", default []}
]
widgets {
 info {
  inheritedWidgetCount=0
  parentClass=&/WBP_AthenaHUDSituationalVehicleFeedback.WBP_AthenaHUDSituationalVehicleFeedback_C
  rootWidgetClass=null
  widgetCount=0
 }
 widgets []
}
