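# Athena HUD widget for contextual controls, showing vehicle and key-bind feedback in a scaled, optionally size-overridden callout box.
# On construct it hides vehicle feedback and registers as player-specific widget host for two gameplay tags, unregistering on destruct.
# Hosted content and extension widgets fill the callout vertical box or initially collapsed vehicle-actions spacer overlay, with removal routed by exact tag match.
# Exp events fade the layout out on start and back in on finish.
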WidgetBlueprint /Game/Athena/HUD/AthenaHUDContextualControls
graphs [
 AddContentToHostWidget {nodes [
   AddContentToHostWidget!() -> (then, HostTag:Gameplay Tag Structure, ContentWidget:Widget Object Reference, Padding:Margin Structure, bFill:bool=~) {is=title K2Node_FunctionEntry}
   > n1 = Branch(Condition=MatchesTag!(
     TagOne:Gameplay Tag Structure=Get HostTag() -> (HostTag:Gameplay Tag Structure)
     TagTwo:Gameplay Tag Structure=Get ControllerBindingsCalloutTag() -> (ControllerBindingsCalloutTag:Gameplay Tag Structure)
     bExactMatch:bool=true
    ) -> (ReturnValue:bool))
   > n2 = AddChildToVerticalBox(
    self=Get ControllerCalloutBox() -> (ControllerCalloutBox:Vertical Box Object Reference)
    Content=Get ContentWidget() -> (ContentWidget:Widget Object Reference)
   )
   > SetHorizontalAlignment(self:Vertical Box Slot Object Reference=@n2, InHorizontalAlignment=HAlign_Fill)
   > SetVerticalAlignment(self:Vertical Box Slot Object Reference=@n2, InVerticalAlignment=VAlign_Fill)
   Branch(
    execute=@n1.else
    Condition=MatchesTag(TagOne=Get HostTag(), TagTwo=Get VehicleActionsSpacerTag() -> (VehicleActionsSpacerTag:Gameplay Tag Structure), bExactMatch=true)
   )
   > n3 = AddChildToOverlay!(
    execute
    self:Overlay Object Reference=Get VehicleActionSpacerOverlay() -> (VehicleActionSpacerOverlay:Overlay Object Reference)
    Content:Widget Object Reference=Get ContentWidget()
   ) -> (then, ReturnValue:Overlay Slot Object Reference)
   > SetHorizontalAlignment(self:Overlay Slot Object Reference=@n3, InHorizontalAlignment=HAlign_Fill)
   > SetVerticalAlignment(self:Overlay Slot Object Reference=@n3, InVerticalAlignment=VAlign_Fill)
  ]}
 EventGraph {nodes [
   Event OnExpEventFinished() {is=title K2Node_Event}
   > PlayAnimationByName!(
    execute
    self:Fort Simple Widget Animations Panel Object Reference=Get FortSimpleWidgetAnimationsPanel() -> (FortSimpleWidgetAnimationsPanel:Fort Simple Widget Animations Panel Object Reference)
    Name:"Name (by ref)"=Get AnimIntroName() -> (AnimIntroName:Name=None)
    Speed:float=1
    StartAtTime:float=0
   ) -> (then, ReturnValue:bool)
   Event OnExpEventStarted() {is=title K2Node_Event}
   > PlayAnimationByName(self=Get FortSimpleWidgetAnimationsPanel(), Name=Get AnimOutroName() -> (AnimOutroName:Name=None), Speed=1, StartAtTime=0)
   Event PreConstruct()
   > SetSizeBoxOverride()
   n1 = Event OnControllerBindingCalloutExtensionAdded!() -> (OutputDelegate, then, ExtensionWidget:User Widget Object Reference) {is=title K2Node_Event}
   > n2 = AddChildToVerticalBox(self=Get ControllerCalloutBox(), Content=@n1.ExtensionWidget)
   > SetHorizontalAlignment(self:Vertical Box Slot Object Reference=@n2, InHorizontalAlignment=HAlign_Fill)
   > SetVerticalAlignment(self:Vertical Box Slot Object Reference=@n2, InVerticalAlignment=VAlign_Fill)
   n3 = Event OnVehicleActionsSpacerExtensionAdded!() -> (OutputDelegate, then, ExtensionWidget:User Widget Object Reference) {is=title K2Node_Event}
   > n4 = AddChildToOverlay(self=Get VehicleActionSpacerOverlay(), Content=@n3.ExtensionWidget)
   > SetHorizontalAlignment(self:Overlay Slot Object Reference=@n4, InHorizontalAlignment=HAlign_Fill)
   > SetVerticalAlignment(self:Overlay Slot Object Reference=@n4, InVerticalAlignment=VAlign_Fill)
   Event Construct()
   > HideFeedback(self:WBP Athena HUDSituational Vehicle Feedback Object Reference=Get AthenaVehicleFeedback() -> (AthenaVehicleFeedback:WBP Athena Vehicle Feedback Object Reference))
   > RegisterWidgetHostTags()
   Event Destruct()
   > UnregisterWidgetHostTags()
   n5 = Event Add Content To Host Widget!() -> (
    OutputDelegate
    then
    WidgetHostTag:"Gameplay Tag Structure (by ref)"
    ContentWidget:Widget Object Reference
    Padding:"Margin Structure (by ref)"
    bFill:bool
   ) {is=title K2Node_Event}
   > AddContentToHostWidget!(
    execute
    self
    HostTag:Gameplay Tag Structure=@n5.WidgetHostTag
    ContentWidget:Widget Object Reference=@n5.ContentWidget
    Padding:Margin Structure=@n5.Padding
    bFill:bool=@n5.bFill
   ) -> (then) {is=title K2Node_CallFunction}
  ]}
 RegisterWidgetHostTags {nodes [
   RegisterWidgetHostTags() {is=title K2Node_FunctionEntry}
   > RegisterWidgetHost!(
    execute
    WidgetHost:Widget Object Reference
    WidgetHostTag:"Gameplay Tag Structure (by ref)"=Get ControllerBindingsCalloutTag()
    IsPlayerSpecific:bool=true
   ) -> (then)
   > RegisterWidgetHost(WidgetHostTag=Get VehicleActionsSpacerTag(), IsPlayerSpecific=true)
  ]}
 RemoveContentFromWidgetHost {nodes [
   Remove Content From Host Widget!() -> (
    then
    WidgetHostTag:"Gameplay Tag Structure (by ref)"
    ContentWidget:Widget Object Reference
    RemovalType:EWidgetSlottingRemovalType Enum=Unparent
   ) {is=title K2Node_FunctionEntry}
   > n1 = Branch(Condition=MatchesTag(TagOne=Get WidgetHostTag() -> (WidgetHostTag:"Gameplay Tag Structure (by ref)"), TagTwo=Get ControllerBindingsCalloutTag(), bExactMatch=true))
   > n2 = RemoveChild(self=Get ControllerCalloutBox(), Content=Get ContentWidget())
   > Return Node!(execute, ReturnValue:bool=@n2)
   n3 = Branch(execute=@n1.else, Condition=MatchesTag(TagOne=Get WidgetHostTag(), TagTwo=Get VehicleActionsSpacerTag(), bExactMatch=true))
   > n4 = RemoveChild(self=Get VehicleActionSpacerOverlay(), Content=Get ContentWidget())
   > Return Node(ReturnValue=@n4)
   Return Node(execute=@n3.else, ReturnValue=false)
  ]}
 SetSizeBoxOverride {nodes [
   SetSizeBoxOverride() {is=title K2Node_FunctionEntry}
   > n1 = Branch(Condition=Get IsSizeBoxEnabled() -> (IsSizeBoxEnabled:bool))
   n2 = Get SizeBox() -> (SizeBox:Size Box Object Reference)
   > SetWidthOverride(self=@n2, InWidthOverride=Get SizeBoxWidthOverride() -> (SizeBoxWidthOverride:float))
   > SetHeightOverride(self=@n2, InHeightOverride=Get SizeBoxHeightOverride() -> (SizeBoxHeightOverride:float))
   ClearWidthOverride(execute=@n1.else, self:Size Box Object Reference=@n2)
   > ClearHeightOverride(self:Size Box Object Reference=@n2)
  ]}
 UnregisterWidgetHostTags {nodes [
   UnregisterWidgetHostTags() {is=title K2Node_FunctionEntry}
   > UnregisterWidgetHost!(execute, WidgetHost:Widget Object Reference, WidgetHostTag:"Gameplay Tag Structure (by ref)"=Get ControllerBindingsCalloutTag()) -> (then)
   > UnregisterWidgetHost(WidgetHostTag=Get VehicleActionsSpacerTag())
  ]}
]
variables [
 AnimIntroName {type=string, default=Intro}
 AnimOutroName {type=string, default=Outro}
 ControllerBindingsCalloutTag {type=GameplayTag, default {tagName=DynamicUI.HUD.Fort.Athena.ContextualControls.ControllerBindingsCallout}}
 IsSizeBoxEnabled {type=boolean, default=true}
 SizeBoxHeightOverride {type=number, default=500}
 SizeBoxWidthOverride {type=number, default=500}
 VehicleActionsSpacerTag {type=GameplayTag, default {tagName=DynamicUI.HUD.Fort.Athena.ContextualControls.VehicleActionsSpacer}}
]
widgets {
 info {
  inheritedWidgetCount=0
  parentClass=/Script/FortniteUI.AthenaHUDContextualControls
  rootWidgetClass=/Script/FortniteUI.FortSimpleWidgetAnimationsPanel
  widgetCount=8
 }
 widgets [
  FortSimpleWidgetAnimationsPanel {
   widgetClassPath=&FortSimpleWidgetAnimationsPanel
   bIsVariable=true
   widget=FortSimpleWidgetAnimationsPanel
   properties {
    animations [
     Intro {
      bIsEnabled=true
      onAnimationsFinished=()
      targets [{
        alpha=&/HUD_ContextualControls_IntroAlpha
        bResetOnFinish=false
        bShouldAnimateAlpha=true
        bShouldAnimateScale=false
        bShouldAnimateTranslation=false
        scale=null
        targetName=HorizontalBox_2
        translation=null
        materialScalarAnimations []
       }]
     }
     Outro {
      bIsEnabled=true
      onAnimationsFinished=()
      targets [{
        alpha=&/HUD_ContextualControls_OutroAlpha
        bResetOnFinish=false
        bShouldAnimateAlpha=true
        bShouldAnimateScale=false
        bShouldAnimateTranslation=false
        scale=null
        targetName=HorizontalBox_2
        translation=null
        materialScalarAnimations []
       }]
     }
    ]
    renderTransformPivot=<0 0.5>
   }
  }
  HorizontalBox_2 {
   widgetClassPath=&HorizontalBox
   parent=FortSimpleWidgetAnimationsPanel
   widget=HorizontalBox_2
   slot {class=FortSimpleWidgetAnimationsSlot}
  }
  VehicleActionSpacerOverlay {
   widgetClassPath=&Overlay
   bIsVariable=true
   parent=HorizontalBox_2
   widget=VehicleActionSpacerOverlay
   properties {visibility=Collapsed}
   slot {class=HorizontalBoxSlot, properties {horizontalAlignment=HAlign_Left}}
  }
  SizeBox {
   widgetClassPath=&SizeBox
   bIsVariable=true
   parent=HorizontalBox_2
   widget=SizeBox
   properties {bOverride_HeightOverride=true, bOverride_WidthOverride=true, heightOverride=500, widthOverride=500}
   slot {class=HorizontalBoxSlot, properties {horizontalAlignment=HAlign_Left}}
  }
  ScaleBox_1 {
   widgetClassPath=&ScaleBox
   parent=SizeBox
   widget=ScaleBox_1
   properties {stretch=UserSpecified, userSpecifiedScale=0.8}
   slot {class=SizeBoxSlot, properties {horizontalAlignment=HAlign_Left, verticalAlignment=VAlign_Bottom}}
  }
  ControllerCalloutBox {
   widgetClassPath=&VerticalBox
   bIsVariable=true
   parent=ScaleBox_1
   widget=ControllerCalloutBox
   properties {renderTransformPivot=<0 0.5>}
   slot {class=ScaleBoxSlot, properties {horizontalAlignment=HAlign_Left}}
  }
  AthenaVehicleFeedback {
   widgetClassPath=&/WBP_AthenaVehicleFeedback
   bIsVariable=true
   parent=ControllerCalloutBox
   widget=AthenaVehicleFeedback
   properties {maxLinesNeeded=11, renderTransform {angle=0, scale=<1 1>, shear=<0 0>, translation=<0 0>}}
   slot {class=VerticalBoxSlot}
  }
  AthenaGenericKeyBindsFeedback {
   widgetClassPath=&/AthenaGenericKeyBindsFeedback
   parent=ControllerCalloutBox
   widget=AthenaGenericKeyBindsFeedback
   properties {elementTag {tagName=HUD.Vehicle.Controls}, renderTransform {angle=0, scale=<1 1>, shear=<0 0>, translation=<0 0>}}
   slot {class=VerticalBoxSlot, properties {horizontalAlignment=HAlign_Left}}
  }
 ]
}
