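# HUD feedback line widget showing a keybind beside rich contextual action text, based on AthenaHUDFeedbackLineBase.
# OnShowEnabledUpdated switches the content border between enabled and disabled colors, dimming the line by alpha when disabled.
# Event_DangerPulse shows or hides a red-yellow glow behind the line, playing the DangerPulse animation at the supplied speed or stopping it.
# DangerPulse animates opacity and scale on the glow layers.
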WidgetBlueprint /Game/Athena/HUD/AthenaHUDFeedbackLine
animations [DangerPulse {
  binding_count=3
  length_seconds=0.8521
  object=&self:DangerPulse
  sequence {bindings [
    Image_YellowBG {
     id=#1
     possessed_class=Image
     tracks [
      {
       type=MovieSceneFloatTrack
       display_name=Render Opacity
       sections [{type=MovieSceneFloatSection, channels [None {type=MovieSceneScriptingFloatChannel, keys [[2.9167,1],[19.9583,0.2]]}]}]
      }
      {
       type=MovieScene2DTransformTrack
       sections [{
         type=MovieScene2DTransformSection
         channels [
          Scale.X {type=MovieSceneScriptingFloatChannel, keys [[2.9167,1],[19.9583,0.8]]}
          Scale.Y {type=MovieSceneScriptingFloatChannel, keys [[2.9167,1],[19.9583,0.8]]}
         ]
        }]
      }
     ]
    }
    Image_RedBG {
     id=#2
     possessed_class=Image
     tracks [
      {
       type=MovieSceneFloatTrack
       display_name=Render Opacity
       sections [{type=MovieSceneFloatSection, channels [None {type=MovieSceneScriptingFloatChannel}]}]
      }
      {
       type=MovieScene2DTransformTrack
       sections [{
         type=MovieScene2DTransformSection
         channels [
          Scale.X {type=MovieSceneScriptingFloatChannel, keys [[2.9167,1],[19.9583,0.8]]}
          Scale.Y {type=MovieSceneScriptingFloatChannel, keys [[2.9167,1],[19.9583,0.8]]}
         ]
        }]
      }
     ]
    }
    Overlay_Glow {
     id=#3
     possessed_class=Overlay
     tracks [{
       type=MovieSceneFloatTrack
       display_name=Render Opacity
       sections [{type=MovieSceneFloatSection, channels [None {type=MovieSceneScriptingFloatChannel, keys [[2.9167,1],[19.9583,0.4]]}]}]
      }]
    }
   ]}
 }]
graphs [EventGraph {nodes [
   n1 = Event_DangerPulse!() -> (OutputDelegate, then, Enable:bool=true, AnimSpeed:double=~) {is=title K2Node_CustomEvent}
   > SetVisibility(
    self=Get Overlay_Glow() -> (Overlay_Glow:Overlay Object Reference)
    InVisibility=Select(Option 0=Hidden, Option 1=HitTestInvisible, Index=@n1.Enable)
   )
   > n2 = Branch(Condition=@n1.Enable)
   n3 = Get DangerPulse() -> (DangerPulse:Widget Animation Object Reference)
   > PlayAnimation(InAnimation=@n3, PlaybackSpeed=@n1.AnimSpeed)
   StopAnimation(execute=@n2.else, InAnimation=@n3)
   n4 = Event OnShowEnabledUpdated!() -> (OutputDelegate, then, bShowEnabled:bool) {is=title K2Node_Event}
   > SetContentColorAndOpacity!(
    execute
    self:Border Object Reference=Get Border_EnabledState() -> (Border_EnabledState:Border Object Reference)
    InContentColorAndOpacity:Linear Color Structure=SelectColor(
     A=Get EnabledBorderColor() -> (EnabledBorderColor:Linear Color Structure)
     B=Get DisabledBorderColor() -> (DisabledBorderColor:Linear Color Structure)
     bPickA=@n4.bShowEnabled
    )
   ) -> (then)
  ]}]
variables [
 DisabledBorderColor {type=LinearColor, default=<rgba 1 1 1 0.4>}
 EnabledBorderColor {type=LinearColor, default=<rgba 1 1 1 1>}
]
widgets {
 info {inheritedWidgetCount=1, parentClass=/Script/FortniteUI.AthenaHUDFeedbackLineBase, rootWidgetClass=&Overlay, widgetCount=10}
 widgets [
  Overlay_0 {widgetClassPath=&Overlay, widget=Overlay_0}
  Overlay_Glow {
   widgetClassPath=&Overlay
   bIsVariable=true
   parent=Overlay_0
   widget=Overlay_Glow
   properties {visibility=Collapsed}
   slot {class=OverlaySlot, properties {horizontalAlignment=HAlign_Center, verticalAlignment=VAlign_Center, padding=<ltrb -200 -200 -200 -200>}}
  }
  Image_Shade {
   widgetClassPath=&Image
   parent=Overlay_Glow
   widget=Image_Shade
   properties {
    brush {
     bIsDynamicallyLoaded=false
     drawAs=Image
     imageType=NoImage
     mirroring=NoMirror
     resourceName=None
     resourceObject=/Game/Athena/HUD/Materials/M_CountdownBG_Alpha_Inst
     tiling=NoTile
     imageSize=<482.0438 58.3627>
     margin=<ltrb 0 0 0 0>
     outlineSettings {
      bUseBrushTransparency=false
      roundingType=HalfHeightRadius
      width=0
      color {colorUseRule=UseColor_Specified, specifiedColor=<rgba 0 0 0 0>}
      cornerRadii=<0 0 0 0>
     }
     tintColor {colorUseRule=UseColor_Specified, specifiedColor=<rgba 0.0156 0.0063 0.0063 1>}
     uVRegion {bIsValid=false, max=<0 0>, min=<0 0>}
    }
    colorAndOpacity=<rgba 1 1 1 0.2286>
   }
   slot {class=OverlaySlot, properties {horizontalAlignment=HAlign_Fill, verticalAlignment=VAlign_Center}}
  }
  Image_RedBG {
   widgetClassPath=&Image
   parent=Overlay_Glow
   widget=Image_RedBG
   properties {
    brush {
     bIsDynamicallyLoaded=false
     drawAs=Image
     imageType=NoImage
     mirroring=NoMirror
     resourceName=None
     resourceObject=/Game/Athena/HUD/Materials/M_CountdownBG_Base
     tiling=NoTile
     imageSize=<514.2557 100>
     margin=<ltrb 0 0 0 0>
     outlineSettings {
      bUseBrushTransparency=false
      roundingType=HalfHeightRadius
      width=0
      color {colorUseRule=UseColor_Specified, specifiedColor=<rgba 0 0 0 0>}
      cornerRadii=<0 0 0 0>
     }
     tintColor {colorUseRule=UseColor_Specified, specifiedColor=<rgba 1 1 1 1>}
     uVRegion {bIsValid=false, max=<0 0>, min=<0 0>}
    }
    colorAndOpacity=<rgba 1 0 0 1>
   }
   slot {class=OverlaySlot, properties {horizontalAlignment=HAlign_Fill, verticalAlignment=VAlign_Center}}
  }
  Image_YellowBG {
   widgetClassPath=&Image
   parent=Overlay_Glow
   widget=Image_YellowBG
   properties {
    brush {
     bIsDynamicallyLoaded=false
     drawAs=Image
     imageType=NoImage
     mirroring=NoMirror
     resourceName=None
     resourceObject=/Game/Athena/HUD/Materials/M_CountdownBG_Core
     tiling=NoTile
     imageSize=<514.2557 100>
     margin=<ltrb 0 0 0 0>
     outlineSettings {
      bUseBrushTransparency=false
      roundingType=HalfHeightRadius
      width=0
      color {colorUseRule=UseColor_Specified, specifiedColor=<rgba 0 0 0 0>}
      cornerRadii=<0 0 0 0>
     }
     tintColor {colorUseRule=UseColor_Specified, specifiedColor=<rgba 1 1 1 1>}
     uVRegion {bIsValid=false, max=<0 0>, min=<0 0>}
    }
    colorAndOpacity=<rgba 1 0.2789 0 1>
   }
   slot {class=OverlaySlot, properties {horizontalAlignment=HAlign_Fill}}
  }
  Border_EnabledState {
   widgetClassPath=&Border
   bIsVariable=true
   parent=Overlay_0
   widget=Border_EnabledState
   properties {
    background {
     bIsDynamicallyLoaded=false
     drawAs=NoDrawType
     imageType=NoImage
     mirroring=NoMirror
     resourceName=None
     resourceObject=null
     tiling=NoTile
     imageSize=<32 32>
     margin=<ltrb 0 0 0 0>
     outlineSettings {
      bUseBrushTransparency=false
      roundingType=HalfHeightRadius
      width=0
      color {colorUseRule=UseColor_Specified, specifiedColor=<rgba 0 0 0 0>}
      cornerRadii=<0 0 0 0>
     }
     tintColor {colorUseRule=UseColor_Specified, specifiedColor=<rgba 1 1 1 1>}
     uVRegion {bIsValid=false, max=<0 0>, min=<0 0>}
    }
    padding=<ltrb 0 0 0 0>
   }
   slot {class=OverlaySlot}
  }
  FeedbackLine {
   widgetClassPath=&HorizontalBox
   parent=Border_EnabledState
   widget=FeedbackLine
   slot {class=BorderSlot, properties {padding=<ltrb 0 0 0 0>}}
  }
  FeedbackKeyBind {
   widgetClassPath=/FortUILibrary/Input/KeybindWidget
   bIsVariable=true
   parent=FeedbackLine
   widget=FeedbackKeyBind
   properties {bShowUnboundStatus=true}
   slot {class=HorizontalBoxSlot, properties {verticalAlignment=VAlign_Center}}
  }
  FeedbackText {
   widgetClassPath=/Script/UMG.RichTextBlock
   bIsVariable=true
   parent=FeedbackLine
   widget=FeedbackText
   properties {
    text="{ContextualAction}"
    textStyleSet=/Game/Athena/HUD/TextStyles/PlayersLeftStyles
    defaultTextStyleOverride {
     overflowPolicy=Clip
     transformPolicy=None
     colorAndOpacity {colorUseRule=UseColor_Specified, specifiedColor=<rgba 1 1 1 1>}
     font {
      bForceMonospaced=false
      bMaterialIsStencil=false
      fontMaterial=null
      fontObject=&BurbankSmall
      letterSpacing=0
      monospacedWidth=1
      size=18
      skewAmount=0
      typefaceFontName=Bold
      outlineSettings {
       bApplyOutlineToDropShadows=true
       bMiteredCorners=false
       bSeparateFillAlpha=true
       outlineMaterial=&M-FontShadow
       outlineSize=3
       outlineColor=<rgba 0 0 0 0.5>
      }
     }
     highlightColor {colorUseRule=UseColor_Specified, specifiedColor=<rgba 0 0 0 1>}
     highlightShape {
      bIsDynamicallyLoaded=false
      drawAs=Image
      imageType=NoImage
      mirroring=NoMirror
      resourceName=None
      resourceObject=null
      tiling=NoTile
      imageSize=<32 32>
      margin=<ltrb 0 0 0 0>
      outlineSettings {
       bUseBrushTransparency=false
       roundingType=HalfHeightRadius
       width=0
       color {colorUseRule=UseColor_Specified, specifiedColor=<rgba 0 0 0 0>}
       cornerRadii=<0 0 0 0>
      }
      tintColor {colorUseRule=UseColor_Specified, specifiedColor=<rgba 1 1 1 1>}
      uVRegion {bIsValid=false, max=<0 0>, min=<0 0>}
     }
     selectedBackgroundColor {colorUseRule=UseColor_Foreground, specifiedColor=<rgba 1 0 1 0>}
     shadowColorAndOpacity=<rgba 0 0 0 1>
     shadowOffset=<0 0>
     strikeBrush {
      bIsDynamicallyLoaded=false
      drawAs=Image
      imageType=NoImage
      mirroring=NoMirror
      resourceName=None
      resourceObject=null
      tiling=NoTile
      imageSize=<32 32>
      margin=<ltrb 0 0 0 0>
      outlineSettings {
       bUseBrushTransparency=false
       roundingType=HalfHeightRadius
       width=0
       color {colorUseRule=UseColor_Specified, specifiedColor=<rgba 0 0 0 0>}
       cornerRadii=<0 0 0 0>
      }
      tintColor {colorUseRule=UseColor_Specified, specifiedColor=<rgba 1 1 1 1>}
      uVRegion {bIsValid=false, max=<0 0>, min=<0 0>}
     }
     underlineBrush {
      bIsDynamicallyLoaded=false
      drawAs=Image
      imageType=NoImage
      mirroring=NoMirror
      resourceName=None
      resourceObject=null
      tiling=NoTile
      imageSize=<32 32>
      margin=<ltrb 0 0 0 0>
      outlineSettings {
       bUseBrushTransparency=false
       roundingType=HalfHeightRadius
       width=0
       color {colorUseRule=UseColor_Specified, specifiedColor=<rgba 0 0 0 0>}
       cornerRadii=<0 0 0 0>
      }
      tintColor {colorUseRule=UseColor_Specified, specifiedColor=<rgba 1 1 1 1>}
      uVRegion {bIsValid=false, max=<0 0>, min=<0 0>}
     }
    }
   }
   slot {class=HorizontalBoxSlot, properties {verticalAlignment=VAlign_Center, padding=<ltrb 10 0 0 0>}}
  }
  TutorialHighlight {widgetClassPath=&TutorialHighlight}
 ]
}
