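# Compact HUD item list entry widget built on AthenaItemIcon, showing an item's icon with its stack count beneath, styled with resource icon materials.
# A VerticalBox stacks a fixed-size icon Image over a centered CommonTextBlock that collapses when empty and uses a font color material with a shadow outline.
# The other inherited slots (secondary count, ammo, rechargeable, secondary panel, tier) are left unplaced, and the event graph is empty, leaving behavior to the parent class.
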WidgetBlueprint /Game/Athena/HUD/AthenaHUDItemListEntry
graphs [EventGraph {links [], nodes []}]
variables []
widgets {
 info {inheritedWidgetCount=8, parentClass=/Script/FortniteUI.AthenaItemIcon, rootWidgetClass=&VerticalBox, widgetCount=10}
 widgets [
  VerticalBox_0 {widgetClassPath=&VerticalBox, widget=VerticalBox_0}
  SizeBox_1 {
   widgetClassPath=&SizeBox
   parent=VerticalBox_0
   widget=SizeBox_1
   properties {bOverride_HeightOverride=true, bOverride_WidthOverride=true, heightOverride=40, widthOverride=48}
   slot {class=VerticalBoxSlot, properties {padding=<ltrb 4 0 4 0>}}
  }
  Image_ItemIcon {
   widgetClassPath=&Image
   bInherited=true
   parent=SizeBox_1
   widget=Image_ItemIcon
   properties {
    visibility=SelfHitTestInvisible
    brush {
     bIsDynamicallyLoaded=false
     drawAs=Image
     imageType=NoImage
     mirroring=NoMirror
     resourceName=None
     resourceObject=/Game/Athena/HUD/Inventory/Art/M_UI_HUD_ResourceIcon
     tiling=NoTile
     imageSize=<64 64>
     margin=<ltrb 0 0 0 0>
     outlineSettings {
      bUseBrushTransparency=false
      roundingType=HalfHeightRadius
      width=0
      color {colorUseRule=UseColor_Specified, specifiedColor=<rgba 0 0 0 0>}
      cornerRadii=<0 0 0 0>
     }
     tintColor {colorUseRule=UseColor_Specified, specifiedColor=<rgba 1 1 1 1>}
     uVRegion {bIsValid=false, max=<0 0>, min=<0 0>}
    }
   }
   slot {class=SizeBoxSlot}
  }
  Text_StackCount {
   widgetClassPath=&CommonTextBlock
   bInherited=true
   parent=VerticalBox_0
   widget=Text_StackCount
   properties {
    bAutoCollapseWithEmptyText=true
    justification=Center
    mobileFontSizeMultiplier=1.125
    text=999
    font {
     bForceMonospaced=false
     bMaterialIsStencil=false
     fontMaterial=/Game/Athena/HUD/Inventory/Art/M_UI_Inventory_Resourece_FontColor
     fontObject=&BurbankSmall
     letterSpacing=0
     monospacedWidth=1
     size=16
     skewAmount=0
     typefaceFontName=Bold
     outlineSettings {
      bApplyOutlineToDropShadows=false
      bMiteredCorners=false
      bSeparateFillAlpha=true
      outlineMaterial=&M-FontShadow
      outlineSize=3
      outlineColor=<rgba 0 0 0 0.5>
     }
    }
    renderTransformPivot=<0.5 0.75>
    shadowOffset=<0 0>
   }
   slot {class=VerticalBoxSlot}
  }
  Text_SecondaryStackCount {widgetClassPath=&CommonTextBlock, bInherited=true, widget=null}
  Image_AmmoIcon {widgetClassPath=&Image, bInherited=true, widget=null}
  Image_SecondaryAmmoIcon {widgetClassPath=&Image, bInherited=true, widget=null}
  Image_RechargeableIcon {widgetClassPath=&Image, bInherited=true, widget=null}
  InfoPanel_SecondaryItem {widgetClassPath=&Overlay, bInherited=true, widget=null}
  ItemTierWidget {widgetClassPath=/Script/FortniteUI.AthenaItemTierWidget, bInherited=true, widget=null}
 ]
}
