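# Athena HUD widget that shows the number of players left in the match as an icon beside a count.
# A horizontal box holds the alive icon and a rich text block styled by the PlayersLeftStyles text style set.
# The native AthenaPlayersLeftBase parent supplies the text through the SetPlayersLeftText event, which the Blueprint writes straight into the text block.
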WidgetBlueprint /Game/Athena/HUD/AthenaPlayersLeftWidget
graphs [EventGraph {nodes [
   n1 = Event SetPlayersLeftText!() -> (OutputDelegate, then, PlayersText:"Text (by ref)") {is=title K2Node_Event}
   > SetText(self=Get PlayersLeftText() -> (PlayersLeftText:Rich Text Block Object Reference), InText=@n1.PlayersText)
  ]}]
variables []
widgets {
 info {inheritedWidgetCount=1, parentClass=/Script/FortniteUI.AthenaPlayersLeftBase, rootWidgetClass=&HorizontalBox, widgetCount=4}
 widgets [
  HorizontalBox_0 {widgetClassPath=&HorizontalBox, widget=HorizontalBox_0}
  PlayersImage {
   widgetClassPath=&Image
   bIsVariable=true
   parent=HorizontalBox_0
   widget=PlayersImage
   properties {brush {
     bIsDynamicallyLoaded=false
     drawAs=Image
     imageType=NoImage
     mirroring=NoMirror
     resourceName=None
     resourceObject=/Game/Athena/HUD/Art/AliveIco
     tiling=NoTile
     imageSize=<36 36>
     margin=<ltrb 0 0 0 0>
     outlineSettings {
      bUseBrushTransparency=false
      roundingType=HalfHeightRadius
      width=0
      color {colorUseRule=UseColor_Specified, specifiedColor=<rgba 0 0 0 0>}
      cornerRadii=<0 0 0 0>
     }
     tintColor {colorUseRule=UseColor_Specified, specifiedColor=<rgba 1 1 1 1>}
     uVRegion {bIsValid=false, max=<0 0>, min=<0 0>}
    }}
   slot {class=HorizontalBoxSlot, properties {verticalAlignment=VAlign_Center}}
  }
  PlayersLeftText {
   widgetClassPath=/Script/UMG.RichTextBlock
   bIsVariable=true
   parent=HorizontalBox_0
   widget=PlayersLeftText
   properties {text=100, textStyleSet=/Game/Athena/HUD/TextStyles/PlayersLeftStyles}
   slot {class=HorizontalBoxSlot, properties {verticalAlignment=VAlign_Center, padding=<ltrb 1 0 0 0>}}
  }
  TutorialHighlight {widgetClassPath=&TutorialHighlight}
 ]
}
