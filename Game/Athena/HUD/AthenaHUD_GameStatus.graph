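# Athena HUD game status strip below the minimap, showing game phase, players left, kills and initially collapsed team kills in a scaled horizontal box.
# It registers as a player-specific widget host on construct, unregisters on destruct, and adds content matching its tag, ignoring padding and fill.
# Extension widgets join the same box with padding and default hide tags.
# Removal only supports unparenting; collapse requests return false.
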WidgetBlueprint /Game/Athena/HUD/AthenaHUD_GameStatus
graphs [
 AddContentToHostWidget {nodes [
   n1 = AddContentToHostWidget!() -> (then, Widget Host Tag:Gameplay Tag Structure, Content Widget:Widget Object Reference, Padding:Margin Structure, Fill:bool=~) {is=title K2Node_FunctionEntry}
   > Branch(Condition=MatchesTag!(
     TagOne:Gameplay Tag Structure=Get ContentHostTag() -> (ContentHostTag:Gameplay Tag Structure)
     TagTwo:Gameplay Tag Structure=@n1
     bExactMatch:bool=true
    ) -> (ReturnValue:bool))
   > AddChildToHorizontalBox!(
    execute
    self:Horizontal Box Object Reference=Get GameInfoBox() -> (GameInfoBox:Horizontal Box Object Reference)
    Content:Widget Object Reference=@n1.Content Widget
   ) -> (then, ReturnValue:Horizontal Box Slot Object Reference)
  ]}
 EventGraph {nodes [
   Event Destruct()
   > UnregisterHostSlots()
   Event Construct()
   > RegisterHostSlots()
   n1 = Event Add Content To Host Widget!() -> (
    OutputDelegate
    then
    WidgetHostTag:"Gameplay Tag Structure (by ref)"
    ContentWidget:Widget Object Reference
    Padding:"Margin Structure (by ref)"
    bFill:bool
   ) {is=title K2Node_Event}
   > AddContentToHostWidget!(
    execute
    self
    Widget Host Tag:Gameplay Tag Structure=@n1.WidgetHostTag
    Content Widget:Widget Object Reference=@n1.ContentWidget
    Padding:Margin Structure=@n1.Padding
    Fill:bool=@n1.bFill
   ) -> (then) {is=title K2Node_CallFunction}
   n2 = Event OnExtensionWidgetCreated!() -> (OutputDelegate, then, InUserWidget:User Widget Object Reference, InExtensionSlot:EUIExtensionSlot Enum=Invalid) {is=title K2Node_Event}
   > n3 = AddChildToHorizontalBox(self=Get GameInfoBox(), Content=@n2.InUserWidget)
   > SetHorizontalAlignment(self:Horizontal Box Slot Object Reference=@n3, InHorizontalAlignment=HAlign_Fill)
   > SetVerticalAlignment(self:Horizontal Box Slot Object Reference=@n3, InVerticalAlignment=VAlign_Fill)
   > SetPadding!(
    execute
    self:Horizontal Box Slot Object Reference=@n3
    InPadding_Left:float=5
    InPadding_Top:float=0
    InPadding_Right:float=5
    InPadding_Bottom:float=0
   ) -> (then)
   > n4 = Cast To FortHUDElementWidget!(execute, Object=@n2.InUserWidget) -> (then, CastFailed, AsFort HUDElement Widget:Fort HUDElement Widget Object Reference) {is=title K2Node_DynamicCast}
   > AppendHUDElementTags!(
    execute
    self:Fort HUDElement Widget Object Reference=@n4
    AdditionalTags:"Gameplay Tag Container Structure (by ref)"=Get GameInfoBoxDefaultHUDTags() -> (GameInfoBoxDefaultHUDTags:Gameplay Tag Container Structure)
   ) -> (then)
  ]}
 RegisterHostSlots {nodes [
   RegisterHostSlots() {is=title K2Node_FunctionEntry}
   > RegisterWidgetHost!(
    execute
    WidgetHost:Widget Object Reference
    WidgetHostTag:"Gameplay Tag Structure (by ref)"=Get ContentHostTag()
    IsPlayerSpecific:bool=true
   ) -> (then)
  ]}
 RemoveContentFromWidgetHost {nodes [
   n1 = Remove Content From Host Widget!() -> (
    then
    WidgetHostTag:"Gameplay Tag Structure (by ref)"
    ContentWidget:Widget Object Reference
    RemovalType:EWidgetSlottingRemovalType Enum=Unparent
   ) {is=title K2Node_FunctionEntry}
   > n2 = Branch(Condition=MatchesTag(TagOne=Get ContentHostTag(), TagTwo=@n1, bExactMatch=true))
   > n3 = Switch on EWidgetSlottingRemovalType!(execute, Selection:EWidgetSlottingRemovalType Enum=@n1.RemovalType) -> (Unparent:Exec, Collapse:Exec) {is=title K2Node_SwitchEnum}
   > n4 = RemoveChild(self=Get GameInfoBox(), Content=@n1.ContentWidget)
   > Return Node!(execute, ReturnValue:bool=@n4)
   Return Node(execute=[@n3.Collapse, @n2.else], ReturnValue=false)
  ]}
 UnregisterHostSlots {nodes [
   UnregisterHostSlots() {is=title K2Node_FunctionEntry}
   > UnregisterWidgetHost!(execute, WidgetHost:Widget Object Reference, WidgetHostTag:"Gameplay Tag Structure (by ref)"=Get ContentHostTag()) -> (then)
  ]}
]
variables [
 ContentHostTag {type=GameplayTag, default {tagName=DynamicUI.HUD.Fort.Athena.GameInfoContent}}
 GameInfoBoxDefaultHUDTags {
  type=GameplayTagContainer
  default {gameplayTags [
    {tagName=Athena.HideForCapture}
    {tagName=HUD.Athena.Hide.WithInventory}
    {tagName=HUD.Athena.Hide.WithMap}
    {tagName=HUD.LTM.HideForTryHard}
    {tagName=HUD.Event.HideForSpecialEvent}
    {tagName=Athena.Creative.HideForLoading}
    {tagName=HUD.Hide.WhileLoading}
   ]}
 }
]
widgets {
 info {
  inheritedWidgetCount=0
  namedSlotCount=1
  parentClass=/Script/FortniteUI.FortExtensionWidgetContainer
  rootWidgetClass=&/FortHUDElementWrapper.FortHUDElementWrapper_C
  widgetCount=8
 }
 widgets [
  FortHUDElementWrapper {
   widgetClassPath=&/FortHUDElementWrapper
   bIsVariable=true
   widget=FortHUDElementWrapper
   properties {elementTag {tagName=HUD.BelowMinimap.GameStatus}}
  }
  SizeBox_2 {
   widgetClassPath=&SizeBox
   bIsVariable=true
   namedSlotHost=FortHUDElementWrapper
   widget=SizeBox_2
   properties {bOverride_MinDesiredWidth=true, minDesiredWidth=300, widthOverride=300}
  }
  ScaleBox_2 {
   widgetClassPath=&ScaleBox
   bIsVariable=true
   parent=SizeBox_2
   widget=ScaleBox_2
   properties {stretch=UserSpecified, userSpecifiedScale=0.8}
   slot {class=SizeBoxSlot}
  }
  GameInfoBox {widgetClassPath=&HorizontalBox, bIsVariable=true, parent=ScaleBox_2, widget=GameInfoBox, slot {class=ScaleBoxSlot}}
  AthenaGamePhaseWidget {
   widgetClassPath=/Game/Athena/HUD/Phase/AthenaGamePhaseWidget
   parent=GameInfoBox
   widget=AthenaGamePhaseWidget
   properties {
    elementTag {tagName=HUD.Athena.GamePhase.MinimapInfo}
    hUDElementTag {gameplayTags [
      {tagName=Athena.HideForCapture}
      {tagName=HUD.Athena.Hide.WithInventory}
      {tagName=HUD.Athena.Hide.WithMap}
      {tagName=HUD.LTM.HideForTryHard}
      {tagName=HUD.Event.HideForSpecialEvent}
      {tagName=Athena.Creative.HideForLoading}
      {tagName=HUD.Hide.WhileLoading}
      {tagName=HUD.Athena.HideForPostGame}
      {tagName=HUD.LTM.HideForCrucible}
      {tagName=Athena.LTM.Mash.Title}
      {tagName=HUD.Athena.HideforRadial}
      {tagName=Athena.Creative.HUDController.HideStormTimer}
      {tagName=HUD.Hide.CreativeJoiningLobby}
     ]}
   }
   slot {class=HorizontalBoxSlot, properties {horizontalAlignment=HAlign_Center, padding=<ltrb 5 0 5 0>}}
  }
  AthenaPlayersLeftWidget {
   widgetClassPath=&/AthenaPlayersLeftWidget
   parent=GameInfoBox
   widget=AthenaPlayersLeftWidget
   properties {hUDElementTag {gameplayTags [
      {tagName=Athena.HideForCapture}
      {tagName=HUD.Athena.Hide.WithInventory}
      {tagName=HUD.Athena.Hide.WithMap}
      {tagName=HUD.LTM.HideForTryHard}
      {tagName=HUD.Event.HideForSpecialEvent}
      {tagName=Athena.Creative.HideForLoading}
      {tagName=HUD.Hide.WhileLoading}
      {tagName=HUD.Athena.HideForPostGame}
      {tagName=HUD.LTM.HideForCrucible}
      {tagName=Athena.LTM.Mash.Title}
      {tagName=HUD.Athena.HideforRadial}
      {tagName=HUD.Hide.CreativeJoiningLobby}
      {tagName=Athena.Creative.HUDController.HidePlayerCount}
     ]}}
   slot {class=HorizontalBoxSlot, properties {horizontalAlignment=HAlign_Center, padding=<ltrb 5 0 5 0>}}
  }
  AthenaKillsWidget {
   widgetClassPath=&/AthenaKillsWidget
   parent=GameInfoBox
   widget=AthenaKillsWidget
   properties {hUDElementTag {gameplayTags [
      {tagName=Athena.HideForCapture}
      {tagName=HUD.Athena.Hide.WithInventory}
      {tagName=HUD.Athena.Hide.WithMap}
      {tagName=HUD.LTM.HideForTryHard}
      {tagName=HUD.Event.HideForSpecialEvent}
      {tagName=Athena.Creative.HideForLoading}
      {tagName=HUD.Hide.WhileLoading}
      {tagName=HUD.Athena.HideForPostGame}
      {tagName=HUD.LTM.HideForCrucible}
      {tagName=Athena.LTM.Mash.Title}
      {tagName=HUD.Athena.HideforRadial}
      {tagName=HUD.Hide.CreativeJoiningLobby}
      {tagName=HUD.Athena.HideOnSpectate}
      {tagName=Athena.Creative.HUDController.HideEliminationCounter}
     ]}}
   slot {class=HorizontalBoxSlot, properties {horizontalAlignment=HAlign_Center, padding=<ltrb 5 0 5 0>}}
  }
  AthenaTeamKillsWidget {
   widgetClassPath=&/AthenaTeamKillsWidget
   bIsVariable=true
   parent=GameInfoBox
   widget=AthenaTeamKillsWidget
   properties {
    visibility=Collapsed
    hUDElementTag {gameplayTags [
      {tagName=Athena.HideForCapture}
      {tagName=HUD.Athena.Hide.WithInventory}
      {tagName=HUD.Athena.Hide.WithMap}
      {tagName=HUD.LTM.HideForTryHard}
      {tagName=HUD.Event.HideForSpecialEvent}
      {tagName=Athena.Creative.HideForLoading}
      {tagName=HUD.Hide.WhileLoading}
      {tagName=HUD.Athena.HideForPostGame}
      {tagName=HUD.LTM.HideForCrucible}
      {tagName=Athena.LTM.Mash.Title}
      {tagName=HUD.Athena.HideforRadial}
      {tagName=HUD.Hide.CreativeJoiningLobby}
      {tagName=Athena.Creative.HUDController.HideTeamInfo}
     ]}
   }
   slot {class=HorizontalBoxSlot, properties {horizontalAlignment=HAlign_Center, padding=<ltrb 5 0 5 0>}}
  }
 ]
}
