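# Athena HUD widget that shows the local player's name beside a crown icon marking party leadership.
# On construct it binds to the party context's local player state change event and applies the current info, then unbinds on destruct.
# State change updates are applied only when the player controller is not a FortCaptureSpectatorController.
# UpdateLocalPlayerInfo reveals the widget, sets the name, and shows the crown only for the leader of a party larger than one.
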WidgetBlueprint /Game/Athena/HUD/AthenaPlayerInfo
graphs [
 EventGraph {nodes [
   n1 = HandlePlayerStateChanged!() -> (OutputDelegate, then, PlayerInfo:"Fort Team Member Info Structure (by ref)") {is=title K2Node_CustomEvent}
   > n2 = Cast To FortCaptureSpectatorController!(execute, Object=GetPlayerController!(PlayerIndex:int=0) -> (ReturnValue:Player Controller Object Reference)) -> (then, CastFailed, AsFort Capture Spectator Controller:Fort Capture Spectator Controller Object Reference) {is=title K2Node_DynamicCast}
   UpdateLocalPlayerInfo!(execute=@n2.CastFailed, self, Player Info:"Fort Team Member Info Structure (by ref)"=@n1.PlayerInfo) -> (then)
   Event Construct()
   n3 = Get FortPartyContext!() -> (ReturnValue:Fort Party Context Object Reference) {is=title K2Node_GetBlueprintContextEx}
   > Bind Event to On Local Player State Changed(self:Fort Party Context Object Reference=@n3, Delegate=@n1) {is=title K2Node_AddDelegate}
   > UpdateLocalPlayerInfo(Player Info=GetLocalPlayerTeamMemberInfo!(self:Fort Party Context Object Reference=@n3) -> (LocalPlayerInfo:Fort Team Member Info Structure))
   Event Destruct()
   > Unbind Event from On Local Player State Changed(self:Fort Party Context Object Reference=@n3, Delegate=@n1) {is=title K2Node_RemoveDelegate}
  ]}
 UpdateLocalPlayerInfo {nodes [
   n1 = UpdateLocalPlayerInfo!() -> (then, Player Info:"Fort Team Member Info Structure (by ref)") {is=title K2Node_FunctionEntry}
   > SetVisibility(InVisibility=Select(Option 0=Collapsed, Option 1=SelfHitTestInvisible, Index=true))
   n2 = Get TextPlayerName() -> (TextPlayerName:Common Text Object Reference)
   > "SetText (Text)"(
    self=@n2
    InText=Break Fort Team Member Info!(FortTeamMemberInfo:"Fort Team Member Info Structure (by ref)"=@n1) -> (
     MemberUniqueId:Unique Net Id Structure
     PartyLeaderUniqueId:Unique Net Id Structure
     ConsoleUniqueId:Unique Net Id Structure
     Platform:String
     PlayerName:Text
     bPartyLeader:bool=~
     bIsInZone:bool=~
     bHasBoostXp:bool=~
     bHasRestXp:bool=~
     BannerIconId:String
     BannerColorId:String
     bBattlePassPurchased:bool=~
     BattlePassLevel:int
     NumPlayersInParty:int
     PlayerIndex:int
     TeamAffiliation:Byte
     HeroClass:Text
     HeroLevel:Text
     HeroXP:int
     HeroItem:Fort Item Object Reference
     SelectedGadgetItems:Array of Fort Item Object References
     HeroIcon:Slate Brush Structure
     HomeBaseInfo:Fort Home Base Info Structure
    ) {is=title K2Node_BreakStruct}.PlayerName
   )
   > SetVisibility(self=@n2, InVisibility=SelfHitTestInvisible)
   n3 = Get FortPartyContext()
   n4 = IsLocalPlayerPartyLeader!(self:Fort Party Context Object Reference=@n3) -> (ReturnValue:bool)
   n5 = "integer > integer"(A=GetNumPlayersInLocalParty!(self:Fort Party Context Object Reference=@n3) -> (ReturnValue:int=0), B=1) {is=title K2Node_CallFunction}
   > SetVisibility(
    self=Get LeaderImage() -> (LeaderImage:Image Object Reference)
    InVisibility=Select(Option 0=Collapsed, Option 1=SelfHitTestInvisible, Index=AND Boolean(A=@n5, B=@n4))
   )
  ]}
]
variables []
widgets {
 info {inheritedWidgetCount=1, parentClass=&FortHUDElementWidget, rootWidgetClass=&HorizontalBox, widgetCount=4}
 widgets [
  "[Horizontal Box] Player Info" {widgetClassPath=&HorizontalBox, bIsVariable=true, widget="[Horizontal Box] Player Info"}
  LeaderImage {
   widgetClassPath=&Image
   bIsVariable=true
   parent="[Horizontal Box] Player Info"
   widget=LeaderImage
   properties {
    visibility=Collapsed
    brush {
     bIsDynamicallyLoaded=false
     drawAs=Image
     imageType=NoImage
     mirroring=NoMirror
     resourceName=None
     resourceObject=/Game/UI/Foundation/Shell/Textures/Shell-Top-Crown
     tiling=NoTile
     imageSize=<20 20>
     margin=<ltrb 0 0 0 0>
     outlineSettings {
      bUseBrushTransparency=false
      roundingType=HalfHeightRadius
      width=0
      color {colorUseRule=UseColor_Specified, specifiedColor=<rgba 0 0 0 0>}
      cornerRadii=<0 0 0 0>
     }
     tintColor {colorUseRule=UseColor_Specified, specifiedColor=<rgba 1 1 1 1>}
     uVRegion {bIsValid=false, max=<0 0>, min=<0 0>}
    }
    colorAndOpacity=<rgba 1 1 1 0.4>
   }
   slot {class=HorizontalBoxSlot, properties {verticalAlignment=VAlign_Center, padding=<ltrb 4 0 4 0>}}
  }
  TextPlayerName {
   widgetClassPath=&CommonTextBlock
   bIsVariable=true
   parent="[Horizontal Box] Player Info"
   widget=TextPlayerName
   properties {
    text=Player Name
    visibility=Collapsed
    colorAndOpacity {colorUseRule=UseColor_Specified, specifiedColor=<rgba 1 1 1 0.4>}
    font {
     bForceMonospaced=false
     bMaterialIsStencil=false
     fontMaterial=null
     fontObject=&NotoSans
     letterSpacing=0
     monospacedWidth=1
     size=10
     skewAmount=0
     typefaceFontName=NotoSans-Regular
     outlineSettings {
      bApplyOutlineToDropShadows=false
      bMiteredCorners=false
      bSeparateFillAlpha=false
      outlineMaterial=null
      outlineSize=0
      outlineColor=<rgba 0 0 0 1>
     }
    }
   }
   slot {class=HorizontalBoxSlot, properties {verticalAlignment=VAlign_Center}}
  }
  TutorialHighlight {widgetClassPath=&TutorialHighlight}
 ]
}
