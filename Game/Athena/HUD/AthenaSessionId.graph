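# AthenaSessionId is a Fortnite Athena HUD element widget that displays the current session ID as small, centered, semi-transparent text.
# On Construct it reads the session ID from the FortGlobalUIContext and writes it to the Text block.
# The text collapses when the ID is empty and is otherwise hit-test invisible.
# The value is set only at construction, with no later refresh in the graph.
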WidgetBlueprint /Game/Athena/HUD/AthenaSessionId
graphs [
 EventGraph {nodes [
   Event Construct()
   > UpdateSessionId()
  ]}
 UpdateSessionId {nodes [
   UpdateSessionId() {is=title K2Node_FunctionEntry}
   n1 = "To Text (String)"(InString=GetSessionId!(self:Fort Global UIContext Object Reference=Get FortGlobalUIContext!() -> (ReturnValue:Fort Global UIContext Object Reference) {is=title K2Node_GetBlueprintContextEx}) -> (ReturnValue:String))
   n2 = Get Text!() -> (Text:Text Object Reference)
   > "SetText (Text)"(self=@n2, InText=@n1)
   > SetVisibility(self=@n2, InVisibility=Select(Option 0=HitTestInvisible, Option 1=Collapsed, Index=TextIsEmpty(InText=@n1)))
  ]}
]
variables []
widgets {
 info {inheritedWidgetCount=1, parentClass=&FortHUDElementWidget, rootWidgetClass=&TextBlock, widgetCount=2}
 widgets [
  Text {
   widgetClassPath=&TextBlock
   bIsVariable=true
   widget=Text
   properties {
    justification=Center
    text=SESSION ID
    colorAndOpacity {colorUseRule=UseColor_Specified, specifiedColor=<rgba 1 1 1 0.4>}
    font {
     bForceMonospaced=false
     bMaterialIsStencil=false
     fontMaterial=null
     fontObject=&NotoSans
     letterSpacing=0
     monospacedWidth=1
     size=10
     skewAmount=0
     typefaceFontName=NotoSans-Regular
     outlineSettings {
      bApplyOutlineToDropShadows=false
      bMiteredCorners=false
      bSeparateFillAlpha=false
      outlineMaterial=null
      outlineSize=0
      outlineColor=<rgba 0 0 0 1>
     }
    }
   }
  }
  TutorialHighlight {widgetClassPath=&TutorialHighlight}
 ]
}
