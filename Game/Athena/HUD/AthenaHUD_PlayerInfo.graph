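# Athena HUD element widget that lays out player, session, server address and module identifier readouts.
# A scaled-down ScaleBox stacks two rows: a collapsed player info widget with session ID and server address, then module ID.
# The three ID widgets carry gameplay tags for hiding during capture, post-game and post-game spectating; module ID reuses the SessionId element tag.
# The event graph has only unconnected PreConstruct, Construct and Tick events, so the widget adds no logic.
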WidgetBlueprint /Game/Athena/HUD/AthenaHUD_PlayerInfo
graphs [EventGraph {nodes [
   Event PreConstruct()
   Event Construct()
   Event Tick()
  ]}]
variables []
widgets {
 info {inheritedWidgetCount=1, parentClass=&FortHUDElementWidget, rootWidgetClass=&ScaleBox, widgetCount=9}
 widgets [
  ScaleBox_1 {widgetClassPath=&ScaleBox, widget=ScaleBox_1, properties {stretch=UserSpecified, userSpecifiedScale=0.8}}
  VerticalBox_0 {widgetClassPath=&VerticalBox, parent=ScaleBox_1, widget=VerticalBox_0, slot {class=ScaleBoxSlot}}
  HorizontalBox_Playerinfo {widgetClassPath=&HorizontalBox, parent=VerticalBox_0, widget=HorizontalBox_Playerinfo, slot {class=VerticalBoxSlot}}
  AthenaPlayerInfo_C_0 {
   widgetClassPath=&/AthenaPlayerInfo
   parent=HorizontalBox_Playerinfo
   widget=AthenaPlayerInfo_C_0
   properties {visibility=Collapsed}
   slot {class=HorizontalBoxSlot}
  }
  AthenaSessionId {
   widgetClassPath=&/AthenaSessionId
   parent=HorizontalBox_Playerinfo
   widget=AthenaSessionId
   properties {
    elementTag {tagName=HUD.Athena.SessionId}
    hUDElementTag {gameplayTags [{tagName=Athena.HideForCapture}, {tagName=HUD.Athena.HideForPostGameSpectate}, {tagName=HUD.Athena.HideForPostGame}]}
   }
   slot {class=HorizontalBoxSlot, properties {padding=<ltrb 20 0 20 0>}}
  }
  AthenaServerAddress {
   widgetClassPath=&/AthenaServerAddress
   parent=HorizontalBox_Playerinfo
   widget=AthenaServerAddress
   properties {
    elementTag {tagName=HUD.Athena.ServerAddress}
    hUDElementTag {gameplayTags [{tagName=Athena.HideForCapture}, {tagName=HUD.Athena.HideForPostGameSpectate}, {tagName=HUD.Athena.HideForPostGame}]}
   }
   slot {class=HorizontalBoxSlot}
  }
  HorizontalBox_ModuleInfo {widgetClassPath=&HorizontalBox, parent=VerticalBox_0, widget=HorizontalBox_ModuleInfo, slot {class=VerticalBoxSlot}}
  AthenaModuleId {
   widgetClassPath=&/AthenaModuleId
   bIsVariable=true
   parent=HorizontalBox_ModuleInfo
   widget=AthenaModuleId
   properties {
    elementTag {tagName=HUD.Athena.SessionId}
    hUDElementTag {gameplayTags [{tagName=Athena.HideForCapture}, {tagName=HUD.Athena.HideForPostGame}, {tagName=HUD.Athena.HideForPostGameSpectate}]}
   }
   slot {class=HorizontalBoxSlot}
  }
  TutorialHighlight {widgetClassPath=&TutorialHighlight}
 ]
}
