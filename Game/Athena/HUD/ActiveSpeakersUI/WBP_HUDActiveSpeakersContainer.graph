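# HUD widget that displays the list of active voice chat speakers, built on the native FortVoiceSpeakerContainerWidget with no Blueprint logic.
# An inherited DynamicEntryBox creates WBP_HUDActiveSpeakerEntry widgets in a right-to-left vertical wrap, inside a fixed-size, right-aligned SizeBox.
# The inherited channel panel, with its speaker icon and "PARTY" label, sits collapsed inside a collapsed overlay named DeleteMe.
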WidgetBlueprint /Game/Athena/HUD/ActiveSpeakersUI/WBP_HUDActiveSpeakersContainer
graphs [EventGraph {links [], nodes []}]
variables []
widgets {
 info {inheritedWidgetCount=3, parentClass=/Script/FortniteUI.FortVoiceSpeakerContainerWidget, rootWidgetClass=&Overlay, widgetCount=7}
 widgets [
  Overlay_1 {widgetClassPath=&Overlay, widget=Overlay_1}
  SizeBox_HeightReservation {
   widgetClassPath=&SizeBox
   parent=Overlay_1
   widget=SizeBox_HeightReservation
   properties {
    bOverride_HeightOverride=true
    bOverride_MaxDesiredHeight=true
    bOverride_MaxDesiredWidth=true
    bOverride_MinDesiredHeight=true
    bOverride_WidthOverride=true
    heightOverride=400
    maxDesiredHeight=400
    maxDesiredWidth=400
    minDesiredHeight=52
    widthOverride=500
   }
   slot {class=OverlaySlot, properties {horizontalAlignment=HAlign_Right}}
  }
  DynamicEntry_Speakers {
   widgetClassPath=&DynamicEntryBox
   bInherited=true
   bIsVariable=true
   parent=SizeBox_HeightReservation
   widget=DynamicEntry_Speakers
   properties {
    entryBoxType=VerticalWrap
    entryWidgetClass=&/WBP_HUDActiveSpeakerEntry.WBP_HUDActiveSpeakerEntry_C
    flowDirectionPreference=RightToLeft
    numDesignerPreviewEntries=16
    entrySpacing=<10 0>
    renderTransform {angle=0, scale=<1 1>, shear=<0 0>, translation=<100 0>}
   }
   slot {class=SizeBoxSlot}
  }
  DeleteMe {widgetClassPath=&Overlay, parent=Overlay_1, widget=DeleteMe, properties {visibility=Collapsed}, slot {class=OverlaySlot}}
  Panel_ChannelContainer {
   widgetClassPath=&HorizontalBox
   bInherited=true
   bIsVariable=true
   parent=DeleteMe
   widget=Panel_ChannelContainer
   properties {visibility=Collapsed}
   slot {class=OverlaySlot}
  }
  SpeakerIcon {
   widgetClassPath=/Script/FortniteUI.FortMobileImage
   bIsVariable=true
   parent=Panel_ChannelContainer
   widget=SpeakerIcon
   properties {
    mobileSizeMultiplier=1.4
    brush {
     bIsDynamicallyLoaded=false
     drawAs=Image
     imageType=NoImage
     mirroring=NoMirror
     resourceName=None
     resourceObject=/Game/Athena/HUD/Art/M_VoiceChatStatusIcon
     tiling=NoTile
     imageSize=<32 32>
     margin=<ltrb 0 0 0 0>
     outlineSettings {
      bUseBrushTransparency=false
      roundingType=HalfHeightRadius
      width=0
      color {colorUseRule=UseColor_Specified, specifiedColor=<rgba 0 0 0 0>}
      cornerRadii=<0 0 0 0>
     }
     tintColor {colorUseRule=UseColor_Specified, specifiedColor=<rgba 1 1 1 0.5>}
     uVRegion {bIsValid=false, max=<0 0>, min=<0 0>}
    }
    colorAndOpacity=<rgba 0.4452 1 1 1>
   }
   slot {class=HorizontalBoxSlot, properties {padding=<ltrb 8 0 0 0>}}
  }
  Text_ChannelDisplayName {
   widgetClassPath=&CommonTextBlock
   bInherited=true
   parent=Panel_ChannelContainer
   widget=Text_ChannelDisplayName
   properties {
    mobileFontSizeMultiplier=1.4
    text=PARTY
    colorAndOpacity {colorUseRule=UseColor_Specified, specifiedColor=<rgba 0.4452 1 1 0.5>}
    font {
     bForceMonospaced=false
     bMaterialIsStencil=false
     fontMaterial=null
     fontObject=&BurbankSmall
     letterSpacing=0
     monospacedWidth=1
     size=17
     skewAmount=0
     typefaceFontName=Black
     outlineSettings {
      bApplyOutlineToDropShadows=false
      bMiteredCorners=false
      bSeparateFillAlpha=false
      outlineMaterial=null
      outlineSize=0
      outlineColor=<rgba 0 0 0 1>
     }
    }
    shadowOffset=<0 0>
   }
   slot {class=HorizontalBoxSlot, properties {verticalAlignment=VAlign_Center}}
  }
 ]
}
