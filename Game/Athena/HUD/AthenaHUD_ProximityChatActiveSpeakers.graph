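# HUD element widget for Athena that shows active voice speakers with proximity chat support.
# A user-scaled ScaleBox slightly shrinks nested vertical and horizontal boxes around a WBP_HUDActiveSpeakersContainer that has proximity chat support switched on.
# The event graph holds only unconnected stub events, so behavior comes from that container and the FortHUDElementWidget parent class.
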WidgetBlueprint /Game/Athena/HUD/AthenaHUD_ProximityChatActiveSpeakers
graphs [EventGraph {nodes [
   Event PreConstruct()
   Event Construct()
   Event Tick()
  ]}]
variables []
widgets {
 info {inheritedWidgetCount=1, parentClass=&FortHUDElementWidget, rootWidgetClass=&ScaleBox, widgetCount=5}
 widgets [
  ScaleBox_1 {widgetClassPath=&ScaleBox, widget=ScaleBox_1, properties {stretch=UserSpecified, userSpecifiedScale=0.8}}
  VerticalBox_0 {widgetClassPath=&VerticalBox, parent=ScaleBox_1, widget=VerticalBox_0, slot {class=ScaleBoxSlot}}
  HorizontalBox_VoiceIndicators {widgetClassPath=&HorizontalBox, parent=VerticalBox_0, widget=HorizontalBox_VoiceIndicators, slot {class=VerticalBoxSlot}}
  WBP_HUDActiveSpeakersContainer {
   widgetClassPath=/Game/Athena/HUD/ActiveSpeakersUI/WBP_HUDActiveSpeakersContainer
   bIsVariable=true
   parent=HorizontalBox_VoiceIndicators
   widget=WBP_HUDActiveSpeakersContainer
   properties {bShouldSupportProximityChat=true}
   slot {class=HorizontalBoxSlot}
  }
  TutorialHighlight {widgetClassPath=&TutorialHighlight}
 ]
}
