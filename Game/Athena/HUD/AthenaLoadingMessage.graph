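# Athena HUD loading message widget, built on FortAthenaLoadingMessage, that shows a throbber beside uppercase status text.
# The root horizontal box starts collapsed and holds the spinner image and a CommonTextBlock.
# When LoadingScreenStateChanged enables the load screen, it reveals the box and displays the supplied HUD reason text, falling back to localized "Relocating" if empty.
# When the load screen is disabled, it collapses the box.
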WidgetBlueprint /Game/Athena/HUD/AthenaLoadingMessage
graphs [EventGraph {nodes [
   n1 = Event LoadingScreenStateChanged!() -> (OutputDelegate, then, bEnableLoadScreen:bool, HUDReasonText:"Text (by ref)") {is=title K2Node_Event}
   > n2 = Branch(Condition=@n1.bEnableLoadScreen)
   n3 = Get LoadingBox() -> (LoadingBox:Horizontal Box Object Reference)
   > SetVisibility(self=@n3, InVisibility=SelfHitTestInvisible)
   > "SetText (Text)"(
    self=Get RelocatingTextBlock() -> (RelocatingTextBlock:Common Text Object Reference)
    InText=Select(Option 0:Text=@n1.HUDReasonText, Option 1:Text="LOCTEXT(\"Relocating\")", Index=TextIsEmpty(InText=@n1.HUDReasonText)) -> (ReturnValue:Text)
   )
   SetVisibility(execute=@n2.else, self=@n3, InVisibility=Collapsed)
  ]}]
variables []
widgets {
 info {inheritedWidgetCount=0, parentClass=/Script/FortniteUI.FortAthenaLoadingMessage, rootWidgetClass=&HorizontalBox, widgetCount=3}
 widgets [
  LoadingBox {widgetClassPath=&HorizontalBox, bIsVariable=true, widget=LoadingBox, properties {visibility=Collapsed}}
  ImageSpinner {
   widgetClassPath=&Image
   parent=LoadingBox
   widget=ImageSpinner
   properties {
    visibility=SelfHitTestInvisible
    brush {
     bIsDynamicallyLoaded=false
     drawAs=Image
     imageType=NoImage
     mirroring=NoMirror
     resourceName=None
     resourceObject=/FortUILibrary/UIKit/Throbber/MI_UIKit_Monochrome_Throbber
     tiling=NoTile
     imageSize=<42 42>
     margin=<ltrb 0 0 0 0>
     outlineSettings {
      bUseBrushTransparency=false
      roundingType=HalfHeightRadius
      width=0
      color {colorUseRule=UseColor_Specified, specifiedColor=<rgba 0 0 0 0>}
      cornerRadii=<0 0 0 0>
     }
     tintColor {colorUseRule=UseColor_Specified, specifiedColor=<rgba 1 1 1 1>}
     uVRegion {bIsValid=false, max=<0 0>, min=<0 0>}
    }
   }
   slot {class=HorizontalBoxSlot, properties {horizontalAlignment=HAlign_Right, verticalAlignment=VAlign_Center, padding=<ltrb 0 0 20 0>}}
  }
  RelocatingTextBlock {
   widgetClassPath=&CommonTextBlock
   bIsVariable=true
   parent=LoadingBox
   widget=RelocatingTextBlock
   properties {
    text=Relocating
    textTransformPolicy=ToUpper
    font {
     bForceMonospaced=false
     bMaterialIsStencil=false
     fontMaterial=null
     fontObject=&Burbank
     letterSpacing=0
     monospacedWidth=1
     size=18
     skewAmount=0
     typefaceFontName=BurbankRegular-Black
     outlineSettings {
      bApplyOutlineToDropShadows=false
      bMiteredCorners=false
      bSeparateFillAlpha=false
      outlineMaterial=null
      outlineSize=1
      outlineColor=<rgba 0.0153 0 0.0156 0.5>
     }
    }
    shadowOffset=<0 0>
   }
   slot {class=HorizontalBoxSlot, properties {horizontalAlignment=HAlign_Left, verticalAlignment=VAlign_Center}}
  }
 ]
}
