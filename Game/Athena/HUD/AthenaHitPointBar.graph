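# HUD widget for a player hit point bar, configured by BarType as health, shield, armor or vehicle health.
# It async-loads fill and delta materials per type, then base-class value events drive their progress parameters, using console variables for segmented bars.
# Health and shield losses pulse the number and shake the bar, escalating at low values.
# Health bars also show blocked damage, a severity shadow and red down-but-not-out styling.
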WidgetBlueprint /Game/Athena/HUD/AthenaHitPointBar
animations [
 Anim_BarShake {
  binding_count=1
  length_seconds=0.1849
  object=&self:Anim_BarShake
  sequence {bindings [SizeBox_Bar {
     id=#1
     possessed_class=SizeBox
     tracks [{
       type=MovieScene2DTransformTrack
       sections [{
         type=MovieScene2DTransformSection
         end_frame=3
         channels [
          Translation.X {type=MovieSceneScriptingFloatChannel, keys [[0,0],[0.8513,-6.8849],[1.4223,3.9219],[2.184,-3.421],[2.7867,0.8919],[3.698,0]]}
          Translation.Y {type=MovieSceneScriptingFloatChannel, keys [[0,0],[0.8513,6.9314],[1.4223,-5.0319],[2.184,4.061],[2.7867,-2.5554],[3.698,0]]}
          Angle {type=MovieSceneScriptingFloatChannel}
          Scale.X {type=MovieSceneScriptingFloatChannel}
          Scale.Y {type=MovieSceneScriptingFloatChannel}
          Shear.X {type=MovieSceneScriptingFloatChannel}
          Shear.Y {type=MovieSceneScriptingFloatChannel}
         ]
        }]
      }]
    }]}
 }
 Anim_BarShake_1 {
  binding_count=1
  length_seconds=0.1849
  object=&self:Anim_BarShake_1
  sequence {bindings [SizeBox_Bar {
     id=#1
     possessed_class=SizeBox
     tracks [{
       type=MovieScene2DTransformTrack
       sections [{
         type=MovieScene2DTransformSection
         end_frame=3
         channels [
          Translation.X {type=MovieSceneScriptingFloatChannel, keys [[0,0],[0.8513,-6.8849],[1.4223,3.9219],[2.184,-3.421],[2.7867,0.8919],[3.698,0]]}
          Translation.Y {type=MovieSceneScriptingFloatChannel, keys [[0,0],[0.8513,6.9314],[1.4223,-5.0319],[2.184,4.061],[2.7867,-2.5554],[3.698,0]]}
          Angle {type=MovieSceneScriptingFloatChannel}
          Scale.X {type=MovieSceneScriptingFloatChannel}
          Scale.Y {type=MovieSceneScriptingFloatChannel}
          Shear.X {type=MovieSceneScriptingFloatChannel}
          Shear.Y {type=MovieSceneScriptingFloatChannel}
         ]
        }]
      }]
    }]}
 }
 Anim_BarShake_2 {
  binding_count=1
  length_seconds=0.1849
  object=&self:Anim_BarShake_2
  sequence {bindings [SizeBox_Bar {
     id=#1
     possessed_class=SizeBox
     tracks [{
       type=MovieScene2DTransformTrack
       sections [{
         type=MovieScene2DTransformSection
         end_frame=3
         channels [
          Translation.X {type=MovieSceneScriptingFloatChannel, keys [[0,0],[0.8513,-9.27],[1.4223,6.3372],[2.184,-5.474],[2.5827,1.7071],[3.698,0]]}
          Translation.Y {type=MovieSceneScriptingFloatChannel, keys [[0,0],[0.8513,9.075],[1.4223,-7.2057],[2.184,4.061],[2.5953,-2.9479],[3.698,0]]}
          Angle {type=MovieSceneScriptingFloatChannel}
          Scale.X {type=MovieSceneScriptingFloatChannel}
          Scale.Y {type=MovieSceneScriptingFloatChannel}
          Shear.X {type=MovieSceneScriptingFloatChannel}
          Shear.Y {type=MovieSceneScriptingFloatChannel}
         ]
        }]
      }]
    }]}
 }
 Anim_ValueImpact {
  binding_count=2
  length_seconds=0.2513
  object=&self:Anim_ValueImpact
  sequence {bindings [
    CurrentValue {
     id=#2
     possessed_class=CommonTextBlock
     tracks [{
       type=MovieScene2DTransformTrack
       sections [{
         type=MovieScene2DTransformSection
         channels [
          Scale.X {type=MovieSceneScriptingFloatChannel, keys [[0,1],[2,1.666],[5,1]]}
          Scale.Y {type=MovieSceneScriptingFloatChannel, keys [[0,1],[2,1.666],[5,1]]}
         ]
        }]
      }]
    }
    Image_FillBar {
     id=#3
     possessed_class=Image
     tracks [{
       type=MovieSceneWidgetMaterialTrack
       sections [{
         type=MovieSceneParameterSection
         channels [AnimateMarkerIntensity {type=MovieSceneScriptingFloatChannel, keys [[-0.026,0],[0,0],[0.974,1],[5,0]]}]
        }]
      }]
    }
   ]}
 }
]
graphs [
 EventGraph {nodes [
   OnBarShakeAnimationFinished() {is=title K2Node_CustomEvent}
   > InvalidateLayoutAndVolatility(self:Widget Object Reference=Get SizeBox_Bar() -> (SizeBox_Bar:Size Box Object Reference))
   InitializeBar() {is=title K2Node_CustomEvent}
   > n1 = Load Soft Hit Bar Material!(
    execute:Exec
    Target:Image Object Reference=Get Image_FillBar() -> (Image_FillBar:Image Object Reference)
    SoftMaterial:Material Interface Soft Object Reference=Select!(
     Health:Material Interface Soft Object Reference=/Game/Athena/HUD/Materials/M_UI_HitPointBar_Health
     Shield:Material Interface Soft Object Reference=GetShieldMaterial!() -> (Shield Material:Material Interface Soft Object Reference)
     Overshield:Material Interface Soft Object Reference
     Armor:Material Interface Soft Object Reference=Get Armor_Segments() -> (Armor_Segments:Material Interface Soft Object Reference)
     Stamina:Material Interface Soft Object Reference
     VehicleHealth:Material Interface Soft Object Reference=/Game/Athena/HUD/Materials/M_UI_HitPointBar_Vehicle
     SignalInStorm:Material Interface Soft Object Reference
     Index:EHealthBarType Enum=Get BarType() -> (BarType:EHealthBarType Enum=Health)
    ) -> (ReturnValue:Material Interface Soft Object Reference)
   ) -> (then:Exec, LoadCompleted:Exec) {is=title K2Node_MacroInstance}
   > n2 = Load Soft Hit Bar Material(
    Target=Get Image_DeltaBar() -> (Image_DeltaBar:Image Object Reference)
    SoftMaterial=Select(
     Health=/Game/Athena/HUD/Materials/M_UI_Delta_Health
     Shield=GetShieldDeltaMaterial!() -> (ReturnValue:Material Interface Soft Object Reference)
     Armor=/Game/Athena/HUD/Materials/M_UI_HitPointBar_Armor_SegmentsBacking
     VehicleHealth=/Game/Athena/HUD/Materials/M_UI_Delta_Vehicle
     Index=Get BarType() -> (BarType=Health)
    )
   )
   > RefreshIcons()
   > SetUserSpecifiedScale!(
    execute
    self:Scale Box Object Reference=Get ScaleBox_Icon() -> (ScaleBox_Icon:Scale Box Object Reference)
    InUserSpecifiedScale:float=Select(
     Health:double=1
     Shield:double=1
     Overshield:double=0.8
     Armor:double=0
     Stamina:double=1
     VehicleHealth:double=1
     SignalInStorm:double=1
     Index=Get BarType() -> (BarType=Health)
    ) -> (ReturnValue:double)
   ) -> (then)
   > SetPadding!(
    execute
    self:Overlay Slot Object Reference=SlotAsOverlaySlot(Widget=Get HBox_TextArea() -> (HBox_TextArea:Horizontal Box Object Reference))
    InPadding:Margin Structure=Make Margin(Left=16, Top=0, Right=0, Bottom=0)
   ) -> (then)
   > Branch(Condition="Equal (Enum)"(A:EHealthBarType Enum=Get BarType() -> (BarType=Health), B:EHealthBarType Enum=Armor))
   > UpdateSegmentFill()
   Branch(execute=@n2.LoadCompleted, Condition=Get HasReceivedDeltaUpdate() -> (HasReceivedDeltaUpdate:bool))
   > UpdateMaxValue()
   Branch(execute=@n1.LoadCompleted, Condition=Get HasReceivedValueUpdate() -> (HasReceivedValueUpdate:bool))
   > UpdateCurrentValue()
   OnValueImpactAnimationFinished() {is=title K2Node_CustomEvent}
   > SetRenderScale!(execute, self:Widget Object Reference=Get CurrentValue() -> (CurrentValue:Common Text Object Reference), Scale:Vector 2D Structure=(X=1,Y=1)) -> (then)
   > n3 = GetDynamicMaterial(self=Get Image_FillBar())
   > SetScalarParameterValue(self=@n3, ParameterName=AnimateMarkerIntensity, Value=0)
   Event OnDeltaChanged() {is=title K2Node_Event}
   > Update Delta Bar()
   Event Construct()
   > InitializeBar() {is=title K2Node_CallFunction}
   > BindToAnimationFinished(Animation=Get Anim_BarShake() -> (Anim_BarShake:Widget Animation Object Reference), Delegate=Create Event())
   > BindToAnimationFinished(Animation=Get Anim_ValueImpact() -> (Anim_ValueImpact:Widget Animation Object Reference), Delegate=Create Event())
   n4 = Event PreConstruct()
   > Set HUDElementTag!(
    execute
    HUDElementTag:Gameplay Tag Container Structure=Get TextHudElementTags() -> (TextHudElementTags:Gameplay Tag Container Structure)
    self:Fort HUDElement Widget Object Reference=Get FortHUDElementWrapper_Text() -> (FortHUDElementWrapper_Text:Fort HUDElement Wrapper Object Reference)
   ) -> (then, Output_Get:Gameplay Tag Container Structure)
   > Branch(Condition=@n4.IsDesignTime)
   > InitializeBar()
   > SetInformationVisibility!(
    execute
    self
    Icon_Visibility:bool=Get bIcon_isVisible() -> (bIcon_isVisible:bool)
    Value_Visibility:bool=Get bValue_isVisible() -> (bValue_isVisible:bool)
   ) -> (then)
   Event OnMaxValueChanged!() -> (OutputDelegate, then, Value:float) {is=title K2Node_Event}
   > Set HasReceivedDeltaUpdate(HasReceivedDeltaUpdate:bool=true) -> (Output_Get:bool)
   > UpdateMaxValue()
   n5 = Event OnDBNOStateChanged!() -> (OutputDelegate, then, IsDBNO:bool) {is=title K2Node_Event}
   > UpdateDBNOState!(execute, self, bIsDBNO:bool=@n5.IsDBNO) -> (then)
   > InvalidateLayoutAndVolatility()
   n6 = Event OnBlockedValueChanged!() -> (OutputDelegate, then, InterpolationProgress:float) {is=title K2Node_Event}
   > UpdateBlockedDamageBar!(execute, self, BlockedValueInterpolation:double=@n6.InterpolationProgress) -> (then)
   n7 = Event OnValueChangedWithReason!() -> (OutputDelegate, then, Value:float, Reason:EFortHitPointModificationReason Enum=Invalid) {is=title K2Node_Event}
   > Set ValueChangedReason(ValueChangedReason:EFortHitPointModificationReason Enum=@n7.Reason) -> (Output_Get:EFortHitPointModificationReason Enum=Invalid)
   > Set HasReceivedValueUpdate(HasReceivedValueUpdate:bool=true) -> (Output_Get:bool)
   > UpdateCurrentValue()
   n8 = Event OnExtensionWidgetAdded!() -> (OutputDelegate, then, ExtensionSlot:EUIExtensionSlot Enum=Invalid, ExtensionWidget:User Widget Object Reference) {is=title K2Node_Event}
   > n9 = AddChildToOverlay!(
    execute
    self:Overlay Object Reference=Get Overlay_ExtensionWidget() -> (Overlay_ExtensionWidget:Overlay Object Reference)
    Content:Widget Object Reference=@n8.ExtensionWidget
   ) -> (then, ReturnValue:Overlay Slot Object Reference)
   > SetHorizontalAlignment(self=@n9, InHorizontalAlignment=HAlign_Fill)
   > SetVerticalAlignment(self=@n9, InVerticalAlignment=VAlign_Fill)
   Event Destruct()
   > UnbindAllFromAnimationFinished!(execute, self:User Widget Object Reference, Animation:Widget Animation Object Reference=Get Anim_BarShake()) -> (then)
   > UnbindAllFromAnimationFinished(Animation=Get Anim_ValueImpact())
   n10 = Event OnBlockedDamageAnimationStarted!() -> (OutputDelegate, then, BlockedBarStartPosition:float) {is=title K2Node_Event}
   > ShowBlockedDamageBar!(execute, self, StartValue:double=@n10.BlockedBarStartPosition) -> (then)
  ]}
 GetShieldDeltaMaterial {nodes [
   GetShieldDeltaMaterial() {is=title K2Node_FunctionEntry}
   > Return Node!(
    execute
    ReturnValue:Material Interface Soft Object Reference=Select!(
     Option 0:Material Interface Soft Object Reference=Get ShieldDelta_Full() -> (ShieldDelta_Full:Material Interface Soft Object Reference)
     Option 1:Material Interface Soft Object Reference=Get ShieldDelta_Segments() -> (ShieldDelta_Segments:Material Interface Soft Object Reference)
     Index:bool=IsSegmentedShield!() -> (ReturnValue:bool)
    ) -> (ReturnValue:Material Interface Soft Object Reference)
   )
  ]}
 GetShieldMaterial {nodes [
   GetShieldMaterial() {is=title K2Node_FunctionEntry}
   > Return Node!(
    execute
    Shield Material:Material Interface Soft Object Reference=Select(
     Option 0=Get Shield_Full() -> (Shield_Full:Material Interface Soft Object Reference)
     Option 1=Get Shield_Segments() -> (Shield_Segments:Material Interface Soft Object Reference)
     Index=IsSegmentedShield()
    )
   )
  ]}
 IsSegmentedShield {nodes [
   IsSegmentedShield() {is=title K2Node_FunctionEntry}
   n1 = IsShieldSegmented!(self:Athena Player Hit Point Bar Base Object Reference) -> (ReturnValue:bool)
   n2 = "Equal (Enum)"(A=Get BarType() -> (BarType=Health), B=Shield)
   n3 = Get BarType() -> (BarType=Health)
   > Return Node!(execute, ReturnValue:bool=OR Boolean(A=AND Boolean(A=@n2, B=@n1), B="Equal (Enum)"(A=@n3, B=Armor)))
  ]}
 LoadSoftHitBarMaterial {nodes [
   n1 = Inputs!() -> (execute, Target:Image Object Reference, SoftMaterial:Material Interface Soft Object Reference)
   > n2 = Async Load Asset(Asset=@n1.SoftMaterial)
   > Outputs!(then:Exec, LoadCompleted:Exec=@n4.then)
   n3 = Cast To MaterialInterface!(execute=@n2.Completed, Object=@n2) -> (then, CastFailed, AsMaterial Interface:Material Interface Object Reference) {is=title K2Node_DynamicCast}
   > n4 = SetBrushFromMaterial(self=@n1, Material=@n3)
  ]}
 RefreshIcons {nodes [
   RefreshIcons() {is=title K2Node_FunctionEntry}
   > SetBrushFromSoftMaterial!(
    execute
    self:Image Object Reference=Get Image_ValueIcon() -> (Image_ValueIcon:Image Object Reference)
    SoftMaterial:Material Interface Soft Object Reference=Select!(
     Health:Material Interface Soft Object Reference=/Game/Athena/HUD/Materials/M_UI_HitPoint_Type_Health
     Shield:Material Interface Soft Object Reference=/Game/Athena/HUD/Materials/M_UI_HitPoint_Type_Shield
     Overshield:Material Interface Soft Object Reference
     Armor:Material Interface Soft Object Reference
     Stamina:Material Interface Soft Object Reference
     VehicleHealth:Material Interface Soft Object Reference=/Game/Athena/HUD/Materials/M_UI_HitPoint_Type_Vehicle
     SignalInStorm:Material Interface Soft Object Reference
     Index:EHealthBarType Enum=Get BarType() -> (BarType=Health)
    ) -> (ReturnValue)
   ) -> (then)
  ]}
 SetDBNOStyle {nodes [
   SetDBNOStyle() {is=title K2Node_FunctionEntry}
   > n1 = GetDynamicMaterial(self=Get Image_FillBar())
   > SetVectorParameterValue(self=@n1, ParameterName=Marker_Color, Value=(R=1,G=0.0999,B=0.0999,A=1))
   > SetVectorParameterValue(self=@n1, ParameterName=Progress_Color_Start, Value=(R=0.5776,G=0,B=0.0232,A=1))
   > SetVectorParameterValue(self=@n1, ParameterName=Progress_Color_Stop, Value=(R=1,G=0,B=0.1144,A=1))
  ]}
 SetHealthStyle {nodes [
   SetHealthStyle() {is=title K2Node_FunctionEntry}
   > Branch(Condition="Equal (Enum)"(A=Get BarType() -> (BarType=Health), B=Health))
   > n1 = GetDynamicMaterial(self=Get Image_FillBar())
   > n2 = Is Valid(InputObject=@n1)
   n3 = Get DBNOStateCache() -> (DBNOStateCache:bool)
   FortShippingLog!(
    execute=@n2.Is Not Valid
    WorldContextObject:Object Reference=Self-Reference()
    InString:String=Append(A="SetHealthStyle get fill bar material failed. DBNO state is: ", B="To String (Boolean)"(InBool=@n3))
    bLogAsWarning:bool=false
   ) -> (then)
   SetVectorParameterValue(
    execute=@n2.Is Valid
    self=@n1
    ParameterName=Marker_Color
    Value=SelectColor(A=(R=1,G=0.0999,B=0.0999,A=1), B=(R=0.3419,G=1,B=0.1248,A=1), bPickA=Get DBNOStateCache())
   )
   > SetVectorParameterValue(
    self=@n1
    ParameterName=Progress_Color_Start
    Value=SelectColor(A=(R=0.5776,G=0,B=0.0232,A=1), B=(R=0,G=0.5029,B=0.0168,A=1), bPickA=Get DBNOStateCache())
   )
   > SetVectorParameterValue(
    self=@n1
    ParameterName=Progress_Color_Stop
    Value=SelectColor(A=(R=1,G=0,B=0.1144,A=1), B=(R=0.0666,G=0.7529,B=0.0185,A=1), bPickA=Get DBNOStateCache())
   )
  ]}
 SetInformationVisibility {nodes [
   SetInformationVisibility!() -> (then, Icon_Visibility:bool=~, Value_Visibility:bool=~) {is=title K2Node_FunctionEntry}
   > n1 = Sequence()
   > SetVisibility(
    self=Get SizeBox_IconFixedSpacing() -> (SizeBox_IconFixedSpacing:Size Box Object Reference)
    InVisibility=Select!(Option 0:ESlateVisibility Enum=Collapsed, Option 1:ESlateVisibility Enum=HitTestInvisible, Index:bool=Get bIcon_isVisible()) -> (ReturnValue:ESlateVisibility Enum)
   )
   SetVisibility(
    execute=@n1.then_1
    self=Get SizeBox_Value() -> (SizeBox_Value:Size Box Object Reference)
    InVisibility=Select(Option 0=Collapsed, Option 1=SelfHitTestInvisible, Index=Get bValue_isVisible())
   )
  ]}
 SetSize {nodes [
   n1 = SetSize!() -> (then, useLargeSize:bool=~) {is=title K2Node_FunctionEntry}
   n2 = Select(Option 0:double=Get NormalHeight() -> (NormalHeight:double), Option 1:double=Get TallHeight() -> (TallHeight:double), Index=@n1) -> (ReturnValue:double)
   > SetHeightOverride(self=Get SizeBox_Bar(), InHeightOverride=@n2)
   > SetDesiredSizeOverride!(execute, self:Image Object Reference=Get Image_ValueIcon(), DesiredSize:Vector 2D Structure=MakeVector2D(X=@n2, Y=@n2)) -> (then)
  ]}
 ShowBlockedDamageBar {nodes [
   ShowBlockedDamageBar!() -> (then, StartValue:double=~) {is=title K2Node_FunctionEntry}
   > Branch(Condition="Equal (Enum)"(A=Get BarType() -> (BarType=Health), B=Health))
   > n1 = GetDynamicMaterial(self=Get Image_BlockedDelta() -> (Image_BlockedDelta:Image Object Reference))
   > SetScalarParameterValue(
    self=@n1
    ParameterName=Progress
    Value=GetCurrentValuePercentage!(self:Athena Player Hit Point Bar Base Object Reference) -> (ReturnValue:float)
   )
   > n2 = GetDynamicMaterial(self=Get Image_FillBar())
   > SetScalarParameterValue(self=@n2, ParameterName=Progress, Value=Get StartValue() -> (StartValue:double))
   n3 = Get Image_BlockedDelta()
   > SetBrushTintColor!(execute, self:Image Object Reference=@n3, TintColor:Slate Color Structure=Make Slate Color(SpecifiedColor=(R=1,G=1,B=1,A=1))) -> (then)
   > SetVisibility(self=@n3, InVisibility=Visible)
   > n4 = GetDynamicMaterial(self=Get Image_FillBar())
   > SetScalarParameterValue(self=@n4, ParameterName=BlendMarkerColor, Value=0)
  ]}
 Update Delta Bar {nodes [
   Update Delta Bar() {is=title K2Node_FunctionEntry}
   > n1 = Branch(Condition=IsSegmentedShield() {is=title K2Node_CallFunction})
   > UpdateSegmentDelta()
   n2 = GetDynamicMaterial(execute=@n1.else, self=Get Image_DeltaBar())
   > SetScalarParameterValue(
    self=@n2
    ParameterName=Progress
    Value=GetLastValuePercentage!(self:Athena Player Hit Point Bar Base Object Reference) -> (ReturnValue:float)
   )
  ]}
 Update Fill Bar {nodes [
   Update Fill Bar() {is=title K2Node_FunctionEntry}
   > n1 = Branch(Condition=IsSegmentedShield())
   > UpdateSegmentFill()
   n2 = GetDynamicMaterial(execute=@n1.else, self=Get Image_FillBar())
   > SetScalarParameterValue(self=@n2, ParameterName=Progress, Value=GetCurrentValuePercentage())
   > SetHealthStyle() {is=title K2Node_CallFunction}
  ]}
 UpdateBlockedDamageBar {nodes [
   n1 = UpdateBlockedDamageBar!() -> (then, BlockedValueInterpolation:double=~) {is=title K2Node_FunctionEntry}
   > Branch(Condition="Equal (Enum)"(A=Get BarType() -> (BarType=Health), B=Health))
   > n2 = Branch(Condition="float < float"(A=@n1, B=1))
   n3 = Get BlockedValueInterpolation() -> (BlockedValueInterpolation:double)
   > SetBrushTintColor(
    self=Get Image_BlockedDelta()
    TintColor=Make Slate Color(SpecifiedColor="Lerp (LinearColor)"(A=(R=1,G=1,B=1,A=1), B=(R=0,G=0.93,B=0,A=1), Alpha=@n3))
   )
   n4 = GetDynamicMaterial(execute=@n2.else, self=Get Image_FillBar())
   > SetScalarParameterValue(self=@n4, ParameterName=Progress, Value=GetCurrentValuePercentage())
   > SetScalarParameterValue(self=@n4, ParameterName=BlendMarkerColor, Value=0.75)
   > SetVisibility(self=Get Image_BlockedDelta(), InVisibility=Hidden)
  ]}
 UpdateCurrentValue {nodes [
   UpdateCurrentValue() {is=title K2Node_FunctionEntry}
   > StopAllAnimations(self:User Widget Object Reference)
   n1 = Break Widget Transform!(WidgetTransform:"Widget Transform Structure (by ref)"=Get RenderTransform!(self:Widget Object Reference=Get SizeBox_Bar()) -> (RenderTransform:Widget Transform Structure)) -> (Translation:Vector 2D Structure, Scale:Vector 2D Structure, Shear:Vector 2D Structure, Angle:float=~) {is=title K2Node_BreakStruct}
   > SetRenderTransform!(
    execute
    self:Widget Object Reference=Get SizeBox_Bar()
    InTransform:Widget Transform Structure=Make Widget Transform!(Translation:Vector 2D Structure=(X=0,Y=0), Scale:Vector 2D Structure=@n1.Scale, Shear:Vector 2D Structure=@n1.Shear, Angle:float=@n1.Angle) -> (WidgetTransform:Widget Transform Structure) {is=title K2Node_MakeStruct}
   ) -> (then)
   > n2 = Sequence!(execute) -> (then_0, then_1, then_2:Exec, then_3:Exec)
   Update Fill Bar(execute=@n2.then_2) {is=title K2Node_CallFunction}
   > "SetText (Text)"(
    self=Get CurrentValue()
    InText="To Text (Float)"(
     Value=Get ValueCurrent() -> (ValueCurrent:float)
     MinimumIntegralDigits=1
     MaximumIntegralDigits=324
     MinimumFractionalDigits=0
     MaximumFractionalDigits=0
    )
   )
   > Branch(Condition="Equal (Enum)"(
     A:EFortHitPointModificationReason Enum=Get ValueChangedReason() -> (ValueChangedReason:EFortHitPointModificationReason Enum=Invalid)
     B:EFortHitPointModificationReason Enum=InitalSet
    ))
   > ForceLayoutPrepass(self:Widget Object Reference=Get SizeBox_Value())
   n3 = Get BarType() -> (BarType=Health)
   Branch(
    execute=@n2.then_3
    Condition=OR Boolean(A="Equal (Enum)"(A:EHealthBarType Enum=@n3, B:EHealthBarType Enum=Shield), B="Equal (Enum)"(A=@n3, B=Health))
   )
   n4 = Get ValueCurrent()
   > n5 = Branch(Condition="float > float"(A=Get ValueLast() -> (ValueLast:float), B=@n4) {is=title K2Node_CallFunction})
   > PlayAnimation(InAnimation=Get Anim_ValueImpact())
   > PlayAnimation(InAnimation=Get Anim_BarShake(), NumLoopsToPlay=3, bRestoreState=true)
   n6 = SafeDivide(A=Get ValueCurrent(), B=Get ValueMax() -> (ValueMax:float))
   > Branch(Condition="float < float"!(A:double=@n6, B:double=0.5) -> (ReturnValue:bool) {is=title K2Node_CallFunction})
   > PlayAnimation(InAnimation=Get Anim_BarShake_1() -> (Anim_BarShake_1:Widget Animation Object Reference), NumLoopsToPlay=5, bRestoreState=true)
   > Branch(Condition="float < float"(A=@n6, B=0.2))
   > PlayAnimation(InAnimation=Get Anim_BarShake_2() -> (Anim_BarShake_2:Widget Animation Object Reference), NumLoopsToPlay=5, bRestoreState=true)
   Branch(execute=@n5.else, Condition="float >= float"(A=@n4, B=Get ValueMax()) {is=title K2Node_CallFunction})
   n7 = MakeLiteralInt!(Value:int=1) -> (ReturnValue:int=0)
   > SetNumLoopsToPlay!(execute, self:User Widget Object Reference, InAnimation:Widget Animation Object Reference=Get Anim_BarShake(), NumLoopsToPlay:int=@n7) -> (then)
   > SetNumLoopsToPlay(InAnimation=Get Anim_BarShake_1(), NumLoopsToPlay=@n7)
   > SetNumLoopsToPlay(InAnimation=Get Anim_BarShake_2(), NumLoopsToPlay=@n7)
   Branch(execute=@n2.then_0, Condition="Equal (Enum)"(A=Get BarType() -> (BarType=Health), B=Health))
   > n8 = GetDynamicMaterial(self=Get Image_CastShadow() -> (Image_CastShadow:Image Object Reference))
   > SetScalarParameterValue(self=@n8, ParameterName=Severity, Value=float - float(B=GetCurrentValuePercentage()) {is=title K2Node_CallFunction})
   Branch(execute=@n2.then_1, Condition="Equal (Enum)"(A=Get BarType() -> (BarType=Health), B=Shield))
   n9 = NOT Boolean(A=Get bShouldDisplayShieldBar() -> (bShouldDisplayShieldBar:bool))
   n10 = AND Boolean(
    A="float <= float"!(A:double=Get ValueCurrent(), B:double=0) -> (ReturnValue:bool) {is=title K2Node_CallFunction}
    B=Get hideShieldOnEmpty() -> (hideShieldOnEmpty:bool)
   )
   > SetVisibility(InVisibility=Select(Option 0:ESlateVisibility Enum=HitTestInvisible, Option 1:ESlateVisibility Enum=Collapsed, Index=OR Boolean(A=@n10, B=@n9)) -> (ReturnValue:ESlateVisibility Enum))
  ]}
 UpdateDBNOState {nodes [
   n1 = UpdateDBNOState!() -> (then, bIsDBNO:bool=~) {is=title K2Node_FunctionEntry}
   > n2 = Branch(Condition="Equal (Boolean)"!(A:bool=@n1, B:bool=Get DBNOStateCache()) -> (ReturnValue:bool))
   n3 = Set DBNOStateCache(execute=@n2.else, DBNOStateCache:bool=@n1) -> (Output_Get:bool)
   > Switch on EHealthBarType!(execute, Selection:EHealthBarType Enum=Get BarType() -> (BarType=Health)) -> (Health:Exec, Shield:Exec, Overshield:Exec, Armor:Exec, Stamina:Exec, VehicleHealth:Exec, SignalInStorm:Exec) {is=title K2Node_SwitchEnum}
   > n4 = Branch(Condition=@n3)
   > n5 = GetDynamicMaterial(self=Get Image_DeltaBar())
   > SetScalarParameterValue(self=@n5, ParameterName=Progress, Value=0)
   > SetBrushFromTexture!(
    execute
    self:Image Object Reference=Get Image_ValueIcon()
    Texture:Texture 2D Object Reference=/Game/Athena/HUD/Art/squadArt/squad_Eliminated
    bMatchSize:bool=false
   ) -> (then)
   > SetHealthStyle()
   > n6 = GetDynamicMaterial(self=Get Image_BlockedDelta())
   > SetScalarParameterValue(self=@n6, ParameterName=Progress, Value=0)
   SetHealthStyle(execute=@n4.else)
   > RefreshIcons() {is=title K2Node_CallFunction}
  ]}
 UpdateLimit {nodes [
   n1 = UpdateLimit!() -> (then, ShowLimitMarker:bool=~, LimitPercentage:double=~) {is=title K2Node_FunctionEntry}
   > n2 = Branch(Condition=@n1)
   n3 = Get Overlay_Limit() -> (Overlay_Limit:Overlay Object Reference)
   n4 = BreakVector2D(InVec=GetLocalSize!(Geometry:"Geometry Structure (by ref)"=GetCachedGeometry!(self:Widget Object Reference=@n3) -> (ReturnValue:"Geometry Structure (by ref)")) -> (ReturnValue:Vector 2D Structure))
   n5 = Get LimitPercentage() -> (LimitPercentage:double)
   > SetRenderTranslation(self=Get LimitMark_Image() -> (LimitMark_Image:Image Object Reference), Translation=MakeVector2D(X=float * float(A=@n4, B=@n5), Y=0))
   n6 = Get Overlay_Limit()
   > n7 = SetVisibility(self=@n6, InVisibility=SelfHitTestInvisible)
   Return Node!(execute=[@n8.then, @n7.then])
   n8 = SetVisibility(execute=@n2.else, self=@n6, InVisibility=Collapsed)
   float + float(A=@n5, B=0.001) {is=title K2Node_CommutativeAssociativeBinaryOperator}
  ]}
 UpdateMaxValue {nodes [
   UpdateMaxValue() {is=title K2Node_FunctionEntry}
   > Update Fill Bar()
   > Update Delta Bar() {is=title K2Node_CallFunction}
  ]}
 UpdateSegmentBlocked {nodes [
   n1 = GetConsoleVariableFloatValue!(execute, VariableName:String=PlayerHitPointBar.MaxValueWithSegments) -> (then, ReturnValue:float)
   > n2 = GetDynamicMaterial(self=Get Image_BlockedDelta())
   n3 = Get ValueMax()
   n4 = Round(A=float * float(A=float / float(A=@n3, B=@n1), B=100) {is=title K2Node_PromotableOperator}) {is=title K2Node_CallFunction}
   > SetScalarParameterValue(self=@n2, ParameterName=Progress, Value="To Float (Integer)"(InInt=int / int(A=@n4, B=100)))
   > n5 = GetConsoleVariableFloatValue(VariableName=PlayerHitPointBar.TotalSegmentationCount)
   > SetScalarParameterValue(self=@n2, ParameterName=Segments, Value=@n5)
   UpdateSegmentBlocked() {is=title K2Node_FunctionEntry}
  ]}
 UpdateSegmentDelta {nodes [
   UpdateSegmentDelta() {is=title K2Node_FunctionEntry}
   > n1 = GetConsoleVariableFloatValue(VariableName=PlayerHitPointBar.MaxValueWithSegments)
   > n2 = GetDynamicMaterial(self=Get Image_DeltaBar())
   n3 = Get ValueMax()
   n4 = Round(A=float * float(A=float / float(A=@n3, B=@n1), B=100))
   > SetScalarParameterValue(self=@n2, ParameterName=Progress, Value="To Float (Integer)"(InInt=int / int(A=@n4, B=100)))
   > n5 = GetConsoleVariableFloatValue(VariableName=PlayerHitPointBar.TotalSegmentationCount)
   > SetScalarParameterValue(self=@n2, ParameterName=Segments, Value=@n5)
  ]}
 UpdateSegmentFill {nodes [
   UpdateSegmentFill() {is=title K2Node_FunctionEntry}
   > n1 = GetConsoleVariableFloatValue(VariableName=PlayerHitPointBar.MaxValueWithSegments)
   > n2 = GetDynamicMaterial(self=Get Image_FillBar())
   > SetScalarParameterValue(self=@n2, ParameterName=Progress, Value=float / float(A=Get ValueCurrent(), B=@n1))
   > n3 = GetConsoleVariableFloatValue(VariableName=PlayerHitPointBar.TotalSegmentationCount)
   > SetScalarParameterValue(self=@n2, ParameterName=Segments, Value=@n3)
  ]}
]
variables [
 Armor_Segments {type=MaterialInterface, default=/Game/Athena/HUD/Materials/M_UI_HitPointBar_Armor_Segments}
 bIcon_isVisible {type=boolean, default=true}
 bValue_isVisible {type=boolean, default=true}
 DBNOStateCache {type=boolean}
 GreenBarColor {type=LinearColor, default=<rgba 0 0.75 0 1>}
 HasReceivedDeltaUpdate {type=boolean}
 HasReceivedValueUpdate {type=boolean}
 hideShieldOnEmpty {type=boolean}
 MaxValueWithSegments {type=number, default=150}
 Shield_Full {type=MaterialInterface, default=/Game/Athena/HUD/Materials/M_UI_HitPointBar_Shield}
 Shield_Segments {type=MaterialInterface, default=/Game/Athena/HUD/Materials/M_UI_HitPointBar_Shield_Segments}
 ShieldDelta_Full {type=MaterialInterface, default=/Game/Athena/HUD/Materials/M_UI_Delta_Shield}
 ShieldDelta_Segments {type=MaterialInterface, default=/Game/Athena/HUD/Materials/M_UI_HitPointBar_SegmentsBacking}
 TextHudElementTags {type=GameplayTagContainer, default {gameplayTags []}}
 TotalSegmentCount {type=number, default=6}
 ValueChangedReason {type=EFortHitPointModificationReason, default=Invalid}
 WhiteBarColor {type=LinearColor, default=<rgba 0.3663 0.807 0.3372 1>}
]
widgets {
 info {
  inheritedWidgetCount=1
  namedSlotCount=1
  parentClass=/Script/FortniteUI.AthenaPlayerHitPointBarBase
  rootWidgetClass=&HorizontalBox
  widgetCount=20
 }
 widgets [
  HorizontalBox_66 {widgetClassPath=&HorizontalBox, widget=HorizontalBox_66}
  SizeBox_IconFixedSpacing {
   widgetClassPath=&SizeBox
   bIsVariable=true
   parent=HorizontalBox_66
   widget=SizeBox_IconFixedSpacing
   properties {bOverride_HeightOverride=true, bOverride_WidthOverride=true, heightOverride=24, widthOverride=24}
   slot {class=HorizontalBoxSlot, properties {horizontalAlignment=HAlign_Center, verticalAlignment=VAlign_Center, padding=<ltrb -12 -12 0 -12>}}
  }
  ScaleBox_Icon {
   widgetClassPath=&ScaleBox
   bIsVariable=true
   parent=SizeBox_IconFixedSpacing
   widget=ScaleBox_Icon
   properties {stretch=UserSpecified}
   slot {class=SizeBoxSlot}
  }
  Image_ValueIcon {
   widgetClassPath=&Image
   bIsVariable=true
   parent=ScaleBox_Icon
   widget=Image_ValueIcon
   properties {brush {
     bIsDynamicallyLoaded=false
     drawAs=Image
     imageType=NoImage
     mirroring=NoMirror
     resourceName=None
     resourceObject=&M_UI_HitPoint_Type_Health
     tiling=NoTile
     imageSize=<24 24>
     margin=<ltrb 0 0 0 0>
     outlineSettings {
      bUseBrushTransparency=false
      roundingType=HalfHeightRadius
      width=0
      color {colorUseRule=UseColor_Specified, specifiedColor=<rgba 0 0 0 0>}
      cornerRadii=<0 0 0 0>
     }
     tintColor {colorUseRule=UseColor_Specified, specifiedColor=<rgba 1 1 1 1>}
     uVRegion {bIsValid=false, max=<0 0>, min=<0 0>}
    }}
   slot {class=ScaleBoxSlot}
  }
  HorizontalBox_1 {
   widgetClassPath=&HorizontalBox
   parent=HorizontalBox_66
   widget=HorizontalBox_1
   slot {class=HorizontalBoxSlot, properties {size {sizeRule=Fill, value=1}}}
  }
  SizeBox_Bar {
   widgetClassPath=&SizeBox
   bIsVariable=true
   parent=HorizontalBox_1
   widget=SizeBox_Bar
   properties {heightOverride=24, widthOverride=450}
   slot {class=HorizontalBoxSlot, properties {size {sizeRule=Fill, value=1}}}
  }
  "[Overlay]" {
   widgetClassPath=&Overlay
   parent=SizeBox_Bar
   widget="[Overlay]"
   properties {visibility=HitTestInvisible}
   slot {class=SizeBoxSlot}
  }
  Overlay_Limit {
   widgetClassPath=&Overlay
   bIsVariable=true
   parent="[Overlay]"
   widget=Overlay_Limit
   properties {visibility=Collapsed, renderTransform {angle=0, scale=<1 1>, shear=<0 0>, translation=<0 1>}}
   slot {class=OverlaySlot, properties {horizontalAlignment=HAlign_Fill, verticalAlignment=VAlign_Fill}}
  }
  LimitMark_Image {
   widgetClassPath=&Image
   bIsVariable=true
   parent=Overlay_Limit
   widget=LimitMark_Image
   properties {
    brush {
     bIsDynamicallyLoaded=false
     drawAs=Image
     imageType=NoImage
     mirroring=NoMirror
     resourceName=None
     resourceObject=null
     tiling=NoTile
     imageSize=<2 32>
     margin=<ltrb 0 0 0 0>
     outlineSettings {
      bUseBrushTransparency=false
      roundingType=HalfHeightRadius
      width=0
      color {colorUseRule=UseColor_Specified, specifiedColor=<rgba 0 0 0 0>}
      cornerRadii=<0 0 0 0>
     }
     tintColor {colorUseRule=UseColor_Specified, specifiedColor=<rgba 1 1 1 1>}
     uVRegion {bIsValid=false, max=<0 0>, min=<0 0>}
    }
    renderTransform {angle=0, scale=<1 1.2>, shear=<0 0>, translation=<0 0>}
   }
   slot {class=OverlaySlot, properties {verticalAlignment=VAlign_Center}}
  }
  HorizontalBox_2 {
   widgetClassPath=&HorizontalBox
   parent="[Overlay]"
   widget=HorizontalBox_2
   properties {visibility=HitTestInvisible}
   slot {class=OverlaySlot, properties {horizontalAlignment=HAlign_Fill, verticalAlignment=VAlign_Fill}}
  }
  Image_CastShadow {
   widgetClassPath=&Image
   bIsVariable=true
   parent=HorizontalBox_2
   widget=Image_CastShadow
   properties {
    visibility=SelfHitTestInvisible
    brush {
     bIsDynamicallyLoaded=false
     drawAs=Image
     imageType=NoImage
     mirroring=NoMirror
     resourceName=None
     resourceObject=/Game/Athena/HUD/Team/M_UI_BarBG_Pulse_Inst
     tiling=NoTile
     imageSize=<32 32>
     margin=<ltrb 0 0 0 0>
     outlineSettings {
      bUseBrushTransparency=false
      roundingType=HalfHeightRadius
      width=0
      color {colorUseRule=UseColor_Specified, specifiedColor=<rgba 0 0 0 0>}
      cornerRadii=<0 0 0 0>
     }
     tintColor {colorUseRule=UseColor_Specified, specifiedColor=<rgba 1 1 1 1>}
     uVRegion {bIsValid=false, max=<0 0>, min=<0 0>}
    }
   }
   slot {class=HorizontalBoxSlot, properties {size {sizeRule=Fill, value=1}}}
  }
  Image_DeltaBar {
   widgetClassPath=&Image
   bIsVariable=true
   parent="[Overlay]"
   widget=Image_DeltaBar
   properties {brush {
     bIsDynamicallyLoaded=false
     drawAs=Image
     imageType=NoImage
     mirroring=NoMirror
     resourceName=None
     resourceObject=&M_UI_HitPointBar_Health
     tiling=NoTile
     imageSize=<32 32>
     margin=<ltrb 0 0 0 0>
     outlineSettings {
      bUseBrushTransparency=false
      roundingType=HalfHeightRadius
      width=0
      color {colorUseRule=UseColor_Specified, specifiedColor=<rgba 0 0 0 0>}
      cornerRadii=<0 0 0 0>
     }
     tintColor {colorUseRule=UseColor_Specified, specifiedColor=<rgba 1 1 1 1>}
     uVRegion {bIsValid=false, max=<0 0>, min=<0 0>}
    }}
   slot {class=OverlaySlot, properties {horizontalAlignment=HAlign_Fill, verticalAlignment=VAlign_Fill}}
  }
  Image_BlockedDelta {
   widgetClassPath=&Image
   bIsVariable=true
   parent="[Overlay]"
   widget=Image_BlockedDelta
   properties {
    visibility=Hidden
    brush {
     bIsDynamicallyLoaded=false
     drawAs=Image
     imageType=FullColor
     mirroring=NoMirror
     resourceName=None
     resourceObject=/Game/Athena/HUD/Materials/MI_UI_HitPointBar_Blocked
     tiling=NoTile
     imageSize=<32 32>
     margin=<ltrb 0 0 0 0>
     outlineSettings {
      bUseBrushTransparency=false
      roundingType=HalfHeightRadius
      width=0
      color {colorUseRule=UseColor_Specified, specifiedColor=<rgba 0 0 0 0>}
      cornerRadii=<0 0 0 0>
     }
     tintColor {colorUseRule=UseColor_Specified, specifiedColor=<rgba 1 1 1 1>}
     uVRegion {bIsValid=false, max=<0 0>, min=<0 0>}
    }
   }
   slot {class=OverlaySlot, properties {horizontalAlignment=HAlign_Fill, verticalAlignment=VAlign_Fill}}
  }
  Image_FillBar {
   widgetClassPath=&Image
   bIsVariable=true
   parent="[Overlay]"
   widget=Image_FillBar
   properties {brush {
     bIsDynamicallyLoaded=false
     drawAs=Image
     imageType=NoImage
     mirroring=NoMirror
     resourceName=None
     resourceObject=&M_UI_HitPointBar_Health
     tiling=NoTile
     imageSize=<32 32>
     margin=<ltrb 0 0 0 0>
     outlineSettings {
      bUseBrushTransparency=false
      roundingType=HalfHeightRadius
      width=0
      color {colorUseRule=UseColor_Specified, specifiedColor=<rgba 0 0 0 0>}
      cornerRadii=<0 0 0 0>
     }
     tintColor {colorUseRule=UseColor_Specified, specifiedColor=<rgba 1 1 1 1>}
     uVRegion {bIsValid=false, max=<0 0>, min=<0 0>}
    }}
   slot {class=OverlaySlot, properties {horizontalAlignment=HAlign_Fill, verticalAlignment=VAlign_Fill}}
  }
  Overlay_ExtensionWidget {
   widgetClassPath=&Overlay
   bIsVariable=true
   parent="[Overlay]"
   widget=Overlay_ExtensionWidget
   slot {class=OverlaySlot, properties {horizontalAlignment=HAlign_Fill, verticalAlignment=VAlign_Fill}}
  }
  SizeBox_Value {
   widgetClassPath=&SizeBox
   bIsVariable=true
   parent=HorizontalBox_1
   widget=SizeBox_Value
   properties {bOverride_WidthOverride=true, widthOverride=200, renderTransform {angle=0, scale=<1 1>, shear=<0 0>, translation=<4 0>}}
   slot {class=HorizontalBoxSlot, properties {horizontalAlignment=HAlign_Left, verticalAlignment=VAlign_Center, padding=<ltrb 0 0 -200 0>}}
  }
  HBox_TextArea {
   widgetClassPath=&HorizontalBox
   bIsVariable=true
   parent=SizeBox_Value
   widget=HBox_TextArea
   slot {class=SizeBoxSlot, properties {verticalAlignment=VAlign_Center, padding=<ltrb 4 0 0 0>}}
  }
  FortHUDElementWrapper_Text {
   widgetClassPath=&/FortHUDElementWrapper
   bIsVariable=true
   parent=HBox_TextArea
   widget=FortHUDElementWrapper_Text
   properties {elementTag {tagName=HUD.Athena.Player.HitPointBarText}}
   slot {class=HorizontalBoxSlot}
  }
  CurrentValue {
   widgetClassPath=&CommonTextBlock
   bIsVariable=true
   namedSlotHost=FortHUDElementWrapper_Text
   widget=CurrentValue
   properties {
    justification=Right
    lineHeightPercentage=0.8
    text=999?
    font {
     bForceMonospaced=false
     bMaterialIsStencil=false
     fontMaterial=null
     fontObject=&BurbankSmall
     letterSpacing=0
     monospacedWidth=1
     size=15
     skewAmount=0
     typefaceFontName=Black
     outlineSettings {
      bApplyOutlineToDropShadows=true
      bMiteredCorners=false
      bSeparateFillAlpha=true
      outlineMaterial=&M-FontShadow
      outlineSize=2
      outlineColor=<rgba 0 0 0 0.5>
     }
    }
    margin=<ltrb 0 -6 0 1>
    renderTransformPivot=<0 0.75>
    shadowOffset=<0 0>
   }
  }
  TutorialHighlight {widgetClassPath=&TutorialHighlight}
 ]
}
