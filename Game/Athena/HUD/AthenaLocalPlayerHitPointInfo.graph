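# Local player HUD widget showing health, shield and vehicle hit point bars, plus a ranked icon looked up from Habanero display data.
# Construct binds overshield changes to vehicle bar padding and Creative quickbar equips to widget collapse, then sets the rank icon.
# With Fade on Full, pawn health changes fade out full bars and quickly reverse the animation otherwise.
# PreConstruct optionally enlarges the bars; the vehicle fade logic has no incoming execution.
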WidgetBlueprint /Game/Athena/HUD/AthenaLocalPlayerHitPointInfo
animations [
 FadeStatus {
  binding_count=1
  length_seconds=4
  object=&self:FadeStatus
  sequence {bindings [VerticalBox_23 {
     id=#1
     possessed_class=VerticalBox
     tracks [{
       type=MovieSceneFloatTrack
       display_name=Render Opacity
       sections [{type=MovieSceneFloatSection, channels [None {type=MovieSceneScriptingFloatChannel, keys [[0,1],[40,1],[80,0]]}]}]
      }]
    }]}
 }
 FadeVehicle {
  binding_count=1
  length_seconds=4
  object=&self:FadeVehicle
  sequence {bindings [VerticalBox_PlayerStatus {
     id=#2
     display_name=SizeBox_2
     possessed_class=VerticalBox
     tracks [{
       type=MovieSceneFloatTrack
       display_name=Render Opacity
       sections [{type=MovieSceneFloatSection, channels [None {type=MovieSceneScriptingFloatChannel, keys [[0,1],[40,1],[80,0]]}]}]
      }]
    }]}
 }
]
graphs [
 EventGraph {nodes [
   n1 = OnPawnHealthChanged_Event() {is=title K2Node_CustomEvent}
   n2 = GetPawnFromPlayerStateZone!(
    execute=@n18.then
    InActor:Actor Object Reference=GetPlayerStateZone!(self:Fort HUDContext Object Reference=Get FortHUDContext!() -> (ReturnValue:Fort HUDContext Object Reference) {is=title K2Node_GetBlueprintContextEx}) -> (ReturnValue:Fort Player State Zone Object Reference)
   ) -> (then, OutPawn:Fort Player Pawn Object Reference, PawnReturned:bool)
   n3 = GetMaxHealth!(self:Fort Pawn Object Reference=@n2) -> (ReturnValue:float)
   n4 = GetHealth!(self:Fort Pawn Object Reference=@n2) -> (ReturnValue:float)
   n5 = GetShield!(self:Fort Pawn Object Reference=@n2) -> (ReturnValue:float)
   n6 = GetMaxShield!(self:Fort Pawn Object Reference=@n2) -> (ReturnValue:float)
   n7 = Get isFaded() -> (isFaded:bool)
   n8 = Branch(
    execute=@n1.then
    Condition=AND Boolean!(
     A="Equal (Float)"!(A:double=@n4, B:double=@n3) -> (ReturnValue:bool) {is=title K2Node_CallFunction}
     B="Equal (Float)"(A=@n5, B=@n6)
     C:bool=NOT Boolean(A=@n7)
    ) -> (ReturnValue)
   )
   n9 = Get FadeStatus() -> (FadeStatus:Widget Animation Object Reference)
   > StopAnimation(InAnimation=@n9)
   > PlayAnimation(InAnimation=@n9)
   > n10 = Set isFaded(isFaded:bool=true) -> (Output_Get:bool)
   Reroute Node(InputPin=[@n10.then, @n11.then])
   StopAnimation(execute=@n8.else, InAnimation=@n9)
   > PlayAnimationReverse(InAnimation=@n9, PlaybackSpeed=3)
   > n11 = Set isFaded(isFaded=false)
   n12 = UpdateOvershieldBarVisibility!() -> (OutputDelegate, then, bNewVisible:bool=~) {is=title K2Node_CustomEvent}
   > OnOvershieldVisibilityChanged()
   Event Construct()
   > n13 = Sequence!(execute) -> (then_0, then_1, then_2:Exec)
   > n14 = IsTacticalOvershieldEnabled!(execute) -> (then, ReturnValue:bool)
   > n15 = Branch(Condition=@n14)
   n16 = GetPlayerStateZone(self=Get FortHUDContext())
   > n17 = Is Valid(InputObject=@n16)
   FortShippingLog!(
    execute=@n17.Is Not Valid
    WorldContextObject:Object Reference=Self-Reference()
    InString:String=Overshield Test: Event Graph. Get Player State Zone failed
    bLogAsWarning:bool=false
   ) -> (then)
   Bind Event to On Overshield Visibility Changed(execute=@n17.Is Valid, self:Fort Player State Zone Object Reference=@n16, Delegate=@n12) {is=title K2Node_AddDelegate}
   > OnOvershieldVisibilityChanged()
   FortShippingLog(execute=@n15.else, WorldContextObject=Self-Reference(), InString=Overshield Test: Overshield Disabled, bLogAsWarning=false)
   Bind Event to On Creative Quickbar Equipped Changed(
    execute=@n13.then_1
    self:Creative HUDLContext Object Reference=Get CreativeHUDLContext!() -> (ReturnValue:Creative HUDLContext Object Reference) {is=title K2Node_GetBlueprintContextEx}
    Delegate=Create Event()
   ) {is=title K2Node_AddDelegate}
   > n18 = Branch(Condition=Get Fade on Full() -> (Fade on Full:bool))
   SetRankIcon(execute=@n13.then_2)
   Event PreConstruct()
   > Branch(Condition=Get Use Large Format() -> (Use Large Format:bool))
   > SetSize!(
    execute
    self:Athena Hit Point Bar Object Reference=[
     Get HitPointBarVehicleHealth() -> (HitPointBarVehicleHealth:Athena Hit Point Bar Object Reference)
     Get HitPointBarHealth() -> (HitPointBarHealth:Athena Hit Point Bar Object Reference)
     Get HitPointBarShield() -> (HitPointBarShield:Athena Hit Point Bar Object Reference)
    ]
    useLargeSize:bool=true
   ) -> (then)
   n19 = GetHealth!(execute:Exec, self:Object Reference=GetVehicleActor!(self:Fort Player Pawn Object Reference=@n2) -> (ReturnValue:Actor Object Reference)) -> (then:Exec, ReturnValue:float) {is=title K2Node_Message}
   > Branch(Condition=IsVisible(self=Get HitPointBarVehicleHealth()))
   > n20 = Branch(Condition="Equal (Float)"(A=@n19, B=1))
   n21 = Get FadeVehicle() -> (FadeVehicle:Widget Animation Object Reference)
   > PlayAnimation(InAnimation=@n21)
   PlayAnimationReverse(execute=@n20.else, InAnimation=@n21)
   Assign On Pawn Health Changed!(execute:Exec=@n2.then, self:Fort Pawn Object Reference=@n2, Delegate:"Delegate (by ref)"=@n1) -> (then:Exec) {is=title K2Node_AssignDelegate}
  ]}
 OnCreativeQuickbarEquippedChanged {nodes [
   n1 = OnCreativeQuickbarEquippedChanged!() -> (then, bIsEquipped:bool=~) {is=title K2Node_FunctionEntry}
   > n2 = Branch(Condition=@n1)
   > SetVisibility(InVisibility=Collapsed)
   SetVisibility(execute=@n2.else, InVisibility=SelfHitTestInvisible)
  ]}
 OnOvershieldVisibilityChanged {nodes [
   OnOvershieldVisibilityChanged() {is=title K2Node_FunctionEntry}
   n1 = GetPlayerStateZone(self=Get FortHUDContext())
   > n2 = Is Valid(InputObject=@n1)
   FortShippingLog(
    execute=@n2.Is Not Valid
    WorldContextObject=Self-Reference()
    InString=Overshield Test: OnOvershieldVisibilityChanged. Get Player State Zone failed
    bLogAsWarning=false
   )
   n3 = GetOwningPlayer()
   n4 = AND Boolean!(
    A=NOT Boolean(A=IsUsingTouch(PlayerController=@n3))
    B=ShouldOvershieldBarBeVisible!(self:Fort Player State Zone Object Reference=@n1) -> (ReturnValue:bool)
   ) -> (ReturnValue)
   SetPadding!(
    execute=@n2.Is Valid
    self:Size Box Slot Object Reference=SlotAsSizeBoxSlot!(Widget:Widget Object Reference=Get HitPointBarVehicleHealth()) -> (ReturnValue:Size Box Slot Object Reference)
    InPadding:Margin Structure=Make Margin(Left=0, Top=0, Right=0, Bottom=SelectFloat(A=15, B=0, bPickA=@n4))
   ) -> (then)
  ]}
 SetRankIcon {nodes [
   SetRankIcon() {is=title K2Node_FunctionEntry}
   n1 = Break Ranked Progress Replicated Data!(RankedProgressReplicatedData:"Ranked Progress Replicated Data Structure (by ref)"=GetRankedProgressFromHabanero!(self:Athena HUDContext Object Reference=Get AthenaHUDContext!() -> (ReturnValue:Athena HUDContext Object Reference) {is=title K2Node_GetBlueprintContextEx}) -> (ReturnValue:Ranked Progress Replicated Data Structure)) -> (RankType:String, Rank:int, HighestRank:int) {is=title K2Node_BreakStruct}
   n2 = Find(
    TargetMap:Map of Strings to Fort Habanero Mode List Structures=Get EntryData!(self:Fort Habanero Display Data Object Reference=Get RankInfo() -> (RankInfo:Fort Habanero Display Data Object Reference)) -> (EntryData:Map of Strings to Fort Habanero Mode List Structures)
    Key=@n1
   ) -> (Value:Fort Habanero Mode List Structure)
   n3 = Break Fort Habanero Mode List!(FortHabaneroModeList:"Fort Habanero Mode List Structure (by ref)"=@n2) -> (TierList:Array of Fort Habanero Tier Structures) {is=title K2Node_BreakStruct}
   n4 = AND Boolean!(
    A=GetShowRankedIcon!() -> (ReturnValue:bool)
    B=@n2.ReturnValue
    C:bool=Is Valid Index!(TargetArray:Array of Fort Habanero Tier Structures=@n3, IndexToTest:int=@n1.Rank) -> (ReturnValue:bool)
   ) -> (ReturnValue)
   > SetVisibility(
    self=Get SizeBoxRank() -> (SizeBoxRank:Size Box Object Reference)
    InVisibility=Select(Option 0=Collapsed, Option 1=SelfHitTestInvisible, Index=@n4)
   )
   > Branch(Condition=@n4)
   > SetBrushFromSoftTexture!(
    execute
    self:Image Object Reference=Get RankIconImg() -> (RankIconImg:Image Object Reference)
    SoftTexture:Texture 2D Soft Object Reference=Break Fort Habanero Tier!(FortHabaneroTier:"Fort Habanero Tier Structure (by ref)"="Get (a copy)"(Array:Array of Fort Habanero Tier Structures=@n3, Dimension 1=@n1.Rank) -> (Output:Fort Habanero Tier Structure)) -> (
     SimpleIconSmall:Texture 2D Soft Object Reference
     AdditionalLeftIcon:Material Instance Soft Object Reference
     AdditionalRightIcon:Material Instance Soft Object Reference
    ) {is=title K2Node_BreakStruct}
    bMatchSize:bool=true
   ) -> (then)
  ]}
]
variables [
 Fade on Full {type=boolean}
 inVehicle {type=boolean}
 isFaded {type=boolean}
 RankInfo {type=FortHabaneroDisplayData, default=/Game/UI/Habanero/HabaneroDisplayData}
 Use Large Format {type=boolean}
 VehicleHealthBarHeightOverride {type=number, default=5}
 VehicleHealthBarWidthOverride {type=number, default=500}
]
widgets {
 info {
  inheritedWidgetCount=1
  namedSlotCount=3
  parentClass=/Script/FortniteUI.AthenaPlayerHitPointInfo
  rootWidgetClass=&SizeBox
  widgetCount=23
 }
 widgets [
  SizeBox_Bars {widgetClassPath=&SizeBox, bIsVariable=true, widget=SizeBox_Bars, properties {bOverride_WidthOverride=true, widthOverride=345}}
  Overlay_0 {widgetClassPath=&Overlay, parent=SizeBox_Bars, widget=Overlay_0, slot {class=SizeBoxSlot}}
  Overlay_1 {
   widgetClassPath=&Overlay
   parent=Overlay_0
   widget=Overlay_1
   slot {class=OverlaySlot, properties {horizontalAlignment=HAlign_Fill, verticalAlignment=VAlign_Bottom}}
  }
  AthenaCurieStatusWidget {
   widgetClassPath=/Game/Athena/HUD/Curie/AthenaCurieStatusWidget
   bIsVariable=true
   parent=Overlay_1
   widget=AthenaCurieStatusWidget
   properties {
    elementTag {tagName=HUD.Athena.Player.Health}
    renderTransform {angle=0, scale=<1 1>, shear=<0 0>, translation=<12 -10>}
    renderTransformPivot=<0 0.8>
   }
   slot {class=OverlaySlot, properties {horizontalAlignment=HAlign_Fill, verticalAlignment=VAlign_Fill, padding=<ltrb -70 -100 -60 -30>}}
  }
  Overlay_LocalPlayerExtensions {
   widgetClassPath=/Game/UI/ExtensionWidgets/UnderLocalPlayerInfo
   bIsVariable=true
   parent=Overlay_1
   widget=Overlay_LocalPlayerExtensions
   slot {class=OverlaySlot, properties {horizontalAlignment=HAlign_Center, verticalAlignment=VAlign_Center}}
  }
  VerticalBox_23 {
   widgetClassPath=&VerticalBox
   parent=Overlay_1
   widget=VerticalBox_23
   slot {class=OverlaySlot, properties {horizontalAlignment=HAlign_Fill, verticalAlignment=VAlign_Fill}}
  }
  FortHUDElementWrapper_Vehicle {
   widgetClassPath=&/FortHUDElementWrapper
   bIsVariable=true
   parent=VerticalBox_23
   widget=FortHUDElementWrapper_Vehicle
   properties {
    elementTag {tagName=HUD.Athena.Player.Vehicle}
    hUDElementTag {gameplayTags [{tagName=Athena.HideForCapture}, {tagName=Athena.Creative.HUDController.HideHealth}]}
   }
   slot {class=VerticalBoxSlot, properties {verticalAlignment=VAlign_Top, padding=<ltrb 0 0 0 6>}}
  }
  SizeBox_2 {
   widgetClassPath=&SizeBox
   bIsVariable=true
   namedSlotHost=FortHUDElementWrapper_Vehicle
   widget=SizeBox_2
   properties {bOverride_HeightOverride=true, heightOverride=32}
  }
  HitPointBarVehicleHealth {
   widgetClassPath=&/AthenaHitPointBar
   bIsVariable=true
   parent=SizeBox_2
   widget=HitPointBarVehicleHealth
   properties {
    barType=VehicleHealth
    flowDirectionPreference=LeftToRight
    visibility=Collapsed
    elementTag {tagName=HUD.BottomLeft}
    hUDElementTag {gameplayTags [{tagName=Athena.HideForCapture}, {tagName=Athena.Creative.HUDController.HideVehicleDamageStatsUI}]}
   }
   slot {class=SizeBoxSlot}
  }
  FortHUDElementWrapper_Shield {
   widgetClassPath=&FortHUDElementWrapper
   bIsVariable=true
   parent=VerticalBox_23
   widget=FortHUDElementWrapper_Shield
   properties {
    elementTag {tagName=HUD.Athena.Player.Shield}
    hUDElementTag {gameplayTags [{tagName=Athena.HideForCapture}, {tagName=Athena.Creative.HUDController.HideShields}]}
   }
   slot {class=VerticalBoxSlot, properties {padding=<ltrb 0 0 0 6>}}
  }
  SizeBox_1 {
   widgetClassPath=&SizeBox
   namedSlotHost=FortHUDElementWrapper_Shield
   widget=SizeBox_1
   properties {bOverride_HeightOverride=true, heightOverride=16}
  }
  HitPointBarShield {
   widgetClassPath=&AthenaHitPointBar
   bIsVariable=true
   parent=SizeBox_1
   widget=HitPointBarShield
   properties {
    barType=Shield
    flowDirectionPreference=LeftToRight
    visibility=SelfHitTestInvisible
    elementTag {tagName=HUD.BottomLeft}
    textHudElementTags {gameplayTags [{tagName=Athena.Creative.HUDController.HideShieldNumbers}]}
   }
   slot {class=SizeBoxSlot}
  }
  FortHUDElementWrapper_Health {
   widgetClassPath=&FortHUDElementWrapper
   bIsVariable=true
   parent=VerticalBox_23
   widget=FortHUDElementWrapper_Health
   properties {
    elementTag {tagName=HUD.Athena.Player.Health}
    hUDElementTag {gameplayTags [{tagName=Athena.HideForCapture}, {tagName=Athena.Creative.HUDController.HideHealth}]}
   }
   slot {class=VerticalBoxSlot, properties {verticalAlignment=VAlign_Top}}
  }
  SizeBox_0 {
   widgetClassPath=&SizeBox
   namedSlotHost=FortHUDElementWrapper_Health
   widget=SizeBox_0
   properties {bOverride_HeightOverride=true, heightOverride=32}
  }
  HitPointBarHealth {
   widgetClassPath=&AthenaHitPointBar
   bIsVariable=true
   parent=SizeBox_0
   widget=HitPointBarHealth
   properties {
    bCanAddExtensionWidgets=true
    extensionSlotToUse=PlayerHealthbarOverlay
    flowDirectionPreference=LeftToRight
    valueCurrent=0.9857
    visibility=SelfHitTestInvisible
    elementTag {tagName=HUD.BottomLeft}
    textHudElementTags {gameplayTags [{tagName=Athena.Creative.HUDController.HideHealthNumbers}]}
   }
   slot {class=SizeBoxSlot}
  }
  SizeBox {
   widgetClassPath=&SizeBox
   parent=VerticalBox_23
   widget=SizeBox
   properties {bOverride_HeightOverride=true, heightOverride=22}
   slot {class=VerticalBoxSlot, properties {verticalAlignment=VAlign_Top, padding=<ltrb 0 3 0 0>}}
  }
  AthenaHitPointBar_CNStorm2 {
   widgetClassPath=/Game/Athena/UI/GAPP/AthenaHitPointBar_CNStorm2
   bIsVariable=true
   parent=SizeBox
   widget=AthenaHitPointBar_CNStorm2
   properties {flowDirectionPreference=LeftToRight, visibility=Collapsed, elementTag {tagName=HUD.BottomLeft}}
   slot {class=SizeBoxSlot}
  }
  CanvasPanel_0 {
   widgetClassPath=&CanvasPanel
   parent=Overlay_0
   widget=CanvasPanel_0
   slot {class=OverlaySlot, properties {horizontalAlignment=HAlign_Fill, verticalAlignment=VAlign_Fill}}
  }
  Overlay_2 {
   widgetClassPath=&Overlay
   parent=CanvasPanel_0
   widget=Overlay_2
   slot {class=CanvasPanelSlot, properties {layoutData {alignment=<0 0>, anchors {maximum=<1 1>, minimum=<0 0>}, offsets=<ltrb 0 0 0 0>}}}
  }
  TutorialHighlight {
   widgetClassPath=&TutorialHighlight
   bIsVariable=true
   parent=Overlay_2
   widget=TutorialHighlight
   properties {customHighlightWidget=/Game/Athena/HUD/Tutorial/Tutorial_Outline_Square_Static.Tutorial_Outline_Square_Static_C}
   slot {class=OverlaySlot, properties {horizontalAlignment=HAlign_Fill, verticalAlignment=VAlign_Fill}}
  }
  SizeBox_HealthbarOverlay {
   widgetClassPath=&SizeBox
   parent=Overlay_0
   widget=SizeBox_HealthbarOverlay
   properties {bOverride_HeightOverride=true, bOverride_WidthOverride=true, heightOverride=36}
   slot {class=OverlaySlot, properties {horizontalAlignment=HAlign_Fill, verticalAlignment=VAlign_Bottom, padding=<ltrb -20 0 0 -2>}}
  }
  SizeBoxRank {
   widgetClassPath=&SizeBox
   bIsVariable=true
   parent=Overlay_0
   widget=SizeBoxRank
   properties {bOverride_HeightOverride=true, bOverride_WidthOverride=true, heightOverride=32, widthOverride=32}
   slot {class=OverlaySlot, properties {verticalAlignment=VAlign_Bottom, padding=<ltrb -60 0 0 0>}}
  }
  RankIconImg {
   widgetClassPath=&Image
   bIsVariable=true
   parent=SizeBoxRank
   widget=RankIconImg
   properties {
    visibility=SelfHitTestInvisible
    brush {
     bIsDynamicallyLoaded=false
     drawAs=Image
     imageType=FullColor
     mirroring=NoMirror
     resourceName=None
     resourceObject=/Game/UI/Frontend/Lobby/Habanero/Habanero_Materials/MaterialInstances/Icons/SmallColor/T_UI_Habanero_SmallColor_T03_05
     tiling=NoTile
     imageSize=<64 64>
     margin=<ltrb 0 0 0 0>
     outlineSettings {
      bUseBrushTransparency=false
      roundingType=HalfHeightRadius
      width=0
      color {colorUseRule=UseColor_Specified, specifiedColor=<rgba 0 0 0 0>}
      cornerRadii=<0 0 0 0>
     }
     tintColor {colorUseRule=UseColor_Specified, specifiedColor=<rgba 1 1 1 1>}
     uVRegion {bIsValid=false, max=<0 0>, min=<0 0>}
    }
   }
   slot {class=SizeBoxSlot}
  }
 ]
}
