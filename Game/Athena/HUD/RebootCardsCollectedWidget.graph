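# HUD widget showing a reboot card icon and the count of cards collected, built on FortRebootCardsCollectedWidget.
# On construct it hides the separator on mobile and, once game state is ready, caches whether resurrection is enabled.
# If that and the runtime inventory option allow, count changes play an intro from zero or a spin otherwise, then update text and visibility.
# It collapses at zero by default, but the mobile layout editor forces it visible.
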WidgetBlueprint /Game/Athena/HUD/RebootCardsCollectedWidget
animations [
 IntroAnim {
  binding_count=1
  length_seconds=0.4
  object=&self:IntroAnim
  sequence {bindings [ImageRebootIcon {
     id=#1
     possessed_class=Image
     tracks [
      {
       type=MovieScene2DTransformTrack
       sections [{
         type=MovieScene2DTransformSection
         end_frame=8
         start_frame=0
         channels [
          Translation.X {type=MovieSceneScriptingFloatChannel}
          Translation.Y {type=MovieSceneScriptingFloatChannel}
          Angle {type=MovieSceneScriptingFloatChannel, keys [[0,-69],[6,0]]}
          Scale.X {type=MovieSceneScriptingFloatChannel, keys [[0,2],[6,0.9],[8,1]]}
          Scale.Y {type=MovieSceneScriptingFloatChannel, keys [[0,2],[6,0.9],[8,1]]}
          Shear.X {type=MovieSceneScriptingFloatChannel}
          Shear.Y {type=MovieSceneScriptingFloatChannel}
         ]
        }]
      }
      {
       type=MovieSceneFloatTrack
       display_name=Render Opacity
       sections [{type=MovieSceneFloatSection, end_frame=8, start_frame=0, channels [None {type=MovieSceneScriptingFloatChannel, keys [[0,0],[2,1],[8,1]]}]}]
      }
     ]
    }]}
 }
 Updated {
  binding_count=1
  length_seconds=0.9
  object=&self:Updated
  sequence {bindings [ImageRebootIcon {
     id=#2
     possessed_class=Image
     tracks [{
       type=MovieScene2DTransformTrack
       sections [{
         type=MovieScene2DTransformSection
         end_frame=18
         start_frame=0
         channels [
          Translation.X {type=MovieSceneScriptingFloatChannel}
          Translation.Y {type=MovieSceneScriptingFloatChannel}
          Angle {type=MovieSceneScriptingFloatChannel, keys [[0,0],[2,0],[12,360],[18,360]]}
          Scale.X {type=MovieSceneScriptingFloatChannel, keys [[0,1],[2,0.9],[5,1.204],[12,1.15],[15,0.848],[18,1]]}
          Scale.Y {type=MovieSceneScriptingFloatChannel, keys [[0,1],[2,0.9],[5,1.204],[12,1.15],[15,0.848],[18,1]]}
          Shear.X {type=MovieSceneScriptingFloatChannel}
          Shear.Y {type=MovieSceneScriptingFloatChannel}
         ]
        }]
      }]
    }]}
 }
]
graphs [
 CanShowRebootCards {nodes [
   CanShowRebootCards() {is=title K2Node_FunctionEntry}
   n1 = GetRuntimeOptions()
   > Return Node!(
    execute
    CanShow:bool=AND Boolean(
     A=Get ResurectionEnabled() -> (ResurectionEnabled:bool)
     B=GetRebootShowInInventory!(self:Fort Runtime Options Object Reference=@n1) -> (ReturnValue:bool)
    )
   )
  ]}
 EventGraph {nodes [
   Event Construct()
   > n1 = GameStateReadyAsync!(execute:Exec) -> (then:Exec, OnReady:Exec, GameState:Game State Base Object Reference) {is=title K2Node_AsyncAction}
   > Branch(Condition=IsMobileGame())
   > SetVisibility(self="Get [Image] Separator"() -> ("[Image] Separator":Image Object Reference), InVisibility=Collapsed)
   Is Valid(exec=@n1.OnReady, InputObject=@n1)
   n2 = GetComponentByClass!(self:Actor Object Reference=@n1, ComponentClass:Actor Component Class Reference=/Script/FortniteGame.FortGameStateComponent_Reboot) -> (ReturnValue:Fort Game State Component Reboot Object Reference)
   > Is Valid(InputObject=@n2)
   > n3 = IsResurrectionEnabled!(execute, self:Fort Game State Component Reboot Object Reference=@n2, PlayerPawn:Fort Player Pawn Object Reference) -> (then, ReturnValue:bool)
   > Set ResurectionEnabled(ResurectionEnabled:bool=@n3) -> (Output_Get:bool)
   > Branch(Condition=CanShowRebootCards!() -> (CanShow:bool) {is=title K2Node_CallFunction})
   > UpdateValue!(execute, self, NewValue:int=Get NumCardsCollected() -> (NumCardsCollected:int=0)) -> (then)
   > UpdateVisibility()
   Event OnRebootCardsCollectedChanged() {is=title K2Node_Event}
   > Branch(Condition=CanShowRebootCards())
   n4 = "integer > integer"(A=Get NumCardsCollected() -> (NumCardsCollected=0), B=0) {is=title K2Node_CallFunction}
   n5 = "Equal (Integer)"(A=Get DisplayedValue() -> (DisplayedValue:int=0), B=0) {is=title K2Node_CallFunction}
   > PlayAnimation(InAnimation=Select(
     Option 0:Widget Animation Object Reference=Get Updated() -> (Updated:Widget Animation Object Reference)
     Option 1:Widget Animation Object Reference=Get IntroAnim() -> (IntroAnim:Widget Animation Object Reference)
     Index=AND Boolean(A=@n5, B=@n4)
    ) -> (ReturnValue:Widget Animation Object Reference))
   > UpdateValue(NewValue=Get NumCardsCollected() -> (NumCardsCollected=0))
   > UpdateVisibility()
  ]}
 UpdateValue {nodes [
   n1 = UpdateValue!() -> (then, NewValue:int) {is=title K2Node_FunctionEntry}
   > "SetText (Text)"(self=Get Value() -> (Value:Common Text Object Reference), InText="To Text (Integer)"(Value=@n1))
   > Set DisplayedValue(DisplayedValue:int=@n1) -> (Output_Get:int=0)
  ]}
 UpdateVisibility {nodes [
   UpdateVisibility() {is=title K2Node_FunctionEntry}
   n1 = Get DisplayedValue() -> (DisplayedValue=0)
   > SetVisibility(InVisibility=Select(
     Option 0:ESlateVisibility Enum=Get NoCardVisibility() -> (NoCardVisibility:ESlateVisibility Enum=Visible)
     Option 1:ESlateVisibility Enum=SelfHitTestInvisible
     Index="integer > integer"(A=@n1) {is=title K2Node_PromotableOperator}
    ) -> (ReturnValue:ESlateVisibility Enum))
  ]}
 UpdateVisibilityForMobileLayoutEditor {nodes [
   n1 = UpdateVisibilityForMobileLayoutEditor!() -> (then, bInLayoutMode:bool=~) {is=title K2Node_FunctionEntry}
   > n2 = Branch(Condition=@n1)
   > SetVisibility(InVisibility=SelfHitTestInvisible)
   UpdateVisibility(execute=@n2.else) {is=title K2Node_CallFunction}
  ]}
]
variables [
 DisplayedValue {type=integer}
 NoCardVisibility {type=ESlateVisibility, default=Collapsed}
 ResurectionEnabled {type=boolean}
]
widgets {
 info {
  inheritedWidgetCount=0
  parentClass=/Script/FortniteUI.FortRebootCardsCollectedWidget
  rootWidgetClass=&HorizontalBox
  widgetCount=8
 }
 widgets [
  HorizontalBox_1 {widgetClassPath=&HorizontalBox, widget=HorizontalBox_1}
  Background {
   widgetClassPath=&CommonBorder
   bIsVariable=true
   parent=HorizontalBox_1
   widget=Background
   properties {
    background {
     bIsDynamicallyLoaded=false
     drawAs=NoDrawType
     imageType=NoImage
     mirroring=NoMirror
     resourceName=None
     resourceObject=/FortUILibrary/Styles/Border/Materials/M_Border_HUD_VGradient_Inst
     tiling=NoTile
     imageSize=<32 32>
     margin=<ltrb 0.49 0.49 0.49 0.49>
     outlineSettings {
      bUseBrushTransparency=false
      roundingType=HalfHeightRadius
      width=0
      color {colorUseRule=UseColor_Specified, specifiedColor=<rgba 0 0 0 0>}
      cornerRadii=<0 0 0 0>
     }
     tintColor {colorUseRule=UseColor_Specified, specifiedColor=<rgba 1 1 1 1>}
     uVRegion {bIsValid=false, max=<0 0>, min=<0 0>}
    }
    padding=<ltrb 0 0 0 0>
   }
   slot {class=HorizontalBoxSlot}
  }
  HorizontalBox_0 {
   widgetClassPath=&HorizontalBox
   parent=Background
   widget=HorizontalBox_0
   slot {class=BorderSlot, properties {padding=<ltrb 0 0 0 0>}}
  }
  SizeBoxContainer {
   widgetClassPath=&SizeBox
   bIsVariable=true
   parent=HorizontalBox_0
   widget=SizeBoxContainer
   properties {bOverride_HeightOverride=true, bOverride_WidthOverride=true, heightOverride=80, maxDesiredWidth=64, minDesiredWidth=64, widthOverride=64}
   slot {class=HorizontalBoxSlot}
  }
  VerticalBox_0 {widgetClassPath=&VerticalBox, parent=SizeBoxContainer, widget=VerticalBox_0, slot {class=SizeBoxSlot}}
  ImageRebootIcon {
   widgetClassPath=&Image
   bIsVariable=true
   parent=VerticalBox_0
   widget=ImageRebootIcon
   properties {
    visibility=HitTestInvisible
    brush {
     bIsDynamicallyLoaded=false
     drawAs=Image
     imageType=FullColor
     mirroring=NoMirror
     resourceName=None
     resourceObject=/Game/Athena/HUD/Materials/MI_UI_RebootCardIcon_PickupManager
     tiling=NoTile
     imageSize=<46 46>
     margin=<ltrb 0 0 0 0>
     outlineSettings {
      bUseBrushTransparency=false
      roundingType=HalfHeightRadius
      width=0
      color {colorUseRule=UseColor_Specified, specifiedColor=<rgba 0 0 0 0>}
      cornerRadii=<0 0 0 0>
     }
     tintColor {colorUseRule=UseColor_Specified, specifiedColor=<rgba 1 1 1 1>}
     uVRegion {bIsValid=false, max=<0 0>, min=<0 0>}
    }
    colorAndOpacity=<rgba 0 0.8745 1 1>
   }
   slot {class=VerticalBoxSlot, properties {horizontalAlignment=HAlign_Center, verticalAlignment=VAlign_Center, padding=<ltrb 0 4 0 0>}}
  }
  Value {
   widgetClassPath=&CommonTextBlock
   bIsVariable=true
   parent=VerticalBox_0
   widget=Value
   properties {
    text=0
    font {
     bForceMonospaced=false
     bMaterialIsStencil=false
     fontMaterial=null
     fontObject=&BurbankSmall
     letterSpacing=0
     monospacedWidth=1
     size=16
     skewAmount=0
     typefaceFontName=Bold
     outlineSettings {
      bApplyOutlineToDropShadows=false
      bMiteredCorners=false
      bSeparateFillAlpha=false
      outlineMaterial=&M-FontShadow
      outlineSize=3
      outlineColor=<rgba 0 0 0 0.5>
     }
    }
    shadowOffset=<0 0>
   }
   slot {class=VerticalBoxSlot, properties {horizontalAlignment=HAlign_Center, padding=<ltrb 0 -2 0 0>}}
  }
  "[Image] Separator" {
   widgetClassPath=&Image
   bIsVariable=true
   parent=HorizontalBox_1
   widget="[Image] Separator"
   properties {
    renderOpacity=0.25
    visibility=SelfHitTestInvisible
    brush {
     bIsDynamicallyLoaded=false
     drawAs=Image
     imageType=NoImage
     mirroring=NoMirror
     resourceName=None
     resourceObject=/Game/UI/Foundation/Textures/Rule/Rule-Dotted-HUD
     tiling=Vertical
     imageSize=<8 48>
     margin=<ltrb 0 0 0 0>
     outlineSettings {
      bUseBrushTransparency=false
      roundingType=HalfHeightRadius
      width=0
      color {colorUseRule=UseColor_Specified, specifiedColor=<rgba 0 0 0 0>}
      cornerRadii=<0 0 0 0>
     }
     tintColor {colorUseRule=UseColor_Specified, specifiedColor=<rgba 1 1 1 1>}
     uVRegion {bIsValid=false, max=<0 0>, min=<0 0>}
    }
    colorAndOpacity=<rgba 1 1 1 0.75>
   }
   slot {class=HorizontalBoxSlot, properties {horizontalAlignment=HAlign_Center, verticalAlignment=VAlign_Center, padding=<ltrb 2 0 0 0>}}
  }
 ]
}
