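# SoundCue for the low-material warning UI sound, judging by its name and folder.
# A single wave player plays the Low_Material_Warning wave directly as the root, with no randomization, mixing or modulation nodes.
# The cue routes to the Fort_UI_InGame_NonCritical sound class at a reduced volume multiplier.
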SoundCue /Game/Sounds/Fort_UI_Sounds/LowMaterials/Low_Material_Warning_Cue
root=WavePlayer_0
nodes [{id=WavePlayer_0, properties {soundWaveAssetPtr=&/Low_Material_Warning}}]
properties {maxDistance=2097152, soundClassObject=&Fort_UI_InGame_NonCritical, volumeMultiplier=0.6}
