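# No_Material_Warning_Cue is a UI sound cue, likely the warning played when building materials run out, judging by its name.
# A single Wave Player plays the No_Material_Warning wave directly as the root, with no randomization or modulation.
# The cue is routed to the Fort_UI sound class at reduced volume.
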SoundCue /Game/Sounds/Fort_UI_Sounds/LowMaterials/No_Material_Warning_Cue
root=WavePlayer_0
nodes [{id=WavePlayer_0, properties {soundWaveAssetPtr=&/No_Material_Warning}}]
properties {maxDistance=2097152, soundClassObject=&Fort_UI, volumeMultiplier=0.5}
