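# HUD widget for one building resource in Athena (Battle Royale), showing icon and count, based on FortResourceWidget.
# Active-state changes restyle the background, with a mobile variant, and scale the icon; reaching max count shears the text and shows an underline.
# In Creative or Playground, Construct loosens the fixed width and sets padding.
# Touch start selects the material; move/end compare TouchStart minus Now against 0.2 ms, effectively always returning Handled.
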WidgetBlueprint /Game/UI/InGame/HUD/AthenaPackResource
graphs [
 EventGraph {nodes [
   n1 = Event OnIsActiveResourceChanged!() -> (OutputDelegate, then, bIsActiveResource:bool) {is=title K2Node_Event}
   n2 = IsMobileGame()
   > SetRenderScale!(
    execute
    self:Widget Object Reference=Get LazyImage_ResourceIcon() -> (LazyImage_ResourceIcon:Fort Lazy Image Object Reference)
    Scale:Vector 2D Structure=Select(Option 0:Vector 2D Structure=(X=0.6,Y=0.6), Option 1:Vector 2D Structure=(X=1,Y=1), Index=OR Boolean(A=@n1.bIsActiveResource, B=@n2)) -> (ReturnValue:Vector 2D Structure)
   ) -> (then)
   n3 = Get BackgroundStyle_HighlightedMobile() -> (BackgroundStyle_HighlightedMobile:Common Border Style Class Reference)
   n4 = Get BackgroundStyle_Highlighted() -> (BackgroundStyle_Highlighted:Common Border Style Class Reference)
   > SetStyle!(
    execute
    self:Common Border Object Reference=Get Background() -> (Background:Common Border Object Reference)
    InStyle:Common Border Style Class Reference=Select(
     Option 0:Common Border Style Class Reference=Get BackgroundStyle_Regular() -> (BackgroundStyle_Regular:Common Border Style Class Reference)
     Option 1:Common Border Style Class Reference=Select(Option 0=@n4, Option 1=@n3, Index=@n2)
     Index=@n1.bIsActiveResource
    ) -> (ReturnValue:Common Border Style Class Reference)
   ) -> (then)
   > InvalidateLayoutAndVolatility(self:Widget Object Reference)
   n5 = Event OnResourceCountUpdated!() -> (OutputDelegate, then, CurrentCount:int=0, MaxCount:int=0, bHasInfiniteResources:bool) {is=title K2Node_Event}
   n6 = "integer < integer"(A=@n5.CurrentCount, B=@n5.MaxCount)
   > SetVisibility(
    self=Get Image_MaxUnderline() -> (Image_MaxUnderline:Image Object Reference)
    InVisibility=Select(Option 0:ESlateVisibility Enum=HitTestInvisible, Option 1:ESlateVisibility Enum=Collapsed, Index=@n6) -> (ReturnValue:ESlateVisibility Enum)
   )
   > SetRenderShear!(
    execute
    self:Widget Object Reference=Get NumericText_ResourceCount() -> (NumericText_ResourceCount:Common Numeric Text Block Object Reference)
    Shear:Vector 2D Structure=Select(Option 0:Vector 2D Structure=(X=-12,Y=0), Option 1:Vector 2D Structure=(X=0,Y=0), Index=@n6) -> (ReturnValue:Vector 2D Structure)
   ) -> (then)
   Event Construct()
   > n7 = PlaylistDataReadyAsync!(execute:Exec) -> (
    then:Exec
    OnReady:Exec
    GameState:Fort Game State Athena Object Reference
    Playlist:Fort Playlist Object Reference
    PlaylistContextTags:"Gameplay Tag Container Structure (by ref)"
   ) {is=title K2Node_AsyncAction}
   Branch(execute=@n7.OnReady, Condition=IsCreativeOrPlaygroundModePlaylist!() -> (ReturnValue:bool))
   n8 = Get SizeBoxContainer() -> (SizeBoxContainer:Size Box Object Reference)
   > SetMinDesiredWidth!(
    execute
    self:Size Box Object Reference=@n8
    InMinDesiredWidth:float=Get WidthOverride!(self:Size Box Object Reference=@n8) -> (WidthOverride:float)
   ) -> (then)
   > ClearWidthOverride(self:Size Box Object Reference=@n8)
   > SetPadding!(execute, self:User Widget Object Reference, InPadding_Left:float=5, InPadding_Top:float=0, InPadding_Right:float=5, InPadding_Bottom:float=0) -> (then)
  ]}
 OnTouchEnded {nodes [
   OnTouchEnded!() -> (then, MyGeometry:Geometry Structure, InTouchEvent:"Pointer Event Structure (by ref)") {is=title K2Node_FunctionEntry}
   n1 = Get Total Milliseconds!(A:Timespan Structure=DateTime - DateTime!(
     A:Date Time Structure=Get TouchStart() -> (TouchStart:Date Time Structure)
     B:Date Time Structure=Now!() -> (ReturnValue:Date Time Structure)
    ) -> (ReturnValue:Timespan Structure)) -> (ReturnValue:double) {is=title K2Node_CallFunction}
   > Return Node!(
    execute
    ReturnValue:Event Reply Structure=Select(
     Option 0:Event Reply Structure=Handled()
     Option 1:Event Reply Structure=Unhandled()
     Index="float >= float"(A=@n1, B=0.2) {is=title K2Node_CallFunction}
    ) -> (ReturnValue:Event Reply Structure)
   )
  ]}
 OnTouchMoved {nodes [
   OnTouchMoved!() -> (then, MyGeometry:Geometry Structure, InTouchEvent:"Pointer Event Structure (by ref)") {is=title K2Node_FunctionEntry}
   n1 = Get Total Milliseconds(A=DateTime - DateTime(A=Get TouchStart(), B=Now()))
   > Return Node(ReturnValue=Select(Option 0=Handled(), Option 1=Unhandled(), Index="float >= float"(A=@n1, B=0.2)))
  ]}
 OnTouchStarted {nodes [
   OnTouchStarted!() -> (then, MyGeometry:Geometry Structure, InTouchEvent:"Pointer Event Structure (by ref)") {is=title K2Node_FunctionEntry}
   > Set TouchStart(TouchStart:Date Time Structure=Now()) -> (Output_Get:Date Time Structure)
   n1 = Get ResourceType() -> (ResourceType:EFortResourceType Enum=Wood)
   > n2 = Branch(Condition="Not Equal (Enum)"!(A:EFortResourceType Enum=@n1, B:EFortResourceType Enum=None) -> (ReturnValue:bool))
   > n3 = SetResourceMaterial!(
    execute
    self:Athena HUDContext Object Reference=Get AthenaHUDContext!() -> (ReturnValue:Athena HUDContext Object Reference) {is=title K2Node_GetBlueprintContextEx}
    NewMaterial:EFortResourceType Enum=@n1
   ) -> (then)
   Return Node(execute=[@n3.then, @n2.else], ReturnValue=Handled())
  ]}
]
variables [
 BackgroundStyle_Highlighted {type=Class, default=/FortUILibrary/Styles/Border/Border_HUD_ResourceBg_Highlighted.Border_HUD_ResourceBg_Highlighted_C}
 BackgroundStyle_HighlightedMobile {type=Class, default=/FortUILibrary/Styles/Border/Border_HUD_ResourceBg_HighlightedBacchus.Border_HUD_ResourceBg_HighlightedBacchus_C}
 BackgroundStyle_Regular {type=Class, default=/FortUILibrary/Styles/Border/Border_HUD_ResourceBg.Border_HUD_ResourceBg_C}
 TouchStart {type=DateTime, default=0001.01.01-00.00.00}
]
widgets {
 info {inheritedWidgetCount=3, parentClass=/Script/FortniteUI.FortResourceWidget, rootWidgetClass=&Overlay, widgetCount=10}
 widgets [
  ContentOverlay {widgetClassPath=&Overlay, widget=ContentOverlay}
  Background {
   widgetClassPath=&CommonBorder
   bIsVariable=true
   parent=ContentOverlay
   widget=Background
   properties {
    style=&Border_HUD_ResourceBg_C
    visibility=Visible
    background {
     bIsDynamicallyLoaded=false
     drawAs=Box
     imageType=NoImage
     mirroring=NoMirror
     resourceName=None
     resourceObject=/FortUILibrary/Foundation/Buttons/Materials/TopBar/M-Button-OutlineTransparent-BacchusSelector
     tiling=NoTile
     imageSize=<24 24>
     margin=<ltrb 0.5 0.5 0.5 0.5>
     outlineSettings {
      bUseBrushTransparency=false
      roundingType=HalfHeightRadius
      width=0
      color {colorUseRule=UseColor_Specified, specifiedColor=<rgba 0 0 0 0>}
      cornerRadii=<0 0 0 0>
     }
     tintColor {colorUseRule=UseColor_Specified, specifiedColor=<rgba 1 1 1 1>}
     uVRegion {bIsValid=false, max=<0 0>, min=<0 0>}
    }
    padding=<ltrb 0 0 0 0>
   }
   slot {class=OverlaySlot, properties {horizontalAlignment=HAlign_Fill, verticalAlignment=VAlign_Fill}}
  }
  SizeBoxContainer {
   widgetClassPath=&SizeBox
   bIsVariable=true
   parent=ContentOverlay
   widget=SizeBoxContainer
   properties {bOverride_HeightOverride=true, bOverride_WidthOverride=true, heightOverride=80, maxDesiredWidth=64, minDesiredWidth=64, widthOverride=60}
   slot {class=OverlaySlot, properties {horizontalAlignment=HAlign_Center, verticalAlignment=VAlign_Center}}
  }
  VerticalBox_0 {widgetClassPath=&VerticalBox, parent=SizeBoxContainer, widget=VerticalBox_0, slot {class=SizeBoxSlot}}
  Overlay_1 {
   widgetClassPath=&Overlay
   parent=VerticalBox_0
   widget=Overlay_1
   slot {class=VerticalBoxSlot, properties {horizontalAlignment=HAlign_Center, verticalAlignment=VAlign_Center, padding=<ltrb -32 -32 -32 -32>}}
  }
  Spacer_56 {
   widgetClassPath=&Spacer
   parent=Overlay_1
   widget=Spacer_56
   properties {size=<1 118.6356>}
   slot {class=OverlaySlot, properties {horizontalAlignment=HAlign_Center, verticalAlignment=VAlign_Center}}
  }
  LazyImage_ResourceIcon {
   widgetClassPath=&FortLazyImage
   bInherited=true
   parent=Overlay_1
   widget=LazyImage_ResourceIcon
   properties {
    brush {
     bIsDynamicallyLoaded=false
     drawAs=Image
     imageType=NoImage
     mirroring=NoMirror
     resourceName=None
     resourceObject=/Game/Athena/HUD/Atlas/Shared/T-Icon-IBeam_Sprite
     tiling=NoTile
     imageSize=<54 54>
     margin=<ltrb 0 0 0 0>
     outlineSettings {
      bUseBrushTransparency=false
      roundingType=HalfHeightRadius
      width=0
      color {colorUseRule=UseColor_Specified, specifiedColor=<rgba 0 0 0 0>}
      cornerRadii=<0 0 0 0>
     }
     tintColor {colorUseRule=UseColor_Specified, specifiedColor=<rgba 1 1 1 1>}
     uVRegion {bIsValid=false, max=<0 0>, min=<0 0>}
    }
    renderTransform {angle=0, scale=<0.6 0.6>, shear=<0 0>, translation=<0 -38>}
    renderTransformPivot=<0.5 1>
   }
   slot {class=OverlaySlot, properties {horizontalAlignment=HAlign_Center, verticalAlignment=VAlign_Bottom}}
  }
  NumericText_ResourceCount {
   widgetClassPath=/Script/CommonUI.CommonNumericTextBlock
   bInherited=true
   parent=VerticalBox_0
   widget=NumericText_ResourceCount
   properties {
    text=0
    font {
     bForceMonospaced=false
     bMaterialIsStencil=false
     fontMaterial=null
     fontObject=&BurbankSmall
     letterSpacing=0
     monospacedWidth=1
     size=16
     skewAmount=0
     typefaceFontName=Bold
     outlineSettings {
      bApplyOutlineToDropShadows=false
      bMiteredCorners=false
      bSeparateFillAlpha=true
      outlineMaterial=&M-FontShadow
      outlineSize=3
      outlineColor=<rgba 0 0 0 0.5>
     }
    }
    formattingSpecification {
     alwaysSign=false
     maximumFractionalDigits=0
     maximumIntegralDigits=324
     minimumFractionalDigits=0
     minimumIntegralDigits=1
     roundingMode=HalfFromZero
     useGrouping=false
    }
    shadowOffset=<0 0>
   }
   slot {class=VerticalBoxSlot, properties {horizontalAlignment=HAlign_Center, padding=<ltrb 0 -6 0 0>}}
  }
  Image_MaxUnderline {
   widgetClassPath=&Image
   bIsVariable=true
   parent=VerticalBox_0
   widget=Image_MaxUnderline
   properties {
    visibility=Collapsed
    brush {
     bIsDynamicallyLoaded=false
     drawAs=Image
     imageType=NoImage
     mirroring=NoMirror
     resourceName=None
     resourceObject=/Game/Athena/HUD/Art/Materials/M_UI_HUD_Resources_SelectionIndicator
     tiling=NoTile
     imageSize=<0 4>
     margin=<ltrb 0 0 0 0>
     outlineSettings {
      bUseBrushTransparency=false
      roundingType=HalfHeightRadius
      width=0
      color {colorUseRule=UseColor_Specified, specifiedColor=<rgba 0 0 0 0>}
      cornerRadii=<0 0 0 0>
     }
     tintColor {colorUseRule=UseColor_Specified, specifiedColor=<rgba 1 1 1 1>}
     uVRegion {bIsValid=false, max=<0 0>, min=<0 0>}
    }
    renderTransform {angle=0, scale=<1 1>, shear=<-12 0>, translation=<0 0>}
   }
   slot {class=VerticalBoxSlot, properties {verticalAlignment=VAlign_Bottom, padding=<ltrb 8 0 8 0>}}
  }
  NumericText_ResourceCost {widgetClassPath=&CommonNumericTextBlock, bInherited=true, widget=null}
 ]
}
