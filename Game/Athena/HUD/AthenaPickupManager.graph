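# Battle Royale HUD widget on FortPickupManagerWidget showing an ingredient list, reboot cards, cash, wood, stone, metal and gold currency counters.
# Once playlist data is ready, it collapses cash in Creative or Playground playlists and otherwise shows it.
# One tick after a resource visibility change, it shows the dotted separator only if cash and another resource counter are visible.
# It also plays warning cues for low or depleted building resources.
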WidgetBlueprint /Game/Athena/HUD/AthenaPickupManager
graphs [EventGraph {nodes [
   Event Construct()
   > n1 = PlaylistDataReadyAsync!(execute:Exec) -> (
    then:Exec
    OnReady:Exec
    GameState:Fort Game State Athena Object Reference
    Playlist:Fort Playlist Object Reference
    PlaylistContextTags:"Gameplay Tag Container Structure (by ref)"
   ) {is=title K2Node_AsyncAction}
   SetVisibility(
    execute=@n1.OnReady
    self=Get PackResourceCash() -> (PackResourceCash:Athena Pack Resource Object Reference)
    InVisibility=Select(Option 0=SelfHitTestInvisible, Option 1=Collapsed, Index=IsCreativeOrPlaygroundModePlaylist!() -> (ReturnValue:bool))
   )
   Event OnResourceWidgetVisibilityChanged!() -> (OutputDelegate, then, ResourceWidget:Fort Resource Widget Object Reference) {is=title K2Node_Event}
   > DelayUntilNextTick!(execute) -> (then)
   n2 = IsVisible(self=Get PackResourceCash())
   n3 = OR Boolean!(
    A=IsVisible(self=Get PackResourceWood() -> (PackResourceWood:Athena Pack Resource Object Reference))
    B=IsVisible(self=Get PackResourceStone() -> (PackResourceStone:Athena Pack Resource Object Reference))
    C:bool=IsVisible(self=Get PackResourceMetal() -> (PackResourceMetal:Athena Pack Resource Object Reference))
    D:bool=IsVisible(self=Get PackResourceGoldCurrency() -> (PackResourceGoldCurrency:Athena Pack Resource Object Reference))
   ) -> (ReturnValue)
   > SetVisibility(
    self="Get [Image] Separator"() -> ("[Image] Separator":Image Object Reference)
    InVisibility=Select(Option 0=Collapsed, Option 1=SelfHitTestInvisible, Index=AND Boolean(A=@n2, B=@n3))
   )
   Event OnLowOnBuildingResources() {is=title K2Node_Event}
   > PlaySound2D(Sound=Get LowOnBuildingResourcesSound() -> (LowOnBuildingResourcesSound:Sound Base Object Reference))
   Event OnOutOfBuildingResources() {is=title K2Node_Event}
   > PlaySound2D(Sound=Get OutOfBuildingResourcesSound() -> (OutOfBuildingResourcesSound:Sound Base Object Reference))
  ]}]
variables [
 LowOnBuildingResourcesSound {type=SoundBase, default=/Game/Sounds/Fort_UI_Sounds/LowMaterials/Low_Material_Warning_Cue}
 OutOfBuildingResourcesSound {type=SoundBase, default=/Game/Sounds/Fort_UI_Sounds/LowMaterials/No_Material_Warning_Cue}
]
widgets {
 info {
  inheritedWidgetCount=2
  namedSlotCount=1
  parentClass=/Script/FortniteUI.FortPickupManagerWidget
  rootWidgetClass=&Overlay
  widgetCount=13
 }
 widgets [
  Overlay_0 {widgetClassPath=&Overlay, widget=Overlay_0}
  VerticalBox_1 {widgetClassPath=&VerticalBox, parent=Overlay_0, widget=VerticalBox_1, slot {class=OverlaySlot}}
  IngredientList {
   widgetClassPath=&/AthenaHUDItemList
   parent=VerticalBox_1
   widget=IngredientList
   properties {
    elementTag {tagName=HUD.Athena.IngredientsList}
    hUDElementTag {gameplayTags [
      {tagName=HUD.Athena.HideOnBus}
      {tagName=Athena.HideForCapture}
      {tagName=HUD.Athena.HideForGadget}
      {tagName=HUD.Athena.Hide.WithInventory}
      {tagName=HUD.Athena.Hide.WithRCPawn}
      {tagName=HUD.Athena.HideForAbility}
      {tagName=HUD.Event.HideForSpecialEvent}
      {tagName=Athena.Creative.HideForLoading}
      {tagName=HUD.Hide.WhileLoading}
      {tagName=HUD.Athena.HideForPostGame}
      {tagName=HUD.Athena.HideforRadial}
      {tagName=HUD.Athena.HideforNoBuilding}
      {tagName=HUD.Athena.HideforVehicle}
      {tagName=HUD.Athena.Hide.WithMap}
      {tagName=HUD.Athena.HideForConversation}
      {tagName=Athena.Creative.HUDController.HideCraftingResources}
     ]}
    itemFilter {bRequiresItemDetails=false, itemTypeFilters [WorldResource]}
    itemFilterTags {gameplayTags [{tagName=Item.Ingredient}]}
   }
   slot {class=VerticalBoxSlot, properties {horizontalAlignment=HAlign_Right, padding=<ltrb 0 8 0 16>}}
  }
  Panel_AllResources {
   widgetClassPath=&HorizontalBox
   bInherited=true
   parent=VerticalBox_1
   widget=Panel_AllResources
   slot {class=VerticalBoxSlot, properties {horizontalAlignment=HAlign_Left, verticalAlignment=VAlign_Top}}
  }
  RebootCardsCollectedWidget {
   widgetClassPath=&/RebootCardsCollectedWidget
   bIsVariable=true
   parent=Panel_AllResources
   widget=RebootCardsCollectedWidget
   properties {visibility=Collapsed}
   slot {class=HorizontalBoxSlot, properties {padding=<ltrb 0 0 4 0>}}
  }
  CashHUDElementWrapper {
   widgetClassPath=&/FortHUDElementWrapper
   bIsVariable=true
   parent=Panel_AllResources
   widget=CashHUDElementWrapper
   properties {elementTag {tagName=HUD.CashResource}, hUDElementTag {gameplayTags [{tagName=HUD.Athena.HideForCreative}]}}
   slot {class=HorizontalBoxSlot}
  }
  PackResourceCash {
   widgetClassPath=/Game/UI/InGame/HUD/AthenaPackResource
   bIsVariable=true
   namedSlotHost=CashHUDElementWrapper
   widget=PackResourceCash
   properties {
    bPickUpVisible=true
    itemDefinition=/Game/Items/ResourcePickups/Athena_WadsItemData
    resourceType=GoldCurrency
    visibility=SelfHitTestInvisible
   }
  }
  "[Image] Separator" {
   widgetClassPath=&Image
   bIsVariable=true
   parent=Panel_AllResources
   widget="[Image] Separator"
   properties {
    renderOpacity=0.25
    visibility=SelfHitTestInvisible
    brush {
     bIsDynamicallyLoaded=false
     drawAs=Image
     imageType=NoImage
     mirroring=NoMirror
     resourceName=None
     resourceObject=/Game/UI/Foundation/Textures/Rule/Rule-Dotted-HUD
     tiling=Vertical
     imageSize=<8 48>
     margin=<ltrb 0 0 0 0>
     outlineSettings {
      bUseBrushTransparency=false
      roundingType=HalfHeightRadius
      width=0
      color {colorUseRule=UseColor_Specified, specifiedColor=<rgba 0 0 0 0>}
      cornerRadii=<0 0 0 0>
     }
     tintColor {colorUseRule=UseColor_Specified, specifiedColor=<rgba 1 1 1 1>}
     uVRegion {bIsValid=false, max=<0 0>, min=<0 0>}
    }
    colorAndOpacity=<rgba 1 1 1 0.75>
   }
   slot {class=HorizontalBoxSlot, properties {horizontalAlignment=HAlign_Center, verticalAlignment=VAlign_Center, padding=<ltrb 2 0 4 0>}}
  }
  PackResourceWood {
   widgetClassPath=&AthenaPackResource
   bIsVariable=true
   parent=Panel_AllResources
   widget=PackResourceWood
   properties {resourceType=Wood, visibility=SelfHitTestInvisible}
   slot {class=HorizontalBoxSlot, properties {horizontalAlignment=HAlign_Left, verticalAlignment=VAlign_Center, size {sizeRule=Fill, value=1}}}
  }
  PackResourceStone {
   widgetClassPath=&AthenaPackResource
   bIsVariable=true
   parent=Panel_AllResources
   widget=PackResourceStone
   properties {resourceType=Stone, visibility=SelfHitTestInvisible}
   slot {class=HorizontalBoxSlot, properties {horizontalAlignment=HAlign_Center, verticalAlignment=VAlign_Center, size {sizeRule=Fill, value=1}}}
  }
  PackResourceMetal {
   widgetClassPath=&AthenaPackResource
   bIsVariable=true
   parent=Panel_AllResources
   widget=PackResourceMetal
   properties {visibility=SelfHitTestInvisible}
   slot {class=HorizontalBoxSlot, properties {horizontalAlignment=HAlign_Center, verticalAlignment=VAlign_Center, size {sizeRule=Fill, value=1}}}
  }
  PackResourceGoldCurrency {
   widgetClassPath=&AthenaPackResource
   bIsVariable=true
   parent=Panel_AllResources
   widget=PackResourceGoldCurrency
   properties {bCanUseShorthandCount=true, resourceType=GoldCurrency, visibility=Collapsed}
   slot {class=HorizontalBoxSlot, properties {horizontalAlignment=HAlign_Right, verticalAlignment=VAlign_Center, size {sizeRule=Fill, value=1}}}
  }
  TutorialHighlight {
   widgetClassPath=&TutorialHighlight
   parent=Overlay_0
   widget=TutorialHighlight
   properties {
    customHighlightWidget=/Game/Athena/HUD/Tutorial/Tutorial_Outline_Square_Static.Tutorial_Outline_Square_Static_C
    visibility=SelfHitTestInvisible
   }
   slot {class=OverlaySlot, properties {horizontalAlignment=HAlign_Fill, verticalAlignment=VAlign_Fill}}
  }
 ]
}
