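# Pickup message HUD widget for Athena that shows an item icon beside a stack count.
# Setup sets icon visibility from whether the soft texture is null, applies the brush if valid, then sets the count.
# Intro pops the row with an additive glow flash, and Outro slides it right while fading.
# Setup's Intro PlayAnimation node has no incoming exec, and nothing here plays Outro.
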WidgetBlueprint /Game/Athena/HUD/AthenaPickupMessageItem
animations [
 Intro {
  binding_count=2
  length_seconds=0.4405
  object=&self:Intro
  sequence {bindings [
    Image_glow {
     id=#1
     possessed_class=Image
     tracks [
      {
       type=MovieScene2DTransformTrack
       sections [{
         type=MovieScene2DTransformSection
         channels [
          Scale.X {type=MovieSceneScriptingFloatChannel, keys [[0,0.87],[8.648,1.2]]}
          Scale.Y {type=MovieSceneScriptingFloatChannel, keys [[0,0.88],[4.0253,1.3568],[8.648,0.091]]}
         ]
        }]
      }
      {
       type=MovieSceneFloatTrack
       display_name=Render Opacity
       sections [{type=MovieSceneFloatSection, channels [None {type=MovieSceneScriptingFloatChannel, keys [[0,0],[0.9437,1],[8.8107,0]]}]}]
      }
     ]
    }
    HorizontalBox_27 {
     id=#2
     possessed_class=HorizontalBox
     tracks [
      {
       type=MovieScene2DTransformTrack
       sections [{
         type=MovieScene2DTransformSection
         channels [
          Translation.X {type=MovieSceneScriptingFloatChannel, const [0,1]}
          Translation.Y {type=MovieSceneScriptingFloatChannel}
          Angle {type=MovieSceneScriptingFloatChannel}
          Scale.X {type=MovieSceneScriptingFloatChannel, keys [[0,0],[1,2],[4,1]]}
          Scale.Y {type=MovieSceneScriptingFloatChannel, keys [[0,0],[1,2],[4,1]]}
          Shear.X {type=MovieSceneScriptingFloatChannel}
          Shear.Y {type=MovieSceneScriptingFloatChannel}
         ]
        }]
      }
      {
       type=MovieSceneFloatTrack
       display_name=Render Opacity
       sections [{type=MovieSceneFloatSection, channels [None {type=MovieSceneScriptingFloatChannel, const [1,1]}]}]
      }
     ]
    }
   ]}
 }
 Outro {
  binding_count=1
  length_seconds=0.357
  object=&self:Outro
  sequence {bindings [HorizontalBox_27 {
     id=#3
     possessed_class=HorizontalBox
     tracks [
      {
       type=MovieScene2DTransformTrack
       sections [{
         type=MovieScene2DTransformSection
         channels [
          Translation.X {type=MovieSceneScriptingFloatChannel, keys [[0,0],[7.1403,28.6168]]}
          Translation.Y {type=MovieSceneScriptingFloatChannel}
          Angle {type=MovieSceneScriptingFloatChannel}
          Scale.X {type=MovieSceneScriptingFloatChannel}
          Scale.Y {type=MovieSceneScriptingFloatChannel}
          Shear.X {type=MovieSceneScriptingFloatChannel}
          Shear.Y {type=MovieSceneScriptingFloatChannel}
         ]
        }]
      }
      {
       type=MovieSceneFloatTrack
       display_name=Render Opacity
       sections [{type=MovieSceneFloatSection, channels [None {type=MovieSceneScriptingFloatChannel, keys [[0,1],[7.1403,0]]}]}]
      }
     ]
    }]}
 }
]
graphs [
 EventGraph {links [], nodes []}
 Setup {nodes [
   n1 = PlayAnimation(InAnimation=Get Intro())
   n2 = Setup!() -> (then, Icon:Texture 2D Soft Object Reference, Count:int) {is=title K2Node_FunctionEntry}
   Set Stack Count!(
    execute=[@n1.then, @n3.then, @n5.then]
    self:Item Stack Counter Object Reference="Get [Stack Counter] Count"() -> ("[Stack Counter] Count":Item Stack Counter Object Reference)
    Count:int=@n2.Count
   ) -> (then) {is=title K2Node_CallFunction}
   n3 = SetBrushFromSoftTexture!(
    execute=@n5.else
    self:Image Object Reference=Get ItemIcon() -> (ItemIcon:Image Object Reference)
    SoftTexture:Texture 2D Soft Object Reference=@n2
    bMatchSize:bool=false
   ) -> (then)
   n4 = IsAssetNull!(Asset:Soft Object Reference=@n2) -> (ReturnValue:bool)
   SetVisibility(execute=@n2.then, InVisibility=Select(Option 0=HitTestInvisible, Option 1=Collapsed, Index=@n4))
   > n5 = Branch(Condition=@n4)
  ]}
]
variables []
widgets {
 info {inheritedWidgetCount=0, parentClass=&UserWidget, rootWidgetClass=&Overlay, widgetCount=5}
 widgets [
  Overlay_0 {widgetClassPath=&Overlay, widget=Overlay_0}
  HorizontalBox_27 {
   widgetClassPath=&HorizontalBox
   parent=Overlay_0
   widget=HorizontalBox_27
   properties {renderTransformPivot=<1 0.5>}
   slot {class=OverlaySlot}
  }
  ItemIcon {
   widgetClassPath=&Image
   bIsVariable=true
   parent=HorizontalBox_27
   widget=ItemIcon
   properties {brush {
     bIsDynamicallyLoaded=false
     drawAs=Image
     imageType=NoImage
     mirroring=NoMirror
     resourceName=None
     resourceObject=/Game/Global/Textures/Default/Blanks/T_Blank_White_ZeroAlpha_sRGBColor
     tiling=NoTile
     imageSize=<40 40>
     margin=<ltrb 0 0 0 0>
     outlineSettings {
      bUseBrushTransparency=false
      roundingType=HalfHeightRadius
      width=0
      color {colorUseRule=UseColor_Specified, specifiedColor=<rgba 0 0 0 0>}
      cornerRadii=<0 0 0 0>
     }
     tintColor {colorUseRule=UseColor_Specified, specifiedColor=<rgba 1 1 1 1>}
     uVRegion {bIsValid=false, max=<0 0>, min=<0 0>}
    }}
   slot {class=HorizontalBoxSlot}
  }
  "[Stack Counter] Count" {
   widgetClassPath=/Game/UI/Foundation/Items/ItemStackCounter
   bIsVariable=true
   parent=HorizontalBox_27
   widget="[Stack Counter] Count"
   properties {
    visibility=Hidden
    visibilityWithNoStack=Visible
    multiSizeFont {
     font_L {
      bForceMonospaced=false
      bMaterialIsStencil=false
      fontMaterial=null
      fontObject=&BurbankSmall
      letterSpacing=0
      monospacedWidth=1
      size=32
      skewAmount=0
      typefaceFontName=Bold
      outlineSettings {
       bApplyOutlineToDropShadows=true
       bMiteredCorners=false
       bSeparateFillAlpha=false
       outlineMaterial=null
       outlineSize=0
       outlineColor=<rgba 0 0 0 1>
      }
     }
     font_M {
      bForceMonospaced=false
      bMaterialIsStencil=false
      fontMaterial=null
      fontObject=&BurbankSmall
      letterSpacing=0
      monospacedWidth=1
      size=24
      skewAmount=0
      typefaceFontName=Bold
      outlineSettings {
       bApplyOutlineToDropShadows=true
       bMiteredCorners=false
       bSeparateFillAlpha=false
       outlineMaterial=null
       outlineSize=0
       outlineColor=<rgba 0 0 0 1>
      }
     }
     font_S {
      bForceMonospaced=false
      bMaterialIsStencil=false
      fontMaterial=null
      fontObject=&BurbankSmall
      letterSpacing=0
      monospacedWidth=1
      size=16
      skewAmount=0
      typefaceFontName=Bold
      outlineSettings {
       bApplyOutlineToDropShadows=true
       bMiteredCorners=false
       bSeparateFillAlpha=false
       outlineMaterial=null
       outlineSize=0
       outlineColor=<rgba 0 0 0 1>
      }
     }
     font_XL {
      bForceMonospaced=false
      bMaterialIsStencil=false
      fontMaterial=null
      fontObject=&BurbankSmall
      letterSpacing=0
      monospacedWidth=1
      size=48
      skewAmount=0
      typefaceFontName=Bold
      outlineSettings {
       bApplyOutlineToDropShadows=true
       bMiteredCorners=false
       bSeparateFillAlpha=false
       outlineMaterial=null
       outlineSize=0
       outlineColor=<rgba 0 0 0 1>
      }
     }
     font_XS {
      bForceMonospaced=false
      bMaterialIsStencil=false
      fontMaterial=null
      fontObject=&BurbankSmall
      letterSpacing=0
      monospacedWidth=1
      size=9
      skewAmount=0
      typefaceFontName=Bold
      outlineSettings {
       bApplyOutlineToDropShadows=true
       bMiteredCorners=false
       bSeparateFillAlpha=false
       outlineMaterial=null
       outlineSize=0
       outlineColor=<rgba 0 0 0 1>
      }
     }
     font_XXS {
      bForceMonospaced=false
      bMaterialIsStencil=false
      fontMaterial=null
      fontObject=&BurbankSmall
      letterSpacing=0
      monospacedWidth=1
      size=9
      skewAmount=0
      typefaceFontName=Bold
      outlineSettings {
       bApplyOutlineToDropShadows=true
       bMiteredCorners=false
       bSeparateFillAlpha=false
       outlineMaterial=null
       outlineSize=0
       outlineColor=<rgba 0 0 0 1>
      }
     }
    }
    multiSizeMargin {
     margin_L=<ltrb 32 0 11 0>
     margin_M=<ltrb 18 0 6 0>
     margin_S=<ltrb 25 0 24 0>
     margin_XL=<ltrb 32 0 11 0>
     margin_XS=<ltrb 9 0 4 0>
     margin_XXS=<ltrb 9 0 4 0>
    }
   }
   slot {class=HorizontalBoxSlot, properties {verticalAlignment=VAlign_Center}}
  }
  Image_glow {
   widgetClassPath=&Image
   bIsVariable=true
   parent=Overlay_0
   widget=Image_glow
   properties {
    renderOpacity=0
    brush {
     bIsDynamicallyLoaded=false
     drawAs=Image
     imageType=NoImage
     mirroring=NoMirror
     resourceName=None
     resourceObject=/Game/Athena/HUD/LTM/Wax/Materials/M_UI_TextGlow_Add
     tiling=NoTile
     imageSize=<32 32>
     margin=<ltrb 0 0 0 0>
     outlineSettings {
      bUseBrushTransparency=false
      roundingType=HalfHeightRadius
      width=0
      color {colorUseRule=UseColor_Specified, specifiedColor=<rgba 0 0 0 0>}
      cornerRadii=<0 0 0 0>
     }
     tintColor {colorUseRule=UseColor_Specified, specifiedColor=<rgba 1 1 1 1>}
     uVRegion {bIsValid=false, max=<0 0>, min=<0 0>}
    }
    colorAndOpacity=<rgba 0.0148 0.0718 0.401 1>
    renderTransformPivot=<0.7659 0.5>
   }
   slot {class=OverlaySlot, properties {horizontalAlignment=HAlign_Fill, verticalAlignment=VAlign_Fill, padding=<ltrb -40 -10 -40 -10>}}
  }
 ]
}
