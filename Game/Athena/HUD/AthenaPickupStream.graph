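# Athena HUD widget that shows the stream of item pickup messages, built on the native PickupStream class.
# RefreshPickupEntry casts each entry to AthenaPickupMessageItem and passes it the icon and count.
# PreConstruct sets the message box's vertical alignment from a variable (default top), overriding the designer's bottom slot, then calls AdaptForTouch.
# AdaptForTouch shifts the fixed-size container down for compact touch UI on screens of 4:3 or narrower.
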WidgetBlueprint /Game/Athena/HUD/AthenaPickupStream
graphs [
 AdaptForTouch {nodes [
   AdaptForTouch() {is=title K2Node_FunctionEntry}
   n1 = GetOwningPlayer()
   > Branch(Condition=AND Boolean(
     A=IsUsingCompactUI()
     B=IsUsingInputType!(InputType:ECommonInputType Enum=Touch, PlayerController:Player Controller Object Reference=@n1) -> (ReturnValue:bool)
    ))
   > Branch(Condition=CompareToScreenAspectRatio!(Comparison:EAspectRatioComparison Enum=LessThanOrEqualTo, AspectRatio:EAspectRatio Enum=Ratio_4_3) -> (ReturnValue:bool))
   > SetRenderTransform!(
    execute
    self:Widget Object Reference=Get SizeBox_PickupStream() -> (SizeBox_PickupStream:Size Box Object Reference)
    InTransform_Translation:Vector 2D Structure=(X=0,Y=180)
    InTransform_Scale:Vector 2D Structure=(X=1,Y=1)
    InTransform_Shear:Vector 2D Structure=(X=0,Y=0)
    InTransform_Angle:float=0
   ) -> (then)
  ]}
 EventGraph {nodes [
   n1 = Event RefreshPickupEntry!() -> (OutputDelegate, then, EntryWidget:User Widget Object Reference, Icon:"Texture 2D Soft Object Reference (by ref)", Count:int=0) {is=title K2Node_Event}
   > n2 = Cast To AthenaPickupMessageItem!(execute, Object=@n1.EntryWidget) -> (then, CastFailed, AsAthena Pickup Message Item:Athena Pickup Message Item Object Reference) {is=title K2Node_DynamicCast}
   > Setup!(execute, self:Athena Pickup Message Item Object Reference=@n2, Icon:Texture 2D Soft Object Reference=@n1.Icon, Count:int=@n1.Count) -> (then)
   Event PreConstruct()
   > n3 = Cast To SizeBoxSlot!(
    execute
    Object=Get Slot!(self:Widget Object Reference=Get VerticalBox_PickupMessages() -> (VerticalBox_PickupMessages:Vertical Box Object Reference)) -> (Slot:Panel Slot Object Reference)
   ) -> (then, CastFailed, AsSize Box Slot:Size Box Slot Object Reference) {is=title K2Node_DynamicCast}
   > SetVerticalAlignment(
    self:Size Box Slot Object Reference=@n3
    InVerticalAlignment=Get Vertical Alignment() -> (Vertical Alignment:EVerticalAlignment Enum=VAlign_Fill)
   )
   > AdaptForTouch() {is=title K2Node_CallFunction}
  ]}
]
variables [Vertical Alignment {type=EVerticalAlignment, default=VAlign_Top}]
widgets {
 info {inheritedWidgetCount=2, parentClass=/Script/FortniteUI.PickupStream, rootWidgetClass=&SizeBox, widgetCount=3}
 widgets [
  SizeBox_PickupStream {
   widgetClassPath=&SizeBox
   bIsVariable=true
   widget=SizeBox_PickupStream
   properties {bOverride_HeightOverride=true, bOverride_WidthOverride=true, heightOverride=200, widthOverride=150}
  }
  VerticalBox_PickupMessages {
   widgetClassPath=&VerticalBox
   bInherited=true
   bIsVariable=true
   parent=SizeBox_PickupStream
   widget=VerticalBox_PickupMessages
   slot {class=SizeBoxSlot, properties {horizontalAlignment=HAlign_Right, verticalAlignment=VAlign_Bottom}}
  }
  TutorialHighlight {widgetClassPath=&TutorialHighlight}
 ]
}
