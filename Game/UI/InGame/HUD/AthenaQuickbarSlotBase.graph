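# Base widget for one Athena HUD quickbar slot, hosting a native item cell, cooldown, durability, keybinds and an extension overlay.
# Maximize and minimize play GrowAnimation, whose scale factor lerps cell size; tagged finish events resize the slot and cooldown countdowns toggle.
# Selection lifts extensions and plays primary or secondary material animations, skipped on simplified HUD.
# Cooldown delegates hide keybinds during item-activation or death cooldowns.
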WidgetBlueprint /Game/UI/InGame/HUD/AthenaQuickbarSlotBase
animations [
 Selected {
  binding_count=1
  display_label=""
  length_seconds=0.1
  object=&self:Selected
  sequence {bindings [NativeModeSwitcher {
     id=#1
     display_name=NativeCell
     possessed_class=WidgetSwitcher
     tracks [
      {
       type=MovieScene2DTransformTrack
       sections [{
         type=MovieScene2DTransformSection
         channels [
          Translation.X {type=MovieSceneScriptingFloatChannel, const [0,2]}
          Translation.Y {type=MovieSceneScriptingFloatChannel, keys [[0,0],[2,-6]]}
          Angle {type=MovieSceneScriptingFloatChannel}
          Scale.X {type=MovieSceneScriptingFloatChannel}
          Scale.Y {type=MovieSceneScriptingFloatChannel}
          Shear.X {type=MovieSceneScriptingFloatChannel}
          Shear.Y {type=MovieSceneScriptingFloatChannel}
         ]
        }]
      }
      {
       type=MovieSceneWidgetMaterialTrack
       display_name=Selected Cell Brush.Brush Material
       sections [{type=MovieSceneParameterSection, channels [AnimateBorder {type=MovieSceneScriptingFloatChannel, keys [[0,0],[2,1]]}]}]
      }
     ]
    }]}
 }
 GrowAnimation {
  binding_count=1
  length_seconds=0.1
  object=&self:GrowAnimation
  sequence {
   bindings [AthenaQuickbarSlotBase_C_34 {
     id=#2
     possessed_class=AthenaQuickbarSlotBase_C
     tracks [{
       type=MovieSceneFloatTrack
       display_name=Animated Box Scale Factor
       sections [{type=MovieSceneFloatSection, end_frame=6, start_frame=0, channels [None {type=MovieSceneScriptingFloatChannel, keys [[0,0],[6,1]]}]}]
      }]
    }]
   display_rate {numerator=60}
  }
 }
 SelectedSecondary {
  binding_count=1
  length_seconds=0.1
  object=&self:SelectedSecondary
  sequence {bindings [NativeModeSwitcher {
     id=#1
     display_name=NativeCell
     possessed_class=WidgetSwitcher
     tracks [
      {
       type=MovieScene2DTransformTrack
       sections [{
         type=MovieScene2DTransformSection
         channels [
          Translation.X {type=MovieSceneScriptingFloatChannel, const [0,2]}
          Translation.Y {type=MovieSceneScriptingFloatChannel, keys [[0,0],[2,-6]]}
          Angle {type=MovieSceneScriptingFloatChannel}
          Scale.X {type=MovieSceneScriptingFloatChannel}
          Scale.Y {type=MovieSceneScriptingFloatChannel}
          Shear.X {type=MovieSceneScriptingFloatChannel}
          Shear.Y {type=MovieSceneScriptingFloatChannel}
         ]
        }]
      }
      {
       type=MovieSceneWidgetMaterialTrack
       display_name=Selected Cell Brush.Brush Material
       sections [{
         type=MovieSceneParameterSection
         channels [
          AnimateBorder {type=MovieSceneScriptingFloatChannel, const [0,1]}
          AnimateGlow {type=MovieSceneScriptingFloatChannel, keys [[0,0],[2,1]]}
         ]
        }]
      }
     ]
    }]}
 }
]
graphs [
 EventGraph {nodes [
   n1 = Event MaximizeSlot!() -> (OutputDelegate, then, bShouldSkipAnimation:bool) {is=title K2Node_Event}
   > n2 = Branch(Condition=@n1.bShouldSkipAnimation)
   > StopAnimation(InAnimation=Get GrowAnimation() -> (GrowAnimation:Widget Animation Object Reference))
   > SetAnimatedBoxScaleFactor!(execute, self, Value:double=1) -> (then)
   > OnMaximizedStarted()
   n3 = PlayAnimationForward(execute=@n2.else, InAnimation=Get GrowAnimation())
   > SetUserTag!(execute, Target:"Widget Animation Handle Structure (by ref)"=@n3, InUserTag:Name=Maximized) -> (then)
   > n4 = Cast To QuickbarSlotCooldown!(execute, Object=Get QuickbarSlotCooldown() -> (QuickbarSlotCooldown:Fort Item Cooldown Widget Object Reference)) -> (then, CastFailed, AsQuickbar Slot Cooldown:Quickbar Slot Cooldown Object Reference) {is=title K2Node_DynamicCast}
   > Show Countdown!(execute, self:Quickbar Slot Cooldown Object Reference=@n4, Show:bool=true) -> (then)
   n5 = Event MinimizeSlot!() -> (OutputDelegate, then, bShouldSkipAnimation:bool) {is=title K2Node_Event}
   > n6 = Branch(Condition=@n5.bShouldSkipAnimation)
   > StopAnimation(InAnimation=Get GrowAnimation())
   > SetAnimatedBoxScaleFactor(Value=0)
   > OnMinimizedFinished()
   n7 = PlayAnimationReverse(execute=@n6.else, InAnimation=Get GrowAnimation())
   > SetUserTag(Target=@n7, InUserTag=Minimize)
   > n8 = Cast To QuickbarSlotCooldown(Object=Get QuickbarSlotCooldown())
   > Show Countdown(self=@n8, Show=false)
   Event PreConstruct()
   > n9 = Branch(Condition=Get IsSecondaryQuickbar() -> (IsSecondaryQuickbar:bool))
   > n10 = SetSecondaryQuickbarVisuals()
   UpdateItemBoxSize(execute=[@n10.then, @n9.else])
   > UpdateKeyBindingText(self:Fort Quick Bar Slot Base Object Reference)
   > n11 = Switch on EFortQuickBars!(execute, Selection:EFortQuickBars Enum=Get QuickBarType() -> (QuickBarType:EFortQuickBars Enum=Primary)) -> (Primary:Exec, Secondary:Exec, Creative:Exec, Max_None:Exec) {is=title K2Node_SwitchEnum}
   SetVisibility(
    execute=[@n11.Creative, @n11.Primary]
    self=Get KeybindWidget!(self:Fort Quickbar Keybind Widget Object Reference=Get QuickbarKeybindWidgetTop() -> (QuickbarKeybindWidgetTop:Fort Quickbar Keybind Widget Object Reference)) -> (KeybindWidget:Fort Keybind Widget Object Reference)
    InVisibility=Hidden
   )
   SetVisibility(
    execute=@n11.Secondary
    self=Get KeybindWidget(self=Get QuickbarKeybindWidgetBottom() -> (QuickbarKeybindWidgetBottom:Fort Quickbar Keybind Widget Object Reference))
    InVisibility=Hidden
   )
   Event Construct()
   > InitializeInvalidation()
   > InitializeCooldowns()
   > RefreshItem(self:Fort Quick Bar Slot Base Object Reference)
   n12 = Event SetSlotSelected!() -> (OutputDelegate, then, bSelected:bool) {is=title K2Node_Event}
   > Set bIsSelected(bIsSelected:bool=@n12.bSelected) -> (Output_Get:bool)
   > SetIsSelected!(
    execute
    self:Athena Quick Bar Native Cell Object Reference=Get NativeCell() -> (NativeCell:Athena Quick Bar Native Cell Object Reference)
    bInSelected:bool=@n12.bSelected
   ) -> (then)
   > SetRenderTranslation(
    self=Get OverlayExtensionWidgets() -> (OverlayExtensionWidgets:Overlay Object Reference)
    Translation=Select(Option 0:Vector 2D Structure, Option 1:Vector 2D Structure=(X=0,Y=-13.5), Index=@n12.bSelected) -> (ReturnValue:Vector 2D Structure)
   )
   > UpdateKeyBindingText()
   > UpdateKeyBindingVisibility(self:Fort Quick Bar Slot Base Object Reference)
   > n13 = Branch(Condition=@n12.bSelected)
   > n14 = Branch(Condition=Get DoNotPlaySelectionAnimation() -> (DoNotPlaySelectionAnimation:bool))
   n15 = Select(
    Option 0:Widget Animation Object Reference=Get Selected() -> (Selected:Widget Animation Object Reference)
    Option 1:Widget Animation Object Reference=Get SelectedSecondary() -> (SelectedSecondary:Widget Animation Object Reference)
    Index=Get IsSecondaryQuickBar() -> (IsSecondaryQuickBar:bool)
   ) -> (ReturnValue:Widget Animation Object Reference)
   PlayAnimation(execute=@n14.else, InAnimation=@n15)
   n16 = Branch(execute=@n13.else, Condition=Get DoNotPlaySelectionAnimation())
   PlayAnimation(execute=@n16.else, InAnimation=@n15, StartAtTime=GetEndTime(self=@n15), PlayMode=Reverse)
   Event OnRefreshItem() {is=title K2Node_Event}
   > UpdateItemBoxSize()
   n17 = Event OnExtensionWidgetAdded!() -> (OutputDelegate, then, ExtensionSlot:EUIExtensionSlot Enum=Invalid, ExtensionWidget:User Widget Object Reference) {is=title K2Node_Event}
   > n18 = AddChildToOverlay!(execute, self:Overlay Object Reference=Get OverlayExtensionWidgets(), Content:Widget Object Reference=@n17.ExtensionWidget) -> (then, ReturnValue:Overlay Slot Object Reference)
   > SetHorizontalAlignment(self=@n18, InHorizontalAlignment=HAlign_Fill)
   > SetVerticalAlignment(self=@n18, InVerticalAlignment=VAlign_Fill)
   "Animation Finished Tag:Minimize (GrowAnimation)"!() -> (OutputDelegate:Delegate, then:Exec) {is=title K2Node_WidgetAnimationEvent, userTag=Minimize}
   > OnMinimizedFinished()
   "Animation Finished Tag:Maximized (GrowAnimation)"!() -> (OutputDelegate:Delegate, then:Exec) {is=title K2Node_WidgetAnimationEvent, userTag=Maximized}
   > OnMaximizedStarted()
  ]}
 InitializeCooldowns {nodes [
   InitializeCooldowns() {is=title K2Node_FunctionEntry}
   > n1 = Cast To QuickbarSlotCooldown(Object=Get QuickbarSlotCooldown())
   > Initialize!(
    execute
    self:Quickbar Slot Cooldown Object Reference=@n1
    QuickbarItemDisplayStyle:ItemDisplayStyle Enum=Get ItemDisplayStyle() -> (ItemDisplayStyle:ItemDisplayStyle Enum=NewEnumerator2)
    DesiredCooldownTypesSupported:Array of EFortItemCooldownType Enums=Get CooldownTypesSupported() -> (CooldownTypesSupported:Array of EFortItemCooldownType Enums=None)
   ) -> (then)
   > InitializeAthenaCooldown(self:Quickbar Slot Cooldown Object Reference=@n1)
   > Branch(Condition=Get bHideKeybindsWhenAbilityUnavailable() -> (bHideKeybindsWhenAbilityUnavailable:bool))
   n2 = Get QuickbarSlotCooldown()
   > Bind Event to OnCooldownStarted(self:Fort Item Cooldown Widget Object Reference=@n2, Delegate=Create Event()) {is=title K2Node_AddDelegate}
   > Bind Event to OnCooldownStopped(self:Fort Item Cooldown Widget Object Reference=@n2, Delegate=Create Event()) {is=title K2Node_AddDelegate}
  ]}
 InitializeInvalidation {nodes [
   InitializeInvalidation() {is=title K2Node_FunctionEntry}
   > Is Simplified HUD!(Exec:Exec) -> (Simplified:Exec, Default:Exec) {is=title K2Node_MacroInstance}
   > Set DoNotPlaySelectionAnimation(DoNotPlaySelectionAnimation:bool=true) -> (Output_Get:bool)
  ]}
 OnCooldownStarted {nodes [
   n1 = OnCooldownStarted!() -> (then, CooldownType:EFortItemCooldownType Enum=None, CooldownDuration:float=~) {is=title K2Node_FunctionEntry}
   > n2 = Switch on EFortItemCooldownType!(execute, Selection:EFortItemCooldownType Enum=@n1) -> (
    None:Exec
    AmmoRegeneration:Exec
    ItemActivation:Exec
    WeaponReloading:Exec
    Death:Exec
    AthenaWeaponFireCooldown:Exec
    AbilitySetActivateByInputAbility:Exec
    DirectProgressUpdate:Exec
   ) {is=title K2Node_SwitchEnum}
   Set bKeybindsHidden(execute=[@n2.Death, @n2.ItemActivation], bKeybindsHidden:bool=true) -> (Output_Get:bool)
   > UpdateKeyBindingVisibility()
  ]}
 OnCooldownStopped {nodes [
   n1 = OnCooldownStopped!() -> (then, CooldownType:EFortItemCooldownType Enum=None) {is=title K2Node_FunctionEntry}
   > n2 = Switch on EFortItemCooldownType(Selection=@n1)
   Set bKeybindsHidden(execute=[@n2.Death, @n2.ItemActivation], bKeybindsHidden=false)
   > UpdateKeyBindingVisibility()
  ]}
 OnMaximizedStarted {nodes [
   OnMaximizedStarted() {is=title K2Node_FunctionEntry}
   > Resize!(execute, self:Fort Quick Bar Slot Base Object Reference, CardSize:EFortItemCardSize Enum=S) -> (then)
  ]}
 OnMinimizedFinished {nodes [
   OnMinimizedFinished() {is=title K2Node_FunctionEntry}
   > Resize(CardSize=XS)
  ]}
 SetAnimatedBoxScaleFactor {nodes [
   n1 = SetAnimatedBoxScaleFactor!() -> (then, Value:double=~) {is=title K2Node_FunctionEntry}
   > Set AnimatedBoxScaleFactor(AnimatedBoxScaleFactor:double=@n1) -> (Output_Get:double)
   > UpdateItemBoxSize()
  ]}
 SetSecondaryQuickbarVisuals {nodes [
   SetSecondaryQuickbarVisuals() {is=title K2Node_FunctionEntry}
   n1 = Get SelectedSecondary()
   > PlayAnimation(InAnimation=@n1, StartAtTime=GetEndTime(self=@n1), PlayMode=Reverse)
   > SetVisibility(
    self=[Get PaddingLeft() -> (PaddingLeft:Spacer Object Reference), Get PaddingRight() -> (PaddingRight:Spacer Object Reference)]
    InVisibility=Collapsed
   )
  ]}
 UpdateItemBoxSize {nodes [
   UpdateItemBoxSize() {is=title K2Node_FunctionEntry}
   > n1 = Branch(Condition=Get bUseSlotSizeBySlotUsage() -> (bUseSlotSizeBySlotUsage:bool))
   n2 = GetSlotSizeForItem!(self:Fort Quick Bar Slot Base Object Reference, InItem:Fort Item Object Reference=Get Item() -> (Item:Fort Item Object Reference)) -> (ReturnValue:Vector 2D Structure)
   n3 = BreakVector2D(InVec=@n2)
   n4 = BreakVector2D(InVec=vector2d * float!(A:Vector 2D Structure=@n2, B:double=0.7) -> (ReturnValue:Vector 2D Structure))
   n5 = Get AnimatedBoxScaleFactor() -> (AnimatedBoxScaleFactor:double)
   n6 = Lerp(A=@n4, B=@n3, Alpha=@n5)
   n7 = Lerp(A=@n4.Y, B=@n3.Y, Alpha=@n5)
   > DesiredSizeOverride!(execute, self:Athena Quick Bar Native Cell Object Reference=Get NativeCell(), WidthOverride:float=@n6, HeightOverride:float=@n7) -> (then)
   n8 = Get SizeBox_ExtensionOverlay() -> (SizeBox_ExtensionOverlay:Size Box Object Reference)
   > SetHeightOverride(self=@n8, InHeightOverride=@n7)
   > SetWidthOverride(self=@n8, InWidthOverride=@n6)
   n9 = Lerp(A=64, B=92, Alpha=@n5)
   DesiredSizeOverride(execute=@n1.else, self=Get NativeCell(), WidthOverride=@n9, HeightOverride=@n9)
   n10 = Get SizeBox_ExtensionOverlay()
   > SetHeightOverride(self=@n10, InHeightOverride=@n9)
   > SetWidthOverride(self=@n10, InWidthOverride=@n9)
  ]}
]
variables [
 AnimatedBoxScaleFactor {type=number, default=1}
 CooldownTypesSupported {type="array<EFortItemCooldownType>", default []}
 IsSecondaryQuickbar {type=boolean}
 ItemDisplayStyle {type=ItemDisplayStyle, default=World}
]
widgets {
 info {inheritedWidgetCount=13, parentClass=/Script/FortniteUI.AthenaQuickBarSlot, rootWidgetClass=&HorizontalBox, widgetCount=24}
 widgets [
  HorizontalBox_0 {widgetClassPath=&HorizontalBox, widget=HorizontalBox_0}
  PaddingLeft {
   widgetClassPath=&Spacer
   bIsVariable=true
   parent=HorizontalBox_0
   widget=PaddingLeft
   properties {size=<2 1>}
   slot {class=HorizontalBoxSlot}
  }
  VerticalBox_0 {widgetClassPath=&VerticalBox, parent=HorizontalBox_0, widget=VerticalBox_0, slot {class=HorizontalBoxSlot}}
  QuickbarKeybindWidgetTop {
   widgetClassPath=&/QuickbarKeybindTop
   bInherited=true
   bIsVariable=true
   parent=VerticalBox_0
   widget=QuickbarKeybindWidgetTop
   slot {class=VerticalBoxSlot, properties {horizontalAlignment=HAlign_Center, padding=<ltrb 0 0 0 16>}}
  }
  Overlay_0 {
   widgetClassPath=&Overlay
   parent=VerticalBox_0
   widget=Overlay_0
   slot {class=VerticalBoxSlot, properties {horizontalAlignment=HAlign_Center, verticalAlignment=VAlign_Center}}
  }
  Slot_BehindImage {
   widgetClassPath=&NamedSlot
   bIsVariable=true
   parent=Overlay_0
   widget=Slot_BehindImage
   properties {bExposeOnInstanceOnly=true, visibility=Collapsed}
   slot {class=OverlaySlot, properties {horizontalAlignment=HAlign_Fill, verticalAlignment=VAlign_Fill}}
  }
  NativeCell {
   widgetClassPath=/Script/FortniteUI.AthenaQuickBarNativeCell
   bInherited=true
   parent=Overlay_0
   widget=NativeCell
   properties {
    bSupportsWideItems=true
    selectionAnimLength_Seconds=0.125
    ammoTypeIcon_Padding=<8 8>
    ammoTypeIcon_Size=<18 18>
    emptyCell_Brush {
     bIsDynamicallyLoaded=false
     drawAs=Box
     imageType=NoImage
     mirroring=NoMirror
     resourceName=None
     resourceObject=/Game/Athena/HUD/Materials/M_UI_HUD_Inventory_Selection_Indicator_Empty
     tiling=NoTile
     imageSize=<80 80>
     margin=<ltrb 0.5 0.5 0.5 0.5>
     outlineSettings {
      bUseBrushTransparency=false
      roundingType=HalfHeightRadius
      width=0
      color {colorUseRule=UseColor_Specified, specifiedColor=<rgba 0 0 0 0>}
      cornerRadii=<0 0 0 0>
     }
     tintColor {colorUseRule=UseColor_Specified, specifiedColor=<rgba 1 1 1 1>}
     uVRegion {bIsValid=false, max=<0 0>, min=<0 0>}
    }
    emptyCell_Padding=<0 0>
    gadgetFuel_Brush {
     bIsDynamicallyLoaded=false
     drawAs=Image
     imageType=NoImage
     mirroring=NoMirror
     resourceName=None
     resourceObject=/Game/Athena/UI/InGame/Jetpack/M_ProgressBar_FuelBar
     tiling=NoTile
     imageSize=<9 32>
     margin=<ltrb 0.49 0.49 0.49 0.49>
     outlineSettings {
      bUseBrushTransparency=false
      roundingType=HalfHeightRadius
      width=0
      color {colorUseRule=UseColor_Specified, specifiedColor=<rgba 0 0 0 0>}
      cornerRadii=<0 0 0 0>
     }
     tintColor {colorUseRule=UseColor_Specified, specifiedColor=<rgba 1 1 1 1>}
     uVRegion {bIsValid=false, max=<0 0>, min=<0 0>}
    }
    gadgetFuel_Padding=<4 4>
    rechargeableIcon_Brush {
     bIsDynamicallyLoaded=false
     drawAs=NoDrawType
     imageType=NoImage
     mirroring=NoMirror
     resourceName=None
     resourceObject=/Game/UI/Foundation/Textures/Icons/Ammo/T-Icon-RechargeAmmo-32
     tiling=NoTile
     imageSize=<16 16>
     margin=<ltrb 0 0 0 0>
     outlineSettings {
      bUseBrushTransparency=false
      roundingType=HalfHeightRadius
      width=0
      color {colorUseRule=UseColor_Specified, specifiedColor=<rgba 0 0 0 0>}
      cornerRadii=<0 0 0 0>
     }
     tintColor {colorUseRule=UseColor_Specified, specifiedColor=<rgba 1 1 1 1>}
     uVRegion {bIsValid=false, max=<0 0>, min=<0 0>}
    }
    rechargeableIcon_Padding=<0 8>
    secondaryAmmoTypeIconBuildMode_Padding=<10 45>
    secondaryAmmoTypeIcon_Padding=<20 70>
    secondaryAmmoTypeIcon_Size=<18 18>
    secondaryAmmoTypeText_Padding=<8 5>
    selectedCell_Brush {
     bIsDynamicallyLoaded=false
     drawAs=Box
     imageType=NoImage
     mirroring=NoMirror
     resourceName=None
     resourceObject=/Game/Athena/HUD/Materials/M_UI_HUD_Inventory_Selection_Indicator_Base
     tiling=NoTile
     imageSize=<80 80>
     margin=<ltrb 0.5 0.5 0.5 0.5>
     outlineSettings {
      bUseBrushTransparency=false
      roundingType=HalfHeightRadius
      width=0
      color {colorUseRule=UseColor_Specified, specifiedColor=<rgba 0 0 0 0>}
      cornerRadii=<0 0 0 0>
     }
     tintColor {colorUseRule=UseColor_Specified, specifiedColor=<rgba 1 1 1 1>}
     uVRegion {bIsValid=false, max=<0 0>, min=<0 0>}
    }
    selectionOffset=<0 -6>
   }
   slot {class=OverlaySlot}
  }
  Overlay_2 {widgetClassPath=&Overlay, parent=NativeCell, widget=Overlay_2, slot {class=PanelSlot}}
  SizeBox_5 {
   widgetClassPath=&SizeBox
   parent=Overlay_2
   widget=SizeBox_5
   properties {bOverride_HeightOverride=true, bOverride_WidthOverride=true, heightOverride=92, widthOverride=92}
   slot {class=OverlaySlot, properties {horizontalAlignment=HAlign_Fill, verticalAlignment=VAlign_Fill}}
  }
  QuickbarSlotCooldown {
   widgetClassPath=&/QuickbarSlotCooldown
   bInherited=true
   bIsVariable=true
   parent=SizeBox_5
   widget=QuickbarSlotCooldown
   properties {visibility=HitTestInvisible}
   slot {class=SizeBoxSlot}
  }
  SizeBox_69 {
   widgetClassPath=&SizeBox
   parent=Overlay_2
   widget=SizeBox_69
   properties {bOverride_HeightOverride=true, bOverride_WidthOverride=true, heightOverride=96, widthOverride=96}
   slot {class=OverlaySlot, properties {horizontalAlignment=HAlign_Fill, verticalAlignment=VAlign_Fill}}
  }
  QuickbarSlotDurability {
   widgetClassPath=&/QuickbarSlotDurability
   bInherited=true
   bIsVariable=true
   parent=SizeBox_69
   widget=QuickbarSlotDurability
   properties {visibility=SelfHitTestInvisible}
   slot {class=SizeBoxSlot}
  }
  Image_Disabled {
   widgetClassPath=&Image
   bInherited=true
   bIsVariable=true
   parent=Overlay_2
   widget=Image_Disabled
   properties {
    visibility=Hidden
    brush {
     bIsDynamicallyLoaded=false
     drawAs=Image
     imageType=NoImage
     mirroring=NoMirror
     resourceName=None
     resourceObject=null
     tiling=NoTile
     imageSize=<32 32>
     margin=<ltrb 0 0 0 0>
     outlineSettings {
      bUseBrushTransparency=false
      roundingType=HalfHeightRadius
      width=0
      color {colorUseRule=UseColor_Specified, specifiedColor=<rgba 0 0 0 0>}
      cornerRadii=<0 0 0 0>
     }
     tintColor {colorUseRule=UseColor_Specified, specifiedColor=<rgba 0 0 0 0.4>}
     uVRegion {bIsValid=false, max=<0 0>, min=<0 0>}
    }
   }
   slot {class=OverlaySlot, properties {horizontalAlignment=HAlign_Fill, verticalAlignment=VAlign_Fill}}
  }
  SizeBox_ExtensionOverlay {
   widgetClassPath=&SizeBox
   bIsVariable=true
   parent=Overlay_0
   widget=SizeBox_ExtensionOverlay
   slot {class=OverlaySlot, properties {horizontalAlignment=HAlign_Fill, verticalAlignment=VAlign_Fill}}
  }
  OverlayExtensionWidgets {
   widgetClassPath=&Overlay
   bIsVariable=true
   parent=SizeBox_ExtensionOverlay
   widget=OverlayExtensionWidgets
   slot {class=SizeBoxSlot}
  }
  TutorialHighlight {
   widgetClassPath=&TutorialHighlight
   parent=Overlay_0
   widget=TutorialHighlight
   properties {customHighlightWidget=/Game/Athena/HUD/Tutorial/Tutorial_Outline_Square.Tutorial_Outline_Square_C}
   slot {class=OverlaySlot, properties {horizontalAlignment=HAlign_Fill, verticalAlignment=VAlign_Fill}}
  }
  QuickbarKeybindWidgetBottom {
   widgetClassPath=&/QuickbarKeybindBottom
   bInherited=true
   bIsVariable=true
   parent=VerticalBox_0
   widget=QuickbarKeybindWidgetBottom
   slot {class=VerticalBoxSlot, properties {horizontalAlignment=HAlign_Center, padding=<ltrb 0 4 0 0>}}
  }
  PaddingRight {
   widgetClassPath=&Spacer
   bIsVariable=true
   parent=HorizontalBox_0
   widget=PaddingRight
   properties {size=<2 1>}
   slot {class=HorizontalBoxSlot}
  }
  SizeBox_BottomKeybind {widgetClassPath=&SizeBox, bInherited=true, widget=null}
  SizeBox_TopKeybind {widgetClassPath=&SizeBox, bInherited=true, widget=null}
  EmptyImage {widgetClassPath=&Image, bInherited=true, widget=null}
  ItemCardMaximized {widgetClassPath=/Script/FortniteUI.FortMultiSizeItemCard, bInherited=true, widget=null}
  ItemCardMinimized {widgetClassPath=&FortMultiSizeItemCard, bInherited=true, widget=null}
  SlotInteraction {widgetClassPath=&Image, bInherited=true, widget=null}
 ]
}
