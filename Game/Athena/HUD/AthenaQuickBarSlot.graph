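# Athena HUD quick bar slot widget, extending AthenaQuickbarSlotBase with inventory refresh and touch-based item selection.
# On Construct it binds to the inventory context's quickbar-contents-changed event and refreshes the item if its slot changed.
# On touch platforms, when enabled by a flag, a touch either selects the slot directly or passes selection to the HUD context, chosen by the Fort.Rollback.IndependentTapAnywherePrimaryAction console variable.
# Both paths request reload-or-use if already selected.
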WidgetBlueprint /Game/Athena/HUD/AthenaQuickBarSlot
graphs [
 EventGraph {nodes [
   n1 = Handle OnQuickbarContentsChanged!() -> (OutputDelegate, then, QuickbarIndex:EFortQuickBars Enum=Primary, ChangedSlots:Array of Integers) {is=title K2Node_CustomEvent}
   > RefreshItemIfSlotChanged!(
    execute
    self:Fort Quick Bar Slot Base Object Reference
    InQuickBar:EFortQuickBars Enum=@n1.QuickbarIndex
    InSlot:Array of Integers=@n1.ChangedSlots
   ) -> (then)
   Event Construct()
   > Parent: Construct()
   > Bind Event to On Quickbar Contents Changed(
    self:Fort Inventory Context Object Reference=Get FortInventoryContext!() -> (ReturnValue:Fort Inventory Context Object Reference) {is=title K2Node_GetBlueprintContextEx}
    Delegate=@n1
   ) {is=title K2Node_AddDelegate}
  ]}
 OnTouchStarted {nodes [
   OnTouchStarted!() -> (then, MyGeometry:Geometry Structure, InTouchEvent:"Pointer Event Structure (by ref)") {is=title K2Node_FunctionEntry}
   > n1 = Branch(Condition=AND Boolean(A=IsTouchInputPlatform!() -> (ReturnValue:bool), B=Get bShouldTakeItemSelectionInputFromHUD() -> (bShouldTakeItemSelectionInputFromHUD:bool)))
   > n2 = GetConsoleVariableBoolValue!(execute, VariableName:String=Fort.Rollback.IndependentTapAnywherePrimaryAction) -> (then, ReturnValue:bool)
   > n3 = Branch(Condition=@n2)
   > StartSelectSlot!(execute, self:Athena Quick Bar Slot Object Reference, bReloadOrUseIfAlreadySelected:bool=true, bStopUseImmediately:bool=true) -> (then)
   > Return Node!(execute, ReturnValue:Event Reply Structure=Handled())
   TakeItemSelectionInputFromHUD!(
    execute=@n3.else
    self:Athena HUDContext Object Reference=Get AthenaHUDContext!() -> (ReturnValue:Athena HUDContext Object Reference) {is=title K2Node_GetBlueprintContextEx}
    TouchIndex:int=Get Pointer Index!(Input:"Pointer Event Structure (by ref)"=Get InTouchEvent() -> (InTouchEvent:"Pointer Event Structure (by ref)")) -> (ReturnValue:int=0) {is=title K2Node_CallFunction}
    SlotIdx:int=Get SlotIndex() -> (SlotIndex:int=0)
    InQuickBar:EFortQuickBars Enum=Get QuickBarType() -> (QuickBarType:EFortQuickBars Enum=Primary)
    bReloadOrUseIfAlreadySelected:bool=true
    bStopUseImmediately:bool=true
   ) -> (then, bIsUsing:bool)
   > Return Node(ReturnValue=Handled())
   Return Node(execute=@n1.else, ReturnValue=Unhandled())
  ]}
]
variables [bShouldTakeItemSelectionInputFromHUD {type=boolean, default=true}]
widgets {
 info {inheritedWidgetCount=0, parentClass=/Game/UI/InGame/HUD/AthenaQuickbarSlotBase.AthenaQuickbarSlotBase_C, rootWidgetClass=null, widgetCount=0}
 widgets []
}
