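# Athena HUD primary quickbar widget, with a standalone first slot and a row of five more slots.
# Construct gathers the slots, binds inventory, input and interaction events, and reads whether the playlist hides the quickbar.
# An interacted pickup drives a preview frame, positioned from slot geometry and sized by slots taken.
# Contents changes can add or remove extra slots at runtime; every weapon rail update passes Show=false, keeping it hidden.
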WidgetBlueprint /Game/Athena/HUD/AthenaQuickbarPrimary
animations [
 In {
  binding_count=4
  length_seconds=1
  object=&self:In
  sequence {bindings [
    Border_Highlight_1 {
     id=#1
     possessed_class=Border
     tracks [{
       type=MovieSceneColorTrack
       display_name=Brush Color
       sections [{
         type=MovieSceneColorSection
         end_frame=20
         start_frame=0
         channels [
          Color.R {type=MovieSceneScriptingFloatChannel, const [1,1]}
          Color.G {type=MovieSceneScriptingFloatChannel, const [1,1]}
          Color.B {type=MovieSceneScriptingFloatChannel, const [1,1]}
          Color.A {type=MovieSceneScriptingFloatChannel, keys [[3.9997,0],[19.9993,1],[19.9997,0.5]]}
         ]
        }]
      }]
    }
    Border_Highlight_2 {
     id=#2
     possessed_class=Border
     tracks [
      {
       type=MovieSceneFloatVectorTrack
       display_name=Desired Size Scale
       sections [{
         type=MovieSceneFloatVectorSection
         end_frame=20
         start_frame=0
         channels [
          Vector.X {type=MovieSceneScriptingFloatChannel, keys [[3.9997,2],[19.9993,1]]}
          Vector.Y {type=MovieSceneScriptingFloatChannel, keys [[3.9997,2],[19.9993,1]]}
         ]
        }]
      }
      {
       type=MovieSceneColorTrack
       display_name=Brush Color
       sections [{
         type=MovieSceneColorSection
         end_frame=20
         start_frame=0
         channels [
          Color.R {type=MovieSceneScriptingFloatChannel}
          Color.G {type=MovieSceneScriptingFloatChannel}
          Color.B {type=MovieSceneScriptingFloatChannel}
          Color.A {type=MovieSceneScriptingFloatChannel, keys [[3.9997,0],[19.9997,0.2]]}
         ]
        }]
      }
      {
       type=MovieSceneByteTrack
       display_name=Visibility
       sections [{type=MovieSceneByteSection, end_frame=19, start_frame=0, channels [None {type=MovieSceneScriptingByteChannel, const [0,1]}]}]
      }
     ]
    }
    ball {
     id=#3
     possessed_class=Image
     tracks [
      {
       type=MovieSceneByteTrack
       display_name=Visibility
       sections [{type=MovieSceneByteSection, end_frame=19, start_frame=0, channels [None {type=MovieSceneScriptingByteChannel, const [0,1]}]}]
      }
      {
       type=MovieSceneColorTrack
       display_name=Color and Opacity
       sections [{
         type=MovieSceneColorSection
         end_frame=19
         start_frame=0
         channels [
          Color.R {type=MovieSceneScriptingFloatChannel, const [1,1]}
          Color.G {type=MovieSceneScriptingFloatChannel, const [1,1]}
          Color.B {type=MovieSceneScriptingFloatChannel, const [1,1]}
          Color.A {type=MovieSceneScriptingFloatChannel, keys [[5,0],[8,0.4],[17,0]]}
         ]
        }]
      }
      {
       type=MovieScene2DTransformTrack
       sections [{
         type=MovieScene2DTransformSection
         end_frame=19
         start_frame=0
         channels [
          Translation.X {type=MovieSceneScriptingFloatChannel}
          Translation.Y {type=MovieSceneScriptingFloatChannel}
          Angle {type=MovieSceneScriptingFloatChannel}
          Scale.X {type=MovieSceneScriptingFloatChannel, keys [[8,1],[17,2.4]]}
          Scale.Y {type=MovieSceneScriptingFloatChannel, keys [[8,1],[17,2.4]]}
          Shear.X {type=MovieSceneScriptingFloatChannel}
          Shear.Y {type=MovieSceneScriptingFloatChannel}
         ]
        }]
      }
     ]
    }
   ]}
 }
 Loop {
  binding_count=3
  length_seconds=1
  object=&self:Loop
  sequence {bindings [
    Border_Highlight_1 {
     id=#4
     possessed_class=Border
     tracks [{
       type=MovieSceneColorTrack
       display_name=Brush Color
       sections [{
         type=MovieSceneColorSection
         end_frame=20
         start_frame=0
         channels [
          Color.R {type=MovieSceneScriptingFloatChannel, const [1,1]}
          Color.G {type=MovieSceneScriptingFloatChannel, const [1,1]}
          Color.B {type=MovieSceneScriptingFloatChannel, const [1,1]}
          Color.A {type=MovieSceneScriptingFloatChannel, keys [[0,1],[8,0.2],[16,0.2],[19.9993,1]]}
         ]
        }]
      }]
    }
    Border_Highlight_2 {
     id=#5
     possessed_class=Border
     tracks [
      {
       type=MovieSceneFloatVectorTrack
       display_name=Desired Size Scale
       sections [{
         type=MovieSceneFloatVectorSection
         end_frame=20
         start_frame=0
         channels [
          Vector.X {type=MovieSceneScriptingFloatChannel, keys [[3.9997,2],[19.9993,1]]}
          Vector.Y {type=MovieSceneScriptingFloatChannel, keys [[3.9997,2],[19.9993,1]]}
         ]
        }]
      }
      {
       type=MovieSceneColorTrack
       display_name=Brush Color
       sections [{
         type=MovieSceneColorSection
         end_frame=20
         start_frame=0
         channels [
          Color.R {type=MovieSceneScriptingFloatChannel}
          Color.G {type=MovieSceneScriptingFloatChannel}
          Color.B {type=MovieSceneScriptingFloatChannel}
          Color.A {type=MovieSceneScriptingFloatChannel, keys [[3.9997,0],[19.9997,0.2]]}
         ]
        }]
      }
      {
       type=MovieSceneByteTrack
       display_name=Visibility
       sections [{type=MovieSceneByteSection, end_frame=19, start_frame=0, channels [None {type=MovieSceneScriptingByteChannel, const [0,1]}]}]
      }
     ]
    }
    ball {
     id=#6
     possessed_class=Image
     tracks [{
       type=MovieSceneByteTrack
       display_name=Visibility
       sections [{type=MovieSceneByteSection, end_frame=19, start_frame=0, channels [None {type=MovieSceneScriptingByteChannel, const [1,1]}]}]
      }]
    }
   ]}
 }
]
graphs [
 EventGraph {nodes [
   HandleInputMethodChanged_Bind!() -> (OutputDelegate, then, InputType:ECommonInputType Enum) {is=title K2Node_CustomEvent}
   > HandleShowHideWeaponRail()
   OnCreativeSwapPhoneHoldChanged!() -> (OutputDelegate, then, bIsSwapped:bool=~) {is=title K2Node_CustomEvent}
   > UpdateKeyBindingText(self:Fort Quick Bar Slot Base Object Reference=Get QuickbarSlot1() -> (QuickbarSlot1:Athena Quick Bar Slot Object Reference))
   OnQuickBarSlotFocusChanged_Bind!() -> (OutputDelegate, then, QuickbarIndex:EFortQuickBars Enum=Primary, Slot:int) {is=title K2Node_CustomEvent}
   > RefreshPickupPreview()
   Event HandleKeybindsChanged() {is=title K2Node_Event}
   > Parent: HandleKeybindsChanged() {is=title K2Node_CallParentFunction}
   Event Destruct()
   n1 = Get CommonInputSubsystem()
   > Is Valid(InputObject=@n1)
   > Unbind Event from On Input Method Changed(self=@n1, Delegate=Create Event())
   > Parent: Destruct()
   Event Maximize() {is=title K2Node_Event}
   > Parent: Maximize() {is=title K2Node_CallParentFunction}
   > HandleShowHideWeaponRail()
   > SetVisibility(self=Get PickupPreview() -> (PickupPreview:Size Box Object Reference), InVisibility=Hidden)
   > Delay(Duration=0)
   > Delay(Duration=0)
   > RefreshPickupPreview()
   Event Minimize() {is=title K2Node_Event}
   > Parent: Minimize() {is=title K2Node_CallParentFunction}
   > Show Primary Quickbar Rail!(execute, self, Show:bool=false) -> (then)
   > SetVisibility(self=Get PickupPreview(), InVisibility=Hidden)
   > Delay(Duration=0)
   > Delay(Duration=0)
   > RefreshPickupPreview()
   n2 = Event OnQuickbarContentsChanged!() -> (OutputDelegate, then, QuickbarIndex:EFortQuickBars Enum=Primary, ChangedSlots:Array of Integers) {is=title K2Node_Event}
   > Parent: OnQuickbarContentsChanged!(execute, QuickbarIndex:EFortQuickBars Enum=@n2.QuickbarIndex, ChangedSlots:Array of Integers=@n2.ChangedSlots) -> (then) {is=title K2Node_CallParentFunction}
   > HandleShowHideWeaponRail()
   > RefreshPickupPreview()
   Event Construct()
   > n3 = Sequence!(execute) -> (then_0, then_1, then_2:Exec, then_3:Exec, then_4:Exec, then_5:Exec, then_6:Exec, then_7:Exec, then_8:Exec)
   Parent: Construct(execute=@n3.then_1)
   n4 = Create Event()
   n5 = Get FortHUDContext!() -> (ReturnValue:Fort HUDContext Object Reference) {is=title K2Node_GetBlueprintContextEx}
   Bind Event to On Player Can Interact(execute=@n3.then_6, self:Fort HUDContext Object Reference=@n5, Delegate=@n4) {is=title K2Node_AddDelegate}
   n6 = PlaylistDataReadyAsync!(execute:Exec=@n3.then_8) -> (
    then:Exec
    OnReady:Exec
    GameState:Fort Game State Athena Object Reference
    Playlist:Fort Playlist Object Reference
    PlaylistContextTags:"Gameplay Tag Container Structure (by ref)"
   ) {is=title K2Node_AsyncAction}
   Set bPlaylistHidesQuickbar(
    execute=@n6.OnReady
    bPlaylistHidesQuickbar:bool=Compare Tag Container to Other Tags!(
     TagCase_0:"Gameplay Tag Structure (by ref)"=Get QuickbarTag() -> (QuickbarTag:Gameplay Tag Structure)
     Gameplay Tag Container:"Gameplay Tag Container Structure (by ref)"=Get HUDElementsToHide!(self:Fort Playlist Object Reference=@n6.Playlist) -> (HUDElementsToHide:Gameplay Tag Container Structure)
    ) -> (Case_0 True:bool=~) {is=title GameplayTagsK2Node_MultiCompareGameplayTagContainerSingleTags}
   ) -> (Output_Get:bool)
   n7 = Set QuickbarSlots(
    execute=@n3.then_0
    QuickbarSlots:Array of Fort Quick Bar Slot Base Object References=Make Array!(
     "[0]":Fort Quick Bar Slot Base Object Reference=Get QuickbarSlot1()
     "[1]":Fort Quick Bar Slot Base Object Reference=Get QuickbarSlot2() -> (QuickbarSlot2:Athena Quick Bar Slot Object Reference)
     "[2]":Fort Quick Bar Slot Base Object Reference=Get QuickbarSlot3() -> (QuickbarSlot3:Athena Quick Bar Slot Object Reference)
     "[3]":Fort Quick Bar Slot Base Object Reference=Get QuickbarSlot4() -> (QuickbarSlot4:Athena Quick Bar Slot Object Reference)
     "[4]":Fort Quick Bar Slot Base Object Reference=Get QuickbarSlot5() -> (QuickbarSlot5:Athena Quick Bar Slot Object Reference)
     "[5]":Fort Quick Bar Slot Base Object Reference=Get QuickbarSlot6() -> (QuickbarSlot6:Athena Quick Bar Slot Object Reference)
    ) -> (Array:Array of Fort Quick Bar Slot Base Object References)
   ) -> (Output_Get:Array of Fort Quick Bar Slot Base Object References)
   > Set NonPickaxeSlotsNum(NonPickaxeSlotsNum:int=int - int(A=Length(TargetArray:Array of Fort Quick Bar Slot Base Object References=@n7))) -> (Output_Get:int=0)
   n8 = Cast To FortPlayerControllerGameplay!(execute=@n3.then_4, Object=GetOwningPlayer()) -> (then, CastFailed, AsFort Player Controller Gameplay:Fort Player Controller Gameplay Object Reference) {is=title K2Node_DynamicCast}
   > Bind Event to On Creative Swap Phone and Harvest Hold Changed Delegate(self:Fort Player Controller Gameplay Object Reference=@n8, Delegate=Create Event()) {is=title K2Node_AddDelegate}
   Bind Event to On Interact Updated(execute=@n3.then_7, self:Fort HUDContext Object Reference=@n5, Delegate=@n4) {is=title K2Node_AddDelegate}
   Bind Event to On Input Method Changed(execute=@n3.then_3, self=Get CommonInputSubsystem(), Delegate=Create Event())
   n9 = GetQuickbarFocus!(self:Fort Inventory Context Object Reference=Get FortInventoryContext!() -> (ReturnValue:Fort Inventory Context Object Reference) {is=title K2Node_GetBlueprintContextEx}) -> (OutQuickBar:EFortQuickBars Enum=Primary, OutSlot:int=0, OutSecondarySlot:int=0, OutPreviousFocusedSlot:int=0)
   OnQuickbarSlotFocusChanged!(execute=@n3.then_2, self:Fort Quickbar Base Object Reference, QuickbarIndex:EFortQuickBars Enum=@n9, SlotIndex:int=@n9.OutSlot) -> (then)
   n10 = Get FortInventoryContext()
   Assign On Quickbar Slot Focus Changed!(execute:Exec=@n3.then_5, self:Fort Inventory Context Object Reference=@n10, Delegate:"Delegate (by ref)"=Create Event()) -> (then:Exec) {is=title K2Node_AssignDelegate}
   > Bind Event to On Quickbar Contents Changed(self:Fort Inventory Context Object Reference=@n10, Delegate=Create Event()) {is=title K2Node_AddDelegate}
  ]}
 GetSpecificSlot {nodes [
   n1 = GetSpecificSlot!() -> (then, SlotNumber:int=0) {is=title K2Node_FunctionEntry}
   n2 = "Get (a copy)"(
    Array:Array of Fort Quick Bar Slot Base Object References=Get QuickbarSlots() -> (QuickbarSlots:Array of Fort Quick Bar Slot Base Object References)
    Dimension 1=@n1
   ) -> (Output:Fort Quick Bar Slot Base Object Reference)
   > Branch(Condition=IsValid(Object=@n2))
   > Return Node!(execute, ReturnValue:Fort Quick Bar Slot Base Object Reference=@n2)
  ]}
 HandleInteractionChanged {nodes [
   n1 = HandleInteractionChanged!() -> (then, Context Info:Fort Interact Context Info Object Reference) {is=title K2Node_FunctionEntry}
   > n2 = Is Valid(InputObject=@n1)
   > n3 = Cast To FortPickup!(execute, Object=Get ReceivingActor!(self:Fort Interact Context Info Object Reference=@n1) -> (ReceivingActor:Actor Weak object)) -> (then, CastFailed, AsFort Pickup:Fort Pickup Object Reference) {is=title K2Node_DynamicCast}
   > n4 = Set CurrentPickupPreviewItem(CurrentPickupPreviewItem:Fort Item Object Reference=Get PrimaryPickupDummyItem!(self:Fort Pickup Object Reference=@n3) -> (PrimaryPickupDummyItem:Fort Item Object Reference)) -> (Output_Get:Fort Item Object Reference)
   RefreshPickupPreview(execute=[@n5.then, @n4.then])
   n5 = Set CurrentPickupPreviewItem(execute=[@n3.CastFailed, @n2.Is Not Valid])
  ]}
 HandleShowHideWeaponRail {nodes [
   HandleShowHideWeaponRail() {is=title K2Node_FunctionEntry}
   n1 = Get CommonInputSubsystem!() -> (ReturnValue:Common Input Object Reference) {is=title K2Node_GetBlueprintContextEx}
   > n2 = Branch(Condition="Equal (Enum)"(A=GetCurrentInputType(self=@n1), B=Gamepad))
   > n3 = Branch(Condition=HasOneNonHarvestWeapon!() -> (HasNonHarvestWeapon:bool))
   > Show Primary Quickbar Rail(Show=false)
   Show Primary Quickbar Rail(execute=[@n3.else, @n2.else], Show=false)
  ]}
 HasOneNonHarvestWeapon {nodes [
   HasOneNonHarvestWeapon() {is=title K2Node_FunctionEntry}
   > n1 = For Loop(FirstIndex=1, LastIndex=5)
   Return Node!(execute=@n1.Completed, HasNonHarvestWeapon:bool=false)
   n2 = Get QuickbarSlots()
   > Is Valid(InputObject=Get Item!(self:Fort Quick Bar Slot Base Object Reference="Get (a copy)"(Array=@n2, Dimension 1=@n1)) -> (Item:Fort Item Object Reference))
   > Return Node(HasNonHarvestWeapon=true)
  ]}
 On Quickbar Contents Changed {nodes [
   n1 = On Quickbar Contents Changed!() -> (then, QuickbarIndex:EFortQuickBars Enum=Primary, ChangedSlots:Array of Integers) {is=title K2Node_FunctionEntry}
   > Switch on EFortQuickBars!(execute, Selection:EFortQuickBars Enum=@n1) -> (Primary:Exec, Secondary:Exec, Creative:Exec, Max_None:Exec) {is=title K2Node_SwitchEnum}
   > n2 = For Each Loop(Array:Array of Integers=@n1.ChangedSlots) -> (Array Element:int)
   > n3 = Set Current Slot(Current Slot:int=@n2) -> (Output_Get:int=0)
   > Branch(Condition="integer > integer"(A=@n3, B=Get NonPickaxeSlotsNum() -> (NonPickaxeSlotsNum:int=0)))
   n4 = Get FortInventoryContext()
   n5 = Get QuickbarIndex() -> (QuickbarIndex:EFortQuickBars Enum=Primary)
   n6 = Get Current Slot() -> (Current Slot:int=0)
   > Set Has Item In Current Slot(Has Item In Current Slot:bool=IsValid(Object=GetQuickBarSlottedItem!(self:Fort Inventory Context Object Reference=@n4, InQuickBar:EFortQuickBars Enum=@n5, Slot:int=@n6) -> (ReturnValue:Fort Item Object Reference))) -> (Output_Get:bool)
   > Set Additional Slot Created(Additional Slot Created:bool=Contains Item(
     TargetArray:Array of Integers=Get AdditionalSlots() -> (AdditionalSlots:Array of Integers)
     ItemToFind:"Integer (by ref)"=Get Current Slot() -> (Current Slot=0)
    )) -> (Output_Get:bool)
   n7 = Get Additional Slot Created() -> (Additional Slot Created:bool)
   > n8 = Branch(Condition=AND Boolean(A=Get Has Item In Current Slot() -> (Has Item In Current Slot:bool), B=NOT Boolean(A=@n7)))
   > n9 = Create Widget!(
    execute:Exec
    Class:User Widget Class Reference=&/AthenaQuickBarSlot.AthenaQuickBarSlot_C
    OwningPlayer:Player Controller Object Reference=GetOwningPlayer()
   ) -> (then:Exec, ReturnValue:Athena Quick Bar Slot Object Reference)
   > AddChild(self=Get HorizontalBox_SlotContainer() -> (HorizontalBox_SlotContainer:Horizontal Box Object Reference), Content=@n9)
   > SetQuickbarIndexRuntime!(execute, self:Fort Quick Bar Slot Base Object Reference=@n9, Index:int=Get Current Slot() -> (Current Slot=0), bRefresh:bool=true) -> (then)
   > Add(TargetArray:Array of Integers=Get AdditionalSlots(), NewItem:"Integer (by ref)"=Get Current Slot() -> (Current Slot=0)) {is=title K2Node_CallArrayFunction}
   > Add(
    TargetArray:Array of Fort Quick Bar Slot Base Object References=Get QuickBarSlots() -> (QuickBarSlots:Array of Fort Quick Bar Slot Base Object References)
    NewItem:"Fort Quick Bar Slot Base Object Reference (by ref)"=@n9
   )
   n10 = Get Has Item In Current Slot()
   Branch(execute=@n8.else, Condition=AND Boolean(A=NOT Boolean(A=@n10), B=Get Additional Slot Created()))
   n11 = Get Current Slot() -> (Current Slot=0)
   n12 = Get HorizontalBox_SlotContainer()
   > n13 = RemoveChild(
    self=@n12
    Content=GetChildAt!(self:Panel Widget Object Reference=@n12, Index:int=int - int(A=@n11)) -> (ReturnValue:Widget Object Reference)
   )
   > Branch(Condition=@n13)
   > Remove Item!(execute, TargetArray:Array of Integers=Get AdditionalSlots(), Item:"Integer (by ref)"=Get Current Slot() -> (Current Slot=0)) -> (then, ReturnValue:bool)
   > Remove Index!(
    execute
    TargetArray:Array of Fort Quick Bar Slot Base Object References=Get QuickBarSlots()
    IndexToRemove:int=Get Current Slot() -> (Current Slot=0)
   ) -> (then)
  ]}
 RefreshPickupPreview {nodes [
   RefreshPickupPreview() {is=title K2Node_FunctionEntry}
   n1 = Get CurrentPickupPreviewItem() -> (CurrentPickupPreviewItem:Fort Item Object Reference)
   > n2 = Is Valid(InputObject=@n1)
   > n3 = PreviewItemPickUp!(
    execute
    self:Fort HUDContext Object Reference=Get FortHUDContext()
    InItem:Fort Item Object Reference=@n1
    InQuickBarType:EFortQuickBars Enum=Primary
   ) -> (then, OutSlotIndex:int=0, OutItemsToPush:Map of Integers to Integers, OutItemsToDrop:Array of Integers, ReturnValue:bool)
   n4 = Get PckupPreviewTranslation() -> (PckupPreviewTranslation:"Array of Float (double-precision)s")
   n5 = Get bPlaylistHidesQuickbar() -> (bPlaylistHidesQuickbar:bool)
   > n6 = Branch(Condition=AND Boolean!(
     A=@n3.ReturnValue
     B=Is Valid Index!(TargetArray:"Array of Float (double-precision)s"=@n4, IndexToTest:int=@n3) -> (ReturnValue:bool)
     C:bool=NOT Boolean(A=@n5)
    ) -> (ReturnValue))
   n7 = GetChildrenCount!(self:Panel Widget Object Reference=Get HorizontalBox_SlotContainer()) -> (ReturnValue:int=0)
   > n8 = Branch(Condition=AND Boolean!(
     A="integer > integer"(A=@n3, B=0) {is=title K2Node_CallFunction}
     B="integer <= integer"!(A:int=@n3, B:int=@n7) -> (ReturnValue:bool) {is=title K2Node_CallFunction}
    ) -> (ReturnValue))
   n9 = Break Widget Transform!(WidgetTransform:"Widget Transform Structure (by ref)"=Get RenderTransform!(self:Widget Object Reference=Get PickupPreview()) -> (RenderTransform:Widget Transform Structure)) -> (Translation:Vector 2D Structure, Scale:Vector 2D Structure, Shear:Vector 2D Structure, Angle:float=~) {is=title K2Node_BreakStruct}
   n10 = BreakVector2D(InVec=@n9)
   n11 = GetChildAt(self=Get HorizontalBox_SlotContainer(), Index=int - int(A=@n3, B=1) {is=title K2Node_CallFunction})
   n12 = BreakVector2D(InVec=GetLocalTopLeft!(Geometry:"Geometry Structure (by ref)"=GetCachedGeometry!(self:Widget Object Reference=@n11) -> (ReturnValue:"Geometry Structure (by ref)")) -> (ReturnValue:Vector 2D Structure))
   > SetRenderTransform!(
    execute
    self:Widget Object Reference=Get PickupPreview()
    InTransform:Widget Transform Structure=Make Widget Transform!(
     Translation:Vector 2D Structure=MakeVector2D(X=@n12, Y=@n10.Y)
     Scale:Vector 2D Structure=@n9.Scale
     Shear:Vector 2D Structure=@n9.Shear
     Angle:float=@n9.Angle
    ) -> (WidgetTransform:Widget Transform Structure) {is=title K2Node_MakeStruct}
   ) -> (then)
   n13 = GetNumberOfSlotsTaken!(self:Fort Item Object Reference=Get CurrentPickupPreviewItem()) -> (ReturnValue:Byte=0)
   > SetWidthOverride(
    self=Get PickupPreview()
    InWidthOverride="Get (a ref)"(
     Array:"Array of Float (double-precision)s"=Get PickupPreviewWidth() -> (PickupPreviewWidth:"Array of Float (double-precision)s")
     Dimension 1="To Integer (Byte)"!(InByte:Byte=@n13) -> (ReturnValue:int=0)
    ) -> (Output:"Float (double-precision) (by ref)")
   )
   > SetVisibility(self=Get PickupPreview(), InVisibility=HitTestInvisible)
   SetVisibility(execute=[@n8.else, @n6.else, @n2.Is Not Valid], self=Get PickupPreview(), InVisibility=Hidden)
   n14 = Get CurrentPickupPreviewItem()
   n15 = Is Valid(InputObject=@n14)
   > n16 = PreviewItemPickUp(self=Get FortHUDContext(), InItem=@n14, InQuickBarType=Primary) -> (OutSlotIndex=0)
   > n17 = Branch(Condition=@n16.ReturnValue)
   n18 = Select!(
    Option 0:Athena Quick Bar Slot Object Reference=Get QuickbarSlot1()
    Option 1:Athena Quick Bar Slot Object Reference=Get QuickbarSlot2()
    Option 2:Athena Quick Bar Slot Object Reference=Get QuickbarSlot3()
    Option 3:Athena Quick Bar Slot Object Reference=Get QuickbarSlot4()
    Option 4:Athena Quick Bar Slot Object Reference=Get QuickbarSlot5()
    Option 5:Athena Quick Bar Slot Object Reference=Get QuickbarSlot6()
    Option 6:Athena Quick Bar Slot Object Reference=Get QuickbarSlot7() -> (QuickbarSlot7:Athena Quick Bar Slot Object Reference)
    Index:int=@n16
   ) -> (ReturnValue:Athena Quick Bar Slot Object Reference)
   n19 = BreakVector2D(InVec=GetLocalTopLeft(Geometry=GetTickSpaceGeometry!(self:Widget Object Reference=@n18) -> (ReturnValue:"Geometry Structure (by ref)")))
   n20 = "To Integer (Byte)"(InByte=GetNumberOfSlotsTaken(self=Get CurrentPickupPreviewItem()) -> (ReturnValue=0)) -> (ReturnValue=0)
   n21 = Select(
    Option 0=Get QuickbarSlot1()
    Option 1=Get QuickbarSlot2()
    Option 2=Get QuickbarSlot3()
    Option 3=Get QuickbarSlot4()
    Option 4=Get QuickbarSlot5()
    Option 5=Get QuickbarSlot6()
    Option 6=Get QuickbarSlot7()
    Index=int + int(A=@n16, B=int - int(A=@n20, B=1)) -> (ReturnValue=0) {is=title K2Node_CommutativeAssociativeBinaryOperator}
   )
   n22 = Get NativeCell!(self:Fort Quick Bar Slot Base Object Reference=@n21) -> (NativeCell:Athena Quick Bar Native Cell Object Reference)
   n23 = BreakVector2D(InVec=GetLocalSize!(Geometry:"Geometry Structure (by ref)"=GetTickSpaceGeometry(self=@n22)) -> (ReturnValue:Vector 2D Structure))
   n24 = BreakVector2D(InVec=GetLocalTopLeft(Geometry=GetTickSpaceGeometry(self=@n21)))
   > SetWidthOverride(
    self=Get PickupPreview()
    InWidthOverride=float - float(A=float + float(A=@n24, B=@n23) {is=title K2Node_CommutativeAssociativeBinaryOperator}, B=@n19) {is=title K2Node_CallFunction}
   )
   n25 = float + float(A=@n24.Y, B=@n23.Y)
   n26 = GetTickSpaceGeometry(self=Get NativeCell(self=@n18))
   n27 = float + float(A=@n19.Y, B=BreakVector2D(InVec=GetLocalSize(Geometry=@n26)).Y)
   > SetHeightOverride(
    self=Get PickupPreview()
    InHeightOverride=float - float(
     A="Max (Float)"(A=@n25, B=@n27)
     B="Min (Float)"(A:double=@n24.Y, B:double=@n19.Y) -> (ReturnValue:double) {is=title K2Node_CommutativeAssociativeBinaryOperator}
    )
   )
   n28 = Break Widget Transform(WidgetTransform=Get RenderTransform(self=Get PickupPreview())) -> (Angle=~)
   n29 = BreakVector2D(InVec=@n28)
   > SetRenderTransform(
    self=Get PickupPreview()
    InTransform=Make Widget Transform(Translation=MakeVector2D(X=@n19, Y=@n29.Y), Scale=@n28.Scale, Shear=@n28.Shear, Angle=@n28.Angle)
   )
   > SetVisibility(self=Get PickupPreview(), InVisibility=HitTestInvisible)
   SetVisibility(execute=[@n17.else, @n15.Is Not Valid], self=Get PickupPreview(), InVisibility=Hidden)
   "Get (a ref)"(Array=Get PckupPreviewTranslation(), Dimension 1=@n3)
  ]}
 Show Primary Quickbar Rail {nodes [
   n1 = Show Primary Quickbar Rail!() -> (then, Show:bool=~) {is=title K2Node_FunctionEntry}
   > SetVisibility(
    self=Get SizeBox_WeaponsRail() -> (SizeBox_WeaponsRail:Size Box Object Reference)
    InVisibility=Select!(Option 0:ESlateVisibility Enum=Hidden, Option 1:ESlateVisibility Enum=SelfHitTestInvisible, Index:bool=@n1) -> (ReturnValue:ESlateVisibility Enum)
   )
  ]}
]
variables [
 AdditionalSlots {type="array<integer>"}
 bPlaylistHidesQuickbar {type=boolean}
 CurrentPickupPreviewItem {type=FortItem, default=null}
 NonPickaxeSlotsNum {type=integer}
 QuickbarSlot7 {type=AthenaQuickBarSlot, default=null}
 QuickbarTag {type=GameplayTag, default {tagName=HUD.Quickbar}}
]
widgets {
 info {inheritedWidgetCount=2, parentClass=/Game/UI/InGame/HUD/QuickbarBase.QuickbarBase_C, rootWidgetClass=&ScaleBox, widgetCount=28}
 widgets [
  ScaleBox_0 {
   widgetClassPath=&ScaleBox
   widget=ScaleBox_0
   properties {stretch=UserSpecified, userSpecifiedScale=0.85, visibility=HitTestInvisible}
  }
  Overlay_0 {widgetClassPath=&Overlay, parent=ScaleBox_0, widget=Overlay_0, slot {class=ScaleBoxSlot}}
  HorizontalBox_1 {widgetClassPath=&HorizontalBox, parent=Overlay_0, widget=HorizontalBox_1, slot {class=OverlaySlot}}
  InvalidationBox_Slot_1 {widgetClassPath=/Script/UMG.InvalidationBox, parent=HorizontalBox_1, widget=InvalidationBox_Slot_1, slot {class=HorizontalBoxSlot}}
  QuickbarSlot1 {
   widgetClassPath=&/AthenaQuickBarSlot
   bIsVariable=true
   parent=InvalidationBox_Slot_1
   widget=QuickbarSlot1
   properties {cooldownTypesSupported [Death]}
   slot {class=PanelSlot}
  }
  Overlay_1 {
   widgetClassPath=&Overlay
   parent=HorizontalBox_1
   widget=Overlay_1
   slot {class=HorizontalBoxSlot, properties {padding=<ltrb 0 0 8 0>}}
  }
  QuickbarPrimary_Underlay_171 {
   widgetClassPath=/Game/UI/ExtensionWidgets/QuickbarPrimary_Underlay
   parent=Overlay_1
   widget=QuickbarPrimary_Underlay_171
   slot {class=OverlaySlot, properties {horizontalAlignment=HAlign_Fill, verticalAlignment=VAlign_Fill}}
  }
  SizeBox_WeaponsRail {
   widgetClassPath=&SizeBox
   bIsVariable=true
   parent=Overlay_1
   widget=SizeBox_WeaponsRail
   properties {bOverride_HeightOverride=true, heightOverride=4, visibility=Hidden}
   slot {class=OverlaySlot, properties {horizontalAlignment=HAlign_Fill, verticalAlignment=VAlign_Bottom, padding=<ltrb 0 0 0 21>}}
  }
  CommonBorder_90 {
   widgetClassPath=&CommonBorder
   parent=SizeBox_WeaponsRail
   widget=CommonBorder_90
   properties {style=/FortUILibrary/Styles/Border/Border_HUD_Line_Horz.Border_HUD_Line_Horz_C}
   slot {class=SizeBoxSlot}
  }
  HorizontalBox_SlotContainer {widgetClassPath=&HorizontalBox, bIsVariable=true, parent=Overlay_1, widget=HorizontalBox_SlotContainer, slot {class=OverlaySlot}}
  InvalidationBox_Slot_2 {
   widgetClassPath=&InvalidationBox
   parent=HorizontalBox_SlotContainer
   widget=InvalidationBox_Slot_2
   slot {class=HorizontalBoxSlot}
  }
  QuickbarSlot2 {
   widgetClassPath=&AthenaQuickBarSlot
   bIsVariable=true
   parent=InvalidationBox_Slot_2
   widget=QuickbarSlot2
   properties {
    bCheckVisibilityBasedOffInventoryCapacity=true
    bShowBottomArrows=true
    slotIndex=1
    cooldownTypesSupported [Death, WeaponReloading, ItemActivation, AthenaWeaponFireCooldown, AmmoRegeneration]
   }
   slot {class=PanelSlot}
  }
  InvalidationBox_Slot_3 {
   widgetClassPath=&InvalidationBox
   parent=HorizontalBox_SlotContainer
   widget=InvalidationBox_Slot_3
   slot {class=HorizontalBoxSlot}
  }
  QuickbarSlot3 {
   widgetClassPath=&AthenaQuickBarSlot
   bIsVariable=true
   parent=InvalidationBox_Slot_3
   widget=QuickbarSlot3
   properties {
    bCheckVisibilityBasedOffInventoryCapacity=true
    bShowBottomArrows=true
    slotIndex=2
    cooldownTypesSupported [Death, WeaponReloading, ItemActivation, AthenaWeaponFireCooldown, AmmoRegeneration]
   }
   slot {class=PanelSlot}
  }
  InvalidationBox_Slot_4 {
   widgetClassPath=&InvalidationBox
   parent=HorizontalBox_SlotContainer
   widget=InvalidationBox_Slot_4
   slot {class=HorizontalBoxSlot}
  }
  QuickbarSlot4 {
   widgetClassPath=&AthenaQuickBarSlot
   bIsVariable=true
   parent=InvalidationBox_Slot_4
   widget=QuickbarSlot4
   properties {
    bCheckVisibilityBasedOffInventoryCapacity=true
    bShowBottomArrows=true
    slotIndex=3
    cooldownTypesSupported [Death, WeaponReloading, ItemActivation, AthenaWeaponFireCooldown, AmmoRegeneration]
   }
   slot {class=PanelSlot}
  }
  InvalidationBox_Slot_5 {
   widgetClassPath=&InvalidationBox
   parent=HorizontalBox_SlotContainer
   widget=InvalidationBox_Slot_5
   slot {class=HorizontalBoxSlot}
  }
  QuickbarSlot5 {
   widgetClassPath=&AthenaQuickBarSlot
   bIsVariable=true
   parent=InvalidationBox_Slot_5
   widget=QuickbarSlot5
   properties {
    bCheckVisibilityBasedOffInventoryCapacity=true
    bShowBottomArrows=true
    slotIndex=4
    cooldownTypesSupported [Death, WeaponReloading, ItemActivation, AthenaWeaponFireCooldown, AmmoRegeneration]
   }
   slot {class=PanelSlot}
  }
  InvalidationBox_Slot_6 {
   widgetClassPath=&InvalidationBox
   parent=HorizontalBox_SlotContainer
   widget=InvalidationBox_Slot_6
   slot {class=HorizontalBoxSlot}
  }
  QuickbarSlot6 {
   widgetClassPath=&AthenaQuickBarSlot
   bIsVariable=true
   parent=InvalidationBox_Slot_6
   widget=QuickbarSlot6
   properties {
    bCheckVisibilityBasedOffInventoryCapacity=true
    bShowBottomArrows=true
    slotIndex=5
    cooldownTypesSupported [Death, WeaponReloading, ItemActivation, AthenaWeaponFireCooldown, AmmoRegeneration]
   }
   slot {class=PanelSlot}
  }
  PickupPreview {
   widgetClassPath=&SizeBox
   bIsVariable=true
   parent=Overlay_1
   widget=PickupPreview
   properties {
    bOverride_HeightOverride=true
    bOverride_WidthOverride=true
    heightOverride=94
    visibility=Hidden
    widthOverride=94
    renderTransform {angle=0, scale=<1 1>, shear=<0 0>, translation=<0 12>}
    renderTransformPivot=<0 0.5>
   }
   slot {class=OverlaySlot}
  }
  Border {
   widgetClassPath=&Border
   parent=PickupPreview
   widget=Border
   properties {
    visibility=SelfHitTestInvisible
    background {
     bIsDynamicallyLoaded=false
     drawAs=Box
     imageType=NoImage
     mirroring=NoMirror
     resourceName=None
     resourceObject=/Game/UI/InGame/Quickbar/M_UI_FramePreview_Inst
     tiling=NoTile
     imageSize=<80 80>
     margin=<ltrb 0 0 0 0>
     outlineSettings {
      bUseBrushTransparency=false
      roundingType=HalfHeightRadius
      width=0
      color {colorUseRule=UseColor_Specified, specifiedColor=<rgba 0 0 0 0>}
      cornerRadii=<0 0 0 0>
     }
     tintColor {colorUseRule=UseColor_Specified, specifiedColor=<rgba 1 1 1 1>}
     uVRegion {bIsValid=false, max=<0 0>, min=<0 0>}
    }
    padding=<ltrb -200 -200 -200 -200>
    renderTransform {angle=0, scale=<1 1>, shear=<0 0>, translation=<0 2>}
    renderTransformPivot=<0 0>
   }
   slot {class=SizeBoxSlot, properties {padding=<ltrb -10 -10 -10 -10>}}
  }
  PickupPreviewHighlight {
   widgetClassPath=&Overlay
   parent=Overlay_1
   widget=PickupPreviewHighlight
   properties {visibility=Collapsed}
   slot {class=OverlaySlot, properties {horizontalAlignment=HAlign_Center, verticalAlignment=VAlign_Center}}
  }
  Border_Highlight_1 {
   widgetClassPath=&Border
   parent=PickupPreviewHighlight
   widget=Border_Highlight_1
   properties {
    horizontalAlignment=HAlign_Center
    verticalAlignment=VAlign_Center
    visibility=SelfHitTestInvisible
    background {
     bIsDynamicallyLoaded=false
     drawAs=Border
     imageType=NoImage
     mirroring=NoMirror
     resourceName=None
     resourceObject=/Game/Athena/HUD/Art/Highlight/M_AthenaQuickBarHighlight_Inst
     tiling=NoTile
     imageSize=<80 80>
     margin=<ltrb 0.4 0.4 0.4 0.4>
     outlineSettings {
      bUseBrushTransparency=false
      roundingType=HalfHeightRadius
      width=0
      color {colorUseRule=UseColor_Specified, specifiedColor=<rgba 0 0 0 0>}
      cornerRadii=<0 0 0 0>
     }
     tintColor {colorUseRule=UseColor_Specified, specifiedColor=<rgba 1 1 1 1>}
     uVRegion {bIsValid=false, max=<0 0>, min=<0 0>}
    }
    brushColor=<rgba 1 1 1 0.2>
    padding=<ltrb 40 40 40 40>
    renderTransformPivot=<0 0>
   }
   slot {class=OverlaySlot, properties {horizontalAlignment=HAlign_Center, verticalAlignment=VAlign_Center}}
  }
  Border_Highlight_2 {
   widgetClassPath=&Border
   parent=PickupPreviewHighlight
   widget=Border_Highlight_2
   properties {
    horizontalAlignment=HAlign_Center
    verticalAlignment=VAlign_Center
    visibility=SelfHitTestInvisible
    background {
     bIsDynamicallyLoaded=false
     drawAs=Border
     imageType=NoImage
     mirroring=NoMirror
     resourceName=None
     resourceObject=&M_AthenaQuickBarHighlight_Inst
     tiling=NoTile
     imageSize=<80 80>
     margin=<ltrb 0.4 0.4 0.4 0.4>
     outlineSettings {
      bUseBrushTransparency=false
      roundingType=HalfHeightRadius
      width=0
      color {colorUseRule=UseColor_Specified, specifiedColor=<rgba 0 0 0 0>}
      cornerRadii=<0 0 0 0>
     }
     tintColor {colorUseRule=UseColor_Specified, specifiedColor=<rgba 1 1 1 1>}
     uVRegion {bIsValid=false, max=<0 0>, min=<0 0>}
    }
    brushColor=<rgba 1 1 1 0.2>
    desiredSizeScale=<2 2>
    padding=<ltrb 40 40 40 40>
    renderTransform {angle=0, scale=<1 1>, shear=<0 0>, translation=<0 2>}
    renderTransformPivot=<0 0>
   }
   slot {class=OverlaySlot, properties {horizontalAlignment=HAlign_Center, verticalAlignment=VAlign_Center}}
  }
  ball {
   widgetClassPath=&Image
   parent=PickupPreviewHighlight
   widget=ball
   properties {brush {
     bIsDynamicallyLoaded=false
     drawAs=Image
     imageType=NoImage
     mirroring=NoMirror
     resourceName=None
     resourceObject=/Game/Athena/Tutorial/Contextual/MI_Widget_Contextual_Circle
     tiling=NoTile
     imageSize=<64 64>
     margin=<ltrb 0 0 0 0>
     outlineSettings {
      bUseBrushTransparency=false
      roundingType=HalfHeightRadius
      width=0
      color {colorUseRule=UseColor_Specified, specifiedColor=<rgba 0 0 0 0>}
      cornerRadii=<0 0 0 0>
     }
     tintColor {colorUseRule=UseColor_Specified, specifiedColor=<rgba 1 1 1 1>}
     uVRegion {bIsValid=false, max=<0 0>, min=<0 0>}
    }}
   slot {class=OverlaySlot, properties {horizontalAlignment=HAlign_Center, verticalAlignment=VAlign_Center}}
  }
  TutorialHighlight {
   widgetClassPath=&TutorialHighlight
   bIsVariable=true
   parent=Overlay_0
   widget=TutorialHighlight
   properties {customHighlightWidget=/Game/Athena/HUD/Tutorial/Tutorial_Outline_Square_Static.Tutorial_Outline_Square_Static_C}
   slot {class=OverlaySlot, properties {horizontalAlignment=HAlign_Fill, verticalAlignment=VAlign_Fill}}
  }
  KeybindWidget_SwitchModeKeybind {widgetClassPath=/Script/FortniteUI.FortKeybindWidget, bInherited=true, widget=null}
 ]
}
