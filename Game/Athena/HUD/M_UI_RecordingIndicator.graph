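# Translucent UI material for a HUD recording indicator: a pulsing shape with a Color parameter that defaults to red.
# Time wraps on LoopInterval and drives a normalized sine, remapped by FadeDuration over LoopInterval and inverted to dip once per loop.
# The UseCircleMask static switch optionally multiplies that pulse by an MF_UI_Circle fill sized by CircleSize, then the result is saturated.
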Material /Game/Athena/HUD/M_UI_RecordingIndicator
graph {
 properties {blendMode=BLEND_Translucent, materialDomain=MD_UI, pixelDepthOffsetMode=PDOM_Legacy, usedShadingModels=Default Lit}
 nodes [
  n1 = ScalarParameter() {defaultValue=2, parameterName=LoopInterval}
  n2 = Fmod(A=Time(), B=@n1)
  n3 = Add(A=Sine(None=Divide(A=@n2, B=@n1)))
  n4 = Divide(A=ScalarParameter() {defaultValue=0.6, parameterName=FadeDuration}, B=@n1)
  n5 = OneMinus(None=MaterialFunctionCall(Input=Divide(A=@n3), Input Low=OneMinus(None=@n4), Input High=Constant() {r=1}) {materialFunction=&RemapValue_0-1}.Result)
  n6 = MaterialFunctionCall(Size=ScalarParameter() {defaultValue=1, parameterName=CircleSize}) {materialFunction=&MF_UI_Circle}
  Saturate(None=StaticSwitchParameter(True=Multiply(A=@n5, B=@n6.Fill), False=@n5) {parameterName=UseCircleMask})
  VectorParameter() {parameterName=Color, defaultValue=<rgba 1 0 0 1>}
 ]
}
