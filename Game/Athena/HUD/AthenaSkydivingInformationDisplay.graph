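# Skydiving HUD widget extending AthenaSkydivingInformation, shown from skydive start to exit as a vertical altitude meter.
# Parent events drive dynamic material parameters: negated normalized altitudes offset markers, vertical speed thresholds toggle three player arrows, and glider entry swaps the player marker.
# The ground marker tracks deploy altitude in freefall and ground altitude while gliding.
# Numeric readouts and progress bars still update, but their containers start collapsed and are never revealed here.
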WidgetBlueprint /Game/Athena/HUD/AthenaSkydivingInformationDisplay
graphs [EventGraph {nodes [
   n1 = Event OnNewVerticalSpeed!() -> (OutputDelegate, then, NewVerticalSpeed:float) {is=title K2Node_Event}
   n2 = float * float(A=@n1.NewVerticalSpeed, B=0.036)
   n3 = SetCurrentValue!(
    execute
    self:Common Numeric Text Block Object Reference=Get DropSpeedTextBlock() -> (DropSpeedTextBlock:Common Numeric Text Block Object Reference)
    NewValue:float=@n2
   ) -> (then)
   n4 = Get PlayerMarkerMID() -> (PlayerMarkerMID:Material Instance Dynamic Object Reference)
   SetScalarParameterValue(
    execute=[@n3.then, @n1.then]
    self=@n4
    ParameterName=Arrow3Fade
    Value=Select(Option 0:double=0, Option 1:double=1, Index="float > float"(A=@n2, B=190) {is=title K2Node_CallFunction}) -> (ReturnValue:double)
   )
   > SetScalarParameterValue(self=@n4, ParameterName=Arrow2Fade, Value=Select(Option 0=0, Option 1=1, Index="float > float"(A=@n2, B=115)))
   > SetScalarParameterValue(self=@n4, ParameterName=Arrow1Fade, Value=Select(Option 0=0, Option 1=1, Index="float > float"(A=@n2, B=15)))
   Event Construct()
   > n5 = GetDynamicMaterial(self=Get PlayerMarker() -> (PlayerMarker:Image Object Reference))
   > Set PlayerMarkerMID(PlayerMarkerMID:Material Instance Dynamic Object Reference=@n5) -> (Output_Get:Material Instance Dynamic Object Reference)
   > SetVisibility(self=Get HBox_Root() -> (HBox_Root:Horizontal Box Object Reference), InVisibility=Collapsed)
   n6 = Event OnNewNormalizedPlayerAltitude!() -> (OutputDelegate, then, NormalizedPlayerAltitude:float) {is=title K2Node_Event}
   > SetPercent!(
    execute
    self:Progress Bar Object Reference=Get CurrentHeightBar() -> (CurrentHeightBar:Progress Bar Object Reference)
    InPercent:float=@n6.NormalizedPlayerAltitude
   ) -> (then)
   > SetScalarParameterValue(self=Get PlayerMarkerMID(), ParameterName=YOffset, Value=float - float(B=@n6.NormalizedPlayerAltitude) {is=title K2Node_CallFunction})
   Event OnEnterGlider() {is=title K2Node_Event}
   > SetVisibility(self=Get DeployAtBox() -> (DeployAtBox:Horizontal Box Object Reference), InVisibility=Collapsed)
   > SetVisibility(self=Get DeployHeightBar() -> (DeployHeightBar:Progress Bar Object Reference), InVisibility=Collapsed)
   > SetVisibility(self=Get DeployInBox() -> (DeployInBox:Horizontal Box Object Reference), InVisibility=Collapsed)
   > SetScalarParameterValue(self=Get PlayerMarkerMID(), ParameterName=GliderSwap, Value=1)
   > Set isGliding(isGliding:bool=true) -> (Output_Get:bool)
   Event OnStartSkydive() {is=title K2Node_Event}
   > SetVisibility(self=Get HBox_Root(), InVisibility=HitTestInvisible)
   > SetScalarParameterValue(self=Get PlayerMarkerMID(), ParameterName=GliderSwap, Value=0)
   > Set isGliding(isGliding=false)
   Event OnExitSkydive() {is=title K2Node_Event}
   > SetVisibility(self=Get HBox_Root(), InVisibility=Collapsed)
   n7 = Event OnNewNormalizedGroundAltitude!() -> (OutputDelegate, then, NormalizedGroundAltitude:float) {is=title K2Node_Event}
   > SetPercent(self=Get GroundHeightBar() -> (GroundHeightBar:Progress Bar Object Reference), InPercent=@n7.NormalizedGroundAltitude)
   n8 = Get isGliding() -> (isGliding:bool)
   > Branch(Condition=@n8)
   > n9 = GetDynamicMaterial(self=Get GroundMarker() -> (GroundMarker:Image Object Reference))
   > SetScalarParameterValue(self=@n9, ParameterName=YOffset, Value=float - float(B=@n7.NormalizedGroundAltitude))
   n10 = Event OnNewNormalizedOpeningAltitude!() -> (OutputDelegate, then, NormalizedOpeningAltitude:float) {is=title K2Node_Event}
   > SetPercent(self=Get DeployHeightBar(), InPercent=@n10.NormalizedOpeningAltitude)
   > n11 = Branch(Condition=@n8)
   n12 = GetDynamicMaterial(execute=@n11.else, self=Get GroundMarker())
   > SetScalarParameterValue(self=@n12, ParameterName=YOffset, Value=float - float(B=@n10.NormalizedOpeningAltitude))
   SetCurrentValue(
    self=Get CommonNumericTextBlock_128() -> (CommonNumericTextBlock_128:Common Numeric Text Block Object Reference)
    NewValue=Event OnNewDuration!() -> (OutputDelegate, then, NewDuration:float) {is=title K2Node_Event}.NewDuration
   )
   SetCurrentValue(
    self=Get HorizSpeedValue() -> (HorizSpeedValue:Common Numeric Text Block Object Reference)
    NewValue=Event OnNewHorizontalSpeed!() -> (OutputDelegate, then, NewHorizontalSpeed:float) {is=title K2Node_Event}.NewHorizontalSpeed
   )
   SetCurrentValue(
    self=Get DeployAltitudeDisplay() -> (DeployAltitudeDisplay:Common Numeric Text Block Object Reference)
    NewValue=float / float(A=Event OnNewAltitude!() -> (OutputDelegate, then, NewAltitude:float) {is=title K2Node_Event}.NewAltitude, B=100) {is=title K2Node_CallFunction}
   )
   SetCurrentValue(
    self=Get CurrentAltitudeDisplay() -> (CurrentAltitudeDisplay:Common Numeric Text Block Object Reference)
    NewValue=float / float(A=Event OnNewCurrentAltitude!() -> (OutputDelegate, then, NewCurrentAltitude:float) {is=title K2Node_Event}.NewCurrentAltitude, B=100)
   )
   Event PreConstruct()
  ]}]
variables [
 isGliding {type=boolean}
 PlayerMarkerMID {type=MaterialInstanceDynamic}
]
widgets {
 info {inheritedWidgetCount=1, parentClass=/Script/FortniteUI.AthenaSkydivingInformation, rootWidgetClass=&HorizontalBox, widgetCount=32}
 widgets [
  HBox_Root {widgetClassPath=&HorizontalBox, bIsVariable=true, widget=HBox_Root}
  SizeBox_0 {
   widgetClassPath=&SizeBox
   parent=HBox_Root
   widget=SizeBox_0
   properties {bOverride_HeightOverride=true, bOverride_WidthOverride=true, heightOverride=300, widthOverride=40}
   slot {class=HorizontalBoxSlot}
  }
  OverlayVisualMeter {widgetClassPath=&Overlay, parent=SizeBox_0, widget=OverlayVisualMeter, slot {class=SizeBoxSlot}}
  BottomAnchor {
   widgetClassPath=&Image
   bIsVariable=true
   parent=OverlayVisualMeter
   widget=BottomAnchor
   properties {brush {
     bIsDynamicallyLoaded=false
     drawAs=Image
     imageType=NoImage
     mirroring=NoMirror
     resourceName=None
     resourceObject=null
     tiling=NoTile
     imageSize=<25 2>
     margin=<ltrb 0 0 0 0>
     outlineSettings {
      bUseBrushTransparency=false
      roundingType=HalfHeightRadius
      width=0
      color {colorUseRule=UseColor_Specified, specifiedColor=<rgba 0 0 0 0>}
      cornerRadii=<0 0 0 0>
     }
     tintColor {colorUseRule=UseColor_Specified, specifiedColor=<rgba 1 1 1 1>}
     uVRegion {bIsValid=false, max=<0 0>, min=<0 0>}
    }}
   slot {class=OverlaySlot, properties {horizontalAlignment=HAlign_Right, verticalAlignment=VAlign_Bottom}}
  }
  RightAnchor {
   widgetClassPath=&Image
   bIsVariable=true
   parent=OverlayVisualMeter
   widget=RightAnchor
   properties {brush {
     bIsDynamicallyLoaded=false
     drawAs=Image
     imageType=NoImage
     mirroring=NoMirror
     resourceName=None
     resourceObject=null
     tiling=NoTile
     imageSize=<2 32>
     margin=<ltrb 0 0 0 0>
     outlineSettings {
      bUseBrushTransparency=false
      roundingType=HalfHeightRadius
      width=0
      color {colorUseRule=UseColor_Specified, specifiedColor=<rgba 0 0 0 0>}
      cornerRadii=<0 0 0 0>
     }
     tintColor {colorUseRule=UseColor_Specified, specifiedColor=<rgba 1 1 1 1>}
     uVRegion {bIsValid=false, max=<0 0>, min=<0 0>}
    }}
   slot {class=OverlaySlot, properties {horizontalAlignment=HAlign_Right, verticalAlignment=VAlign_Fill}}
  }
  GroundMarker {
   widgetClassPath=&Image
   bIsVariable=true
   parent=OverlayVisualMeter
   widget=GroundMarker
   properties {brush {
     bIsDynamicallyLoaded=false
     drawAs=Image
     imageType=NoImage
     mirroring=NoMirror
     resourceName=None
     resourceObject=/Game/Athena/HUD/Materials/M_UI_SkydiveMarker_Inst
     tiling=NoTile
     imageSize=<16 4>
     margin=<ltrb 0 0 0 0>
     outlineSettings {
      bUseBrushTransparency=false
      roundingType=HalfHeightRadius
      width=0
      color {colorUseRule=UseColor_Specified, specifiedColor=<rgba 0 0 0 0>}
      cornerRadii=<0 0 0 0>
     }
     tintColor {colorUseRule=UseColor_Specified, specifiedColor=<rgba 1 1 1 1>}
     uVRegion {bIsValid=false, max=<0 0>, min=<0 0>}
    }}
   slot {class=OverlaySlot, properties {horizontalAlignment=HAlign_Right, verticalAlignment=VAlign_Fill}}
  }
  PlayerMarker {
   widgetClassPath=&Image
   bIsVariable=true
   parent=OverlayVisualMeter
   widget=PlayerMarker
   properties {brush {
     bIsDynamicallyLoaded=false
     drawAs=Image
     imageType=NoImage
     mirroring=NoMirror
     resourceName=None
     resourceObject=/Game/Athena/HUD/Materials/M_UI_SkydivePlayer_Inst
     tiling=NoTile
     imageSize=<25.5038 4>
     margin=<ltrb 0 0 0 0>
     outlineSettings {
      bUseBrushTransparency=false
      roundingType=HalfHeightRadius
      width=0
      color {colorUseRule=UseColor_Specified, specifiedColor=<rgba 0 0 0 0>}
      cornerRadii=<0 0 0 0>
     }
     tintColor {colorUseRule=UseColor_Specified, specifiedColor=<rgba 1 1 1 1>}
     uVRegion {bIsValid=false, max=<0 0>, min=<0 0>}
    }}
   slot {class=OverlaySlot, properties {horizontalAlignment=HAlign_Fill, verticalAlignment=VAlign_Fill}}
  }
  VerticalBoxTextOutput {
   widgetClassPath=&VerticalBox
   parent=HBox_Root
   widget=VerticalBoxTextOutput
   properties {visibility=Collapsed}
   slot {class=HorizontalBoxSlot, properties {padding=<ltrb 0 0 5 0>}}
  }
  HorizontalBox_1 {
   widgetClassPath=&HorizontalBox
   parent=VerticalBoxTextOutput
   widget=HorizontalBox_1
   properties {visibility=Collapsed}
   slot {class=VerticalBoxSlot, properties {horizontalAlignment=HAlign_Left, verticalAlignment=VAlign_Bottom, size {sizeRule=Fill, value=1}}}
  }
  Altimeter_Current_StaticText {
   widgetClassPath=&CommonTextBlock
   parent=HorizontalBox_1
   widget=Altimeter_Current_StaticText
   properties {
    style=/FortUILibrary/Styles/Typography/Foundation/TextStyle_BurbankSmall_M_White.TextStyle_BurbankSmall_M_White_C
    text="Altitude: "
    font {
     bForceMonospaced=false
     bMaterialIsStencil=false
     fontMaterial=null
     fontObject=&BurbankSmall
     letterSpacing=0
     monospacedWidth=1
     size=24
     skewAmount=0
     typefaceFontName=Black
     outlineSettings {
      bApplyOutlineToDropShadows=false
      bMiteredCorners=false
      bSeparateFillAlpha=false
      outlineMaterial=null
      outlineSize=0
      outlineColor=<rgba 0 0 0 1>
     }
    }
    shadowOffset=<0 0>
   }
   slot {class=HorizontalBoxSlot, properties {verticalAlignment=VAlign_Bottom, size {sizeRule=Fill, value=1}}}
  }
  CurrentAltitudeDisplay {
   widgetClassPath=/Script/CommonUI.CommonNumericTextBlock
   bIsVariable=true
   parent=HorizontalBox_1
   widget=CurrentAltitudeDisplay
   properties {
    style=&TextStyle-Base_C
    text=Text Block
    font {
     bForceMonospaced=false
     bMaterialIsStencil=false
     fontMaterial=null
     fontObject=&NotoSans
     letterSpacing=0
     monospacedWidth=1
     size=24
     skewAmount=0
     typefaceFontName=NotoSans-Regular
     outlineSettings {
      bApplyOutlineToDropShadows=false
      bMiteredCorners=false
      bSeparateFillAlpha=false
      outlineMaterial=null
      outlineSize=0
      outlineColor=<rgba 0 0 0 1>
     }
    }
    shadowOffset=<0 0>
   }
   slot {class=HorizontalBoxSlot, properties {horizontalAlignment=HAlign_Right, verticalAlignment=VAlign_Bottom, size {sizeRule=Fill, value=1}}}
  }
  CommonTextBlock {
   widgetClassPath=&CommonTextBlock
   parent=HorizontalBox_1
   widget=CommonTextBlock
   properties {
    style=/FortUILibrary/Styles/Typography/Foundation/TextStyle-BurbankSmall-XS-White.TextStyle-BurbankSmall-XS-White_C
    text=meters
    font {
     bForceMonospaced=false
     bMaterialIsStencil=false
     fontMaterial=null
     fontObject=&BurbankSmall
     letterSpacing=0
     monospacedWidth=1
     size=13
     skewAmount=0
     typefaceFontName=Black
     outlineSettings {
      bApplyOutlineToDropShadows=false
      bMiteredCorners=false
      bSeparateFillAlpha=false
      outlineMaterial=null
      outlineSize=0
      outlineColor=<rgba 0 0 0 1>
     }
    }
    shadowOffset=<0 0>
   }
   slot {class=HorizontalBoxSlot, properties {horizontalAlignment=HAlign_Left}}
  }
  DeployAtBox {
   widgetClassPath=&HorizontalBox
   bIsVariable=true
   parent=VerticalBoxTextOutput
   widget=DeployAtBox
   properties {visibility=Collapsed}
   slot {class=VerticalBoxSlot, properties {horizontalAlignment=HAlign_Left, verticalAlignment=VAlign_Bottom}}
  }
  AltimiterStaticText {
   widgetClassPath=&CommonTextBlock
   parent=DeployAtBox
   widget=AltimiterStaticText
   properties {
    style=&TextStyle_BurbankSmall_M_White_C
    text="Deploy At: "
    font {
     bForceMonospaced=false
     bMaterialIsStencil=false
     fontMaterial=null
     fontObject=&BurbankSmall
     letterSpacing=0
     monospacedWidth=1
     size=24
     skewAmount=0
     typefaceFontName=Black
     outlineSettings {
      bApplyOutlineToDropShadows=false
      bMiteredCorners=false
      bSeparateFillAlpha=false
      outlineMaterial=null
      outlineSize=0
      outlineColor=<rgba 0 0 0 1>
     }
    }
    shadowOffset=<0 0>
   }
   slot {class=HorizontalBoxSlot, properties {horizontalAlignment=HAlign_Center, verticalAlignment=VAlign_Center}}
  }
  DeployAltitudeDisplay {
   widgetClassPath=&CommonNumericTextBlock
   bIsVariable=true
   parent=DeployAtBox
   widget=DeployAltitudeDisplay
   properties {
    style=&TextStyle-Base_C
    text=Text Block
    font {
     bForceMonospaced=false
     bMaterialIsStencil=false
     fontMaterial=null
     fontObject=&NotoSans
     letterSpacing=0
     monospacedWidth=1
     size=24
     skewAmount=0
     typefaceFontName=NotoSans-Regular
     outlineSettings {
      bApplyOutlineToDropShadows=false
      bMiteredCorners=false
      bSeparateFillAlpha=false
      outlineMaterial=null
      outlineSize=0
      outlineColor=<rgba 0 0 0 1>
     }
    }
    shadowOffset=<0 0>
   }
   slot {class=HorizontalBoxSlot, properties {horizontalAlignment=HAlign_Right, size {sizeRule=Fill, value=1}}}
  }
  CommonTextBlock_228 {
   widgetClassPath=&CommonTextBlock
   parent=DeployAtBox
   widget=CommonTextBlock_228
   properties {
    style=&TextStyle-BurbankSmall-XS-White_C
    text=meters
    font {
     bForceMonospaced=false
     bMaterialIsStencil=false
     fontMaterial=null
     fontObject=&BurbankSmall
     letterSpacing=0
     monospacedWidth=1
     size=13
     skewAmount=0
     typefaceFontName=Black
     outlineSettings {
      bApplyOutlineToDropShadows=false
      bMiteredCorners=false
      bSeparateFillAlpha=false
      outlineMaterial=null
      outlineSize=0
      outlineColor=<rgba 0 0 0 1>
     }
    }
    shadowOffset=<0 0>
   }
   slot {class=HorizontalBoxSlot}
  }
  DeployInBox {
   widgetClassPath=&HorizontalBox
   bIsVariable=true
   parent=VerticalBoxTextOutput
   widget=DeployInBox
   properties {visibility=Collapsed}
   slot {class=VerticalBoxSlot, properties {horizontalAlignment=HAlign_Left, verticalAlignment=VAlign_Top}}
  }
  TimerStaticText {
   widgetClassPath=&CommonTextBlock
   parent=DeployInBox
   widget=TimerStaticText
   properties {
    style=&TextStyle_BurbankSmall_M_White_C
    text="Deploy In: "
    font {
     bForceMonospaced=false
     bMaterialIsStencil=false
     fontMaterial=null
     fontObject=&BurbankSmall
     letterSpacing=0
     monospacedWidth=1
     size=24
     skewAmount=0
     typefaceFontName=Black
     outlineSettings {
      bApplyOutlineToDropShadows=false
      bMiteredCorners=false
      bSeparateFillAlpha=false
      outlineMaterial=null
      outlineSize=0
      outlineColor=<rgba 0 0 0 1>
     }
    }
    shadowOffset=<0 0>
   }
   slot {class=HorizontalBoxSlot, properties {verticalAlignment=VAlign_Top}}
  }
  CommonNumericTextBlock_128 {
   widgetClassPath=&CommonNumericTextBlock
   bIsVariable=true
   parent=DeployInBox
   widget=CommonNumericTextBlock_128
   properties {
    numericType=Seconds
    style=&TextStyle_BurbankSmall_M_White_C
    text=Text Block
    font {
     bForceMonospaced=false
     bMaterialIsStencil=false
     fontMaterial=null
     fontObject=&BurbankSmall
     letterSpacing=0
     monospacedWidth=1
     size=24
     skewAmount=0
     typefaceFontName=Black
     outlineSettings {
      bApplyOutlineToDropShadows=false
      bMiteredCorners=false
      bSeparateFillAlpha=false
      outlineMaterial=null
      outlineSize=0
      outlineColor=<rgba 0 0 0 1>
     }
    }
    shadowOffset=<0 0>
   }
   slot {class=HorizontalBoxSlot}
  }
  HorizontalBox_2 {
   widgetClassPath=&HorizontalBox
   parent=VerticalBoxTextOutput
   widget=HorizontalBox_2
   slot {class=VerticalBoxSlot, properties {horizontalAlignment=HAlign_Left, verticalAlignment=VAlign_Top}}
  }
  TimerStaticText_1 {
   widgetClassPath=&CommonTextBlock
   parent=HorizontalBox_2
   widget=TimerStaticText_1
   properties {
    style=&TextStyle_BurbankSmall_M_White_C
    text="Drop Speed: "
    font {
     bForceMonospaced=false
     bMaterialIsStencil=false
     fontMaterial=null
     fontObject=&BurbankSmall
     letterSpacing=0
     monospacedWidth=1
     size=24
     skewAmount=0
     typefaceFontName=Black
     outlineSettings {
      bApplyOutlineToDropShadows=false
      bMiteredCorners=false
      bSeparateFillAlpha=false
      outlineMaterial=null
      outlineSize=0
      outlineColor=<rgba 0 0 0 1>
     }
    }
    shadowOffset=<0 0>
   }
   slot {class=HorizontalBoxSlot, properties {verticalAlignment=VAlign_Top, size {sizeRule=Fill, value=1}}}
  }
  DropSpeedTextBlock {
   widgetClassPath=&CommonNumericTextBlock
   bIsVariable=true
   parent=HorizontalBox_2
   widget=DropSpeedTextBlock
   properties {
    style=&TextStyle_BurbankSmall_M_White_C
    text=Text Block
    font {
     bForceMonospaced=false
     bMaterialIsStencil=false
     fontMaterial=null
     fontObject=&BurbankSmall
     letterSpacing=0
     monospacedWidth=1
     size=24
     skewAmount=0
     typefaceFontName=Black
     outlineSettings {
      bApplyOutlineToDropShadows=false
      bMiteredCorners=false
      bSeparateFillAlpha=false
      outlineMaterial=null
      outlineSize=0
      outlineColor=<rgba 0 0 0 1>
     }
    }
    shadowOffset=<0 0>
   }
   slot {class=HorizontalBoxSlot}
  }
  CommonTextBlock_1 {
   widgetClassPath=&CommonTextBlock
   parent=HorizontalBox_2
   widget=CommonTextBlock_1
   properties {
    style=&TextStyle-BurbankSmall-XS-White_C
    text=km/h
    font {
     bForceMonospaced=false
     bMaterialIsStencil=false
     fontMaterial=null
     fontObject=&BurbankSmall
     letterSpacing=0
     monospacedWidth=1
     size=13
     skewAmount=0
     typefaceFontName=Black
     outlineSettings {
      bApplyOutlineToDropShadows=false
      bMiteredCorners=false
      bSeparateFillAlpha=false
      outlineMaterial=null
      outlineSize=0
      outlineColor=<rgba 0 0 0 1>
     }
    }
    shadowOffset=<0 0>
   }
   slot {class=HorizontalBoxSlot}
  }
  HorizontalBox {
   widgetClassPath=&HorizontalBox
   parent=VerticalBoxTextOutput
   widget=HorizontalBox
   slot {class=VerticalBoxSlot, properties {horizontalAlignment=HAlign_Left, verticalAlignment=VAlign_Top, size {sizeRule=Fill, value=1}}}
  }
  HorizSpeed {
   widgetClassPath=&CommonTextBlock
   parent=HorizontalBox
   widget=HorizSpeed
   properties {
    style=&TextStyle_BurbankSmall_M_White_C
    text="Ground Speed: "
    font {
     bForceMonospaced=false
     bMaterialIsStencil=false
     fontMaterial=null
     fontObject=&BurbankSmall
     letterSpacing=0
     monospacedWidth=1
     size=24
     skewAmount=0
     typefaceFontName=Black
     outlineSettings {
      bApplyOutlineToDropShadows=false
      bMiteredCorners=false
      bSeparateFillAlpha=false
      outlineMaterial=null
      outlineSize=0
      outlineColor=<rgba 0 0 0 1>
     }
    }
    shadowOffset=<0 0>
   }
   slot {class=HorizontalBoxSlot, properties {verticalAlignment=VAlign_Top, size {sizeRule=Fill, value=1}}}
  }
  HorizSpeedValue {
   widgetClassPath=&CommonNumericTextBlock
   bIsVariable=true
   parent=HorizontalBox
   widget=HorizSpeedValue
   properties {
    style=&TextStyle_BurbankSmall_M_White_C
    text=Text Block
    font {
     bForceMonospaced=false
     bMaterialIsStencil=false
     fontMaterial=null
     fontObject=&BurbankSmall
     letterSpacing=0
     monospacedWidth=1
     size=24
     skewAmount=0
     typefaceFontName=Black
     outlineSettings {
      bApplyOutlineToDropShadows=false
      bMiteredCorners=false
      bSeparateFillAlpha=false
      outlineMaterial=null
      outlineSize=0
      outlineColor=<rgba 0 0 0 1>
     }
    }
    shadowOffset=<0 0>
   }
   slot {class=HorizontalBoxSlot}
  }
  CommonTextBlock_2 {
   widgetClassPath=&CommonTextBlock
   parent=HorizontalBox
   widget=CommonTextBlock_2
   properties {
    style=&TextStyle-BurbankSmall-XS-White_C
    text=km/h
    font {
     bForceMonospaced=false
     bMaterialIsStencil=false
     fontMaterial=null
     fontObject=&BurbankSmall
     letterSpacing=0
     monospacedWidth=1
     size=13
     skewAmount=0
     typefaceFontName=Black
     outlineSettings {
      bApplyOutlineToDropShadows=false
      bMiteredCorners=false
      bSeparateFillAlpha=false
      outlineMaterial=null
      outlineSize=0
      outlineColor=<rgba 0 0 0 1>
     }
    }
    shadowOffset=<0 0>
   }
   slot {class=HorizontalBoxSlot}
  }
  OverlayProgressBars {
   widgetClassPath=&Overlay
   parent=HBox_Root
   widget=OverlayProgressBars
   properties {visibility=Collapsed}
   slot {class=HorizontalBoxSlot}
  }
  CurrentHeightBar {
   widgetClassPath=/Script/UMG.ProgressBar
   bIsVariable=true
   parent=OverlayProgressBars
   widget=CurrentHeightBar
   properties {
    barFillType=BottomToTop
    percent=0.6
    fillColorAndOpacity=<rgba 0 0.5 1 1>
    widgetStyle {
     enableFillAnimation=false
     backgroundImage {
      bIsDynamicallyLoaded=false
      drawAs=Box
      imageType=FullColor
      mirroring=NoMirror
      resourceName=../../../Engine/Content/Slate/Common/ProgressBar_Background.png
      resourceObject=null
      tiling=NoTile
      imageSize=<0 0>
      margin=<ltrb 0.4167 0.4167 0.4167 0.4167>
      outlineSettings {
       bUseBrushTransparency=false
       roundingType=HalfHeightRadius
       width=0
       color {colorUseRule=UseColor_Specified, specifiedColor=<rgba 0 0 0 0>}
       cornerRadii=<0 0 0 0>
      }
      tintColor {colorUseRule=UseColor_Specified, specifiedColor=<rgba 1 1 1 0>}
      uVRegion {bIsValid=false, max=<0 0>, min=<0 0>}
     }
     fillImage {
      bIsDynamicallyLoaded=false
      drawAs=Box
      imageType=FullColor
      mirroring=NoMirror
      resourceName=../../../Engine/Content/Slate/Common/ProgressBar_Fill.png
      resourceObject=null
      tiling=NoTile
      imageSize=<0 0>
      margin=<ltrb 0.4167 0.4167 0.4167 0.4167>
      outlineSettings {
       bUseBrushTransparency=false
       roundingType=HalfHeightRadius
       width=0
       color {colorUseRule=UseColor_Specified, specifiedColor=<rgba 0 0 0 0>}
       cornerRadii=<0 0 0 0>
      }
      tintColor {colorUseRule=UseColor_Specified, specifiedColor=<rgba 1 1 1 1>}
      uVRegion {bIsValid=false, max=<0 0>, min=<0 0>}
     }
     marqueeImage {
      bIsDynamicallyLoaded=false
      drawAs=Image
      imageType=FullColor
      mirroring=NoMirror
      resourceName=../../../Engine/Content/Slate/Common/ProgressBar_Marquee.png
      resourceObject=null
      tiling=Horizontal
      imageSize=<20 12>
      margin=<ltrb 0 0 0 0>
      outlineSettings {
       bUseBrushTransparency=false
       roundingType=HalfHeightRadius
       width=0
       color {colorUseRule=UseColor_Specified, specifiedColor=<rgba 0 0 0 0>}
       cornerRadii=<0 0 0 0>
      }
      tintColor {colorUseRule=UseColor_Specified, specifiedColor=<rgba 1 1 1 1>}
      uVRegion {bIsValid=false, max=<0 0>, min=<0 0>}
     }
    }
   }
   slot {class=OverlaySlot, properties {verticalAlignment=VAlign_Fill}}
  }
  DeployHeightBar {
   widgetClassPath=&ProgressBar
   bIsVariable=true
   parent=OverlayProgressBars
   widget=DeployHeightBar
   properties {
    barFillType=BottomToTop
    percent=0.4
    fillColorAndOpacity=<rgba 0.4798 1 0.3867 1>
    widgetStyle {
     enableFillAnimation=false
     backgroundImage {
      bIsDynamicallyLoaded=false
      drawAs=Box
      imageType=FullColor
      mirroring=NoMirror
      resourceName=../../../Engine/Content/Slate/Common/ProgressBar_Background.png
      resourceObject=null
      tiling=NoTile
      imageSize=<0 0>
      margin=<ltrb 0.4167 0.4167 0.4167 0.4167>
      outlineSettings {
       bUseBrushTransparency=false
       roundingType=HalfHeightRadius
       width=0
       color {colorUseRule=UseColor_Specified, specifiedColor=<rgba 0 0 0 0>}
       cornerRadii=<0 0 0 0>
      }
      tintColor {colorUseRule=UseColor_Specified, specifiedColor=<rgba 1 1 1 0>}
      uVRegion {bIsValid=false, max=<0 0>, min=<0 0>}
     }
     fillImage {
      bIsDynamicallyLoaded=false
      drawAs=Box
      imageType=FullColor
      mirroring=NoMirror
      resourceName=../../../Engine/Content/Slate/Common/ProgressBar_Fill.png
      resourceObject=null
      tiling=NoTile
      imageSize=<0 0>
      margin=<ltrb 0.4167 0.4167 0.4167 0.4167>
      outlineSettings {
       bUseBrushTransparency=false
       roundingType=HalfHeightRadius
       width=0
       color {colorUseRule=UseColor_Specified, specifiedColor=<rgba 0 0 0 0>}
       cornerRadii=<0 0 0 0>
      }
      tintColor {colorUseRule=UseColor_Specified, specifiedColor=<rgba 1 1 1 1>}
      uVRegion {bIsValid=false, max=<0 0>, min=<0 0>}
     }
     marqueeImage {
      bIsDynamicallyLoaded=false
      drawAs=Image
      imageType=FullColor
      mirroring=NoMirror
      resourceName=../../../Engine/Content/Slate/Common/ProgressBar_Marquee.png
      resourceObject=null
      tiling=Horizontal
      imageSize=<20 12>
      margin=<ltrb 0 0 0 0>
      outlineSettings {
       bUseBrushTransparency=false
       roundingType=HalfHeightRadius
       width=0
       color {colorUseRule=UseColor_Specified, specifiedColor=<rgba 0 0 0 0>}
       cornerRadii=<0 0 0 0>
      }
      tintColor {colorUseRule=UseColor_Specified, specifiedColor=<rgba 1 1 1 1>}
      uVRegion {bIsValid=false, max=<0 0>, min=<0 0>}
     }
    }
   }
   slot {class=OverlaySlot, properties {verticalAlignment=VAlign_Fill}}
  }
  GroundHeightBar {
   widgetClassPath=&ProgressBar
   bIsVariable=true
   parent=OverlayProgressBars
   widget=GroundHeightBar
   properties {
    barFillType=BottomToTop
    percent=0.2
    fillColorAndOpacity=<rgba 1 0.4109 0.1269 1>
    widgetStyle {
     enableFillAnimation=false
     backgroundImage {
      bIsDynamicallyLoaded=false
      drawAs=Box
      imageType=FullColor
      mirroring=NoMirror
      resourceName=../../../Engine/Content/Slate/Common/ProgressBar_Background.png
      resourceObject=null
      tiling=NoTile
      imageSize=<0 0>
      margin=<ltrb 0.4167 0.4167 0.4167 0.4167>
      outlineSettings {
       bUseBrushTransparency=false
       roundingType=HalfHeightRadius
       width=0
       color {colorUseRule=UseColor_Specified, specifiedColor=<rgba 0 0 0 0>}
       cornerRadii=<0 0 0 0>
      }
      tintColor {colorUseRule=UseColor_Specified, specifiedColor=<rgba 1 1 1 0>}
      uVRegion {bIsValid=false, max=<0 0>, min=<0 0>}
     }
     fillImage {
      bIsDynamicallyLoaded=false
      drawAs=Box
      imageType=FullColor
      mirroring=NoMirror
      resourceName=../../../Engine/Content/Slate/Common/ProgressBar_Fill.png
      resourceObject=null
      tiling=NoTile
      imageSize=<0 0>
      margin=<ltrb 0.4167 0.4167 0.4167 0.4167>
      outlineSettings {
       bUseBrushTransparency=false
       roundingType=HalfHeightRadius
       width=0
       color {colorUseRule=UseColor_Specified, specifiedColor=<rgba 0 0 0 0>}
       cornerRadii=<0 0 0 0>
      }
      tintColor {colorUseRule=UseColor_Specified, specifiedColor=<rgba 1 1 1 1>}
      uVRegion {bIsValid=false, max=<0 0>, min=<0 0>}
     }
     marqueeImage {
      bIsDynamicallyLoaded=false
      drawAs=Image
      imageType=FullColor
      mirroring=NoMirror
      resourceName=../../../Engine/Content/Slate/Common/ProgressBar_Marquee.png
      resourceObject=null
      tiling=Horizontal
      imageSize=<20 12>
      margin=<ltrb 0 0 0 0>
      outlineSettings {
       bUseBrushTransparency=false
       roundingType=HalfHeightRadius
       width=0
       color {colorUseRule=UseColor_Specified, specifiedColor=<rgba 0 0 0 0>}
       cornerRadii=<0 0 0 0>
      }
      tintColor {colorUseRule=UseColor_Specified, specifiedColor=<rgba 1 1 1 1>}
      uVRegion {bIsValid=false, max=<0 0>, min=<0 0>}
     }
    }
   }
   slot {class=OverlaySlot, properties {verticalAlignment=VAlign_Fill, padding=<ltrb 0.5 0.5 0.5 0.5>}}
  }
  TutorialHighlight {widgetClassPath=&TutorialHighlight}
 ]
}
