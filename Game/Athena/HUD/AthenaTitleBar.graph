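# AthenaTitleBar is an activatable CommonUI widget that places a window title bar area over an invisible click-catching layer.
# A height-fixed SizeBox in the root Overlay holds the WindowTitleBarArea and stretches horizontally.
# A fully transparent button fills the overlay beneath it, and clicking that button deactivates the widget.
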WidgetBlueprint /Game/Athena/HUD/AthenaTitleBar
graphs [EventGraph {nodes [
   "On Clicked (ClickCatcher)"() {is=title K2Node_ComponentBoundEvent}
   > DeactivateWidget()
  ]}]
variables []
widgets {
 info {inheritedWidgetCount=0, parentClass=&CommonActivatableWidget, rootWidgetClass=&Overlay, widgetCount=4}
 widgets [
  Overlay_0 {widgetClassPath=&Overlay, widget=Overlay_0, properties {visibility=Visible}}
  ClickCatcher {
   widgetClassPath=&Button
   parent=Overlay_0
   widget=ClickCatcher
   properties {backgroundColor=<rgba 0 0 0 0>, colorAndOpacity=<rgba 0 0 0 0>}
   slot {class=OverlaySlot, properties {horizontalAlignment=HAlign_Fill, verticalAlignment=VAlign_Fill}}
  }
  SizeBox_0 {
   widgetClassPath=&SizeBox
   parent=Overlay_0
   widget=SizeBox_0
   properties {bOverride_HeightOverride=true, heightOverride=50}
   slot {class=OverlaySlot, properties {horizontalAlignment=HAlign_Fill}}
  }
  WindowTitleBarArea_22 {widgetClassPath=/Script/UMG.WindowTitleBarArea, parent=SizeBox_0, widget=WindowTitleBarArea_22, slot {class=SizeBoxSlot}}
 ]
}
