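# AthenaWatcherEye is a HUD widget showing an animated eye that reacts to spectators joining and leaving.
# Animations drive one Image's vertical scale and eye material parameters for lids and pupil.
# Begin_Watcher clears the transition timer, shows and opens the eye, then looks around or acts surprised; End_Watcher reverses Open and starts that timer.
# Spectator events play look-around, occasional mad, blink or surprise only when no animation is running.
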WidgetBlueprint /Game/Athena/HUD/AthenaWatcherEye
animations [
 Open {
  binding_count=1
  display_label=""
  length_seconds=0.25
  object=&self:Open
  sequence {bindings [I_Watcher {
     id=#1
     possessed_class=Image
     tracks [
      {
       type=MovieScene2DTransformTrack
       sections [{
         type=MovieScene2DTransformSection
         end_frame=5
         start_frame=0
         channels [
          Translation.X {type=MovieSceneScriptingFloatChannel}
          Translation.Y {type=MovieSceneScriptingFloatChannel}
          Angle {type=MovieSceneScriptingFloatChannel}
          Scale.X {type=MovieSceneScriptingFloatChannel, const [1,4]}
          Scale.Y {type=MovieSceneScriptingFloatChannel, keys [[0,0],[3,1.15],[5,1]]}
          Shear.X {type=MovieSceneScriptingFloatChannel}
          Shear.Y {type=MovieSceneScriptingFloatChannel}
         ]
        }]
      }
      {
       type=MovieSceneWidgetMaterialTrack
       display_name=Brush Material
       sections [{
         type=MovieSceneParameterSection
         end_frame=0
         start_frame=0
         channels [
          BotLid {type=MovieSceneScriptingFloatChannel, const [0,1]}
          TopLid {type=MovieSceneScriptingFloatChannel, const [0,1]}
          Pupil_Pos_X {type=MovieSceneScriptingFloatChannel, const [0,1]}
          Pupil_Pos_Y {type=MovieSceneScriptingFloatChannel, const [0,1]}
          Pupil_Radius {type=MovieSceneScriptingFloatChannel, const [0.45,1]}
         ]
        }]
      }
     ]
    }]}
 }
 LookAround {
  binding_count=1
  display_label=""
  length_seconds=2.25
  object=&self:LookAround
  sequence {bindings [I_Watcher {
     id=#2
     possessed_class=Image
     tracks [{
       type=MovieSceneWidgetMaterialTrack
       display_name=Brush Material
       sections [{
         type=MovieSceneParameterSection
         end_frame=45
         start_frame=0
         channels [
          Pupil_Pos_X {type=MovieSceneScriptingFloatChannel, keys [[0,0],[2,-0.2439],[12.9997,-0.2439],[16,0.2],[30,0.2],[45,0]]}
          TopLid {type=MovieSceneScriptingFloatChannel, const [0,1]}
          BotLid {type=MovieSceneScriptingFloatChannel, const [0,1]}
         ]
        }]
      }]
    }]}
 }
 Blink {
  binding_count=1
  display_label=""
  length_seconds=0.5
  object=&self:Blink
  sequence {bindings [I_Watcher {
     id=#3
     possessed_class=Image
     tracks [{
       type=MovieScene2DTransformTrack
       sections [{
         type=MovieScene2DTransformSection
         end_frame=10
         start_frame=0
         channels [
          Translation.X {type=MovieSceneScriptingFloatChannel}
          Translation.Y {type=MovieSceneScriptingFloatChannel}
          Angle {type=MovieSceneScriptingFloatChannel}
          Scale.X {type=MovieSceneScriptingFloatChannel, const [1,3]}
          Scale.Y {type=MovieSceneScriptingFloatChannel, keys [[0,1],[5,0],[10,1]]}
          Shear.X {type=MovieSceneScriptingFloatChannel}
          Shear.Y {type=MovieSceneScriptingFloatChannel}
         ]
        }]
      }]
    }]}
 }
 Surprise {
  binding_count=1
  display_label=""
  length_seconds=1.05
  object=&self:Surprise
  sequence {bindings [I_Watcher {
     id=#4
     possessed_class=Image
     tracks [{
       type=MovieSceneWidgetMaterialTrack
       display_name=Brush Material
       sections [{
         type=MovieSceneParameterSection
         end_frame=21
         start_frame=0
         channels [
          WideEye {type=MovieSceneScriptingFloatChannel, keys [[0,1],[4,0.2],[13,0.2],[18.9997,1]]}
          Pupil_Radius {
           type=MovieSceneScriptingFloatChannel
           keys [[2,0.45],[6.9997,0.3],[8,0.32],[8.9997,0.3],[10,0.32],[11,0.3],[11.9997,0.32],[13,0.3],[20,0.45]]
          }
          BotLid {type=MovieSceneScriptingFloatChannel, keys [[0,0],[2,-0.08],[5,0.18]]}
          TopLid {type=MovieSceneScriptingFloatChannel, const [0,1]}
         ]
        }]
      }]
    }]}
 }
 Blink_New {
  binding_count=1
  display_label=""
  length_seconds=0.5
  object=&self:Blink_New
  sequence {bindings [I_Watcher {
     id=#3
     possessed_class=Image
     tracks [
      {
       type=MovieScene2DTransformTrack
       sections [{
         type=MovieScene2DTransformSection
         end_frame=10
         start_frame=0
         channels [
          Translation.X {type=MovieSceneScriptingFloatChannel}
          Translation.Y {type=MovieSceneScriptingFloatChannel}
          Angle {type=MovieSceneScriptingFloatChannel}
          Scale.X {type=MovieSceneScriptingFloatChannel}
          Scale.Y {type=MovieSceneScriptingFloatChannel}
          Shear.X {type=MovieSceneScriptingFloatChannel}
          Shear.Y {type=MovieSceneScriptingFloatChannel}
         ]
        }]
      }
      {
       type=MovieSceneWidgetMaterialTrack
       display_name=Brush Material
       sections [{
         type=MovieSceneParameterSection
         end_frame=10
         start_frame=0
         channels [TopLid {type=MovieSceneScriptingFloatChannel, keys [[0,0],[4,1],[10,0]]}]
        }]
      }
     ]
    }]}
 }
 Mad {
  binding_count=1
  display_label=""
  length_seconds=1.1
  object=&self:Mad
  sequence {bindings [I_Watcher {
     id=#3
     possessed_class=Image
     tracks [
      {
       type=MovieScene2DTransformTrack
       sections [{
         type=MovieScene2DTransformSection
         end_frame=10
         start_frame=0
         channels [
          Translation.X {type=MovieSceneScriptingFloatChannel}
          Translation.Y {type=MovieSceneScriptingFloatChannel}
          Angle {type=MovieSceneScriptingFloatChannel}
          Scale.X {type=MovieSceneScriptingFloatChannel}
          Scale.Y {type=MovieSceneScriptingFloatChannel}
          Shear.X {type=MovieSceneScriptingFloatChannel}
          Shear.Y {type=MovieSceneScriptingFloatChannel}
         ]
        }]
      }
      {
       type=MovieSceneWidgetMaterialTrack
       display_name=Brush Material
       sections [{
         type=MovieSceneParameterSection
         end_frame=18
         start_frame=-5
         channels [
          TopLid {type=MovieSceneScriptingFloatChannel, keys [[-4.0003,0],[0,0.38],[11,0.38],[17.9997,0]]}
          Pupil_Pos_Y {type=MovieSceneScriptingFloatChannel}
          Pupil_Pos_X {type=MovieSceneScriptingFloatChannel}
          BotLid {type=MovieSceneScriptingFloatChannel, keys [[-4.0003,0],[1,0.34],[12,0.34],[18,0]]}
         ]
        }]
      }
     ]
    }]}
 }
]
graphs [EventGraph {nodes [
   n1 = Begin_Watcher!() -> (OutputDelegate, then, ManySpectatorsJoining:bool=~) {is=title K2Node_CustomEvent}
   > Clear and Invalidate Timer by Handle(Handle=Get WatcherFinishedTransitionTimer() -> (WatcherFinishedTransitionTimer:Timer Handle Structure))
   > n2 = Branch(Condition=IsAnyAnimationPlaying!(self:User Widget Object Reference) -> (ReturnValue:bool))
   > n3 = StopAllAnimations(self:User Widget Object Reference)
   Show_Watcher(execute=[@n3.then, @n2.else])
   n4 = Get Open() -> (Open:Widget Animation Object Reference)
   > PlayAnimation(InAnimation=@n4)
   > Delay(Duration=GetEndTime(self=@n4))
   > PlayAnimation(InAnimation=Select(
     Option 0:Widget Animation Object Reference=Get LookAround() -> (LookAround:Widget Animation Object Reference)
     Option 1:Widget Animation Object Reference=Get Surprise() -> (Surprise:Widget Animation Object Reference)
     Index=@n1.ManySpectatorsJoining
    ) -> (ReturnValue:Widget Animation Object Reference))
   SpectatorAdded_Watcher() {is=title K2Node_CustomEvent}
   > n5 = Branch(Condition=IsAnyAnimationPlaying())
   n6 = RandomFloatInRange(Min=0, Max=100)
   PlayAnimation(
    execute=@n5.else
    InAnimation=Select(
     Option 0=Get LookAround()
     Option 1=Get Mad() -> (Mad:Widget Animation Object Reference)
     Index="float > float"(A=@n6, B=85) {is=title K2Node_CallFunction}
    )
   )
   SpectatorRemoved_Watcher() {is=title K2Node_CustomEvent}
   > n7 = Branch(Condition=IsAnyAnimationPlaying())
   PlayAnimation(execute=@n7.else, InAnimation=Get Blink_New() -> (Blink_New:Widget Animation Object Reference))
   ManySpectatorsAdded_Watcher() {is=title K2Node_CustomEvent}
   > n8 = Branch(Condition=IsAnyAnimationPlaying())
   PlayAnimation(execute=@n8.else, InAnimation=Get Surprise())
   End_Watcher() {is=title K2Node_CustomEvent}
   n9 = Get Open()
   > PlayAnimation(InAnimation=@n9, PlayMode=Reverse)
   > n10 = Set Timer by Event(Delegate=Create Event(), Time=GetEndTime(self=@n9), bLooping=false)
   > Set WatcherFinishedTransitionTimer(WatcherFinishedTransitionTimer:Timer Handle Structure=@n10) -> (Output_Get:Timer Handle Structure)
   Hide_Watcher() {is=title K2Node_CustomEvent}
   n11 = Get I_Watcher() -> (I_Watcher:Image Object Reference)
   > SetVisibility(self=@n11, InVisibility=Hidden)
   Show_Watcher() {is=title K2Node_CustomEvent}
   > SetVisibility(self=@n11, InVisibility=SelfHitTestInvisible)
  ]}]
variables [WatcherFinishedTransitionTimer {type=TimerHandle}]
widgets {
 info {inheritedWidgetCount=0, parentClass=&UserWidget, rootWidgetClass=&Image, widgetCount=1}
 widgets [I_Watcher {
   widgetClassPath=&Image
   bIsVariable=true
   widget=I_Watcher
   properties {
    visibility=Hidden
    brush {
     bIsDynamicallyLoaded=false
     drawAs=Image
     imageType=NoImage
     mirroring=NoMirror
     resourceName=None
     resourceObject=/Game/Athena/HUD/Art/Materials/M_HUD_WatcherEye
     tiling=NoTile
     imageSize=<42 42>
     margin=<ltrb 0 0 0 0>
     outlineSettings {
      bUseBrushTransparency=false
      roundingType=HalfHeightRadius
      width=0
      color {colorUseRule=UseColor_Specified, specifiedColor=<rgba 0 0 0 0>}
      cornerRadii=<0 0 0 0>
     }
     tintColor {colorUseRule=UseColor_Specified, specifiedColor=<rgba 1 1 1 1>}
     uVRegion {bIsValid=false, max=<0 0>, min=<0 0>}
    }
   }
  }]
}
