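# HUD progress bar for Creative minigames, faded in when an allowed activity (default Race) starts and out when it ends.
# OnPlayerStatUpdate caches scores, scales the player's bar against the goal, and positions three rival markers from sorted team scores.
# Optional text shows placement and the lead or deficit.
# UpdateCloseToVictory compares the top score to entry/exit thresholds, toggling music and sending a friend or enemy message through the game state.
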WidgetBlueprint /Game/Athena/HUD/BP_Creative_MinigameProgressWidget
animations [
 PlayerProgress_PotentialIntro {
  binding_count=1
  length_seconds=0.25
  object=&self:PlayerProgress_PotentialIntro
  sequence {bindings [CommonBorderPlayerProgress_Potential {
     id=#1
     possessed_class=CommonBorder
     tracks [{
       type=MovieSceneFloatTrack
       display_name=Render Opacity
       sections [{type=MovieSceneFloatSection, end_frame=5, start_frame=0, channels [None {type=MovieSceneScriptingFloatChannel, keys [[0,0],[5,1]]}]}]
      }]
    }]}
 }
 Intro {
  binding_count=1
  length_seconds=0.5
  object=&self:Intro
  sequence {bindings [BP_Creative_MinigameProgressWidget_C_6 {
     id=#2
     possessed_class=BP_Creative_MinigameProgressWidget_C
     tracks [{
       type=MovieSceneFloatTrack
       display_name=Render Opacity
       sections [{type=MovieSceneFloatSection, end_frame=10, start_frame=0, channels [None {type=MovieSceneScriptingFloatChannel, keys [[0,0],[10,1]]}]}]
      }]
    }]}
 }
]
graphs [
 DisableCloseToVictoryMusic {nodes [
   DisableCloseToVictoryMusic() {is=title K2Node_FunctionEntry}
   n1 = Get CloseToVictoryMusicInstance() -> (CloseToVictoryMusicInstance:Audio Component Object Reference)
   > Is Valid(InputObject=@n1)
   > FadeOut!(
    execute
    self:Audio Component Object Reference=@n1
    FadeOutDuration:float=Get CloseToVictoryMusicFade() -> (CloseToVictoryMusicFade:double)
    FadeVolumeLevel:float=0
    FadeCurve:EAudioFaderCurve Enum=Linear
   ) -> (then)
  ]}
 DisplayCloseToVictoryMessage {nodes [
   n1 = DisplayCloseToVictoryMessage!() -> (then, FriendCloseToVictory:bool=~) {is=title K2Node_FunctionEntry}
   n2 = GetGameStateAthena!() -> (ReturnValue:Fort Game State Athena Object Reference)
   > Is Valid(InputObject=@n2)
   n3 = Get CloseToVictoryFriendMessage() -> (CloseToVictoryFriendMessage:Text)
   n4 = Get CloseToVictoryEnemyMessage() -> (CloseToVictoryEnemyMessage:Text)
   > HandleGameModeMessageRequest!(
    execute
    self:Fort Game State Athena Object Reference=@n2
    MessageData:Athena Game Message Data Structure=Make Athena Game Message Data!(
     MsgType:EAthenaGameMsgType Enum=DefaultMessage
     MsgText:Text=Select(Option 0:Text=@n4, Option 1:Text=@n3, Index=@n1) -> (ReturnValue:Text)
     MsgSound:Sound Cue Object Reference
     MsgMediaSource:Media Source Object Reference
     MsgDelay:float=0
     bIsTeamBased:bool=False
     TeamIndex:int=0
     DisplayTime:float=5
     MessageFlags:EAthenaGameMessageDataFlags Enum=None
     TargetPlayerController:Player Controller Object Reference
    ) -> (AthenaGameMessageData:Athena Game Message Data Structure) {is=title K2Node_MakeStruct}
   ) -> (then)
  ]}
 EnableCloseToVictoryMusic {nodes [
   EnableCloseToVictoryMusic!() -> (then, Friend Close To Victory:bool=~) {is=title K2Node_FunctionEntry}
   n1 = Get CloseToVictoryMusic() -> (CloseToVictoryMusic:Sound Cue Object Reference)
   > n2 = Is Valid(InputObject=@n1)
   n3 = Get CloseToVictoryMusicInstance()
   FadeOut(execute=@n2.Is Not Valid, self=@n3, FadeOutDuration=Get CloseToVictoryMusicFade(), FadeVolumeLevel=0, FadeCurve=Linear)
   Stop(execute=@n2.Is Valid, self:Audio Component Object Reference=@n3)
   > n4 = SpawnSound2D!(
    execute
    Sound:Sound Base Object Reference=@n1
    VolumeMultiplier:float=1
    PitchMultiplier:float=1
    StartTime:float=0
    ConcurrencySettings:Sound Concurrency Object Reference
    bPersistAcrossLevelTransition:bool=false
    bAutoDestroy:bool=true
   ) -> (then, ReturnValue:Audio Component Object Reference)
   > Set CloseToVictoryMusicInstance(CloseToVictoryMusicInstance:Audio Component Object Reference=@n4) -> (Output_Get:Audio Component Object Reference)
  ]}
 EventGraph {nodes [
   Event Construct()
   > Parent: Construct()
   > EventSetProgress!(
    execute
    self:Team Placement Progress Enemy Object Reference=[
     Get ProgressEnemy_2() -> (ProgressEnemy_2:Team Placement Progress Enemy Object Reference)
     Get ProgressEnemy_3() -> (ProgressEnemy_3:Team Placement Progress Enemy Object Reference)
     Get ProgressEnemy_1() -> (ProgressEnemy_1:Team Placement Progress Enemy Object Reference)
    ]
    isFirst:bool=false
    Progress:double=0
   ) -> (then)
   n1 = Event OnPlayerStatUpdate!() -> (OutputDelegate, then, CurrentPlayerTeamScore:int=0, CurrentPlayerTeamPlacement:int=0, GoalScore:int=0, SortedTeamStats:Array of Integers) {is=title K2Node_Event}
   > n2 = Sequence!(execute) -> (then_0, then_1, then_2:Exec, then_3:Exec, then_4:Exec, then_5:Exec, then_6:Exec)
   Branch(execute=@n2.then_1, Condition=Get ScoreUpdated() -> (ScoreUpdated:bool))
   n3 = float / float(A="To Float (Integer)"(InInt=Get CurrentPlayerTeamScore() -> (CurrentPlayerTeamScore:int=0)), B=GetGoalScoreSafe!() -> (ReturnValue:double)) {is=title K2Node_CallFunction}
   > SetDesiredSizeScale!(
    execute
    self:Border Object Reference=Get CommonBorderPlayerProgress_Potential() -> (CommonBorderPlayerProgress_Potential:Common Border Object Reference)
    InScale:Vector 2D Structure=MakeVector2D(X="Clamp (Float)"(Value=@n3, Min=0, Max=1), Y=1)
   ) -> (then)
   > SetPlayerProgress()
   > SetCurrentValue!(
    execute
    self:Common Numeric Text Block Object Reference=Get TargetScoreText() -> (TargetScoreText:Common Numeric Text Block Object Reference)
    NewValue:float=GetGoalScoreSafe()
   ) -> (then)
   SetEnemyProgress!(
    execute=@n2.then_3
    self
    TeamProgressWidget:Team Placement Progress Enemy Object Reference=Get ProgressEnemy_3()
    PositionIndex:int=int - int(A=Get CurrentPlayerTeamPlacement() -> (CurrentPlayerTeamPlacement:int=0), B=2) {is=title K2Node_CallFunction}
   ) -> (then)
   n4 = "To Float (Integer)"(InInt=@n1.GoalScore)
   n5 = Get GoalScore() -> (GoalScore:double)
   n6 = Get CurrentPlayerTeamScore() -> (CurrentPlayerTeamScore=0)
   n7 = Set ScoreUpdated(
    execute=@n2.then_0
    ScoreUpdated:bool=OR Boolean(
     A="Not Equal (Integer)"!(A:int=@n1.CurrentPlayerTeamScore, B:int=@n6) -> (ReturnValue:bool) {is=title K2Node_CallFunction}
     B="Not Equal (Float)"!(A:double=@n5, B:double=@n4) -> (ReturnValue:bool) {is=title K2Node_CallFunction}
    )
   ) -> (Output_Get:bool)
   > n8 = Branch(Condition=@n7)
   n9 = GetGoalScoreSafe()
   n10 = "To Float (Integer)"(InInt=Get CurrentPlayerTeamScore() -> (CurrentPlayerTeamScore=0))
   > n11 = Set CurrentProgress(CurrentProgress:double="Clamp (Float)"(Value=float / float(A=@n10, B=@n9), Min=0, Max=1)) -> (Output_Get:double)
   Set CurrentPlayerTeamScore(execute=[@n11.then, @n8.else], CurrentPlayerTeamScore:int=@n1.CurrentPlayerTeamScore) -> (Output_Get:int=0)
   > Set CurrentPlayerTeamPlacement(CurrentPlayerTeamPlacement:int=@n1.CurrentPlayerTeamPlacement) -> (Output_Get:int=0)
   > Set GoalScore(GoalScore:double="To Float (Integer)"(InInt=@n1.GoalScore)) -> (Output_Get:double)
   > n12 = Set SortedTeamScores(SortedTeamScores:Array of Integers=@n1.SortedTeamStats) -> (Output_Get:Array of Integers)
   > Set TotalTeams(TotalTeams:int=Length(TargetArray:Array of Integers=@n12)) -> (Output_Get:int=0)
   n13 = For Each Loop(
    Exec=@n2.then_2
    Array:Array of Team Placement Progress Enemy Object References=Make Array!(
     "[0]":Team Placement Progress Enemy Object Reference=Get ProgressEnemy_1()
     "[1]":Team Placement Progress Enemy Object Reference=Get ProgressEnemy_2()
    ) -> (Array:Array of Team Placement Progress Enemy Object References)
   ) -> (Array Element:Team Placement Progress Enemy Object Reference)
   > SetEnemyProgress(TeamProgressWidget=@n13, PositionIndex=@n13.Array Index)
   n14 = Get UseCloseToVictoryMusicAndMessage() -> (UseCloseToVictoryMusicAndMessage:Scalable Float Structure)
   Branch(
    execute=@n2.then_6
    Condition="float > float"(A=GetValueAtLevel!(Input:"Scalable Float Structure (by ref)"=@n14, Level:float=0) -> (ReturnValue:float), B=0) {is=title K2Node_CallFunction}
   )
   > UpdateCloseToVictory()
   n15 = Get ScoreUpdated()
   Branch(
    execute=@n2.then_4
    Condition=AND Boolean(A=Get DisplayLeaderRelativeMessageConst() -> (DisplayLeaderRelativeMessageConst:bool), B=NOT Boolean(A=@n15))
   )
   > SetLeaderReleativeMessage()
   Branch(execute=@n2.then_5, Condition=Get DisplayLocalTeamPlacementUnderCaratConst() -> (DisplayLocalTeamPlacementUnderCaratConst:bool))
   n16 = Get SortedTeamScores() -> (SortedTeamScores:Array of Integers)
   > Branch(Condition=Is Valid Index!(TargetArray:Array of Integers=@n16, IndexToTest:int) -> (ReturnValue:bool))
   n17 = Get Text_LocalTeamPlacement() -> (Text_LocalTeamPlacement:Common Text Object Reference)
   n18 = "Get (a copy)"(Array:Array of Integers=@n16) -> (Output:int)
   > SetVisibility(
    self=@n17
    InVisibility=Select(
     Option 0:ESlateVisibility Enum=HitTestInvisible
     Option 1:ESlateVisibility Enum=Collapsed
     Index="Equal (Integer)"(A=@n18, B=0) {is=title K2Node_CallFunction}
    ) -> (ReturnValue:ESlateVisibility Enum)
   )
   > "SetText (Text)"(
    self=@n17
    InText=Format Text!(Format="LOCTEXT(\"## {Team Placement}\")", Team Placement:int=Get CurrentPlayerTeamPlacement() -> (CurrentPlayerTeamPlacement=0)) -> (Result)
   )
   n19 = Event On Handle Minigame Started!() -> (OutputDelegate, then, StartData:"Minigame Activity Started Data Structure (by ref)") {is=title K2Node_Event}
   > Branch(Condition=Contains Item(
     TargetArray:Array of Texts=Get AllowedMinigameActivities() -> (AllowedMinigameActivities:Array of Texts)
     ItemToFind:"Text (by ref)"=Break Minigame Activity Started Data!(MinigameActivityStartedData:"Minigame Activity Started Data Structure (by ref)"=@n19.StartData) -> (MinigameActivityName:Text, Stat:EMinigameActivityStat Enum, InitialStatValue:float=~, InitialScore:int, TotalScore:int) {is=title K2Node_BreakStruct}
    ))
   > SetVisibility(InVisibility=HitTestInvisible)
   > PlayAnimation(InAnimation=Get Intro())
   n20 = Event PreConstruct()
   > Parent: PreConstruct(IsDesignTime=@n20.IsDesignTime)
   > n21 = Sequence!(execute) -> (then_0, then_1)
   > SetBrushTintColor!(
    execute
    self:Image Object Reference=Get Image_Progress() -> (Image_Progress:Image Object Reference)
    TintColor:Slate Color Structure=Get PlayerProgressTint() -> (PlayerProgressTint:Slate Color Structure)
   ) -> (then)
   > SetBrushTintColor(
    self=Get Image_Progress_Potential() -> (Image_Progress_Potential:Image Object Reference)
    TintColor=Get PlayerPotentialProgressTint() -> (PlayerPotentialProgressTint:Slate Color Structure)
   )
   > SetPadding!(
    execute
    self:Border Object Reference=Get CommonBorder_TopPaddingContainer() -> (CommonBorder_TopPaddingContainer:Common Border Object Reference)
    InPadding:Margin Structure=Make Margin(Left=0, Top=Get TopPadding() -> (TopPadding:double), Right=0, Bottom=0)
   ) -> (then)
   n22 = Get ScoreTypeTextBox() -> (ScoreTypeTextBox:Common Text Object Reference)
   SetVisibility(
    execute=@n21.then_1
    self=@n22
    InVisibility=Select(Option 0=Collapsed, Option 1=SelfHitTestInvisible, Index=Get DisplayScoreTypeTextConst() -> (DisplayScoreTypeTextConst:bool))
   )
   > "SetText (Text)"(self=@n22, InText=Get ScoreTypeTextConst() -> (ScoreTypeTextConst:Text))
   > SetVisibility(
    self=Get LeaderRelativeMessageBox() -> (LeaderRelativeMessageBox:Horizontal Box Object Reference)
    InVisibility=Select(Option 0=Collapsed, Option 1=HitTestInvisible, Index=Get DisplayLeaderRelativeMessageConst())
   )
   Event On Handle Minigame Ended!() -> (OutputDelegate, then, EndData:"Minigame Activity Ended Data Structure (by ref)") {is=title K2Node_Event}
   > n23 = "Play Animation with Finished event (legacy)"!(
    execute:Exec
    Widget:User Widget Object Reference=Self-Reference()
    InAnimation:Widget Animation Object Reference=Get Intro()
    StartAtTime:float=0
    NumLoopsToPlay:int=1
    PlayMode:EUMGSequencePlayMode Enum=Reverse
    PlaybackSpeed:float=1
   ) -> (then:Exec, Result:UMGSequence Player Object Reference, Finished:Exec) {is=title K2Node_PlayAnimation}
   SetVisibility(execute=@n23.Finished, InVisibility=Collapsed)
   "Play Animation with Finished event (legacy)"(
    Widget=Self-Reference()
    InAnimation=Get PlayerProgress_PotentialIntro() -> (PlayerProgress_PotentialIntro:Widget Animation Object Reference)
    StartAtTime=0
    NumLoopsToPlay=1
    PlaybackSpeed=1
   )
  ]}
 GetGoalScoreSafe {nodes [
   GetGoalScoreSafe() {is=title K2Node_FunctionEntry}
   > Return Node!(execute, ReturnValue:double="Max (Float)"(A=Get GoalScore(), B=1))
  ]}
 InterpProgress {nodes [
   InterpProgress() {is=title K2Node_FunctionEntry}
   n1 = GetGoalScoreSafe() {is=title K2Node_CallFunction}
   n2 = "To Float (Integer)"(InInt=Get CurrentPlayerTeamScore() -> (CurrentPlayerTeamScore=0))
   > n3 = Set ProgressGoal(ProgressGoal:double="Clamp (Float)"(Value=float / float(A=@n2, B=@n1), Min=0, Max=1)) -> (Output_Get:double)
   > Set CurrentProgress(CurrentProgress=@n3)
   > SetDesiredSizeScale(
    self=Get CommonBorderPlayerProgress() -> (CommonBorderPlayerProgress:Common Border Object Reference)
    InScale=MakeVector2D(X=Get CurrentProgress() -> (CurrentProgress:double), Y=1)
   )
   > Branch(Condition="Equal (Float)"!(A:double=Get ProgressGoal() -> (ProgressGoal:double), B:double=Get CurrentProgress()) -> (ReturnValue:bool) {is=title K2Node_CallFunction})
   > SetActiveWidget(
    self:Widget Switcher Object Reference=Get CommonWidgetSwitcher_PlayerProgress() -> (CommonWidgetSwitcher_PlayerProgress:Common Widget Switcher Legacy Object Reference)
    Widget=Get Image_Progress()
   )
   > Clear and Invalidate Timer by Handle(Handle=Get InterpProgressTimer() -> (InterpProgressTimer:Timer Handle Structure))
   FInterpTo_Constant!(
    Current:double=Get CurrentProgress()
    Target:double=Get ProgressGoal()
    DeltaTime:double=Get InterpUpdateTime() -> (InterpUpdateTime:double)
    InterpSpeed:double=Get ProgressInterpSpeed() -> (ProgressInterpSpeed:double)
   ) -> (ReturnValue:double)
  ]}
 SetEnemyProgress {nodes [
   n1 = SetEnemyProgress!() -> (then, TeamProgressWidget:Team Placement Progress Enemy Object Reference, PositionIndex:int) {is=title K2Node_FunctionEntry}
   n2 = Get SortedTeamScores()
   n3 = int - int(A=Get CurrentPlayerTeamPlacement() -> (CurrentPlayerTeamPlacement=0), B=1)
   n4 = Get PositionIndex() -> (PositionIndex:int=0)
   > n5 = Branch(Condition=AND Boolean(A=Is Valid Index(TargetArray=@n2, IndexToTest=@n1.PositionIndex), B="Not Equal (Integer)"(A=@n3, B=@n4)))
   n6 = GetGoalScoreSafe()
   n7 = "Get (a copy)"(Array=Get SortedTeamScores(), Dimension 1=Get PositionIndex() -> (PositionIndex=0))
   n8 = Get TeamProgressWidget() -> (TeamProgressWidget:Team Placement Progress Enemy Object Reference)
   > EventSetProgress(self=@n8, isFirst=false, Progress=float / float(A="To Float (Integer)"(InInt=@n7), B="Clamp (Float)"(Value=@n6, Min=1, Max=@n6)))
   EventSetProgress(execute=@n5.else, self=@n8, isFirst=false, Progress=0)
  ]}
 SetLeaderReleativeMessage {nodes [
   SetLeaderReleativeMessage() {is=title K2Node_FunctionEntry}
   > n1 = Sequence()
   n2 = Get SortedTeamScores()
   > Branch(Condition=Is Valid Index(TargetArray=@n2))
   > n3 = Set CurrentTopScore(CurrentTopScore:int="Get (a copy)"(Array=@n2)) -> (Output_Get:int=0)
   > SetVisibility(self=Get LeaderRelativeMessageBox(), InVisibility=Select(Option 0=HitTestInvisible, Option 1=Collapsed, Index="Equal (Integer)"(A=@n3, B=0)))
   n4 = Branch(
    execute=@n1.then_1
    Condition="Equal (Integer)"(A=Get CurrentPlayerTeamScore() -> (CurrentPlayerTeamScore=0), B=Get CurrentTopScore() -> (CurrentTopScore:int=0))
   )
   n5 = Get SortedTeamScores()
   > Branch(Condition=Is Valid Index(TargetArray=@n5, IndexToTest=1))
   > n6 = Set NextLeadingScore(NextLeadingScore:int="Get (a copy)"(Array=@n5, Dimension 1=1)) -> (Output_Get:int=0)
   > n7 = Branch(Condition="Equal (Integer)"(A=@n6, B=Get CurrentPlayerTeamScore() -> (CurrentPlayerTeamScore=0)))
   > "SetText (Text)"(
    self=Get RelativeLeaderMessage() -> (RelativeLeaderMessage:Common Text Object Reference)
    InText=Get LeaderRelativeTiedMessage() -> (LeaderRelativeTiedMessage:Text)
   )
   > "SetText (Text)"(self=Get LeaderRelativePointValue() -> (LeaderRelativePointValue:Common Text Object Reference))
   "SetText (Text)"(execute=@n7.else, self=Get RelativeLeaderMessage(), InText=Get LeaderRelativeMessageLeading() -> (LeaderRelativeMessageLeading:Text))
   n8 = Get CurrentPlayerTeamScore() -> (CurrentPlayerTeamScore=0)
   n9 = Get NextLeadingScore() -> (NextLeadingScore:int=0)
   n10 = Get LeaderRelativePointValue()
   > "SetText (Text)"(self=@n10, InText="To Text (Integer)"(Value=int - int(A=@n8, B=@n9)))
   > SetColorAndOpacity(self=@n10, InColorAndOpacity=Get LeadingTeamPointValueColor() -> (LeadingTeamPointValueColor:Slate Color Structure))
   "SetText (Text)"(execute=@n4.else, self=Get RelativeLeaderMessage(), InText=Get LeaderRelativeMessageTrailing() -> (LeaderRelativeMessageTrailing:Text))
   n11 = Get CurrentPlayerTeamScore() -> (CurrentPlayerTeamScore=0)
   n12 = Get CurrentTopScore() -> (CurrentTopScore=0)
   n13 = Get LeaderRelativePointValue()
   > "SetText (Text)"(self=@n13, InText="To Text (Integer)"(Value=int - int(A=@n12, B=@n11)))
   > SetColorAndOpacity(self=@n13, InColorAndOpacity=Get LosingPointValueColor() -> (LosingPointValueColor:Slate Color Structure))
  ]}
 SetPlayerProgress {nodes [
   SetPlayerProgress() {is=title K2Node_FunctionEntry}
   > InterpProgress() {is=title K2Node_CallFunction}
   n1 = Truncate(A=GetGoalScoreSafe())
   n2 = Get CurrentPlayerTeamScore() -> (CurrentPlayerTeamScore=0)
   > "SetText (Text)"(
    self=Get Text_YourPlacement() -> (Text_YourPlacement:Common Text Object Reference)
    InText="To Text (Integer)"(Value="Clamp (Integer)"(Value=@n2, Min=0, Max=@n1) -> (ReturnValue=0), bUseGrouping=false)
   )
  ]}
 UpdateCloseToVictory {nodes [
   UpdateCloseToVictory() {is=title K2Node_FunctionEntry}
   n1 = Get GoalScore()
   n2 = GetGameStateAthena()
   n3 = HasMatchEnded!(self:Game State Base Object Reference=@n2) -> (ReturnValue:bool)
   n4 = Length(TargetArray=Get SortedTeamScores())
   > n5 = Branch(Condition=AND Boolean!(
     A="float > float"(A=@n1, B=0)
     B="integer > integer"(A=@n4, B=0) {is=title K2Node_CallFunction}
     C:bool=IsValid(Object=@n2)
     D:bool=NOT Boolean(A=@n3)
    ) -> (ReturnValue))
   n6 = Get CloseToVictoryEntryThreshold() -> (CloseToVictoryEntryThreshold:Scalable Float Structure)
   n7 = "Get (a copy)"(Array=Get SortedTeamScores())
   n8 = float / float(A="To Float (Integer)"(InInt=@n7), B=Get GoalScore())
   > n9 = Branch(Condition="float >= float"(A=@n8, B=GetValueAtLevel(Input=@n6, Level=0)) {is=title K2Node_CallFunction})
   > n10 = Branch(Condition=Get CloseToVictoryActive() -> (CloseToVictoryActive:bool))
   Set CloseToVictoryActive(execute=@n10.else, CloseToVictoryActive:bool=true) -> (Output_Get:bool)
   n11 = "Equal (Integer)"(A=Get CurrentPlayerTeamScore() -> (CurrentPlayerTeamScore=0), B=@n7)
   > DisplayCloseToVictoryMessage!(execute, self, FriendCloseToVictory:bool=@n11) -> (then) {is=title K2Node_CallFunction}
   > EnableCloseToVictoryMusic!(execute, self, Friend Close To Victory:bool=@n11) -> (then) {is=title K2Node_CallFunction}
   n12 = GetValueAtLevel(Input=Get CloseToVictoryExitThreshold() -> (CloseToVictoryExitThreshold:Scalable Float Structure), Level=0)
   n13 = Branch(
    execute=@n9.else
    Condition=AND Boolean!(A="float < float"!(A:double=@n8, B:double=@n12) -> (ReturnValue:bool) {is=title K2Node_CallFunction}, B=Get CloseToVictoryActive()) -> (ReturnValue)
   )
   Set CloseToVictoryActive(execute=[@n13.then, @n14.then], CloseToVictoryActive=false)
   > DisableCloseToVictoryMusic() {is=title K2Node_CallFunction}
   n14 = Branch(execute=@n5.else, Condition=Get CloseToVictoryActive())
  ]}
]
variables [
 AllowedMinigameActivities {type="array<string>", default [Race]}
 CloseToVictoryActive {type=boolean}
 CloseToVictoryEnemyMessage {type=string, default=""}
 CloseToVictoryEntryThreshold {type=ScalableFloat, default {registryType=None, value=0.95, curve {curveTable=null, rowName=None}}}
 CloseToVictoryExitThreshold {type=ScalableFloat, default {registryType=None, value=0.95, curve {curveTable=null, rowName=None}}}
 CloseToVictoryFriendMessage {type=string, default=""}
 CloseToVictoryMusic {type=SoundCue, default=null, class=~, properties=~}
 CloseToVictoryMusicFade {type=number}
 CloseToVictoryMusicInstance {type=AudioComponent, default=null}
 CurrentPlayerTeamPlacement {type=integer}
 CurrentPlayerTeamScore {type=integer}
 CurrentProgress {type=number}
 DisplayLeaderRelativeMessageConst {type=boolean}
 DisplayLocalTeamPlacementUnderCaratConst {type=boolean}
 DisplayScoreTypeTextConst {type=boolean}
 GoalIconImage {type=Texture2D, default=/Game/UI/Icons/Icon-Coin/Icon-Dungeon-Coin-S}
 GoalScore {type=number}
 InterpProgressFunctionName {type=string, default=InterpProgress}
 InterpProgressTimer {type=TimerHandle}
 InterpUpdateTime {type=number, default=0.016}
 IsOutro {type=boolean}
 LeaderRelativeMessageLeading {type=string, default=LEADING BY}
 LeaderRelativeMessageTrailing {type=string, default=TRAILING BY}
 LeaderRelativeTiedMessage {type=string, default=TIED FOR THE LEAD}
 LeadingTeamPointValueColor {type=SlateColor, default {colorUseRule=UseColor_Specified, specifiedColor=<rgba 0.0203 0.7835 0.9823 1>}}
 LosingPointValueColor {type=SlateColor, default {colorUseRule=UseColor_Specified, specifiedColor=<rgba 1 0 0 1>}}
 PlayerPotentialProgressTint {type=SlateColor, default {colorUseRule=UseColor_Specified, specifiedColor=<rgba 0.3613 0.9734 1 1>}}
 PlayerProgressTint {type=SlateColor, default {colorUseRule=UseColor_Specified, specifiedColor=<rgba 0.0203 0.7835 0.9823 1>}}
 PreviousTeamScore {type=integer}
 ProgressGoal {type=number}
 ProgressInterpDelay {type=number}
 ProgressInterpSpeed {type=number, default=5}
 ScoreTypeTextConst {type=string, default=pts}
 ScoreUpdated {type=boolean}
 SortedTeamScores {type="array<integer>"}
 TopPadding {type=number, default=100}
 TotalTeams {type=integer}
 UseCloseToVictoryMusicAndMessage {type=ScalableFloat, default {registryType=None, value=1, curve {curveTable=null, rowName=None}}}
]
widgets {
 info {inheritedWidgetCount=0, parentClass=/Script/FortniteUI.CreativeMinigameProgressWidget, rootWidgetClass=&Overlay, widgetCount=49}
 widgets [
  OverlayProgressContainer {
   widgetClassPath=&Overlay
   widget=OverlayProgressContainer
   properties {renderTransform {angle=0, scale=<1 1>, shear=<0 0>, translation=<0 -30>}}
  }
  CommonBorder_TopPaddingContainer {
   widgetClassPath=&CommonBorder
   bIsVariable=true
   parent=OverlayProgressContainer
   widget=CommonBorder_TopPaddingContainer
   properties {verticalAlignment=VAlign_Top, padding=<ltrb 0 50 0 0>}
   slot {class=OverlaySlot, properties {horizontalAlignment=HAlign_Fill}}
  }
  VerticalBox_0 {
   widgetClassPath=&VerticalBox
   parent=CommonBorder_TopPaddingContainer
   widget=VerticalBox_0
   slot {class=BorderSlot, properties {verticalAlignment=VAlign_Top, padding=<ltrb 0 50 0 0>}}
  }
  SizeBox_0 {widgetClassPath=&SizeBox, parent=VerticalBox_0, widget=SizeBox_0, properties {heightOverride=60}, slot {class=VerticalBoxSlot}}
  VerticalBoxProgressVbox {
   widgetClassPath=&VerticalBox
   parent=SizeBox_0
   widget=VerticalBoxProgressVbox
   slot {class=SizeBoxSlot, properties {verticalAlignment=VAlign_Top}}
  }
  HorizontalBox_0 {
   widgetClassPath=&HorizontalBox
   parent=VerticalBoxProgressVbox
   widget=HorizontalBox_0
   slot {class=VerticalBoxSlot, properties {verticalAlignment=VAlign_Top, size {sizeRule=Fill, value=1}}}
  }
  HorizontalBox_2 {
   widgetClassPath=&HorizontalBox
   parent=HorizontalBox_0
   widget=HorizontalBox_2
   slot {class=HorizontalBoxSlot, properties {horizontalAlignment=HAlign_Right, verticalAlignment=VAlign_Center, size {sizeRule=Fill, value=1}}}
  }
  Text_YourPlacement {
   widgetClassPath=&CommonTextBlock
   bIsVariable=true
   parent=HorizontalBox_2
   widget=Text_YourPlacement
   properties {
    justification=Center
    style=/FortUILibrary/Styles/Typography/Foundation/TextStyle_BurbankBlack_M_White.TextStyle_BurbankBlack_M_White_C
    text=22
    textTransformPolicy=ToUpper
    visibility=HitTestInvisible
    font {
     bForceMonospaced=false
     bMaterialIsStencil=false
     fontMaterial=null
     fontObject=&Burbank
     letterSpacing=0
     monospacedWidth=1
     size=24
     skewAmount=0
     typefaceFontName=BurbankRegular-Black
     outlineSettings {
      bApplyOutlineToDropShadows=false
      bMiteredCorners=false
      bSeparateFillAlpha=false
      outlineMaterial=&M-FontShadow
      outlineSize=3
      outlineColor=<rgba 0 0 0 0.5>
     }
    }
    renderTransform {angle=0, scale=<1 1>, shear=<-8 0>, translation=<0 0>}
    shadowColorAndOpacity=<rgba 0 0 0 0.5>
    shadowOffset=<3 3>
   }
   slot {class=HorizontalBoxSlot, properties {horizontalAlignment=HAlign_Right, verticalAlignment=VAlign_Top, padding=<ltrb 0 0 10 0>}}
  }
  SizeBox_Progress {
   widgetClassPath=&SizeBox
   parent=HorizontalBox_2
   widget=SizeBox_Progress
   properties {widthOverride=530}
   slot {class=HorizontalBoxSlot, properties {horizontalAlignment=HAlign_Center, verticalAlignment=VAlign_Top, padding=<ltrb 0 15 0 0>}}
  }
  HorizontalBoxProgressBars {widgetClassPath=&HorizontalBox, parent=SizeBox_Progress, widget=HorizontalBoxProgressBars, slot {class=SizeBoxSlot}}
  CommonBorder_YourPlacement {
   widgetClassPath=&CommonBorder
   bIsVariable=true
   parent=HorizontalBoxProgressBars
   widget=CommonBorder_YourPlacement
   properties {
    horizontalAlignment=HAlign_Right
    visibility=Collapsed
    contentColorAndOpacity=<rgba 1 0.6038 0 1>
    padding=<ltrb 0 0 0 0>
    renderTransform {angle=0, scale=<1 1>, shear=<0 0>, translation=<-4 -7>}
   }
   slot {class=HorizontalBoxSlot, properties {horizontalAlignment=HAlign_Right, verticalAlignment=VAlign_Top}}
  }
  SizeBox_1 {
   widgetClassPath=&SizeBox
   parent=CommonBorder_YourPlacement
   widget=SizeBox_1
   properties {bOverride_WidthOverride=true}
   slot {class=BorderSlot, properties {horizontalAlignment=HAlign_Right, padding=<ltrb 0 0 0 0>}}
  }
  Overlay_2 {
   widgetClassPath=&Overlay
   parent=HorizontalBoxProgressBars
   widget=Overlay_2
   slot {class=HorizontalBoxSlot, properties {horizontalAlignment=HAlign_Center}}
  }
  CommonBorderBarShadow {
   widgetClassPath=&CommonBorder
   parent=Overlay_2
   widget=CommonBorderBarShadow
   properties {
    horizontalAlignment=HAlign_Left
    verticalAlignment=VAlign_Top
    contentColorAndOpacity=<rgba 0.0052 0.0052 0.0056 0.25>
    padding=<ltrb 0 0 0 0>
    renderTransform {angle=0, scale=<1 1>, shear=<0 0>, translation=<4 4>}
   }
   slot {class=OverlaySlot, properties {verticalAlignment=VAlign_Fill}}
  }
  SizeBox_8 {
   widgetClassPath=&SizeBox
   bIsVariable=true
   parent=CommonBorderBarShadow
   widget=SizeBox_8
   properties {bOverride_HeightOverride=true, bOverride_WidthOverride=true, heightOverride=10, widthOverride=520}
   slot {class=BorderSlot, properties {horizontalAlignment=HAlign_Left, verticalAlignment=VAlign_Top, padding=<ltrb 0 0 0 0>}}
  }
  Image_7 {
   widgetClassPath=&Image
   parent=SizeBox_8
   widget=Image_7
   properties {
    visibility=HitTestInvisible
    brush {
     bIsDynamicallyLoaded=false
     drawAs=Box
     imageType=NoImage
     mirroring=NoMirror
     resourceName=None
     resourceObject=/Game/Athena/HUD/LTM/Wax/Materials/M_UI_SkewBoxBorder_Progress
     tiling=NoTile
     imageSize=<15 15>
     margin=<ltrb 0.5 0 0.5 0>
     outlineSettings {
      bUseBrushTransparency=false
      roundingType=HalfHeightRadius
      width=0
      color {colorUseRule=UseColor_Specified, specifiedColor=<rgba 0 0 0 0>}
      cornerRadii=<0 0 0 0>
     }
     tintColor {colorUseRule=UseColor_Specified, specifiedColor=<rgba 1 1 1 1>}
     uVRegion {bIsValid=false, max=<0 0>, min=<0 0>}
    }
    renderTransform {angle=0, scale=<-1 -1>, shear=<0 0>, translation=<0 0>}
   }
   slot {class=SizeBoxSlot}
  }
  CommonBorderBarBG {
   widgetClassPath=&CommonBorder
   parent=Overlay_2
   widget=CommonBorderBarBG
   properties {
    horizontalAlignment=HAlign_Left
    verticalAlignment=VAlign_Top
    contentColorAndOpacity=<rgba 0.0052 0.0052 0.0056 0.8>
    padding=<ltrb 0 0 0 0>
   }
   slot {class=OverlaySlot, properties {verticalAlignment=VAlign_Fill}}
  }
  SizeBox_6 {
   widgetClassPath=&SizeBox
   bIsVariable=true
   parent=CommonBorderBarBG
   widget=SizeBox_6
   properties {bOverride_HeightOverride=true, bOverride_WidthOverride=true, heightOverride=10, widthOverride=520}
   slot {class=BorderSlot, properties {horizontalAlignment=HAlign_Left, verticalAlignment=VAlign_Top, padding=<ltrb 0 0 0 0>}}
  }
  Image_5 {
   widgetClassPath=&Image
   parent=SizeBox_6
   widget=Image_5
   properties {
    visibility=HitTestInvisible
    brush {
     bIsDynamicallyLoaded=false
     drawAs=Box
     imageType=NoImage
     mirroring=NoMirror
     resourceName=None
     resourceObject=/Game/Athena/HUD/LTM/TeamPlacement/Materials/M_UI_SkewBoxBorder_Progress
     tiling=NoTile
     imageSize=<15 15>
     margin=<ltrb 0.5 0 0.5 0>
     outlineSettings {
      bUseBrushTransparency=false
      roundingType=HalfHeightRadius
      width=0
      color {colorUseRule=UseColor_Specified, specifiedColor=<rgba 0 0 0 0>}
      cornerRadii=<0 0 0 0>
     }
     tintColor {colorUseRule=UseColor_Specified, specifiedColor=<rgba 1 1 1 1>}
     uVRegion {bIsValid=false, max=<0 0>, min=<0 0>}
    }
    renderTransform {angle=0, scale=<-1 -1>, shear=<0 0>, translation=<0 0>}
   }
   slot {class=SizeBoxSlot}
  }
  ProgressEnemy_3 {
   widgetClassPath=/Game/Athena/HUD/LTM/TeamPlacement/TeamPlacement_ProgressEnemy
   bIsVariable=true
   parent=Overlay_2
   widget=ProgressEnemy_3
   properties {isFirst=false, progress=0.4}
   slot {class=OverlaySlot}
  }
  ProgressEnemy_2 {
   widgetClassPath=&TeamPlacement_ProgressEnemy
   bIsVariable=true
   parent=Overlay_2
   widget=ProgressEnemy_2
   properties {isFirst=false}
   slot {class=OverlaySlot}
  }
  ProgressEnemy_1 {
   widgetClassPath=&TeamPlacement_ProgressEnemy
   bIsVariable=true
   parent=Overlay_2
   widget=ProgressEnemy_1
   properties {progress=0.8}
   slot {class=OverlaySlot}
  }
  CommonBorderPlayerProgress_Potential {
   widgetClassPath=&CommonBorder
   bIsVariable=true
   parent=Overlay_2
   widget=CommonBorderPlayerProgress_Potential
   properties {
    horizontalAlignment=HAlign_Left
    verticalAlignment=VAlign_Top
    visibility=HitTestInvisible
    contentColorAndOpacity=<rgba 0.9823 0.9823 0.9823 1>
    desiredSizeScale=<0.3 1>
    padding=<ltrb 0 0 0 0>
   }
   slot {class=OverlaySlot}
  }
  Overlay {
   widgetClassPath=&Overlay
   parent=CommonBorderPlayerProgress_Potential
   widget=Overlay
   slot {class=BorderSlot, properties {horizontalAlignment=HAlign_Left, verticalAlignment=VAlign_Top, padding=<ltrb 0 0 0 0>}}
  }
  SizeBox {
   widgetClassPath=&SizeBox
   bIsVariable=true
   parent=Overlay
   widget=SizeBox
   properties {bOverride_HeightOverride=true, bOverride_WidthOverride=true, heightOverride=10, widthOverride=520}
   slot {class=OverlaySlot, properties {horizontalAlignment=HAlign_Fill}}
  }
  Image_Progress_Potential {
   widgetClassPath=&Image
   bIsVariable=true
   parent=SizeBox
   widget=Image_Progress_Potential
   properties {
    visibility=HitTestInvisible
    brush {
     bIsDynamicallyLoaded=false
     drawAs=Box
     imageType=NoImage
     mirroring=NoMirror
     resourceName=None
     resourceObject=&M_UI_SkewBoxBorder_Progress
     tiling=NoTile
     imageSize=<15 15>
     margin=<ltrb 0.5 0 0.5 0>
     outlineSettings {
      bUseBrushTransparency=false
      roundingType=HalfHeightRadius
      width=0
      color {colorUseRule=UseColor_Specified, specifiedColor=<rgba 0 0 0 0>}
      cornerRadii=<0 0 0 0>
     }
     tintColor {colorUseRule=UseColor_Specified, specifiedColor=<rgba 0.3613 0.9734 1 1>}
     uVRegion {bIsValid=false, max=<0 0>, min=<0 0>}
    }
    renderTransform {angle=0, scale=<-1 -1>, shear=<0 0>, translation=<0 0>}
   }
   slot {class=SizeBoxSlot}
  }
  CommonBorderPlayerProgress {
   widgetClassPath=&CommonBorder
   bIsVariable=true
   parent=Overlay_2
   widget=CommonBorderPlayerProgress
   properties {
    horizontalAlignment=HAlign_Left
    verticalAlignment=VAlign_Top
    contentColorAndOpacity=<rgba 0.9823 0.9823 0.9823 1>
    desiredSizeScale=<0.2 1>
    padding=<ltrb 0 0 0 0>
   }
   slot {class=OverlaySlot}
  }
  Overlay_3 {
   widgetClassPath=&Overlay
   parent=CommonBorderPlayerProgress
   widget=Overlay_3
   slot {class=BorderSlot, properties {horizontalAlignment=HAlign_Left, verticalAlignment=VAlign_Top, padding=<ltrb 0 0 0 0>}}
  }
  CommonBorder_TeamPlacement {
   widgetClassPath=&CommonBorder
   bIsVariable=true
   parent=Overlay_3
   widget=CommonBorder_TeamPlacement
   properties {
    horizontalAlignment=HAlign_Center
    verticalAlignment=VAlign_Top
    padding=<ltrb 0 0 0 0>
    renderTransform {angle=0, scale=<1 1>, shear=<0 0>, translation=<-6 11>}
   }
   slot {class=OverlaySlot, properties {horizontalAlignment=HAlign_Right, verticalAlignment=VAlign_Center}}
  }
  Overlay_4 {
   widgetClassPath=&Overlay
   parent=CommonBorder_TeamPlacement
   widget=Overlay_4
   slot {class=BorderSlot, properties {horizontalAlignment=HAlign_Center, verticalAlignment=VAlign_Top, padding=<ltrb 0 0 0 0>}}
  }
  Image {
   widgetClassPath=&Image
   bIsVariable=true
   parent=Overlay_4
   widget=Image
   properties {
    brush {
     bIsDynamicallyLoaded=false
     drawAs=Image
     imageType=NoImage
     mirroring=NoMirror
     resourceName=None
     resourceObject=/Game/Athena/HUD/LTM/TeamPlacement/Materials/M_UI_TeamProgressCaret_Ally
     tiling=NoTile
     imageSize=<32 32>
     margin=<ltrb 0 0 0 0>
     outlineSettings {
      bUseBrushTransparency=false
      roundingType=HalfHeightRadius
      width=0
      color {colorUseRule=UseColor_Specified, specifiedColor=<rgba 0 0 0 0>}
      cornerRadii=<0 0 0 0>
     }
     tintColor {colorUseRule=UseColor_Specified, specifiedColor=<rgba 1 1 1 1>}
     uVRegion {bIsValid=false, max=<0 0>, min=<0 0>}
    }
    colorAndOpacity=<rgba 0 0 0 0.5>
    renderTransform {angle=0, scale=<1 -1>, shear=<0 0>, translation=<2.5 1>}
   }
   slot {class=OverlaySlot, properties {horizontalAlignment=HAlign_Center, padding=<ltrb -30 0 -30 0>}}
  }
  Image_111 {
   widgetClassPath=&Image
   bIsVariable=true
   parent=Overlay_4
   widget=Image_111
   properties {
    brush {
     bIsDynamicallyLoaded=false
     drawAs=Image
     imageType=NoImage
     mirroring=NoMirror
     resourceName=None
     resourceObject=&M_UI_TeamProgressCaret_Ally
     tiling=NoTile
     imageSize=<32 32>
     margin=<ltrb 0 0 0 0>
     outlineSettings {
      bUseBrushTransparency=false
      roundingType=HalfHeightRadius
      width=0
      color {colorUseRule=UseColor_Specified, specifiedColor=<rgba 0 0 0 0>}
      cornerRadii=<0 0 0 0>
     }
     tintColor {colorUseRule=UseColor_Specified, specifiedColor=<rgba 1 1 1 1>}
     uVRegion {bIsValid=false, max=<0 0>, min=<0 0>}
    }
    colorAndOpacity=<rgba 0.0203 0.7835 0.9823 1>
    renderTransform {angle=0, scale=<1 -1>, shear=<0 0>, translation=<0 0>}
   }
   slot {class=OverlaySlot, properties {horizontalAlignment=HAlign_Center, padding=<ltrb -30 0 -30 0>}}
  }
  Text_LocalTeamPlacement {
   widgetClassPath=&CommonTextBlock
   bIsVariable=true
   parent=Overlay_4
   widget=Text_LocalTeamPlacement
   properties {
    text="##4"
    textTransformPolicy=ToUpper
    visibility=Collapsed
    font {
     bForceMonospaced=false
     bMaterialIsStencil=false
     fontMaterial=null
     fontObject=&Burbank
     letterSpacing=0
     monospacedWidth=1
     size=20
     skewAmount=0
     typefaceFontName=BurbankRegular-Black
     outlineSettings {
      bApplyOutlineToDropShadows=false
      bMiteredCorners=false
      bSeparateFillAlpha=false
      outlineMaterial=null
      outlineSize=1
      outlineColor=<rgba 0.0052 0.0052 0.0056 0.6>
     }
    }
    renderTransform {angle=0, scale=<1 1>, shear=<-8 0>, translation=<5 -7>}
    shadowColorAndOpacity=<rgba 0 0 0 0.5>
    shadowOffset=<3 3>
   }
   slot {class=OverlaySlot, properties {horizontalAlignment=HAlign_Right, verticalAlignment=VAlign_Bottom, padding=<ltrb -200 15 0 0>}}
  }
  SizeBox_4 {
   widgetClassPath=&SizeBox
   bIsVariable=true
   parent=Overlay_3
   widget=SizeBox_4
   properties {bOverride_HeightOverride=true, bOverride_WidthOverride=true, heightOverride=10, widthOverride=520}
   slot {class=OverlaySlot, properties {horizontalAlignment=HAlign_Fill}}
  }
  CommonWidgetSwitcher_PlayerProgress {
   widgetClassPath=&CommonWidgetSwitcherLegacy
   bIsVariable=true
   parent=SizeBox_4
   widget=CommonWidgetSwitcher_PlayerProgress
   slot {class=SizeBoxSlot}
  }
  Image_Progress {
   widgetClassPath=&Image
   bIsVariable=true
   parent=CommonWidgetSwitcher_PlayerProgress
   widget=Image_Progress
   properties {
    visibility=HitTestInvisible
    brush {
     bIsDynamicallyLoaded=false
     drawAs=Box
     imageType=NoImage
     mirroring=NoMirror
     resourceName=None
     resourceObject=&M_UI_SkewBoxBorder_Progress
     tiling=NoTile
     imageSize=<15 15>
     margin=<ltrb 0.5 0 0.5 0>
     outlineSettings {
      bUseBrushTransparency=false
      roundingType=HalfHeightRadius
      width=0
      color {colorUseRule=UseColor_Specified, specifiedColor=<rgba 0 0 0 0>}
      cornerRadii=<0 0 0 0>
     }
     tintColor {colorUseRule=UseColor_Specified, specifiedColor=<rgba 0.0203 0.7835 0.9823 1>}
     uVRegion {bIsValid=false, max=<0 0>, min=<0 0>}
    }
    renderTransform {angle=0, scale=<-1 -1>, shear=<0 0>, translation=<0 0>}
   }
   slot {class=WidgetSwitcherSlot}
  }
  SizeBox_Goal {
   widgetClassPath=&SizeBox
   parent=HorizontalBox_2
   widget=SizeBox_Goal
   properties {bOverride_WidthOverride=true, heightOverride=32.5807, widthOverride=80}
   slot {
    class=HorizontalBoxSlot
    properties {horizontalAlignment=HAlign_Left, verticalAlignment=VAlign_Top, padding=<ltrb -4 0 0 0>, size {sizeRule=Fill, value=1}}
   }
  }
  CommonBorder_0 {
   widgetClassPath=&CommonBorder
   parent=SizeBox_Goal
   widget=CommonBorder_0
   properties {
    background {
     bIsDynamicallyLoaded=false
     drawAs=Box
     imageType=NoImage
     mirroring=NoMirror
     resourceName=None
     resourceObject=null
     tiling=NoTile
     imageSize=<64 64>
     margin=<ltrb 0 0 0 0>
     outlineSettings {
      bUseBrushTransparency=false
      roundingType=HalfHeightRadius
      width=0
      color {colorUseRule=UseColor_Specified, specifiedColor=<rgba 0 0 0 0>}
      cornerRadii=<0 0 0 0>
     }
     tintColor {colorUseRule=UseColor_Specified, specifiedColor=<rgba 0 0 0 0>}
     uVRegion {bIsValid=false, max=<0 0>, min=<0 0>}
    }
    padding=<ltrb 0 0 0 0>
   }
   slot {class=SizeBoxSlot}
  }
  HorizontalBox_1 {
   widgetClassPath=&HorizontalBox
   parent=CommonBorder_0
   widget=HorizontalBox_1
   slot {class=BorderSlot, properties {padding=<ltrb 0 0 0 0>}}
  }
  Spacer_152 {widgetClassPath=&Spacer, parent=HorizontalBox_1, widget=Spacer_152, properties {size=<10 1>}, slot {class=HorizontalBoxSlot}}
  TargetScoreText {
   widgetClassPath=/Script/CommonUI.CommonNumericTextBlock
   bIsVariable=true
   parent=HorizontalBox_1
   widget=TargetScoreText
   properties {
    currentNumericValue=100
    style=&TextStyle_BurbankBlack_M_White_C
    text=Text Block
    font {
     bForceMonospaced=false
     bMaterialIsStencil=false
     fontMaterial=null
     fontObject=&Burbank
     letterSpacing=0
     monospacedWidth=1
     size=24
     skewAmount=0
     typefaceFontName=BurbankRegular-Black
     outlineSettings {
      bApplyOutlineToDropShadows=false
      bMiteredCorners=false
      bSeparateFillAlpha=false
      outlineMaterial=&M-FontShadow
      outlineSize=3
      outlineColor=<rgba 0 0 0 0.5>
     }
    }
    formattingSpecification {
     alwaysSign=false
     maximumFractionalDigits=0
     maximumIntegralDigits=324
     minimumFractionalDigits=0
     minimumIntegralDigits=1
     roundingMode=HalfFromZero
     useGrouping=false
    }
    renderTransform {angle=0, scale=<1 1>, shear=<-8 0>, translation=<0 0>}
    shadowColorAndOpacity=<rgba 0 0 0 0.5>
    shadowOffset=<3 3>
   }
   slot {class=HorizontalBoxSlot, properties {horizontalAlignment=HAlign_Left, verticalAlignment=VAlign_Center}}
  }
  Spacer_2 {widgetClassPath=&Spacer, parent=HorizontalBox_1, widget=Spacer_2, properties {size=<10 1>}, slot {class=HorizontalBoxSlot}}
  ScoreTypeTextBox {
   widgetClassPath=&CommonTextBlock
   bIsVariable=true
   parent=HorizontalBox_1
   widget=ScoreTypeTextBox
   properties {
    text=pts
    textTransformPolicy=ToUpper
    visibility=Collapsed
    font {
     bForceMonospaced=false
     bMaterialIsStencil=false
     fontMaterial=null
     fontObject=&Burbank
     letterSpacing=0
     monospacedWidth=1
     size=15
     skewAmount=0
     typefaceFontName=BurbankRegular-Black
     outlineSettings {
      bApplyOutlineToDropShadows=false
      bMiteredCorners=false
      bSeparateFillAlpha=false
      outlineMaterial=null
      outlineSize=0
      outlineColor=<rgba 0 0 0 1>
     }
    }
    renderTransform {angle=0, scale=<1 1>, shear=<-8 0>, translation=<0 0>}
    shadowColorAndOpacity=<rgba 0 0 0 0.5>
    shadowOffset=<3 3>
   }
   slot {class=HorizontalBoxSlot, properties {horizontalAlignment=HAlign_Left, verticalAlignment=VAlign_Center, padding=<ltrb -10 -26 0 -30>}}
  }
  LeaderRelativeMessageBox {
   widgetClassPath=&HorizontalBox
   bIsVariable=true
   parent=VerticalBox_0
   widget=LeaderRelativeMessageBox
   properties {visibility=Visible, renderTransform {angle=0, scale=<1 1>, shear=<-8 0>, translation=<0 0>}}
   slot {class=VerticalBoxSlot, properties {horizontalAlignment=HAlign_Center, padding=<ltrb 0 5 0 0>}}
  }
  RelativeLeaderMessage {
   widgetClassPath=&CommonTextBlock
   bIsVariable=true
   parent=LeaderRelativeMessageBox
   widget=RelativeLeaderMessage
   properties {
    text=LEADING BY
    textTransformPolicy=ToUpper
    font {
     bForceMonospaced=false
     bMaterialIsStencil=false
     fontMaterial=null
     fontObject=&Burbank
     letterSpacing=0
     monospacedWidth=1
     size=30
     skewAmount=0
     typefaceFontName=BurbankRegular-Black
     outlineSettings {
      bApplyOutlineToDropShadows=false
      bMiteredCorners=false
      bSeparateFillAlpha=false
      outlineMaterial=null
      outlineSize=0
      outlineColor=<rgba 0 0 0 1>
     }
    }
    shadowColorAndOpacity=<rgba 0 0 0 0.5>
    shadowOffset=<3 3>
   }
   slot {class=HorizontalBoxSlot, properties {horizontalAlignment=HAlign_Left, verticalAlignment=VAlign_Top}}
  }
  Spacer_274 {
   widgetClassPath=&Spacer
   parent=LeaderRelativeMessageBox
   widget=Spacer_274
   properties {size=<2 1>}
   slot {class=HorizontalBoxSlot}
  }
  LeaderRelativePointValue {
   widgetClassPath=&CommonTextBlock
   bIsVariable=true
   parent=LeaderRelativeMessageBox
   widget=LeaderRelativePointValue
   properties {
    text=400
    textTransformPolicy=ToUpper
    colorAndOpacity {colorUseRule=UseColor_Specified, specifiedColor=<rgba 0.0203 0.7835 0.9823 1>}
    font {
     bForceMonospaced=false
     bMaterialIsStencil=false
     fontMaterial=null
     fontObject=&Burbank
     letterSpacing=0
     monospacedWidth=1
     size=30
     skewAmount=0
     typefaceFontName=BurbankRegular-Black
     outlineSettings {
      bApplyOutlineToDropShadows=false
      bMiteredCorners=false
      bSeparateFillAlpha=false
      outlineMaterial=null
      outlineSize=0
      outlineColor=<rgba 0 0 0 1>
     }
    }
    shadowColorAndOpacity=<rgba 0 0 0 0.5>
    shadowOffset=<3 3>
   }
   slot {class=HorizontalBoxSlot}
  }
  Spacer_1 {
   widgetClassPath=&Spacer
   parent=LeaderRelativeMessageBox
   widget=Spacer_1
   properties {size=<2 1>}
   slot {class=HorizontalBoxSlot}
  }
  ScoreTypeTextBox_1 {
   widgetClassPath=&CommonTextBlock
   bIsVariable=true
   parent=LeaderRelativeMessageBox
   widget=ScoreTypeTextBox_1
   properties {
    text=PTS
    textTransformPolicy=ToUpper
    font {
     bForceMonospaced=false
     bMaterialIsStencil=false
     fontMaterial=null
     fontObject=&Burbank
     letterSpacing=0
     monospacedWidth=1
     size=20
     skewAmount=0
     typefaceFontName=BurbankRegular-Black
     outlineSettings {
      bApplyOutlineToDropShadows=false
      bMiteredCorners=false
      bSeparateFillAlpha=false
      outlineMaterial=null
      outlineSize=0
      outlineColor=<rgba 0 0 0 1>
     }
    }
    shadowColorAndOpacity=<rgba 0 0 0 0.5>
    shadowOffset=<3 3>
   }
   slot {class=HorizontalBoxSlot, properties {horizontalAlignment=HAlign_Left, verticalAlignment=VAlign_Center, padding=<ltrb 0 -25 0 -30>}}
  }
 ]
}
