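# Translucent UI material for the HUD that lays a shine streak over a flat color, white by default.
# The shine texture is sampled with slightly rotated, clamped UVs shifted by a TimeOffset scalar, so the sweep is set manually rather than by a Time node.
# Linear Dodge blends the shine onto the Color parameter, and a Max combines Color alpha with the shine's red channel.
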Material /Game/Athena/HUD/Materials/M_UI_OutlineShine_White_Manual
graph {
 properties {blendMode=BLEND_Translucent, materialDomain=MD_UI, pixelDepthOffsetMode=PDOM_Legacy, usedShadingModels=Default Lit}
 nodes [
  n1 = TextureCoordinate()
  n2 = Constant() {r=-0.05}
  n3 = TextureSample(UVs=Add(
    A=MaterialFunctionCall(UVs=@n1, "Rotation Angle (0-1)"=@n2) {materialFunction=&CustomRotator}.Rotated Values
    B=ScalarParameter() {defaultValue=-10, parameterName=TimeOffset}
   )) {samplerSource=SSM_Clamp_WorldGroupSettings, texture=/Game/UI/Foundation/Textures/Pattern/T-UI-Shine}
  n4 = VectorParameter() {parameterName=Color, defaultValue=<rgba 1 1 1 1>}
  MaterialFunctionCall!(Blend=@n3.RGB, Base=@n4.RGB) {materialFunction=/Engine/Functions/Engine_MaterialFunctions03/Blends/Blend_LinearDodge}
  Max(A=@n4.A, B=@n3.R)
 ]
}
