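# Sound cue that plays the outro sting for the Papaya timed trial quest, routed to the Fort_Quests sound class.
# A short fixed delay node feeds a single wave player for the outro wave.
# The cue boosts volume, uses a very large max distance, and overrides concurrency to cap simultaneous instances.
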SoundCue /Game/Athena/Sounds/Quests/TimedTrial_Papaya/Papaya_TimedTrial_Outro_Cue
root=Delay_0
nodes [
 {id=Delay_0, children [WavePlayer_0], properties {delayMax=0.02, delayMin=0.02}}
 {id=WavePlayer_0, properties {soundWaveAssetPtr=&/Papaya_TimedTrial_Outro_01}}
]
properties {
 bOverrideConcurrency=true
 maxDistance=2097152
 soundClassObject=&Fort_Quests
 volumeMultiplier=2.016
 concurrencyOverrides {maxCount=2}
}
