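# Creative race HUD built on AthenaMinigameActivityWidget, with timer, progress, best lap and an animated end-of-race rank.
# Minigame events toggle sections and play the intro; successful endings play FinalScore and NewRecord, then a retriggerable delay runs the outro unless another minigame is active.
# SetRank maps an index to a letter grade, font material and glows; ToTimeFormat formats times.
# On construct it sends Event.Checkpoint.Ready to the owning pawn.
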WidgetBlueprint /Game/Athena/HUD/BP_Creative_RaceHudWidget
animations [
 Reset {
  binding_count=6
  length_seconds=0.798
  object=&self:Reset
  sequence {bindings [
    Image_Triangle {
     id=#1
     possessed_class=Image
     tracks [
      {
       type=MovieSceneFloatTrack
       display_name=Render Opacity
       sections [{type=MovieSceneFloatSection, channels [None {type=MovieSceneScriptingFloatChannel, const [0,1]}]}]
      }
      {
       type=MovieSceneByteTrack
       display_name=Visibility
       sections [{type=MovieSceneByteSection, channels [None {type=MovieSceneScriptingByteChannel, const [2,1]}]}]
      }
     ]
    }
    OverlayAddTime {
     id=#2
     possessed_class=Overlay
     tracks [{
       type=MovieSceneFloatTrack
       display_name=Render Opacity
       sections [{type=MovieSceneFloatSection, end_frame=0, start_frame=0, channels [None {type=MovieSceneScriptingFloatChannel, const [0,1]}]}]
      }]
    }
    CommonBorder_Criteria {
     id=#3
     possessed_class=CommonBorder
     tracks [
      {
       type=MovieSceneFloatVectorTrack
       display_name=Desired Size Scale
       sections [{
         type=MovieSceneFloatVectorSection
         channels [
          Vector.X {type=MovieSceneScriptingFloatChannel, const [1,1]}
          Vector.Y {type=MovieSceneScriptingFloatChannel}
         ]
        }]
      }
      {
       type=MovieSceneFloatTrack
       display_name=Render Opacity
       sections [{type=MovieSceneFloatSection, channels [None {type=MovieSceneScriptingFloatChannel, const [1,2]}]}]
      }
     ]
    }
    VerticalBox_0 {id=#4, possessed_class=VerticalBox, tracks []}
    CommonBorder_Rank {
     id=#5
     possessed_class=CommonBorder
     tracks [{
       type=MovieSceneByteTrack
       display_name=Visibility
       sections [{type=MovieSceneByteSection, channels [None {type=MovieSceneScriptingByteChannel, const [1,1]}]}]
      }]
    }
    HorizontalBoxSlot_0 {
     id=#6
     parent_id=#4
     possessed_class=HorizontalBoxSlot
     tracks [{
       type=MovieSceneMarginTrack
       display_name=Padding
       sections [{
         type=MovieSceneMarginSection
         end_frame=0
         start_frame=0
         channels [
          Left {type=MovieSceneScriptingFloatChannel, const [0,1]}
          Top {type=MovieSceneScriptingFloatChannel}
          Right {type=MovieSceneScriptingFloatChannel}
          Bottom {type=MovieSceneScriptingFloatChannel}
         ]
        }]
      }]
    }
   ]}
 }
 Intro {
  binding_count=1
  length_seconds=0.5
  object=&self:Intro
  sequence {bindings [Overlay_Container {
     id=#7
     possessed_class=Overlay
     tracks [
      {
       type=MovieScene2DTransformTrack
       sections [{
         type=MovieScene2DTransformSection
         channels [
          Translation.X {type=MovieSceneScriptingFloatChannel}
          Translation.Y {type=MovieSceneScriptingFloatChannel, keys [[0,-350],[8,12.5],[10,0]]}
          Angle {type=MovieSceneScriptingFloatChannel}
          Scale.X {type=MovieSceneScriptingFloatChannel}
          Scale.Y {type=MovieSceneScriptingFloatChannel}
          Shear.X {type=MovieSceneScriptingFloatChannel}
          Shear.Y {type=MovieSceneScriptingFloatChannel}
         ]
        }]
      }
      {
       type=MovieSceneFloatTrack
       display_name=Render Opacity
       sections [{type=MovieSceneFloatSection, channels [None {type=MovieSceneScriptingFloatChannel, keys [[0,0],[10.0003,1]]}]}]
      }
     ]
    }]}
 }
 Outro {
  binding_count=1
  length_seconds=0.4243
  object=&self:Outro
  sequence {bindings [Overlay_Container {
     id=#8
     possessed_class=Overlay
     tracks [
      {
       type=MovieSceneFloatTrack
       display_name=Render Opacity
       sections [{type=MovieSceneFloatSection, channels [None {type=MovieSceneScriptingFloatChannel, keys [[0,1],[8.4867,0]]}]}]
      }
      {
       type=MovieScene2DTransformTrack
       sections [{
         type=MovieScene2DTransformSection
         channels [
          Translation.X {type=MovieSceneScriptingFloatChannel, keys [[0,0],[8.4867,31.5151]]}
          Translation.Y {type=MovieSceneScriptingFloatChannel}
          Angle {type=MovieSceneScriptingFloatChannel}
          Scale.X {type=MovieSceneScriptingFloatChannel}
          Scale.Y {type=MovieSceneScriptingFloatChannel}
          Shear.X {type=MovieSceneScriptingFloatChannel}
          Shear.Y {type=MovieSceneScriptingFloatChannel}
         ]
        }]
      }
     ]
    }]}
 }
 FinalScore {
  binding_count=13
  length_seconds=4.5893
  object=&self:FinalScore
  sequence {bindings [
    Image_redglow {
     id=#9
     possessed_class=Image
     tracks [
      {
       type=MovieSceneFloatTrack
       display_name=Render Opacity
       sections [{type=MovieSceneFloatSection, channels [None {type=MovieSceneScriptingFloatChannel, keys [[9.786,0],[12.786,1]]}]}]
      }
      {
       type=MovieScene2DTransformTrack
       sections [{
         type=MovieScene2DTransformSection
         channels [
          Translation.X {type=MovieSceneScriptingFloatChannel}
          Translation.Y {type=MovieSceneScriptingFloatChannel, const [15,2]}
          Angle {type=MovieSceneScriptingFloatChannel}
          Scale.X {type=MovieSceneScriptingFloatChannel, const [1,2]}
          Scale.Y {type=MovieSceneScriptingFloatChannel, const [1,2]}
          Shear.X {type=MovieSceneScriptingFloatChannel}
          Shear.Y {type=MovieSceneScriptingFloatChannel}
         ]
        }]
      }
     ]
    }
    VerticalBox_0 {id=#10, possessed_class=VerticalBox, tracks []}
    Image_topglow {
     id=#11
     possessed_class=Image
     tracks [
      {
       type=MovieSceneFloatTrack
       display_name=Render Opacity
       sections [{type=MovieSceneFloatSection, channels [None {type=MovieSceneScriptingFloatChannel, keys [[9.786,0],[12.786,1]]}]}]
      }
      {
       type=MovieScene2DTransformTrack
       sections [{
         type=MovieScene2DTransformSection
         channels [
          Translation.X {type=MovieSceneScriptingFloatChannel}
          Translation.Y {type=MovieSceneScriptingFloatChannel}
          Angle {type=MovieSceneScriptingFloatChannel}
          Scale.X {type=MovieSceneScriptingFloatChannel, const [1,2]}
          Scale.Y {type=MovieSceneScriptingFloatChannel, const [1,2]}
          Shear.X {type=MovieSceneScriptingFloatChannel}
          Shear.Y {type=MovieSceneScriptingFloatChannel}
         ]
        }]
      }
     ]
    }
    Overlay_FinalRank {
     id=#12
     possessed_class=Overlay
     tracks [{
       type=MovieScene2DTransformTrack
       sections [{
         type=MovieScene2DTransformSection
         channels [
          Translation.X {type=MovieSceneScriptingFloatChannel}
          Translation.Y {type=MovieSceneScriptingFloatChannel}
          Angle {type=MovieSceneScriptingFloatChannel}
          Scale.X {type=MovieSceneScriptingFloatChannel, keys [[3.5407,0],[7.6437,1],[9.786,1]]}
          Scale.Y {type=MovieSceneScriptingFloatChannel, keys [[3.5407,0],[7.6437,1],[9.786,1]]}
          Shear.X {type=MovieSceneScriptingFloatChannel}
          Shear.Y {type=MovieSceneScriptingFloatChannel}
         ]
        }]
      }]
    }
    Image_flare_2 {
     id=#13
     possessed_class=Image
     tracks [
      {
       type=MovieSceneFloatTrack
       display_name=Render Opacity
       sections [{type=MovieSceneFloatSection, channels [None {type=MovieSceneScriptingFloatChannel, keys [[9.786,0],[20.786,1]]}]}]
      }
      {
       type=MovieScene2DTransformTrack
       sections [{
         type=MovieScene2DTransformSection
         channels [
          Translation.X {type=MovieSceneScriptingFloatChannel}
          Translation.Y {type=MovieSceneScriptingFloatChannel, keys [[9.786,-25.9791],[91.786,-34.0433]]}
          Scale.X {type=MovieSceneScriptingFloatChannel, keys [[9.786,0],[35.5393,1],[91.786,1.8134]]}
          Scale.Y {type=MovieSceneScriptingFloatChannel, keys [[9.786,0],[35.5393,1],[91.786,1.8134]]}
          Shear.X {type=MovieSceneScriptingFloatChannel}
          Shear.Y {type=MovieSceneScriptingFloatChannel}
         ]
        }]
      }
      {
       type=MovieSceneWidgetMaterialTrack
       sections [{type=MovieSceneParameterSection, channels [TimeOffset {type=MovieSceneScriptingFloatChannel, const [2.4061,1]}]}]
      }
     ]
    }
    Image_flare_3 {
     id=#14
     possessed_class=Image
     tracks [
      {
       type=MovieSceneFloatTrack
       display_name=Render Opacity
       sections [{type=MovieSceneFloatSection, channels [None {type=MovieSceneScriptingFloatChannel, keys [[9.786,0],[12.786,1]]}]}]
      }
      {
       type=MovieScene2DTransformTrack
       sections [{
         type=MovieScene2DTransformSection
         channels [
          Translation.X {type=MovieSceneScriptingFloatChannel}
          Translation.Y {type=MovieSceneScriptingFloatChannel, keys [[9.786,29.5483],[90.786,13.7359]]}
          Angle {type=MovieSceneScriptingFloatChannel}
          Scale.X {type=MovieSceneScriptingFloatChannel, keys [[9.786,0],[19.786,1.4392],[91.786,1.6264]]}
          Scale.Y {type=MovieSceneScriptingFloatChannel, keys [[9.786,0],[19.786,1.4392],[91.786,1.6264]]}
          Shear.X {type=MovieSceneScriptingFloatChannel}
          Shear.Y {type=MovieSceneScriptingFloatChannel}
         ]
        }]
      }
      {
       type=MovieSceneWidgetMaterialTrack
       sections [{type=MovieSceneParameterSection, channels [TimeOffset {type=MovieSceneScriptingFloatChannel, const [5.0122,1]}]}]
      }
     ]
    }
    Image_whiteglow_1 {
     id=#15
     possessed_class=Image
     tracks [
      {
       type=MovieScene2DTransformTrack
       sections [{
         type=MovieScene2DTransformSection
         channels [
          Translation.X {type=MovieSceneScriptingFloatChannel}
          Translation.Y {type=MovieSceneScriptingFloatChannel}
          Angle {type=MovieSceneScriptingFloatChannel}
          Scale.X {type=MovieSceneScriptingFloatChannel, keys [[11.593,1.2],[34.1177,0]]}
          Scale.Y {type=MovieSceneScriptingFloatChannel, keys [[11.593,1.5],[34.1177,0]]}
          Shear.X {type=MovieSceneScriptingFloatChannel}
          Shear.Y {type=MovieSceneScriptingFloatChannel}
         ]
        }]
      }
      {
       type=MovieSceneFloatTrack
       display_name=Render Opacity
       sections [{type=MovieSceneFloatSection, channels [None {type=MovieSceneScriptingFloatChannel, keys [[11.593,1],[34.1177,0]]}]}]
      }
     ]
    }
    Image_flare {
     id=#16
     possessed_class=Image
     tracks [
      {
       type=MovieSceneFloatTrack
       display_name=Render Opacity
       sections [{type=MovieSceneFloatSection, channels [None {type=MovieSceneScriptingFloatChannel, keys [[9.786,0],[20.786,1]]}]}]
      }
      {
       type=MovieScene2DTransformTrack
       sections [{
         type=MovieScene2DTransformSection
         channels [
          Scale.X {type=MovieSceneScriptingFloatChannel, keys [[9.786,0],[30.6293,1]]}
          Scale.Y {type=MovieSceneScriptingFloatChannel, keys [[9.786,0],[30.6293,1]]}
          Shear.X {type=MovieSceneScriptingFloatChannel}
          Shear.Y {type=MovieSceneScriptingFloatChannel}
         ]
        }]
      }
     ]
    }
    Image_star {
     id=#17
     possessed_class=Image
     tracks [
      {
       type=MovieSceneFloatTrack
       display_name=Render Opacity
       sections [{type=MovieSceneFloatSection, channels [None {type=MovieSceneScriptingFloatChannel, keys [[9.786,0],[12.786,1]]}]}]
      }
      {
       type=MovieScene2DTransformTrack
       sections [{
         type=MovieScene2DTransformSection
         channels [
          Translation.X {type=MovieSceneScriptingFloatChannel, const [-25,1]}
          Translation.Y {type=MovieSceneScriptingFloatChannel, const [-19,2]}
          Angle {type=MovieSceneScriptingFloatChannel, keys [[9.786,-50],[20.786,-9.3678],[37.786,-0.6125]]}
          Scale.X {type=MovieSceneScriptingFloatChannel, keys [[9.786,0],[17.786,1.0848],[35.8367,0.75]]}
          Scale.Y {type=MovieSceneScriptingFloatChannel, keys [[9.786,0],[17.786,1.0848],[35.8367,0.75]]}
          Shear.X {type=MovieSceneScriptingFloatChannel}
          Shear.Y {type=MovieSceneScriptingFloatChannel}
         ]
        }]
      }
     ]
    }
    HorizontalBoxSlot_0 {
     id=#18
     parent_id=#10
     possessed_class=HorizontalBoxSlot
     tracks [{
       type=MovieSceneMarginTrack
       display_name=Padding
       sections [{
         type=MovieSceneMarginSection
         end_frame=7
         start_frame=0
         channels [
          Left {type=MovieSceneScriptingFloatChannel, keys [[0,0],[7.6437,110]]}
          Top {type=MovieSceneScriptingFloatChannel, const [0,1]}
          Right {type=MovieSceneScriptingFloatChannel, const [0,2]}
          Bottom {type=MovieSceneScriptingFloatChannel, const [0,1]}
         ]
        }]
      }]
    }
    CommonBorder_Rank {
     id=#5
     possessed_class=CommonBorder
     tracks [
      {
       type=MovieSceneFloatVectorTrack
       display_name=Desired Size Scale
       sections [{
         type=MovieSceneFloatVectorSection
         channels [
          Vector.X {type=MovieSceneScriptingFloatChannel, keys [[0,0],[9,1]]}
          Vector.Y {type=MovieSceneScriptingFloatChannel}
         ]
        }]
      }
      {
       type=MovieSceneFloatTrack
       display_name=Render Opacity
       sections [{type=MovieSceneFloatSection, channels [None {type=MovieSceneScriptingFloatChannel, const [1,2]}]}]
      }
      {
       type=MovieSceneByteTrack
       display_name=Visibility
       sections [{type=MovieSceneByteSection, channels [None {type=MovieSceneScriptingByteChannel, const [4,1]}]}]
      }
     ]
    }
    CommonBorder_Criteria {
     id=#19
     possessed_class=CommonBorder
     tracks [{
       type=MovieSceneFloatVectorTrack
       display_name=Desired Size Scale
       sections [{
         type=MovieSceneFloatVectorSection
         channels [
          Vector.X {type=MovieSceneScriptingFloatChannel, keys [[0,1],[9,0]]}
          Vector.Y {type=MovieSceneScriptingFloatChannel}
         ]
        }]
      }]
    }
    Image_flare_1 {
     id=#20
     possessed_class=Image
     tracks [
      {
       type=MovieSceneFloatTrack
       display_name=Render Opacity
       sections [{type=MovieSceneFloatSection, channels [None {type=MovieSceneScriptingFloatChannel, keys [[9.786,0],[25.5437,1]]}]}]
      }
      {
       type=MovieScene2DTransformTrack
       sections [{
         type=MovieScene2DTransformSection
         channels [
          Translation.X {type=MovieSceneScriptingFloatChannel}
          Translation.Y {type=MovieSceneScriptingFloatChannel, keys [[9.786,-15.3894],[91.786,-26.661]]}
          Scale.X {type=MovieSceneScriptingFloatChannel, keys [[9.786,0],[21.786,1.0801],[72.786,0.5876]]}
          Scale.Y {type=MovieSceneScriptingFloatChannel, keys [[9.786,0],[21.786,1.0801],[72.786,0.5876]]}
          Shear.X {type=MovieSceneScriptingFloatChannel}
          Shear.Y {type=MovieSceneScriptingFloatChannel}
         ]
        }]
      }
      {
       type=MovieSceneWidgetMaterialTrack
       sections [{type=MovieSceneParameterSection, channels [TimeOffset {type=MovieSceneScriptingFloatChannel, const [0.8,1]}]}]
      }
     ]
    }
   ]}
 }
 FinalTime {
  binding_count=3
  length_seconds=0.6
  object=&self:FinalTime
  sequence {bindings [
    Text_AddTime {
     id=#21
     possessed_class=CommonTextBlock
     tracks [{
       type=MovieScene2DTransformTrack
       sections [{
         type=MovieScene2DTransformSection
         channels [
          Translation.X {type=MovieSceneScriptingFloatChannel}
          Translation.Y {type=MovieSceneScriptingFloatChannel}
          Angle {type=MovieSceneScriptingFloatChannel}
          Shear.X {type=MovieSceneScriptingFloatChannel}
          Shear.Y {type=MovieSceneScriptingFloatChannel}
         ]
        }]
      }]
    }
    Image_timeglow {
     id=#22
     possessed_class=Image
     tracks [
      {
       type=MovieScene2DTransformTrack
       sections [{
         type=MovieScene2DTransformSection
         channels [
          Translation.X {type=MovieSceneScriptingFloatChannel}
          Translation.Y {type=MovieSceneScriptingFloatChannel}
          Angle {type=MovieSceneScriptingFloatChannel}
          Scale.X {type=MovieSceneScriptingFloatChannel, keys [[0,1],[12,1.5]]}
          Scale.Y {type=MovieSceneScriptingFloatChannel, const [1,1]}
          Shear.X {type=MovieSceneScriptingFloatChannel}
          Shear.Y {type=MovieSceneScriptingFloatChannel}
         ]
        }]
      }
      {
       type=MovieSceneFloatTrack
       display_name=Render Opacity
       sections [{type=MovieSceneFloatSection, channels [None {type=MovieSceneScriptingFloatChannel, keys [[0,1],[12,0]]}]}]
      }
     ]
    }
    OverlayAddTime {
     id=#23
     possessed_class=Overlay
     tracks [{
       type=MovieSceneFloatTrack
       display_name=Render Opacity
       sections [{type=MovieSceneFloatSection, channels [None {type=MovieSceneScriptingFloatChannel, keys [[0,0],[1.243,1],[7,1]]}]}]
      }]
    }
   ]}
 }
 AddTime {
  binding_count=3
  length_seconds=1.011
  object=&self:AddTime
  sequence {bindings [
    Text_AddTime {
     id=#21
     possessed_class=CommonTextBlock
     tracks [{
       type=MovieScene2DTransformTrack
       sections [{
         type=MovieScene2DTransformSection
         channels [
          Translation.X {type=MovieSceneScriptingFloatChannel}
          Translation.Y {type=MovieSceneScriptingFloatChannel}
          Angle {type=MovieSceneScriptingFloatChannel}
          Scale.X {type=MovieSceneScriptingFloatChannel, keys [[0,1],[2,1.7441],[7,1]]}
          Scale.Y {type=MovieSceneScriptingFloatChannel, keys [[0,1],[2,1.7441],[7,1]]}
          Shear.X {type=MovieSceneScriptingFloatChannel}
          Shear.Y {type=MovieSceneScriptingFloatChannel}
         ]
        }]
      }]
    }
    Image_timeglow {
     id=#22
     possessed_class=Image
     tracks [
      {
       type=MovieScene2DTransformTrack
       sections [{
         type=MovieScene2DTransformSection
         channels [
          Translation.X {type=MovieSceneScriptingFloatChannel}
          Translation.Y {type=MovieSceneScriptingFloatChannel}
          Angle {type=MovieSceneScriptingFloatChannel}
          Scale.X {type=MovieSceneScriptingFloatChannel, keys [[0,1],[12,1.5]]}
          Scale.Y {type=MovieSceneScriptingFloatChannel, const [1,1]}
          Shear.X {type=MovieSceneScriptingFloatChannel}
          Shear.Y {type=MovieSceneScriptingFloatChannel}
         ]
        }]
      }
      {
       type=MovieSceneFloatTrack
       display_name=Render Opacity
       sections [{type=MovieSceneFloatSection, channels [None {type=MovieSceneScriptingFloatChannel, keys [[0,1],[12,0]]}]}]
      }
     ]
    }
    OverlayAddTime {
     id=#23
     possessed_class=Overlay
     tracks [{
       type=MovieSceneFloatTrack
       display_name=Render Opacity
       sections [{type=MovieSceneFloatSection, channels [None {type=MovieSceneScriptingFloatChannel, keys [[0,0],[1.243,1],[7,1],[20.2197,0]]}]}]
      }]
    }
   ]}
 }
 NewRecord {
  binding_count=2
  length_seconds=0.4943
  object=&self:NewRecord
  sequence {bindings [
    CommonTextBlock_NewRecord {
     id=#24
     possessed_class=CommonTextBlock
     tracks [{
       type=MovieScene2DTransformTrack
       sections [{
         type=MovieScene2DTransformSection
         channels [
          Translation.X {type=MovieSceneScriptingFloatChannel}
          Translation.Y {type=MovieSceneScriptingFloatChannel}
          Angle {type=MovieSceneScriptingFloatChannel}
          Scale.X {type=MovieSceneScriptingFloatChannel, keys [[0,3],[3.001,1],[4.2963,1]]}
          Scale.Y {type=MovieSceneScriptingFloatChannel, keys [[0,3],[3.001,1],[4.2963,1]]}
          Shear.X {type=MovieSceneScriptingFloatChannel}
          Shear.Y {type=MovieSceneScriptingFloatChannel}
         ]
        }]
      }]
    }
    Image_Triangle {
     id=#25
     possessed_class=Image
     tracks [
      {
       type=MovieSceneFloatTrack
       display_name=Render Opacity
       sections [{type=MovieSceneFloatSection, channels [None {type=MovieSceneScriptingFloatChannel, keys [[0,0],[9.8853,1]]}]}]
      }
      {
       type=MovieSceneByteTrack
       display_name=Visibility
       sections [{type=MovieSceneByteSection, channels [None {type=MovieSceneScriptingByteChannel, const [3,1]}]}]
      }
     ]
    }
   ]}
 }
 OldRecord {
  binding_count=2
  length_seconds=0.65
  object=&self:OldRecord
  sequence {bindings [
    Text_TimeDelta {
     id=#26
     possessed_class=CommonTextBlock
     tracks [{
       type=MovieScene2DTransformTrack
       sections [{
         type=MovieScene2DTransformSection
         end_frame=13
         start_frame=0
         channels [
          Translation.X {type=MovieSceneScriptingFloatChannel, keys [[0,453.6279],[4.9137,-28.3125],[5.9137,0]]}
          Translation.Y {type=MovieSceneScriptingFloatChannel}
          Angle {type=MovieSceneScriptingFloatChannel}
          Scale.X {type=MovieSceneScriptingFloatChannel}
          Scale.Y {type=MovieSceneScriptingFloatChannel}
          Shear.X {type=MovieSceneScriptingFloatChannel}
          Shear.Y {type=MovieSceneScriptingFloatChannel}
         ]
        }]
      }]
    }
    Text_PreviousRecord {
     id=#27
     possessed_class=CommonTextBlock
     tracks [{
       type=MovieScene2DTransformTrack
       sections [{
         type=MovieScene2DTransformSection
         end_frame=13
         start_frame=0
         channels [
          Translation.X {type=MovieSceneScriptingFloatChannel, keys [[0,-384.2262],[4.9137,26.6106],[5.9137,0]]}
          Translation.Y {type=MovieSceneScriptingFloatChannel}
          Angle {type=MovieSceneScriptingFloatChannel}
          Scale.X {type=MovieSceneScriptingFloatChannel}
          Scale.Y {type=MovieSceneScriptingFloatChannel}
          Shear.X {type=MovieSceneScriptingFloatChannel}
          Shear.Y {type=MovieSceneScriptingFloatChannel}
         ]
        }]
      }]
    }
   ]}
 }
]
graphs [
 EventGraph {nodes [
   Play Final Score() {is=title K2Node_CustomEvent}
   > PlaySound2D(Sound=/Game/Athena/Sounds/Quests/TimedTrial_Papaya/Papaya_TimedTrial_Score_cue.Papaya_TimedTrial_Score_Cue)
   n1 = SetRank!() -> (OutputDelegate, then, Rank:int) {is=title K2Node_CustomEvent}
   > SetVisibility(
    self=[Get Image_redglow() -> (Image_redglow:Image Object Reference), Get Image_flare_1() -> (Image_flare_1:Image Object Reference)]
    InVisibility=Select(Option 0=HitTestInvisible, Option 1=Hidden, Index="integer > integer"(A=@n1.Rank, B=0) {is=title K2Node_CallFunction})
   )
   > SetVisibility(
    self=[Get Image_topglow() -> (Image_topglow:Image Object Reference), Get Image_flare() -> (Image_flare:Image Object Reference)]
    InVisibility=Select(Option 0=HitTestInvisible, Option 1=Hidden, Index="integer > integer"(A=@n1.Rank, B=1))
   )
   n2 = Get Text_Rank() -> (Text_Rank:Common Text Object Reference)
   > "SetText (Text)"(
    self=@n2
    InText=Select(
     Option 0:Text=Select!(
      Option 0="LOCTEXT(\"S\")"
      Option 1="LOCTEXT(\"A\")"
      Option 2:Text="LOCTEXT(\"B\")"
      Option 3:Text="LOCTEXT(\"C\")"
      Option 4:Text="LOCTEXT(\"D\")"
      Index:int=@n1.Rank
     ) -> (ReturnValue)
     Option 1:Text="LOCTEXT(\"D\")"
     Index="integer > integer"(A=@n1.Rank, B=3)
    ) -> (ReturnValue:Text)
   )
   n3 = Break Slate Font Info!(SlateFontInfo:"Slate Font Info Structure (by ref)"=Get Font!(self:Text Object Reference=@n2) -> (Font:Slate Font Info Structure)) -> (
    FontObject:Object Reference
    FontMaterial:Object Reference
    OutlineSettings:Font Outline Settings Structure
    TypefaceFontName:Name
    Size:float=~
    LetterSpacing:int
    SkewAmount:float=~
    bForceMonospaced:bool=~
    bMaterialIsStencil:bool=~
    MonospacedWidth:float=~
   ) {is=title K2Node_BreakStruct}
   > SetFont(
    self=@n2
    InFontInfo=MakeSlateFontInfo!(
     FontObject:Object Reference=@n3
     FontMaterial:Object Reference=Select(
      Option 0:Object Reference=/Game/Athena/HUD/Materials/Victory/M_WinText_Gold_Inst
      Option 1:Object Reference=/Game/Athena/HUD/Materials/Victory/M_WinText_Silver_Inst
      Option 2:Object Reference=/Game/Athena/HUD/Materials/M_UI_OutlineShine_White_Manual
      Option 3:Object Reference=&M_UI_OutlineShine_White_Manual
      Option 4:Object Reference=&M_UI_OutlineShine_White_Manual
      Index=@n1.Rank
     ) -> (ReturnValue:Object Reference)
     OutlineSettings:Font Outline Settings Structure=@n3.OutlineSettings
     TypefaceFontName:Name=@n3.TypefaceFontName
     Size:float=@n3.Size
     LetterSpacing:int=0
     SkewAmount:float=0
     bForceMonospaced:bool=false
     bMaterialIsStencil:bool=false
     MonospacedWidth:float=1
    ) -> (ReturnValue:Slate Font Info Structure)
   )
   EventIntro() {is=title K2Node_CustomEvent}
   > SetVisibility(InVisibility=HitTestInvisible)
   > EventReset()
   > PlayAnimation(InAnimation=Get Intro())
   n4 = EventOutro!() -> (OutputDelegate, then, Delay:double=~) {is=title K2Node_CustomEvent}
   > RetriggerableDelay(Duration=@n4.Delay)
   > n5 = Branch(Condition=Get bMinigameActive() -> (bMinigameActive:bool))
   StopAnimation(execute=@n5.else, InAnimation=Get Intro())
   n6 = Get Outro() -> (Outro:Widget Animation Object Reference)
   > n7 = PlayAnimation(InAnimation=@n6)
   n8 = OutroFinished() {is=title K2Node_CustomEvent}
   n9 = Set Timer by Event(execute=@n7.then, Delegate=@n8, Time=GetEndTime(self=@n6), bLooping=false)
   > Set OutroTimerHandle(OutroTimerHandle:Timer Handle Structure=@n9) -> (Output_Get:Timer Handle Structure)
   Play Outro() {is=title K2Node_CustomEvent}
   > PlaySound2D(Sound=/Game/Athena/Sounds/Quests/TimedTrial_Papaya/Papaya_TimedTrial_Outro_Cue)
   EventReset() {is=title K2Node_CustomEvent}
   > Clear and Invalidate Timer by Handle(Handle=Get OutroTimerHandle() -> (OutroTimerHandle:Timer Handle Structure))
   > Clear and Invalidate Timer by Handle(Handle=Get RankAnimationTimerHandle() -> (RankAnimationTimerHandle:Timer Handle Structure))
   > StopAllAnimations(self:User Widget Object Reference)
   > PlayAnimation(InAnimation=Get Reset() -> (Reset:Widget Animation Object Reference))
   > SetActiveWidgetIndex(
    self:Common Visibility Switcher Object Reference=Get VisibilitySwitcher_State() -> (VisibilitySwitcher_State:Common Visibility Switcher Object Reference)
    Index=0
   )
   n10 = Event On Handle Minigame Timer UI Enabled Changed!() -> (OutputDelegate, then, bEnabled:bool) {is=title K2Node_Event}
   > SetVisibility(
    self=Get TimerHorizontalBox() -> (TimerHorizontalBox:Horizontal Box Object Reference)
    InVisibility=Select!(Option 0:ESlateVisibility Enum=Collapsed, Option 1:ESlateVisibility Enum=HitTestInvisible, Index:bool=@n10.bEnabled) -> (ReturnValue:ESlateVisibility Enum)
   )
   Event Construct()
   > BindToAnimationStarted!(
    execute
    self:User Widget Object Reference
    Animation:Widget Animation Object Reference=Get FinalScore() -> (FinalScore:Widget Animation Object Reference)
    Delegate:Delegate=Create Event()
   ) -> (then)
   > BindToAnimationStarted(Animation=Get Outro(), Delegate=Create Event())
   > DelayUntilNextTick!(execute) -> (then)
   > SendGameplayEventToActor!(
    execute
    Actor:Actor Object Reference=GetOwningPlayerPawn!(self:User Widget Object Reference) -> (ReturnValue:Pawn Object Reference)
    EventTag:Gameplay Tag Structure=(TagName="Event.Checkpoint.Ready")
    Payload:Gameplay Event Data Structure
   ) -> (then)
   Event On Handle Minigame Started!() -> (OutputDelegate, then, StartData:"Minigame Activity Started Data Structure (by ref)") {is=title K2Node_Event}
   > Branch(Condition=Get RaceUIEnabled() -> (RaceUIEnabled:bool))
   > EventIntro() {is=title K2Node_CallFunction}
   n11 = Event On Handle Minigame Ended!() -> (
    OutputDelegate
    then
    EndData_bSuccessfullyCompleted:bool=~
    EndData_Stat:EMinigameActivityStat Enum
    EndData_FinalStatValue:float=~
    EndData_FinalStatBestValue:float=~
    EndData_FinalScore:int
    EndData_TotalScore:int
    EndData_Rank:int
   ) {is=title K2Node_Event}
   > n12 = Branch(Condition=AND Boolean(A=@n11.EndData_bSuccessfullyCompleted, B=Get RaceUIEnabled()))
   > PlayAnimation(InAnimation=Get FinalScore())
   > PlayAnimation(InAnimation=Get NewRecord() -> (NewRecord:Widget Animation Object Reference))
   > EventOutro!(execute, self, Delay:double=Get RaceCompleteOutroDelay() -> (RaceCompleteOutroDelay:double)) -> (then) {is=title K2Node_CallFunction}
   EventOutro(execute=@n12.else, Delay=Get RaceFailedOutroDelay() -> (RaceFailedOutroDelay:double))
   n13 = Event On Handle Minigame Progress UI Enabled Changed!() -> (OutputDelegate, then, bEnabled:bool) {is=title K2Node_Event}
   > SetVisibility(
    self=Get Progress_Vertical_Box() -> (Progress_Vertical_Box:Vertical Box Object Reference)
    InVisibility=Select(Option 0=Collapsed, Option 1=HitTestInvisible, Index=@n13.bEnabled)
   )
   n14 = Event On Handle Minigame UI Enabled Changed!() -> (OutputDelegate, then, bEnabled:bool) {is=title K2Node_Event}
   > Set RaceUIEnabled(RaceUIEnabled:bool=@n14.bEnabled) -> (Output_Get:bool)
   n15 = Event On Handle Minigame Best Lap Time UI Enabled Changed!() -> (OutputDelegate, then, bEnabled:bool) {is=title K2Node_Event}
   > SetVisibility(
    self=Get BestLapHorizontalBox() -> (BestLapHorizontalBox:Horizontal Box Object Reference)
    InVisibility=Select(Option 0=Collapsed, Option 1=HitTestInvisible, Index=@n15.bEnabled)
   )
   SetVisibility(execute=@n8.then, InVisibility=Collapsed)
  ]}
 ToTimeFormat {nodes [
   n1 = ToTimeFormat!() -> (then, Time:double=~) {is=title K2Node_FunctionEntry}
   n2 = "Division (Whole and Remainder)"!(Dividend:double=@n1, Divisor:double=60) -> (Remainder:double, ReturnValue:int=0)
   n3 = Select(
    Option 0:double="To Float (Integer)"(InInt=@n2.ReturnValue)
    Option 1:double=0
    Index="integer < integer"!(A:int=@n2.ReturnValue, B:int=0) -> (ReturnValue:bool) {is=title K2Node_CallFunction}
   ) -> (ReturnValue:double)
   n4 = "Absolute (Float)"(A=@n2)
   > Return Node!(
    execute
    FormattedTime:Text=Format Text!(
     Format=MakeLiteralText(Value="LOCTEXT(\"{minutes}:{seconds}\")")
     seconds:Text="To Text (Float)"(Value=@n4, MinimumIntegralDigits=2, MaximumIntegralDigits=2, MinimumFractionalDigits=2, MaximumFractionalDigits=2)
     minutes:Text="To Text (Float)"(Value=@n3, MinimumIntegralDigits=1, MaximumIntegralDigits=2, MinimumFractionalDigits=0, MaximumFractionalDigits=0)
    ) -> (Result)
   )
  ]}
]
variables [
 OutroTimerHandle {type=TimerHandle}
 PreviousBestTime {type=number}
 RaceCompleteOutroDelay {type=number, default=8}
 RaceFailedOutroDelay {type=number, default=3}
 RaceUIEnabled {type=boolean, default=true}
 RankAnimationTimerHandle {type=TimerHandle}
 ThisRunFinalTime {type=number}
]
widgets {
 info {inheritedWidgetCount=6, parentClass=/Script/FortniteUI.AthenaMinigameActivityWidget, rootWidgetClass=&Overlay, widgetCount=41}
 widgets [
  Overlay_29 {widgetClassPath=&Overlay, widget=Overlay_29}
  Overlay_Container {
   widgetClassPath=&Overlay
   parent=Overlay_29
   widget=Overlay_Container
   slot {class=OverlaySlot, properties {padding=<ltrb 0 50 0 0>}}
  }
  Image_bg {
   widgetClassPath=&Image
   parent=Overlay_Container
   widget=Image_bg
   properties {
    visibility=SelfHitTestInvisible
    brush {
     bIsDynamicallyLoaded=false
     drawAs=Image
     imageType=NoImage
     mirroring=NoMirror
     resourceName=None
     resourceObject=null
     tiling=NoTile
     imageSize=<410 70>
     margin=<ltrb 0 0 0 0>
     outlineSettings {
      bUseBrushTransparency=false
      roundingType=HalfHeightRadius
      width=0
      color {colorUseRule=UseColor_Specified, specifiedColor=<rgba 0 0 0 0>}
      cornerRadii=<0 0 0 0>
     }
     tintColor {colorUseRule=UseColor_Specified, specifiedColor=<rgba 0 0 0 0.5>}
     uVRegion {bIsValid=false, max=<0 0>, min=<0 0>}
    }
   }
   slot {class=OverlaySlot, properties {horizontalAlignment=HAlign_Fill, verticalAlignment=VAlign_Fill}}
  }
  Image_Triangle {
   widgetClassPath=&Image
   bIsVariable=true
   parent=Overlay_Container
   widget=Image_Triangle
   properties {
    renderOpacity=0
    visibility=Collapsed
    brush {
     bIsDynamicallyLoaded=false
     drawAs=Image
     imageType=NoImage
     mirroring=NoMirror
     resourceName=None
     resourceObject=/Game/UI/HUD/Materials/M_FX_TriangleScroll-Burst_Fast_Vertical_Gold
     tiling=NoTile
     imageSize=<250 250>
     margin=<ltrb 0 0 0 0>
     outlineSettings {
      bUseBrushTransparency=false
      roundingType=HalfHeightRadius
      width=0
      color {colorUseRule=UseColor_Specified, specifiedColor=<rgba 0 0 0 0>}
      cornerRadii=<0 0 0 0>
     }
     tintColor {colorUseRule=UseColor_Specified, specifiedColor=<rgba 1 1 1 1>}
     uVRegion {bIsValid=false, max=<0 0>, min=<0 0>}
    }
   }
   slot {class=OverlaySlot, properties {verticalAlignment=VAlign_Bottom, padding=<ltrb -20 -300 0 0>}}
  }
  VisibilitySwitcher_State {
   widgetClassPath=&CommonVisibilitySwitcher
   bIsVariable=true
   parent=Overlay_Container
   widget=VisibilitySwitcher_State
   slot {class=OverlaySlot, properties {horizontalAlignment=HAlign_Fill, verticalAlignment=VAlign_Fill, padding=<ltrb 20 20 20 20>}}
  }
  HorizontalBoxActivityStatus {
   widgetClassPath=&HorizontalBox
   parent=VisibilitySwitcher_State
   widget=HorizontalBoxActivityStatus
   slot {class=CommonVisibilitySwitcherSlot}
  }
  CommonBorder_Rank {
   widgetClassPath=&CommonBorder
   parent=HorizontalBoxActivityStatus
   widget=CommonBorder_Rank
   properties {renderOpacity=0, visibility=Collapsed, desiredSizeScale=<0 1>, padding=<ltrb 0 0 0 0>}
   slot {class=HorizontalBoxSlot, properties {horizontalAlignment=HAlign_Left, verticalAlignment=VAlign_Center, padding=<ltrb 0 -60 0 -60>}}
  }
  Overlay_FinalRank {
   widgetClassPath=&Overlay
   parent=CommonBorder_Rank
   widget=Overlay_FinalRank
   properties {renderTransformPivot=<0.5 0.7561>}
   slot {class=BorderSlot, properties {padding=<ltrb 0 0 0 0>}}
  }
  Text_Rank {
   widgetClassPath=&CommonTextBlock
   bIsVariable=true
   parent=Overlay_FinalRank
   widget=Text_Rank
   properties {
    text=FINAL
    font {
     bForceMonospaced=false
     bMaterialIsStencil=false
     fontMaterial=&M_WinText_Gold_Inst
     fontObject=&Burbank
     letterSpacing=0
     monospacedWidth=1
     size=40
     skewAmount=0
     typefaceFontName=BurbankRegular-Black
     outlineSettings {
      bApplyOutlineToDropShadows=false
      bMiteredCorners=false
      bSeparateFillAlpha=false
      outlineMaterial=null
      outlineSize=4
      outlineColor=<rgba 0 0 0 1>
     }
    }
    shadowOffset=<0 0>
   }
   slot {class=OverlaySlot, properties {verticalAlignment=VAlign_Center, padding=<ltrb 0 -20 15 -20>}}
  }
  CommonTextBlock {
   widgetClassPath=&CommonTextBlock
   parent=Overlay_FinalRank
   widget=CommonTextBlock
   properties {
    text=Rank
    textTransformPolicy=ToUpper
    visibility=Hidden
    font {
     bForceMonospaced=false
     bMaterialIsStencil=false
     fontMaterial=null
     fontObject=&Burbank
     letterSpacing=0
     monospacedWidth=1
     size=18
     skewAmount=0
     typefaceFontName=BurbankRegular-Black
     outlineSettings {
      bApplyOutlineToDropShadows=false
      bMiteredCorners=false
      bSeparateFillAlpha=false
      outlineMaterial=null
      outlineSize=2
      outlineColor=<rgba 0 0 0 1>
     }
    }
    renderTransform {angle=0, scale=<1 1>, shear=<0 0>, translation=<10 2>}
    shadowOffset=<0 0>
   }
   slot {class=OverlaySlot, properties {verticalAlignment=VAlign_Bottom, padding=<ltrb 30 0 -400 0>}}
  }
  Overlay_90 {
   widgetClassPath=&Overlay
   parent=Overlay_FinalRank
   widget=Overlay_90
   properties {renderTransform {angle=0, scale=<1 1>, shear=<12 0>, translation=<0 0>}}
   slot {class=OverlaySlot, properties {horizontalAlignment=HAlign_Center, verticalAlignment=VAlign_Center}}
  }
  Image_star {
   widgetClassPath=&Image
   bIsVariable=true
   parent=Overlay_90
   widget=Image_star
   properties {
    brush {
     bIsDynamicallyLoaded=false
     drawAs=Image
     imageType=NoImage
     mirroring=NoMirror
     resourceName=None
     resourceObject=/Game/UI/InGame/Quest/Materials/M_UI_SparkStar
     tiling=NoTile
     imageSize=<128 128>
     margin=<ltrb 0 0 0 0>
     outlineSettings {
      bUseBrushTransparency=false
      roundingType=HalfHeightRadius
      width=0
      color {colorUseRule=UseColor_Specified, specifiedColor=<rgba 0 0 0 0>}
      cornerRadii=<0 0 0 0>
     }
     tintColor {colorUseRule=UseColor_Specified, specifiedColor=<rgba 1 1 1 1>}
     uVRegion {bIsValid=false, max=<0 0>, min=<0 0>}
    }
    renderTransform {angle=0, scale=<1 1>, shear=<0 0>, translation=<-25 -19>}
   }
   slot {class=OverlaySlot, properties {horizontalAlignment=HAlign_Center, verticalAlignment=VAlign_Center, padding=<ltrb -64 -64 -64 -64>}}
  }
  Image_redglow {
   widgetClassPath=&Image
   bIsVariable=true
   parent=Overlay_90
   widget=Image_redglow
   properties {
    visibility=Hidden
    brush {
     bIsDynamicallyLoaded=false
     drawAs=Image
     imageType=NoImage
     mirroring=NoMirror
     resourceName=None
     resourceObject=/Game/Athena/HUD/LTM/Wax/Materials/M_UI_TextGlow_Add_Inst
     tiling=NoTile
     imageSize=<150 150>
     margin=<ltrb 0 0 0 0>
     outlineSettings {
      bUseBrushTransparency=false
      roundingType=HalfHeightRadius
      width=0
      color {colorUseRule=UseColor_Specified, specifiedColor=<rgba 0 0 0 0>}
      cornerRadii=<0 0 0 0>
     }
     tintColor {colorUseRule=UseColor_Specified, specifiedColor=<rgba 0.5583 0.0637 0 1>}
     uVRegion {bIsValid=false, max=<0 0>, min=<0 0>}
    }
    renderTransform {angle=0, scale=<1 1>, shear=<0 0>, translation=<13 15>}
   }
   slot {class=OverlaySlot, properties {horizontalAlignment=HAlign_Center, verticalAlignment=VAlign_Center, padding=<ltrb -75 -75 -75 -75>}}
  }
  Image_topglow {
   widgetClassPath=&Image
   bIsVariable=true
   parent=Overlay_90
   widget=Image_topglow
   properties {
    visibility=Hidden
    brush {
     bIsDynamicallyLoaded=false
     drawAs=Image
     imageType=NoImage
     mirroring=NoMirror
     resourceName=None
     resourceObject=/Game/Athena/HUD/LTM/Wax/Materials/M_UI_TextGlow_Pulse_Add
     tiling=NoTile
     imageSize=<199.3698 158.721>
     margin=<ltrb 0 0 0 0>
     outlineSettings {
      bUseBrushTransparency=false
      roundingType=HalfHeightRadius
      width=0
      color {colorUseRule=UseColor_Specified, specifiedColor=<rgba 0 0 0 0>}
      cornerRadii=<0 0 0 0>
     }
     tintColor {colorUseRule=UseColor_Specified, specifiedColor=<rgba 0.0722 0 1 1>}
     uVRegion {bIsValid=false, max=<0 0>, min=<0 0>}
    }
    renderTransform {angle=0, scale=<1 1>, shear=<-3 -16>, translation=<-15 -13>}
   }
   slot {class=OverlaySlot, properties {horizontalAlignment=HAlign_Center, verticalAlignment=VAlign_Center, padding=<ltrb -100 -100 -100 -100>}}
  }
  Image_flare {
   widgetClassPath=&Image
   bIsVariable=true
   parent=Overlay_90
   widget=Image_flare
   properties {
    brush {
     bIsDynamicallyLoaded=false
     drawAs=Image
     imageType=NoImage
     mirroring=NoMirror
     resourceName=None
     resourceObject=/Game/Athena/HUD/LTM/Ashton/Materials/M_Sharp_Cross_Flare_Flicker_Inst
     tiling=NoTile
     imageSize=<128 128>
     margin=<ltrb 0 0 0 0>
     outlineSettings {
      bUseBrushTransparency=false
      roundingType=HalfHeightRadius
      width=0
      color {colorUseRule=UseColor_Specified, specifiedColor=<rgba 0 0 0 0>}
      cornerRadii=<0 0 0 0>
     }
     tintColor {colorUseRule=UseColor_Specified, specifiedColor=<rgba 1 1 1 1>}
     uVRegion {bIsValid=false, max=<0 0>, min=<0 0>}
    }
    renderTransform {angle=0, scale=<1 1>, shear=<0 0>, translation=<-31 32>}
   }
   slot {class=OverlaySlot, properties {horizontalAlignment=HAlign_Center, verticalAlignment=VAlign_Center, padding=<ltrb -32 -32 -32 -32>}}
  }
  Image_flare_1 {
   widgetClassPath=&Image
   bIsVariable=true
   parent=Overlay_90
   widget=Image_flare_1
   properties {
    brush {
     bIsDynamicallyLoaded=false
     drawAs=Image
     imageType=NoImage
     mirroring=NoMirror
     resourceName=None
     resourceObject=&M_Sharp_Cross_Flare_Flicker_Inst
     tiling=NoTile
     imageSize=<130 130>
     margin=<ltrb 0 0 0 0>
     outlineSettings {
      bUseBrushTransparency=false
      roundingType=HalfHeightRadius
      width=0
      color {colorUseRule=UseColor_Specified, specifiedColor=<rgba 0 0 0 0>}
      cornerRadii=<0 0 0 0>
     }
     tintColor {colorUseRule=UseColor_Specified, specifiedColor=<rgba 1 1 1 1>}
     uVRegion {bIsValid=false, max=<0 0>, min=<0 0>}
    }
    renderTransform {angle=0, scale=<1 1>, shear=<0 0>, translation=<24 -27.1139>}
   }
   slot {class=OverlaySlot, properties {horizontalAlignment=HAlign_Center, verticalAlignment=VAlign_Center, padding=<ltrb -50 -50 -50 -50>}}
  }
  Image_flare_2 {
   widgetClassPath=&Image
   bIsVariable=true
   parent=Overlay_90
   widget=Image_flare_2
   properties {
    brush {
     bIsDynamicallyLoaded=false
     drawAs=Image
     imageType=NoImage
     mirroring=NoMirror
     resourceName=None
     resourceObject=&M_Sharp_Cross_Flare_Flicker_Inst
     tiling=NoTile
     imageSize=<45 45>
     margin=<ltrb 0 0 0 0>
     outlineSettings {
      bUseBrushTransparency=false
      roundingType=HalfHeightRadius
      width=0
      color {colorUseRule=UseColor_Specified, specifiedColor=<rgba 0 0 0 0>}
      cornerRadii=<0 0 0 0>
     }
     tintColor {colorUseRule=UseColor_Specified, specifiedColor=<rgba 1 1 1 1>}
     uVRegion {bIsValid=false, max=<0 0>, min=<0 0>}
    }
    renderTransform {angle=0, scale=<1 1>, shear=<0 0>, translation=<-30 -44.4265>}
   }
   slot {class=OverlaySlot, properties {horizontalAlignment=HAlign_Center, verticalAlignment=VAlign_Center, padding=<ltrb -32 -32 -32 -32>}}
  }
  Image_flare_3 {
   widgetClassPath=&Image
   bIsVariable=true
   parent=Overlay_90
   widget=Image_flare_3
   properties {
    brush {
     bIsDynamicallyLoaded=false
     drawAs=Image
     imageType=NoImage
     mirroring=NoMirror
     resourceName=None
     resourceObject=&M_Sharp_Cross_Flare_Flicker_Inst
     tiling=NoTile
     imageSize=<32 32>
     margin=<ltrb 0 0 0 0>
     outlineSettings {
      bUseBrushTransparency=false
      roundingType=HalfHeightRadius
      width=0
      color {colorUseRule=UseColor_Specified, specifiedColor=<rgba 0 0 0 0>}
      cornerRadii=<0 0 0 0>
     }
     tintColor {colorUseRule=UseColor_Specified, specifiedColor=<rgba 1 1 1 1>}
     uVRegion {bIsValid=false, max=<0 0>, min=<0 0>}
    }
    renderTransform {angle=0, scale=<1.6264 1.6264>, shear=<0 0>, translation=<24 13.7359>}
   }
   slot {class=OverlaySlot, properties {horizontalAlignment=HAlign_Center, verticalAlignment=VAlign_Center, padding=<ltrb -32 -32 -32 -32>}}
  }
  Image_whiteglow_1 {
   widgetClassPath=&Image
   bIsVariable=true
   parent=Overlay_90
   widget=Image_whiteglow_1
   properties {
    brush {
     bIsDynamicallyLoaded=false
     drawAs=Image
     imageType=NoImage
     mirroring=NoMirror
     resourceName=None
     resourceObject=&M_UI_TextGlow_Add_Inst
     tiling=NoTile
     imageSize=<150 150>
     margin=<ltrb 0 0 0 0>
     outlineSettings {
      bUseBrushTransparency=false
      roundingType=HalfHeightRadius
      width=0
      color {colorUseRule=UseColor_Specified, specifiedColor=<rgba 0 0 0 0>}
      cornerRadii=<0 0 0 0>
     }
     tintColor {colorUseRule=UseColor_Specified, specifiedColor=<rgba 1 0.7554 0.3604 1>}
     uVRegion {bIsValid=false, max=<0 0>, min=<0 0>}
    }
    renderTransform {angle=0, scale=<1.2 1.5>, shear=<0 0>, translation=<0 0>}
   }
   slot {class=OverlaySlot, properties {horizontalAlignment=HAlign_Center, verticalAlignment=VAlign_Center, padding=<ltrb -75 -75 -75 -75>}}
  }
  VerticalBox_0 {
   widgetClassPath=&VerticalBox
   parent=HorizontalBoxActivityStatus
   widget=VerticalBox_0
   slot {class=HorizontalBoxSlot, properties {verticalAlignment=VAlign_Top, size {sizeRule=Fill, value=1}}}
  }
  Text_ActivityName {
   widgetClassPath=&CommonTextBlock
   bInherited=true
   parent=VerticalBox_0
   widget=Text_ActivityName
   properties {
    text="{Activity}"
    textTransformPolicy=ToUpper
    font {
     bForceMonospaced=false
     bMaterialIsStencil=false
     fontMaterial=null
     fontObject=&Burbank
     letterSpacing=0
     monospacedWidth=1
     size=24
     skewAmount=0
     typefaceFontName=BurbankRegular-Black
     outlineSettings {
      bApplyOutlineToDropShadows=false
      bMiteredCorners=false
      bSeparateFillAlpha=false
      outlineMaterial=null
      outlineSize=0
      outlineColor=<rgba 0 0 0 1>
     }
    }
    shadowOffset=<0 0>
   }
   slot {class=VerticalBoxSlot, properties {padding=<ltrb 0 -8 15 -4>}}
  }
  TimerHorizontalBox {
   widgetClassPath=&HorizontalBox
   bIsVariable=true
   parent=VerticalBox_0
   widget=TimerHorizontalBox
   properties {visibility=HitTestInvisible}
   slot {class=VerticalBoxSlot, properties {padding=<ltrb 0 -4 0 -8>}}
  }
  SizeBox_0 {
   widgetClassPath=&SizeBox
   parent=TimerHorizontalBox
   widget=SizeBox_0
   properties {bOverride_MinDesiredWidth=true, minDesiredWidth=120}
   slot {class=HorizontalBoxSlot}
  }
  Text_Stat {
   widgetClassPath=&CommonTextBlock
   bInherited=true
   parent=SizeBox_0
   widget=Text_Stat
   properties {
    text=00:00.00
    font {
     bForceMonospaced=false
     bMaterialIsStencil=false
     fontMaterial=null
     fontObject=&Burbank
     letterSpacing=0
     monospacedWidth=1
     size=24
     skewAmount=0
     typefaceFontName=BurbankRegular-Black
     outlineSettings {
      bApplyOutlineToDropShadows=false
      bMiteredCorners=false
      bSeparateFillAlpha=false
      outlineMaterial=null
      outlineSize=0
      outlineColor=<rgba 0 0 0 1>
     }
    }
    shadowOffset=<0 0>
   }
   slot {class=SizeBoxSlot}
  }
  OverlayAddTime {
   widgetClassPath=&Overlay
   parent=TimerHorizontalBox
   widget=OverlayAddTime
   properties {renderOpacity=0}
   slot {class=HorizontalBoxSlot, properties {padding=<ltrb 15 0 0 0>}}
  }
  Text_AddTime {
   widgetClassPath=&CommonTextBlock
   bInherited=true
   bIsVariable=true
   parent=OverlayAddTime
   widget=Text_AddTime
   properties {
    text=+1s
    colorAndOpacity {colorUseRule=UseColor_Specified, specifiedColor=<rgba 0 0.9606 1 1>}
    font {
     bForceMonospaced=false
     bMaterialIsStencil=false
     fontMaterial=null
     fontObject=&Burbank
     letterSpacing=0
     monospacedWidth=1
     size=18
     skewAmount=0
     typefaceFontName=BurbankRegular-Black
     outlineSettings {
      bApplyOutlineToDropShadows=false
      bMiteredCorners=false
      bSeparateFillAlpha=false
      outlineMaterial=null
      outlineSize=0
      outlineColor=<rgba 0 0 0 1>
     }
    }
    renderTransformPivot=<0.4 0.8>
    shadowOffset=<0 0>
   }
   slot {class=OverlaySlot, properties {verticalAlignment=VAlign_Center}}
  }
  Image_timeglow {
   widgetClassPath=&Image
   bIsVariable=true
   parent=OverlayAddTime
   widget=Image_timeglow
   properties {
    brush {
     bIsDynamicallyLoaded=false
     drawAs=Image
     imageType=NoImage
     mirroring=NoMirror
     resourceName=None
     resourceObject=&M_UI_TextGlow_Add_Inst
     tiling=NoTile
     imageSize=<303.8873 51.0456>
     margin=<ltrb 0 0 0 0>
     outlineSettings {
      bUseBrushTransparency=false
      roundingType=HalfHeightRadius
      width=0
      color {colorUseRule=UseColor_Specified, specifiedColor=<rgba 0 0 0 0>}
      cornerRadii=<0 0 0 0>
     }
     tintColor {colorUseRule=UseColor_Specified, specifiedColor=<rgba 0 0.2157 0.5583 1>}
     uVRegion {bIsValid=false, max=<0 0>, min=<0 0>}
    }
    renderTransform {angle=0, scale=<1 1>, shear=<-36 0>, translation=<0 0>}
    renderTransformPivot=<0.6488 0.5>
   }
   slot {class=OverlaySlot, properties {horizontalAlignment=HAlign_Center, verticalAlignment=VAlign_Center, padding=<ltrb -150 -75 -100 -75>}}
  }
  BestLapHorizontalBox {
   widgetClassPath=&HorizontalBox
   bIsVariable=true
   parent=VerticalBox_0
   widget=BestLapHorizontalBox
   slot {class=VerticalBoxSlot, properties {padding=<ltrb 0 10 0 -10>}}
  }
  Text_BestLapTimeLabel {
   widgetClassPath=&CommonTextBlock
   parent=BestLapHorizontalBox
   widget=Text_BestLapTimeLabel
   properties {
    text=Best Lap
    textTransformPolicy=ToUpper
    font {
     bForceMonospaced=false
     bMaterialIsStencil=false
     fontMaterial=null
     fontObject=&Burbank
     letterSpacing=0
     monospacedWidth=1
     size=16
     skewAmount=0
     typefaceFontName=BurbankRegular-Black
     outlineSettings {
      bApplyOutlineToDropShadows=false
      bMiteredCorners=false
      bSeparateFillAlpha=false
      outlineMaterial=null
      outlineSize=0
      outlineColor=<rgba 0 0 0 1>
     }
    }
    shadowOffset=<0 0>
   }
   slot {class=HorizontalBoxSlot, properties {horizontalAlignment=HAlign_Left, verticalAlignment=VAlign_Top}}
  }
  Text_BestLapTime {
   widgetClassPath=&CommonTextBlock
   bInherited=true
   parent=BestLapHorizontalBox
   widget=Text_BestLapTime
   properties {
    text=00:00.00
    font {
     bForceMonospaced=false
     bMaterialIsStencil=false
     fontMaterial=null
     fontObject=&Burbank
     letterSpacing=0
     monospacedWidth=1
     size=16
     skewAmount=0
     typefaceFontName=BurbankRegular-Black
     outlineSettings {
      bApplyOutlineToDropShadows=false
      bMiteredCorners=false
      bSeparateFillAlpha=false
      outlineMaterial=null
      outlineSize=0
      outlineColor=<rgba 0 0 0 1>
     }
    }
    shadowOffset=<0 0>
   }
   slot {class=HorizontalBoxSlot, properties {horizontalAlignment=HAlign_Left, verticalAlignment=VAlign_Top, padding=<ltrb 15 0 0 0>}}
  }
  OverlayAddTime_1 {
   widgetClassPath=&Overlay
   parent=BestLapHorizontalBox
   widget=OverlayAddTime_1
   properties {renderOpacity=0}
   slot {class=HorizontalBoxSlot, properties {padding=<ltrb 15 0 0 0>}}
  }
  Text_AddTime_1 {
   widgetClassPath=&CommonTextBlock
   bIsVariable=true
   parent=OverlayAddTime_1
   widget=Text_AddTime_1
   properties {
    text=+1s
    colorAndOpacity {colorUseRule=UseColor_Specified, specifiedColor=<rgba 0 0.9606 1 1>}
    font {
     bForceMonospaced=false
     bMaterialIsStencil=false
     fontMaterial=null
     fontObject=&Burbank
     letterSpacing=0
     monospacedWidth=1
     size=18
     skewAmount=0
     typefaceFontName=BurbankRegular-Black
     outlineSettings {
      bApplyOutlineToDropShadows=false
      bMiteredCorners=false
      bSeparateFillAlpha=false
      outlineMaterial=null
      outlineSize=0
      outlineColor=<rgba 0 0 0 1>
     }
    }
    renderTransformPivot=<0.4 0.8>
    shadowOffset=<0 0>
   }
   slot {class=OverlaySlot, properties {verticalAlignment=VAlign_Center}}
  }
  Image_timeglow_1 {
   widgetClassPath=&Image
   bIsVariable=true
   parent=OverlayAddTime_1
   widget=Image_timeglow_1
   properties {
    brush {
     bIsDynamicallyLoaded=false
     drawAs=Image
     imageType=NoImage
     mirroring=NoMirror
     resourceName=None
     resourceObject=&M_UI_TextGlow_Add_Inst
     tiling=NoTile
     imageSize=<303.8873 51.0456>
     margin=<ltrb 0 0 0 0>
     outlineSettings {
      bUseBrushTransparency=false
      roundingType=HalfHeightRadius
      width=0
      color {colorUseRule=UseColor_Specified, specifiedColor=<rgba 0 0 0 0>}
      cornerRadii=<0 0 0 0>
     }
     tintColor {colorUseRule=UseColor_Specified, specifiedColor=<rgba 0 0.2157 0.5583 1>}
     uVRegion {bIsValid=false, max=<0 0>, min=<0 0>}
    }
    renderTransform {angle=0, scale=<1 1>, shear=<-36 0>, translation=<0 0>}
    renderTransformPivot=<0.6488 0.5>
   }
   slot {class=OverlaySlot, properties {horizontalAlignment=HAlign_Center, verticalAlignment=VAlign_Center, padding=<ltrb -150 -75 -100 -75>}}
  }
  CommonBorder_Criteria {
   widgetClassPath=&CommonBorder
   parent=HorizontalBoxActivityStatus
   widget=CommonBorder_Criteria
   properties {
    clipping=ClipToBounds
    horizontalAlignment=HAlign_Right
    verticalAlignment=VAlign_Center
    padding=<ltrb 0 0 0 0>
    renderTransformPivot=<0.5 1>
   }
   slot {class=HorizontalBoxSlot, properties {horizontalAlignment=HAlign_Right, verticalAlignment=VAlign_Top, padding=<ltrb 15 -5 0 0>}}
  }
  Progress_Vertical_Box {
   widgetClassPath=&VerticalBox
   bIsVariable=true
   parent=CommonBorder_Criteria
   widget=Progress_Vertical_Box
   properties {visibility=HitTestInvisible}
   slot {class=BorderSlot, properties {horizontalAlignment=HAlign_Right, verticalAlignment=VAlign_Center, padding=<ltrb 0 0 0 0>}}
  }
  Text_Lap {
   widgetClassPath=&CommonTextBlock
   bInherited=true
   bIsVariable=true
   parent=Progress_Vertical_Box
   widget=Text_Lap
   properties {
    text=Lap X/Y
    textTransformPolicy=ToUpper
    font {
     bForceMonospaced=false
     bMaterialIsStencil=false
     fontMaterial=null
     fontObject=&Burbank
     letterSpacing=0
     monospacedWidth=1
     size=15
     skewAmount=0
     typefaceFontName=BurbankRegular-Black
     outlineSettings {
      bApplyOutlineToDropShadows=false
      bMiteredCorners=false
      bSeparateFillAlpha=false
      outlineMaterial=null
      outlineSize=0
      outlineColor=<rgba 0 0 0 1>
     }
    }
    shadowOffset=<0 0>
   }
   slot {class=VerticalBoxSlot, properties {horizontalAlignment=HAlign_Center}}
  }
  Text_Score {
   widgetClassPath=&CommonTextBlock
   bInherited=true
   bIsVariable=true
   parent=Progress_Vertical_Box
   widget=Text_Score
   properties {
    text=X/Y
    textTransformPolicy=ToUpper
    font {
     bForceMonospaced=false
     bMaterialIsStencil=false
     fontMaterial=null
     fontObject=&Burbank
     letterSpacing=0
     monospacedWidth=1
     size=40
     skewAmount=0
     typefaceFontName=BurbankRegular-Black
     outlineSettings {
      bApplyOutlineToDropShadows=false
      bMiteredCorners=false
      bSeparateFillAlpha=false
      outlineMaterial=null
      outlineSize=0
      outlineColor=<rgba 0 0 0 1>
     }
    }
    shadowOffset=<0 0>
   }
   slot {class=VerticalBoxSlot, properties {horizontalAlignment=HAlign_Center, padding=<ltrb 0 -10 0 0>}}
  }
  CommonTextBlock_NewRecord {
   widgetClassPath=&CommonTextBlock
   parent=VisibilitySwitcher_State
   widget=CommonTextBlock_NewRecord
   properties {
    justification=Center
    lineHeightPercentage=0.85
    text="You set a new \r\nCourse Record!"
    textTransformPolicy=ToUpper
    font {
     bForceMonospaced=false
     bMaterialIsStencil=false
     fontMaterial=/FortUILibrary/Materials/Foundation/M_Font_Shine_Prestige
     fontObject=&Burbank
     letterSpacing=0
     monospacedWidth=1
     size=24
     skewAmount=0
     typefaceFontName=BurbankRegular-Black
     outlineSettings {
      bApplyOutlineToDropShadows=false
      bMiteredCorners=false
      bSeparateFillAlpha=false
      outlineMaterial=null
      outlineSize=0
      outlineColor=<rgba 0 0 0 1>
     }
    }
    shadowOffset=<0 0>
   }
   slot {class=CommonVisibilitySwitcherSlot}
  }
  Verticalbox_OldRecord {
   widgetClassPath=&VerticalBox
   parent=VisibilitySwitcher_State
   widget=Verticalbox_OldRecord
   properties {clipping=ClipToBounds}
   slot {class=CommonVisibilitySwitcherSlot}
  }
  Text_PreviousRecord {
   widgetClassPath=&CommonTextBlock
   parent=Verticalbox_OldRecord
   widget=Text_PreviousRecord
   properties {
    justification=Center
    lineHeightPercentage=0.85
    text=Current Record
    textTransformPolicy=ToUpper
    font {
     bForceMonospaced=false
     bMaterialIsStencil=false
     fontMaterial=null
     fontObject=&Burbank
     letterSpacing=0
     monospacedWidth=1
     size=20
     skewAmount=0
     typefaceFontName=BurbankRegular-Black
     outlineSettings {
      bApplyOutlineToDropShadows=false
      bMiteredCorners=false
      bSeparateFillAlpha=false
      outlineMaterial=null
      outlineSize=0
      outlineColor=<rgba 0 0 0 1>
     }
    }
    shadowOffset=<0 0>
   }
   slot {class=VerticalBoxSlot}
  }
  Text_TimeDelta {
   widgetClassPath=&CommonTextBlock
   bIsVariable=true
   parent=Verticalbox_OldRecord
   widget=Text_TimeDelta
   properties {
    text="0:00.00 (-0:00.00)"
    colorAndOpacity {colorUseRule=UseColor_Specified, specifiedColor=<rgba 0 0.9606 1 1>}
    font {
     bForceMonospaced=false
     bMaterialIsStencil=false
     fontMaterial=null
     fontObject=&Burbank
     letterSpacing=0
     monospacedWidth=1
     size=20
     skewAmount=0
     typefaceFontName=BurbankRegular-Black
     outlineSettings {
      bApplyOutlineToDropShadows=false
      bMiteredCorners=false
      bSeparateFillAlpha=false
      outlineMaterial=null
      outlineSize=0
      outlineColor=<rgba 0 0 0 1>
     }
    }
    renderTransformPivot=<0.4 0.8>
    shadowOffset=<0 0>
   }
   slot {class=VerticalBoxSlot, properties {horizontalAlignment=HAlign_Center, verticalAlignment=VAlign_Center}}
  }
 ]
}
