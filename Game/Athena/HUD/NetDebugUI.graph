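# Net debug HUD widget, based on NetDebugWidget, that lays out ping plus download and upload bandwidth, packet rate and packet loss beside direction arrows.
# On Construct it gathers the seven inherited text blocks into NetDisplayText and calls SetTextStyleForMobile with IsMobileGame.
# That function rebuilds each block's font, using a larger size on mobile.
# TestMobileNetHUDAdjustments uses a FlipFlop to toggle between the two styles for testing.
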WidgetBlueprint /Game/Athena/HUD/NetDebugUI
graphs [
 EventGraph {nodes [
   TestMobileNetHUDAdjustments() {is=title K2Node_CustomEvent}
   > n1 = Flip Flop!(None:Exec) -> (A:Exec, B:Exec, IsA:bool=~) {is=title K2Node_MacroInstance}
   SetTextStyleForMobile!(execute=[@n1.A, @n1.B], self, bIsMobile:bool=@n1) -> (then)
   Event Construct()
   > Set NetDisplayText(NetDisplayText:Array of Common Text Object References=Make Array!(
     "[0]":Common Text Object Reference=Get PingUI() -> (PingUI:Common Text Object Reference)
     "[1]":Common Text Object Reference=Get PacketInRateUI() -> (PacketInRateUI:Common Text Object Reference)
     "[2]":Common Text Object Reference=Get PacketOutRateUI() -> (PacketOutRateUI:Common Text Object Reference)
     "[3]":Common Text Object Reference=Get PacketInLossUI() -> (PacketInLossUI:Common Text Object Reference)
     "[4]":Common Text Object Reference=Get PacketOutLossUI() -> (PacketOutLossUI:Common Text Object Reference)
     "[5]":Common Text Object Reference=Get UpBandwidthUI() -> (UpBandwidthUI:Common Text Object Reference)
     "[6]":Common Text Object Reference=Get DownBandwidthUI() -> (DownBandwidthUI:Common Text Object Reference)
    ) -> (Array:Array of Common Text Object References)) -> (Output_Get:Array of Common Text Object References)
   > SetTextStyleForMobile(bIsMobile=IsMobileGame())
  ]}
 SetTextStyleForMobile {nodes [
   n1 = SetTextStyleForMobile!() -> (then, bIsMobile:bool=~) {is=title K2Node_FunctionEntry}
   > n2 = For Each Loop(Array:Array of Common Text Object References=Get NetDisplayText() -> (NetDisplayText:Array of Common Text Object References)) -> (Array Element:Common Text Object Reference)
   n3 = Select(Option 0:int=8, Option 1:int=12, Index=@n1) -> (ReturnValue:int)
   n4 = Get!(self:Text Object Reference=@n2) -> (
    Font_FontObject:Object Reference
    Font_FontMaterial:Object Reference
    Font_OutlineSettings:Font Outline Settings Structure
    Font_TypefaceFontName:Name
    Font_Size:float=~
    Font_LetterSpacing:int
    Font_SkewAmount:float=~
    Font_bForceMonospaced:bool=~
    Font_bMaterialIsStencil:bool=~
    Font_MonospacedWidth:float=~
   )
   > SetFont(
    self=@n2
    InFontInfo=MakeSlateFontInfo!(
     FontObject:Object Reference=@n4
     FontMaterial:Object Reference=@n4.Font_FontMaterial
     OutlineSettings:Font Outline Settings Structure=@n4.Font_OutlineSettings
     TypefaceFontName:Name=@n4.Font_TypefaceFontName
     Size:float="To Float (Integer)"(InInt=@n3)
     LetterSpacing:int=0
     SkewAmount:float=0
     bForceMonospaced:bool=false
     bMaterialIsStencil:bool=false
     MonospacedWidth:float=1
    ) -> (ReturnValue:Slate Font Info Structure)
   )
  ]}
]
variables [NetDisplayText {type="array<CommonTextBlock>", default []}]
widgets {
 info {inheritedWidgetCount=7, parentClass=/Script/NetUI.NetDebugWidget, rootWidgetClass=&VerticalBox, widgetCount=13}
 widgets [
  VerticalBox_2 {widgetClassPath=&VerticalBox, bIsVariable=true, widget=VerticalBox_2}
  PingUI {
   widgetClassPath=&CommonTextBlock
   bInherited=true
   parent=VerticalBox_2
   widget=PingUI
   properties {
    text="{PING}"
    font {
     bForceMonospaced=false
     bMaterialIsStencil=false
     fontMaterial=null
     fontObject=&NotoSans
     letterSpacing=0
     monospacedWidth=1
     size=8
     skewAmount=0
     typefaceFontName=NotoSans-Bold
     outlineSettings {
      bApplyOutlineToDropShadows=false
      bMiteredCorners=false
      bSeparateFillAlpha=false
      outlineMaterial=null
      outlineSize=0
      outlineColor=<rgba 0 0 0 1>
     }
    }
    shadowOffset=<0 0>
   }
   slot {class=VerticalBoxSlot, properties {horizontalAlignment=HAlign_Left}}
  }
  HorizontalBox_414 {widgetClassPath=&HorizontalBox, parent=VerticalBox_2, widget=HorizontalBox_414, slot {class=VerticalBoxSlot}}
  Image_2 {
   widgetClassPath=&Image
   bIsVariable=true
   parent=HorizontalBox_414
   widget=Image_2
   properties {brush {
     bIsDynamicallyLoaded=false
     drawAs=Image
     imageType=NoImage
     mirroring=NoMirror
     resourceName=None
     resourceObject=/Game/Athena/HUD/Art/NetDirectionArrow
     tiling=NoTile
     imageSize=<20 20>
     margin=<ltrb 0 0 0 0>
     outlineSettings {
      bUseBrushTransparency=false
      roundingType=HalfHeightRadius
      width=0
      color {colorUseRule=UseColor_Specified, specifiedColor=<rgba 0 0 0 0>}
      cornerRadii=<0 0 0 0>
     }
     tintColor {colorUseRule=UseColor_Specified, specifiedColor=<rgba 1 1 1 1>}
     uVRegion {bIsValid=false, max=<0 0>, min=<0 0>}
    }}
   slot {class=HorizontalBoxSlot, properties {horizontalAlignment=HAlign_Center, verticalAlignment=VAlign_Top}}
  }
  VerticalBox_1 {
   widgetClassPath=&VerticalBox
   parent=HorizontalBox_414
   widget=VerticalBox_1
   slot {class=HorizontalBoxSlot, properties {padding=<ltrb 2 2 2 2>}}
  }
  DownBandwidthUI {
   widgetClassPath=&CommonTextBlock
   bInherited=true
   bIsVariable=true
   parent=VerticalBox_1
   widget=DownBandwidthUI
   properties {
    text="{DOWN}"
    font {
     bForceMonospaced=false
     bMaterialIsStencil=false
     fontMaterial=null
     fontObject=&NotoSans
     letterSpacing=0
     monospacedWidth=1
     size=8
     skewAmount=0
     typefaceFontName=NotoSans-Bold
     outlineSettings {
      bApplyOutlineToDropShadows=false
      bMiteredCorners=false
      bSeparateFillAlpha=false
      outlineMaterial=null
      outlineSize=0
      outlineColor=<rgba 0 0 0 1>
     }
    }
    shadowOffset=<0 0>
   }
   slot {class=VerticalBoxSlot}
  }
  PacketInRateUI {
   widgetClassPath=&CommonTextBlock
   bInherited=true
   parent=VerticalBox_1
   widget=PacketInRateUI
   properties {
    text="{RATE}"
    font {
     bForceMonospaced=false
     bMaterialIsStencil=false
     fontMaterial=null
     fontObject=&NotoSans
     letterSpacing=0
     monospacedWidth=1
     size=8
     skewAmount=0
     typefaceFontName=NotoSans-Bold
     outlineSettings {
      bApplyOutlineToDropShadows=false
      bMiteredCorners=false
      bSeparateFillAlpha=false
      outlineMaterial=null
      outlineSize=0
      outlineColor=<rgba 0 0 0 1>
     }
    }
    shadowOffset=<0 0>
   }
   slot {class=VerticalBoxSlot}
  }
  PacketInLossUI {
   widgetClassPath=&CommonTextBlock
   bInherited=true
   parent=VerticalBox_1
   widget=PacketInLossUI
   properties {
    text="{LOSS}"
    font {
     bForceMonospaced=false
     bMaterialIsStencil=false
     fontMaterial=null
     fontObject=&NotoSans
     letterSpacing=0
     monospacedWidth=1
     size=8
     skewAmount=0
     typefaceFontName=NotoSans-Bold
     outlineSettings {
      bApplyOutlineToDropShadows=false
      bMiteredCorners=false
      bSeparateFillAlpha=false
      outlineMaterial=null
      outlineSize=0
      outlineColor=<rgba 0 0 0 1>
     }
    }
    shadowOffset=<0 0>
   }
   slot {class=VerticalBoxSlot}
  }
  Image_1 {
   widgetClassPath=&Image
   bIsVariable=true
   parent=HorizontalBox_414
   widget=Image_1
   properties {
    brush {
     bIsDynamicallyLoaded=false
     drawAs=Image
     imageType=NoImage
     mirroring=NoMirror
     resourceName=None
     resourceObject=&NetDirectionArrow
     tiling=NoTile
     imageSize=<20 20>
     margin=<ltrb 0 0 0 0>
     outlineSettings {
      bUseBrushTransparency=false
      roundingType=HalfHeightRadius
      width=0
      color {colorUseRule=UseColor_Specified, specifiedColor=<rgba 0 0 0 0>}
      cornerRadii=<0 0 0 0>
     }
     tintColor {colorUseRule=UseColor_Specified, specifiedColor=<rgba 1 1 1 1>}
     uVRegion {bIsValid=false, max=<0 0>, min=<0 0>}
    }
    renderTransform {angle=0, scale=<1 -1>, shear=<0 0>, translation=<0 0>}
   }
   slot {class=HorizontalBoxSlot, properties {horizontalAlignment=HAlign_Center, verticalAlignment=VAlign_Top}}
  }
  VerticalBox_0 {
   widgetClassPath=&VerticalBox
   parent=HorizontalBox_414
   widget=VerticalBox_0
   slot {class=HorizontalBoxSlot, properties {padding=<ltrb 2 2 2 2>}}
  }
  UpBandwidthUI {
   widgetClassPath=&CommonTextBlock
   bInherited=true
   parent=VerticalBox_0
   widget=UpBandwidthUI
   properties {
    text="{UP}"
    font {
     bForceMonospaced=false
     bMaterialIsStencil=false
     fontMaterial=null
     fontObject=&NotoSans
     letterSpacing=0
     monospacedWidth=1
     size=8
     skewAmount=0
     typefaceFontName=NotoSans-Bold
     outlineSettings {
      bApplyOutlineToDropShadows=false
      bMiteredCorners=false
      bSeparateFillAlpha=false
      outlineMaterial=null
      outlineSize=0
      outlineColor=<rgba 0 0 0 1>
     }
    }
    shadowOffset=<0 0>
   }
   slot {class=VerticalBoxSlot}
  }
  PacketOutRateUI {
   widgetClassPath=&CommonTextBlock
   bInherited=true
   bIsVariable=true
   parent=VerticalBox_0
   widget=PacketOutRateUI
   properties {
    text="{RATE}"
    font {
     bForceMonospaced=false
     bMaterialIsStencil=false
     fontMaterial=null
     fontObject=&NotoSans
     letterSpacing=0
     monospacedWidth=1
     size=8
     skewAmount=0
     typefaceFontName=NotoSans-Bold
     outlineSettings {
      bApplyOutlineToDropShadows=false
      bMiteredCorners=false
      bSeparateFillAlpha=false
      outlineMaterial=null
      outlineSize=0
      outlineColor=<rgba 0 0 0 1>
     }
    }
    shadowOffset=<0 0>
   }
   slot {class=VerticalBoxSlot}
  }
  PacketOutLossUI {
   widgetClassPath=&CommonTextBlock
   bInherited=true
   parent=VerticalBox_0
   widget=PacketOutLossUI
   properties {
    text="{LOSS}"
    font {
     bForceMonospaced=false
     bMaterialIsStencil=false
     fontMaterial=null
     fontObject=&NotoSans
     letterSpacing=0
     monospacedWidth=1
     size=8
     skewAmount=0
     typefaceFontName=NotoSans-Bold
     outlineSettings {
      bApplyOutlineToDropShadows=false
      bMiteredCorners=false
      bSeparateFillAlpha=false
      outlineMaterial=null
      outlineSize=0
      outlineColor=<rgba 0 0 0 1>
     }
    }
    shadowOffset=<0 0>
   }
   slot {class=VerticalBoxSlot}
  }
 ]
}
