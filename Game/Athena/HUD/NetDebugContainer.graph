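# NetDebugContainer is a HUD element widget that stacks a net debug readout above a small net graph.
# Both sit in a vertical box inside a fixed-width SizeBox, wrapped in an InvalidationBox; on Construct, touch input clears the width override.
# Visibility changes start the NetDebugUI timer for visible or hit-test-invisible states and stop it when collapsed or hidden.
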WidgetBlueprint /Game/Athena/HUD/NetDebugContainer
graphs [EventGraph {nodes [
   Event Construct()
   > Branch(Condition=IsUsingInputType!(InputType:ECommonInputType Enum=Touch, PlayerController:Player Controller Object Reference=GetOwningPlayer()) -> (ReturnValue:bool))
   > ClearWidthOverride(self:Size Box Object Reference=Get SizeBox_0() -> (SizeBox_0:Size Box Object Reference))
   n1 = Event OnVisibilitySetEvent!() -> (OutputDelegate, then, InVisibility:ESlateVisibility Enum=Visible) {is=title K2Node_Event}
   > n2 = Switch on ESlateVisibility!(execute, Selection:ESlateVisibility Enum=@n1.InVisibility) -> (Visible:Exec, Collapsed:Exec, Hidden:Exec, HitTestInvisible:Exec, SelfHitTestInvisible:Exec) {is=title K2Node_SwitchEnum}
   n3 = Get NetDebugUI() -> (NetDebugUI:Net Debug UI Object Reference)
   StopTimer(execute=[@n2.Collapsed, @n2.Hidden], self:Net Debug Widget Object Reference=@n3)
   StartTimer(execute=[@n2.HitTestInvisible, @n2.SelfHitTestInvisible, @n2.Visible], self:Net Debug Widget Object Reference=@n3)
  ]}]
variables []
widgets {
 info {inheritedWidgetCount=1, parentClass=&FortHUDElementWidget, rootWidgetClass=/Script/UMG.InvalidationBox, widgetCount=6}
 widgets [
  InvalidationBox_0 {widgetClassPath=&InvalidationBox, bIsVariable=true, widget=InvalidationBox_0}
  SizeBox_0 {
   widgetClassPath=&SizeBox
   bIsVariable=true
   parent=InvalidationBox_0
   widget=SizeBox_0
   properties {bOverride_WidthOverride=true, widthOverride=180}
   slot {class=PanelSlot}
  }
  VerticalBox_0 {widgetClassPath=&VerticalBox, parent=SizeBox_0, widget=VerticalBox_0, slot {class=SizeBoxSlot}}
  NetDebugUI {
   widgetClassPath=&/NetDebugUI
   bIsVariable=true
   parent=VerticalBox_0
   widget=NetDebugUI
   properties {visibility=HitTestInvisible}
   slot {class=VerticalBoxSlot}
  }
  NetGraph {
   widgetClassPath=&/NetGraph
   bIsVariable=true
   parent=VerticalBox_0
   widget=NetGraph
   properties {backgroundColor=<rgba 0 0 0 0.1137>, desiredSize=<150 25>}
   slot {class=VerticalBoxSlot}
  }
  TutorialHighlight {widgetClassPath=&TutorialHighlight}
 ]
}
