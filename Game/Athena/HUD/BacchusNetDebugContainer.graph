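# BacchusNetDebugContainer is a BacchusHUDElement widget Blueprint that places the net debug display in the Athena HUD.
# The inherited root ScaleBox uses user-specified stretch, is hit-test invisible, and holds a NetDebugContainer child filling it in both directions.
# It has no Blueprint variables or event graph logic, so any behavior comes from the parent class and the contained widget.
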WidgetBlueprint /Game/Athena/HUD/BacchusNetDebugContainer
graphs [EventGraph {links [], nodes []}]
variables []
widgets {
 info {inheritedWidgetCount=3, parentClass=/Script/FortniteUI.BacchusHUDElement, rootWidgetClass=&ScaleBox, widgetCount=4}
 widgets [
  ScaleBox_HUDWidgetScale {
   widgetClassPath=&ScaleBox
   bInherited=true
   widget=ScaleBox_HUDWidgetScale
   properties {stretch=UserSpecified, visibility=HitTestInvisible}
  }
  NetDebugContainer {
   widgetClassPath=&/NetDebugContainer
   parent=ScaleBox_HUDWidgetScale
   widget=NetDebugContainer
   slot {class=ScaleBoxSlot, properties {horizontalAlignment=HAlign_Fill, verticalAlignment=VAlign_Fill}}
  }
  Image_Background {widgetClassPath=&Image, bInherited=true, widget=null}
  TutorialHighlight {widgetClassPath=&TutorialHighlight}
 ]
}
