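; Blueprint interface for the Creative global options part of the Athena HUD, judging by its name and path.
; It declares a single function, OnGlobalOptionsDeactivated, with no parameters or logic shown.
; The name indicates a notification for when the global options are deactivated, with any response left to implementers.
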(fn OnGlobalOptionsDeactivated)
