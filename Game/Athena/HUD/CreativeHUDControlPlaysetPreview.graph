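# Creative-mode HUD widget that shows keybind hints for playset preview: Paste, Push, Pull and Exit.
# It extends AthenaHUDSituationalFeedback, pairing action names from FortGlobalUIContext with labels and setting MaxLinesNeeded before parent Construct.
# ClearAndSetFeedbackLines clears the Feedback box, then re-adds inherited FeedbackLines widgets, indexed by child count, with each bound action and label.
# OnKeybindsChanged refreshes the names and rebuilds the lines.
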WidgetBlueprint /Game/Athena/HUD/CreativeHUDControlPlaysetPreview
graphs [
 EventGraph {nodes [
   ClearAndSetFeedbackLines() {is=title K2Node_CustomEvent}
   > ClearChildren(self=Get Feedback() -> (Feedback:Vertical Box Object Reference))
   > n1 = For Each Loop(Array:Array of Names=Get PlaysetPreviewTooActionNames() -> (PlaysetPreviewTooActionNames:Array of Names)) -> (Array Element:Name)
   n2 = GetChildrenCount!(self:Panel Widget Object Reference=Get Feedback()) -> (ReturnValue:int=0)
   > Branch(Condition="integer < integer"!(A:int=@n2, B:int=Get MaxLinesNeeded() -> (MaxLinesNeeded:int=0)) -> (ReturnValue:bool) {is=title K2Node_CallFunction})
   n3 = "Get (a copy)"(
    Array:Array of Athena HUDFeedback Line Object References=Get FeedbackLines() -> (FeedbackLines:Array of Athena HUDFeedback Line Object References)
    Dimension 1=@n2
   ) -> (Output:Athena HUDFeedback Line Object Reference)
   > SetVisibility(self=@n3, InVisibility=SelfHitTestInvisible)
   > SetBoundAction!(
    execute
    self:Fort Keybind Widget Object Reference=Get FeedbackKeyBind!(self:Athena HUDFeedback Line Object Reference=@n3) -> (FeedbackKeyBind:Keybind Widget Object Reference)
    NewBoundAction:Name=@n1
   ) -> (then)
   > SetText(
    self=Get FeedbackText!(self:Athena HUDFeedback Line Object Reference=@n3) -> (FeedbackText:Rich Text Block Object Reference)
    InText="Get (a ref)"(Array:Array of Texts=Get PlaysetPreviewToolActionLabels() -> (PlaysetPreviewToolActionLabels:Array of Texts), Dimension 1=@n1.Array Index) -> (Output:"Text (by ref)")
   )
   > AddChild(self=Get Feedback(), Content=@n3)
   OnKeybindsChanged!() -> (OutputDelegate, then, IsFlying:bool=~) {is=title K2Node_CustomEvent}
   > SetupActionNamesAndLabels()
   > ClearAndSetFeedbackLines() {is=title K2Node_CallFunction}
   Event Construct()
   > SetupActionNamesAndLabels()
   > Set MaxLinesNeeded(MaxLinesNeeded:int=Length(TargetArray:Array of Names=Get PlaysetPreviewTooActionNames())) -> (Output_Get:int=0)
   > Parent: Construct()
   > ClearAndSetFeedbackLines()
  ]}
 SetupActionNamesAndLabels {nodes [
   SetupActionNamesAndLabels() {is=title K2Node_FunctionEntry}
   n1 = Get FortGlobalUIContext!() -> (ReturnValue:Fort Global UIContext Object Reference) {is=title K2Node_GetBlueprintContextEx}
   > Set PlaysetPreviewTooActionNames(PlaysetPreviewTooActionNames:Array of Names=Make Array!(
     "[0]":Name=GetCreativeMoveBuildingsOnGridDuplicateName!(self:Fort Global UIContext Object Reference=@n1) -> (ReturnValue:Name=None)
     "[1]":Name=GetCreativeMoveObjectsFreelyPushName!(self:Fort Global UIContext Object Reference=@n1) -> (ReturnValue:Name=None)
     "[2]":Name=GetCreativeMoveObjectsFreelyPullName!(self:Fort Global UIContext Object Reference=@n1) -> (ReturnValue:Name=None)
     "[3]":Name=GetCreativeMoveBuildingsOnGridExitName!(self:Fort Global UIContext Object Reference=@n1) -> (ReturnValue:Name=None)
    ) -> (Array:Array of Names)) -> (Output_Get:Array of Names)
   > Set PlaysetPreviewToolActionLabels(PlaysetPreviewToolActionLabels:Array of Texts=Make Array(
     "[0]":Text=Get PasteLabelText() -> (PasteLabelText:Text)
     "[1]":Text=Get PushLabelText() -> (PushLabelText:Text)
     "[2]":Text=Get PullLabelText() -> (PullLabelText:Text)
     "[3]":Text=Get ExitLabelText() -> (ExitLabelText:Text)
    ) -> (Array:Array of Texts)) -> (Output_Get:Array of Texts)
  ]}
]
variables [
 ExitLabelText {type=string, default=Exit}
 PasteLabelText {type=string, default=Paste}
 PlaysetPreviewTooActionNames {type="array<string>", default []}
 PlaysetPreviewToolActionLabels {type="array<string>", default []}
 PullLabelText {type=string, default=Pull}
 PushLabelText {type=string, default=Push}
]
widgets {
 info {
  inheritedWidgetCount=0
  parentClass=&/AthenaHUDSituationalFeedback.AthenaHUDSituationalFeedback_C
  rootWidgetClass=null
  widgetCount=0
 }
 widgets []
}
