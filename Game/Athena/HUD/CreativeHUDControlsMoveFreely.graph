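# Creative-mode HUD widget, derived from AthenaHUDSituationalFeedback, that shows keybind hints for moving objects freely with the move tool.
# SetupActionNamesAndLabels builds paired arrays of UI-context action names and labels plus fixed line indices, which ClearAndSetFeedbackLines binds to feedback rows.
# Update refreshes state-dependent rows through text macros for scale/rotation, drop, grid snap and collision.
# It also sets visibility from SelectionProperty, DropsDisabled and ScaleDisabled.
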WidgetBlueprint /Game/Athena/HUD/CreativeHUDControlsMoveFreely
graphs [
 EventGraph {nodes [
   Update!() -> (OutputDelegate, then, IsFlying:bool=~) {is=title K2Node_CustomEvent}
   > n1 = Cast To FortPlayerControllerAthena(Object=GetOwningPlayer())
   > n2 = Sequence!(execute) -> (then_0, then_1, then_2:Exec)
   n3 = "Get (a copy)"(
    Array:Array of Athena HUDFeedback Line Object References=Get FeedbackLines() -> (FeedbackLines:Array of Athena HUDFeedback Line Object References)
    Dimension 1=Get MakeScaleOrRotationAxisTextIndex() -> (MakeScaleOrRotationAxisTextIndex:int=0)
   ) -> (Output:Athena HUDFeedback Line Object Reference)
   > SetText(
    self=Get FeedbackText!(self:Athena HUDFeedback Line Object Reference=@n3) -> (FeedbackText:Rich Text Block Object Reference)
    InText=Make Scale or Rotation Axis Text!() -> (ReturnValue:Text) {is=title K2Node_MacroInstance}
   )
   n4 = "Get (a copy)"(Array=Get FeedbackLines(), Dimension 1=Get MakeDropToFloorTextIndex() -> (MakeDropToFloorTextIndex:int=0))
   > SetText(self=Get FeedbackText(self=@n4), InText=Make Drop to Floor Text!() -> (ReturnValue:Text) {is=title K2Node_MacroInstance})
   n5 = "Get (a copy)"(Array=Get FeedbackLines(), Dimension 1=Get MakeGridSnapIndexTextIndex() -> (MakeGridSnapIndexTextIndex:int=0))
   > SetText(self=Get FeedbackText(self=@n5), InText=Make Grid Snap Index Text!() -> (ReturnValue:Text) {is=title K2Node_MacroInstance})
   n6 = "Get (a copy)"(Array=Get FeedbackLines(), Dimension 1=Get PhasingTextIndex() -> (PhasingTextIndex:int=0))
   > SetText(self=Get FeedbackText(self=@n6), InText=Make Phased Text!() -> (ReturnValue:Text) {is=title K2Node_MacroInstance})
   n7 = Get RotateClockwiseOrScaleIndex() -> (RotateClockwiseOrScaleIndex:int=0)
   n8 = Get FeedbackLines()
   > SetText(
    self=Get FeedbackText(self="Get (a copy)"(Array=@n8, Dimension 1=@n7))
    InText=Make Scale Up or Rotate Clockwise Text!() -> (ReturnValue:Text) {is=title K2Node_MacroInstance}
   )
   n9 = Get FeedbackLines()
   n10 = Get RotateCounterClockwiseOrScaleIndex() -> (RotateCounterClockwiseOrScaleIndex:int=0)
   > SetText(
    self=Get FeedbackText(self="Get (a copy)"(Array=@n9, Dimension 1=@n10))
    InText=Make Scale Down or Rotate Counter Clockwise Text!() -> (ReturnValue:Text) {is=title K2Node_MacroInstance}
   )
   > n11 = Sequence!(execute) -> (then_0, then_1)
   n12 = "Get (a copy)"(Array=Get FeedbackLines(), Dimension 1=Get ScaleActiveIndex() -> (ScaleActiveIndex:int=0))
   > SetText(self=Get FeedbackText(self=@n12), InText=Make Scale Active Text!() -> (ReturnValue:Text) {is=title K2Node_MacroInstance})
   n13 = "Get (a copy)"(Array=Get FeedbackLines(), Dimension 1=Get RotateResetOrScaleIndex() -> (RotateResetOrScaleIndex:int=0))
   > SetText(self=Get FeedbackText(self=@n13), InText=Make Scale or Rotate Reset Text!() -> (ReturnValue:Text) {is=title K2Node_MacroInstance})
   SetVisibility(
    execute=@n11.then_1
    self=@n12
    InVisibility=Select!(
     SingleObject:ESlateVisibility Enum=Visible
     SingleObjectMoveOnGrid:ESlateVisibility Enum=Collapsed
     MultipleObjects:ESlateVisibility Enum=Collapsed
     MultipleObjectsMoveOnGrid:ESlateVisibility Enum=Collapsed
     None:ESlateVisibility Enum=Visible
     Index:ESelectionProperty Enum=Get SelectionProperty() -> (SelectionProperty:ESelectionProperty Enum=SingleObject)
    ) -> (ReturnValue)
   )
   > SetVisibility(
    self=@n13
    InVisibility=Select(
     SingleObject=Visible
     SingleObjectMoveOnGrid=Collapsed
     MultipleObjects=Visible
     MultipleObjectsMoveOnGrid=Collapsed
     None=Visible
     Index=Get SelectionProperty() -> (SelectionProperty=SingleObject)
    )
   )
   SetVisibility(
    execute=@n2.then_1
    self=@n3
    InVisibility=Select(
     SingleObject=Visible
     SingleObjectMoveOnGrid=Collapsed
     MultipleObjects=Visible
     MultipleObjectsMoveOnGrid=Collapsed
     None=Visible
     Index=Get SelectionProperty() -> (SelectionProperty=SingleObject)
    )
   )
   > n14 = Branch(Condition=Get DropsDisabled() -> (DropsDisabled:bool))
   > n15 = SetVisibility(self=@n4, InVisibility=Collapsed)
   n16 = Select(
    SingleObject=SelfHitTestInvisible
    SingleObjectMoveOnGrid=Collapsed
    MultipleObjects=SelfHitTestInvisible
    MultipleObjectsMoveOnGrid=Collapsed
    None=SelfHitTestInvisible
    Index=Get SelectionProperty() -> (SelectionProperty=SingleObject)
   )
   n17 = MakeLiteralBool(Value=false)
   SetVisibility(
    execute=[@n19.then, @n15.then]
    self=@n5
    InVisibility=Select!(Option 0:ESlateVisibility Enum=@n16, Option 1:ESlateVisibility Enum=@n16, Index:bool=@n17) -> (ReturnValue)
   )
   > SetVisibility(self=@n6, InVisibility=Select(Option 0=SelfHitTestInvisible, Option 1=Collapsed, Index=@n17))
   n18 = Get SelectionProperty() -> (SelectionProperty=SingleObject)
   n19 = SetVisibility(
    execute=@n14.else
    self=@n4
    InVisibility=Select(
     Option 0=Select!(
      SingleObject:ESlateVisibility Enum=SelfHitTestInvisible
      SingleObjectMoveOnGrid:ESlateVisibility Enum=Collapsed
      MultipleObjects:ESlateVisibility Enum=SelfHitTestInvisible
      MultipleObjectsMoveOnGrid:ESlateVisibility Enum=Collapsed
      None:ESlateVisibility Enum=SelfHitTestInvisible
      Index:ESelectionProperty Enum=@n18
     ) -> (ReturnValue)
     Option 1=Collapsed
     Index=@n17
    )
   )
   Branch(execute=@n2.then_2, Condition=Get ScaleDisabled() -> (ScaleDisabled:bool))
   > SetVisibility(self="Get (a copy)"(Array=Get FeedbackLines(), Dimension 1=Get ScaleActiveIndex() -> (ScaleActiveIndex=0)), InVisibility=Collapsed)
   ClearAndSetFeedbackLines() {is=title K2Node_CustomEvent}
   > ClearChildren(self=Get Feedback() -> (Feedback:Vertical Box Object Reference))
   > n20 = For Each Loop(Array:Array of Names=Get MoveFreelyMoveToolHoldingItemActionNames() -> (MoveFreelyMoveToolHoldingItemActionNames:Array of Names)) -> (Array Element:Name)
   n21 = GetChildrenCount!(self:Panel Widget Object Reference=Get Feedback()) -> (ReturnValue:int=0)
   > Branch(Condition="integer < integer"!(A:int=@n21, B:int=Get MaxLinesNeeded() -> (MaxLinesNeeded:int=0)) -> (ReturnValue:bool) {is=title K2Node_CallFunction})
   n22 = "Get (a copy)"(Array=Get FeedbackLines(), Dimension 1=@n21)
   > SetVisibility(self=@n22, InVisibility=SelfHitTestInvisible)
   > SetBoundAction!(
    execute
    self:Fort Keybind Widget Object Reference=Get FeedbackKeyBind!(self:Athena HUDFeedback Line Object Reference=@n22) -> (FeedbackKeyBind:Keybind Widget Object Reference)
    NewBoundAction:Name=@n20
   ) -> (then)
   > SetText(
    self=Get FeedbackText(self=@n22)
    InText="Get (a ref)"(
     Array:Array of Texts=Get MoveFreelyMoveToolHoldingItemsActionLabels() -> (MoveFreelyMoveToolHoldingItemsActionLabels:Array of Texts)
     Dimension 1=@n20.Array Index
    ) -> (Output:"Text (by ref)")
   )
   > AddChild(self=Get Feedback(), Content=@n22)
   n23 = OnKeybindsChanged!() -> (OutputDelegate, then, IsFlying:bool=~) {is=title K2Node_CustomEvent}
   > SetupActionNamesAndLabels()
   > ClearAndSetFeedbackLines() {is=title K2Node_CallFunction}
   > Update!(execute, self, IsFlying:bool=@n23.IsFlying) -> (then) {is=title K2Node_CallFunction}
   Event Construct()
   > SetupActionNamesAndLabels()
   > Set MaxLinesNeeded(MaxLinesNeeded:int=Length(TargetArray:Array of Names=Get MoveFreelyMoveToolHoldingItemActionNames())) -> (Output_Get:int=0)
   > Parent: Construct()
   > ClearAndSetFeedbackLines()
   IsCreativeQuickmenuEnabled!(self:Fort Player Controller Gameplay Object Reference=@n1) -> (ReturnValue:bool)
   Get FeedbackLines()
  ]}
 MakeDropToFloorText {nodes [
   Inputs!()
   n1 = "To String (Text)"(InText=Get DropToFloorLabelText() -> (DropToFloorLabelText:Text))
   n2 = Get OnLabelText() -> (OnLabelText:Text)
   n3 = Get OffLabelText() -> (OffLabelText:Text)
   n4 = Select!(
    Option 0:String=Append(A=MakeLiteralString(Value=" "), B="To String (Text)"(InText=@n3))
    Option 1:String=Append(A=MakeLiteralString(Value=" "), B="To String (Text)"(InText=@n2))
    Index:bool=Get DropToFloorOn() -> (DropToFloorOn:bool)
   ) -> (ReturnValue:String)
   Outputs!(ReturnValue:Text="To Text (String)"(InString=Append(A=@n1, B=@n4)))
  ]}
 MakeGridSnapIndexText {nodes [
   Inputs()
   n1 = "To String (Text)"(InText=Get GridPrecisionLabelText() -> (GridPrecisionLabelText:Text))
   n2 = Get OffLabelText()
   n3 = "To String (Text)"(InText=Get GridSnap64LabelText() -> (GridSnap64LabelText:Text))
   n4 = MakeLiteralString(Value=" ")
   n5 = "To String (Text)"(InText=Get GridSnap8LabelText() -> (GridSnap8LabelText:Text))
   n6 = MakeLiteralString(Value=" ")
   n7 = "To String (Text)"(InText=Get GridSnap4LabelText() -> (GridSnap4LabelText:Text))
   n8 = MakeLiteralString(Value=" ")
   n9 = "To String (Text)"(InText=Get GridSnap1LabelText() -> (GridSnap1LabelText:Text))
   n10 = MakeLiteralString(Value=" ")
   n11 = "To String (Text)"(InText=Get GridSnap2LabelText() -> (GridSnap2LabelText:Text))
   n12 = MakeLiteralString(Value=" ")
   n13 = "To String (Text)"(InText=Get GridSnap16LabelText() -> (GridSnap16LabelText:Text))
   n14 = MakeLiteralString(Value=" ")
   n15 = "To String (Text)"(InText=Get GridSnap32LabelText() -> (GridSnap32LabelText:Text))
   n16 = MakeLiteralString(Value=" ")
   n17 = Select(
    Option 0=Append(A=MakeLiteralString(Value=" "), B="To String (Text)"(InText=@n2))
    Option 1=Select!(
     Option 0=Append(A=@n16, B=@n15)
     Option 1=Append(A=@n14, B=@n13)
     Option 2:String=Append(A=@n6, B=@n5)
     Option 3:String=Append(A=@n8, B=@n7)
     Option 4:String=Append(A=@n12, B=@n11)
     Option 5:String=Append(A=@n10, B=@n9)
     Option 6:String=Append(A=@n4, B=@n3)
     Index:Byte=Get GridSnapIndex() -> (GridSnapIndex:Byte=0)
    ) -> (ReturnValue)
    Index=Get PrecisionOn() -> (PrecisionOn:bool)
   )
   Outputs(ReturnValue="To Text (String)"(InString=Append(A=@n1, B=@n17)))
  ]}
 MakeLeftAction {nodes [
   Inputs()
   n1 = Append(A=MakeLiteralString(Value=" "), B="To String (Text)"(InText=Get ScaleDownLabelText() -> (ScaleDownLabelText:Text)))
   n2 = Append(A=MakeLiteralString(Value=" "), B="To String (Text)"(InText=Get PullLabelText() -> (PullLabelText:Text)))
   n3 = Append(A=MakeLiteralString(Value=" "), B="To String (Text)"(InText=Get RotateCounterClockwiseLabelText() -> (RotateCounterClockwiseLabelText:Text)))
   n4 = Get TransformationType() -> (TransformationType:ETransformationType Enum=Translation)
   Outputs(ReturnValue="To Text (String)"(InString=Select!(Translation:String=@n2, Rotation:String=@n3, Scale:String=@n1, None:String, Index:ETransformationType Enum=@n4) -> (ReturnValue)))
  ]}
 MakePhasedText {nodes [
   Inputs()
   n1 = "To String (Text)"(InText=Get PhasingLabelText() -> (PhasingLabelText:Text))
   n2 = Get NoBlockersText() -> (NoBlockersText:Text)
   n3 = Get VisibilityText() -> (VisibilityText:Text)
   n4 = Get TerrainText() -> (TerrainText:Text)
   n5 = Select!(
    Option 0:String=Append(A=MakeLiteralString(), B="To String (Text)"(InText=@n3))
    Option 1:String=Append(A=MakeLiteralString(), B="To String (Text)"(InText=@n4))
    Option 2:String=Append(A=MakeLiteralString(), B="To String (Text)"(InText=@n2))
    Index:Byte=Get Phasing() -> (Phasing:EHitTraceRule Enum=Visibility)
   ) -> (ReturnValue)
   Outputs(ReturnValue="To Text (String)"(InString=Append!(A=@n1, B=@n5, C:String="]") -> (ReturnValue)))
  ]}
 MakeRightAction {nodes [
   Inputs()
   n1 = Append!(A=MakeLiteralString(Value=" "), B="To String (Text)"(InText=Get ScaleUpLabelText() -> (ScaleUpLabelText:Text))) -> (ReturnValue)
   n2 = Append(A=MakeLiteralString(Value=" "), B="To String (Text)"(InText=Get PushLabelText() -> (PushLabelText:Text)))
   n3 = Append(A=MakeLiteralString(Value=" "), B="To String (Text)"(InText=Get RotateClockwiseLabelText() -> (RotateClockwiseLabelText:Text)))
   n4 = Get TransformationType() -> (TransformationType=Translation)
   Outputs(ReturnValue="To Text (String)"(InString=Select!(Translation:String=@n2, Rotation:String=@n3, Scale:String=@n1, None:String, Index:ETransformationType Enum=@n4) -> (ReturnValue)))
  ]}
 MakeRotationAxisIndexText {nodes [
   Inputs()
   n1 = "To String (Text)"(InText=Get RotationAxisLabelText() -> (RotationAxisLabelText:Text))
   n2 = Get RollLabelText() -> (RollLabelText:Text)
   n3 = Get YawLabelText() -> (YawLabelText:Text)
   n4 = Get PitchLabelText() -> (PitchLabelText:Text)
   n5 = Select!(
    Option 0:String=Append(A=MakeLiteralString(Value=" "), B="To String (Text)"(InText=@n3))
    Option 1:String=Append(A=MakeLiteralString(Value=" "), B="To String (Text)"(InText=@n4))
    Option 2:String=Append(A=MakeLiteralString(Value=" "), B="To String (Text)"(InText=@n2))
    Index:Byte=Get RotationAxisIndex() -> (RotationAxisIndex:Byte=0)
   ) -> (ReturnValue)
   Outputs(ReturnValue="To Text (String)"(InString=Append(A=@n1, B=@n5)))
  ]}
 MakeScaleActiveText {nodes [
   Inputs()
   n1 = "To String (Text)"(InText=Get ScaleText() -> (ScaleText:Text))
   n2 = Get OnLabelText()
   n3 = Get OffLabelText()
   n4 = Select!(
    Option 0=Append(A=MakeLiteralString(Value=" "), B="To String (Text)"(InText=@n3))
    Option 1=Append(A=MakeLiteralString(Value=" "), B="To String (Text)"(InText=@n2))
    Index:bool=Get ScalingOn() -> (ScalingOn:bool)
   ) -> (ReturnValue)
   Outputs(ReturnValue="To Text (String)"(InString=Append(A=@n1, B=@n4)))
  ]}
 MakeScaleAxisSelectedText {nodes [
   Inputs()
   n1 = "To String (Text)"(InText=Get ScaleAxisLabel() -> (ScaleAxisLabel:Text))
   n2 = Get ZLabelText() -> (ZLabelText:Text)
   n3 = Get AllLabelText() -> (AllLabelText:Text)
   n4 = Get XLabelText() -> (XLabelText:Text)
   n5 = Get YLabelText() -> (YLabelText:Text)
   n6 = Select!(
    All:String=Append(A=MakeLiteralString(Value=" "), B="To String (Text)"(InText=@n3))
    X:String=Append(A=MakeLiteralString(Value=" "), B="To String (Text)"(InText=@n4))
    Y:String=Append(A=MakeLiteralString(Value=" "), B="To String (Text)"(InText=@n5))
    Z:String=Append(A=MakeLiteralString(Value=" "), B="To String (Text)"(InText=@n2))
    Index:EScaleAxis Enum=Get SelectedScaleAxis() -> (SelectedScaleAxis:EScaleAxis Enum=All)
   ) -> (ReturnValue)
   Outputs(ReturnValue="To Text (String)"(InString=Append(A=@n1, B=@n6)))
  ]}
 MakeScaleDownOrRotateCounterClockwiseText {nodes [
   Inputs()
   n1 = Append(A=MakeLiteralString(Value=" "), B="To String (Text)"(InText=Get RotateCounterClockwiseLabelText()))
   n2 = Append(A=MakeLiteralString(Value=" "), B="To String (Text)"(InText=Get ScaleDownLabelText()))
   n3 = Get ScalingOn()
   Outputs(ReturnValue="To Text (String)"(InString=Select!(Option 0:String=@n1, Option 1:String=@n2, Index:bool=@n3) -> (ReturnValue)))
  ]}
 MakeScaleOrRotateResetText {nodes [
   Inputs()
   n1 = Append(A=MakeLiteralString(Value=" "), B="To String (Text)"(InText=Get RotateResetLabelText() -> (RotateResetLabelText:Text)))
   n2 = Append(A=MakeLiteralString(Value=" "), B="To String (Text)"(InText=Get ScaleResetLabelText() -> (ScaleResetLabelText:Text)))
   n3 = Get ScalingOn()
   Outputs(ReturnValue="To Text (String)"(InString=Select(Option 0=@n1, Option 1=@n2, Index=@n3)))
  ]}
 MakeScaleOrRotationAxisText {nodes [
   Inputs()
   n1 = Append(
    A=MakeLiteralString(Value=" ")
    B="To String (Text)"(InText=Make Rotation Axis Index Text!() -> (ReturnValue:Text) {is=title K2Node_MacroInstance})
   )
   n2 = Append(
    A=MakeLiteralString(Value=" ")
    B="To String (Text)"(InText=Make Scale Axis Selected Text!() -> (ReturnValue:Text) {is=title K2Node_MacroInstance})
   )
   n3 = Get ScalingOn()
   Outputs(ReturnValue="To Text (String)"(InString=Select(Option 0=@n1, Option 1=@n2, Index=@n3)))
  ]}
 MakeScaleUpOrRotateClockwiseText {nodes [
   Inputs()
   n1 = Append(A=MakeLiteralString(Value=" "), B="To String (Text)"(InText=Get RotateClockwiseLabelText()))
   n2 = Append(A=MakeLiteralString(Value=" "), B="To String (Text)"(InText=Get ScaleUpLabelText()))
   n3 = Get ScalingOn()
   Outputs(ReturnValue="To Text (String)"(InString=Select(Option 0=@n1, Option 1=@n2, Index=@n3)))
  ]}
 SetupActionNamesAndLabels {nodes [
   SetupActionNamesAndLabels() {is=title K2Node_FunctionEntry}
   n1 = Get FortGlobalUIContext!() -> (ReturnValue:Fort Global UIContext Object Reference) {is=title K2Node_GetBlueprintContextEx}
   n2 = GetCreativeMoveObjectsFreelyExitName!(self:Fort Global UIContext Object Reference=@n1) -> (ReturnValue:Name=None)
   n3 = GetCreativeMoveObjectsFreelySwitchAxisName!(self:Fort Global UIContext Object Reference=@n1) -> (ReturnValue:Name=None)
   n4 = GetCreativeMoveObjectsFreelyChangePrecisionLevelName!(self:Fort Global UIContext Object Reference=@n1) -> (ReturnValue:Name=None)
   > Set MoveFreelyMoveToolHoldingItemActionNames(MoveFreelyMoveToolHoldingItemActionNames:Array of Names=Make Array!(
     "[0]":Name=GetCreativeMoveObjectsFreelyDuplicateName!(self:Fort Global UIContext Object Reference=@n1) -> (ReturnValue:Name=None)
     "[1]":Name=GetCreativeMoveObjectsFreelyPushName!(self:Fort Global UIContext Object Reference=@n1) -> (ReturnValue:Name=None)
     "[2]":Name=GetCreativeMoveObjectsFreelyPullName!(self:Fort Global UIContext Object Reference=@n1) -> (ReturnValue:Name=None)
     "[3]":Name=GetCreativeMoveObjectsFreelyRotateClockwiseName!(self:Fort Global UIContext Object Reference=@n1) -> (ReturnValue:Name=None)
     "[4]":Name=GetCreativeMoveObjectsFreelyRotateCounterclockwiseName!(self:Fort Global UIContext Object Reference=@n1) -> (ReturnValue:Name=None)
     "[5]":Name=@n3
     "[6]":Name=@n3
     "[7]":Name=GetCreativeMoveObjectsFreelyDropToFloorName!(self:Fort Global UIContext Object Reference=@n1) -> (ReturnValue:Name=None)
     "[8]":Name=@n4
     "[9]":Name=@n4
     "[10]":Name=@n2
     "[11]":Name=@n2
    ) -> (Array:Array of Names)) -> (Output_Get:Array of Names)
   > Set MoveFreelyMoveToolHoldingItemsActionLabels(MoveFreelyMoveToolHoldingItemsActionLabels:Array of Texts=Make Array(
     "[0]":Text=Get PasteLabelText() -> (PasteLabelText:Text)
     "[1]":Text=Get PushLabelText()
     "[2]":Text=Get PullLabelText()
     "[3]":Text=Make Scale Up or Rotate Clockwise Text()
     "[4]":Text=Make Scale Down or Rotate Counter Clockwise Text()
     "[5]":Text=Make Scale or Rotation Axis Text()
     "[6]":Text=Make Scale or Rotate Reset Text()
     "[7]":Text=Make Drop to Floor Text()
     "[8]":Text=Make Grid Snap Index Text()
     "[9]":Text=Make Phased Text()
     "[10]":Text=Get ExitLabelText() -> (ExitLabelText:Text)
     "[11]":Text=Make Scale Active Text()
    ) -> (Array:Array of Texts)) -> (Output_Get:Array of Texts)
   > Set MakeScaleOrRotationAxisTextIndex(MakeScaleOrRotationAxisTextIndex:int=5) -> (Output_Get:int=0)
   > Set MakeDropToFloorTextIndex(MakeDropToFloorTextIndex:int=7) -> (Output_Get:int=0)
   > Set MakeGridSnapIndexTextIndex(MakeGridSnapIndexTextIndex:int=8) -> (Output_Get:int=0)
   > Set PhasingTextIndex(PhasingTextIndex:int=9) -> (Output_Get:int=0)
   > Set RotateClockwiseOrScaleIndex(RotateClockwiseOrScaleIndex:int=3) -> (Output_Get:int=0)
   > Set RotateCounterClockwiseOrScaleIndex(RotateCounterClockwiseOrScaleIndex:int=4) -> (Output_Get:int=0)
   > Set ScaleActiveIndex(ScaleActiveIndex:int=11) -> (Output_Get:int=0)
   > Set RotateResetOrScaleIndex(RotateResetOrScaleIndex:int=6) -> (Output_Get:int=0)
  ]}
]
variables [
 AllLabelText {type=string, default="[All]"}
 DropsDisabled {type=boolean}
 DropToFloorLabelText {type=string, default=Drops}
 DropToFloorOn {type=boolean, default=true}
 ExitLabelText {type=string, default=Exit}
 GridPrecisionLabelText {type=string, default=Grid Snap}
 GridSnap16LabelText {type=string, default="[16]"}
 GridSnap1LabelText {type=string, default="[1]"}
 GridSnap2LabelText {type=string, default="[2]"}
 GridSnap32LabelText {type=string, default="[32]"}
 GridSnap4LabelText {type=string, default="[4]"}
 GridSnap64LabelText {type=string, default="[64]"}
 GridSnap8LabelText {type=string, default="[8]"}
 GridSnapIndex {type=integer}
 MakeDropToFloorTextIndex {type=integer}
 MakeGridSnapIndexTextIndex {type=integer}
 MakeScaleOrRotationAxisTextIndex {type=integer}
 MoveFreelyMoveToolHoldingItemActionNames {type="array<string>", default []}
 MoveFreelyMoveToolHoldingItemsActionLabels {type="array<string>", default []}
 NoBlockersText {type=string, default=Nothing}
 OffLabelText {type=string, default="[Off]"}
 OnLabelText {type=string, default="[On]"}
 PasteLabelText {type=string, default=Paste}
 Phasing {type=EHitTraceRule, default=Terrain}
 PhasingLabelText {type=string, default="(Hold) Collision ["}
 PhasingTextIndex {type=integer}
 PitchLabelText {type=string, default="[Pitch]"}
 PrecisionOn {type=boolean}
 PullLabelText {type=string, default=Pull}
 PushLabelText {type=string, default=Push}
 RollLabelText {type=string, default="[Roll]"}
 RotateClockwiseLabelText {type=string, default=Rotate Clockwise}
 RotateClockwiseOrScaleIndex {type=integer}
 RotateCounterClockwiseLabelText {type=string, default=Rotate Counterclockwise}
 RotateCounterClockwiseOrScaleIndex {type=integer}
 RotateResetLabelText {type=string, default="(Hold) Rotation Reset"}
 RotateResetOrScaleIndex {type=integer}
 RotationAxisIndex {type=integer}
 RotationAxisLabelText {type=string, default=Rotation Axis}
 ScaleActiveIndex {type=integer}
 ScaleAxisLabel {type=string, default=Resize Axis}
 ScaleDisabled {type=boolean}
 ScaleDownLabelText {type=string, default=Shrink}
 ScaleResetLabelText {type=string, default="(Hold) Size Reset"}
 ScaleText {type=string, default="(Hold) Resize"}
 ScaleUpLabelText {type=string, default=Grow}
 ScalingOn {type=boolean}
 SelectedScaleAxis {type=EScaleAxis, default=All}
 SelectionProperty {type=ESelectionProperty, default=SingleObject}
 TerrainText {type=string, default=Terrain Only}
 TransformationType {type=ETransformationType, default=Translation}
 VisibilityText {type=string, default=Everything}
 XLabelText {type=string, default="[Width]"}
 YawLabelText {type=string, default="[Yaw]"}
 YLabelText {type=string, default="[Depth]"}
 ZLabelText {type=string, default="[Height]"}
]
widgets {
 info {
  inheritedWidgetCount=0
  parentClass=&/AthenaHUDSituationalFeedback.AthenaHUDSituationalFeedback_C
  rootWidgetClass=null
  widgetCount=0
 }
 widgets []
}
