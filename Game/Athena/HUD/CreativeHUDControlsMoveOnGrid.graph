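# Creative-mode HUD widget, derived from AthenaHUDSituationalFeedback, that shows keybind hints for the move-on-grid tool while holding an item.
# SetupActionNamesAndLabels pairs UI-context action names with Paste, rotate, Mirror and Exit labels in parallel arrays.
# On construct and keybind change, ClearAndSetFeedbackLines refills the Feedback box from existing FeedbackLines, binding each key and label.
# It collapses the Mirror line, found by fixed index, when MirrorDisabled is set.
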WidgetBlueprint /Game/Athena/HUD/CreativeHUDControlsMoveOnGrid
graphs [
 EventGraph {nodes [
   ClearAndSetFeedbackLines() {is=title K2Node_CustomEvent}
   > ClearChildren(self=Get Feedback() -> (Feedback:Vertical Box Object Reference))
   > n1 = For Each Loop(Array:Array of Names=Get MoveOnGridMoveToolHoldingItemActionNames() -> (MoveOnGridMoveToolHoldingItemActionNames:Array of Names)) -> (Array Element:Name)
   n2 = GetChildrenCount!(self:Panel Widget Object Reference=Get Feedback()) -> (ReturnValue:int=0)
   > Branch(Condition="integer < integer"!(A:int=@n2, B:int=Get MaxLinesNeeded() -> (MaxLinesNeeded:int=0)) -> (ReturnValue:bool) {is=title K2Node_CallFunction})
   n3 = "Get (a copy)"(
    Array:Array of Athena HUDFeedback Line Object References=Get FeedbackLines() -> (FeedbackLines:Array of Athena HUDFeedback Line Object References)
    Dimension 1=@n2
   ) -> (Output:Athena HUDFeedback Line Object Reference)
   > SetVisibility(self=@n3, InVisibility=SelfHitTestInvisible)
   > SetBoundAction!(
    execute
    self:Fort Keybind Widget Object Reference=Get FeedbackKeyBind!(self:Athena HUDFeedback Line Object Reference=@n3) -> (FeedbackKeyBind:Keybind Widget Object Reference)
    NewBoundAction:Name=@n1
   ) -> (then)
   > SetText(
    self=Get FeedbackText!(self:Athena HUDFeedback Line Object Reference=@n3) -> (FeedbackText:Rich Text Block Object Reference)
    InText="Get (a ref)"(
     Array:Array of Texts=Get MoveOnGridMoveToolHoldingItemsActionLabels() -> (MoveOnGridMoveToolHoldingItemsActionLabels:Array of Texts)
     Dimension 1=@n1.Array Index
    ) -> (Output:"Text (by ref)")
   )
   > AddChild(self=Get Feedback(), Content=@n3)
   Branch(execute=@n1.Completed, Condition=Get MirrorDisabled() -> (MirrorDisabled:bool))
   > SetVisibility(self="Get (a copy)"(Array=Get FeedbackLines(), Dimension 1=Get MirrorIndex() -> (MirrorIndex:int=0)), InVisibility=Collapsed)
   OnKeybindsChanged!() -> (OutputDelegate, then, IsFlying:bool=~) {is=title K2Node_CustomEvent}
   > SetupActionNamesAndLabels()
   > ClearAndSetFeedbackLines() {is=title K2Node_CallFunction}
   Event Construct()
   > SetupActionNamesAndLabels()
   > Set MaxLinesNeeded(MaxLinesNeeded:int=Length(TargetArray:Array of Names=Get MoveOnGridMoveToolHoldingItemActionNames())) -> (Output_Get:int=0)
   > Parent: Construct()
   > ClearAndSetFeedbackLines()
  ]}
 SetupActionNamesAndLabels {nodes [
   SetupActionNamesAndLabels() {is=title K2Node_FunctionEntry}
   n1 = Get FortGlobalUIContext!() -> (ReturnValue:Fort Global UIContext Object Reference) {is=title K2Node_GetBlueprintContextEx}
   > Set MoveOnGridMoveToolHoldingItemActionNames(MoveOnGridMoveToolHoldingItemActionNames:Array of Names=Make Array!(
     "[0]":Name=GetCreativeMoveBuildingsOnGridDuplicateName!(self:Fort Global UIContext Object Reference=@n1) -> (ReturnValue:Name=None)
     "[1]":Name=GetCreativeMoveBuildingsOnGridRotateClockwiseName!(self:Fort Global UIContext Object Reference=@n1) -> (ReturnValue:Name=None)
     "[2]":Name=GetCreativeMoveBuildingsOnGridRotateCounterclockwiseName!(self:Fort Global UIContext Object Reference=@n1) -> (ReturnValue:Name=None)
     "[3]":Name=GetCreativeMoveBuildingsOnGridMirrorName!(self:Fort Global UIContext Object Reference=@n1) -> (ReturnValue:Name=None)
     "[4]":Name=GetCreativeMoveBuildingsOnGridExitName!(self:Fort Global UIContext Object Reference=@n1) -> (ReturnValue:Name=None)
    ) -> (Array:Array of Names)) -> (Output_Get:Array of Names)
   > Set MoveOnGridMoveToolHoldingItemsActionLabels(MoveOnGridMoveToolHoldingItemsActionLabels:Array of Texts=Make Array(
     "[0]":Text=Get PasteLabelText() -> (PasteLabelText:Text)
     "[1]":Text=Get RotateClockwiseLabelText() -> (RotateClockwiseLabelText:Text)
     "[2]":Text=Get RotateCounterClockwiseLabelText() -> (RotateCounterClockwiseLabelText:Text)
     "[3]":Text=Get MirrorLabelText() -> (MirrorLabelText:Text)
     "[4]":Text=Get ExitLabelText() -> (ExitLabelText:Text)
    ) -> (Array:Array of Texts)) -> (Output_Get:Array of Texts)
   > Set MirrorIndex(MirrorIndex:int=3) -> (Output_Get:int=0)
  ]}
]
variables [
 ExitLabelText {type=string, default=Exit}
 MirrorDisabled {type=boolean}
 MirrorIndex {type=integer}
 MirrorLabelText {type=string, default=Mirror}
 MoveOnGridMoveToolHoldingItemActionNames {type="array<string>", default []}
 MoveOnGridMoveToolHoldingItemsActionLabels {type="array<string>", default []}
 PasteLabelText {type=string, default=Paste}
 RotateClockwiseLabelText {type=string, default=Rotate Clockwise}
 RotateCounterClockwiseLabelText {type=string, default=Rotate Counterclockwise}
]
widgets {
 info {
  inheritedWidgetCount=0
  parentClass=&/AthenaHUDSituationalFeedback.AthenaHUDSituationalFeedback_C
  rootWidgetClass=null
  widgetCount=0
 }
 widgets []
}
