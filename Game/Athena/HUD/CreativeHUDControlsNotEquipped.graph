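# Creative-mode HUD widget, derived from AthenaHUDSituationalFeedback, showing keybind hints for fly, flight speed, phase and heatmap controls, presumably when nothing is equipped.
# SetupActionNamesAndLabels builds parallel action and label arrays, recording line indices; ClearAndSetFeedbackLines fills inherited feedback lines from them.
# Update sets dynamic speed, phase and heatmap text, filters lines by flight state and valid keys, and hides some in minigames.
# Input-method changes rebuild the list.
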WidgetBlueprint /Game/Athena/HUD/CreativeHUDControlsNotEquipped
graphs [
 AddFeedbackLine {nodes [
   n1 = AddFeedbackLine!() -> (then, ActionName:Name, Label:Text) {is=title K2Node_FunctionEntry}
   > Add(TargetArray:Array of Names=Get NotEquippedActionNames() -> (NotEquippedActionNames:Array of Names), NewItem:"Name (by ref)"=@n1) {is=title K2Node_CallArrayFunction}
   > n2 = Add(TargetArray:Array of Texts=Get NotEquippedActionLabels() -> (NotEquippedActionLabels:Array of Texts), NewItem:"Text (by ref)"=@n1.Label)
   > Return Node!(execute, Index:int=@n2)
  ]}
 EventGraph {nodes [
   n1 = Update!() -> (OutputDelegate, then, IsFlying:bool=~) {is=title K2Node_CustomEvent}
   n2 = Get FortGlobalUIContext!() -> (ReturnValue:Fort Global UIContext Object Reference) {is=title K2Node_GetBlueprintContextEx}
   > n3 = Is Valid(InputObject=@n2)
   n4 = Set IsUsingGamepad(execute=@n3.Is Not Valid, IsUsingGamepad:bool=false) -> (Output_Get:bool)
   n5 = Get FeedbackLines() -> (FeedbackLines:Array of Athena HUDFeedback Line Object References)
   n6 = Get GhostModeIndex() -> (GhostModeIndex:int=0)
   SetText(
    execute=[@n4.then, @n21.then]
    self=Get FeedbackText!(self:Athena HUDFeedback Line Object Reference="Get (a copy)"(Array:Array of Athena HUDFeedback Line Object References=@n5, Dimension 1=@n6) -> (Output:Athena HUDFeedback Line Object Reference)) -> (FeedbackText:Rich Text Block Object Reference)
    InText=Make Ghost Mode Text!() -> (Return Value:Text) {is=title K2Node_MacroInstance}
   )
   > SetupHeatmapModeLabels()
   n7 = Get FeedbackLines()
   n8 = Get FlightSpeedLabelIndex() -> (FlightSpeedLabelIndex:int=0)
   > SetText(
    self=Get FeedbackText(self="Get (a copy)"(Array=@n7, Dimension 1=@n8))
    InText=Make Flight Speed Label!() -> (FlightSpeedTextLabel:Text) {is=title K2Node_MacroInstance}
   )
   > n9 = Cast To FortPlayerControllerAthena(Object=GetOwningPlayer())
   > n10 = Sequence()
   > n11 = Branch(Condition=@n1.IsFlying)
   n12 = Get Feedback() -> (Feedback:Vertical Box Object Reference)
   > n13 = For Loop(LastIndex=int - int(A=GetChildrenCount!(self:Panel Widget Object Reference=@n12) -> (ReturnValue:int=0), B=1) {is=title K2Node_CallFunction})
   n14 = IsValidFeedbackKey!(Index:int=@n13) -> (Valid:bool)
   n15 = "Not Equal (Integer)"(A=Get FlyTextIndex() -> (FlyTextIndex:int=0), B=@n13)
   > SetVisibility(
    self=GetChildAt!(self:Panel Widget Object Reference=@n12, Index:int=@n13) -> (ReturnValue:Widget Object Reference)
    InVisibility=Select(Option 0=Collapsed, Option 1=HitTestInvisible, Index=AND Boolean(A=@n15, B=@n14))
   )
   n16 = Get Feedback()
   n17 = For Loop(execute=@n11.else, LastIndex=int - int(A=GetChildrenCount(self=@n16) -> (ReturnValue=0), B=1))
   n18 = IsValidFeedbackKey(Index=@n17)
   n19 = "Equal (Integer)"(A=Get FlyTextIndex() -> (FlyTextIndex=0), B=@n17)
   > SetVisibility(
    self=GetChildAt(self=@n16, Index=@n17)
    InVisibility=Select(Option 0=Collapsed, Option 1=HitTestInvisible, Index=AND Boolean(A=@n19, B=@n18))
   )
   Branch(execute=@n10.then_1, Condition=IsPlayerInAMinigame!(self:Fort Player Controller Athena Object Reference=@n9) -> (ReturnValue:bool))
   > n20 = For Each Loop(Array:Array of Integers=Get MinigameHideWidgets() -> (MinigameHideWidgets:Array of Integers)) -> (Array Element:int)
   > SetVisibility(self=GetChildAt(self=@n16, Index=@n20), InVisibility=Collapsed)
   n21 = Set IsUsingGamepad(execute=@n3.Is Valid, IsUsingGamepad=IsUsingGamepad!(self:Fort Global UIContext Object Reference=@n2) -> (ReturnValue:bool))
   ClearAndSetFeedbackLines() {is=title K2Node_CustomEvent}
   > ClearChildren(self=Get Feedback())
   > n22 = For Each Loop(Array:Array of Names=Get NotEquippedActionNames()) -> (Array Element:Name)
   n23 = GetChildrenCount(self=Get Feedback()) -> (ReturnValue=0)
   > Branch(Condition="integer < integer"!(A:int=@n23, B:int=Get MaxLinesNeeded() -> (MaxLinesNeeded:int=0)) -> (ReturnValue:bool) {is=title K2Node_CallFunction})
   n24 = "Get (a copy)"(Array=Get FeedbackLines(), Dimension 1=@n23)
   > SetVisibility(self=@n24, InVisibility=SelfHitTestInvisible)
   > SetBoundAction!(
    execute
    self:Fort Keybind Widget Object Reference=Get FeedbackKeyBind!(self:Athena HUDFeedback Line Object Reference=@n24) -> (FeedbackKeyBind:Keybind Widget Object Reference)
    NewBoundAction:Name=@n22
   ) -> (then)
   > SetText(
    self=Get FeedbackText(self=@n24)
    InText="Get (a ref)"(Array:Array of Texts=Get NotEquippedActionLabels(), Dimension 1=@n22.Array Index) -> (Output:"Text (by ref)")
   )
   > AddChild(self=Get Feedback(), Content=@n24)
   n25 = OnKeybindsChanged!() -> (OutputDelegate, then, IsFlying:bool=~) {is=title K2Node_CustomEvent}
   > SetupActionNamesAndLabels()
   > ClearAndSetFeedbackLines() {is=title K2Node_CallFunction}
   > Update!(execute, self, IsFlying:bool=@n25.IsFlying) -> (then) {is=title K2Node_CallFunction}
   OnInputMethodChanged!() -> (OutputDelegate, then, InputType:ECommonInputType Enum) {is=title K2Node_CustomEvent}
   > SetupActionNamesAndLabels()
   > ClearAndSetFeedbackLines()
   > n26 = Cast To FortPlayerControllerGameplay!(execute, Object=GetOwningPlayer()) -> (then, CastFailed, AsFort Player Controller Gameplay:Fort Player Controller Gameplay Object Reference) {is=title K2Node_DynamicCast}
   > Update(IsFlying=HasCreativeFlyingInputComponent!(self:Fort Player Controller Gameplay Object Reference=@n26) -> (ReturnValue:bool))
   Update(execute=@n26.CastFailed, IsFlying=false)
   Event Construct()
   > SetupActionNamesAndLabels()
   > Set MaxLinesNeeded(MaxLinesNeeded:int=Length(TargetArray:Array of Names=Get NotEquippedActionNames())) -> (Output_Get:int=0)
   > Parent: Construct()
   > ClearAndSetFeedbackLines()
   > Assign On Input Method Changed!(execute:Exec, self:Common Input Object Reference=Get CommonInputSubsystem(), Delegate:"Delegate (by ref)"=Create Event()) -> (then:Exec) {is=title K2Node_AssignDelegate}
  ]}
 IsValidFeedbackKey {nodes [
   n1 = IsValidFeedbackKey!() -> (then, Index:int) {is=title K2Node_FunctionEntry}
   n2 = Get NotEquippedActionNames()
   > n3 = Branch(Condition=Is Valid Index!(TargetArray:Array of Names=@n2, IndexToTest:int=@n1) -> (ReturnValue:bool))
   n4 = "Get (a ref)"(Array:Array of Names=@n2, Dimension 1=@n1) -> (Output:"Name (by ref)")
   n5 = Get FortGlobalUIContext()
   > Return Node!(
    execute
    Valid:bool=Is Valid Key!(Key:"Key Structure (by ref)"=GetKeyForAction!(self:Fort Global UIContext Object Reference=@n5, Action:Name=@n4, bForceGamepadKey:bool=false, PresetNameOverride:Name=None) -> (ReturnValue:Key Structure)) -> (ReturnValue:bool)
   )
   Return Node(execute=@n3.else, Valid=false)
  ]}
 MakeFlightSpeedLabel {nodes [
   Inputs!()
   n1 = Get FlightSpeeds() -> (FlightSpeeds:"Array of Float (double-precision)s")
   n2 = "Get (a ref)"(
    Array:"Array of Float (double-precision)s"=Get FlightSpeeds()
    Dimension 1="Clamp (Integer)"(
     Value=Get FlightSpeedLevel() -> (FlightSpeedLevel:int=0)
     Min=0
     Max=Last Index!(TargetArray:"Array of Float (double-precision)s"=@n1) -> (ReturnValue=0)
    ) -> (ReturnValue=0)
   ) -> (Output:"Float (double-precision) (by ref)")
   Outputs!(FlightSpeedTextLabel:Text=Format Text!(
     Format="LOCTEXT(\"{0} [{1}x]\")"
     0:Text=Get FlightSpeedTextLabel() -> (FlightSpeedTextLabel:Text)
     1:Text="To Text (Float)"(Value=@n2, bUseGrouping=false, MinimumIntegralDigits=1, MaximumIntegralDigits=324, MinimumFractionalDigits=1, MaximumFractionalDigits=1)
    ) -> (Result))
  ]}
 MakeGhostModeText {nodes [
   Inputs()
   n1 = Get GhostModeLabelText() -> (GhostModeLabelText:Text)
   n2 = Format Text(Format="LOCTEXT(\"{0} {1}\")", 0=Get Hold Label Text() -> (Hold Label Text:Text), 1=@n1)
   n3 = Get IsUsingGamepad() -> (IsUsingGamepad:bool)
   n4 = Get GhostModeEnabled() -> (GhostModeEnabled:bool)
   n5 = Get Off Label Text() -> (Off Label Text:Text)
   n6 = Get On Label Text() -> (On Label Text:Text)
   Outputs!(Return Value:Text=Format Text(
     Format="LOCTEXT(\"{0} {1}\")"
     0=Select(Option 0:Text=@n1, Option 1:Text=@n2, Index=@n3) -> (ReturnValue:Text)
     1=Select(Option 0=@n5, Option 1=@n6, Index=@n4)
    ))
  ]}
 MakeHeatmapModeText {nodes [
   MakeHeatmapModeText() {is=title K2Node_FunctionEntry}
   > n1 = Is Valid(InputObject=Get CachedMapManagerCreative() -> (CachedMapManagerCreative:Fort In Game Map Manager Creative Object Reference))
   n2 = Get CachedMapManagerCreative()
   > Return Node!(
    execute
    HeatmapModeText:"Text (by ref)"=Select!(
     Visible:"Text (by ref)"=MakeLiteralText(Value="LOCTEXT(\"Heatmap [On]\")")
     VisibleWithValues:"Text (by ref)"=MakeLiteralText(Value="LOCTEXT(\"Heatmap [Memory Values On]\")")
     NotVisible:"Text (by ref)"=MakeLiteralText(Value="LOCTEXT(\"Heatmap [Off]\")")
     Index:EFortCreativeHeatmapDrawState Enum=GetHeatmapDrawState!(self:Fort In Game Map Manager Creative Object Reference=@n2) -> (ReturnValue:EFortCreativeHeatmapDrawState Enum=Visible)
    ) -> (ReturnValue:"Text (by ref)")
   )
   Return Node(execute=@n1.Is Not Valid, HeatmapModeText=Get HeatmapToggleTextLabel() -> (HeatmapToggleTextLabel:Text))
  ]}
 SetupActionNamesAndLabels {nodes [
   SetupActionNamesAndLabels() {is=title K2Node_FunctionEntry}
   > Clear(TargetArray:Array of Names=Get NotEquippedActionNames())
   > Clear(TargetArray:Array of Texts=Get NotEquippedActionLabels())
   > Clear(TargetArray:Array of Integers=Get MinigameHideWidgets())
   > Set HeatmapToggleIndex(HeatmapToggleIndex:int=-1) -> (Output_Get:int=0)
   > n1 = Sequence!(execute) -> (then_0, then_1, then_2:Exec, then_3:Exec, then_4:Exec, then_5:Exec)
   n2 = Get FortGlobalUIContext()
   n3 = AddFeedbackLine!(
    execute=@n1.then_2
    self
    ActionName:Name=GetCreativeFlyDownActionName!(self:Fort Global UIContext Object Reference=@n2) -> (ReturnValue:Name=None)
    Label:Text=Get FlyDownLabelText() -> (FlyDownLabelText:Text)
   ) -> (then, Index:int=0) {is=title K2Node_CallFunction}
   > Set FlyDownTextIndex(FlyDownTextIndex:int=@n3) -> (Output_Get:int=0)
   n4 = GetGameStateAthena!() -> (ReturnValue:Fort Game State Athena Object Reference)
   Is Valid(exec=@n1.then_4, InputObject=@n4)
   > Branch(Condition=NOT Boolean(A=IsVKEditPlaylist!(self:Fort Game State Athena Object Reference=@n4) -> (ReturnValue:bool)))
   > n5 = AddFeedbackLine(
    ActionName=GetCreativeToggleHeatmapName!(self:Fort Global UIContext Object Reference=@n2) -> (ReturnValue:Name=None)
    Label=MakeHeatmapModeText!() -> (HeatmapModeText:Text) {is=title K2Node_CallFunction}
   ) -> (Index=0)
   > Set HeatmapToggleIndex(HeatmapToggleIndex=@n5) -> (Output_Get=0)
   n6 = AddFeedbackLine(
    execute=@n1.then_3
    ActionName=GetCreativeMoveObjectsFreelyPullName!(self:Fort Global UIContext Object Reference=@n2) -> (ReturnValue:Name=None)
    Label=Get FlightSpeedTextLabel()
   ) -> (Index=0)
   > Set FlightSpeedLabelIndex(FlightSpeedLabelIndex:int=@n6) -> (Output_Get:int=0)
   > Add(TargetArray:Array of Integers=Get MinigameHideWidgets(), NewItem:"Integer (by ref)"=Get FlightSpeedLabelIndex() -> (FlightSpeedLabelIndex=0))
   n7 = AddFeedbackLine(
    execute=@n1.then_5
    ActionName=GetCreativeToggleGhostName!(self:Fort Global UIContext Object Reference=@n2) -> (ReturnValue:Name=None)
    Label=Get GhostModeLabelText()
   ) -> (Index=0)
   > n8 = Set GhostModeIndex(GhostModeIndex:int=@n7) -> (Output_Get:int=0)
   > Add(TargetArray=Get MinigameHideWidgets(), NewItem=@n8)
   n9 = AddFeedbackLine(
    execute=@n1.then_0
    ActionName=GetJumpActionName!(self:Fort Global UIContext Object Reference=@n2) -> (ReturnValue:Name=None)
    Label=Get FlyLabelText() -> (FlyLabelText:Text)
   ) -> (Index=0)
   > Set FlyTextIndex(FlyTextIndex:int=@n9) -> (Output_Get:int=0)
   n10 = AddFeedbackLine(execute=@n1.then_1, ActionName=GetJumpActionName(self=@n2) -> (ReturnValue=None), Label=Get FlyExitLabelText() -> (FlyExitLabelText:Text)) -> (Index=0)
   > Set FlyExitTextIndex(FlyExitTextIndex:int=@n10) -> (Output_Get:int=0)
  ]}
 SetupHeatmapModeLabels {nodes [
   SetupHeatmapModeLabels() {is=title K2Node_FunctionEntry}
   > Branch(Condition=Is Valid Index(
     TargetArray:Array of Athena HUDFeedback Line Object References=Get FeedbackLines()
     IndexToTest=Get HeatmapToggleIndex() -> (HeatmapToggleIndex:int=0)
    ))
   > n1 = Is Valid(InputObject=Get CachedMapManagerCreative())
   n2 = GetUIMapManager!() -> (ReturnValue:Fort In Game Map Manager Object Reference)
   Is Valid(exec=@n1.Is Not Valid, InputObject=@n2)
   > n3 = Cast To FortInGameMapManagerCreative!(execute, Object=@n2) -> (then, CastFailed, AsFort in Game Map Manager Creative:Fort In Game Map Manager Creative Object Reference) {is=title K2Node_DynamicCast}
   > n4 = Set CachedMapManagerCreative(CachedMapManagerCreative:Fort In Game Map Manager Creative Object Reference=@n3) -> (Output_Get:Fort In Game Map Manager Creative Object Reference)
   n5 = Get FeedbackText(self="Get (a copy)"(Array=Get FeedbackLines(), Dimension 1=Get HeatmapToggleIndex() -> (HeatmapToggleIndex=0)))
   n6 = SetText(execute=[@n1.Is Valid, @n4.then], self=@n5, InText=MakeHeatmapModeText())
   Return Node!(execute=[@n6.then, @n7.then])
   n7 = SetText(execute=@n3.CastFailed, self=@n5, InText=Get HeatmapToggleTextLabel())
  ]}
]
variables [
 CachedMapManagerCreative {type=FortInGameMapManagerCreative, default=null}
 FlightSpeedLabelIndex {type=integer}
 FlightSpeedLevel {type=integer}
 FlightSpeeds {type="array<number>", default [0.1,0.5,1,1.5,2,3,4,5,10,15,20,25,30]}
 FlightSpeedTextLabel {type=string, default=Flight Speed}
 FlyDownLabelText {type=string, default=Fly Down}
 FlyDownTextIndex {type=integer}
 FlyExitLabelText {type=string, default="(Double Tap) Exit Fly"}
 FlyExitTextIndex {type=integer}
 FlyLabelText {type=string, default="(Double Tap) Fly"}
 FlyTextIndex {type=integer}
 GhostModeEnabled {type=boolean}
 GhostModeIndex {type=integer}
 GhostModeLabelText {type=string, default=Phase}
 HeatmapToggleIndex {type=integer}
 HeatmapToggleTextLabel {type=string, default=Heatmap}
 Hold Label Text {type=string, default=(Hold)}
 IsUsingGamepad {type=boolean}
 Level0Speed {type=string, default=0.5x}
 Level10Speed {type=string, default=25.0x}
 Level11Speed {type=string, default=30.0x}
 Level1Speed {type=string, default=1.0x}
 Level2Speed {type=string, default=1.5x}
 Level3Speed {type=string, default=2.0x}
 Level4Speed {type=string, default=3.0x}
 Level5Speed {type=string, default=4.0x}
 Level6Speed {type=string, default=5.0x}
 Level7Speed {type=string, default=10.0x}
 Level8Speed {type=string, default=15.0x}
 Level9Speed {type=string, default=20.0x}
 MinigameHideWidgets {type="array<integer>"}
 NotEquippedActionLabels {type="array<string>", default []}
 NotEquippedActionNames {type="array<string>", default []}
 Off Label Text {type=string, default="[Off]"}
 On Label Text {type=string, default="[On]"}
]
widgets {
 info {
  inheritedWidgetCount=0
  parentClass=&/AthenaHUDSituationalFeedback.AthenaHUDSituationalFeedback_C
  rootWidgetClass=null
  widgetCount=0
 }
 widgets []
}
