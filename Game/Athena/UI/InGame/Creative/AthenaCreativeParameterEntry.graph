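# Row widget for a Creative parameter, built on FortMatchmakingKnobsSpinnerButton, with a name label, left/right value selector, save spinner and collapsed text-entry field.
# Arrow or row clicks slide out the value, shift it on animation finish, then reverse the opposite animation to slide in.
# Directional navigation reuses those slides with a press sound.
# It also handles hover highlighting, index-staggered entrance, mobile arrow sizing and hover clearing outside gamepad input.
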WidgetBlueprint /Game/Athena/UI/InGame/Creative/AthenaCreativeParameterEntry
animations [
 Left {
  binding_count=1
  display_label=""
  length_seconds=0.1
  object=&self:Left
  sequence {bindings [MyText {
     id=#1
     possessed_class=CommonTextBlock
     tracks [
      {
       type=MovieScene2DTransformTrack
       sections [{
         type=MovieScene2DTransformSection
         end_frame=2
         start_frame=0
         channels [
          Translation.X {type=MovieSceneScriptingFloatChannel, keys [[0,0],[2,-100]]}
          Translation.Y {type=MovieSceneScriptingFloatChannel, const [0,1]}
          Angle {type=MovieSceneScriptingFloatChannel, const [0,1]}
          Scale.X {type=MovieSceneScriptingFloatChannel, const [1,1]}
          Scale.Y {type=MovieSceneScriptingFloatChannel, const [1,1]}
          Shear.X {type=MovieSceneScriptingFloatChannel, const [0,1]}
          Shear.Y {type=MovieSceneScriptingFloatChannel, const [0,1]}
         ]
        }]
      }
      {
       type=MovieSceneFloatTrack
       display_name=Render Opacity
       sections [{type=MovieSceneFloatSection, end_frame=2, start_frame=0, channels [None {type=MovieSceneScriptingFloatChannel, keys [[0,1],[2,0]]}]}]
      }
     ]
    }]}
 }
 Right {
  binding_count=1
  display_label=""
  length_seconds=0.1
  object=&self:Right
  sequence {bindings [MyText {
     id=#1
     possessed_class=CommonTextBlock
     tracks [
      {
       type=MovieScene2DTransformTrack
       sections [{
         type=MovieScene2DTransformSection
         end_frame=2
         start_frame=0
         channels [
          Translation.X {type=MovieSceneScriptingFloatChannel, keys [[0,0],[2,100]]}
          Translation.Y {type=MovieSceneScriptingFloatChannel, const [0,1]}
          Angle {type=MovieSceneScriptingFloatChannel, const [0,1]}
          Scale.X {type=MovieSceneScriptingFloatChannel, const [1,1]}
          Scale.Y {type=MovieSceneScriptingFloatChannel, const [1,1]}
          Shear.X {type=MovieSceneScriptingFloatChannel, const [0,1]}
          Shear.Y {type=MovieSceneScriptingFloatChannel, const [0,1]}
         ]
        }]
      }
      {
       type=MovieSceneFloatTrack
       display_name=Render Opacity
       sections [{type=MovieSceneFloatSection, end_frame=2, start_frame=0, channels [None {type=MovieSceneScriptingFloatChannel, keys [[0,1],[2,0]]}]}]
      }
     ]
    }]}
 }
 In {
  binding_count=1
  length_seconds=0.35
  object=&self:In
  sequence {bindings [Border_Fill {
     id=#2
     possessed_class=Border
     tracks [
      {
       type=MovieScene2DTransformTrack
       sections [{
         type=MovieScene2DTransformSection
         end_frame=7
         start_frame=0
         channels [
          Translation.X {type=MovieSceneScriptingFloatChannel, keys [[0,300],[5,-25],[7,0]]}
          Translation.Y {type=MovieSceneScriptingFloatChannel}
          Angle {type=MovieSceneScriptingFloatChannel}
          Scale.X {type=MovieSceneScriptingFloatChannel}
          Scale.Y {type=MovieSceneScriptingFloatChannel}
          Shear.X {type=MovieSceneScriptingFloatChannel}
          Shear.Y {type=MovieSceneScriptingFloatChannel}
         ]
        }]
      }
      {
       type=MovieSceneFloatTrack
       display_name=Render Opacity
       sections [{type=MovieSceneFloatSection, end_frame=7, start_frame=0, channels [None {type=MovieSceneScriptingFloatChannel, keys [[0,0],[7,1]]}]}]
      }
     ]
    }]}
 }
 OnHovered {
  binding_count=1
  length_seconds=0.05
  object=&self:OnHovered
  sequence {bindings [Border_Fill {
     id=#3
     possessed_class=Border
     tracks [{
       type=MovieSceneColorTrack
       display_name=Brush Color
       sections [{
         type=MovieSceneColorSection
         end_frame=1
         start_frame=0
         channels [
          Color.R {type=MovieSceneScriptingFloatChannel}
          Color.G {type=MovieSceneScriptingFloatChannel}
          Color.B {type=MovieSceneScriptingFloatChannel}
          Color.A {type=MovieSceneScriptingFloatChannel, keys [[0,0],[1,1]]}
         ]
        }]
      }]
    }]}
 }
 OnUnHovered {
  binding_count=1
  length_seconds=0.2
  object=&self:OnUnHovered
  sequence {bindings [Border_Fill {
     id=#3
     possessed_class=Border
     tracks [{
       type=MovieSceneColorTrack
       display_name=Brush Color
       sections [{
         type=MovieSceneColorSection
         end_frame=4
         start_frame=0
         channels [
          Color.R {type=MovieSceneScriptingFloatChannel}
          Color.G {type=MovieSceneScriptingFloatChannel}
          Color.B {type=MovieSceneScriptingFloatChannel}
          Color.A {type=MovieSceneScriptingFloatChannel, keys [[0,1],[4,0]]}
         ]
        }]
      }]
    }]}
 }
]
graphs [
 DoCustomNavigation_0 {nodes [
   DoCustomNavigation_0!() -> (then, Navigation:EUINavigation Enum) {is=title K2Node_FunctionEntry}
   > ShiftTextLeft(self:Common Rotator Object Reference)
   > Return Node!(execute, ReturnValue:Widget Object Reference=Self-Reference())
  ]}
 EventGraph {nodes [
   n1 = InputModeChanged!() -> (OutputDelegate, then, GamePad:bool=~) {is=title K2Node_CustomEvent}
   n2 = SetStyle!(
    execute
    self:Common Button Base Object Reference
    InStyle:Common Button Style Class Reference=Select(
     Option 0:Common Button Style Class Reference=/FortUILibrary/Foundation/Buttons/ButtonStyles/ButtonStyle-TextOnlyBase_Empty.ButtonStyle-TextOnlyBase_Empty_C
     Option 1:Common Button Style Class Reference=/FortUILibrary/Foundation/Buttons/ButtonStyles/ButtonStyle-TextOnlyBase_EmptyFocus.ButtonStyle-TextOnlyBase_EmptyFocus_C
     Index=@n1.GamePad
    ) -> (ReturnValue:Common Button Style Class Reference)
   ) -> (then)
   n3 = Set isGamepad(execute=[@n2.then, @n1.then], isGamepad:bool=@n1.GamePad) -> (Output_Get:bool)
   > n4 = Branch(Condition=@n3)
   ClearHoverState(execute=@n4.else)
   n5 = "On Button Clicked (LeftButton)"!() -> (OutputDelegate, then, Button:Common Button Legacy Object Reference) {is=title K2Node_ComponentBoundEvent}
   n6 = AfterleftisDone() {is=title K2Node_CustomEvent}
   n7 = Get Left() -> (Left:Widget Animation Object Reference)
   BindToAnimationFinished!(execute=@n5.then, self:Widget Animation Object Reference=@n7, Widget:User Widget Object Reference=Self-Reference(), Delegate=@n6) -> (then)
   > PlayAnimation(InAnimation=@n7)
   n8 = "On Button Clicked (RightButton)"!() -> (OutputDelegate, then, Button:Common Button Legacy Object Reference) {is=title K2Node_ComponentBoundEvent}
   n9 = AfterRightIsDone() {is=title K2Node_CustomEvent}
   n10 = Get Right() -> (Right:Widget Animation Object Reference)
   BindToAnimationFinished(execute=[@n8.then, @n20.then], self=@n10, Widget=Self-Reference(), Delegate=@n9)
   > PlayAnimation(InAnimation=@n10)
   n11 = AnimateRightFinished() {is=title K2Node_CustomEvent}
   n12 = Self-Reference()
   n13 = Get Left()
   > UnbindAllFromAnimationFinished!(execute, self:User Widget Object Reference=@n12, Animation:Widget Animation Object Reference=@n13) -> (then)
   > PlayAnimation(InAnimation=@n13, PlayMode=Reverse)
   n14 = AnimateLeftFinished() {is=title K2Node_CustomEvent}
   n15 = Get Right()
   > UnbindAllFromAnimationFinished(self=@n12, Animation=@n15)
   > PlayAnimation(InAnimation=@n15, PlayMode=Reverse)
   ClearHoverState() {is=title K2Node_CustomEvent}
   n16 = Get Border_Fill() -> (Border_Fill:Border Object Reference)
   n17 = BreakColor(InColor=Get BrushColor!(self:Border Object Reference=@n16) -> (BrushColor:Linear Color Structure))
   > SetBrushColor(self=@n16, InBrushColor=MakeColor!(R:float=@n17, G:float=@n17.G, B:float=@n17.B, A:float=0) -> (ReturnValue:Linear Color Structure))
   n18 = InputUpdated!() -> (OutputDelegate, then, bNewInputType:ECommonInputType Enum) {is=title K2Node_CustomEvent}
   > InputModeChanged!(execute, self, GamePad:bool="Equal (Enum)"(A=@n18.bNewInputType, B=Gamepad)) -> (then) {is=title K2Node_CallFunction}
   n19 = AnimateIn!() -> (OutputDelegate, then, Index:int) {is=title K2Node_CustomEvent}
   > SetRenderOpacity(InOpacity=0)
   > Delay(Duration=int * float!(A:int=@n19.Index, B:double=0.05) -> (ReturnValue:double))
   > PlayAnimation(InAnimation=Get In() -> (In:Widget Animation Object Reference), PlaybackSpeed=1.3)
   > SetRenderOpacity(InOpacity=1)
   n20 = Event On Clicked()
   Event On Hovered()
   > Parent: On Hovered() {is=title K2Node_CallParentFunction}
   n21 = Get OnHovered() -> (OnHovered:Widget Animation Object Reference)
   > PlayAnimationForward(InAnimation=@n21)
   > UpdateSelectionState!(execute, self, Selected:bool=true) -> (then)
   Event OnNavigatedRight() {is=title K2Node_Event}
   n22 = Get Right()
   > BindToAnimationFinished!(execute, self:User Widget Object Reference=Self-Reference(), Animation:Widget Animation Object Reference=@n22, Delegate=@n11) -> (then)
   > PlayAnimation(InAnimation=@n22)
   > PlayButtonStyleSound()
   Event OnNavigatedLeft() {is=title K2Node_Event}
   n23 = Get Left()
   > BindToAnimationFinished(self=Self-Reference(), Animation=@n23, Delegate=@n14)
   > PlayAnimation(InAnimation=@n23)
   > PlayButtonStyleSound()
   n24 = Event OnLoadingStateChanged!() -> (OutputDelegate, then, bCurrentlyLoading:bool) {is=title K2Node_Event}
   > SetVisibility(
    self=Get SaveSpinner() -> (SaveSpinner:Image Object Reference)
    InVisibility=Select(Option 0:ESlateVisibility Enum=Collapsed, Option 1:ESlateVisibility Enum=Visible, Index=@n24.bCurrentlyLoading) -> (ReturnValue:ESlateVisibility Enum)
   )
   Event Construct()
   n25 = Get CommonInputSubsystem!() -> (ReturnValue:Common Input Object Reference) {is=title K2Node_GetBlueprintContextEx}
   > InputModeChanged(GamePad="Equal (Enum)"(A=GetCurrentInputType(self=@n25), B=Gamepad))
   > Assign On Input Method Changed!(execute:Exec, self:Common Input Object Reference=@n25, Delegate:"Delegate (by ref)"=@n18) -> (then:Exec) {is=title K2Node_AssignDelegate}
   > Set Button Sizes() {is=title K2Node_CallFunction}
   > UpdateSelectionState(Selected=false)
   Event On Unhovered()
   > Parent: On Unhovered() {is=title K2Node_CallParentFunction}
   > PlayAnimationReverse(InAnimation=@n21)
   > UpdateSelectionState(Selected=false)
   n26 = Event SetSpinnerOptionsVisiblity!() -> (OutputDelegate, then, bVisible:bool) {is=title K2Node_Event}
   > n27 = Branch(Condition=@n26.bVisible)
   > n28 = Set VisibilitySetting(VisibilitySetting:ESlateVisibility Enum=Visible) -> (Output_Get:ESlateVisibility Enum=Visible)
   SetVisibility(
    execute=[@n29.then, @n28.then]
    self=Get OptionSelectionArea() -> (OptionSelectionArea:Horizontal Box Object Reference)
    InVisibility=Get VisibilitySetting() -> (VisibilitySetting:ESlateVisibility Enum=Visible)
   )
   n29 = Set VisibilitySetting(execute=@n27.else, VisibilitySetting=Collapsed) -> (Output_Get=Visible)
   Branch(Condition=Event PreConstruct().IsDesignTime)
   > UpdateSelectionState(Selected=false)
   n30 = Get Left()
   PlayAnimation(execute=@n33.then, InAnimation=@n30, PlayMode=Reverse)
   n31 = Get Right()
   PlayAnimation(execute=@n34.then, InAnimation=@n31, PlayMode=Reverse)
   K2Node_Event_4 = CustomEvent!() -> (OutputDelegate, then, bUsingGamepad:bool=~) {is=title K2Node_CustomEvent}
   ShiftTextRight(execute=@n9.then, self:Common Rotator Object Reference)
   n32 = Self-Reference()
   > n33 = UnbindAllFromAnimationFinished!(execute, self:Widget Animation Object Reference=[@n30, @n31], Widget:User Widget Object Reference=@n32) -> (then)
   ShiftTextLeft(execute=@n6.then)
   > n34 = UnbindAllFromAnimationFinished(self=[@n30, @n31], Widget=@n32)
  ]}
 PlayButtonStyleSound {nodes [
   PlayButtonStyleSound() {is=title K2Node_FunctionEntry}
   n1 = GetStyle!(self:Common Button Base Object Reference=Get LeftButton() -> (LeftButton:Simple Common Button Object Reference)) -> (ReturnValue:Common Button Style Object Reference)
   > n2 = Cast To SoundBase!(
    execute
    Object=Break Slate Sound!(SlateSound:"Slate Sound Structure (by ref)"=Get PressedSlateSound!(self:Common Button Style Object Reference=@n1) -> (PressedSlateSound:Slate Sound Structure)) -> (ResourceObject:Object Reference) {is=title K2Node_BreakStruct}
   ) -> (then, CastFailed, AsSound Base:Sound Base Object Reference) {is=title K2Node_DynamicCast}
   > PlaySound2D(Sound=@n2)
  ]}
 Set Button Sizes {nodes [
   Set Button Sizes() {is=title K2Node_FunctionEntry}
   > Branch(Condition=IsMobileGame())
   n1 = Get Mobile Size Override() -> (Mobile Size Override:double)
   n2 = Get LeftButtonSize() -> (LeftButtonSize:Size Box Object Reference)
   n3 = Get RightButtonSize() -> (RightButtonSize:Size Box Object Reference)
   > SetHeightOverride(self=[@n2, @n3], InHeightOverride=@n1)
   > SetWidthOverride(self=[@n2, @n3], InWidthOverride=@n1)
  ]}
 UpdateSelectionState {nodes [
   UpdateSelectionState!() -> (then, Selected:bool=~) {is=title K2Node_FunctionEntry}
   > SetVisibility(
    self=Get Image_Selection() -> (Image_Selection:Image Object Reference)
    InVisibility=Select(Option 0=Hidden, Option 1=SelfHitTestInvisible, Index=Get Selected() -> (Selected:bool))
   )
   > Set BrushColor!(
    execute
    BrushColor:Linear Color Structure=Select(Option 0:Linear Color Structure=(R=1,G=1,B=1,A=0), Option 1:Linear Color Structure=(R=1,G=1,B=1,A=0.5), Index=Get Selected()) -> (ReturnValue:Linear Color Structure)
    self:Border Object Reference=Get Border_Fill()
   ) -> (then, Output_Get:Linear Color Structure)
  ]}
]
variables [
 isGamepad {type=boolean}
 Mobile Size Override {type=number, default=50}
 NormalTextColor {type=SlateColor, default {colorUseRule=UseColor_Specified, specifiedColor=<rgba 1 1 1 1>}}
 SelectedTextColor {type=SlateColor, default {colorUseRule=UseColor_Specified, specifiedColor=<rgba 0.0048 0.004 0.0116 1>}}
 VisibilitySetting {type=ESlateVisibility, default=Visible}
]
widgets {
 info {
  inheritedWidgetCount=5
  parentClass=/Script/FortniteUI.FortMatchmakingKnobsSpinnerButton
  rootWidgetClass=&Overlay
  widgetCount=21
 }
 widgets [
  Overlay_0 {widgetClassPath=&Overlay, widget=Overlay_0}
  Image_Selection {
   widgetClassPath=&Image
   bIsVariable=true
   parent=Overlay_0
   widget=Image_Selection
   properties {brush {
     bIsDynamicallyLoaded=false
     drawAs=Image
     imageType=NoImage
     mirroring=NoMirror
     resourceName=None
     resourceObject=/FortUILibrary/UIKit/Buttons/MaterialInstances/TileOutline/MI_UIKit_Outline
     tiling=NoTile
     imageSize=<6 64>
     margin=<ltrb 0 0 0 0>
     outlineSettings {
      bUseBrushTransparency=false
      roundingType=HalfHeightRadius
      width=0
      color {colorUseRule=UseColor_Specified, specifiedColor=<rgba 0 0 0 0>}
      cornerRadii=<0 0 0 0>
     }
     tintColor {colorUseRule=UseColor_Specified, specifiedColor=<rgba 1 1 1 1>}
     uVRegion {bIsValid=false, max=<0 0>, min=<0 0>}
    }}
   slot {class=OverlaySlot, properties {horizontalAlignment=HAlign_Fill, verticalAlignment=VAlign_Fill, padding=<ltrb 0 -8 0 -8>}}
  }
  Border_Fill {
   widgetClassPath=&Border
   bIsVariable=true
   parent=Overlay_0
   widget=Border_Fill
   properties {
    navigation=&self:WidgetTree.Border_Fill.WidgetNavigation_0
    verticalAlignment=VAlign_Center
    background {
     bIsDynamicallyLoaded=false
     drawAs=Image
     imageType=NoImage
     mirroring=NoMirror
     resourceName=None
     resourceObject=/FortUILibrary/UIKit/Buttons/MaterialInstances/CTAButtons/MI_UIKit_Button_QuietRoundedCorners
     tiling=NoTile
     imageSize=<32 32>
     margin=<ltrb 0 0 0 0>
     outlineSettings {
      bUseBrushTransparency=false
      roundingType=FixedRadius
      width=0
      color {colorUseRule=UseColor_Specified, specifiedColor=<rgba 0 0 0 0>}
      cornerRadii=<0 0 0 0>
     }
     tintColor {colorUseRule=UseColor_Specified, specifiedColor=<rgba 1 1 1 0.1>}
     uVRegion {bIsValid=false, max=<0 0>, min=<0 0>}
    }
    padding=<ltrb 22 0 0 0>
   }
   slot {class=OverlaySlot, properties {horizontalAlignment=HAlign_Fill, verticalAlignment=VAlign_Fill, padding=<ltrb 4 -4 4 -4>}}
  }
  SB_Contents {
   widgetClassPath=&SizeBox
   parent=Border_Fill
   widget=SB_Contents
   properties {bOverride_HeightOverride=true, heightOverride=42, maxDesiredWidth=1000, minDesiredHeight=70.6997, minDesiredWidth=100, widthOverride=1000}
   slot {class=BorderSlot, properties {verticalAlignment=VAlign_Center, padding=<ltrb 22 0 0 0>}}
  }
  CommonBorder_Container {
   widgetClassPath=&CommonBorder
   bIsVariable=true
   parent=SB_Contents
   widget=CommonBorder_Container
   properties {padding=<ltrb 0 0 0 0>}
   slot {class=SizeBoxSlot}
  }
  HBox_Content {
   widgetClassPath=&HorizontalBox
   parent=CommonBorder_Container
   widget=HBox_Content
   slot {class=BorderSlot, properties {padding=<ltrb 0 0 0 0>}}
  }
  SaveSpinner {
   widgetClassPath=&Image
   bIsVariable=true
   parent=HBox_Content
   widget=SaveSpinner
   properties {
    visibility=Collapsed
    brush {
     bIsDynamicallyLoaded=false
     drawAs=Image
     imageType=NoImage
     mirroring=NoMirror
     resourceName=None
     resourceObject=/FortUILibrary/UIKit/Throbber/MI_UIKit_Monochrome_Throbber
     tiling=NoTile
     imageSize=<35 35>
     margin=<ltrb 0 0 0 0>
     outlineSettings {
      bUseBrushTransparency=false
      roundingType=HalfHeightRadius
      width=0
      color {colorUseRule=UseColor_Specified, specifiedColor=<rgba 0 0 0 0>}
      cornerRadii=<0 0 0 0>
     }
     tintColor {colorUseRule=UseColor_Specified, specifiedColor=<rgba 1 1 1 1>}
     uVRegion {bIsValid=false, max=<0 0>, min=<0 0>}
    }
   }
   slot {class=HorizontalBoxSlot, properties {horizontalAlignment=HAlign_Right, verticalAlignment=VAlign_Top}}
  }
  SizeBox_0 {
   widgetClassPath=&SizeBox
   parent=HBox_Content
   widget=SizeBox_0
   properties {bOverride_MaxDesiredWidth=true, maxDesiredWidth=380, minDesiredWidth=200, widthOverride=550}
   slot {class=HorizontalBoxSlot, properties {size {sizeRule=Fill, value=1}}}
  }
  TextBlock_OptionDisplayName {
   widgetClassPath=&CommonTextBlock
   bInherited=true
   bIsVariable=true
   parent=SizeBox_0
   widget=TextBlock_OptionDisplayName
   properties {
    scrollStyle=&TextScrollStyle-Base_C
    text="Flight Speed [1.x] Flight Speed [1.x] Flight Speed [1.x]"
    colorAndOpacity {colorUseRule=UseColor_Specified, specifiedColor=<rgba 1 1 1 0.9>}
    font {
     bForceMonospaced=false
     bMaterialIsStencil=false
     fontMaterial=null
     fontObject=&HeadingNow
     letterSpacing=0
     monospacedWidth=1
     size=16.5
     skewAmount=0
     typefaceFontName=75 Medium
     outlineSettings {
      bApplyOutlineToDropShadows=false
      bMiteredCorners=false
      bSeparateFillAlpha=false
      outlineMaterial=/Game/UI/Foundation/Materials/M-FontShadow
      outlineSize=1
      outlineColor=<rgba 0.0052 0.0052 0.0056 0.5>
     }
    }
   }
   slot {class=SizeBoxSlot, properties {verticalAlignment=VAlign_Center}}
  }
  SizeBox_3 {
   widgetClassPath=&SizeBox
   parent=HBox_Content
   widget=SizeBox_3
   properties {bOverride_WidthOverride=true, maxDesiredWidth=200, minDesiredWidth=200, widthOverride=242}
   slot {class=HorizontalBoxSlot, properties {horizontalAlignment=HAlign_Right, padding=<ltrb 10 0 16 0>}}
  }
  OptionSelectionArea {widgetClassPath=&HorizontalBox, bIsVariable=true, parent=SizeBox_3, widget=OptionSelectionArea, slot {class=SizeBoxSlot}}
  LeftButtonSize {
   widgetClassPath=&SizeBox
   bIsVariable=true
   parent=OptionSelectionArea
   widget=LeftButtonSize
   properties {bOverride_HeightOverride=true, bOverride_WidthOverride=true, heightOverride=30, widthOverride=20}
   slot {class=HorizontalBoxSlot, properties {horizontalAlignment=HAlign_Center, verticalAlignment=VAlign_Center}}
  }
  LeftButton {
   widgetClassPath=/Game/Athena/UI/Matchmaking/SimpleCommonButton
   bIsVariable=true
   parent=LeftButtonSize
   widget=LeftButton
   properties {
    onButtonBaseClicked=(&self:WidgetTree.LeftButton.HandleOnButtonClicked)
    onButtonBaseDoubleClicked=(&self:WidgetTree.LeftButton.HandleOnButtonDoubleClicked)
    onButtonBaseHovered=(&self:WidgetTree.LeftButton.HandleOnButtonHovered)
    onButtonBaseUnhovered=(&self:WidgetTree.LeftButton.HandleOnButtonUnhovered)
    onSelectedChangedBase=(&self:WidgetTree.LeftButton.HandleOnSelectedChanged)
    style=/FortUILibrary/Foundation/Buttons/ButtonStyles/ButtonStyle_SimpleCycleArrow_Left_White.ButtonStyle_SimpleCycleArrow_Left_White_C
    hoveredSlateSoundOverride {resourceObject=/Game/Sounds/Fort_UI_Sounds/FrontEnd/DefaultSounds/PressedButtons/MSS_UI_Button_Hover_Default}
    pressedSlateSoundOverride {resourceObject=/Game/Sounds/Fort_UI_Sounds/FrontEnd/DefaultSounds/PressedButtons/MSS_UI_Button_Pressed_Default}
   }
   slot {class=SizeBoxSlot, properties {horizontalAlignment=HAlign_Center, verticalAlignment=VAlign_Center}}
  }
  SizeBox_2 {
   widgetClassPath=&SizeBox
   parent=OptionSelectionArea
   widget=SizeBox_2
   properties {bOverride_WidthOverride=true, clipping=ClipToBounds, widthOverride=342}
   slot {class=HorizontalBoxSlot, properties {verticalAlignment=VAlign_Center, size {sizeRule=Fill, value=1}}}
  }
  Border_4 {
   widgetClassPath=&Border
   parent=SizeBox_2
   widget=Border_4
   properties {
    horizontalAlignment=HAlign_Center
    background {
     bIsDynamicallyLoaded=false
     drawAs=Image
     imageType=NoImage
     mirroring=NoMirror
     resourceName=None
     resourceObject=null
     tiling=NoTile
     imageSize=<32 32>
     margin=<ltrb 0 0 0 0>
     outlineSettings {
      bUseBrushTransparency=false
      roundingType=HalfHeightRadius
      width=0
      color {colorUseRule=UseColor_Specified, specifiedColor=<rgba 0 0 0 0>}
      cornerRadii=<0 0 0 0>
     }
     tintColor {colorUseRule=UseColor_Specified, specifiedColor=<rgba 0 0.0244 0.0625 0>}
     uVRegion {bIsValid=false, max=<0 0>, min=<0 0>}
    }
    padding=<ltrb 0 0 0 0>
   }
   slot {class=SizeBoxSlot}
  }
  MyText {
   widgetClassPath=&CommonTextBlock
   bInherited=true
   bIsVariable=true
   parent=Border_4
   widget=MyText
   properties {
    scrollStyle=&TextScrollStyle-Base_C
    text=Off
    colorAndOpacity {colorUseRule=UseColor_Specified, specifiedColor=<rgba 1 1 1 0.9>}
    font {
     bForceMonospaced=false
     bMaterialIsStencil=false
     fontMaterial=null
     fontObject=&HeadingNow
     letterSpacing=0
     monospacedWidth=1
     size=16.5
     skewAmount=0
     typefaceFontName=75 Medium
     outlineSettings {
      bApplyOutlineToDropShadows=false
      bMiteredCorners=false
      bSeparateFillAlpha=false
      outlineMaterial=&M-FontShadow
      outlineSize=1
      outlineColor=<rgba 0.0052 0.0052 0.0056 0.5>
     }
    }
   }
   slot {class=BorderSlot, properties {horizontalAlignment=HAlign_Center, padding=<ltrb 0 0 0 0>}}
  }
  RightButtonSize {
   widgetClassPath=&SizeBox
   bIsVariable=true
   parent=OptionSelectionArea
   widget=RightButtonSize
   properties {bOverride_HeightOverride=true, bOverride_WidthOverride=true, heightOverride=30, widthOverride=20}
   slot {class=HorizontalBoxSlot, properties {horizontalAlignment=HAlign_Center, verticalAlignment=VAlign_Center}}
  }
  RightButton {
   widgetClassPath=&SimpleCommonButton
   bIsVariable=true
   parent=RightButtonSize
   widget=RightButton
   properties {
    onButtonBaseClicked=(&self:WidgetTree.RightButton.HandleOnButtonClicked)
    onButtonBaseDoubleClicked=(&self:WidgetTree.RightButton.HandleOnButtonDoubleClicked)
    onButtonBaseHovered=(&self:WidgetTree.RightButton.HandleOnButtonHovered)
    onButtonBaseUnhovered=(&self:WidgetTree.RightButton.HandleOnButtonUnhovered)
    onSelectedChangedBase=(&self:WidgetTree.RightButton.HandleOnSelectedChanged)
    style=/FortUILibrary/Foundation/Buttons/ButtonStyles/ButtonStyle_SimpleCycleArrow_Right_White.ButtonStyle_SimpleCycleArrow_Right_White_C
    hoveredSlateSoundOverride {resourceObject=&MSS_UI_Button_Hover_Default}
    pressedSlateSoundOverride {resourceObject=&MSS_UI_Button_Pressed_Default}
   }
   slot {class=SizeBoxSlot}
  }
  SizeBox_TextEntry {
   widgetClassPath=&SizeBox
   bInherited=true
   bIsVariable=true
   parent=HBox_Content
   widget=SizeBox_TextEntry
   properties {bOverride_MaxDesiredWidth=true, maxDesiredWidth=750, visibility=Collapsed}
   slot {class=HorizontalBoxSlot, properties {horizontalAlignment=HAlign_Right, padding=<ltrb 10 0 0 0>}}
  }
  FilteredTextEntry_Widget {
   widgetClassPath=/FortUILibrary/UIKit/InputField/WBP_UIKit_FilteredInputField
   bInherited=true
   bIsVariable=true
   parent=SizeBox_TextEntry
   widget=FilteredTextEntry_Widget
   slot {class=SizeBoxSlot}
  }
  InputActionWidget {widgetClassPath=&CommonActionWidget, bInherited=true, widget=null}
 ]
}
