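# Widget Blueprint for one Creative options category in the HUD, showing a category heading above a list of that category's options.
# A vertical box stacks an uppercase title text block over a CommonListView that builds rows from AthenaCreativeParameterEntry and has touch scrolling disabled.
# Both inner widgets are inherited from the native FortCreativeOptionsCategory parent, and the event graph holds only unconnected stock events, so the asset supplies just layout and styling.
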WidgetBlueprint /Game/Athena/HUD/GlolbalOptionCategoryListView
graphs [EventGraph {nodes [
   Event PreConstruct()
   Event Construct()
   Event Tick()
  ]}]
variables []
widgets {
 info {inheritedWidgetCount=2, parentClass=/Script/FortniteUI.FortCreativeOptionsCategory, rootWidgetClass=&VerticalBox, widgetCount=3}
 widgets [
  VerticalBox_0 {widgetClassPath=&VerticalBox, widget=VerticalBox_0}
  Text_CategoryText {
   widgetClassPath=&CommonTextBlock
   bInherited=true
   parent=VerticalBox_0
   widget=Text_CategoryText
   properties {
    text=Category
    textTransformPolicy=ToUpper
    font {
     bForceMonospaced=false
     bMaterialIsStencil=false
     fontMaterial=null
     fontObject=&HeadingNow
     letterSpacing=0
     monospacedWidth=1
     size=21
     skewAmount=0
     typefaceFontName=86 Bold
     outlineSettings {
      bApplyOutlineToDropShadows=false
      bMiteredCorners=false
      bSeparateFillAlpha=false
      outlineMaterial=null
      outlineSize=2
      outlineColor=<rgba 0 0 0 0.05>
     }
    }
   }
   slot {class=VerticalBoxSlot}
  }
  ListView_Options {
   widgetClassPath=&CommonListView
   bInherited=true
   bIsVariable=true
   parent=VerticalBox_0
   widget=ListView_Options
   properties {
    bEnableTouchScrolling=false
    clipping=Inherit
    entryWidgetClass=/Game/Athena/UI/InGame/Creative/AthenaCreativeParameterEntry.AthenaCreativeParameterEntry_C
   }
   slot {class=VerticalBoxSlot}
  }
 ]
}
