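# Creative-mode Quick Menu modal, built on FortCreativeGlobalOptions, listing Movement, Object Placement and Tools options, with Show Map and Close buttons.
# Opening plays a skewed slide-in, hides the HUD, centers the cursor and restores the last-used selection on gamepad.
# Closing caches indices and plays the slide-out, then pops the widget into camera mode; deactivation restores the HUD and notifies listeners.
# Custom navigation functions link the category lists, targeting edge entries on gamepad.
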WidgetBlueprint /Game/Athena/HUD/CreativeHUDGlobalOptions_Modal
animations [
 In {
  binding_count=4
  length_seconds=0.35
  object=&self:In
  sequence {bindings [
    CommonCustomNavigation_70 {
     id=#1
     possessed_class=CommonCustomNavigation
     tracks [{
       type=MovieSceneFloatTrack
       display_name=Render Opacity
       sections [{type=MovieSceneFloatSection, end_frame=7, start_frame=0, channels [None {type=MovieSceneScriptingFloatChannel, keys [[0,0],[7,1]]}]}]
      }]
    }
    Border_90 {
     id=#2
     display_name=SizeBox_Shell
     possessed_class=Border
     tracks [{
       type=MovieScene2DTransformTrack
       sections [{
         type=MovieScene2DTransformSection
         end_frame=7
         start_frame=0
         channels [
          Translation.X {type=MovieSceneScriptingFloatChannel, thinned_from=4, keys [[0,-500],[7,0],[11,0]]}
          Translation.Y {type=MovieSceneScriptingFloatChannel, const [0,1]}
          Angle {type=MovieSceneScriptingFloatChannel, const [0,1]}
          Scale.X {type=MovieSceneScriptingFloatChannel, const [1,1]}
          Scale.Y {type=MovieSceneScriptingFloatChannel, const [1,1]}
          Shear.X {type=MovieSceneScriptingFloatChannel, keys [[0,10],[2,10],[7,0],[11,0]]}
          Shear.Y {type=MovieSceneScriptingFloatChannel, keys [[0,-10],[2,-10],[7,0],[11,0]]}
         ]
        }]
      }]
    }
    VerticalBox_Body {
     id=#3
     possessed_class=VerticalBox
     tracks [{
       type=MovieSceneFloatTrack
       display_name=Render Opacity
       sections [{type=MovieSceneFloatSection, end_frame=7, start_frame=0, channels [None {type=MovieSceneScriptingFloatChannel, keys [[0,0],[4,0],[7,1]]}]}]
      }]
    }
    OptionDisplayName {
     id=#4
     possessed_class=CommonTextBlock
     tracks [{
       type=MovieSceneFloatTrack
       display_name=Render Opacity
       sections [{type=MovieSceneFloatSection, end_frame=7, start_frame=3, channels [None {type=MovieSceneScriptingFloatChannel, keys [[4,0],[7,1]]}]}]
      }]
    }
   ]}
 }
 Out {
  binding_count=1
  length_seconds=0.35
  object=&self:Out
  sequence {bindings [Border_90 {
     id=#2
     display_name=SizeBox_Shell
     possessed_class=Border
     tracks [{
       type=MovieScene2DTransformTrack
       sections [{
         type=MovieScene2DTransformSection
         end_frame=7
         start_frame=0
         channels [
          Translation.X {type=MovieSceneScriptingFloatChannel, keys [[0,0],[7,-800]]}
          Translation.Y {type=MovieSceneScriptingFloatChannel, const [0,1]}
          Angle {type=MovieSceneScriptingFloatChannel, const [0,1]}
          Scale.X {type=MovieSceneScriptingFloatChannel, const [1,1]}
          Scale.Y {type=MovieSceneScriptingFloatChannel, const [1,1]}
          Shear.X {type=MovieSceneScriptingFloatChannel, const [0,1]}
          Shear.Y {type=MovieSceneScriptingFloatChannel, const [0,1]}
         ]
        }]
      }]
    }]}
 }
]
graphs [
 CenterMouseCursor {nodes [
   CenterMouseCursor() {is=title K2Node_FunctionEntry}
   n1 = GetPlayerController!(PlayerIndex:int=0) -> (ReturnValue:Player Controller Object Reference)
   > Is Valid(InputObject=@n1)
   n2 = GetViewportSize!(self:Player Controller Object Reference=@n1) -> (SizeX:int=0, SizeY:int=0)
   > Set Mouse Position!(
    execute
    self:Player Controller Object Reference=@n1
    X:int=int / int!(A:int=@n2, B:int=2) -> (ReturnValue:int=0) {is=title K2Node_CallFunction}
    Y:int=int / int(A=@n2.SizeY, B=2) -> (ReturnValue=0)
   ) -> (then) {is=title K2Node_CallFunction}
  ]}
 CloseMenu {nodes [
   CloseMenu() {is=title K2Node_FunctionEntry}
   > n1 = Branch(Condition=IsAnimationPlaying(InAnimation=Get Out() -> (Out:Widget Animation Object Reference)))
   UpdateCachedIndices(execute=@n1.else, self:Fort Creative Global Options Object Reference)
   > UnbindAllFromAnimationFinished!(execute, self:User Widget Object Reference, Animation:Widget Animation Object Reference=Get Out()) -> (then)
   n2 = Get Out()
   > PlayAnimation(InAnimation=@n2, PlaybackSpeed=3)
   > BindToAnimationFinished(Animation=@n2, Delegate=Create Event())
   > Return Node!(execute, bDummyVariable:bool=true)
  ]}
 EventGraph {nodes [
   n1 = SetInputActionHandler!(
    execute
    self:Common Activatable Panel Legacy Object Reference=Self-Reference()
    InputActionRow:Data Table Row Handle Structure=Get SpecialLeft() -> (SpecialLeft:Data Table Row Handle Structure)
    CommitedEvent:Delegate=Create Event()
   ) -> (then)
   SetVisibility(execute=[@n1.then, @n15.then], InVisibility=SelfHitTestInvisible)
   n2 = Get In() -> (In:Widget Animation Object Reference)
   > n3 = PlayAnimation(InAnimation=@n2, PlaybackSpeed=1.5)
   n4 = CenterWidget() {is=title K2Node_CustomEvent}
   BindToAnimationFinished(execute=@n3.then, Animation=@n2, Delegate=@n4)
   > CenterMouseCursor() {is=title K2Node_CallFunction}
   > n5 = Cast To FortPlayerControllerGameplay!(execute, Object=GetOwningPlayer()) -> (then, CastFailed, AsFort Player Controller Gameplay:Fort Player Controller Gameplay Object Reference) {is=title K2Node_DynamicCast}
   > SetHUDElementVisibilityWithContext!(
    execute
    self:Fort Player Controller Object Reference=@n5
    HUDElementTags:"Gameplay Tag Container Structure (by ref)"=Get HideGameplayTag() -> (HideGameplayTag:Gameplay Tag Container Structure)
    bHideElements:bool=true
    ReasonName:Name=None
    LogContext:String
   ) -> (then)
   n6 = "On Button Clicked (Button_Cancel)"!() -> (OutputDelegate, then, Button:Common Button Legacy Object Reference) {is=title K2Node_ComponentBoundEvent}
   CloseMenu!(execute=[@n6.then, @n9.then, @n10.then], self) -> (then, bDummyVariable:bool) {is=title K2Node_CallFunction}
   "On Button Clicked (Button_Accept)"!() -> (OutputDelegate, then, Button:Common Button Legacy Object Reference) {is=title K2Node_ComponentBoundEvent}
   > n7 = Cast To FortPlayerController(Object=GetOwningPlayer())
   > On Out Finished()
   n8 = GetMiniMapComponent!(self:Fort Player Controller Object Reference=@n7) -> (ReturnValue:Fort Controller Component Minimap Object Reference)
   > Is Valid(InputObject=@n8)
   > ToggleFullScreenMap(self:Fort Controller Component Minimap Object Reference=@n8)
   n9 = "On Pressed (InputCatcher)"() {is=title K2Node_ComponentBoundEvent}
   n10 = "On Button Clicked (Button_SecondaryClose2)"!() -> (OutputDelegate, then, Button:Common Button Legacy Object Reference) {is=title K2Node_ComponentBoundEvent}
   On Out Finished() {is=title K2Node_CustomEvent}
   > PopContentWidget!(execute) -> (then)
   > EnterCameraMode(self:Fort HUDContext Object Reference=Get FortHUDContext!() -> (ReturnValue:Fort HUDContext Object Reference) {is=title K2Node_GetBlueprintContextEx})
   n11 = Play Intro Audio() {is=title K2Node_CustomEvent}
   > PlaySound2D(Sound=/Game/Sounds/Fort_UI_Sounds/Gameplay/UI_QuickMenu_Open_Cue)
   n12 = Play Outro Audio() {is=title K2Node_CustomEvent}
   > PlaySound2D(Sound=/Game/Sounds/Fort_UI_Sounds/Gameplay/UI_QuickMenu_Close_Cue)
   Event Construct()
   > BindToAnimationStarted!(execute, self:User Widget Object Reference, Animation:Widget Animation Object Reference=Get In(), Delegate:Delegate=@n11) -> (then)
   > BindToAnimationStarted(Animation=Get Out(), Delegate=@n12)
   n13 = Event InitFromObject!() -> (OutputDelegate, then, InitObject:Object Reference) {is=title K2Node_Event}
   > n14 = Cast To UserWidget!(execute, Object=@n13.InitObject) -> (then, CastFailed, AsUser Widget:User Widget Object Reference) {is=title K2Node_DynamicCast}
   > Set InitObject(InitObject:User Widget Object Reference=@n14) -> (Output_Get:User Widget Object Reference)
   Event On Activated()
   > Call External on Activated!(
    execute:Exec
    self:Creative HUDGlobal Options Modal Object Reference
    GlobalOptions:Creative HUDGlobal Options Modal Object Reference=Self-Reference()
   ) -> (then:Exec) {is=title K2Node_CallDelegate}
   > n15 = "SetText (Text)"(self=Get Text_ToolTipText() -> (Text_ToolTipText:Common Text Object Reference), InText=GetEmptyText!() -> (ReturnValue:Text))
   Event On Deactivated()
   > n16 = Cast To CreativeGlobalOptionsInterface!(execute, Object=Get InitObject() -> (InitObject:User Widget Object Reference)) -> (then, CastFailed, AsCreative Global Options Interface:Creative Global Options Interface Interface) {is=title K2Node_DynamicCast}
   > n17 = OnGlobalOptionsDeactivated(self:Creative Global Options Interface Interface=@n16)
   Call External on Deactivated!(
    execute:Exec=[@n17.then, @n16.CastFailed]
    self:Creative HUDGlobal Options Modal Object Reference
    GlobalOptions:Creative HUDGlobal Options Modal Object Reference=Self-Reference()
   ) -> (then:Exec) {is=title K2Node_CallDelegate}
   > SetVisibility(InVisibility=Collapsed)
   > n18 = Cast To FortPlayerControllerGameplay(Object=GetOwningPlayer())
   > SetHUDElementVisibilityWithContext(self=@n18, HUDElementTags=Get HideGameplayTag(), bHideElements=false, ReasonName=None)
   > ForwardOnCreativeQuickMenuDisplayedChanged!(
    execute
    self:Creative HUDLContext Object Reference=Get CreativeHUDLContext!() -> (ReturnValue:Creative HUDLContext Object Reference) {is=title K2Node_GetBlueprintContextEx}
    bIsDisplayed:bool=false
   ) -> (then)
   NavigateToIndex!(
    execute=@n20.then
    self:List View Object Reference=GetOptionsList!(self:Fort Creative Options Category Object Reference=GetOptionListFromMemoryPointer!(self:Fort Creative Global Options Object Reference) -> (ReturnValue:Fort Creative Options Category Object Reference)) -> (ReturnValue:Common List View Object Reference)
    Index:int=Get CachedLastUsedIndexInCategory() -> (CachedLastUsedIndexInCategory:int=0)
   ) -> (then)
   SetInputActionHandler(
    InputActionRow=Get SecondaryCloseDataTableRowHandle() -> (SecondaryCloseDataTableRowHandle:Data Table Row Handle Structure)
    CommitedEvent=Create Event()
   )
   n19 = IsUsingGamepad!(self:Fort Global UIContext Object Reference=Get FortGlobalUIContext() -> (ReturnValue:Fort Global UIContext Object Reference) {is=title K2Node_GetSubsystem}) -> (ReturnValue:bool)
   n20 = Branch(execute=@n4.then, Condition=@n19)
  ]}
 ExternalOnActivated {nodes [ExternalOnActivated!() -> (then, GlobalOptions:Creative HUDGlobal Options Modal Object Reference) {is=title K2Node_FunctionEntry}]}
 ExternalOnDeactivated {nodes [ExternalOnDeactivated!() -> (then, GlobalOptions:Creative HUDGlobal Options Modal Object Reference) {is=title K2Node_FunctionEntry}]}
 NavigateDownToMoveToolListView {nodes [
   n1 = NavigateDownToMoveToolListView!() -> (then, Navigation:EUINavigation Enum) {is=title K2Node_FunctionEntry}
   > n2 = Switch on EUINavigation!(execute, Selection:EUINavigation Enum=@n1) -> (Left:Exec, Right:Exec, Up:Exec, Down:Exec, Next:Exec, Previous:Exec, Invalid:Exec) {is=title K2Node_SwitchEnum}
   n3 = Branch(execute=@n2.Down, Condition=IsUsingGamepad(self=Get FortGlobalUIContext()))
   n4 = GetOptionsList(self=Get Category_MoveTool() -> (Category_MoveTool:Fort Creative Options Category Object Reference))
   > Return Node!(
    execute
    ReturnValue:Widget Object Reference="Get (a copy)"(Array:Array of User Widget Object References=GetDisplayedEntryWidgets!(self:List View Base Object Reference=@n4) -> (ReturnValue:Array of User Widget Object References)) -> (Output:User Widget Object Reference)
   )
   Return Node(execute=@n3.else, ReturnValue=GetOptionsList(self=Get Category_MoveTool()))
  ]}
 NavigateDownToToolsetListView {nodes [
   n1 = NavigateDownToToolsetListView!() -> (then, Navigation:EUINavigation Enum) {is=title K2Node_FunctionEntry}
   > n2 = Switch on EUINavigation(Selection=@n1)
   n3 = Branch(execute=@n2.Down, Condition=IsUsingGamepad(self=Get FortGlobalUIContext()))
   n4 = GetOptionsList(self=Get Category_Toolset() -> (Category_Toolset:Glolbal Option Category List View Object Reference))
   > Return Node(ReturnValue="Get (a copy)"(Array=GetDisplayedEntryWidgets(self=@n4)))
   Return Node(execute=@n3.else, ReturnValue=GetOptionsList(self=Get Category_Toolset()))
  ]}
 NavigateUpToMoveToolListView {nodes [
   n1 = NavigateUpToMoveToolListView!() -> (then, Navigation:EUINavigation Enum) {is=title K2Node_FunctionEntry}
   > n2 = Switch on EUINavigation(Selection=@n1)
   n3 = Branch(execute=@n2.Up, Condition=IsUsingGamepad(self=Get FortGlobalUIContext()))
   n4 = GetDisplayedEntryWidgets(self=GetOptionsList(self=Get Category_MoveTool()))
   n5 = Length(TargetArray:Array of User Widget Object References=@n4)
   > Return Node(ReturnValue="Get (a copy)"(Array=@n4, Dimension 1=int - int(A=@n5, B=1) {is=title K2Node_CallFunction}))
   Return Node(execute=@n3.else, ReturnValue=GetOptionsList(self=Get Category_MoveTool()))
  ]}
 NavigateUpToPlayerListView {nodes [
   n1 = NavigateUpToPlayerListView!() -> (then, Navigation:EUINavigation Enum) {is=title K2Node_FunctionEntry}
   > n2 = Switch on EUINavigation(Selection=@n1)
   n3 = Branch(execute=@n2.Up, Condition=IsUsingGamepad(self=Get FortGlobalUIContext()))
   n4 = GetDisplayedEntryWidgets(self=GetOptionsList(self=Get Category_Player() -> (Category_Player:Fort Creative Options Category Object Reference)))
   n5 = Length(TargetArray=@n4)
   > Return Node(ReturnValue="Get (a copy)"(Array=@n4, Dimension 1=int - int(A=@n5, B=1)))
   Return Node(execute=@n3.else, ReturnValue=GetOptionsList(self=Get Category_Player()))
  ]}
 TryPushWidgetAsActivatablePanel {nodes [
   n1 = TryPushWidgetAsActivatablePanel!() -> (then, Object:Object Reference, inHideHeader:bool=~, inHideFooter:bool=~, inHideChatWidget:bool=~) {is=title K2Node_FunctionEntry}
   > n2 = Cast To CommonActivatablePanelLegacy!(execute, Object=@n1) -> (then, CastFailed, AsCommon Activatable Panel Legacy:Common Activatable Panel Legacy Object Reference) {is=title K2Node_DynamicCast}
   > n3 = Branch(Condition=IsInZone!(self:Fort Global UIContext Object Reference=Get FortGlobalUIContext!() -> (ReturnValue:Fort Global UIContext Object Reference) {is=title K2Node_GetBlueprintContextEx}) -> (ReturnValue:bool))
   n4 = Get FortHUDContext()
   > EnterCameraMode(self=@n4)
   > EnterCursorMode!(execute, self:Fort HUDContext Object Reference=@n4, ActionName:Name=None, CursorModeWidget:User Widget Object Reference=@n2) -> (then)
   PushContentWidget_Adv!(
    execute=@n3.else
    self:Fort Front End Context Object Reference=Get FortFrontEndContext!() -> (ReturnValue:Fort Front End Context Object Reference) {is=title K2Node_GetBlueprintContextEx}
    Widget:Widget Object Reference=@n2
    bHideHeader:bool=Get inHideHeader() -> (inHideHeader:bool)
    bHideFooter:bool=Get inHideFooter() -> (inHideFooter:bool)
    bHideChatWidget:bool=Get inHideChatWidget() -> (inHideChatWidget:bool)
   ) -> (then)
  ]}
]
variables [
 ExternalOnActivated {}
 ExternalOnDeactivated {}
 HideGameplayTag {type=GameplayTagContainer, default {gameplayTags [{tagName=HUD}]}}
 InitObject {type=UserWidget, default=null}
 SecondaryCloseDataTableRowHandle {type=DataTableRowHandle, default {dataTable=/FortUILibrary/Input/CreativeInputActionDataTable, rowName=CreativeToggleGlobalOptions}}
 SpecialLeft {type=DataTableRowHandle, default {dataTable=&CreativeInputActionDataTable, rowName=CreativeToggleGlobalOptions}}
]
widgets {
 info {
  inheritedWidgetCount=4
  namedSlotCount=1
  parentClass=/Script/FortniteUI.FortCreativeGlobalOptions
  rootWidgetClass=&SafeZone
  widgetCount=33
 }
 widgets [
  SafeZone_1 {widgetClassPath=&SafeZone, bIsVariable=true, widget=SafeZone_1}
  Overlay_1 {widgetClassPath=&Overlay, parent=SafeZone_1, widget=Overlay_1, slot {class=SafeZoneSlot}}
  Lightbox {
   widgetClassPath=/Game/UI/Foundation/Lightbox/Lightbox
   bIsVariable=true
   parent=Overlay_1
   widget=Lightbox
   properties {bConsumePointerInput=true, showWindowControlButtons=false, backgroundColor=<rgba 1 1 1 0>, colorAndOpacity=<rgba 1 1 1 0>}
   slot {class=OverlaySlot, properties {horizontalAlignment=HAlign_Fill, verticalAlignment=VAlign_Fill}}
  }
  InputCatcher {
   widgetClassPath=&Button
   namedSlotHost=Lightbox
   widget=InputCatcher
   properties {renderOpacity=0, backgroundColor=<rgba 1 1 1 0>, colorAndOpacity=<rgba 1 1 1 0>}
  }
  CommonCustomNavigation_70 {
   widgetClassPath=/Script/CommonUI.CommonCustomNavigation
   bIsVariable=true
   parent=Overlay_1
   widget=CommonCustomNavigation_70
   properties {background {
     bIsDynamicallyLoaded=false
     drawAs=NoDrawType
     imageType=NoImage
     mirroring=NoMirror
     resourceName=None
     resourceObject=null
     tiling=NoTile
     imageSize=<32 32>
     margin=<ltrb 0 0 0 0>
     outlineSettings {
      bUseBrushTransparency=false
      roundingType=HalfHeightRadius
      width=0
      color {colorUseRule=UseColor_Specified, specifiedColor=<rgba 0 0 0 0>}
      cornerRadii=<0 0 0 0>
     }
     tintColor {colorUseRule=UseColor_Specified, specifiedColor=<rgba 1 1 1 1>}
     uVRegion {bIsValid=false, max=<0 0>, min=<0 0>}
    }}
   slot {class=OverlaySlot, properties {verticalAlignment=VAlign_Center}}
  }
  SizeBox_Shell {
   widgetClassPath=&SizeBox
   bIsVariable=true
   parent=CommonCustomNavigation_70
   widget=SizeBox_Shell
   properties {bOverride_MinDesiredHeight=true, bOverride_WidthOverride=true, minDesiredHeight=572, widthOverride=640}
   slot {class=BorderSlot, properties {padding=<ltrb 0 0 0 0>}}
  }
  Overlay_23 {widgetClassPath=&Overlay, parent=SizeBox_Shell, widget=Overlay_23, slot {class=SizeBoxSlot}}
  Image_Shadow {
   widgetClassPath=&Image
   parent=Overlay_23
   widget=Image_Shadow
   properties {
    brush {
     bIsDynamicallyLoaded=false
     drawAs=Image
     imageType=NoImage
     mirroring=NoMirror
     resourceName=None
     resourceObject=/Game/Athena/UI/InGame/Creative/Art/M_SimpleSkewBox_Cr_QuickMenuShadow
     tiling=NoTile
     imageSize=<32 32>
     margin=<ltrb 0 0 0 0>
     outlineSettings {
      bUseBrushTransparency=false
      roundingType=HalfHeightRadius
      width=0
      color {colorUseRule=UseColor_Specified, specifiedColor=<rgba 0 0 0 0>}
      cornerRadii=<0 0 0 0>
     }
     tintColor {colorUseRule=UseColor_Specified, specifiedColor=<rgba 1 1 1 1>}
     uVRegion {bIsValid=false, max=<0 0>, min=<0 0>}
    }
    colorAndOpacity=<rgba 0 0 0 0.6>
    renderTransform {angle=0, scale=<1.3287 1.35>, shear=<0 0>, translation=<9.1124 2.2896>}
   }
   slot {class=OverlaySlot, properties {horizontalAlignment=HAlign_Fill, verticalAlignment=VAlign_Fill, padding=<ltrb 0 24 0 0>}}
  }
  Image_366 {
   widgetClassPath=&Image
   parent=Overlay_23
   widget=Image_366
   properties {brush {
     bIsDynamicallyLoaded=false
     drawAs=Image
     imageType=NoImage
     mirroring=NoMirror
     resourceName=None
     resourceObject=/Game/Athena/UI/InGame/Creative/Art/M_UI_BoltGradientBorder_CreativeQuickMenuBase
     tiling=NoTile
     imageSize=<32 32>
     margin=<ltrb 0 0 0 0>
     outlineSettings {
      bUseBrushTransparency=false
      roundingType=HalfHeightRadius
      width=0
      color {colorUseRule=UseColor_Specified, specifiedColor=<rgba 0 0 0 0>}
      cornerRadii=<0 0 0 0>
     }
     tintColor {colorUseRule=UseColor_Specified, specifiedColor=<rgba 1 1 1 1>}
     uVRegion {bIsValid=false, max=<0 0>, min=<0 0>}
    }}
   slot {class=OverlaySlot, properties {horizontalAlignment=HAlign_Fill, verticalAlignment=VAlign_Fill, padding=<ltrb 0 68 0 0>}}
  }
  VerticalBox_1 {
   widgetClassPath=&VerticalBox
   parent=Overlay_23
   widget=VerticalBox_1
   slot {class=OverlaySlot, properties {horizontalAlignment=HAlign_Fill, verticalAlignment=VAlign_Fill}}
  }
  Overlay_184 {widgetClassPath=&Overlay, parent=VerticalBox_1, widget=Overlay_184, slot {class=VerticalBoxSlot}}
  Image_279 {
   widgetClassPath=&Image
   parent=Overlay_184
   widget=Image_279
   properties {brush {
     bIsDynamicallyLoaded=false
     drawAs=Image
     imageType=NoImage
     mirroring=NoMirror
     resourceName=None
     resourceObject=/Game/Athena/UI/InGame/Creative/Art/M_UI_BoltGradientBorder_CreativeQuickMenu
     tiling=NoTile
     imageSize=<32 76>
     margin=<ltrb 0 0 0 0>
     outlineSettings {
      bUseBrushTransparency=false
      roundingType=HalfHeightRadius
      width=0
      color {colorUseRule=UseColor_Specified, specifiedColor=<rgba 0 0 0 0>}
      cornerRadii=<0 0 0 0>
     }
     tintColor {colorUseRule=UseColor_Specified, specifiedColor=<rgba 1 1 1 1>}
     uVRegion {bIsValid=false, max=<0 0>, min=<0 0>}
    }}
   slot {class=OverlaySlot, properties {horizontalAlignment=HAlign_Fill, verticalAlignment=VAlign_Fill}}
  }
  OptionDisplayName {
   widgetClassPath=&CommonTextBlock
   bIsVariable=true
   parent=Overlay_184
   widget=OptionDisplayName
   properties {
    text=Quick Menu
    textTransformPolicy=ToUpper
    font {
     bForceMonospaced=false
     bMaterialIsStencil=false
     fontMaterial=null
     fontObject=&BurbankSmall
     letterSpacing=0
     monospacedWidth=1
     size=24
     skewAmount=0
     typefaceFontName=Black
     outlineSettings {
      bApplyOutlineToDropShadows=false
      bMiteredCorners=false
      bSeparateFillAlpha=false
      outlineMaterial=null
      outlineSize=0
      outlineColor=<rgba 0.0052 0.0052 0.0056 1>
     }
    }
   }
   slot {class=OverlaySlot, properties {horizontalAlignment=HAlign_Fill, verticalAlignment=VAlign_Bottom, padding=<ltrb 12 0 24 6>}}
  }
  CommonVisibilityWidget_1 {
   widgetClassPath=&CommonVisibilityWidgetLegacy
   parent=Overlay_184
   widget=CommonVisibilityWidget_1
   properties {bShowForGamepad=false}
   slot {class=OverlaySlot, properties {horizontalAlignment=HAlign_Right, verticalAlignment=VAlign_Bottom, padding=<ltrb 0 0 12 16>}}
  }
  SizeBox_1 {
   widgetClassPath=&SizeBox
   parent=CommonVisibilityWidget_1
   widget=SizeBox_1
   properties {bOverride_HeightOverride=true, bOverride_WidthOverride=true, heightOverride=28, widthOverride=28}
   slot {class=BorderSlot}
  }
  Button_SecondaryClose2 {
   widgetClassPath=&IconTextButton
   bIsVariable=true
   parent=SizeBox_1
   widget=Button_SecondaryClose2
   properties {
    bHideInputAction=true
    bIsFocusable=false
    buttonText=""
    onButtonBaseClicked=(&self:WidgetTree.Button_SecondaryClose2.HandleOnButtonClicked)
    onButtonBaseDoubleClicked=(&self:WidgetTree.Button_SecondaryClose2.HandleOnButtonDoubleClicked)
    onButtonBaseHovered=(&self:WidgetTree.Button_SecondaryClose2.HandleOnButtonHovered)
    onButtonBaseUnhovered=(&self:WidgetTree.Button_SecondaryClose2.HandleOnButtonUnhovered)
    onSelectedChangedBase=(&self:WidgetTree.Button_SecondaryClose2.HandleOnSelectedChanged)
    style=/FortUILibrary/Foundation/Buttons/ButtonStyles/ButtonStyle_MobileCloseX.ButtonStyle_MobileCloseX_C
    triggeringInputAction {dataTable=&CreativeInputActionDataTable, rowName=CreativeToggleGlobalOptions}
   }
   slot {class=SizeBoxSlot}
  }
  VerticalBox_Body {
   widgetClassPath=&VerticalBox
   bIsVariable=true
   parent=VerticalBox_1
   widget=VerticalBox_Body
   slot {class=VerticalBoxSlot, properties {size {sizeRule=Fill, value=1}}}
  }
  ScrollBox_0 {
   widgetClassPath=&ScrollBox
   parent=VerticalBox_Body
   widget=ScrollBox_0
   properties {navigationDestination=Center, scrollWhenFocusChanges=AnimatedScroll}
   slot {class=VerticalBoxSlot, properties {padding=<ltrb 12 12 12 12>, size {sizeRule=Fill, value=1}}}
  }
  VerticalBox_2 {
   widgetClassPath=&VerticalBox
   parent=ScrollBox_0
   widget=VerticalBox_2
   properties {clipping=OnDemand}
   slot {class=ScrollBoxSlot}
  }
  Category_Player {
   widgetClassPath=&/GlolbalOptionCategoryListView
   bInherited=true
   bIsVariable=true
   parent=VerticalBox_2
   widget=Category_Player
   properties {
    navigation=&self:WidgetTree.Category_Player.WidgetNavigation_0
    optionCategory {categoryDisplayName=Movement, categoryTag=player}
   }
   slot {class=VerticalBoxSlot}
  }
  Category_MoveTool {
   widgetClassPath=&GlolbalOptionCategoryListView
   bInherited=true
   bIsVariable=true
   parent=VerticalBox_2
   widget=Category_MoveTool
   properties {
    navigation=&self:WidgetTree.Category_MoveTool.WidgetNavigation_0
    optionCategory {categoryDisplayName=Object Placement, categoryTag=movetool}
   }
   slot {class=VerticalBoxSlot, properties {padding=<ltrb 0 12 0 0>}}
  }
  Category_Toolset {
   widgetClassPath=&GlolbalOptionCategoryListView
   bIsVariable=true
   parent=VerticalBox_2
   widget=Category_Toolset
   properties {
    navigation=&self:WidgetTree.Category_Toolset.WidgetNavigation_0
    optionCategory {categoryDisplayName=Tools, categoryTag=toolset}
   }
   slot {class=VerticalBoxSlot, properties {padding=<ltrb 0 24 0 0>}}
  }
  SizeBox_0 {
   widgetClassPath=&SizeBox
   parent=VerticalBox_Body
   widget=SizeBox_0
   properties {bOverride_HeightOverride=true, heightOverride=132.554}
   slot {class=VerticalBoxSlot}
  }
  Overlay_2 {widgetClassPath=&Overlay, parent=SizeBox_0, widget=Overlay_2, slot {class=SizeBoxSlot}}
  Image_160 {
   widgetClassPath=&Image
   parent=Overlay_2
   widget=Image_160
   properties {brush {
     bIsDynamicallyLoaded=false
     drawAs=Image
     imageType=NoImage
     mirroring=NoMirror
     resourceName=None
     resourceObject=/Game/Athena/UI/InGame/Creative/Art/M_UI_BoltGradientBorder_Creative_QuickMenu
     tiling=NoTile
     imageSize=<32 76>
     margin=<ltrb 0 0 0 0>
     outlineSettings {
      bUseBrushTransparency=false
      roundingType=HalfHeightRadius
      width=0
      color {colorUseRule=UseColor_Specified, specifiedColor=<rgba 0 0 0 0>}
      cornerRadii=<0 0 0 0>
     }
     tintColor {colorUseRule=UseColor_Specified, specifiedColor=<rgba 1 1 1 1>}
     uVRegion {bIsValid=false, max=<0 0>, min=<0 0>}
    }}
   slot {class=OverlaySlot, properties {horizontalAlignment=HAlign_Fill, verticalAlignment=VAlign_Fill}}
  }
  VerticalBox_268 {
   widgetClassPath=&VerticalBox
   parent=Overlay_2
   widget=VerticalBox_268
   slot {class=OverlaySlot, properties {horizontalAlignment=HAlign_Fill, verticalAlignment=VAlign_Fill}}
  }
  Overlay_182 {
   widgetClassPath=&Overlay
   parent=VerticalBox_268
   widget=Overlay_182
   slot {class=VerticalBoxSlot, properties {verticalAlignment=VAlign_Center, size {sizeRule=Fill, value=1}}}
  }
  Text_ToolTipText {
   widgetClassPath=&CommonTextBlock
   bInherited=true
   bIsVariable=true
   parent=Overlay_182
   widget=Text_ToolTipText
   properties {
    scrollStyle=&TextScrollStyle-Base_C
    text="[TooltipText]"
    colorAndOpacity {colorUseRule=UseColor_Specified, specifiedColor=<rgba 1 1 1 0.8>}
    font {
     bForceMonospaced=false
     bMaterialIsStencil=false
     fontMaterial=/Game/Athena/UI/Challenges/Art/M_UI_Challenge_DynamicColor_Text_WhiteFill
     fontObject=&BurbankSmall
     letterSpacing=0
     monospacedWidth=1
     size=18
     skewAmount=0
     typefaceFontName=Black
     outlineSettings {
      bApplyOutlineToDropShadows=false
      bMiteredCorners=false
      bSeparateFillAlpha=false
      outlineMaterial=null
      outlineSize=0
      outlineColor=<rgba 0 0 0 1>
     }
    }
    shadowOffset=<0 0>
   }
   slot {class=OverlaySlot, properties {horizontalAlignment=HAlign_Fill, padding=<ltrb 24 16 24 0>}}
  }
  Overlay_0 {widgetClassPath=&Overlay, parent=VerticalBox_268, widget=Overlay_0, slot {class=VerticalBoxSlot}}
  HorizontalBox_0 {
   widgetClassPath=&HorizontalBox
   parent=Overlay_0
   widget=HorizontalBox_0
   slot {class=OverlaySlot, properties {horizontalAlignment=HAlign_Fill, padding=<ltrb 0 8 8 8>}}
  }
  Button_Accept {
   widgetClassPath=&IconTextButton
   bIsVariable=true
   parent=HorizontalBox_0
   widget=Button_Accept
   properties {
    bIsFocusable=false
    buttonText=Show Map
    minWidth=200
    onButtonBaseClicked=(&self:WidgetTree.Button_Accept.HandleOnButtonClicked)
    onButtonBaseDoubleClicked=(&self:WidgetTree.Button_Accept.HandleOnButtonDoubleClicked)
    onButtonBaseHovered=(&self:WidgetTree.Button_Accept.HandleOnButtonHovered)
    onButtonBaseUnhovered=(&self:WidgetTree.Button_Accept.HandleOnButtonUnhovered)
    onSelectedChangedBase=(&self:WidgetTree.Button_Accept.HandleOnSelectedChanged)
    style=/FortUILibrary/Foundation/Buttons/ButtonStyles/Athena/Athena_ButtonStyle_AngledBlueMenuButton_Small.Athena_ButtonStyle_AngledBlueMenuButton_Small_C
    visibility=Visible
    triggeringInputAction {dataTable=&InputActionDataTable, rowName=GenericFaceButtonTop}
   }
   slot {class=HorizontalBoxSlot, properties {horizontalAlignment=HAlign_Right, verticalAlignment=VAlign_Center, padding=<ltrb 24 0 0 0>}}
  }
  Overlay_3 {
   widgetClassPath=&Overlay
   parent=HorizontalBox_0
   widget=Overlay_3
   slot {class=HorizontalBoxSlot, properties {horizontalAlignment=HAlign_Right, verticalAlignment=VAlign_Center, size {sizeRule=Fill, value=1}}}
  }
  Button_Cancel {
   widgetClassPath=&IconTextButton
   bInherited=true
   bIsVariable=true
   parent=Overlay_3
   widget=Button_Cancel
   properties {
    bIsFocusable=false
    buttonText=Close
    minWidth=200
    onButtonBaseClicked=(&self:WidgetTree.Button_Cancel.HandleOnButtonClicked)
    onButtonBaseDoubleClicked=(&self:WidgetTree.Button_Cancel.HandleOnButtonDoubleClicked)
    onButtonBaseHovered=(&self:WidgetTree.Button_Cancel.HandleOnButtonHovered)
    onButtonBaseUnhovered=(&self:WidgetTree.Button_Cancel.HandleOnButtonUnhovered)
    onSelectedChangedBase=(&self:WidgetTree.Button_Cancel.HandleOnSelectedChanged)
    style=&Athena_ButtonStyle_AngledBlueMenuButton_Small_C
    visibility=Visible
    foregroundColor {colorUseRule=UseColor_Foreground, specifiedColor=<rgba 1 0.8289 0 1>}
    triggeringInputAction {dataTable=&CreativeInputActionDataTable, rowName=Close}
   }
   slot {class=OverlaySlot}
  }
 ]
}
