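# CreativeHUDControls is the Creative mode HUD widget showing contextual control hints for flying and the move tool.
# Player, pawn, vehicle and move tool delegates set per-panel visibility states and push tool state into five child panels, refreshed through UpdateVisibilities.
# It polls by timer for the pawn, rebinding on possession changes.
# It also opens the async-created global options modal, plays pitch-shifted feedback sounds and hides menu keybinds whose keys active panels use.
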WidgetBlueprint /Game/Athena/HUD/CreativeHUDControls
graphs [
 CanWeFlyRightNow {nodes [
   CanWeFlyRightNow() {is=title K2Node_FunctionEntry}
   n1 = GetOwningPlayer()
   > Is Valid(InputObject=@n1)
   > n2 = Cast To FortPlayerPawnAthena(Object=Get Controlled Pawn!(self:Controller Object Reference=@n1) -> (ReturnValue:Pawn Object Reference) {is=title K2Node_CallFunction})
   n3 = IsDrivingVehicle!(self:Fort Player Pawn Object Reference=@n2) -> (ReturnValue:bool)
   n4 = IsDBNO!(self:Fort Pawn Object Reference=@n2) -> (ReturnValue:bool)
   n5 = IsPassengerInVehicle!(self:Fort Player Pawn Object Reference=@n2) -> (ReturnValue:bool)
   > n6 = Set Local_CanFly(Local_CanFly:bool=NOT Boolean(A=OR Boolean!(A=@n3, B=@n4, C:bool=@n5) -> (ReturnValue))) -> (Output_Get:bool)
   Return Node!(execute=[@n6.then, @n2.CastFailed], CanFly:bool=Get Local_CanFly() -> (Local_CanFly:bool))
  ]}
 EventGraph {nodes [
   n1 = OnWeaponEquippedDelegate_Bind!() -> (OutputDelegate, then, NewWeapon:Fort Weapon Object Reference, PrevWeapon:Fort Weapon Object Reference) {is=title K2Node_CustomEvent}
   > n2 = Cast To FortCreativeMoveTool!(execute, Object=@n1.NewWeapon) -> (then, CastFailed, AsFort Creative Move Tool:Fort Creative Move Tool Object Reference) {is=title K2Node_DynamicCast}
   > ShowEquippedControls()
   > n3 = Sequence()
   > Bind Event to On Move Tool Interaction Started(self:Fort Creative Move Tool Object Reference=@n2, Delegate=Create Event()) {is=title K2Node_AddDelegate}
   > Bind Event to On Move Tool Interaction Stopped(self:Fort Creative Move Tool Object Reference=@n2, Delegate=Create Event()) {is=title K2Node_AddDelegate}
   > Bind Event to On Move Tool Drop to Floor Changed(self:Fort Creative Move Tool Object Reference=@n2, Delegate=Create Event()) {is=title K2Node_AddDelegate}
   > Bind Event to On Move Tool Precision Changed(self:Fort Creative Move Tool Object Reference=@n2, Delegate=Create Event()) {is=title K2Node_AddDelegate}
   > Bind Event to On Move Tool Rotation Axis Changed(self:Fort Creative Move Tool Object Reference=@n2, Delegate=Create Event()) {is=title K2Node_AddDelegate}
   > Assign On Move Tool Line Of Sight Blocking Changed!(execute:Exec, self:Fort Creative Move Tool Object Reference=@n2, Delegate:"Delegate (by ref)"=Create Event()) -> (then:Exec) {is=title K2Node_AssignDelegate}
   > Bind Event to On Move Tool Transformation Mode Changed(self:Fort Creative Move Tool Object Reference=@n2, Delegate=Create Event()) {is=title K2Node_AddDelegate}
   > Bind Event to On Move Tool Scale Instead Of Rotate Changed(self:Fort Creative Move Tool Object Reference=@n2, Delegate=Create Event()) {is=title K2Node_AddDelegate}
   > Bind Event to On Move Tool Scale Axis Changed(self:Fort Creative Move Tool Object Reference=@n2, Delegate=Create Event()) {is=title K2Node_AddDelegate}
   > Bind Event to On Added Selected Object(self:Fort Creative Move Tool Object Reference=@n2, Delegate=Create Event()) {is=title K2Node_AddDelegate}
   > Bind Event to On Removed Selected Object(self:Fort Creative Move Tool Object Reference=@n2, Delegate=Create Event()) {is=title K2Node_AddDelegate}
   > Bind Event to On Selection Property Changed(self:Fort Creative Move Tool Object Reference=@n2, Delegate=Create Event()) {is=title K2Node_AddDelegate}
   > Assign On Selected Objects Cleared!(execute:Exec, self:Fort Creative Move Tool Object Reference=@n2, Delegate:"Delegate (by ref)"=Create Event()) -> (then:Exec) {is=title K2Node_AssignDelegate}
   > Bind Event to On Move Tool Focus Changed(self:Fort Creative Move Tool Object Reference=@n2, Delegate=Create Event()) {is=title K2Node_AddDelegate}
   UpdateMaxSelectable!(
    execute=@n3.then_1
    self:Creative HUDControls Equipped Object Reference=Get CreativeHUDControlsEquipped() -> (CreativeHUDControlsEquipped:Creative HUDControls Equipped Object Reference)
    Max:int=GetMultiselectLimit!(self:Fort Creative Move Tool Object Reference=@n2) -> (ReturnValue:int=0)
   ) -> (then)
   > OnMoveToolPrecisionChanged_Bind!(
    execute
    self
    PrecisionOn:bool=GetShouldUsePrecisionGridSnapping!(self:Fort Creative Move Tool Object Reference=@n2) -> (ReturnValue:bool)
    GridSnapIndex:Byte=GetGridSnapIndex!(self:Fort Creative Move Tool Object Reference=@n2) -> (ReturnValue:Byte=0)
   ) -> (then)
   > Set DropsDisabled!(
    execute
    DropsDisabled:bool=IsDropsDisabled!(self:Fort Creative Move Tool Object Reference=@n2) -> (ReturnValue:bool)
    self:Creative HUDControls Move Freely Object Reference=Get CreativeHUDControlsMoveFreely() -> (CreativeHUDControlsMoveFreely:Creative HUDControls Move Freely Object Reference)
   ) -> (then, Output_Get:bool)
   > OnMoveToolLineOfSightBlockingChanged_Bind!(
    execute
    self
    bIgnoreLineOfSightBlockers:EHitTraceRule Enum=GetCurrentHitTraceRule!(self:Fort Creative Move Tool Object Reference=@n2) -> (ReturnValue:EHitTraceRule Enum=Visibility)
   ) -> (then)
   ShowUnequippedControls(execute=@n2.CastFailed)
   OnKeybindsChanged_Bind() {is=title K2Node_CustomEvent}
   > OnKeybindsChanged!(
    execute
    self:Creative HUDControls Equipped Object Reference=Get CreativeHUDControlsEquipped()
    IsFlying:bool=Get IsFlying() -> (IsFlying:bool)
   ) -> (then)
   > OnKeybindsChanged(self:Creative HUDControls Move Freely Object Reference=Get CreativeHUDControlsMoveFreely(), IsFlying=Get IsFlying())
   > OnKeybindsChanged(
    self:Creative HUDControls Move On Grid Object Reference=Get CreativeHUDControlsMoveOnGrid() -> (CreativeHUDControlsMoveOnGrid:Creative HUDControls Move On Grid Object Reference)
    IsFlying=Get IsFlying()
   )
   > OnKeybindsChanged(
    self:Creative HUDControls Not Equipped Object Reference=Get CreativeHUDControlsNotEquipped() -> (CreativeHUDControlsNotEquipped:Creative HUDControls Not Equipped Object Reference)
    IsFlying=Get IsFlying()
   )
   > OnKeybindsChanged!(execute, self:Self Object Reference) -> (then)
   n4 = OnCreativeQuickbarFocusedOnUserPrefabSlotChanged!() -> (OutputDelegate, then, IsUserPrefabSlotFocused:bool=~, IsUserPrefabFocusedSlotEmpty:bool=~) {is=title K2Node_CustomEvent}
   n5 = Get CreativeHUDControlsEquipped()
   > Set IsFocusedOnUserPrefab!(execute, IsFocusedOnUserPrefab:bool=@n4.IsUserPrefabSlotFocused, self:Creative HUDControls Equipped Object Reference=@n5) -> (then, Output_Get:bool)
   > Set IsUserPrefabFocusedSlotEmpty!(execute, IsUserPrefabFocusedSlotEmpty:bool=@n4.IsUserPrefabFocusedSlotEmpty, self:Creative HUDControls Equipped Object Reference=@n5) -> (then, Output_Get:bool)
   > UpdateVisibilities()
   n6 = OnWorldInventoryChanged_Bind!() -> (OutputDelegate, then, bIsAllowedToFly:bool=~) {is=title K2Node_CustomEvent}
   > n7 = Set IsAllowedToFly(IsAllowedToFly:bool=@n6.bIsAllowedToFly) -> (Output_Get:bool)
   > n8 = CanWeFlyRightNow!(execute, self) -> (then, CanFly:bool) {is=title K2Node_CallFunction}
   > n9 = Branch(Condition=AND Boolean(A=@n8, B=@n7))
   > n10 = SetVisibility(InVisibility=HitTestInvisible)
   UpdateVisibilities(execute=[@n10.then, @n11.then])
   n11 = SetVisibility(execute=@n9.else, InVisibility=Collapsed)
   UpdateVisibilities() {is=title K2Node_CustomEvent}
   > n12 = Sequence!(execute) -> (then_0, then_1, then_2:Exec, then_3:Exec, then_4:Exec, then_5:Exec, then_6:Exec)
   n13 = Get CreativeHUDControlsEquipped()
   n14 = Get EquippedVisibility() -> (EquippedVisibility:ESlateVisibility Enum=Visible)
   SetVisibility(execute=@n12.then_1, self=@n13, InVisibility=@n14)
   > n15 = Branch(Condition="Not Equal (Enum)"!(A:ESlateVisibility Enum=@n14, B:ESlateVisibility Enum=Collapsed) -> (ReturnValue:bool))
   > Update!(execute, self:Creative HUDControls Equipped Object Reference=@n13, IsFlying:bool=Get IsFlying()) -> (then)
   > ShowFeedback(self:Athena HUDSituational Feedback Object Reference=@n13)
   HideFeedback(execute=@n15.else, self:Athena HUDSituational Feedback Object Reference=@n13)
   n16 = Get CreativeHUDControlsNotEquipped()
   n17 = Get NotEquippedVisibility() -> (NotEquippedVisibility:ESlateVisibility Enum=Visible)
   SetVisibility(execute=@n12.then_2, self=@n16, InVisibility=@n17)
   > n18 = Branch(Condition="Not Equal (Enum)"(A=@n17, B=Collapsed))
   > Update(self:Creative HUDControls Not Equipped Object Reference=@n16, IsFlying=Get IsFlying())
   > ShowFeedback(self=@n16)
   HideFeedback(execute=@n18.else, self=@n16)
   n19 = Get CreativeHUDControlsMoveOnGrid()
   n20 = Get MoveBuildingsOnGridVisibility() -> (MoveBuildingsOnGridVisibility:ESlateVisibility Enum=Visible)
   SetVisibility(execute=@n12.then_3, self=@n19, InVisibility=@n20)
   > n21 = Branch(Condition="Not Equal (Enum)"(A=@n20, B=Collapsed))
   > ClearAndSetFeedbackLines(self:Creative HUDControls Move On Grid Object Reference=@n19)
   > ShowFeedback(self=@n19)
   HideFeedback(execute=@n21.else, self=@n19)
   OnKeybindsChanged(execute=@n12.then_6)
   > n22 = Cast To FortPlayerControllerAthena(Object=GetOwningPlayer())
   n23 = IsPlayerInAMinigame!(self:Fort Player Controller Athena Object Reference=@n22) -> (ReturnValue:bool)
   n24 = NOT Boolean(A=IsCreativeQuickmenuEnabled!(self:Fort Player Controller Gameplay Object Reference=@n22) -> (ReturnValue:bool))
   > SetVisibility(
    self=Get MenusBorder() -> (MenusBorder:Border Object Reference)
    InVisibility=Select(Option 0=SelfHitTestInvisible, Option 1=Collapsed, Index=OR Boolean!(A=@n24, B=@n23) -> (ReturnValue))
   )
   n25 = Cast To FortPlayerControllerGameplay!(execute=@n12.then_0, Object=GetOwningPlayer()) -> (then, CastFailed, AsFort Player Controller Gameplay:Fort Player Controller Gameplay Object Reference) {is=title K2Node_DynamicCast}
   > HideOrShowCreativeHUDControls!(execute, self, PlayerController:Fort Player Controller Gameplay Object Reference=@n25) -> (then, IsVisibile:bool)
   n26 = Get CreativeHUDControlsMoveFreely()
   n27 = Get MoveObjectsFreelyVisibility() -> (MoveObjectsFreelyVisibility:ESlateVisibility Enum=Visible)
   SetVisibility(execute=@n12.then_4, self=@n26, InVisibility=@n27)
   > n28 = Branch(Condition="Not Equal (Enum)"(A=@n27, B=Collapsed))
   > Update(self:Creative HUDControls Move Freely Object Reference=@n26, IsFlying=false)
   > ShowFeedback(self=@n26)
   HideFeedback(execute=@n28.else, self=@n26)
   n29 = Get PlaysetPreviewVisibility() -> (PlaysetPreviewVisibility:ESlateVisibility Enum=Visible)
   n30 = Get CreativeHUDControlPlaysetPreview() -> (CreativeHUDControlPlaysetPreview:Creative HUDControl Playset Preview Object Reference)
   SetVisibility(execute=@n12.then_5, self=@n30, InVisibility=@n29)
   > n31 = Branch(Condition="Not Equal (Enum)"(A=@n29, B=Collapsed))
   > ShowFeedback(self=@n30)
   HideFeedback(execute=@n31.else, self=@n30)
   ShowUnequippedControls() {is=title K2Node_CustomEvent}
   > n32 = Sequence!(execute) -> (then_0, then_1, then_2)
   > Set NotEquippedVisibility(NotEquippedVisibility:ESlateVisibility Enum=HitTestInvisible) -> (Output_Get:ESlateVisibility Enum=Visible)
   Set EquippedVisibility(execute=@n32.then_1, EquippedVisibility:ESlateVisibility Enum=Collapsed) -> (Output_Get:ESlateVisibility Enum=Visible)
   > Set MoveBuildingsOnGridVisibility(MoveBuildingsOnGridVisibility:ESlateVisibility Enum=Collapsed) -> (Output_Get:ESlateVisibility Enum=Visible)
   > Set MoveObjectsFreelyVisibility(MoveObjectsFreelyVisibility:ESlateVisibility Enum=Collapsed) -> (Output_Get:ESlateVisibility Enum=Visible)
   > Set PlaysetPreviewVisibility(PlaysetPreviewVisibility:ESlateVisibility Enum=Collapsed) -> (Output_Get:ESlateVisibility Enum=Visible)
   UpdateVisibilities(execute=@n32.then_2) {is=title K2Node_CallFunction}
   ShowMoveBuildingsOnGridControls() {is=title K2Node_CustomEvent}
   > n33 = Sequence()
   > Set MoveBuildingsOnGridVisibility(MoveBuildingsOnGridVisibility=HitTestInvisible) -> (Output_Get=Visible)
   Set EquippedVisibility(execute=@n33.then_1, EquippedVisibility=Collapsed) -> (Output_Get=Visible)
   > Set NotEquippedVisibility(NotEquippedVisibility=Collapsed) -> (Output_Get=Visible)
   > Set MoveObjectsFreelyVisibility(MoveObjectsFreelyVisibility=Collapsed) -> (Output_Get=Visible)
   > Set PlaysetPreviewVisibility(PlaysetPreviewVisibility=Collapsed) -> (Output_Get=Visible)
   UpdateVisibilities(execute=@n33.then_2)
   OnExitVehicle_Bind() {is=title K2Node_CustomEvent}
   > n34 = Branch(Condition=Get IsAllowedToFly() -> (IsAllowedToFly:bool))
   > n35 = SetVisibility(InVisibility=HitTestInvisible)
   UpdateVisibilities(execute=[@n35.then, @n34.else])
   OnEnterVehicle_Bind() {is=title K2Node_CustomEvent}
   > SetVisibility(InVisibility=Collapsed)
   > Set IsFlying(IsFlying:bool=false) -> (Output_Get:bool)
   ShowMoveObjectsFreelyControls() {is=title K2Node_CustomEvent}
   > n36 = Sequence()
   > Set MoveObjectsFreelyVisibility(MoveObjectsFreelyVisibility=HitTestInvisible) -> (Output_Get=Visible)
   Set EquippedVisibility(execute=@n36.then_1, EquippedVisibility=Collapsed) -> (Output_Get=Visible)
   > Set NotEquippedVisibility(NotEquippedVisibility=Collapsed) -> (Output_Get=Visible)
   > Set MoveBuildingsOnGridVisibility(MoveBuildingsOnGridVisibility=Collapsed) -> (Output_Get=Visible)
   > Set PlaysetPreviewVisibility(PlaysetPreviewVisibility=Collapsed) -> (Output_Get=Visible)
   UpdateVisibilities(execute=@n36.then_2)
   n37 = OnMoveToolLineOfSightBlockingChanged_Bind!() -> (OutputDelegate, then, bIgnoreLineOfSightBlockers:EHitTraceRule Enum=Visibility) {is=title K2Node_CustomEvent}
   > Set Phasing!(
    execute
    Phasing:EHitTraceRule Enum=@n37.bIgnoreLineOfSightBlockers
    self:Creative HUDControls Move Freely Object Reference=Get CreativeHUDControlsMoveFreely()
   ) -> (then, Output_Get:EHitTraceRule Enum=Visibility)
   > UpdateVisibilities()
   WaitForPawn() {is=title K2Node_CustomEvent}
   > UpdateVisibilities()
   n38 = GetOwningPlayer()
   > n39 = Cast To FortPlayerControllerGameplay(Object=@n38)
   > Setup Controls Overall Initial Visibility!(execute:Exec, PlayerController:Fort Player Controller Gameplay Object Reference=@n39) -> (then:Exec) {is=title K2Node_MacroInstance}
   > n40 = Cast To FortPawn!(execute, Object=Get Controlled Pawn(self=@n38)) -> (then, CastFailed, AsFort Pawn:Fort Pawn Object Reference) {is=title K2Node_DynamicCast}
   > Bind Event to On Weapon Equipped Delegate(self:Fort Pawn Object Reference=@n40, Delegate=Create Event()) {is=title K2Node_AddDelegate}
   > Clear and Invalidate Timer by Handle(Handle=Get WaitForPawnTimerHandle() -> (WaitForPawnTimerHandle:Timer Handle Structure))
   > Bind Event to On Possessed Pawn Changed(self:Controller Object Reference=@n38, Delegate=Create Event()) {is=title K2Node_AddDelegate}
   ShowEquippedControls() {is=title K2Node_CustomEvent}
   > n41 = Sequence()
   > Set EquippedVisibility(EquippedVisibility=HitTestInvisible) -> (Output_Get=Visible)
   Set NotEquippedVisibility(execute=@n41.then_1, NotEquippedVisibility=Collapsed) -> (Output_Get=Visible)
   > Set MoveBuildingsOnGridVisibility(MoveBuildingsOnGridVisibility=Collapsed) -> (Output_Get=Visible)
   > Set MoveObjectsFreelyVisibility(MoveObjectsFreelyVisibility=Collapsed) -> (Output_Get=Visible)
   > Set PlaysetPreviewVisibility(PlaysetPreviewVisibility=Collapsed) -> (Output_Get=Visible)
   UpdateVisibilities(execute=@n41.then_2)
   n42 = OnMoveToolTransformationModeChanged!() -> (OutputDelegate, then, TransformationType:ETransformationType Enum=Translation) {is=title K2Node_CustomEvent}
   > Set TransformationType!(
    execute
    TransformationType:ETransformationType Enum=@n42.TransformationType
    self:Creative HUDControls Move Freely Object Reference=Get CreativeHUDControlsMoveFreely()
   ) -> (then, Output_Get:ETransformationType Enum=Translation)
   > UpdateVisibilities()
   n43 = OnMoveToolRotationAxisChanged_Bind!() -> (OutputDelegate, then, RotationAxisIndex:Byte) {is=title K2Node_CustomEvent}
   > Set RotationAxisIndex!(
    execute
    RotationAxisIndex:Byte=@n43.RotationAxisIndex
    self:Creative HUDControls Move Freely Object Reference=Get CreativeHUDControlsMoveFreely()
   ) -> (then, Output_Get:Byte=0)
   > UpdateVisibilities()
   n44 = AddSelectedObject!() -> (OutputDelegate, then, AddedObject:Object Reference, NumSelected:int) {is=title K2Node_CustomEvent}
   > UpdateNumberSelected!(execute, self:Creative HUDControls Equipped Object Reference=Get CreativeHUDControlsEquipped(), CurrentNumber:int=@n44.NumSelected) -> (then)
   > UpdateVisibilities()
   n45 = OnMoveToolScaleAxisChanged!() -> (OutputDelegate, then, ScaleAxis:EScaleAxis Enum=All) {is=title K2Node_CustomEvent}
   > Set SelectedScaleAxis!(
    execute
    SelectedScaleAxis:EScaleAxis Enum=@n45.ScaleAxis
    self:Creative HUDControls Move Freely Object Reference=Get CreativeHUDControlsMoveFreely()
   ) -> (then, Output_Get:EScaleAxis Enum=All)
   > UpdateVisibilities()
   n46 = RemoveSelectedObject!() -> (OutputDelegate, then, RemovedObject:Object Reference, NumSelected:int) {is=title K2Node_CustomEvent}
   > UpdateNumberSelected(self=Get CreativeHUDControlsEquipped(), CurrentNumber=@n46.NumSelected)
   > UpdateVisibilities()
   n47 = OnSelectionPropertyChanged!() -> (OutputDelegate, then, SelectionProperty:ESelectionProperty Enum=SingleObject) {is=title K2Node_CustomEvent}
   > Set SelectionProperty!(
    execute
    SelectionProperty:ESelectionProperty Enum=@n47.SelectionProperty
    self:Creative HUDControls Move Freely Object Reference=Get CreativeHUDControlsMoveFreely()
   ) -> (then, Output_Get:ESelectionProperty Enum=SingleObject)
   > UpdateVisibilities()
   OnSelectedObjectsCleared() {is=title K2Node_CustomEvent}
   > UpdateNumberSelected(self=Get CreativeHUDControlsEquipped(), CurrentNumber=0)
   > UpdateVisibilities()
   n48 = OnCreativeGhostModeChanged!() -> (OutputDelegate, then, GhostEnabled:bool=~) {is=title K2Node_CustomEvent}
   > n49 = Set GhostModeEnabled!(execute, GhostModeEnabled:bool=@n48.GhostEnabled, self:Creative HUDControls Equipped Object Reference=Get CreativeHUDControlsEquipped()) -> (then, Output_Get:bool)
   > Set GhostModeEnabled(GhostModeEnabled=@n49, self:Creative HUDControls Not Equipped Object Reference=Get CreativeHUDControlsNotEquipped())
   > UpdateVisibilities()
   n50 = OnMoveToolInteractionStarted!() -> (OutputDelegate, then, CreativeInteraction:Object Interaction Behavior Object Reference) {is=title K2Node_CustomEvent}
   > n51 = Cast To MoveBuildingsOnGrid!(execute, Object=@n50.CreativeInteraction) -> (then, CastFailed, AsMove Buildings on Grid:Move Buildings On Grid Object Reference) {is=title K2Node_DynamicCast}
   > Set MirrorDisabled!(
    execute
    MirrorDisabled:bool=IsScaleDisabled!(self:Fort Creative Move Tool Object Reference=GetFortCreativeMoveTool!(self:Object Interaction Behavior Object Reference=@n51) -> (ReturnValue:Fort Creative Move Tool Object Reference)) -> (ReturnValue:bool)
    self:Creative HUDControls Move On Grid Object Reference=Get CreativeHUDControlsMoveOnGrid()
   ) -> (then, Output_Get:bool)
   > ShowMoveBuildingsOnGridControls() {is=title K2Node_CallFunction}
   n52 = Cast To MoveObjectsFreely!(execute=@n51.CastFailed, Object=@n50.CreativeInteraction) -> (then, CastFailed, AsMove Objects Freely:Move Objects Freely Object Reference) {is=title K2Node_DynamicCast}
   > Set ScaleDisabled!(
    execute
    ScaleDisabled:bool=IsScaleDisabled(self=GetFortCreativeMoveTool(self=@n52))
    self:Creative HUDControls Move Freely Object Reference=Get CreativeHUDControlsMoveFreely()
   ) -> (then, Output_Get:bool)
   > ShowMoveObjectsFreelyControls() {is=title K2Node_CallFunction}
   Cast To PlaysetPreview!(execute=@n52.CastFailed, Object=@n50.CreativeInteraction) -> (then, CastFailed, AsPlayset Preview:Playset Preview Object Reference) {is=title K2Node_DynamicCast}
   > ShowPlaysetPreviewControls()
   n53 = OnCreativeFlightSpeedChanged!() -> (OutputDelegate, then, SpeedLevel:int, ShowFX:bool=~) {is=title K2Node_CustomEvent}
   > n54 = Set FlightSpeedLevel!(execute, FlightSpeedLevel:int=@n53.SpeedLevel, self:Creative HUDControls Equipped Object Reference=Get CreativeHUDControlsEquipped()) -> (then, Output_Get:int=0)
   > n55 = Set FlightSpeedLevel(FlightSpeedLevel=@n54, self:Creative HUDControls Not Equipped Object Reference=Get CreativeHUDControlsNotEquipped()) -> (Output_Get=0)
   > Play Sound FlightSpeedChanged!(execute, self, newFlightSpeedLevel:int=@n55) -> (then)
   > UpdateVisibilities()
   Show Creative Global Options() {is=title K2Node_CustomEvent}
   > Branch(Condition=NOT Boolean(A=Get IsShowingQuickMenu() -> (IsShowingQuickMenu:bool)))
   > Branch(Condition=NOT Boolean(A=Get IsCreatingQuickMenu() -> (IsCreatingQuickMenu:bool)))
   > n56 = CreateWidgetAsync!(
    execute:Exec
    UserWidgetSoftClass:User Widget Soft Class Reference=&/CreativeHUDGlobalOptions_Modal.CreativeHUDGlobalOptions_Modal_C
    OwningPlayer:Player Controller Object Reference=GetOwningPlayer()
    bSuspendInputUntilComplete:bool=true
   ) -> (then:Exec, OnComplete:Exec, OnFailure:Exec, UserWidget:User Widget Object Reference) {is=title K2Node_AsyncAction}
   > Set IsCreatingQuickMenu(IsCreatingQuickMenu:bool=true) -> (Output_Get:bool)
   Set IsCreatingQuickMenu(execute=@n56.OnComplete, IsCreatingQuickMenu=false)
   > n57 = Cast To CreativeHUDGlobalOptions_Modal!(execute, Object=@n56) -> (then, CastFailed, AsCreative HUDGlobal Options Modal:Creative HUDGlobal Options Modal Object Reference) {is=title K2Node_DynamicCast}
   n58 = On Global Options Activated!() -> (OutputDelegate, then, GlobalOptions:Creative HUDGlobal Options Modal Object Reference) {is=title K2Node_CustomEvent}
   n59 = Bind Event to External on Activated(execute=@n57.then, self:Creative HUDGlobal Options Modal Object Reference=@n57, Delegate=@n58) {is=title K2Node_AddDelegate}
   n60 = On Global Options Deactivated!() -> (OutputDelegate, then, GlobalOptions:Creative HUDGlobal Options Modal Object Reference) {is=title K2Node_CustomEvent}
   Bind Event to External on Deactivated(execute=@n59.then, self:Creative HUDGlobal Options Modal Object Reference=@n57, Delegate=@n60) {is=title K2Node_AddDelegate}
   > TryPushWidgetAsModalPanel!(execute, PushWidget:User Widget Object Reference=@n57) -> (then)
   > SetVisibility(self=@n57, InVisibility=SelfHitTestInvisible)
   > ForwardOnCreativeQuickMenuDisplayedChanged!(
    execute
    self:Creative HUDLContext Object Reference=Get CreativeHUDLContext!() -> (ReturnValue:Creative HUDLContext Object Reference) {is=title K2Node_GetBlueprintContextEx}
    bIsDisplayed:bool=true
   ) -> (then)
   OnQuickMenuEnabledChanged!() -> (OutputDelegate, then, Enabled:bool=~) {is=title K2Node_CustomEvent}
   > UpdateVisibilities()
   OnQuickBarEnabledChanged!() -> (OutputDelegate, then, Enabled:bool=~) {is=title K2Node_CustomEvent}
   > UpdateVisibilities()
   ShowPlaysetPreviewControls() {is=title K2Node_CustomEvent}
   > n61 = Sequence()
   > Set PlaysetPreviewVisibility(PlaysetPreviewVisibility=HitTestInvisible) -> (Output_Get=Visible)
   Set MoveObjectsFreelyVisibility(execute=@n61.then_1, MoveObjectsFreelyVisibility=Collapsed) -> (Output_Get=Visible)
   > Set EquippedVisibility(EquippedVisibility=Collapsed) -> (Output_Get=Visible)
   > Set NotEquippedVisibility(NotEquippedVisibility=Collapsed) -> (Output_Get=Visible)
   > Set MoveBuildingsOnGridVisibility(MoveBuildingsOnGridVisibility=Collapsed) -> (Output_Get=Visible)
   UpdateVisibilities(execute=@n61.then_2)
   n62 = OnMinigamePlayModeChanged_Bind!() -> (OutputDelegate, then, Minigame:Fort Minigame Object Reference, bIsInPlayMode:bool=~) {is=title K2Node_CustomEvent}
   > UpdateVisibilities()
   n63 = PawnChanged!() -> (OutputDelegate, then, OldPawn:Pawn Object Reference, NewPawn:Pawn Object Reference) {is=title K2Node_CustomEvent}
   > Branch(Condition="Not Equal (Object)"(A:Object Reference=@n63.OldPawn, B:Object Reference=@n63.NewPawn) -> (ReturnValue:bool=~) {is=title K2Node_PromotableOperator})
   > n64 = Sequence!(execute) -> (then_0, then_1)
   > UpdateVisibilities()
   n65 = GetOwningPlayer()
   > n66 = Cast To FortPlayerControllerGameplay(Object=@n65)
   > Setup Controls Overall Initial Visibility(PlayerController=@n66)
   > n67 = Cast To FortPawn(Object=Get Controlled Pawn(self=@n65))
   > Bind Event to On Weapon Equipped Delegate(self=@n67, Delegate=Create Event())
   n68 = Cast To FortPawn(execute=@n64.then_1, Object=@n63.OldPawn)
   > Unbind Event from On Weapon Equipped Delegate(self:Fort Pawn Object Reference=@n68, Delegate=Create Event()) {is=title K2Node_RemoveDelegate}
   n69 = OnCreativePlotLinkedVolumeChanged_BInd!() -> (OutputDelegate, then, Volume:Fort Volume Object Reference) {is=title K2Node_CustomEvent}
   > n70 = Sequence()
   > n71 = Get!(execute:Exec) -> (then:Exec, else:Exec, MostRecentMinigame:Fort Minigame Object Reference)
   > Unbind Event from On Play Mode Changed Delegate(self:Fort Minigame Object Reference=@n71, Delegate=@n62) {is=title K2Node_RemoveDelegate}
   Is Valid(exec=@n70.then_1, InputObject=@n69.Volume)
   n72 = GetMinigame!(self:Fort Volume Object Reference=@n69.Volume) -> (ReturnValue:Fort Minigame Object Reference)
   > Is Valid(InputObject=@n72)
   > n73 = Set MostRecentMinigame(MostRecentMinigame:Fort Minigame Object Reference=@n72) -> (Output_Get:Fort Minigame Object Reference)
   > Bind Event to On Play Mode Changed Delegate(self:Fort Minigame Object Reference=@n73, Delegate=@n62) {is=title K2Node_AddDelegate}
   OnMoveToolInteractionStopped() {is=title K2Node_CustomEvent}
   > ShowEquippedControls() {is=title K2Node_CallFunction}
   n74 = OnMoveToolScalingModeChanged!() -> (OutputDelegate, then, ScalingOn:bool=~) {is=title K2Node_CustomEvent}
   > Set ScalingOn!(execute, ScalingOn:bool=@n74.ScalingOn, self:Creative HUDControls Move Freely Object Reference=Get CreativeHUDControlsMoveFreely()) -> (then, Output_Get:bool)
   > UpdateVisibilities()
   n75 = On Move Tool Selected Actor Changed!() -> (OutputDelegate, then, FocusedActor:Actor Object Reference) {is=title K2Node_CustomEvent}
   > OnMoveToolActorSelectedChanged!(
    execute
    self:Creative HUDControls Equipped Object Reference=Get CreativeHUDControlsEquipped()
    Actor:Actor Object Reference=@n75.FocusedActor
   ) -> (then)
   > UpdateVisibilities()
   n76 = OnHeatmapModeChanged!() -> (OutputDelegate, then, NewState:EFortCreativeHeatmapDrawState Enum=Visible) {is=title K2Node_CustomEvent}
   > Play Sound HeatmapModeChanged!(execute, self, newHeatmapMode:EFortCreativeHeatmapDrawState Enum=@n76.NewState) -> (then)
   > UpdateVisibilities()
   n77 = OnCreativeFlyingChanged_Bind!() -> (OutputDelegate, then, Flying:bool=~) {is=title K2Node_CustomEvent}
   > Set IsFlying(IsFlying=@n77.Flying)
   > UpdateVisibilities()
   > Set Can Play Speed Changed(Can Play Speed Changed:bool=false) -> (Output_Get:bool)
   > Delay(Duration=0.1)
   > Set Can Play Speed Changed(Can Play Speed Changed=true)
   OnBeginSkydiving_Bind() {is=title K2Node_CustomEvent}
   > Set IsFlying(IsFlying=false)
   > UpdateVisibilities()
   SetWaitForPawnTimer!() -> (OutputDelegate, then, PlayerController:Fort Player Controller Zone Object Reference) {is=title K2Node_CustomEvent}
   > n78 = Set Timer by Event(Delegate=Create Event(), Time=0.25, bLooping=true)
   > Set WaitForPawnTimerHandle(WaitForPawnTimerHandle:Timer Handle Structure=@n78) -> (Output_Get:Timer Handle Structure)
   n79 = OnMoveToolDropToFloorChanged_Bind!() -> (OutputDelegate, then, bDropToFloorOn:bool=~) {is=title K2Node_CustomEvent}
   > Set DropToFloorOn!(execute, DropToFloorOn:bool=@n79.bDropToFloorOn, self:Creative HUDControls Move Freely Object Reference=Get CreativeHUDControlsMoveFreely()) -> (then, Output_Get:bool)
   > UpdateVisibilities()
   n80 = OnMoveToolPrecisionChanged_Bind!() -> (OutputDelegate, then, PrecisionOn:bool=~, GridSnapIndex:Byte) {is=title K2Node_CustomEvent}
   n81 = Get CreativeHUDControlsMoveFreely()
   > Set PrecisionOn!(execute, PrecisionOn:bool=@n80.PrecisionOn, self:Creative HUDControls Move Freely Object Reference=@n81) -> (then, Output_Get:bool)
   > Set GridSnapIndex!(execute, GridSnapIndex:Byte=@n80.GridSnapIndex, self:Creative HUDControls Move Freely Object Reference=@n81) -> (then, Output_Get:Byte=0)
   > UpdateVisibilities()
   Event Destruct()
   > Parent: Destruct()
   > Unbind Event from On Keybinds Changed(
    self:Fort Global UIContext Object Reference=Get FortGlobalUIContext!() -> (ReturnValue:Fort Global UIContext Object Reference) {is=title K2Node_GetBlueprintContextEx}
    Delegate=Create Event()
   ) {is=title K2Node_RemoveDelegate}
   Event Construct()
   > Parent: Construct()
   > n82 = Sequence!(execute) -> (then_0, then_1, then_2:Exec)
   > SetVisibility(InVisibility=Collapsed)
   > SetWaitForPawnTimer!(execute, self, PlayerController:Fort Player Controller Zone Object Reference) -> (then) {is=title K2Node_CallFunction}
   > n83 = Cast To FortPlayerControllerAthena(Object=GetOwningPlayer())
   > Bind Event to On Started Respawn(self:Fort Player Controller Zone Object Reference=@n83, Delegate=Create Event()) {is=title K2Node_AddDelegate}
   > Bind Event to On Begin Skydiving(self:Fort Player Controller Zone Object Reference=@n83, Delegate=Create Event()) {is=title K2Node_AddDelegate}
   > Bind Event to On End Skydiving(self:Fort Player Controller Zone Object Reference=@n83, Delegate=Create Event()) {is=title K2Node_AddDelegate}
   > Bind Event to On Creative Plot Linked Volume Changed(self:Fort Player Controller Athena Object Reference=@n83, Delegate=Create Event()) {is=title K2Node_AddDelegate}
   > n84 = Cast To FortInGameMapManagerCreative!(execute, Object=GetUIMapManager!() -> (ReturnValue:Fort In Game Map Manager Object Reference)) -> (then, CastFailed, AsFort in Game Map Manager Creative:Fort In Game Map Manager Creative Object Reference) {is=title K2Node_DynamicCast}
   > Bind Event to On Heatmap Draw State Changed(self:Fort In Game Map Manager Creative Object Reference=@n84, Delegate=Create Event()) {is=title K2Node_AddDelegate}
   n85 = Cast To FortPlayerControllerGameplay(execute=@n82.then_2, Object=GetOwningPlayer())
   > Bind Event to On World Inventory Changed Delegate(self:Fort Player Controller Gameplay Object Reference=@n85, Delegate=Create Event()) {is=title K2Node_AddDelegate}
   > Bind Event to On Creative Fly Changed Delegate(self:Fort Player Controller Gameplay Object Reference=@n85, Delegate=Create Event()) {is=title K2Node_AddDelegate}
   > Bind Event to On Creative Ghost Enabled Delegate(self:Fort Player Controller Gameplay Object Reference=@n85, Delegate=Create Event()) {is=title K2Node_AddDelegate}
   > Bind Event to On Creative Flying Speed Changed(self:Fort Player Controller Gameplay Object Reference=@n85, Delegate=Create Event()) {is=title K2Node_AddDelegate}
   > Setup Controls Overall Initial Visibility(PlayerController=@n85)
   > Bind Event to On Show Creative Global Options(self:Fort Player Controller Gameplay Object Reference=@n85, Delegate=Create Event()) {is=title K2Node_AddDelegate}
   > Bind Event to On Creative Quickbar Enabled Changed Delegate(self:Fort Player Controller Gameplay Object Reference=@n85, Delegate=Create Event()) {is=title K2Node_AddDelegate}
   > Bind Event to On Creative Quickmenu Enabled Changed Delegate(self:Fort Player Controller Gameplay Object Reference=@n85, Delegate=Create Event()) {is=title K2Node_AddDelegate}
   n86 = Get CommonInputSubsystem()
   > Bind Event to On Input Method Changed(self=@n86, Delegate=Create Event())
   > Handle Input Method Changed!(execute, self, InputMethod:ECommonInputType Enum=GetCurrentInputType(self=@n86)) -> (then)
   n87 = Get FortGlobalUIContext()
   Bind Event to On Keybinds Changed(execute=@n82.then_1, self:Fort Global UIContext Object Reference=@n87, Delegate=Create Event()) {is=title K2Node_AddDelegate}
   > Bind Event to On Enter Vehicle Driver(self:Fort Global UIContext Object Reference=@n87, Delegate=Create Event()) {is=title K2Node_AddDelegate}
   > Assign On Enter Vehicle Passenger!(execute:Exec, self:Fort Global UIContext Object Reference=@n87, Delegate:"Delegate (by ref)"=Create Event()) -> (then:Exec) {is=title K2Node_AssignDelegate}
   > Assign On Exit Vehicle!(execute:Exec, self:Fort Global UIContext Object Reference=@n87, Delegate:"Delegate (by ref)"=Create Event()) -> (then:Exec) {is=title K2Node_AssignDelegate}
   SetVisibility(execute=@n90.then, InVisibility=Collapsed)
   OnEndSkydiving_Bind() {is=title K2Node_CustomEvent}
   Set IsShowingQuickMenu(execute=@n88.then, IsShowingQuickMenu:bool=false) -> (Output_Get:bool)
   n88 = SetVisibility(execute=@n89.then, InVisibility=HitTestInvisible)
   n89 = Unbind all Events from External on Deactivated!(execute:Exec=@n60.then, self:Creative HUDGlobal Options Modal Object Reference=@n60.GlobalOptions) -> (then:Exec) {is=title K2Node_ClearDelegate}
   Unbind all Events from External on Activated!(execute:Exec=@n58.then, self:Creative HUDGlobal Options Modal Object Reference=@n58.GlobalOptions) -> (then:Exec) {is=title K2Node_ClearDelegate}
   > n90 = Set IsShowingQuickMenu(IsShowingQuickMenu=true)
  ]}
 Handle Input Method Changed {nodes [
   Handle Input Method Changed!() -> (then, InputMethod:ECommonInputType Enum) {is=title K2Node_FunctionEntry}
   > OnKeybindsChanged()
  ]}
 HideOrShowCreativeHUDControls {nodes [
   n1 = HideOrShowCreativeHUDControls!() -> (then, PlayerController:Fort Player Controller Gameplay Object Reference) {is=title K2Node_FunctionEntry}
   > n2 = Is Valid(InputObject=@n1)
   Set IsAllowedToFly(execute=@n2.Is Not Valid, IsAllowedToFly=false)
   > n3 = Set IsCreativeModeEnabled(IsCreativeModeEnabled:bool=false) -> (Output_Get:bool)
   Is Valid(
    exec=[@n3.then, @n10.then]
    InputObject=Get FortHUDContext!() -> (ReturnValue:Fort HUDContext Object Reference) {is=title K2Node_GetBlueprintContextEx}
   )
   > n4 = Cast To FortGameStateAthena!(execute, Object=GetGameState!() -> (ReturnValue:Game State Base Object Reference)) -> (then, CastFailed, AsFort Game State Athena:Fort Game State Athena Object Reference) {is=title K2Node_DynamicCast}
   > n5 = Branch(Condition=IsCreativeOrPlaygroundModePlaylist!(self:Fort Game State Athena Object Reference=@n4) -> (ReturnValue:bool))
   > n6 = CanWeFlyRightNow()
   n7 = Get IsShowingQuickMenu()
   n8 = Get IsAllowedToFly()
   > n9 = Branch(Condition=AND Boolean!(A=AND Boolean!(A=@n6, B=@n8) -> (ReturnValue), B=Get IsCreativeModeEnabled() -> (IsCreativeModeEnabled:bool), C:bool=NOT Boolean(A=@n7)) -> (ReturnValue))
   > SetVisibility(InVisibility=HitTestInvisible)
   > Return Node!(execute, IsVisibile:bool=true)
   SetVisibility(execute=[@n9.else, @n5.else, @n4.CastFailed], InVisibility=Collapsed)
   > Return Node(IsVisibile=false)
   Set IsAllowedToFly(execute=@n2.Is Valid, IsAllowedToFly=IsAllowedToFly!(self:Fort Player Controller Gameplay Object Reference=@n1) -> (ReturnValue:bool))
   > n10 = Set IsCreativeModeEnabled(IsCreativeModeEnabled=IsCreativeModeEnabled!(self:Fort Player Controller Gameplay Object Reference=@n1) -> (ReturnValue:bool))
  ]}
 IsKeyOverridden {nodes [
   n1 = IsKeyOverridden!() -> (then, Key:Key Structure=None) {is=title K2Node_FunctionEntry}
   n2 = Get CreativeHUDControlsMoveFreely()
   n3 = Get CreativeHUDControlsEquipped()
   n4 = Get CreativeHUDControlPlaysetPreview()
   n5 = Get CreativeHUDControlsNotEquipped()
   n6 = Get CreativeHUDControlsMoveOnGrid()
   > Return Node!(
    execute
    Overriden:bool=OR Boolean!(
     A=IsKeyInUse!(self:Athena HUDSituational Feedback Object Reference=@n3, Key:Key Structure=@n1) -> (InUse:bool)
     B=IsKeyInUse(self=@n5, Key=@n1)
     C:bool=IsKeyInUse(self=@n2, Key=@n1)
     D:bool=IsKeyInUse(self=@n6, Key=@n1)
     E:bool=IsKeyInUse(self=@n4, Key=@n1)
    ) -> (ReturnValue)
   )
  ]}
 OnKeybindsChanged {nodes [
   OnKeybindsChanged() {is=title K2Node_FunctionEntry}
   > UpdateKeybindVisibility!(
    execute
    self
    Box:Widget Object Reference=Get QuickMenuHorizontalBox() -> (QuickMenuHorizontalBox:Horizontal Box Object Reference)
    Keybind:Keybind Widget Object Reference=Get QuickMenuKeybindWidget() -> (QuickMenuKeybindWidget:Keybind Widget Object Reference)
   ) -> (then)
   > UpdateKeybindVisibility(
    Box=Get ContentBrowserHorizontalBox() -> (ContentBrowserHorizontalBox:Horizontal Box Object Reference)
    Keybind=Get ContentBrowserKeybindWidget() -> (ContentBrowserKeybindWidget:Keybind Widget Object Reference)
   )
   > UpdateKeybindVisibility(
    Box=Get IslandSettingsHorizontalBox() -> (IslandSettingsHorizontalBox:Horizontal Box Object Reference)
    Keybind=Get IslandSettingsKeybindWidget() -> (IslandSettingsKeybindWidget:Keybind Widget Object Reference)
   )
   > UpdateKeybindVisibility(
    Box=Get CreativeMenuHorizontalBox() -> (CreativeMenuHorizontalBox:Horizontal Box Object Reference)
    Keybind=Get CreativeMenuKeybindWidget() -> (CreativeMenuKeybindWidget:Keybind Widget Object Reference)
   )
  ]}
 Play Sound FlightSpeedChanged {nodes [
   n1 = Play Sound FlightSpeedChanged!() -> (then, newFlightSpeedLevel:int) {is=title K2Node_FunctionEntry}
   > Branch(Condition=AND Boolean!(A=Get IsFlying(), B=Get IsCreativeModeEnabled(), C:bool=Get Can Play Speed Changed() -> (Can Play Speed Changed:bool)) -> (ReturnValue))
   > PlaySound2D(
    Sound=Get Flight Speed Changed Sound() -> (Flight Speed Changed Sound:Sound Base Object Reference)
    PitchMultiplier=Select!(
     Option 0:float=0.85
     Option 1:float=0.89
     Option 2:float=0.93
     Option 3:float=0.96
     Option 4:float=1
     Option 5:float=1.05
     Option 6:float=1.1
     Option 7:float=1.15
     Option 8:float=1.2
     Index:int=@n1
    ) -> (ReturnValue:float)
   )
  ]}
 Play Sound HeatmapModeChanged {nodes [
   n1 = Play Sound HeatmapModeChanged!() -> (then, newHeatmapMode:EFortCreativeHeatmapDrawState Enum=Visible) {is=title K2Node_FunctionEntry}
   > PlaySound2D(
    Sound=Get Heatmap Mode Changed Sound() -> (Heatmap Mode Changed Sound:Sound Base Object Reference)
    PitchMultiplier=Select!(Visible:float=1, VisibleWithValues:float=1.1, NotVisible:float=0.8, Index:EFortCreativeHeatmapDrawState Enum=@n1) -> (ReturnValue)
   )
  ]}
 SetupControlsOverallInitialVisibility {nodes [
   n1 = Inputs!() -> (execute, PlayerController:Fort Player Controller Gameplay Object Reference)
   > n2 = HideOrShowCreativeHUDControls(PlayerController=@n1) {is=title K2Node_CallFunction}
   > n3 = Branch(Condition=@n2)
   > n4 = ShowUnequippedControls() {is=title K2Node_CallFunction}
   UpdateVisibilities(execute=[@n4.then, @n3.else])
   > Outputs!(then:Exec)
  ]}
 UpdateKeybindVisibility {nodes [
   n1 = UpdateKeybindVisibility!() -> (then, Box:Widget Object Reference, Keybind:Keybind Widget Object Reference) {is=title K2Node_FunctionEntry}
   n2 = GetKeyForAction!(
    self:Fort Global UIContext Object Reference=Get FortGlobalUIContext()
    Action:Name=Get BoundAction!(self:Fort Keybind Widget Object Reference=@n1.Keybind) -> (BoundAction:Name=None)
    bForceGamepadKey:bool=false
    PresetNameOverride:Name=None
   ) -> (ReturnValue:Key Structure)
   n3 = Is Valid Key!(Key:"Key Structure (by ref)"=@n2) -> (ReturnValue:bool)
   n4 = NOT Boolean(A=IsKeyOverridden!(Key:Key Structure=@n2) -> (Overriden:bool) {is=title K2Node_CallFunction})
   > SetVisibility(
    self=@n1
    InVisibility=Select!(
     Option 0:ESlateVisibility Enum=Collapsed
     Option 1:ESlateVisibility Enum=HitTestInvisible
     Index:bool=AND Boolean!(A=@n3, B=@n4) -> (ReturnValue)
    ) -> (ReturnValue:ESlateVisibility Enum)
   )
  ]}
]
variables [
 Can Play Speed Changed {type=boolean}
 EquippedVisibility {type=ESlateVisibility, default=Collapsed}
 Flight Speed Changed Sound {type=SoundBase, default=/Game/Sounds/Fort_UI_Sounds/FrontEnd/Leaderboard/Fort_Athena_LeaderboardTab_Pressed_Cue}
 Heatmap Mode Changed Sound {type=SoundBase, default=/Game/Sounds/Fort_UI_Sounds/FrontEnd/Leaderboard/Fortnite_UI_Leaderboard_Tile_Click_Cue}
 IsAllowedToFly {type=boolean}
 IsCreatingQuickMenu {type=boolean}
 IsCreativeModeEnabled {type=boolean}
 IsFlying {type=boolean}
 IsShowingQuickMenu {type=boolean}
 MostRecentMinigame {type=FortMinigame, default=null}
 MoveBuildingsOnGridVisibility {type=ESlateVisibility, default=Collapsed}
 MoveObjectsFreelyVisibility {type=ESlateVisibility, default=Collapsed}
 NotEquippedVisibility {type=ESlateVisibility, default=Collapsed}
 PlaysetPreviewVisibility {type=ESlateVisibility, default=Collapsed}
 WaitForPawnTimerHandle {type=TimerHandle}
]
widgets {
 info {inheritedWidgetCount=1, parentClass=&FortHUDElementWidget, rootWidgetClass=&VerticalBox, widgetCount=22}
 widgets [
  VerticalBox_0 {widgetClassPath=&VerticalBox, widget=VerticalBox_0}
  Overlay_0 {widgetClassPath=&Overlay, parent=VerticalBox_0, widget=Overlay_0, slot {class=VerticalBoxSlot}}
  CreativeHUDControlsEquipped {
   widgetClassPath=&/CreativeHUDControlsEquipped
   bIsVariable=true
   parent=Overlay_0
   widget=CreativeHUDControlsEquipped
   slot {class=OverlaySlot, properties {horizontalAlignment=HAlign_Fill, verticalAlignment=VAlign_Fill}}
  }
  CreativeHUDControlsNotEquipped {
   widgetClassPath=&/CreativeHUDControlsNotEquipped
   bIsVariable=true
   parent=Overlay_0
   widget=CreativeHUDControlsNotEquipped
   properties {elementTag {tagName=Creative.MoveTool.Activated}}
   slot {class=OverlaySlot}
  }
  CreativeHUDControlsMoveFreely {
   widgetClassPath=&/CreativeHUDControlsMoveFreely
   bIsVariable=true
   parent=Overlay_0
   widget=CreativeHUDControlsMoveFreely
   properties {elementTag {tagName=Creative.MoveTool.Activated}}
   slot {class=OverlaySlot}
  }
  CreativeHUDControlsMoveOnGrid {
   widgetClassPath=&/CreativeHUDControlsMoveOnGrid
   bIsVariable=true
   parent=Overlay_0
   widget=CreativeHUDControlsMoveOnGrid
   properties {elementTag {tagName=Creative.MoveTool.Activated}}
   slot {class=OverlaySlot}
  }
  CreativeHUDControlPlaysetPreview {
   widgetClassPath=&/CreativeHUDControlPlaysetPreview
   bIsVariable=true
   parent=Overlay_0
   widget=CreativeHUDControlPlaysetPreview
   properties {elementTag {tagName=Creative.MoveTool.Activated}}
   slot {class=OverlaySlot}
  }
  MenusBorder {
   widgetClassPath=&Border
   bIsVariable=true
   parent=VerticalBox_0
   widget=MenusBorder
   properties {
    horizontalAlignment=HAlign_Left
    visibility=HitTestInvisible
    background {
     bIsDynamicallyLoaded=false
     drawAs=NoDrawType
     imageType=NoImage
     mirroring=NoMirror
     resourceName=None
     resourceObject=null
     tiling=NoTile
     imageSize=<32 32>
     margin=<ltrb 0 0 0 0>
     outlineSettings {
      bUseBrushTransparency=false
      roundingType=HalfHeightRadius
      width=0
      color {colorUseRule=UseColor_Specified, specifiedColor=<rgba 0 0 0 0>}
      cornerRadii=<0 0 0 0>
     }
     tintColor {colorUseRule=UseColor_Specified, specifiedColor=<rgba 1 1 1 1>}
     uVRegion {bIsValid=false, max=<0 0>, min=<0 0>}
    }
    padding=<ltrb 0 0 0 0>
   }
   slot {class=VerticalBoxSlot, properties {horizontalAlignment=HAlign_Left}}
  }
  VerticalBox_1 {
   widgetClassPath=&VerticalBox
   parent=MenusBorder
   widget=VerticalBox_1
   slot {class=BorderSlot, properties {horizontalAlignment=HAlign_Left, padding=<ltrb 0 0 0 0>}}
  }
  CreativeMenuHorizontalBox {
   widgetClassPath=&HorizontalBox
   bIsVariable=true
   parent=VerticalBox_1
   widget=CreativeMenuHorizontalBox
   slot {class=VerticalBoxSlot, properties {horizontalAlignment=HAlign_Left, padding=<ltrb 0 0 0 10>}}
  }
  CreativeMenuKeybindWidget {
   widgetClassPath=/FortUILibrary/Input/KeybindWidget
   bIsVariable=true
   parent=CreativeMenuHorizontalBox
   widget=CreativeMenuKeybindWidget
   properties {boundAction=CreativeMenu, boundKeyFallback=B}
   slot {class=HorizontalBoxSlot, properties {horizontalAlignment=HAlign_Left, size {sizeRule=Fill, value=1}}}
  }
  CreativeMenuText {
   widgetClassPath=/Script/UMG.RichTextBlock
   bIsVariable=true
   parent=CreativeMenuHorizontalBox
   widget=CreativeMenuText
   properties {text=Creative Menu, textStyleSet=/Game/Athena/HUD/TextStyles/PlayersLeftStyles}
   slot {class=HorizontalBoxSlot, properties {horizontalAlignment=HAlign_Left, padding=<ltrb 10 0 0 0>}}
  }
  QuickMenuHorizontalBox {
   widgetClassPath=&HorizontalBox
   bIsVariable=true
   parent=VerticalBox_1
   widget=QuickMenuHorizontalBox
   properties {visibility=Collapsed}
   slot {class=VerticalBoxSlot, properties {horizontalAlignment=HAlign_Left, padding=<ltrb 0 0 0 10>}}
  }
  QuickMenuKeybindWidget {
   widgetClassPath=&KeybindWidget
   bIsVariable=true
   parent=QuickMenuHorizontalBox
   widget=QuickMenuKeybindWidget
   properties {boundAction=ToggleQuickMenu, boundKeyFallback=B}
   slot {class=HorizontalBoxSlot, properties {horizontalAlignment=HAlign_Left, size {sizeRule=Fill, value=1}}}
  }
  QuickMenuText {
   widgetClassPath=&RichTextBlock
   bIsVariable=true
   parent=QuickMenuHorizontalBox
   widget=QuickMenuText
   properties {text=Quick Menu, textStyleSet=&PlayersLeftStyles}
   slot {class=HorizontalBoxSlot, properties {horizontalAlignment=HAlign_Left, padding=<ltrb 10 0 0 0>}}
  }
  ContentBrowserHorizontalBox {
   widgetClassPath=&HorizontalBox
   bIsVariable=true
   parent=VerticalBox_1
   widget=ContentBrowserHorizontalBox
   properties {visibility=Collapsed}
   slot {class=VerticalBoxSlot, properties {horizontalAlignment=HAlign_Left, padding=<ltrb 0 0 0 10>}}
  }
  ContentBrowserKeybindWidget {
   widgetClassPath=&KeybindWidget
   bIsVariable=true
   parent=ContentBrowserHorizontalBox
   widget=ContentBrowserKeybindWidget
   properties {boundAction=ToggleContentBrowser, boundKeyFallback=B}
   slot {class=HorizontalBoxSlot, properties {horizontalAlignment=HAlign_Left, size {sizeRule=Fill, value=1}}}
  }
  ContentBrowserText {
   widgetClassPath=&RichTextBlock
   bIsVariable=true
   parent=ContentBrowserHorizontalBox
   widget=ContentBrowserText
   properties {text=Content, textStyleSet=&PlayersLeftStyles}
   slot {class=HorizontalBoxSlot, properties {horizontalAlignment=HAlign_Left, padding=<ltrb 10 0 0 0>}}
  }
  IslandSettingsHorizontalBox {
   widgetClassPath=&HorizontalBox
   bIsVariable=true
   parent=VerticalBox_1
   widget=IslandSettingsHorizontalBox
   properties {visibility=Collapsed}
   slot {class=VerticalBoxSlot, properties {horizontalAlignment=HAlign_Left, padding=<ltrb 0 0 0 10>}}
  }
  IslandSettingsKeybindWidget {
   widgetClassPath=&KeybindWidget
   bIsVariable=true
   parent=IslandSettingsHorizontalBox
   widget=IslandSettingsKeybindWidget
   properties {boundAction=ToggleIslandSettings, boundKeyFallback=B}
   slot {class=HorizontalBoxSlot, properties {horizontalAlignment=HAlign_Left, size {sizeRule=Fill, value=1}}}
  }
  IslandSettingsText {
   widgetClassPath=&RichTextBlock
   bIsVariable=true
   parent=IslandSettingsHorizontalBox
   widget=IslandSettingsText
   properties {text=Island Settings, textStyleSet=&PlayersLeftStyles}
   slot {class=HorizontalBoxSlot, properties {horizontalAlignment=HAlign_Left, padding=<ltrb 10 0 0 0>}}
  }
  TutorialHighlight {widgetClassPath=&TutorialHighlight}
 ]
}
