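# Creative HUD overlay for previewing the island photo camera used for the published island's screenshot.
# Confirm, Default Location and Cancel buttons forward to parent FortCreativeCaptureScreenshotHUD handlers; player controller delegates bound on Construct drive the flow.
# Camera activation toggles the overlay and HUD-tagged elements, setting bIsPreviewing, which gates accept, reset and cancel.
# Accept shows a timed message before deactivating and committing the preview; preview setup retains a debug PrintString.
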WidgetBlueprint /Game/Athena/HUD/CreativeHUDPreviewScreenshotOverlay
graphs [
 EventGraph {nodes [
   "On Button Clicked (Button_Confirm)"!() -> (OutputDelegate, then, Button:Common Button Legacy Object Reference) {is=title K2Node_ComponentBoundEvent}
   > HandleButtonConfirmClicked(self:Fort Creative Capture Screenshot HUD Object Reference)
   "On Button Clicked (Button_Reset)"!() -> (OutputDelegate, then, Button:Common Button Legacy Object Reference) {is=title K2Node_ComponentBoundEvent}
   > HandleButtonResetClicked(self:Fort Creative Capture Screenshot HUD Object Reference)
   "On Button Clicked (Cancel)"!() -> (OutputDelegate, then, Button:Common Button Legacy Object Reference) {is=title K2Node_ComponentBoundEvent}
   > HandleButtonCancelClicked(self:Fort Creative Capture Screenshot HUD Object Reference)
   n1 = ActivateOverlay!() -> (OutputDelegate, then, bActivated:bool=~) {is=title K2Node_CustomEvent}
   > K2Node_Composite_1 = OverlayVisibility!(InputPin:Exec, bHideElements:bool=@n1.bActivated) {is=title K2Node_Composite}
   n2 = CancelOverlay() {is=title K2Node_CustomEvent}
   > K2Node_Composite_2 = CancelFlow() {is=title K2Node_Composite}
   n3 = OnBeginPreview!() -> (OutputDelegate, then, bActive:bool=~) {is=title K2Node_CustomEvent}
   > K2Node_Composite_0 = PreviewSetup!(execute, Condition:bool=@n3.bActive) {is=title K2Node_Composite}
   n4 = "AcceptOverlay()"() {is=title K2Node_CustomEvent}
   > Branch(Condition=Get bIsPreviewing() -> (bIsPreviewing:bool))
   > EndPreview!(execute, self, Accepted:bool=true) -> (then)
   n5 = EndPreview!() -> (OutputDelegate, then, Accepted:bool=~) {is=title K2Node_CustomEvent}
   > SetVisibility(self=Get InteractionBorder() -> (InteractionBorder:Common Border Object Reference), InVisibility=Collapsed)
   > SetVisibility(
    self=Get PreviewScreenshotMessageWidget() -> (PreviewScreenshotMessageWidget:Preview Screenshot Message Widget Object Reference)
    InVisibility=HitTestInvisible
   )
   > Delay(Duration=Get MessageShowSeconds() -> (MessageShowSeconds:double))
   > SetVisibility(self=Get PreviewScreenshotMessageWidget(), InVisibility=Collapsed)
   > OnBeginPreview!(execute, self, bActive:bool=false) -> (then) {is=title K2Node_CallFunction}
   > ActivateCreativePreviewScreenshot!(
    execute
    self:Fort Player Controller Gameplay Object Reference=Get FortPlayerControllerAthena() -> (FortPlayerControllerAthena:Fort Player Controller Athena Object Reference)
    bActivate:bool=false
   ) -> (then)
   > n6 = Branch(Condition=@n5.Accepted)
   > AcceptPreview(self:Fort Creative Capture Screenshot HUD Object Reference)
   ResetPreview(execute=@n6.else, self:Fort Creative Capture Screenshot HUD Object Reference)
   n7 = ResetOverlay() {is=title K2Node_CustomEvent}
   > Branch(Condition=Get bIsPreviewing())
   > ResetPreview()
   Event Construct()
   > SetVisibility(InVisibility=Collapsed)
   > n8 = Cast To Athena_PlayerController!(execute, Object=GetOwningPlayer()) -> (then, CastFailed, AsAthena Player Controller:Athena Player Controller Object Reference) {is=title K2Node_DynamicCast}
   > n9 = Set FortPlayerControllerAthena(FortPlayerControllerAthena:Fort Player Controller Athena Object Reference=@n8) -> (Output_Get:Fort Player Controller Athena Object Reference)
   > Bind Event to On Creative Preview Screenshot Active Changed(self:Fort Player Controller Gameplay Object Reference=@n9, Delegate=@n3) {is=title K2Node_AddDelegate}
   > Bind Event to On Preview Screenshot Camera Activated(self:Fort Player Controller Athena Object Reference=Get FortPlayerControllerAthena(), Delegate=@n1) {is=title K2Node_AddDelegate}
   > Bind Event to On Creative Preview Screenshot Accepted(self:Fort Player Controller Gameplay Object Reference=Get FortPlayerControllerAthena(), Delegate=@n4) {is=title K2Node_AddDelegate}
   > Bind Event to On Creative Preview Screenshot Reset(self:Fort Player Controller Gameplay Object Reference=Get FortPlayerControllerAthena(), Delegate=@n7) {is=title K2Node_AddDelegate}
   > Bind Event to On Creative Preview Screenshot Cancelled(self:Fort Player Controller Gameplay Object Reference=Get FortPlayerControllerAthena(), Delegate=@n2) {is=title K2Node_AddDelegate}
  ]}
 PreviewSetup {
  path=EventGraph.K2Node_Composite_0.PreviewSetup
  nodes [
   n1 = Inputs!() -> (execute, Condition:bool=~)
   > PrintString!(execute, InString=OnBeingPreview CHPS, bPrintToScreen, bPrintToLog, TextColor, Duration, Key) -> (then)
   > Branch(Condition=@n1)
   > BeginPreview!(execute, self:Fort Creative Capture Screenshot HUD Object Reference) -> (then, ReturnValue:bool)
   Outputs()
  ]
 }
 OverlayVisibility {
  path=EventGraph.K2Node_Composite_1.OverlayVisibility
  nodes [
   n1 = Inputs!() -> (InputPin:Exec, bHideElements:bool=~)
   > Set bIsPreviewing(bIsPreviewing:bool=@n1) -> (Output_Get:bool)
   > SetHUDElementVisibility!(
    execute
    self:Fort HUDContext Object Reference=Get FortHUDContext!() -> (ReturnValue:Fort HUDContext Object Reference) {is=title K2Node_GetBlueprintContextEx}
    HUDElementTags:"Gameplay Tag Container Structure (by ref)"=MakeGameplayTagContainerFromTag!(SingleTag:Gameplay Tag Structure=(TagName="HUD")) -> (ReturnValue:Gameplay Tag Container Structure)
    bHideElements:bool=@n1
    ReasonName:"Name (by ref)"=None
   ) -> (then)
   n2 = Select(Option 0=Collapsed, Option 1=SelfHitTestInvisible, Index=@n1)
   > SetVisibility(InVisibility=@n2)
   > SetVisibility(self=Get PreviewScreenshotMessageWidget(), InVisibility=Collapsed)
   > SetVisibility(self=Get InteractionBorder(), InVisibility=@n2)
   Outputs()
  ]
 }
 CancelFlow {
  path=EventGraph.K2Node_Composite_2.CancelFlow
  nodes [
   Inputs!() -> (execute:Exec)
   > Branch(Condition=Get bIsPreviewing())
   > SetVisibility(self=Get InteractionBorder(), InVisibility=Collapsed)
   > ActivateCreativePreviewScreenshot(self=Get FortPlayerControllerAthena(), bActivate=false)
   > CancelPreview(self:Fort Creative Capture Screenshot HUD Object Reference)
   Outputs()
  ]
 }
]
variables [
 bIsPreviewing {type=boolean}
 FortPlayerControllerAthena {type=FortPlayerControllerAthena, default=null}
 MessageShowSeconds {type=number, default=2}
]
widgets {
 info {
  inheritedWidgetCount=0
  parentClass=/Script/FortniteUI.FortCreativeCaptureScreenshotHUD
  rootWidgetClass=&CanvasPanel
  widgetCount=11
 }
 widgets [
  CanvasPanel_77 {widgetClassPath=&CanvasPanel, widget=CanvasPanel_77}
  InteractionBorder {
   widgetClassPath=&CommonBorder
   bIsVariable=true
   parent=CanvasPanel_77
   widget=InteractionBorder
   properties {
    style=/FortUILibrary/Styles/Border/Border-Bottom-Box-Rounded.Border-Bottom-Box-Rounded_C
    brushColor=<rgba 0.4 0.4 0.4 0.5>
    padding=<ltrb 0 0 0 0>
   }
   slot {
    class=CanvasPanelSlot
    properties {layoutData {alignment=<0 0>, anchors {maximum=<1 1>, minimum=<0 0>}, offsets=<ltrb 0 925.6216 1.9219 1.9629>}}
   }
  }
  HorizontalBox_0 {
   widgetClassPath=&HorizontalBox
   parent=InteractionBorder
   widget=HorizontalBox_0
   slot {class=BorderSlot, properties {padding=<ltrb 0 0 0 0>}}
  }
  VerticalBox_0 {
   widgetClassPath=&VerticalBox
   parent=HorizontalBox_0
   widget=VerticalBox_0
   slot {class=HorizontalBoxSlot, properties {padding=<ltrb 12 12 12 12>, size {sizeRule=Fill, value=1}}}
  }
  RichText_BottomPrompt {
   widgetClassPath=/Script/UMG.RichTextBlock
   parent=VerticalBox_0
   widget=RichText_BottomPrompt
   properties {
    lineHeightPercentage=0.77
    text="<keyword>ISLAND PHOTO PREVIEW</>"
    textStyleSet=/Game/Athena/HUD/LTM/Tutorial/Materials/TutorialCalloutStyles
    renderTransform {angle=0, scale=<1 1>, shear=<-12 0>, translation=<0 0>}
   }
   slot {class=VerticalBoxSlot, properties {horizontalAlignment=HAlign_Left}}
  }
  RichText_BottomPrompt_1 {
   widgetClassPath=&RichTextBlock
   parent=VerticalBox_0
   widget=RichText_BottomPrompt_1
   properties {
    autoWrapText=true
    lineHeightPercentage=0.77
    text=This camera will be used to capture a screenshot for your island once it is published.
    textStyleSet=&TutorialCalloutStyles
   }
   slot {class=VerticalBoxSlot, properties {size {sizeRule=Fill, value=1}}}
  }
  ButtonsArea {
   widgetClassPath=&HorizontalBox
   parent=HorizontalBox_0
   widget=ButtonsArea
   slot {class=HorizontalBoxSlot, properties {horizontalAlignment=HAlign_Right, verticalAlignment=VAlign_Top, padding=<ltrb 0 12 64 0>}}
  }
  Button_Confirm {
   widgetClassPath=&IconTextButton
   bIsVariable=true
   parent=ButtonsArea
   widget=Button_Confirm
   properties {
    buttonText=Confirm
    minWidth=300
    onButtonBaseClicked=(&self:WidgetTree.Button_Confirm.HandleOnButtonClicked)
    onButtonBaseDoubleClicked=(&self:WidgetTree.Button_Confirm.HandleOnButtonDoubleClicked)
    onButtonBaseHovered=(&self:WidgetTree.Button_Confirm.HandleOnButtonHovered)
    onButtonBaseUnhovered=(&self:WidgetTree.Button_Confirm.HandleOnButtonUnhovered)
    onSelectedChangedBase=(&self:WidgetTree.Button_Confirm.HandleOnSelectedChanged)
    style=/FortUILibrary/Foundation/Buttons/ButtonStyles/ButtonStyle_Feature_L_Yellow.ButtonStyle_Feature_L_Yellow_C
    triggeringInputAction {dataTable=&InputActionDataTable, rowName=Confirm-TrapDefender}
   }
   slot {class=HorizontalBoxSlot, properties {padding=<ltrb 10 0 30 0>}}
  }
  Button_Reset {
   widgetClassPath=&IconTextButton
   bIsVariable=true
   parent=ButtonsArea
   widget=Button_Reset
   properties {
    buttonText=Default Location
    minWidth=300
    onButtonBaseClicked=(&self:WidgetTree.Button_Reset.HandleOnButtonClicked)
    onButtonBaseDoubleClicked=(&self:WidgetTree.Button_Reset.HandleOnButtonDoubleClicked)
    onButtonBaseHovered=(&self:WidgetTree.Button_Reset.HandleOnButtonHovered)
    onButtonBaseUnhovered=(&self:WidgetTree.Button_Reset.HandleOnButtonUnhovered)
    onSelectedChangedBase=(&self:WidgetTree.Button_Reset.HandleOnSelectedChanged)
    style=/FortUILibrary/Foundation/Buttons/ButtonStyles/Skew/ButtonStyle-Skew_Red.ButtonStyle-Skew_Red_C
    triggeringInputAction {dataTable=&InputActionDataTable, rowName=PreviewInspectorReset}
   }
   slot {class=HorizontalBoxSlot}
  }
  Cancel {
   widgetClassPath=&IconTextButton
   bIsVariable=true
   parent=ButtonsArea
   widget=Cancel
   properties {
    buttonText=Cancel
    minWidth=50
    onButtonBaseClicked=(&self:WidgetTree.Cancel.HandleOnButtonClicked)
    onButtonBaseDoubleClicked=(&self:WidgetTree.Cancel.HandleOnButtonDoubleClicked)
    onButtonBaseHovered=(&self:WidgetTree.Cancel.HandleOnButtonHovered)
    onButtonBaseUnhovered=(&self:WidgetTree.Cancel.HandleOnButtonUnhovered)
    onSelectedChangedBase=(&self:WidgetTree.Cancel.HandleOnSelectedChanged)
    style=/FortUILibrary/Foundation/Buttons/ButtonStyles/Athena/Athena_ButtonStyle_AngledBlueMenuButton.Athena_ButtonStyle_AngledBlueMenuButton_C
    triggeringInputAction {dataTable=&InputActionDataTable, rowName=GenericBack}
   }
   slot {class=HorizontalBoxSlot, properties {padding=<ltrb 50 0 0 0>}}
  }
  PreviewScreenshotMessageWidget {
   widgetClassPath=/Game/Athena/UI/InGame/Creative/PreviewScreenshotMessageWidget
   bIsVariable=true
   parent=CanvasPanel_77
   widget=PreviewScreenshotMessageWidget
   slot {
    class=CanvasPanelSlot
    properties {layoutData {alignment=<0 0>, anchors {maximum=<0.5 0>, minimum=<0.5 0>}, offsets=<ltrb -228.9609 472 480 160.02>}}
   }
  }
 ]
}
