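# Situational HUD feedback widget that shows keybind hints for Creative moderator mode: flight, phase, invulnerability, moderator bot, screenshot and tool actions.
# On construct it fills parallel action-name and label arrays into inherited feedback lines, then shows or hides with moderator mode.
# While active, bound keybind, ghost, invulnerability and flight delegates trigger Update.
# Update rewrites toggle and speed labels, adds hold hints for gamepad, and collapses flight-dependent lines when not flying.
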WidgetBlueprint /Game/Athena/HUD/CreativeModeratorModeHUDControls
graphs [
 BindDelegates {nodes [
   BindDelegates() {is=title K2Node_FunctionEntry}
   > Bind Event to On Keybinds Changed(
    self:Fort Global UIContext Object Reference=Get FortGlobalUIContext() -> (ReturnValue:Fort Global UIContext Object Reference) {is=title K2Node_GetSubsystem}
    Delegate=Create Event()
   ) {is=title K2Node_AddDelegate}
   n1 = Create Event()
   n2 = Get CreativeHUDLContext!() -> (ReturnValue:Creative HUDLContext Object Reference) {is=title K2Node_GetBlueprintContextEx}
   > Bind Event to On Creative Ghost Changed(self:Creative HUDLContext Object Reference=@n2, Delegate=@n1) {is=title K2Node_AddDelegate}
   > Bind Event to On Creative Invulnerable Changed(self:Creative HUDLContext Object Reference=@n2, Delegate=@n1) {is=title K2Node_AddDelegate}
   > n3 = Cast To FortPlayerControllerGameplay!(execute, Object=GetOwningPlayer()) -> (then, CastFailed, AsFort Player Controller Gameplay:Fort Player Controller Gameplay Object Reference) {is=title K2Node_DynamicCast}
   > Bind Event to On Creative Fly Changed Delegate(self:Fort Player Controller Gameplay Object Reference=@n3, Delegate=@n1) {is=title K2Node_AddDelegate}
   > Bind Event to On Creative Flying Speed Changed(self:Fort Player Controller Gameplay Object Reference=@n3, Delegate=Create Event()) {is=title K2Node_AddDelegate}
  ]}
 EventGraph {nodes [
   Handle On Creative Flying Speed Changed!() -> (OutputDelegate, then, FlyingSpeedLevel:int, bPlayFX:bool=~) {is=title K2Node_CustomEvent}
   > Update!(execute, self, Unused:bool=false) -> (then)
   ClearAndSetFeedbackLines() {is=title K2Node_CustomEvent}
   > ClearChildren(self=Get Feedback() -> (Feedback:Vertical Box Object Reference))
   > n1 = For Each Loop(Array:Array of Names=Get ActionNames() -> (ActionNames:Array of Names)) -> (Array Element:Name)
   n2 = GetChildrenCount!(self:Panel Widget Object Reference=Get Feedback()) -> (ReturnValue:int=0)
   > Branch(Condition="integer < integer"!(A:int=@n2, B:int=Get MaxLinesNeeded() -> (MaxLinesNeeded:int=0)) -> (ReturnValue:bool) {is=title K2Node_CallFunction})
   n3 = "Get (a copy)"(
    Array:Array of Athena HUDFeedback Line Object References=Get FeedbackLines() -> (FeedbackLines:Array of Athena HUDFeedback Line Object References)
    Dimension 1=@n2
   ) -> (Output:Athena HUDFeedback Line Object Reference)
   > SetVisibility(self=@n3, InVisibility=SelfHitTestInvisible)
   > SetBoundAction!(
    execute
    self:Fort Keybind Widget Object Reference=Get FeedbackKeyBind!(self:Athena HUDFeedback Line Object Reference=@n3) -> (FeedbackKeyBind:Keybind Widget Object Reference)
    NewBoundAction:Name=@n1
   ) -> (then)
   > SetText(
    self=Get FeedbackText!(self:Athena HUDFeedback Line Object Reference=@n3) -> (FeedbackText:Rich Text Block Object Reference)
    InText="Get (a ref)"(Array:Array of Texts=Get ActionLabels() -> (ActionLabels:Array of Texts), Dimension 1=@n1.Array Index) -> (Output:"Text (by ref)")
   )
   > AddChild(self=Get Feedback(), Content=@n3)
   Handle On Keybinds Changed() {is=title K2Node_CustomEvent}
   > Update(Unused=false)
   n4 = Handle On Is Creative Moderator Mode Active Changed!() -> (OutputDelegate, then, bIsCreativeModeratorModeActive:bool=~) {is=title K2Node_CustomEvent}
   > n5 = Branch(Condition=@n4.bIsCreativeModeratorModeActive)
   > BindDelegates() {is=title K2Node_CallFunction}
   > Update(Unused=false)
   > ShowFeedback(self:Athena HUDSituational Feedback Object Reference)
   UnbindDelegates(execute=@n5.else)
   > HideFeedback(self:Athena HUDSituational Feedback Object Reference)
   Update!() -> (OutputDelegate, then, Unused:bool=~) {is=title K2Node_CustomEvent}
   > n6 = Sequence!(execute) -> (then_0, then_1, then_2:Exec)
   > Set IsUsingGamepad(IsUsingGamepad:bool=IsUsingGamepad!(self:Fort Global UIContext Object Reference=Get FortGlobalUIContext!() -> (ReturnValue:Fort Global UIContext Object Reference) {is=title K2Node_GetBlueprintContextEx}) -> (ReturnValue:bool)) -> (Output_Get:bool)
   > n7 = Cast To FortPlayerPawnAthena(Object=GetOwningPlayerPawn!(self:User Widget Object Reference) -> (ReturnValue:Pawn Object Reference))
   > n8 = Sequence()
   n9 = Get CharacterMovement!(self:Character Object Reference=@n7) -> (CharacterMovement:Character Movement Component Object Reference)
   > Is Valid(InputObject=@n9)
   > Set IsFlying(IsFlying:bool="Equal (Enum)"(
     A:EMovementMode Enum=Get MovementMode!(self:Character Movement Component Object Reference=@n9) -> (MovementMode:EMovementMode Enum=MOVE_None)
     B:EMovementMode Enum=MOVE_Flying
    )) -> (Output_Get:bool)
   Set IsInGhostMode(
    execute=@n8.then_1
    IsInGhostMode:bool=Get bIsCreativeGhostModeActivated!(self:Fort Player Pawn Athena Object Reference=@n7) -> (bIsCreativeGhostModeActivated:bool)
   ) -> (Output_Get:bool)
   Set IsInvulnerable(execute=@n8.then_2, IsInvulnerable:bool=Get bIsInvulnerable!(self:Fort Pawn Object Reference=@n7) -> (bIsInvulnerable:bool)) -> (Output_Get:bool)
   Set Text by Index!(
    exec=@n6.then_1
    TextToSet:Text=Make Toggle Label!(Label:Text=Get GhostModeLabelText() -> (GhostModeLabelText:Text), Toggle Value:bool=Get IsInGhostMode() -> (IsInGhostMode:bool)) -> (Return Value:Text) {is=title K2Node_MacroInstance}
    Index:int=Get GhostModeIndex() -> (GhostModeIndex:int=0)
   ) -> (outExec:Exec) {is=title K2Node_MacroInstance}
   > Set Text by Index(
    TextToSet=Make Toggle Label(Label=Get InvulnerabilityLabelText() -> (InvulnerabilityLabelText:Text), Toggle Value=Get IsInvulnerable() -> (IsInvulnerable:bool))
    Index=Get InvulnerabilityTextIndex() -> (InvulnerabilityTextIndex:int=0)
   )
   > Set Text by Index(
    TextToSet=Make Flight Speed Label!() -> (FlightSpeedLabel:Text) {is=title K2Node_MacroInstance}
    Index=Get FlightSpeedTextIndex() -> (FlightSpeedTextIndex:int=0)
   )
   n10 = Get Feedback()
   n11 = For Loop(execute=@n6.then_2, LastIndex=int - int(A=GetChildrenCount(self=@n10) -> (ReturnValue=0), B=1) {is=title K2Node_CallFunction})
   n12 = Is Feedback Line Visible!(InIndex:int=@n11, InputPin:int) -> (OutIndex:int, Is Visible:bool=~) {is=title K2Node_MacroInstance}
   > SetVisibility(
    self=GetChildAt!(self:Panel Widget Object Reference=Get Feedback(), Index:int=@n12) -> (ReturnValue:Widget Object Reference)
    InVisibility=Select(Option 0=Collapsed, Option 1=HitTestInvisible, Index=@n12.Is Visible)
   )
   Event Destruct()
   n13 = Get CreativeHUDLContext()
   > Is Valid(InputObject=@n13)
   > Unbind Event from On Is Creative Moderator Mode Active Changed(self:Creative HUDLContext Object Reference=@n13, Delegate=Create Event()) {is=title K2Node_RemoveDelegate}
   > UnbindDelegates()
   Event Construct()
   > Assign On Is Creative Moderator Mode Active Changed!(execute:Exec, self:Creative HUDLContext Object Reference=Get CreativeHUDLContext(), Delegate:"Delegate (by ref)"=Create Event()) -> (then:Exec) {is=title K2Node_AssignDelegate}
   > SetupActionNamesAndLabels()
   > Set MaxLinesNeeded(MaxLinesNeeded:int=Length(TargetArray:Array of Names=Get ActionNames())) -> (Output_Get:int=0)
   > Parent: Construct()
   > ClearAndSetFeedbackLines() {is=title K2Node_CallFunction}
   > HideFeedback()
  ]}
 IsFeedbackLineVisible {nodes [
   n1 = Get FlyTextIndex() -> (FlyTextIndex:int=0)
   n2 = Inputs!() -> (InIndex:int, InputPin:int)
   n3 = Get InvulnerabilityTextIndex() -> (InvulnerabilityTextIndex=0)
   n4 = Get DespawnModeratorBotTextIndex() -> (DespawnModeratorBotTextIndex:int=0)
   n5 = Get SpawnModeratorBotTextIndex() -> (SpawnModeratorBotTextIndex:int=0)
   n6 = Get OpenModeratorPhotoModeTextIndex() -> (OpenModeratorPhotoModeTextIndex:int=0)
   n7 = OR Boolean!(
    A="Equal (Integer)"(A=@n2, B=Get FlyUpTextIndex() -> (FlyUpTextIndex:int=0)) {is=title K2Node_CallFunction}
    B="Equal (Integer)"(A=@n2, B=Get FlyDownTextIndex() -> (FlyDownTextIndex:int=0))
    C:bool="Equal (Integer)"(A=@n2, B=Get GhostModeIndex() -> (GhostModeIndex=0))
    D:bool="Equal (Integer)"(A=@n2, B=Get FlightSpeedTextIndex() -> (FlightSpeedTextIndex=0)) {is=title K2Node_PromotableOperator}
   ) -> (ReturnValue)
   n8 = Get IsFlying() -> (IsFlying:bool)
   n9 = Get OpenModeratorToolModeTextIndex() -> (OpenModeratorToolModeTextIndex:int=0)
   Outputs!(
    OutIndex:int=@n2
    Is Visible:bool=OR Boolean!(
     A="Equal (Integer)"(A=@n2, B=@n1) {is=title K2Node_CallFunction}
     B="Equal (Integer)"(A=@n2, B=@n3)
     C="Equal (Integer)"(A=@n2, B=@n5)
     D="Equal (Integer)"(A=@n2, B=@n4)
     E:bool="Equal (Integer)"(A=@n2, B=@n6)
     F:bool=AND Boolean(A=@n8, B=@n7)
     G:bool="Equal (Integer)"(A=@n2, B=@n9) {is=title K2Node_PromotableOperator}
    ) -> (ReturnValue)
   )
  ]}
 MakeFlightSpeedLabel {nodes [
   Inputs!()
   n1 = Get FlightSpeeds() -> (FlightSpeeds:"Array of Float (double-precision)s")
   n2 = Cast To FortPlayerControllerGameplay!(Object=GetOwningPlayer()) -> (AsFort Player Controller Gameplay, bSuccess:bool=~)
   n3 = "Get (a ref)"(
    Array:"Array of Float (double-precision)s"=Get FlightSpeeds()
    Dimension 1="Clamp (Integer)"(
     Value=GetFlyingModifierIndex!(self:Fort Player Controller Gameplay Object Reference=@n2) -> (ReturnValue:int=0)
     Min=0
     Max=Last Index!(TargetArray:"Array of Float (double-precision)s"=@n1) -> (ReturnValue=0)
    ) -> (ReturnValue=0)
   ) -> (Output:"Float (double-precision) (by ref)")
   Outputs!(FlightSpeedLabel:Text=Format Text!(
     Format="LOCTEXT(\"{0} [{1}x]\")"
     0:Text=Get FlightSpeedLabelText() -> (FlightSpeedLabelText:Text)
     1:Text="To Text (Float)"(Value=@n3, bUseGrouping=false, MinimumIntegralDigits=1, MaximumIntegralDigits=324, MinimumFractionalDigits=1, MaximumFractionalDigits=1)
    ) -> (Result))
  ]}
 MakeToggleLabel {nodes [
   n1 = Get OffLabelText() -> (OffLabelText:Text)
   n2 = Get OnLabelText() -> (OnLabelText:Text)
   n3 = Inputs!() -> (Label:Text, Toggle Value:bool=~)
   n4 = MakeLiteralText()
   n5 = Get HoldLabelText() -> (HoldLabelText:Text)
   n6 = Get IsUsingGamepad() -> (IsUsingGamepad:bool)
   Outputs!(Return Value:Text=Format Text!(
     Format="LOCTEXT(\"{Label}{Hold} {Mode}\")"
     Label:Text=@n3
     Hold:"Text (by ref)"=Select(Option 0:"Text (by ref)"=@n4, Option 1:"Text (by ref)"=@n5, Index=@n6) -> (ReturnValue:"Text (by ref)")
     Mode:"Text (by ref)"=Select(Option 0=@n1, Option 1=@n2, Index=@n3.Toggle Value)
    ) -> (Result))
  ]}
 SetTextByIndex {nodes [
   n1 = Inputs!() -> (exec:Exec, TextToSet:Text, Index:int)
   n2 = Get FeedbackLines()
   > SetText(self=Get FeedbackText(self="Get (a copy)"(Array=@n2, Dimension 1=@n1.Index)), InText=@n1)
   > Outputs!(outExec:Exec)
  ]}
 SetupActionNamesAndLabels {nodes [
   SetupActionNamesAndLabels() {is=title K2Node_FunctionEntry}
   n1 = Get FortGlobalUIContext()
   n2 = GetJumpActionName!(self:Fort Global UIContext Object Reference=@n1) -> (ReturnValue:Name=None)
   > Set ActionNames(ActionNames:Array of Names=Make Array!(
     "[0]":Name=@n2
     "[1]":Name=GetCreativeMoveObjectsFreelyPullName!(self:Fort Global UIContext Object Reference=@n1) -> (ReturnValue:Name=None)
     "[2]":Name=@n2
     "[3]":Name=GetCreativeFlyDownActionName!(self:Fort Global UIContext Object Reference=@n1) -> (ReturnValue:Name=None)
     "[4]":Name=GetCreativeToggleGhostName!(self:Fort Global UIContext Object Reference=@n1) -> (ReturnValue:Name=None)
     "[5]":Name=GetCreativeToggleInvulnerablityName!(self:Fort Global UIContext Object Reference=@n1) -> (ReturnValue:Name=None)
     "[6]":Name=GetCreativeSpawnModeratorBotName!(self:Fort Global UIContext Object Reference=@n1) -> (ReturnValue:Name=None)
     "[7]":Name=GetCreativeDespawnModeratorBotName!(self:Fort Global UIContext Object Reference=@n1) -> (ReturnValue:Name=None)
     "[8]":Name=GetCreativeModeratorOpenPhotoModeName!(self:Fort Global UIContext Object Reference=@n1) -> (ReturnValue:Name=None)
     "[9]":Name=GetCreativeModeratorOpenToolModeName!(self:Fort Global UIContext Object Reference=@n1) -> (ReturnValue:Name=None)
    ) -> (Array:Array of Names)) -> (Output_Get:Array of Names)
   > Set ActionLabels(ActionLabels:Array of Texts=Make Array(
     "[0]":Text=Get FlyLabelText() -> (FlyLabelText:Text)
     "[1]":Text=Get FlightSpeedLabelText()
     "[2]":Text=Get FlyUpLabelText() -> (FlyUpLabelText:Text)
     "[3]":Text=Get FlyDownLabelText() -> (FlyDownLabelText:Text)
     "[4]":Text=Get GhostModeLabelText()
     "[5]":Text=Get InvulnerabilityLabelText()
     "[6]":Text=Get SpawnModeratorBotLabelText() -> (SpawnModeratorBotLabelText:Text)
     "[7]":Text=Get DespawnModeratorBotLabelText() -> (DespawnModeratorBotLabelText:Text)
     "[8]":Text=Get OpenModeratorPhotoModeLabelText() -> (OpenModeratorPhotoModeLabelText:Text)
     "[9]":Text=Get OpenModeratorToolModeLabelText() -> (OpenModeratorToolModeLabelText:Text)
    ) -> (Array:Array of Texts)) -> (Output_Get:Array of Texts)
   > Set FlyTextIndex(FlyTextIndex:int=0) -> (Output_Get:int=0)
   > Set FlightSpeedTextIndex(FlightSpeedTextIndex:int=1) -> (Output_Get:int=0)
   > Set FlyUpTextIndex(FlyUpTextIndex:int=2) -> (Output_Get:int=0)
   > Set FlyDownTextIndex(FlyDownTextIndex:int=3) -> (Output_Get:int=0)
   > Set GhostModeIndex(GhostModeIndex:int=4) -> (Output_Get:int=0)
   > Set InvulnerabilityTextIndex(InvulnerabilityTextIndex:int=5) -> (Output_Get:int=0)
   > Set SpawnModeratorBotTextIndex(SpawnModeratorBotTextIndex:int=6) -> (Output_Get:int=0)
   > Set DespawnModeratorBotTextIndex(DespawnModeratorBotTextIndex:int=7) -> (Output_Get:int=0)
   > Set OpenModeratorPhotoModeTextIndex(OpenModeratorPhotoModeTextIndex:int=8) -> (Output_Get:int=0)
   > Set OpenModeratorToolModeTextIndex(OpenModeratorToolModeTextIndex:int=9) -> (Output_Get:int=0)
  ]}
 UnbindDelegates {nodes [
   UnbindDelegates() {is=title K2Node_FunctionEntry}
   > Unbind Event from On Keybinds Changed(self:Fort Global UIContext Object Reference=Get FortGlobalUIContext() {is=title K2Node_GetSubsystem}, Delegate=Create Event()) {is=title K2Node_RemoveDelegate}
   n1 = Create Event()
   n2 = Get CreativeHUDLContext()
   > Unbind Event from On Creative Ghost Changed(self:Creative HUDLContext Object Reference=@n2, Delegate=@n1) {is=title K2Node_RemoveDelegate}
   > Unbind Event from On Creative Invulnerable Changed(self:Creative HUDLContext Object Reference=@n2, Delegate=@n1) {is=title K2Node_RemoveDelegate}
   > n3 = Cast To FortPlayerControllerGameplay!(execute:Exec, Object=GetOwningPlayer()) -> (then:Exec, CastFailed:Exec, AsFort Player Controller Gameplay)
   > Unbind Event from On Creative Fly Changed Delegate(self:Fort Player Controller Gameplay Object Reference=@n3, Delegate=@n1) {is=title K2Node_RemoveDelegate}
   > Unbind Event from On Creative Flying Speed Changed(self:Fort Player Controller Gameplay Object Reference=@n3, Delegate=Create Event()) {is=title K2Node_RemoveDelegate}
  ]}
]
variables [
 ActionLabels {type="array<string>", default []}
 ActionNames {type="array<string>", default []}
 DespawnModeratorBotLabelText {type=string, default=Despawn Moderator Bot}
 DespawnModeratorBotTextIndex {type=integer}
 FlightSpeedLabelText {type=string, default=Flight Speed}
 FlightSpeeds {type="array<number>", default [0.1,0.5,1,1.5,2,3,4,5,10,15,20,25,30]}
 FlightSpeedTextIndex {type=integer}
 FlyDownLabelText {type=string, default=Fly Down}
 FlyDownTextIndex {type=integer}
 FlyLabelText {type=string, default="Fly (Double-Tap)"}
 FlyTextIndex {type=integer}
 FlyUpLabelText {type=string, default=Fly Up}
 FlyUpTextIndex {type=integer}
 GhostModeIndex {type=integer}
 GhostModeLabelText {type=string, default=Phase}
 HoldLabelText {type=string, default=(Hold)}
 InvulnerabilityLabelText {type=string, default=Invulnerability}
 InvulnerabilityTextIndex {type=integer}
 IsFlying {type=boolean}
 IsInGhostMode {type=boolean}
 IsInvulnerable {type=boolean}
 IsUsingGamepad {type=boolean}
 OffLabelText {type=string, default="[Off]"}
 OnLabelText {type=string, default="[On]"}
 OpenModeratorPhotoModeLabelText {type=string, default=Open Screenshot Tool}
 OpenModeratorPhotoModeTextIndex {type=integer}
 OpenModeratorToolModeLabelText {type=string, default=Open Moderator Tool}
 OpenModeratorToolModeTextIndex {type=integer}
 SpawnModeratorBotLabelText {type=string, default=Spawn Moderator Bot}
 SpawnModeratorBotTextIndex {type=integer}
]
widgets {
 info {
  inheritedWidgetCount=0
  parentClass=&/AthenaHUDSituationalFeedback.AthenaHUDSituationalFeedback_C
  rootWidgetClass=null
  widgetCount=0
 }
 widgets []
}
