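# AthenaTeamScoreWidget is a score-LTM HUD widget that shows team points, placement, and leaderboard standing.
# Score events update the text and a GoalScore-relative material progress bar, then play ScoreUp.
# Placement events switch gold, silver, bronze or generic plates, collapse those past MinimumDisplayPlacement, and play RankUp unless a worse displayable place or an instant update applies.
# Leaderboard events switch between winner, tied, leading and trailing text, looping WinnerShine for winners.
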WidgetBlueprint /Game/Athena/HUD/AthenaTeamScoreWidget
animations [
 RankUp {
  binding_count=8
  display_label=""
  length_seconds=0.4664
  object=&self:RankUp
  sequence {bindings [
    Image_2 {
     id=#1
     possessed_class=Image
     tracks [{
       type=MovieSceneWidgetMaterialTrack
       display_name=Brush Material
       sections [{
         type=MovieSceneParameterSection
         end_frame=9
         start_frame=0
         channels [FillAmount {type=MovieSceneScriptingFloatChannel, keys [[0,1],[7,0]]}]
        }]
      }]
    }
    Image_5 {
     id=#2
     possessed_class=Image
     tracks [
      {
       type=MovieSceneFloatTrack
       display_name=Render Opacity
       sections [{type=MovieSceneFloatSection, end_frame=9, start_frame=0, channels [None {type=MovieSceneScriptingFloatChannel, keys [[0,1],[9,0]]}]}]
      }
      {
       type=MovieScene2DTransformTrack
       sections [{
         type=MovieScene2DTransformSection
         end_frame=9
         start_frame=0
         channels [
          Translation.X {type=MovieSceneScriptingFloatChannel}
          Translation.Y {type=MovieSceneScriptingFloatChannel}
          Angle {type=MovieSceneScriptingFloatChannel}
          Scale.X {type=MovieSceneScriptingFloatChannel, keys [[0,1],[9,3.0745]]}
          Scale.Y {type=MovieSceneScriptingFloatChannel, keys [[0,1],[9,3.0745]]}
          Shear.X {type=MovieSceneScriptingFloatChannel}
          Shear.Y {type=MovieSceneScriptingFloatChannel}
         ]
        }]
      }
      {
       type=MovieSceneWidgetMaterialTrack
       display_name=Brush Material
       sections [{type=MovieSceneParameterSection, end_frame=9, start_frame=0, channels [FillAmount {type=MovieSceneScriptingFloatChannel, const [1,1]}]}]
      }
     ]
    }
    Image_7 {
     id=#3
     possessed_class=Image
     tracks [
      {
       type=MovieSceneFloatTrack
       display_name=Render Opacity
       sections [{type=MovieSceneFloatSection, end_frame=9, start_frame=0, channels [None {type=MovieSceneScriptingFloatChannel, keys [[0,1],[9,0]]}]}]
      }
      {
       type=MovieScene2DTransformTrack
       sections [{
         type=MovieScene2DTransformSection
         end_frame=9
         start_frame=0
         channels [
          Translation.X {type=MovieSceneScriptingFloatChannel}
          Translation.Y {type=MovieSceneScriptingFloatChannel}
          Angle {type=MovieSceneScriptingFloatChannel}
          Scale.X {type=MovieSceneScriptingFloatChannel, keys [[0,1],[9,3.0745]]}
          Scale.Y {type=MovieSceneScriptingFloatChannel, keys [[0,1],[9,3.0745]]}
          Shear.X {type=MovieSceneScriptingFloatChannel}
          Shear.Y {type=MovieSceneScriptingFloatChannel}
         ]
        }]
      }
      {
       type=MovieSceneWidgetMaterialTrack
       display_name=Brush Material
       sections [{type=MovieSceneParameterSection, end_frame=9, start_frame=0, channels [FillAmount {type=MovieSceneScriptingFloatChannel, const [1,1]}]}]
      }
     ]
    }
    Image_6 {
     id=#4
     possessed_class=Image
     tracks [
      {
       type=MovieSceneFloatTrack
       display_name=Render Opacity
       sections [{type=MovieSceneFloatSection, end_frame=9, start_frame=0, channels [None {type=MovieSceneScriptingFloatChannel, keys [[0,1],[9,0]]}]}]
      }
      {
       type=MovieScene2DTransformTrack
       sections [{
         type=MovieScene2DTransformSection
         end_frame=9
         start_frame=0
         channels [
          Translation.X {type=MovieSceneScriptingFloatChannel}
          Translation.Y {type=MovieSceneScriptingFloatChannel}
          Angle {type=MovieSceneScriptingFloatChannel}
          Scale.X {type=MovieSceneScriptingFloatChannel, keys [[0,1],[9,3.0745]]}
          Scale.Y {type=MovieSceneScriptingFloatChannel, keys [[0,1],[9,3.0745]]}
          Shear.X {type=MovieSceneScriptingFloatChannel}
          Shear.Y {type=MovieSceneScriptingFloatChannel}
         ]
        }]
      }
      {
       type=MovieSceneWidgetMaterialTrack
       display_name=Brush Material
       sections [{type=MovieSceneParameterSection, end_frame=9, start_frame=0, channels [FillAmount {type=MovieSceneScriptingFloatChannel, const [1,1]}]}]
      }
     ]
    }
    Image_1 {
     id=#5
     possessed_class=Image
     tracks [{
       type=MovieSceneWidgetMaterialTrack
       display_name=Brush Material
       sections [{
         type=MovieSceneParameterSection
         end_frame=9
         start_frame=0
         channels [FillAmount {type=MovieSceneScriptingFloatChannel, keys [[0,1],[7,0]]}]
        }]
      }]
    }
    Image_4 {
     id=#6
     possessed_class=Image
     tracks [
      {
       type=MovieSceneFloatTrack
       display_name=Render Opacity
       sections [{type=MovieSceneFloatSection, end_frame=9, start_frame=0, channels [None {type=MovieSceneScriptingFloatChannel, keys [[0,1],[9,0]]}]}]
      }
      {
       type=MovieScene2DTransformTrack
       sections [{
         type=MovieScene2DTransformSection
         end_frame=9
         start_frame=0
         channels [
          Translation.X {type=MovieSceneScriptingFloatChannel}
          Translation.Y {type=MovieSceneScriptingFloatChannel}
          Angle {type=MovieSceneScriptingFloatChannel}
          Scale.X {type=MovieSceneScriptingFloatChannel, keys [[0,1],[9,3.0745]]}
          Scale.Y {type=MovieSceneScriptingFloatChannel, keys [[0,1],[9,3.0745]]}
          Shear.X {type=MovieSceneScriptingFloatChannel}
          Shear.Y {type=MovieSceneScriptingFloatChannel}
         ]
        }]
      }
      {
       type=MovieSceneWidgetMaterialTrack
       display_name=Brush Material
       sections [{type=MovieSceneParameterSection, end_frame=9, start_frame=0, channels [FillAmount {type=MovieSceneScriptingFloatChannel, const [1,1]}]}]
      }
     ]
    }
    Image_Generic {
     id=#7
     possessed_class=Image
     tracks [{
       type=MovieSceneWidgetMaterialTrack
       display_name=Brush Material
       sections [{
         type=MovieSceneParameterSection
         end_frame=9
         start_frame=0
         channels [FillAmount {type=MovieSceneScriptingFloatChannel, keys [[0,1],[7,0]]}]
        }]
      }]
    }
    Image_3 {
     id=#8
     possessed_class=Image
     tracks [{
       type=MovieSceneWidgetMaterialTrack
       display_name=Brush Material
       sections [{
         type=MovieSceneParameterSection
         end_frame=9
         start_frame=0
         channels [FillAmount {type=MovieSceneScriptingFloatChannel, keys [[0,1],[7,0]]}]
        }]
      }]
    }
   ]}
 }
 ScoreUp {
  binding_count=2
  display_label=""
  length_seconds=0.39
  object=&self:ScoreUp
  sequence {
   bindings [
    ScoreTextBox {
     id=#9
     possessed_class=CommonTextBlock
     tracks [{
       type=MovieScene2DTransformTrack
       sections [{
         type=MovieScene2DTransformSection
         end_frame=39
         start_frame=0
         channels [
          Translation.X {type=MovieSceneScriptingFloatChannel}
          Translation.Y {type=MovieSceneScriptingFloatChannel}
          Angle {type=MovieSceneScriptingFloatChannel}
          Scale.X {type=MovieSceneScriptingFloatChannel, keys [[0,1],[4,1.2],[20,1]]}
          Scale.Y {type=MovieSceneScriptingFloatChannel, keys [[0,1],[4,1.2],[20,1]]}
          Shear.X {type=MovieSceneScriptingFloatChannel}
          Shear.Y {type=MovieSceneScriptingFloatChannel}
         ]
        }]
      }]
    }
    Image_BackplateProgress {
     id=#10
     possessed_class=Image
     tracks [{
       type=MovieSceneWidgetMaterialTrack
       display_name=Brush Material
       sections [{
         type=MovieSceneParameterSection
         end_frame=39
         start_frame=0
         channels [ShineOffset {type=MovieSceneScriptingFloatChannel, keys [[0,0.7621],[39,-1]]}]
        }]
      }]
    }
   ]
   display_rate {numerator=100}
  }
 }
 WinnerShine {
  binding_count=1
  length_seconds=0.5499
  object=&self:WinnerShine
  sequence {bindings [isWinner {
     id=#11
     possessed_class=CommonTextBlock
     tracks [{
       type=MovieSceneWidgetMaterialTrack
       display_name=Font.FontMaterial
       sections [{type=MovieSceneParameterSection, channels [ShineOffset {type=MovieSceneScriptingFloatChannel, keys [[3.0157,0.5647],[8.4867,-0.5453]]}]}]
      }]
    }]}
 }
]
graphs [
 EventGraph {nodes [
   n1 = UpdatePlacement!(execute, self, Place:int=Get PreviewPlace() -> (PreviewPlace:int=0), Instant:bool=true) -> (then)
   Branch(execute=[@n1.then, @n4.then, @n4.else], Condition=Get isBacchus() -> (isBacchus:bool))
   n2 = Get OverlayPlacePlate() -> (OverlayPlacePlate:Overlay Object Reference)
   > SetHorizontalAlignment(self=SlotAsOverlaySlot(Widget=@n2), InHorizontalAlignment=HAlign_Right)
   > SetRenderTranslation(self=@n2, Translation=(X=61,Y=0))
   n3 = Event PreConstruct()
   > n4 = Branch(Condition=@n3.IsDesignTime)
   Event Construct()
   > "SetText (Text)"(self=Get ScoreTextBox() -> (ScoreTextBox:Common Text Object Reference))
   > n5 = GetDynamicMaterial(self=Get Image_BackplateProgress() -> (Image_BackplateProgress:Image Object Reference))
   > n6 = Set ProgressBarDMI(ProgressBarDMI:Material Instance Dynamic Object Reference=@n5) -> (Output_Get:Material Instance Dynamic Object Reference)
   > SetScalarParameterValue(self=@n6, ParameterName=PercentFill, Value=0)
   n7 = Event OnLeaderboardChanged!() -> (OutputDelegate, then, ScoreDelta:int=0, bIsWinner:bool, bHasScore:bool) {is=title K2Node_Event}
   > UpdateLeaderboard!(execute, self, PointsDelta:int=@n7.ScoreDelta, IsWinner:bool=@n7.bIsWinner, HasScore:bool=@n7.bHasScore) -> (then)
   n8 = Event OnScoreChanged!() -> (OutputDelegate, then, Score:int=0) {is=title K2Node_Event}
   > UpdateScore!(execute, self, Score:int=@n8.Score) -> (then)
   n9 = Event OnPlacementChanged!() -> (OutputDelegate, then, Placement:int=0) {is=title K2Node_Event}
   > UpdatePlacement(Place=@n9.Placement, Instant=false)
   SetRenderTransform!(
    execute
    self:Widget Object Reference=@n2
    InTransform:Widget Transform Structure=Make Widget Transform!(Translation:Vector 2D Structure=(X=61,Y=0), Scale:Vector 2D Structure=(X=1,Y=1), Shear:Vector 2D Structure=(X=0,Y=0), Angle:float=0) -> (WidgetTransform:Widget Transform Structure) {is=title K2Node_MakeStruct}
   ) -> (then)
  ]}
 UpdateLeaderboard {nodes [
   n1 = UpdateLeaderboard!() -> (then, PointsDelta:int, IsWinner:bool=~, HasScore:bool=~) {is=title K2Node_FunctionEntry}
   > Branch(Condition=OR Boolean(A="Not Equal (Integer)"!(A:int=@n1, B:int=0) -> (ReturnValue:bool) {is=title K2Node_CallFunction}, B=@n1.HasScore))
   > n2 = Sequence()
   > n3 = Branch(Condition=@n1.IsWinner)
   > SetActiveWidget(
    self:Widget Switcher Object Reference=Get CommonWidgetSwitcher_GameStatus() -> (CommonWidgetSwitcher_GameStatus:Common Widget Switcher Legacy Object Reference)
    Widget=Get IsWinner() -> (IsWinner:Common Text Object Reference)
   )
   > PlayAnimation(InAnimation=Get WinnerShine() -> (WinnerShine:Widget Animation Object Reference), NumLoopsToPlay=0)
   n4 = Branch(execute=@n3.else, Condition="Equal (Integer)"(A=@n1, B=0) {is=title K2Node_CallFunction})
   > SetActiveWidget(self=Get CommonWidgetSwitcher_GameStatus(), Widget=Get TextBlock_TiedForLead() -> (TextBlock_TiedForLead:Common Text Object Reference))
   n5 = Branch(execute=@n4.else, Condition="integer < integer"!(A:int=@n1, B:int=0) -> (ReturnValue:bool) {is=title K2Node_CallFunction})
   n6 = Get TextBlock_TrailingDelta() -> (TextBlock_TrailingDelta:Common Text Object Reference)
   > SetActiveWidget(self=Get CommonWidgetSwitcher_GameStatus(), Widget=@n6)
   > "SetText (Text)"(
    self=@n6
    InText=Format Text!(Format="LOCTEXT(\"Trailing by {0}\")", 0:int="Absolute (Integer)"!(A:int=@n1) -> (ReturnValue:int=0)) -> (Result)
   )
   n7 = Get TextBlock_LeaderDelta() -> (TextBlock_LeaderDelta:Common Text Object Reference)
   SetActiveWidget(execute=@n5.else, self=Get CommonWidgetSwitcher_GameStatus(), Widget=@n7)
   > "SetText (Text)"(self=@n7, InText=Format Text(Format="LOCTEXT(\"Leading by {0}!\")", 0=@n1))
   SetVisibility(execute=@n2.then_1, self=Get Overlay_Leaderboard() -> (Overlay_Leaderboard:Overlay Object Reference), InVisibility=SelfHitTestInvisible)
  ]}
 UpdatePlacement {nodes [
   n1 = UpdatePlacement!() -> (then, Place:int, Instant:bool=~) {is=title K2Node_FunctionEntry}
   > n2 = Branch(Condition="Equal (Integer)"(A=@n1, B=0))
   > SetVisibility(self=Get OverlayPlacePlate(), InVisibility=Collapsed)
   n3 = "integer <= integer"!(A:int=@n1, B:int=Get MinimumDisplayPlacement() -> (MinimumDisplayPlacement:int=0)) -> (ReturnValue:bool) {is=title K2Node_CallFunction}
   n4 = Select(Option 0:int=0, Option 1:int=@n1, Index=@n3) -> (ReturnValue:int)
   SetActiveWidgetIndex(
    execute=@n2.else
    self=[
     Get WidgetSwitcher_BackgroundFlash() -> (WidgetSwitcher_BackgroundFlash:Widget Switcher Object Reference)
     Get WidgetSwitcher_PlaceBackground() -> (WidgetSwitcher_PlaceBackground:Widget Switcher Object Reference)
    ]
    Index=Select!(Option 0=3, Option 1=0, Option 2:int=1, Option 3:int=2, Option 4:int=3, Index:int=@n4) -> (ReturnValue)
   )
   > SetStyle!(
    execute
    self:Common Text Object Reference=Get Text_PlaceValue() -> (Text_PlaceValue:Common Text Object Reference)
    InStyle:Common Text Style Class Reference=Select(
     Option 0:Common Text Style Class Reference=/Game/Athena/HUD/Materials/Score/TextStyle-ScorePlacementGeneric.TextStyle-ScorePlacementGeneric_C
     Option 1:Common Text Style Class Reference=/Game/Athena/HUD/Materials/Score/TextStyle-ScorePlacementGold.TextStyle-ScorePlacementGold_C
     Option 2:Common Text Style Class Reference=/Game/Athena/HUD/Materials/Score/TextStyle-ScorePlacementSilver.TextStyle-ScorePlacementSilver_C
     Option 3:Common Text Style Class Reference=/Game/Athena/HUD/Materials/Score/TextStyle-ScorePlacementBronze.TextStyle-ScorePlacementBronze_C
     Option 4:Common Text Style Class Reference=&TextStyle-ScorePlacementGeneric_C
     Index=@n4
    ) -> (ReturnValue:Common Text Style Class Reference)
   ) -> (then)
   > n5 = Branch(Condition="integer > integer"(A=@n1, B=3) {is=title K2Node_CallFunction})
   > "SetText (Text)"(self=Get Text_PlaceValue(), InText="To Text (Integer)"(Value=@n1))
   "SetText (Text)"(
    execute=@n5.else
    self=Get Text_PlaceValue()
    InText=Select!(
     Option 0:Text
     Option 1:Text="LOCTEXT(\"1st\")"
     Option 2:Text="LOCTEXT(\"2nd\")"
     Option 3:Text="LOCTEXT(\"3rd\")"
     Option 4:Text
     Option 5:Text
     Index=@n4
    ) -> (ReturnValue:Text)
   )
   > SetVisibility(
    self=Get OverlayPlacePlate()
    InVisibility=Select!(Option 0:ESlateVisibility Enum=Collapsed, Option 1:ESlateVisibility Enum=HitTestInvisible, Index:bool=@n3) -> (ReturnValue:ESlateVisibility Enum)
   )
   n6 = Get CurrentPlaceValue() -> (CurrentPlaceValue:int=0)
   > n7 = Branch(Condition=AND Boolean(A="integer < integer"(A=@n6, B=@n1), B=@n3))
   Set CurrentPlaceValue(execute=[@n9.then, @n7.then, @n8.then], CurrentPlaceValue:int=@n1) -> (Output_Get:int=0)
   > SetVisibility(InVisibility=HitTestInvisible)
   n8 = Branch(execute=@n7.else, Condition=@n1.Instant)
   n9 = PlayAnimation(execute=@n8.else, InAnimation=Get RankUp() -> (RankUp:Widget Animation Object Reference))
  ]}
 UpdateScore {nodes [
   n1 = UpdateScore!() -> (then, Score:int) {is=title K2Node_FunctionEntry}
   > n2 = Branch(Condition="integer >= integer"!(A:int=@n1, B:int=0) -> (ReturnValue:bool) {is=title K2Node_CallFunction})
   > "SetText (Text)"(self=Get ScoreTextBox(), InText=Format Text!(Format="LOCTEXT(\"{points} pts\")", points:int=@n1) -> (Result))
   n3 = Get GoalScore() -> (GoalScore:int=0)
   > SetScalarParameterValue(
    self=Get ProgressBarDMI() -> (ProgressBarDMI:Material Instance Dynamic Object Reference)
    ParameterName=PercentFill
    Value=float / float(A="To Float (Integer)"(InInt=@n1), B="To Float (Integer)"(InInt=@n3)) {is=title K2Node_CallFunction}
   )
   > PlayAnimation(InAnimation=Get ScoreUp() -> (ScoreUp:Widget Animation Object Reference))
   > SetVisibility(InVisibility=HitTestInvisible)
   SetVisibility(execute=@n2.else, InVisibility=Collapsed)
   > "SetText (Text)"(self=Get ScoreTextBox(), InText="LOCTEXT(\"0 pts\")")
  ]}
]
variables [
 CurrentPlaceValue {type=integer, default=200}
 isBacchus {type=boolean}
 MinimumDisplayPlacement {type=integer, default=3}
 PreviewPlace {type=integer, default=1}
 ProgressBarDMI {type=MaterialInstanceDynamic}
]
widgets {
 info {inheritedWidgetCount=1, parentClass=/Script/FortniteUI.AthenaTeamScoreWidgetBase, rootWidgetClass=&Overlay, widgetCount=34}
 widgets [
  Overlay_1 {widgetClassPath=&Overlay, widget=Overlay_1}
  VerticalBoxContent {
   widgetClassPath=&VerticalBox
   parent=Overlay_1
   widget=VerticalBoxContent
   properties {renderTransformPivot=<0 0>}
   slot {class=OverlaySlot, properties {horizontalAlignment=HAlign_Fill, verticalAlignment=VAlign_Fill}}
  }
  Overlay_ProgressBar {
   widgetClassPath=&Overlay
   parent=VerticalBoxContent
   widget=Overlay_ProgressBar
   slot {class=VerticalBoxSlot, properties {verticalAlignment=VAlign_Top}}
  }
  Image_BackplateProgress {
   widgetClassPath=&Image
   bIsVariable=true
   parent=Overlay_ProgressBar
   widget=Image_BackplateProgress
   properties {
    bFlipForRightToLeftFlowDirection=true
    brush {
     bIsDynamicallyLoaded=false
     drawAs=Box
     imageType=NoImage
     mirroring=NoMirror
     resourceName=None
     resourceObject=/Game/Athena/HUD/Materials/Score/M_ScoreLTM_ScoreProgress_Inst
     tiling=NoTile
     imageSize=<32 54.4196>
     margin=<ltrb 0.09 0.09 0.09 0.09>
     outlineSettings {
      bUseBrushTransparency=false
      roundingType=HalfHeightRadius
      width=0
      color {colorUseRule=UseColor_Specified, specifiedColor=<rgba 0 0 0 0>}
      cornerRadii=<0 0 0 0>
     }
     tintColor {colorUseRule=UseColor_Specified, specifiedColor=<rgba 1 1 1 1>}
     uVRegion {bIsValid=false, max=<0 0>, min=<0 0>}
    }
    renderTransform {angle=0, scale=<1 1>, shear=<-5 0>, translation=<0 0>}
   }
   slot {class=OverlaySlot, properties {horizontalAlignment=HAlign_Fill, verticalAlignment=VAlign_Center, padding=<ltrb 35 0 0 0>}}
  }
  Overlay_0 {
   widgetClassPath=&Overlay
   parent=Overlay_ProgressBar
   widget=Overlay_0
   slot {class=OverlaySlot, properties {horizontalAlignment=HAlign_Fill, verticalAlignment=VAlign_Fill}}
  }
  ScaleBoxPlacement {
   widgetClassPath=&ScaleBox
   parent=Overlay_0
   widget=ScaleBoxPlacement
   properties {stretch=UserSpecified, userSpecifiedScale=0.7, renderTransform {angle=0, scale=<1 1>, shear=<0 0>, translation=<-12 0>}}
   slot {class=OverlaySlot, properties {verticalAlignment=VAlign_Center}}
  }
  OverlayPlacePlate {
   widgetClassPath=&Overlay
   bIsVariable=true
   parent=ScaleBoxPlacement
   widget=OverlayPlacePlate
   properties {renderTransform {angle=0, scale=<1 1>, shear=<0 0>, translation=<0 2>}}
   slot {class=ScaleBoxSlot}
  }
  WidgetSwitcher_BackgroundFlash {
   widgetClassPath=&WidgetSwitcher
   bIsVariable=true
   parent=OverlayPlacePlate
   widget=WidgetSwitcher_BackgroundFlash
   properties {activeWidgetIndex=3}
   slot {class=OverlaySlot, properties {horizontalAlignment=HAlign_Center, verticalAlignment=VAlign_Center, padding=<ltrb 0 -75 0 -75>}}
  }
  Image_4 {
   widgetClassPath=&Image
   parent=WidgetSwitcher_BackgroundFlash
   widget=Image_4
   properties {
    brush {
     bIsDynamicallyLoaded=false
     drawAs=Image
     imageType=NoImage
     mirroring=NoMirror
     resourceName=None
     resourceObject=/Game/Athena/HUD/Materials/Score/M_ScoreLTM_PlacementPlate_1st
     tiling=NoTile
     imageSize=<108 79>
     margin=<ltrb 0 0 0 0>
     outlineSettings {
      bUseBrushTransparency=false
      roundingType=HalfHeightRadius
      width=0
      color {colorUseRule=UseColor_Specified, specifiedColor=<rgba 0 0 0 0>}
      cornerRadii=<0 0 0 0>
     }
     tintColor {colorUseRule=UseColor_Specified, specifiedColor=<rgba 1 1 1 1>}
     uVRegion {bIsValid=false, max=<0 0>, min=<0 0>}
    }
    renderTransform {angle=0, scale=<1 1>, shear=<0 0>, translation=<0 4>}
   }
   slot {class=WidgetSwitcherSlot}
  }
  Image_5 {
   widgetClassPath=&Image
   parent=WidgetSwitcher_BackgroundFlash
   widget=Image_5
   properties {brush {
     bIsDynamicallyLoaded=false
     drawAs=Image
     imageType=NoImage
     mirroring=NoMirror
     resourceName=None
     resourceObject=/Game/Athena/HUD/Materials/Score/M_ScoreLTM_PlacementPlate_2nd
     tiling=NoTile
     imageSize=<105 67>
     margin=<ltrb 0 0 0 0>
     outlineSettings {
      bUseBrushTransparency=false
      roundingType=HalfHeightRadius
      width=0
      color {colorUseRule=UseColor_Specified, specifiedColor=<rgba 0 0 0 0>}
      cornerRadii=<0 0 0 0>
     }
     tintColor {colorUseRule=UseColor_Specified, specifiedColor=<rgba 1 1 1 1>}
     uVRegion {bIsValid=false, max=<0 0>, min=<0 0>}
    }}
   slot {class=WidgetSwitcherSlot}
  }
  Image_6 {
   widgetClassPath=&Image
   parent=WidgetSwitcher_BackgroundFlash
   widget=Image_6
   properties {
    brush {
     bIsDynamicallyLoaded=false
     drawAs=Image
     imageType=NoImage
     mirroring=NoMirror
     resourceName=None
     resourceObject=/Game/Athena/HUD/Materials/Score/M_ScoreLTM_PlacementPlate_3rd
     tiling=NoTile
     imageSize=<91 69>
     margin=<ltrb 0 0 0 0>
     outlineSettings {
      bUseBrushTransparency=false
      roundingType=HalfHeightRadius
      width=0
      color {colorUseRule=UseColor_Specified, specifiedColor=<rgba 0 0 0 0>}
      cornerRadii=<0 0 0 0>
     }
     tintColor {colorUseRule=UseColor_Specified, specifiedColor=<rgba 1 1 1 1>}
     uVRegion {bIsValid=false, max=<0 0>, min=<0 0>}
    }
    renderTransform {angle=0, scale=<1 1>, shear=<0 0>, translation=<0 -4>}
   }
   slot {class=WidgetSwitcherSlot}
  }
  Image_7 {
   widgetClassPath=&Image
   parent=WidgetSwitcher_BackgroundFlash
   widget=Image_7
   properties {
    brush {
     bIsDynamicallyLoaded=false
     drawAs=Image
     imageType=NoImage
     mirroring=NoMirror
     resourceName=None
     resourceObject=/Game/Athena/HUD/Materials/Score/M_ScoreLTM_PlacementPlate_Generic
     tiling=NoTile
     imageSize=<91 69>
     margin=<ltrb 0 0 0 0>
     outlineSettings {
      bUseBrushTransparency=false
      roundingType=HalfHeightRadius
      width=0
      color {colorUseRule=UseColor_Specified, specifiedColor=<rgba 0 0 0 0>}
      cornerRadii=<0 0 0 0>
     }
     tintColor {colorUseRule=UseColor_Specified, specifiedColor=<rgba 1 1 1 1>}
     uVRegion {bIsValid=false, max=<0 0>, min=<0 0>}
    }
    renderTransform {angle=0, scale=<1 1>, shear=<0 0>, translation=<0 -4>}
   }
   slot {class=WidgetSwitcherSlot}
  }
  WidgetSwitcher_PlaceBackground {
   widgetClassPath=&WidgetSwitcher
   bIsVariable=true
   parent=OverlayPlacePlate
   widget=WidgetSwitcher_PlaceBackground
   properties {activeWidgetIndex=3}
   slot {class=OverlaySlot, properties {horizontalAlignment=HAlign_Center, verticalAlignment=VAlign_Center, padding=<ltrb 0 -75 0 -75>}}
  }
  Image_1 {
   widgetClassPath=&Image
   parent=WidgetSwitcher_PlaceBackground
   widget=Image_1
   properties {
    brush {
     bIsDynamicallyLoaded=false
     drawAs=Image
     imageType=NoImage
     mirroring=NoMirror
     resourceName=None
     resourceObject=&M_ScoreLTM_PlacementPlate_1st
     tiling=NoTile
     imageSize=<108 79>
     margin=<ltrb 0 0 0 0>
     outlineSettings {
      bUseBrushTransparency=false
      roundingType=HalfHeightRadius
      width=0
      color {colorUseRule=UseColor_Specified, specifiedColor=<rgba 0 0 0 0>}
      cornerRadii=<0 0 0 0>
     }
     tintColor {colorUseRule=UseColor_Specified, specifiedColor=<rgba 1 1 1 1>}
     uVRegion {bIsValid=false, max=<0 0>, min=<0 0>}
    }
    renderTransform {angle=0, scale=<1 1>, shear=<0 0>, translation=<0 4>}
   }
   slot {class=WidgetSwitcherSlot}
  }
  Image_2 {
   widgetClassPath=&Image
   parent=WidgetSwitcher_PlaceBackground
   widget=Image_2
   properties {brush {
     bIsDynamicallyLoaded=false
     drawAs=Image
     imageType=NoImage
     mirroring=NoMirror
     resourceName=None
     resourceObject=&M_ScoreLTM_PlacementPlate_2nd
     tiling=NoTile
     imageSize=<105 67>
     margin=<ltrb 0 0 0 0>
     outlineSettings {
      bUseBrushTransparency=false
      roundingType=HalfHeightRadius
      width=0
      color {colorUseRule=UseColor_Specified, specifiedColor=<rgba 0 0 0 0>}
      cornerRadii=<0 0 0 0>
     }
     tintColor {colorUseRule=UseColor_Specified, specifiedColor=<rgba 1 1 1 1>}
     uVRegion {bIsValid=false, max=<0 0>, min=<0 0>}
    }}
   slot {class=WidgetSwitcherSlot}
  }
  Image_3 {
   widgetClassPath=&Image
   parent=WidgetSwitcher_PlaceBackground
   widget=Image_3
   properties {
    brush {
     bIsDynamicallyLoaded=false
     drawAs=Image
     imageType=NoImage
     mirroring=NoMirror
     resourceName=None
     resourceObject=&M_ScoreLTM_PlacementPlate_3rd
     tiling=NoTile
     imageSize=<91 69>
     margin=<ltrb 0 0 0 0>
     outlineSettings {
      bUseBrushTransparency=false
      roundingType=HalfHeightRadius
      width=0
      color {colorUseRule=UseColor_Specified, specifiedColor=<rgba 0 0 0 0>}
      cornerRadii=<0 0 0 0>
     }
     tintColor {colorUseRule=UseColor_Specified, specifiedColor=<rgba 1 1 1 1>}
     uVRegion {bIsValid=false, max=<0 0>, min=<0 0>}
    }
    renderTransform {angle=0, scale=<1 1>, shear=<0 0>, translation=<0 -4>}
   }
   slot {class=WidgetSwitcherSlot}
  }
  Image_Generic {
   widgetClassPath=&Image
   parent=WidgetSwitcher_PlaceBackground
   widget=Image_Generic
   properties {
    brush {
     bIsDynamicallyLoaded=false
     drawAs=Image
     imageType=NoImage
     mirroring=NoMirror
     resourceName=None
     resourceObject=&M_ScoreLTM_PlacementPlate_Generic
     tiling=NoTile
     imageSize=<91 69>
     margin=<ltrb 0 0 0 0>
     outlineSettings {
      bUseBrushTransparency=false
      roundingType=HalfHeightRadius
      width=0
      color {colorUseRule=UseColor_Specified, specifiedColor=<rgba 0 0 0 0>}
      cornerRadii=<0 0 0 0>
     }
     tintColor {colorUseRule=UseColor_Specified, specifiedColor=<rgba 1 1 1 1>}
     uVRegion {bIsValid=false, max=<0 0>, min=<0 0>}
    }
    renderTransform {angle=0, scale=<1 1>, shear=<0 0>, translation=<0 -4>}
   }
   slot {class=WidgetSwitcherSlot}
  }
  SizeBox_0 {
   widgetClassPath=&SizeBox
   parent=OverlayPlacePlate
   widget=SizeBox_0
   properties {bOverride_HeightOverride=true, bOverride_WidthOverride=true, heightOverride=57.3923, widthOverride=54.4425}
   slot {class=OverlaySlot, properties {horizontalAlignment=HAlign_Center, verticalAlignment=VAlign_Center, padding=<ltrb 0 -20 0 -20>}}
  }
  ScaleBox_0 {widgetClassPath=&ScaleBox, parent=SizeBox_0, widget=ScaleBox_0, slot {class=SizeBoxSlot, properties {padding=<ltrb 0 -4 0 0>}}}
  Text_PlaceValue {
   widgetClassPath=&CommonTextBlock
   bIsVariable=true
   parent=ScaleBox_0
   widget=Text_PlaceValue
   properties {
    style=&TextStyle-ScorePlacementGeneric_C
    text=3rd
    colorAndOpacity {colorUseRule=UseColor_Specified, specifiedColor=<rgba 0.0052 0.0052 0.0056 1>}
    font {
     bForceMonospaced=false
     bMaterialIsStencil=false
     fontMaterial=null
     fontObject=&Burbank
     letterSpacing=0
     monospacedWidth=1
     size=39
     skewAmount=0
     typefaceFontName=Burbank-Black
     outlineSettings {
      bApplyOutlineToDropShadows=false
      bMiteredCorners=false
      bSeparateFillAlpha=true
      outlineMaterial=null
      outlineSize=2
      outlineColor=<rgba 0.5008 0.5008 0.5417 0.651>
     }
    }
    shadowOffset=<0 0>
   }
   slot {class=ScaleBoxSlot}
  }
  SizeBoxScore {
   widgetClassPath=&SizeBox
   parent=Overlay_0
   widget=SizeBoxScore
   properties {bOverride_WidthOverride=true, widthOverride=308}
   slot {class=OverlaySlot, properties {horizontalAlignment=HAlign_Fill, verticalAlignment=VAlign_Center}}
  }
  ScoreOverlay {widgetClassPath=&Overlay, parent=SizeBoxScore, widget=ScoreOverlay, slot {class=SizeBoxSlot}}
  ScoreTextBox {
   widgetClassPath=&CommonTextBlock
   bIsVariable=true
   parent=ScoreOverlay
   widget=ScoreTextBox
   properties {
    text="1,400 pts"
    font {
     bForceMonospaced=false
     bMaterialIsStencil=false
     fontMaterial=null
     fontObject=&NotoSans
     letterSpacing=0
     monospacedWidth=1
     size=20
     skewAmount=0
     typefaceFontName=NotoSans-Bold
     outlineSettings {
      bApplyOutlineToDropShadows=false
      bMiteredCorners=false
      bSeparateFillAlpha=false
      outlineMaterial=null
      outlineSize=1
      outlineColor=<rgba 0 0 0 0.8>
     }
    }
    renderTransformPivot=<0.2 0.8>
   }
   slot {class=OverlaySlot, properties {horizontalAlignment=HAlign_Center, verticalAlignment=VAlign_Center, padding=<ltrb 0 -10 0 -10>}}
  }
  Overlay_Leaderboard {
   widgetClassPath=&Overlay
   bIsVariable=true
   parent=VerticalBoxContent
   widget=Overlay_Leaderboard
   properties {visibility=Collapsed}
   slot {class=VerticalBoxSlot, properties {verticalAlignment=VAlign_Top, padding=<ltrb 40 -10 10 0>}}
  }
  ImageLeaderboardBG {
   widgetClassPath=&Image
   parent=Overlay_Leaderboard
   widget=ImageLeaderboardBG
   properties {
    visibility=HitTestInvisible
    brush {
     bIsDynamicallyLoaded=false
     drawAs=Box
     imageType=NoImage
     mirroring=NoMirror
     resourceName=None
     resourceObject=/FortUILibrary/Foundation/Buttons/Materials/Skewed/M-Skew-Base_Desat
     tiling=NoTile
     imageSize=<256 64>
     margin=<ltrb 0.3 0.3 0.3 0.3>
     outlineSettings {
      bUseBrushTransparency=false
      roundingType=HalfHeightRadius
      width=0
      color {colorUseRule=UseColor_Specified, specifiedColor=<rgba 0 0 0 0>}
      cornerRadii=<0 0 0 0>
     }
     tintColor {colorUseRule=UseColor_Specified, specifiedColor=<rgba 1 1 1 1>}
     uVRegion {bIsValid=false, max=<0 0>, min=<0 0>}
    }
   }
   slot {class=OverlaySlot, properties {horizontalAlignment=HAlign_Fill, verticalAlignment=VAlign_Fill}}
  }
  CommonWidgetSwitcher_GameStatus {
   widgetClassPath=&CommonWidgetSwitcherLegacy
   bIsVariable=true
   parent=Overlay_Leaderboard
   widget=CommonWidgetSwitcher_GameStatus
   slot {class=OverlaySlot, properties {horizontalAlignment=HAlign_Fill, verticalAlignment=VAlign_Fill, padding=<ltrb 0 -5 0 0>}}
  }
  HorizontalBox_0 {
   widgetClassPath=&HorizontalBox
   parent=CommonWidgetSwitcher_GameStatus
   widget=HorizontalBox_0
   properties {visibility=Collapsed}
   slot {class=WidgetSwitcherSlot, properties {padding=<ltrb 35 25 25 0>}}
  }
  TextBlock_places {
   widgetClassPath=&TextBlock
   bIsVariable=true
   parent=HorizontalBox_0
   widget=TextBlock_places
   properties {
    justification=Right
    text="1st\r\n2nd\r\n3rd"
    font {
     bForceMonospaced=false
     bMaterialIsStencil=false
     fontMaterial=null
     fontObject=&BurbankSmall
     letterSpacing=0
     monospacedWidth=1
     size=20
     skewAmount=0
     typefaceFontName=Bold
     outlineSettings {
      bApplyOutlineToDropShadows=false
      bMiteredCorners=false
      bSeparateFillAlpha=false
      outlineMaterial=null
      outlineSize=1
      outlineColor=<rgba 0.0052 0.0052 0.0056 0.4>
     }
    }
    shadowColorAndOpacity=<rgba 0.0052 0.0052 0.0056 1>
    shadowOffset=<3 3>
   }
   slot {class=HorizontalBoxSlot, properties {horizontalAlignment=HAlign_Left, padding=<ltrb 0 0 0 35>}}
  }
  TextBlock_PlaceValues {
   widgetClassPath=&TextBlock
   bIsVariable=true
   parent=HorizontalBox_0
   widget=TextBlock_PlaceValues
   properties {
    text="1,000\r\n900\r\n800"
    font {
     bForceMonospaced=false
     bMaterialIsStencil=false
     fontMaterial=null
     fontObject=&BurbankSmall
     letterSpacing=0
     monospacedWidth=1
     size=20
     skewAmount=0
     typefaceFontName=Bold
     outlineSettings {
      bApplyOutlineToDropShadows=false
      bMiteredCorners=false
      bSeparateFillAlpha=false
      outlineMaterial=null
      outlineSize=1
      outlineColor=<rgba 0.0052 0.0052 0.0056 0.4>
     }
    }
    shadowColorAndOpacity=<rgba 0.0052 0.0052 0.0056 1>
    shadowOffset=<3 3>
   }
   slot {class=HorizontalBoxSlot, properties {horizontalAlignment=HAlign_Left, padding=<ltrb 20 0 0 0>}}
  }
  TextBlock_LeaderDelta {
   widgetClassPath=&CommonTextBlock
   bIsVariable=true
   parent=CommonWidgetSwitcher_GameStatus
   widget=TextBlock_LeaderDelta
   properties {
    text="Leading by 1,000!"
    textTransformPolicy=ToUpper
    colorAndOpacity {colorUseRule=UseColor_Specified, specifiedColor=<rgba 0.9702 1 0.0913 1>}
    font {
     bForceMonospaced=false
     bMaterialIsStencil=false
     fontMaterial=null
     fontObject=&BurbankSmall
     letterSpacing=0
     monospacedWidth=1
     size=20
     skewAmount=0
     typefaceFontName=Black
     outlineSettings {
      bApplyOutlineToDropShadows=false
      bMiteredCorners=false
      bSeparateFillAlpha=false
      outlineMaterial=null
      outlineSize=1
      outlineColor=<rgba 0.0625 0.0049 0 1>
     }
    }
    renderTransform {angle=0, scale=<1 1>, shear=<-12 0>, translation=<0 0>}
    shadowColorAndOpacity=<rgba 0 0 0 1>
    shadowOffset=<3 3>
   }
   slot {class=WidgetSwitcherSlot, properties {horizontalAlignment=HAlign_Center, verticalAlignment=VAlign_Center}}
  }
  TextBlock_TrailingDelta {
   widgetClassPath=&CommonTextBlock
   bIsVariable=true
   parent=CommonWidgetSwitcher_GameStatus
   widget=TextBlock_TrailingDelta
   properties {
    text="Trailing by 1,000"
    textTransformPolicy=ToUpper
    font {
     bForceMonospaced=false
     bMaterialIsStencil=false
     fontMaterial=null
     fontObject=&BurbankSmall
     letterSpacing=0
     monospacedWidth=1
     size=20
     skewAmount=0
     typefaceFontName=Bold
     outlineSettings {
      bApplyOutlineToDropShadows=false
      bMiteredCorners=false
      bSeparateFillAlpha=false
      outlineMaterial=null
      outlineSize=1
      outlineColor=<rgba 0.0052 0.0052 0.0056 1>
     }
    }
    renderTransform {angle=0, scale=<1 1>, shear=<-12 0>, translation=<0 0>}
    shadowColorAndOpacity=<rgba 0 0 0 1>
    shadowOffset=<3 3>
   }
   slot {class=WidgetSwitcherSlot, properties {horizontalAlignment=HAlign_Center, verticalAlignment=VAlign_Center}}
  }
  TextBlock_TiedForLead {
   widgetClassPath=&CommonTextBlock
   bIsVariable=true
   parent=CommonWidgetSwitcher_GameStatus
   widget=TextBlock_TiedForLead
   properties {
    text=TIED FOR 1st
    colorAndOpacity {colorUseRule=UseColor_Specified, specifiedColor=<rgba 0.9702 1 0.0913 1>}
    font {
     bForceMonospaced=false
     bMaterialIsStencil=false
     fontMaterial=null
     fontObject=&BurbankSmall
     letterSpacing=0
     monospacedWidth=1
     size=20
     skewAmount=0
     typefaceFontName=Black
     outlineSettings {
      bApplyOutlineToDropShadows=false
      bMiteredCorners=false
      bSeparateFillAlpha=false
      outlineMaterial=null
      outlineSize=1
      outlineColor=<rgba 0.0625 0.0049 0 1>
     }
    }
    renderTransform {angle=0, scale=<1 1>, shear=<-12 0>, translation=<0 0>}
    shadowColorAndOpacity=<rgba 0 0 0 1>
    shadowOffset=<3 3>
   }
   slot {class=WidgetSwitcherSlot, properties {horizontalAlignment=HAlign_Center, verticalAlignment=VAlign_Center}}
  }
  isWinner {
   widgetClassPath=&CommonTextBlock
   bIsVariable=true
   parent=CommonWidgetSwitcher_GameStatus
   widget=isWinner
   properties {
    text="WINNER!"
    colorAndOpacity {colorUseRule=UseColor_Specified, specifiedColor=<rgba 0.9467 1 0.0683 1>}
    font {
     bForceMonospaced=false
     bMaterialIsStencil=false
     fontMaterial=/Game/Athena/HUD/Materials/Victory/M_1stplace_Gold
     fontObject=&BurbankSmall
     letterSpacing=0
     monospacedWidth=1
     size=26
     skewAmount=0
     typefaceFontName=Black
     outlineSettings {
      bApplyOutlineToDropShadows=false
      bMiteredCorners=false
      bSeparateFillAlpha=false
      outlineMaterial=null
      outlineSize=1
      outlineColor=<rgba 0.0625 0.0049 0 1>
     }
    }
    renderTransform {angle=0, scale=<1 1>, shear=<-12 0>, translation=<0 0>}
    shadowColorAndOpacity=<rgba 0 0 0 1>
    shadowOffset=<3 3>
   }
   slot {class=WidgetSwitcherSlot, properties {horizontalAlignment=HAlign_Center, verticalAlignment=VAlign_Center, padding=<ltrb 0 -4 0 0>}}
  }
  TutorialHighlight {widgetClassPath=&TutorialHighlight}
 ]
}
