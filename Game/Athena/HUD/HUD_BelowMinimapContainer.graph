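# HUD widget that, per its name, serves as the below-minimap container for the Athena storm surge and team score widgets.
# A fixed-width SizeBox wraps a VerticalBox that stacks the two children, both collapsed by default.
# Each child carries its own HUD element gameplay tags naming contexts to hide in, such as loading, capture, post-game and special events.
# The event graph holds only empty stubs, so nothing here controls their visibility at runtime.
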WidgetBlueprint /Game/Athena/HUD/HUD_BelowMinimapContainer
graphs [EventGraph {nodes [
   Event PreConstruct()
   Event Construct()
   Event Tick()
  ]}]
variables []
widgets {
 info {inheritedWidgetCount=0, parentClass=&UserWidget, rootWidgetClass=&SizeBox, widgetCount=4}
 widgets [
  SizeBox_0 {
   widgetClassPath=&SizeBox
   widget=SizeBox_0
   properties {bOverride_MinDesiredHeight=true, bOverride_WidthOverride=true, minDesiredHeight=1, widthOverride=300}
  }
  VerticalBox_29 {widgetClassPath=&VerticalBox, parent=SizeBox_0, widget=VerticalBox_29, slot {class=SizeBoxSlot}}
  AthenaStormSurgeWidget {
   widgetClassPath=&/AthenaStormSurgeWidget
   parent=VerticalBox_29
   widget=AthenaStormSurgeWidget
   properties {
    visibility=Collapsed
    hUDElementTag {gameplayTags [
      {tagName=HUD.Athena.HideOnBus}
      {tagName=Athena.HideForCapture}
      {tagName=HUD.Athena.HideForGadget}
      {tagName=HUD.Athena.Hide.WithInventory}
      {tagName=HUD.Athena.Hide.WithRCPawn}
      {tagName=HUD.Athena.HideForAbility}
      {tagName=HUD.Event.HideForSpecialEvent}
      {tagName=Athena.Creative.HideForLoading}
      {tagName=HUD.Hide.WhileLoading}
      {tagName=HUD.Athena.HideForPostGame}
      {tagName=HUD.Athena.HideforRadial}
      {tagName=HUD.Hide.CreativeJoiningLobby}
     ]}
   }
   slot {class=VerticalBoxSlot}
  }
  TeamScoreWidget {
   widgetClassPath=&/AthenaTeamScoreWidget
   parent=VerticalBox_29
   widget=TeamScoreWidget
   properties {
    visibility=Collapsed
    hUDElementTag {gameplayTags [
      {tagName=HUD.Athena.Hide.WithMap}
      {tagName=HUD.LTM.HideForTryHard}
      {tagName=HUD.Event.HideForSpecialEvent}
      {tagName=HUD.Hide.WhileLoading}
      {tagName=Athena.HideForCapture}
      {tagName=HUD.Athena.HideForPostGameSpectate}
      {tagName=HUD.Athena.HideForPostGame}
      {tagName=HUD.Athena.HideforRadial}
      {tagName=HUD.Hide.CreativeJoiningLobby}
     ]}
    renderTransformPivot=<1 0.5>
   }
   slot {class=VerticalBoxSlot}
  }
 ]
}
