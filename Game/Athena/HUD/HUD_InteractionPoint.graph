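# HUD interaction point widget that marks a spot with either a distant preview or a detail view.
# The distant preview shows a pulsing ping, animated by Intro/Outro, or a supplied icon with optional arrow.
# The detail view shows an icon, two text lines split at a line break, or both, above an arrow hidden when clamped.
# Custom events drive the switchers, brushes and visibility, and Construct applies the DetailText variable.
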WidgetBlueprint /Game/Athena/HUD/HUD_InteractionPoint
graphs [EventGraph {nodes [
   n1 = EventShowDetail!() -> (OutputDelegate, then, ShowDetail:bool=~) {is=title K2Node_CustomEvent}
   > SetActiveWidgetIndex(
    self=Get Switcher_Detail() -> (Switcher_Detail:Common Widget Switcher Legacy Object Reference)
    Index=Select(Option 0:int=0, Option 1:int=1, Index=@n1.ShowDetail) -> (ReturnValue:int)
   )
   EventOutro() {is=title K2Node_CustomEvent}
   > PlayAnimationByName!(
    execute
    self:Fort Simple Widget Animations Panel Object Reference=Get AnimationsPanel_IntroOutro() -> (AnimationsPanel_IntroOutro:Fort Simple Widget Animations Panel Object Reference)
    Name:"Name (by ref)"=MakeLiteralName(Value=Outro)
    Speed:float=1
    StartAtTime:float=0
   ) -> (then, ReturnValue:bool)
   EventIntro() {is=title K2Node_CustomEvent}
   > PlayAnimationByName(self=Get AnimationsPanel_IntroOutro(), Name=MakeLiteralName(Value=Intro), Speed=1, StartAtTime=0)
   n2 = EventSetText!() -> (OutputDelegate, then, DetailText:Text) {is=title K2Node_CustomEvent}
   > "SetText (Text)"(self=[
     Get Text_DetailLine1() -> (Text_DetailLine1:Common Text Object Reference)
     Get Text_DetailLine2() -> (Text_DetailLine2:Common Text Object Reference)
    ])
   n3 = Split(SourceString="To String (Text)"(InText=@n2.DetailText), InStr="\r\n", SearchDir=FromStart)
   n4 = "To Text (String)"(InString=@n3)
   > "SetText (Text)"(self=Get Text_DetailLine1(), InText=@n4)
   > "SetText (Text)"(self=Get Text_DetailLine2(), InText="To Text (String)"(InString=@n3.RightS))
   n5 = SetDetailIcon!() -> (OutputDelegate, then, Icon:Slate Brush Structure, ShowWithText:bool=~, HideArrow:bool=~) {is=title K2Node_CustomEvent}
   n6 = Get Image_DetailIcon_2() -> (Image_DetailIcon_2:Image Object Reference)
   > SetBrush(self=[@n6, Get Image_DetailIcon() -> (Image_DetailIcon:Image Object Reference)], InBrush=@n5.Icon)
   > SetActiveWidgetIndex(
    self=Get Switcher_TextorIcon() -> (Switcher_TextorIcon:Common Widget Switcher Legacy Object Reference)
    Index=Select(Option 0=1, Option 1=0, Index=@n5.ShowWithText)
   )
   > SetVisibility(
    self=@n6
    InVisibility=Select(Option 0:ESlateVisibility Enum=Collapsed, Option 1:ESlateVisibility Enum=HitTestInvisible, Index=@n5.ShowWithText) -> (ReturnValue:ESlateVisibility Enum)
   )
   > SetVisibility(
    self=Get Image_DetailArrow() -> (Image_DetailArrow:Image Object Reference)
    InVisibility=Select(Option 0=HitTestInvisible, Option 1=Hidden, Index=@n5.HideArrow)
   )
   n7 = SetDistantIcon!() -> (OutputDelegate, then, Icon:Slate Brush Structure, Default:bool=~, HideArrow:bool=~) {is=title K2Node_CustomEvent}
   > SetActiveWidgetIndex(
    self=Get Switcher_DistantPreview() -> (Switcher_DistantPreview:Common Widget Switcher Legacy Object Reference)
    Index=Select(Option 0:int=1, Option 1:int=0, Index=@n7.Default) -> (ReturnValue:int)
   )
   > SetBrush(self=Get Image_DistantIcon() -> (Image_DistantIcon:Image Object Reference), InBrush=@n7.Icon)
   > SetVisibility(
    self=Get Image_Arrow() -> (Image_Arrow:Image Object Reference)
    InVisibility=Select(Option 0:ESlateVisibility Enum=HitTestInvisible, Option 1:ESlateVisibility Enum=Collapsed, Index=@n7.HideArrow) -> (ReturnValue:ESlateVisibility Enum)
   )
   n8 = EventClamp!() -> (OutputDelegate, then, isClamped:bool=~, Position:Vector 2D Structure) {is=title K2Node_CustomEvent}
   > SetVisibility(
    self=Get Border_ArrowClampContainer() -> (Border_ArrowClampContainer:Common Border Object Reference)
    InVisibility=Select(Option 0=HitTestInvisible, Option 1=Hidden, Index=@n8.isClamped)
   )
   Event Construct()
   > EventSetText!(execute, self, DetailText:Text=Get DetailText() -> (DetailText:Text)) -> (then) {is=title K2Node_CallFunction}
   PrintString!(
    execute
    InString="To String (Vector2d)"!(InVec:Vector 2D Structure=@n8.Position) -> (ReturnValue:String)
    bPrintToScreen
    bPrintToLog
    TextColor
    Duration
    Key
   ) -> (then)
   SetActiveWidgetIndex(self=Get Switcher_TextorIcon(), Index=0)
   Select(Option 0:Text=@n2.DetailText, Option 1:Text=@n4, Index=@n3.ReturnValue) -> (ReturnValue:Text)
  ]}]
variables [DetailText {type=string, default=""}]
widgets {
 info {inheritedWidgetCount=0, namedSlotCount=1, parentClass=&UserWidget, rootWidgetClass=&ScaleBox, widgetCount=21}
 widgets [
  ScaleBox_Container {
   widgetClassPath=&ScaleBox
   bIsVariable=true
   widget=ScaleBox_Container
   properties {stretch=UserSpecified, userSpecifiedScale=0.75}
  }
  FortHUDElementWrapper {
   widgetClassPath=&/FortHUDElementWrapper
   bIsVariable=true
   parent=ScaleBox_Container
   widget=FortHUDElementWrapper
   properties {elementTag {tagName=HUD.InteractionPoint}}
   slot {class=ScaleBoxSlot, properties {verticalAlignment=VAlign_Bottom}}
  }
  Switcher_Detail {
   widgetClassPath=&CommonWidgetSwitcherLegacy
   bIsVariable=true
   namedSlotHost=FortHUDElementWrapper
   widget=Switcher_Detail
   properties {visibility=HitTestInvisible}
  }
  Switcher_DistantPreview {
   widgetClassPath=&CommonWidgetSwitcherLegacy
   bIsVariable=true
   parent=Switcher_Detail
   widget=Switcher_DistantPreview
   slot {class=WidgetSwitcherSlot, properties {horizontalAlignment=HAlign_Center, verticalAlignment=VAlign_Center}}
  }
  ScaleBox_InteractionIndication {
   widgetClassPath=&ScaleBox
   bIsVariable=true
   parent=Switcher_DistantPreview
   widget=ScaleBox_InteractionIndication
   properties {stretch=UserSpecified, userSpecifiedScale=0.75}
   slot {class=WidgetSwitcherSlot}
  }
  AnimationsPanel_IntroOutro {
   widgetClassPath=/Script/FortniteUI.FortSimpleWidgetAnimationsPanel
   bIsVariable=true
   parent=ScaleBox_InteractionIndication
   widget=AnimationsPanel_IntroOutro
   properties {animations [
     Intro {
      bIsEnabled=true
      onAnimationsFinished=()
      targets [{
        alpha=&/HUD_InteractionPoint_IntroAlpha
        bResetOnFinish=false
        bShouldAnimateAlpha=true
        bShouldAnimateScale=true
        bShouldAnimateTranslation=false
        scale=&/HUD_InteractionPoint_IntroScale
        targetName=Overlay_Ping
        translation=null
        materialScalarAnimations []
       }]
     }
     Outro {
      bIsEnabled=true
      onAnimationsFinished=()
      targets [{
        alpha=&/HUD_InteractionPoint_OutroAlpha
        bResetOnFinish=false
        bShouldAnimateAlpha=true
        bShouldAnimateScale=true
        bShouldAnimateTranslation=false
        scale=&/HUD_InteractionPoint_OutroScale
        targetName=Overlay_Ping
        translation=null
        materialScalarAnimations []
       }]
     }
    ]}
   slot {class=ScaleBoxSlot, properties {horizontalAlignment=HAlign_Fill, verticalAlignment=VAlign_Fill}}
  }
  Overlay_Ping {widgetClassPath=&Overlay, parent=AnimationsPanel_IntroOutro, widget=Overlay_Ping, slot {class=FortSimpleWidgetAnimationsSlot}}
  Image_Ping {
   widgetClassPath=&Image
   bIsVariable=true
   parent=Overlay_Ping
   widget=Image_Ping
   properties {brush {
     bIsDynamicallyLoaded=false
     drawAs=Image
     imageType=NoImage
     mirroring=NoMirror
     resourceName=None
     resourceObject=/Game/UI/Login/SubGameSelect/Materials/M_UI_FocusBorderPulse_Circle_Inst
     tiling=NoTile
     imageSize=<32 32>
     margin=<ltrb 0 0 0 0>
     outlineSettings {
      bUseBrushTransparency=false
      roundingType=HalfHeightRadius
      width=0
      color {colorUseRule=UseColor_Specified, specifiedColor=<rgba 0 0 0 0>}
      cornerRadii=<0 0 0 0>
     }
     tintColor {colorUseRule=UseColor_Specified, specifiedColor=<rgba 1 1 1 1>}
     uVRegion {bIsValid=false, max=<0 0>, min=<0 0>}
    }}
   slot {class=OverlaySlot, properties {horizontalAlignment=HAlign_Fill, verticalAlignment=VAlign_Fill}}
  }
  Image {
   widgetClassPath=&Image
   bIsVariable=true
   parent=Overlay_Ping
   widget=Image
   properties {brush {
     bIsDynamicallyLoaded=false
     drawAs=Image
     imageType=NoImage
     mirroring=NoMirror
     resourceName=None
     resourceObject=/Game/Athena/HUD/Materials/M_UI_FocusBorderPulse_Circle_Solid
     tiling=NoTile
     imageSize=<128 128>
     margin=<ltrb 0 0 0 0>
     outlineSettings {
      bUseBrushTransparency=false
      roundingType=HalfHeightRadius
      width=0
      color {colorUseRule=UseColor_Specified, specifiedColor=<rgba 0 0 0 0>}
      cornerRadii=<0 0 0 0>
     }
     tintColor {colorUseRule=UseColor_Specified, specifiedColor=<rgba 1 1 1 1>}
     uVRegion {bIsValid=false, max=<0 0>, min=<0 0>}
    }}
   slot {class=OverlaySlot, properties {horizontalAlignment=HAlign_Center, verticalAlignment=VAlign_Center}}
  }
  Overlay_31 {widgetClassPath=&Overlay, parent=Switcher_DistantPreview, widget=Overlay_31, slot {class=WidgetSwitcherSlot}}
  Image_Arrow {
   widgetClassPath=&Image
   bIsVariable=true
   parent=Overlay_31
   widget=Image_Arrow
   properties {brush {
     bIsDynamicallyLoaded=false
     drawAs=Image
     imageType=NoImage
     mirroring=NoMirror
     resourceName=None
     resourceObject=/Game/Athena/HUD/Materials/M_UI_FocusBorderPulse_ArrowOnly
     tiling=NoTile
     imageSize=<128 128>
     margin=<ltrb 0 0 0 0>
     outlineSettings {
      bUseBrushTransparency=false
      roundingType=HalfHeightRadius
      width=0
      color {colorUseRule=UseColor_Specified, specifiedColor=<rgba 0 0 0 0>}
      cornerRadii=<0 0 0 0>
     }
     tintColor {colorUseRule=UseColor_Specified, specifiedColor=<rgba 1 1 1 1>}
     uVRegion {bIsValid=false, max=<0 0>, min=<0 0>}
    }}
   slot {class=OverlaySlot, properties {horizontalAlignment=HAlign_Center, verticalAlignment=VAlign_Center}}
  }
  Image_DistantIcon {
   widgetClassPath=&Image
   bIsVariable=true
   parent=Overlay_31
   widget=Image_DistantIcon
   properties {brush {
     bIsDynamicallyLoaded=false
     drawAs=Image
     imageType=NoImage
     mirroring=NoMirror
     resourceName=None
     resourceObject=/Game/Athena/HUD/Materials/M_UI_FocusBorderPulse_MarkerOnly
     tiling=NoTile
     imageSize=<64 64>
     margin=<ltrb 0 0 0 0>
     outlineSettings {
      bUseBrushTransparency=false
      roundingType=HalfHeightRadius
      width=0
      color {colorUseRule=UseColor_Specified, specifiedColor=<rgba 0 0 0 0>}
      cornerRadii=<0 0 0 0>
     }
     tintColor {colorUseRule=UseColor_Specified, specifiedColor=<rgba 1 1 1 1>}
     uVRegion {bIsValid=false, max=<0 0>, min=<0 0>}
    }}
   slot {class=OverlaySlot, properties {horizontalAlignment=HAlign_Center, verticalAlignment=VAlign_Center}}
  }
  VerticalBox_18 {
   widgetClassPath=&VerticalBox
   parent=Switcher_Detail
   widget=VerticalBox_18
   slot {class=WidgetSwitcherSlot, properties {horizontalAlignment=HAlign_Center, verticalAlignment=VAlign_Center}}
  }
  Switcher_TextorIcon {
   widgetClassPath=&CommonWidgetSwitcherLegacy
   bIsVariable=true
   parent=VerticalBox_18
   widget=Switcher_TextorIcon
   slot {class=VerticalBoxSlot, properties {horizontalAlignment=HAlign_Center, verticalAlignment=VAlign_Center}}
  }
  VerticalBox_0 {widgetClassPath=&VerticalBox, parent=Switcher_TextorIcon, widget=VerticalBox_0, slot {class=WidgetSwitcherSlot}}
  Text_DetailLine1 {
   widgetClassPath=&CommonTextBlock
   bIsVariable=true
   parent=VerticalBox_0
   widget=Text_DetailLine1
   properties {
    justification=Center
    text=Lorem
    textTransformPolicy=ToUpper
    font {
     bForceMonospaced=false
     bMaterialIsStencil=false
     fontMaterial=null
     fontObject=&Burbank
     letterSpacing=0
     monospacedWidth=1
     size=24
     skewAmount=0
     typefaceFontName=Burbank-PlayerPower
     outlineSettings {
      bApplyOutlineToDropShadows=false
      bMiteredCorners=false
      bSeparateFillAlpha=false
      outlineMaterial=null
      outlineSize=1
      outlineColor=<rgba 0 0 0 1>
     }
    }
    shadowColorAndOpacity=<rgba 0.0052 0.0052 0.0056 1>
    shadowOffset=<4 4>
   }
   slot {class=VerticalBoxSlot, properties {padding=<ltrb 0 0 0 -8>}}
  }
  Text_DetailLine2 {
   widgetClassPath=&CommonTextBlock
   bIsVariable=true
   parent=VerticalBox_0
   widget=Text_DetailLine2
   properties {
    justification=Center
    text=Ipsum
    textTransformPolicy=ToUpper
    font {
     bForceMonospaced=false
     bMaterialIsStencil=false
     fontMaterial=null
     fontObject=&BurbankSmall
     letterSpacing=0
     monospacedWidth=1
     size=24
     skewAmount=0
     typefaceFontName=Black
     outlineSettings {
      bApplyOutlineToDropShadows=false
      bMiteredCorners=false
      bSeparateFillAlpha=false
      outlineMaterial=null
      outlineSize=1
      outlineColor=<rgba 0 0 0 1>
     }
    }
    renderTransform {angle=0, scale=<1 1>, shear=<-12 0>, translation=<0 0>}
    shadowColorAndOpacity=<rgba 0.0052 0.0052 0.0056 1>
    shadowOffset=<4 4>
   }
   slot {class=VerticalBoxSlot, properties {horizontalAlignment=HAlign_Center, verticalAlignment=VAlign_Center, padding=<ltrb 0 -3 0 0>}}
  }
  Image_DetailIcon_2 {
   widgetClassPath=&Image
   bIsVariable=true
   parent=VerticalBox_0
   widget=Image_DetailIcon_2
   properties {
    visibility=Collapsed
    brush {
     bIsDynamicallyLoaded=false
     drawAs=Image
     imageType=NoImage
     mirroring=NoMirror
     resourceName=None
     resourceObject=/Game/UI/Foundation/Textures/Buttons/T-SlotIcon-Recycle-64
     tiling=NoTile
     imageSize=<64 64>
     margin=<ltrb 0 0 0 0>
     outlineSettings {
      bUseBrushTransparency=false
      roundingType=HalfHeightRadius
      width=0
      color {colorUseRule=UseColor_Specified, specifiedColor=<rgba 0 0 0 0>}
      cornerRadii=<0 0 0 0>
     }
     tintColor {colorUseRule=UseColor_Specified, specifiedColor=<rgba 1 1 1 1>}
     uVRegion {bIsValid=false, max=<0 0>, min=<0 0>}
    }
   }
   slot {class=VerticalBoxSlot, properties {horizontalAlignment=HAlign_Center, verticalAlignment=VAlign_Center}}
  }
  Image_DetailIcon {
   widgetClassPath=&Image
   bIsVariable=true
   parent=Switcher_TextorIcon
   widget=Image_DetailIcon
   properties {brush {
     bIsDynamicallyLoaded=false
     drawAs=Image
     imageType=NoImage
     mirroring=NoMirror
     resourceName=None
     resourceObject=&T-SlotIcon-Recycle-64
     tiling=NoTile
     imageSize=<64 64>
     margin=<ltrb 0 0 0 0>
     outlineSettings {
      bUseBrushTransparency=false
      roundingType=HalfHeightRadius
      width=0
      color {colorUseRule=UseColor_Specified, specifiedColor=<rgba 0 0 0 0>}
      cornerRadii=<0 0 0 0>
     }
     tintColor {colorUseRule=UseColor_Specified, specifiedColor=<rgba 1 1 1 1>}
     uVRegion {bIsValid=false, max=<0 0>, min=<0 0>}
    }}
   slot {class=WidgetSwitcherSlot}
  }
  Border_ArrowClampContainer {
   widgetClassPath=&CommonBorder
   bIsVariable=true
   parent=VerticalBox_18
   widget=Border_ArrowClampContainer
   properties {padding=<ltrb 0 0 0 0>}
   slot {class=VerticalBoxSlot, properties {horizontalAlignment=HAlign_Center, verticalAlignment=VAlign_Bottom, padding=<ltrb 0 -80 0 0>}}
  }
  Image_DetailArrow {
   widgetClassPath=&Image
   bIsVariable=true
   parent=Border_ArrowClampContainer
   widget=Image_DetailArrow
   properties {brush {
     bIsDynamicallyLoaded=false
     drawAs=Image
     imageType=NoImage
     mirroring=NoMirror
     resourceName=None
     resourceObject=/Game/Athena/HUD/LTM/Mash/Materials/M_UI_InteractionArrow
     tiling=NoTile
     imageSize=<124.6503 112.1483>
     margin=<ltrb 0 0 0 0>
     outlineSettings {
      bUseBrushTransparency=false
      roundingType=HalfHeightRadius
      width=0
      color {colorUseRule=UseColor_Specified, specifiedColor=<rgba 0 0 0 0>}
      cornerRadii=<0 0 0 0>
     }
     tintColor {colorUseRule=UseColor_Specified, specifiedColor=<rgba 1 1 1 1>}
     uVRegion {bIsValid=false, max=<0 0>, min=<0 0>}
    }}
   slot {class=BorderSlot, properties {padding=<ltrb 0 0 0 0>}}
  }
 ]
}
