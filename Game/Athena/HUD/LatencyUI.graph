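# LatencyUI is a HUD widget that lays out a latency readout, derived from the native LatencyWidget class.
# A horizontal box wraps a vertical stack of small bold CommonTextBlocks for total, game, render, driver, OS work, GPU and tweak latency.
# The text blocks are inherited widgets holding placeholder text, and the event graph contains only unconnected PreConstruct, Construct and Tick events, so the Blueprint supplies layout and styling alone.
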WidgetBlueprint /Game/Athena/HUD/LatencyUI
graphs [EventGraph {nodes [
   Event PreConstruct()
   Event Construct()
   Event Tick()
  ]}]
variables []
widgets {
 info {inheritedWidgetCount=7, parentClass=/Script/LatencyUI.LatencyWidget, rootWidgetClass=&HorizontalBox, widgetCount=9}
 widgets [
  HorizontalBox_0 {widgetClassPath=&HorizontalBox, widget=HorizontalBox_0}
  VerticalBox_29 {widgetClassPath=&VerticalBox, parent=HorizontalBox_0, widget=VerticalBox_29, slot {class=HorizontalBoxSlot}}
  TotalLatencyUI {
   widgetClassPath=&CommonTextBlock
   bInherited=true
   parent=VerticalBox_29
   widget=TotalLatencyUI
   properties {
    text="{Total}"
    font {
     bForceMonospaced=false
     bMaterialIsStencil=false
     fontMaterial=null
     fontObject=&NotoSans
     letterSpacing=0
     monospacedWidth=1
     size=8
     skewAmount=0
     typefaceFontName=NotoSans-Bold
     outlineSettings {
      bApplyOutlineToDropShadows=false
      bMiteredCorners=false
      bSeparateFillAlpha=false
      outlineMaterial=null
      outlineSize=0
      outlineColor=<rgba 0 0 0 1>
     }
    }
    shadowOffset=<0 0>
   }
   slot {class=VerticalBoxSlot}
  }
  GameLatencyUI {
   widgetClassPath=&CommonTextBlock
   bInherited=true
   parent=VerticalBox_29
   widget=GameLatencyUI
   properties {
    text="{Game}"
    font {
     bForceMonospaced=false
     bMaterialIsStencil=false
     fontMaterial=null
     fontObject=&NotoSans
     letterSpacing=0
     monospacedWidth=1
     size=8
     skewAmount=0
     typefaceFontName=NotoSans-Bold
     outlineSettings {
      bApplyOutlineToDropShadows=false
      bMiteredCorners=false
      bSeparateFillAlpha=false
      outlineMaterial=null
      outlineSize=0
      outlineColor=<rgba 0 0 0 1>
     }
    }
    shadowOffset=<0 0>
   }
   slot {class=VerticalBoxSlot}
  }
  RenderLatencyUI {
   widgetClassPath=&CommonTextBlock
   bInherited=true
   parent=VerticalBox_29
   widget=RenderLatencyUI
   properties {
    text="{Render}"
    font {
     bForceMonospaced=false
     bMaterialIsStencil=false
     fontMaterial=null
     fontObject=&NotoSans
     letterSpacing=0
     monospacedWidth=1
     size=8
     skewAmount=0
     typefaceFontName=NotoSans-Bold
     outlineSettings {
      bApplyOutlineToDropShadows=false
      bMiteredCorners=false
      bSeparateFillAlpha=false
      outlineMaterial=null
      outlineSize=0
      outlineColor=<rgba 0 0 0 1>
     }
    }
    shadowOffset=<0 0>
   }
   slot {class=VerticalBoxSlot}
  }
  DriverLatencyUI {
   widgetClassPath=&CommonTextBlock
   bInherited=true
   parent=VerticalBox_29
   widget=DriverLatencyUI
   properties {
    text="{Driver}"
    font {
     bForceMonospaced=false
     bMaterialIsStencil=false
     fontMaterial=null
     fontObject=&NotoSans
     letterSpacing=0
     monospacedWidth=1
     size=8
     skewAmount=0
     typefaceFontName=NotoSans-Bold
     outlineSettings {
      bApplyOutlineToDropShadows=false
      bMiteredCorners=false
      bSeparateFillAlpha=false
      outlineMaterial=null
      outlineSize=0
      outlineColor=<rgba 0 0 0 1>
     }
    }
    shadowOffset=<0 0>
   }
   slot {class=VerticalBoxSlot}
  }
  OSWorkLatencyUI {
   widgetClassPath=&CommonTextBlock
   bInherited=true
   parent=VerticalBox_29
   widget=OSWorkLatencyUI
   properties {
    text="{OSWork}"
    font {
     bForceMonospaced=false
     bMaterialIsStencil=false
     fontMaterial=null
     fontObject=&NotoSans
     letterSpacing=0
     monospacedWidth=1
     size=8
     skewAmount=0
     typefaceFontName=NotoSans-Bold
     outlineSettings {
      bApplyOutlineToDropShadows=false
      bMiteredCorners=false
      bSeparateFillAlpha=false
      outlineMaterial=null
      outlineSize=0
      outlineColor=<rgba 0 0 0 1>
     }
    }
    shadowOffset=<0 0>
   }
   slot {class=VerticalBoxSlot}
  }
  GPURenderLatencyUI {
   widgetClassPath=&CommonTextBlock
   bInherited=true
   parent=VerticalBox_29
   widget=GPURenderLatencyUI
   properties {
    text="{GPU}"
    font {
     bForceMonospaced=false
     bMaterialIsStencil=false
     fontMaterial=null
     fontObject=&NotoSans
     letterSpacing=0
     monospacedWidth=1
     size=8
     skewAmount=0
     typefaceFontName=NotoSans-Bold
     outlineSettings {
      bApplyOutlineToDropShadows=false
      bMiteredCorners=false
      bSeparateFillAlpha=false
      outlineMaterial=null
      outlineSize=0
      outlineColor=<rgba 0 0 0 1>
     }
    }
    shadowOffset=<0 0>
   }
   slot {class=VerticalBoxSlot}
  }
  TweakLatencyUI {
   widgetClassPath=&CommonTextBlock
   bInherited=true
   parent=VerticalBox_29
   widget=TweakLatencyUI
   properties {
    text="{Tweak}"
    font {
     bForceMonospaced=false
     bMaterialIsStencil=false
     fontMaterial=null
     fontObject=&NotoSans
     letterSpacing=0
     monospacedWidth=1
     size=8
     skewAmount=0
     typefaceFontName=NotoSans-Bold
     outlineSettings {
      bApplyOutlineToDropShadows=false
      bMiteredCorners=false
      bSeparateFillAlpha=false
      outlineMaterial=null
      outlineSize=0
      outlineColor=<rgba 0 0 0 1>
     }
    }
    shadowOffset=<0 0>
   }
   slot {class=VerticalBoxSlot}
  }
 ]
}
