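# LatencyContainer is a HUD element widget that hosts the LatencyUI and LatencyGraph widgets.
# Both sit stacked in a vertical box inside a fixed-width size box, wrapped in an invalidation box.
# On construct, it clears the width override for mobile games.
# Visibility changes start the LatencyUI timer for visible states and stop it for collapsed or hidden, so the timer is not left running off screen.
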WidgetBlueprint /Game/Athena/HUD/LatencyContainer
graphs [EventGraph {nodes [
   Event Construct()
   > Branch(Condition=IsMobileGame())
   > ClearWidthOverride(self:Size Box Object Reference=Get SizeBox_0() -> (SizeBox_0:Size Box Object Reference))
   n1 = Event OnVisibilitySetEvent!() -> (OutputDelegate, then, InVisibility:ESlateVisibility Enum=Visible) {is=title K2Node_Event}
   > n2 = Switch on ESlateVisibility!(execute, Selection:ESlateVisibility Enum=@n1.InVisibility) -> (Visible:Exec, Collapsed:Exec, Hidden:Exec, HitTestInvisible:Exec, SelfHitTestInvisible:Exec) {is=title K2Node_SwitchEnum}
   n3 = Get LatencyUI() -> (LatencyUI:Latency UI Object Reference)
   StopTimer(execute=[@n2.Collapsed, @n2.Hidden], self:Latency Widget Object Reference=@n3)
   StartTimer(execute=[@n2.HitTestInvisible, @n2.SelfHitTestInvisible, @n2.Visible], self:Latency Widget Object Reference=@n3)
  ]}]
variables []
widgets {
 info {inheritedWidgetCount=1, parentClass=&FortHUDElementWidget, rootWidgetClass=/Script/UMG.InvalidationBox, widgetCount=6}
 widgets [
  InvalidationBox_0 {widgetClassPath=&InvalidationBox, bIsVariable=true, widget=InvalidationBox_0}
  SizeBox_0 {
   widgetClassPath=&SizeBox
   bIsVariable=true
   parent=InvalidationBox_0
   widget=SizeBox_0
   properties {bOverride_WidthOverride=true, widthOverride=180}
   slot {class=PanelSlot}
  }
  VerticalBox_0 {widgetClassPath=&VerticalBox, parent=SizeBox_0, widget=VerticalBox_0, slot {class=SizeBoxSlot}}
  LatencyUI {widgetClassPath=&/LatencyUI, bIsVariable=true, parent=VerticalBox_0, widget=LatencyUI, slot {class=VerticalBoxSlot}}
  LatencyGraph {widgetClassPath=&/LatencyGraph, bIsVariable=true, parent=VerticalBox_0, widget=LatencyGraph, slot {class=VerticalBoxSlot}}
  TutorialHighlight {widgetClassPath=&TutorialHighlight}
 ]
}
