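# NetUI is an Athena HUD element Widget Blueprint, based on FortHUDElementWidget, that hosts the network debug display.
# Its root is a NetDebugContainer widget exposed as a variable, alongside a TutorialHighlight widget.
# The Blueprint defines no variables or event logic of its own, so behavior comes from the parent class and the nested widgets.
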WidgetBlueprint /Game/Athena/HUD/NetUI
graphs [EventGraph {links [], nodes []}]
variables []
widgets {
 info {
  inheritedWidgetCount=1
  parentClass=&FortHUDElementWidget
  rootWidgetClass=&/NetDebugContainer.NetDebugContainer_C
  widgetCount=2
 }
 widgets [
  NetDebugContainer {widgetClassPath=&/NetDebugContainer, bIsVariable=true, widget=NetDebugContainer}
  TutorialHighlight {widgetClassPath=&TutorialHighlight}
 ]
}
