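# BadNetworkIndicator is a HUD element widget for the Athena mode, based on FortHUDElementWidget, that displays a bad-network warning icon.
# Its root NetworkIssueIndicator image draws the packet-loss warning material, alongside a TutorialHighlight widget.
# On Construct the widget collapses itself, and nothing in this graph makes it visible again.
# A short Intro animation exists but has no bindings and is not played here.
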WidgetBlueprint /Game/Athena/HUD/BadNetworkIndicator
animations [Intro {binding_count=0, length_seconds=0.25, object=&self:Intro, sequence {bindings []}}]
graphs [EventGraph {nodes [
   Event Construct()
   > SetVisibility(InVisibility=Collapsed)
  ]}]
variables []
widgets {
 info {inheritedWidgetCount=1, parentClass=&FortHUDElementWidget, rootWidgetClass=&Image, widgetCount=2}
 widgets [
  NetworkIssueIndicator {
   widgetClassPath=&Image
   bIsVariable=true
   widget=NetworkIssueIndicator
   properties {brush {
     bIsDynamicallyLoaded=false
     drawAs=Image
     imageType=NoImage
     mirroring=NoMirror
     resourceName=None
     resourceObject=/Game/Athena/HUD/Materials/M_UI_BadNetworkWarning_PacketLoss
     tiling=NoTile
     imageSize=<64 64>
     margin=<ltrb 0 0 0 0>
     outlineSettings {
      bUseBrushTransparency=false
      roundingType=HalfHeightRadius
      width=0
      color {colorUseRule=UseColor_Specified, specifiedColor=<rgba 0 0 0 0>}
      cornerRadii=<0 0 0 0>
     }
     tintColor {colorUseRule=UseColor_Specified, specifiedColor=<rgba 1 1 1 1>}
     uVRegion {bIsValid=false, max=<0 0>, min=<0 0>}
    }}
  }
  TutorialHighlight {widgetClassPath=&TutorialHighlight}
 ]
}
