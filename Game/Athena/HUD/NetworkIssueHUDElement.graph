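# Athena HUD element that hosts the bad-network indicator, tagged HUD.TopLeft.NetIssue, to flag connection issues.
# A FortHUDElementWidget subclass, it wraps the BadNetworkIndicator child widget in a root SizeBox.
# On PreConstruct, SetMobileStyle overrides the SizeBox to a fixed small square when running as a mobile game with touch input.
# Construct and Tick have no logic, and the graph does no network detection itself.
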WidgetBlueprint /Game/Athena/HUD/NetworkIssueHUDElement
graphs [
 EventGraph {nodes [
   Event PreConstruct()
   > SetMobileStyle() {is=title K2Node_CallFunction}
   Event Construct()
   Event Tick()
  ]}
 SetMobileStyle {nodes [
   SetMobileStyle() {is=title K2Node_FunctionEntry}
   n1 = GetOwningPlayer()
   > Branch(Condition=AND Boolean(A=IsMobileGame(), B=IsUsingTouch(PlayerController=@n1)))
   n2 = Get SizeBox_0() -> (SizeBox_0:Size Box Object Reference)
   > SetWidthOverride(self=@n2, InWidthOverride=40)
   > SetHeightOverride(self=@n2, InHeightOverride=40)
  ]}
]
variables []
widgets {
 info {inheritedWidgetCount=1, parentClass=&FortHUDElementWidget, rootWidgetClass=&SizeBox, widgetCount=3}
 widgets [
  SizeBox_0 {widgetClassPath=&SizeBox, bIsVariable=true, widget=SizeBox_0}
  BadNetworkIndicator {
   widgetClassPath=&/BadNetworkIndicator
   bIsVariable=true
   parent=SizeBox_0
   widget=BadNetworkIndicator
   properties {elementTag {tagName=HUD.TopLeft.NetIssue}}
   slot {class=SizeBoxSlot}
  }
  TutorialHighlight {widgetClassPath=&TutorialHighlight}
 ]
}
