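# HUD_NetStats is an Athena HUD widget that groups network status readouts in a single horizontal row.
# The row holds child widgets LatencyContainer, NetworkIssueHUDElement and NetUI.
# The latter two carry HUD element tags for hiding during post-game and post-game spectating, with the network issue element also tagged for capture and conversation hiding.
# The blueprint's event graph holds only unconnected event stubs, with no logic of its own.
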WidgetBlueprint /Game/Athena/HUD/HUD_NetStats
graphs [EventGraph {nodes [
   Event PreConstruct()
   Event Construct()
   Event Tick()
  ]}]
variables []
widgets {
 info {inheritedWidgetCount=0, parentClass=&UserWidget, rootWidgetClass=&HorizontalBox, widgetCount=4}
 widgets [
  NetStatsBox {widgetClassPath=&HorizontalBox, widget=NetStatsBox}
  LatencyContainer {widgetClassPath=&/LatencyContainer, parent=NetStatsBox, widget=LatencyContainer, slot {class=HorizontalBoxSlot}}
  NetworkIssueHUDElement {
   widgetClassPath=&/NetworkIssueHUDElement
   parent=NetStatsBox
   widget=NetworkIssueHUDElement
   properties {hUDElementTag {gameplayTags [
      {tagName=Athena.HideForCapture}
      {tagName=HUD.Athena.HideForPostGameSpectate}
      {tagName=HUD.Athena.HideForPostGame}
      {tagName=HUD.Athena.HideForConversation}
     ]}}
   slot {class=HorizontalBoxSlot, properties {verticalAlignment=VAlign_Top, padding=<ltrb 0 0 6 0>}}
  }
  NetUI {
   widgetClassPath=&/NetUI
   parent=NetStatsBox
   widget=NetUI
   properties {
    visibility=HitTestInvisible
    hUDElementTag {gameplayTags [{tagName=HUD.Athena.HideForPostGameSpectate}, {tagName=HUD.Athena.HideForPostGame}]}
   }
   slot {class=HorizontalBoxSlot}
  }
 ]
}
