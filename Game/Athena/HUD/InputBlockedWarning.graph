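# HUD widget that shows an "INPUT BLOCKED" banner with the icon of the filtered input device.
# OnInputFiltered looks up the icon for the attempted input type, collapsing the banner if none exists.
# Otherwise it reveals the banner and restarts a short one-shot timer, looping ShowWarning only when no timer is active; that animation has no bindings.
# A WarningFinished event collapses the banner, though the timer's delegate target is not named.
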WidgetBlueprint /Game/Athena/HUD/InputBlockedWarning
animations [ShowWarning {binding_count=0, display_label="", length_seconds=5, object=&self:ShowWarning, sequence {bindings []}}]
graphs [EventGraph {nodes [
   WarningFinished() {is=title K2Node_CustomEvent}
   > SetVisibility(self=Get RootBorder() -> (RootBorder:Border Object Reference), InVisibility=Collapsed)
   n1 = Event OnInputFiltered!() -> (OutputDelegate, then, AttemptedInputType:ECommonInputType Enum) {is=title K2Node_Event}
   > n2 = GetInputTypeIcon!(
    execute
    self:Fort Global UIContext Object Reference=Get FortGlobalUIContext() -> (ReturnValue:Fort Global UIContext Object Reference) {is=title K2Node_GetSubsystem}
    InputType:ECommonInputType Enum=@n1.AttemptedInputType
   ) -> (then, ReturnValue:Texture 2D Object Reference)
   > n3 = Is Valid(InputObject=@n2)
   SetVisibility(execute=@n3.Is Not Valid, self=Get RootBorder(), InVisibility=Collapsed)
   SetBrushFromTexture!(
    execute=@n3.Is Valid
    self:Image Object Reference=Get InputIcon() -> (InputIcon:Image Object Reference)
    Texture:Texture 2D Object Reference=@n2
    bMatchSize:bool=false
   ) -> (then)
   > n4 = Branch(Condition=Is Valid Timer Handle!(Handle:Timer Handle Structure=Get WarningTimerHandle() -> (WarningTimerHandle:Timer Handle Structure)) -> (ReturnValue:bool))
   Clear and Invalidate Timer by Handle(execute=[@n6.then, @n4.then], Handle=Get WarningTimerHandle())
   > n5 = Set Timer by Event(Delegate=Create Event(), Time=1, bLooping=false)
   > Set WarningTimerHandle(WarningTimerHandle:Timer Handle Structure=@n5) -> (Output_Get:Timer Handle Structure)
   > SetVisibility(self=Get RootBorder(), InVisibility=HitTestInvisible)
   n6 = PlayAnimation(execute=@n4.else, InAnimation=Get ShowWarning() -> (ShowWarning:Widget Animation Object Reference), NumLoopsToPlay=0)
  ]}]
variables [WarningTimerHandle {type=TimerHandle}]
widgets {
 info {inheritedWidgetCount=1, parentClass=/Script/FortniteUI.FortInputBlockedWarning, rootWidgetClass=&Border, widgetCount=6}
 widgets [
  RootBorder {
   widgetClassPath=&Border
   bIsVariable=true
   widget=RootBorder
   properties {visibility=Collapsed, brushColor=<rgba 0 0 0 0.25>, padding=<ltrb 10 4 10 4>}
  }
  Border_1 {
   widgetClassPath=&Border
   parent=RootBorder
   widget=Border_1
   properties {
    background {
     bIsDynamicallyLoaded=false
     drawAs=Image
     imageType=NoImage
     mirroring=NoMirror
     resourceName=None
     resourceObject=/Game/UI/Foundation/Modals/ErrorWindow/Materials/M_UI_ErrorPopUpBorder
     tiling=NoTile
     imageSize=<32 32>
     margin=<ltrb 0 0 0 0>
     outlineSettings {
      bUseBrushTransparency=false
      roundingType=HalfHeightRadius
      width=0
      color {colorUseRule=UseColor_Specified, specifiedColor=<rgba 0 0 0 0>}
      cornerRadii=<0 0 0 0>
     }
     tintColor {colorUseRule=UseColor_Specified, specifiedColor=<rgba 1 1 1 1>}
     uVRegion {bIsValid=false, max=<0 0>, min=<0 0>}
    }
    brushColor=<rgba 0.7656 0.8728 1 1>
    padding=<ltrb 4 10 4 10>
   }
   slot {class=BorderSlot, properties {padding=<ltrb 10 4 10 4>}}
  }
  HorizontalBox_1 {
   widgetClassPath=&HorizontalBox
   parent=Border_1
   widget=HorizontalBox_1
   slot {class=BorderSlot, properties {padding=<ltrb 4 10 4 10>}}
  }
  InputIcon {
   widgetClassPath=&Image
   bIsVariable=true
   parent=HorizontalBox_1
   widget=InputIcon
   properties {brush {
     bIsDynamicallyLoaded=false
     drawAs=Image
     imageType=NoImage
     mirroring=NoMirror
     resourceName=None
     resourceObject=/Game/UI/Icons/Input/T_UI_Icon_Input_Controller_128
     tiling=NoTile
     imageSize=<80 80>
     margin=<ltrb 0 0 0 0>
     outlineSettings {
      bUseBrushTransparency=false
      roundingType=HalfHeightRadius
      width=0
      color {colorUseRule=UseColor_Specified, specifiedColor=<rgba 0 0 0 0>}
      cornerRadii=<0 0 0 0>
     }
     tintColor {colorUseRule=UseColor_Specified, specifiedColor=<rgba 1 1 1 1>}
     uVRegion {bIsValid=false, max=<0 0>, min=<0 0>}
    }}
   slot {class=HorizontalBoxSlot, properties {verticalAlignment=VAlign_Center}}
  }
  Title {
   widgetClassPath=&CommonTextBlock
   bIsVariable=true
   parent=HorizontalBox_1
   widget=Title
   properties {
    justification=Center
    style=/FortUILibrary/Styles/Typography/Foundation/TextStyle-Header-M.TextStyle-Header-M_C
    text=INPUT BLOCKED
    font {
     bForceMonospaced=false
     bMaterialIsStencil=false
     fontMaterial=null
     fontObject=&Burbank
     letterSpacing=0
     monospacedWidth=1
     size=32
     skewAmount=0
     typefaceFontName=Burbank-Black
     outlineSettings {
      bApplyOutlineToDropShadows=false
      bMiteredCorners=false
      bSeparateFillAlpha=false
      outlineMaterial=null
      outlineSize=0
      outlineColor=<rgba 0 0 0 1>
     }
    }
    shadowOffset=<0 0>
   }
   slot {class=HorizontalBoxSlot, properties {verticalAlignment=VAlign_Center, padding=<ltrb 10 0 0 0>}}
  }
  TutorialHighlight {widgetClassPath=&TutorialHighlight}
 ]
}
