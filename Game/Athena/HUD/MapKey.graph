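# MapKey is a HUD element widget, based on FortHUDElementWidget, that shows the key binding for toggling the full-screen map in Athena.
# A horizontal row places a GlobalActionKeybindWidget, set to the ToggleFullScreenMap action, beside a small map icon from the shared HUD atlas.
# A TutorialHighlight widget is also present.
# The event graph is empty, so the widget is layout only and adds no Blueprint logic.
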WidgetBlueprint /Game/Athena/HUD/MapKey
graphs [EventGraph {links [], nodes []}]
variables []
widgets {
 info {inheritedWidgetCount=1, parentClass=&FortHUDElementWidget, rootWidgetClass=&HorizontalBox, widgetCount=4}
 widgets [
  MapRow {widgetClassPath=&HorizontalBox, widget=MapRow}
  KeybindMap {
   widgetClassPath=/Game/UI/InGame/KeybindWidget/GlobalActionKeybindWidget
   parent=MapRow
   widget=KeybindMap
   properties {globalAction=ToggleFullScreenMap}
   slot {class=HorizontalBoxSlot, properties {verticalAlignment=VAlign_Center}}
  }
  Map {
   widgetClassPath=&Image
   bIsVariable=true
   parent=MapRow
   widget=Map
   properties {brush {
     bIsDynamicallyLoaded=false
     drawAs=Image
     imageType=NoImage
     mirroring=NoMirror
     resourceName=None
     resourceObject=/Game/Athena/HUD/Atlas/Shared/icon_map_Sprite
     tiling=NoTile
     imageSize=<32 32>
     margin=<ltrb 0 0 0 0>
     outlineSettings {
      bUseBrushTransparency=false
      roundingType=HalfHeightRadius
      width=0
      color {colorUseRule=UseColor_Specified, specifiedColor=<rgba 0 0 0 0>}
      cornerRadii=<0 0 0 0>
     }
     tintColor {colorUseRule=UseColor_Specified, specifiedColor=<rgba 1 1 1 1>}
     uVRegion {bIsValid=false, max=<0 0>, min=<0 0>}
    }}
   slot {class=HorizontalBoxSlot, properties {verticalAlignment=VAlign_Center, padding=<ltrb 4 0 0 0>}}
  }
  TutorialHighlight {widgetClassPath=&TutorialHighlight}
 ]
}
