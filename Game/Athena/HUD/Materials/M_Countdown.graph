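# Translucent UI material for a HUD countdown, combining a two-color gradient with a procedural animated flame.
# The gradient blends FirstColor and SecondColor using a powered vertical UV ramp and an inverted radial falloff.
# Two panning fire noises, one bump-offsetting the other, form flames with a sine-pulsed radial exponent and a FlameHeight-shifted radial mask.
# That result drives a dark-red-to-yellow ramp through Firepower and a SmoothStep mask between MaskMin and MaskMax.
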Material /Game/Athena/HUD/Materials/M_Countdown
graph {
 properties {blendMode=BLEND_Translucent, materialDomain=MD_UI, pixelDepthOffsetMode=PDOM_Legacy, usedShadingModels=Default Lit}
 nodes [
  n1 = ComponentMask(None=Multiply(A=TextureCoordinate() {vTiling=1.5}, B=ScalarParameter() {defaultValue=1.5, parameterName=UVScale})) {g=true}
  n2 = ScalarParameter() {defaultValue=5, parameterName=GradientPower}
  n3 = VectorParameter() {parameterName=RadiusPosition, defaultValue=<rgba 1 0 0 1>}
  n4 = MaterialFunctionCall(
   CenterPosition=AppendVector(A=@n3.R, B=@n3.R)
   Radius=ScalarParameter() {defaultValue=0.5, parameterName=Radius}
   Density=ScalarParameter() {defaultValue=2.33, parameterName=Density}
  ) {materialFunction=&RadialGradientExponential}
  LinearInterpolate(
   A=VectorParameter() {parameterName=FirstColor, defaultValue=<rgba 1 1 1 1>}.RGB
   B=VectorParameter() {parameterName=SecondColor, defaultValue=<rgba 1 0.8815 0.487 1>}.RGB
   Alpha=Multiply(A=Power(Base=@n1, Exp=@n2), B=OneMinus(None=@n4.RadialGradientExponential))
  )
  n5 = TextureSample(UVs=Panner() {speedY=0.5}) {texture=/Game/Effects/Fort_Effects/Textures/Fire/T_FireNoise_02}
  n6 = ScalarParameter() {defaultValue=2, parameterName=FlameDistort}
  n7 = TextureSample(UVs=Panner(Coordinate=MaterialFunctionCall!(HeightTexture=@n5.R, TextureCoordinates, ReferencePlane, HeightRatioInput=@n6, CameraVector) {materialFunction=/Engine/Functions/Engine_MaterialFunctions02/Utility/BumpOffset_Advanced.BumpOffset_advanced}.UVs) {speedY=0.2}) {texture=/Game/Global/Textures/Default/Noise/T_FireNoise_01}
  n8 = Constant() {r=1}
  n9 = Constant() {r=2}
  n10 = MaterialFunctionCall(Value=Time()) {materialFunction=&LinearSine}
  n11 = Multiply(A=MaterialFunctionCall(CenterPosition=Constant2Vector() {g=1, r=0.5}, Radius=LinearInterpolate(A=@n8, B=@n9, Alpha=@n10.Linear Sine), Density=Constant() {r=0.7164}) {materialFunction=&RadialGradientExponential}.RadialGradientExponential) {constB=-1}
  n12 = MaterialFunctionCall(Texture Scale=Constant2Vector() {g=2, r=3}) {materialFunction=&ScaleUVsByCenter}
  n13 = VectorParameter() {parameterName=FlameHeight, defaultValue=<rgba 0 0 0 1>}
  n14 = AppendVector(A=@n13.R, B=@n13.R)
  n15 = Multiply(
   A=LinearInterpolate(B=Power(Base=@n5.R) {constExponent=-0.75}, Alpha=Power(Base=@n7.R, Exp=@n11) {constExponent=5})
   B=MaterialFunctionCall(
    UVs=Add(A=@n12.UVs, B=@n14)
    CenterPosition=Constant2Vector() {g=1, r=0.5}
    Radius=Constant() {r=0.8043}
    Density=Constant() {r=2.3384}
    Invert Density=StaticBool() {value=true}
   ) {materialFunction=&RadialGradientExponential}.RadialGradientExponential
  )
  LinearInterpolate(
   A=Constant3Vector() {constant=<rgba 0.3281 0.0154 0 0>}
   B=Constant3Vector() {constant=<rgba 1 0.8698 0.141 0>}
   Alpha=Power(Base=@n15, Exp=ScalarParameter() {defaultValue=2, parameterName=Firepower})
  )
  MaterialFunctionCall(HeightTexture=@n7.R, HeightRatioInput=@n6) {materialFunction=&BumpOffset_advanced}
  MaterialFunctionCall(
   Alpha=@n15
   Min=ScalarParameter() {defaultValue=0.1, parameterName=MaskMin}
   Max=ScalarParameter() {defaultValue=0.75, parameterName=MaskMax}
  ) {materialFunction=&SmoothStep}
  ScalarParameter() {defaultValue=1, parameterName=Transparency}
 ]
}
