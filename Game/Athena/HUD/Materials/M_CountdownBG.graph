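# Additive UI material for a countdown background in the Athena HUD, drawing a tiled diagonal pattern inside an exponential radial gradient.
# The pattern repeats over mesh or screen-aligned UVs via a static switch, fades toward white by PatternFade, then is multiplied by the gradient.
# An optional time-driven LinearSine pulse multiplies the result, but both lerp ends read PulseConstraints.R, so as wired it yields a constant, not an oscillation.
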Material /Game/Athena/HUD/Materials/M_CountdownBG
graph {
 properties {blendMode=BLEND_Additive, materialDomain=MD_UI, pixelDepthOffsetMode=PDOM_Legacy, usedShadingModels=Default Lit}
 nodes [
  n1 = MaterialFunctionCall!(TextureResolution) {materialFunction=/Engine/Functions/Engine_MaterialFunctions02/Texturing/ScreenAlignedPixelToPixelUVs}
  n2 = TextureCoordinate()
  n3 = TextureSample(UVs=Multiply(
    A=StaticSwitchParameter(True=@n1.Result, False=@n2) {parameterName=ScreenAlignedTexture}
    B=ScalarParameter() {defaultValue=30, parameterName=PatternRepeat}
   )) {texture=&/T_UI_DiagonalPattern}
  n4 = Constant3Vector() {constant=<rgba 1 1 1 0>}
  n5 = ScalarParameter() {parameterName=PatternFade}
  n6 = ScalarParameter() {defaultValue=0.5, parameterName=Radius}
  n7 = ScalarParameter() {defaultValue=2.6353, parameterName=Density}
  n8 = StaticBool() {value=true}
  n9 = VectorParameter() {group=Pulse, parameterName=PulseConstraints, defaultValue=<rgba 0 1 0 1>}
  n10 = Time()
  n11 = ScalarParameter() {defaultValue=1, group=Pulse, parameterName=PulseSpeed}
  n12 = MaterialFunctionCall(Value=Fmod(A=Multiply(A=@n10, B=@n11), B=Constant() {r=1})) {materialFunction=&LinearSine}
  Multiply(
   A=Multiply(
    A=LinearInterpolate(A=@n3.RGB, B=@n4, Alpha=@n5)
    B=MaterialFunctionCall(Radius=@n6, Density=@n7, Invert Density=@n8) {materialFunction=&RadialGradientExponential}.RadialGradientExponential
   )
   B=StaticSwitchParameter(True=LinearInterpolate(A=@n9.R, B=@n9.R, Alpha=@n12.Linear Sine), False=Constant() {r=1}) {group=Pulse, parameterName=isPulsing}
  )
 ]
}
