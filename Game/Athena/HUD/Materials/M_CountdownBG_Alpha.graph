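# Translucent UI material for the countdown background in the Athena HUD, producing a diagonal-pattern radial gradient.
# It tiles the diagonal pattern texture by a PatternRepeat parameter, then blends it toward solid white using PatternFade.
# The result is multiplied by an exponential radial gradient with inverted density, shaped by Radius and Density parameters.
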Material /Game/Athena/HUD/Materials/M_CountdownBG_Alpha
graph {
 properties {blendMode=BLEND_Translucent, materialDomain=MD_UI, pixelDepthOffsetMode=PDOM_Legacy, usedShadingModels=Default Lit}
 nodes [
  n1 = Multiply(A=TextureCoordinate(), B=ScalarParameter() {defaultValue=30, parameterName=PatternRepeat})
  Multiply(
   A=LinearInterpolate(
    A=TextureSample(UVs=@n1) {texture=&/T_UI_DiagonalPattern}.RGB
    B=Constant3Vector() {constant=<rgba 1 1 1 0>}
    Alpha=ScalarParameter() {parameterName=PatternFade}
   )
   B=MaterialFunctionCall(
    Radius=ScalarParameter() {defaultValue=0.5, parameterName=Radius}
    Density=ScalarParameter() {defaultValue=2.6353, parameterName=Density}
    Invert Density=StaticBool() {value=true}
   ) {materialFunction=&RadialGradientExponential}.RadialGradientExponential
  )
 ]
}
