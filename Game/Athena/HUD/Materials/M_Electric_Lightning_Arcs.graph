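# Additive UI material for the HUD that draws animated, looping electric lightning arcs.
# A FlipBook driven by offset, speed-scaled time animates the arc texture, with CustomRotator applying an adjustable rotation.
# A scaled noise texture, raised to an exponent that ramps up cyclically, masks the arcs for a repeating decay.
# The masked green and red channels get density powers, tint as glow and core, then sum and scale by OpacityMult.
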Material /Game/Athena/HUD/Materials/M_Electric_Lightning_Arcs
graph {
 properties {blendMode=BLEND_Additive, materialDomain=MD_UI, numCustomizedUVs=1, pixelDepthOffsetMode=PDOM_Legacy, usedShadingModels=Default Lit}
 nodes [
  Constant() {r=1}
  n1 = TextureCoordinate()
  n2 = ScalarParameter() {defaultValue=2, parameterName=NoiseScale}
  n3 = Multiply(A=Add(A=Time() {period=1}, B=ScalarParameter() {parameterName=TimeOffset}), B=ScalarParameter() {defaultValue=1, parameterName=DecaySpeed})
  n4 = Constant() {r=1}
  n5 = Power(
   Base=TextureSampleParameter2D(UVs=Multiply(A=@n1, B=@n2)) {parameterName=Noise, samplerType=SAMPLERTYPE_LinearColor, texture=&HighContrastNoise}.R
   Exp=LinearInterpolate(Alpha=Fmod(A=@n3, B=@n4)) {constA=0.01, constB=20}
  )
  n6 = Multiply(A=Add(A=Time() {period=4}, B=ScalarParameter() {parameterName=TimeOffset}), B=ScalarParameter() {defaultValue=1, parameterName=Speed2})
  n7 = TextureSample(UVs=MaterialFunctionCall(
    UVs=MaterialFunctionCall("Animation  Phase (0-1)"=@n6) {materialFunction=&FlipBook}.UVs
    "Rotation Angle (0-1)"=ScalarParameter() {parameterName=PostRotation}
   ) {materialFunction=&CustomRotator}.Rotated Values) {texture=/Game/Effects/Fort_Effects/Textures/Electric/T_Electric_Lightning_Arcs}
  n8 = Multiply(A=@n5, B=Power(Base=@n7.G, Exp=ScalarParameter() {defaultValue=1, parameterName="Density (S)"}))
  n9 = VectorParameter() {parameterName=GlowColor, defaultValue=<rgba 1 0.2673 0 1>}
  n10 = Power(Base=Multiply(A=@n5, B=@n7.R), Exp=ScalarParameter() {defaultValue=1, parameterName="Density1 (S)"})
  n11 = VectorParameter() {parameterName=CoreColor, defaultValue=<rgba 1 1 1 1>}
  Multiply(A=Add(A=Multiply(A=@n8, B=@n9.RGB), B=Multiply(A=@n10, B=@n11.RGB)), B=ScalarParameter() {defaultValue=1, parameterName=OpacityMult})
  TextureCoordinate()
 ]
}
