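# Translucent UI material for the HUD's bad-network warning, layering high-latency and packet-loss icons over Wi-Fi bars.
# Display scalars toggle each warning, and severity scalars blend its gradient tint toward a severe color.
# A time-sine strobe with override pulses both warnings and masks the bars beneath them.
# Icons sample separate UV channels, with parameterized offset, rotate and scale chains apparently feeding the customized UVs.
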Material /Game/Athena/HUD/Materials/M_UI_BadNetworkWarning
graph {
 properties {blendMode=BLEND_Translucent, materialDomain=MD_UI, numCustomizedUVs=3, pixelDepthOffsetMode=PDOM_Legacy, usedShadingModels=Default Lit}
 nodes [
  n1 = ScalarParameter() {defaultValue=1, parameterName=DisplayHighLatency}
  n2 = VectorParameter() {parameterName=Exclaim_ColorB, defaultValue=<rgba 0 0.387 1 1>}
  n3 = TextureSample(UVs=TextureCoordinate() {coordinateIndex=2}) {samplerSource=SSM_Clamp_WorldGroupSettings, samplerType=SAMPLERTYPE_LinearColor, texture=&/T_UI_BadPing_SlowArrow}
  n4 = Multiply(A=@n2.RGB, B=Multiply(A=OneMinus(None=@n3.B), B=@n3.R))
  n5 = VectorParameter() {parameterName=Exclaim_ColorA_GradStop1, defaultValue=<rgba 0.7274 0 0.0072 0>}
  n6 = VectorParameter() {parameterName=Exclaim_ColorA_GradStop2, defaultValue=<rgba 1 0 0.2816 1>}
  n7 = TextureCoordinate()
  n8 = VectorParameter() {parameterName=Severe_GradStop2, defaultValue=<rgba 1 0 0.1787 1>}
  n9 = LinearInterpolate(
   A=LinearInterpolate(A=@n5.RGB, B=@n6.RGB, Alpha=ComponentMask(None=@n7) {g=true})
   B=@n8.RGB
   Alpha=ScalarParameter() {defaultValue=1, parameterName=Severity_HighLatency}
  )
  n10 = Multiply(A=@n9, B=@n3.R)
  n11 = ScalarParameter() {defaultValue=1.2, parameterName=StrobeSpeed}
  n12 = LinearInterpolate(
   A=LinearInterpolate(
    A=ScalarParameter() {defaultValue=0.5, parameterName=StrobeLowerOpacityBound}
    Alpha=MaterialFunctionCall(Speed=@n11) {materialFunction=&DebugTimeSine}.Result
   ) {constA=0.25}
   Alpha=ScalarParameter() {defaultValue=1, parameterName=OverrideStrobe}
  )
  n13 = Multiply(A=Multiply(A=Add(A=@n4, B=@n10), B=@n9), B=@n12)
  n14 = ScalarParameter() {defaultValue=1, parameterName=DisplayPacketLoss}
  n15 = TextureCoordinate()
  n16 = VectorParameter() {parameterName=Severe_GradStop1, defaultValue=<rgba 1 0 0 1>}
  n17 = LinearInterpolate(
   A=LinearInterpolate(A=@n5.RGB, B=@n6.RGB, Alpha=ComponentMask(None=@n15) {g=true})
   B=@n16.RGB
   Alpha=ScalarParameter() {defaultValue=1, parameterName=Severity_PacketLoss}
  )
  n18 = TextureSample(UVs=TextureCoordinate() {coordinateIndex=1}) {samplerSource=SSM_Clamp_WorldGroupSettings, samplerType=SAMPLERTYPE_LinearColor, texture=&/T_UI_BadPing}
  n19 = OneMinus(None=@n18.B)
  n20 = Add(A=Multiply(A=Multiply(A=@n17, B=@n18.B), B=@n12), B=Multiply(A=@n2.RGB, B=Multiply(A=@n19, B=@n18.R)))
  n21 = Add(A=@n3.R, B=@n3.R)
  n22 = ScalarParameter() {defaultValue=1, parameterName=DisplayHighLatency}
  n23 = ScalarParameter() {defaultValue=1, parameterName=DisplayPacketLoss}
  n24 = Add(A=@n18.R, B=@n18.R)
  n25 = Clamp(None=Add(A=Multiply(A=@n21, B=@n22), B=Multiply(A=@n23, B=@n24)))
  n26 = OneMinus(None=Multiply(A=@n25, B=@n12))
  n27 = TextureSample(UVs=TextureCoordinate()) {samplerSource=SSM_Clamp_WorldGroupSettings, samplerType=SAMPLERTYPE_LinearColor, texture=&T_UI_BadPing}
  Add(
   A=Add(A=Multiply(A=@n1, B=@n13), B=Multiply(A=@n14, B=@n20))
   B=Multiply(A=VectorParameter() {parameterName=WifiBars_ColorA, defaultValue=<rgba 1 1 1 0>}.RGB, B=Multiply(A=@n26, B=@n27.G))
  )
  Add(A=Multiply(A=@n25, B=@n12), B=Multiply(A=@n27.G, B=ScalarParameter() {defaultValue=1, parameterName=WifiOpacity}))
  n28 = TextureCoordinate()
  Floor(None=Multiply(A=ComponentMask(None=@n28) {r=true}) {constB=2})
  LinearInterpolate(A=@n16.RGB, B=@n8.RGB, Alpha=ComponentMask(None=TextureCoordinate()) {g=true})
  n29 = MaterialFunctionCall(Float2=TextureCoordinate()) {materialFunction=&BreakOutFloat2Components}
  n30 = Add(A=@n29.R, B=ScalarParameter() {parameterName=Exclaim1_XOffset})
  n31 = Add(A=@n29.G, B=ScalarParameter() {parameterName=Exclaim2_YOffset})
  n32 = VectorParameter() {parameterName=Exclaim1_Pivot, defaultValue=<rgba 0 0 -1 -1>}
  MaterialFunctionCall(
   UVs=MaterialFunctionCall(
    UVs=MaterialFunctionCall(X=@n30, Y=@n31) {materialFunction=&MakeFloat2}.Result
    Rotation Center=MaterialFunctionCall(X=@n32.R, Y=@n32.R) {materialFunction=&MakeFloat2}.Result
    "Rotation Angle (0-1)"=ScalarParameter() {parameterName=Exclaim1_Rot}
   ) {materialFunction=&CustomRotator}.Rotated Values
   Texture Scale=MaterialFunctionCall(X=ScalarParameter() {defaultValue=1, parameterName=Exclaim1_XScale}, Y=ScalarParameter() {defaultValue=1, parameterName=Exclaim1_YScale}) {materialFunction=&MakeFloat2}.Result
  ) {materialFunction=&ScaleUVsByCenter}
  n33 = MaterialFunctionCall(Float2=TextureCoordinate()) {materialFunction=&BreakOutFloat2Components}
  n34 = Add(A=@n33.R, B=ScalarParameter() {defaultValue=0.3576, parameterName=Exclaim_XOffset})
  n35 = Add(A=@n33.G, B=ScalarParameter() {parameterName=Exclaim_YOffset})
  n36 = VectorParameter() {parameterName=Exclaim_Pivot, defaultValue=<rgba 0 0 -1 -1>}
  MaterialFunctionCall(
   UVs=MaterialFunctionCall(
    UVs=MaterialFunctionCall(X=@n34, Y=@n35) {materialFunction=&MakeFloat2}.Result
    Rotation Center=MaterialFunctionCall(X=@n36.R, Y=@n36.R) {materialFunction=&MakeFloat2}.Result
    "Rotation Angle (0-1)"=ScalarParameter() {parameterName=Exclaim_Rot}
   ) {materialFunction=&CustomRotator}.Rotated Values
   Texture Scale=MaterialFunctionCall(X=ScalarParameter() {defaultValue=1, parameterName=Exclaim_XScale}, Y=ScalarParameter() {defaultValue=1, parameterName=Exclaim_YScale}) {materialFunction=&MakeFloat2}.Result
  ) {materialFunction=&ScaleUVsByCenter}
 ]
}
