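# M_UI_CardinalNorth is a translucent UI material for the Athena HUD that shapes a north cardinal marker from a box and a triangle.
# A rounded box SDF and a rotatable three-sided polygon SDF each get parameterized position and size.
# A smooth SDF union blends the two shapes, and an anti-aliasing function turns the result into a procedural mask.
# A Color parameter is also defined, though its connection to the output is not shown.
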Material /Game/Athena/HUD/Materials/M_UI_CardinalNorth
graph {
 properties {bAutomaticallySetUsageInEditor=false, blendMode=BLEND_Translucent, materialDomain=MD_UI, usedShadingModels=Default Lit}
 nodes [
  n1 = Add(
   A=TextureCoordinate()
   B=AppendVector(A=ScalarParameter() {parameterName=BoxPositionX}, B=ScalarParameter() {defaultValue=0.15, parameterName=BoxPositionY})
  ) {constB=0}
  n2 = MaterialFunctionCall(
   UVs=MaterialFunctionCall(UV=MaterialFunctionCall!(UVs=@n1) {materialFunction=/FortUILibrary/MaterialFunctions/Basics/MF_UI_9SliceUVs}.UVs) {materialFunction=&MF_UI_UVScaling}.Scaled to Fit
   Size=AppendVector(A=ScalarParameter() {defaultValue=0.5, parameterName=BoxSizeX}, B=ScalarParameter() {defaultValue=0.5, parameterName=BoxSizeY})
   CornerRoundness=ScalarParameter() {defaultValue=0.1, parameterName=BoxCornerRoundness}
  ) {materialFunction=&MF_UI_Box}
  n3 = MaterialFunctionCall(UVs=TextureCoordinate() {vTiling=1.75}, "Rotation Angle (0-1)"=ScalarParameter() {defaultValue=0.5, parameterName=TriangleRotation}) {materialFunction=&CustomRotator}
  n4 = AppendVector(A=ScalarParameter() {parameterName=TrianglePositionX}, B=ScalarParameter() {defaultValue=0.7, parameterName=TrianglePositionY})
  n5 = ScalarParameter() {defaultValue=0.5, parameterName=TriPivotX}
  n6 = ScalarParameter() {defaultValue=0.43, parameterName=TriPivotY}
  n7 = MaterialFunctionCall!(
   UVs=MaterialFunctionCall(UV=Add(A=@n3.Rotated Values, B=@n4), ScalePivotPoint=AppendVector(A=@n5, B=@n6)) {materialFunction=&MF_UI_UVScaling}.Scaled to Fit
   Size=ScalarParameter() {defaultValue=0.58, parameterName=TriangleSize}
   Sides=Constant() {r=3}
   InnerAngle=ScalarParameter() {parameterName=TriangleInnerAngle}
   OutlineThickness
   OutlineAlignment
   GlowMin
   GlowMax
  ) {materialFunction=/FortUILibrary/MaterialFunctions/Shapes/MF_UI_RegularPolygon}
  n8 = ScalarParameter() {defaultValue=0.04, parameterName=Blend Amount}
  MaterialFunctionCall(UV=MaterialFunctionCall(
    DistanceField=MaterialFunctionCall!(DistanceFieldA=@n2.SDF, DistanceFieldB=@n7.SDF, BlendAmount=@n8) {materialFunction=/FortUILibrary/MaterialFunctions/Shapes/Modifiers/MF_UI_SDF_Union}.Smooth
    Threshold=ScalarParameter() {defaultValue=0.1, parameterName=AntiAliasingThreshold}
   ) {materialFunction=&MF_UI_AntiAliasing}.Inner) {materialFunction=&MF_UI_UVScaling}
  VectorParameter() {parameterName=Color, defaultValue=<rgba 0 0 0 1>}
 ]
}
