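# Translucent UI material that draws a distance-field icon with an outline, drop shadow and optional background plate.
# SmoothStep thresholds the icon texture into fill and outline, with a second offset sample supplying the shadow.
# The plate uses scaled, rotatable UVs and a static switch between radial gradient and texture mask.
# Masks combine by Max for coverage and drive color lerps, with a switchable linear-dodge MF_Shine limited to the fill.
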Material /Game/Athena/HUD/Materials/M_UI_DFIconShadow
graph {
 properties {blendMode=BLEND_Translucent, materialDomain=MD_UI, numCustomizedUVs=1, pixelDepthOffsetMode=PDOM_Legacy, usedShadingModels=Default Lit}
 nodes [
  n1 = VectorParameter() {parameterName=ScaleAndOffset, defaultValue=<rgba 1 1 0.5 0.5>}
  MaterialFunctionCall!(
   Texture Scale=AppendVector(A=@n1.R, B=@n1.R)
   UVs=TextureCoordinate()
   Offset XY Texturesize Z=MaterialFunctionCall(R=@n1.B, G=@n1.B, B=Constant() {r=1}) {materialFunction=&AppendMany}.RGB
  ) {materialFunction=/Engine/Functions/Engine_MaterialFunctions01/Texturing/ScaleUVsAroundPoint}
  n2 = VectorParameter() {desc="r=X g=Y b=1 a=Scale", group=BGPlate, parameterName=PlateScale, defaultValue=<rgba 0.5 0.5 1 1>}
  n3 = MaterialFunctionCall(
   Texture Scale=@n2.A
   UVs=MaterialFunctionCall(UVs=TextureCoordinate(), "Rotation Angle (0-1)"=ScalarParameter() {group=BGPlate, parameterName=PlateRotation}) {materialFunction=&CustomRotator}.Rotated Values
   Offset XY Texturesize Z=MaterialFunctionCall(R=@n2.R, G=@n2.R, B=@n2.R) {materialFunction=&AppendMany}.RGB
  ) {materialFunction=&ScaleUVsAroundPoint}
  n4 = ScalarParameter() {defaultValue=0.5, group=BGPlate, parameterName=PlateGradientRadius}
  n5 = ScalarParameter() {defaultValue=2.33, group=BGPlate, parameterName=PlateGradientDensity}
  n6 = ScalarParameter() {defaultValue=0.1, group=BGPlate, parameterName=BGPlateEdge}
  n7 = MaterialFunctionCall(
   Alpha=StaticSwitchParameter(
    True=MaterialFunctionCall(UVs=@n3.UVs, Radius=@n4, Density=@n5) {materialFunction=&RadialGradientExponential}.RadialGradientExponential
    False=TextureSampleParameter2D(UVs=@n3.UVs) {group=BGPlate, parameterName=BGPlateDF, texture=&/T_UI_RadialGradient}.R
   ) {group=BGPlate, parameterName=UseRadialPlate}
   Min=@n6
   Max=Add(A=@n6, B=ScalarParameter() {defaultValue=0.041, group=BGPlate, parameterName=PlateEdgeSoftness})
  ) {materialFunction=&SmoothStep}
  n8 = VectorParameter() {group=BGPlate, parameterName=BGPlateColor, defaultValue=<rgba 0.0052 0.0052 0.0056 0>}
  n9 = VectorParameter() {parameterName=ShadowOffset, defaultValue=<rgba -0.03 -0.03 0 1>}
  n10 = TextureSampleParameter2D(UVs=Add(A=TextureCoordinate(), B=AppendVector(A=@n9.R, B=@n9.R))) {parameterName=Texture, samplerSource=SSM_Clamp_WorldGroupSettings, texture=&/T_UI_RebootIcon}
  n11 = Subtract(A=ScalarParameter() {defaultValue=0.19, parameterName=Edge}, B=ScalarParameter() {defaultValue=0.1, parameterName=OutlineSize})
  n12 = Add(A=@n11, B=ScalarParameter() {defaultValue=0.041, parameterName=EdgeSoftness})
  n13 = TextureSampleParameter2D() {parameterName=Texture, samplerSource=SSM_Clamp_WorldGroupSettings, texture=&T_UI_RebootIcon}
  n14 = Max(
   A=Multiply(
    A=@n9.A
    B=MaterialFunctionCall(Alpha=@n10.R, Min=@n11, Max=@n12) {materialFunction=&SmoothStep}.Result
   )
   B=MaterialFunctionCall(Alpha=@n13.R, Min=@n11, Max=@n12) {materialFunction=&SmoothStep}.Result
  )
  n15 = VectorParameter() {parameterName=BGColor, defaultValue=<rgba 0.0052 0.0052 0.0056 1>}
  n16 = Multiply(A=@n14, B=@n15.A)
  n17 = ScalarParameter() {defaultValue=0.19, parameterName=Edge}
  n18 = MaterialFunctionCall(Alpha=@n13.R, Min=@n17, Max=Add(A=@n17, B=ScalarParameter() {defaultValue=0.041, parameterName=EdgeSoftness})) {materialFunction=&SmoothStep}
  Reroute(None=Max(A=Multiply(A=@n7.Result, B=@n8.A), B=Max(A=@n16, B=@n18.Result)))
  n19 = TextureCoordinate()
  n20 = ComponentMask(None=@n19) {r=true}
  n21 = ComponentMask(None=@n19) {g=true}
  n22 = MaterialFunctionCall!(
   ShineOffset=LinearInterpolate(Alpha=ScalarParameter() {parameterName="ShineOffset (S)", sliderMax=1}) {constA=0.5, constB=-1.5}
   Rotation
   RotationCenter
   UVs=AppendVector(A=Multiply(A=@n20, B=@n21), B=@n21)
  ) {materialFunction=/Game/Athena/HUD/Materials/Victory/MF_Shine}
  n23 = Constant()
  n24 = Constant() {r=1}
  n25 = LinearInterpolate(
   A=LinearInterpolate(A=@n8.RGB, B=@n15.RGB, Alpha=If(A=@n8.A, B=@n23, "A > B"=@n14, "A == B"=@n24, "A < B"=@n24))
   B=VectorParameter() {parameterName=Color, defaultValue=<rgba 1 1 1 1>}.RGB
   Alpha=@n18.Result
  )
  StaticSwitchParameter(
   True=LinearInterpolate(
    A=@n25
    B=MaterialFunctionCall!(Blend=@n22.Result, Base=@n25) {materialFunction=/Engine/Functions/Engine_MaterialFunctions03/Blends/Blend_LinearDodge}.Result
    Alpha=@n18.Result
   )
   False=@n25
  ) {parameterName=UseShine}
 ]
}
