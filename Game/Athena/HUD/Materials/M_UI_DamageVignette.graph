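# Additive UI material in the Athena HUD, named as a damage vignette, built around a screen-edge gradient mask.
# An exponential radial gradient on the texture coordinates, with Radius and Density parameters, is inverted to brighten toward the edges.
# A separate UV scale from UI pixel size divides height by itself, correcting no aspect ratio, and does not feed the gradient.
# No connections to material outputs appear.
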Material /Game/Athena/HUD/Materials/M_UI_DamageVignette
graph {
 properties {blendMode=BLEND_Additive, materialDomain=MD_UI, pixelDepthOffsetMode=PDOM_Legacy, usedShadingModels=Default Lit}
 nodes [
  Constant() {r=1}
  n1 = TextureCoordinate()
  n2 = MaterialFunctionCall(Float2=MaterialFunctionCall() {materialFunction=&GetUserInterfaceUV}.Pixel Size) {materialFunction=&BreakOutFloat2Components}
  Multiply(A=@n1, B=AppendVector(A=Constant() {r=1}, B=Divide(A=@n2.G, B=@n2.G)))
  OneMinus(None=MaterialFunctionCall(
    UVs=@n1
    Radius=ScalarParameter() {defaultValue=1, parameterName="Radius (S)"}
    Density=ScalarParameter() {defaultValue=4, parameterName="Density (S)"}
   ) {materialFunction=&RadialGradientExponential}.RadialGradientExponential)
 ]
}
