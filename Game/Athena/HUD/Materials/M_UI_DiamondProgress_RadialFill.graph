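# Translucent UI material for a HUD diamond progress indicator that fills radially around a box outline.
# A square gradient is offset by Progress, with the IsCountdown static switch inverting both inputs to flip the fill behavior.
# The result is limited by a rerouted value, then Ceil creates a hard step blending BGColor to FillColor.
# MF_UI_Box builds the outline from Scale and FillWidth, feeding the SquareFill and Mask reroutes.
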Material /Game/Athena/HUD/Materials/M_UI_DiamondProgress_RadialFill
graph {
 properties {blendMode=BLEND_Translucent, materialDomain=MD_UI, pixelDepthOffsetMode=PDOM_Legacy, usedShadingModels=Default Lit}
 nodes [
  n1 = VectorParameter() {parameterName=FillColor, defaultValue=<rgba 0 0.3229 0.0127 1>}
  n2 = MaterialFunctionCall!(UVs, Start Point) {materialFunction=/FortUILibrary/MaterialFunctions/Gradients/MF_UI_SquareGradient}
  n3 = OneMinus(None=@n2.Gradient)
  n4 = ScalarParameter() {defaultValue=0.65, parameterName=Progress, sliderMax=1}
  n5 = OneMinus(None=@n4)
  n6 = Min(
   A=Subtract(
    A=StaticSwitchParameter(True=@n2.Gradient, False=@n3) {parameterName=IsCountdown}
    B=StaticSwitchParameter(True=@n4, False=@n5) {parameterName=IsCountdown}
   )
   B=NamedRerouteUsage()
  )
  LinearInterpolate(
   A=VectorParameter() {parameterName=BGColor, defaultValue=<rgba 0.0312 0.0312 0.0312 1>}.RGB
   B=@n1.RGB
   Alpha=Multiply(A=@n1.A, B=Ceil(None=@n6))
  )
  n7 = ScalarParameter() {defaultValue=1, parameterName=Scale}
  n8 = MaterialFunctionCall(
   UVs=MaterialFunctionCall(Texture Scale=@n7) {materialFunction=&ScaleUVsByCenter}.UVs
   Size=@n7
   OutlineThickness=ScalarParameter() {defaultValue=0.2, parameterName=FillWidth}
  ) {materialFunction=&MF_UI_Box}
  n9 = NamedRerouteDeclaration(None=@n8.Outline) {name=SquareFill, nodeColor=<rgba 0.9301 0 1 1>}
  n10 = Ceil(None=@n8.SDFOutline)
  NamedRerouteDeclaration(None=Multiply(A=Subtract(A=@n9, B=@n10), B=ScalarParameter() {defaultValue=1, parameterName=Mask})) {name=Mask, nodeColor=<rgba 0.9301 0 1 1>}
  NamedRerouteUsage()
 ]
}
