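# Additive UI material that draws an animated, two-color electric ring for the HUD.
# A 2x2 lightning-ring flipbook plays on Time, or on ManualTime when nonzero, and its red channel blends Color2 to Color1 over black.
# Rotated, scaled UVs build a spinning radial mask, which sets the anti-aliased threshold (AAThreshold) that cuts the flipbook.
# A cone-mask texture is also sampled on those rotated UVs.
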Material /Game/Athena/HUD/Materials/M_UI_ElectricRing
graph {
 properties {blendMode=BLEND_Additive, materialDomain=MD_UI, pixelDepthOffsetMode=PDOM_Legacy, usedShadingModels=Default Lit}
 nodes [
  Constant2Vector()
  n1 = NamedRerouteUsage()
  n2 = ScalarParameter() {parameterName=TimeOffset}
  n3 = Constant() {r=2}
  n4 = VectorParameter() {parameterName=Flip1Scale, defaultValue=<rgba 1 1 0 0>}
  n5 = Constant() {r=1}
  n6 = MaterialFunctionCall(
   "Animation  Phase (0-1)"=Multiply(A=Add(A=@n1, B=@n2), B=ScalarParameter() {defaultValue=2, parameterName=Speed2})
   Number of Rows=@n3
   Number of Columns=@n3
   UVs=MaterialFunctionCall!(
    Texture Scale=@n4.R
    UVs
    Offset XY Texturesize Z=MaterialFunctionCall(R=@n4.B, G=@n4.B, B=@n5) {materialFunction=&AppendMany}.RGB
   ) {materialFunction=/Engine/Functions/Engine_MaterialFunctions01/Texturing/ScaleUVsAroundPoint}.UVs
  ) {materialFunction=&FlipBook}
  n7 = TextureSample(UVs=@n6.UVs) {texture=&/T_UI_LightningRing_Flipbook}
  LinearInterpolate(
   A=Constant()
   B=LinearInterpolate(
    A=VectorParameter() {parameterName=Color2, defaultValue=<rgba 1 0.0658 0 1>}.RGB
    B=VectorParameter() {parameterName=Color1, defaultValue=<rgba 1 1 1 1>}.RGB
    Alpha=@n7.R
   )
   Alpha=SmoothStep(Value=@n7.R) {constMin=0.01}
  )
  MaterialFunctionCall(UVs=@n6.UVs, Rotation Center=@n6.UVs, "Rotation Angle (0-1)"=NamedRerouteUsage()) {materialFunction=&CustomRotator}
  n8 = ScalarParameter() {parameterName=ManualTime}
  NamedRerouteDeclaration(None=If(A=@n8, B=Constant(), "A > B"=@n8, "A == B"=Time(), "A < B"=@n8)) {name=Time, nodeColor=<rgba 0.875 0 1 1>}
  n9 = VectorParameter() {parameterName=MaskScale, defaultValue=<rgba 1 1 0.5 0.5>}
  n10 = NamedRerouteUsage()
  n11 = ScalarParameter() {parameterName=SpinnerOffset}
  n12 = MaterialFunctionCall(
   Texture Scale=@n9.R
   UVs=MaterialFunctionCall(UVs=TextureCoordinate(), "Rotation Angle (0-1)"=Add(A=@n10, B=@n11)) {materialFunction=&CustomRotator}.Rotated Values
   Offset XY Texturesize Z=MaterialFunctionCall(R=@n9.G, G=@n9.G, B=Constant() {r=1}) {materialFunction=&AppendMany}.RGB
  ) {materialFunction=&ScaleUVsAroundPoint}
  n13 = VectorParameter() {parameterName="CenterPosition (V2)", defaultValue=<rgba 0.5 0 0.5 2>}
  n14 = Saturate(None=Multiply(
    A=@n12.0-1 mask
    B=MaterialFunctionCall(UVs=@n12.UVs, CenterPosition=@n13.RGB, Radius=@n13.RGB, Density=@n13.RGB) {materialFunction=&RadialGradientExponential}.RadialGradientExponential
   ))
  Saturate(None=MaterialFunctionCall!(
    Threshold=OneMinus(None=@n14)
    Gradient=Saturate(None=@n7.R)
    Filter Width Override=ScalarParameter() {parameterName="Filter Width Override (S)"}
    Filter Width Fudge Factor=ScalarParameter() {parameterName="Filter Width Fudge Factor (S)"}
   ) {materialFunction=/Game/UI/Materials/AAThreshold}.Result)
  TextureSample(UVs=@n12.UVs) {samplerSource=SSM_Clamp_WorldGroupSettings, texture=/Game/Athena/HUD/Vehicles/Materials/T_UI_ConeMask}
 ]
}
