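# Translucent UI material for the Athena HUD elimination marker, drawing a tintable, outlined icon from the T_UI_ElimIndicator texture.
# The texture's red channel acts like a distance field, thresholded by SmoothStep using Edge and EdgeSoftness to form an icon mask that blends OutlineColor to IconColor.
# A second SmoothStep, bounded by Outline and Outline minus OutlineSoftness, forms the outline mask, which Max combines with the icon mask.
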Material /Game/Athena/HUD/Materials/M_UI_EliminationMarker
graph {
 properties {blendMode=BLEND_Translucent, materialDomain=MD_UI, pixelDepthOffsetMode=PDOM_Legacy, usedShadingModels=Default Lit}
 nodes [
  n1 = TextureSample() {texture=&/T_UI_ElimIndicator}
  n2 = ScalarParameter() {defaultValue=0.2, parameterName=Edge}
  n3 = MaterialFunctionCall(Alpha=@n1.R, Min=@n2, Max=Add(A=@n2, B=ScalarParameter() {defaultValue=0.01, parameterName=EdgeSoftness})) {materialFunction=&SmoothStep}
  LinearInterpolate(
   A=VectorParameter() {group=Outline, parameterName=OutlineColor, defaultValue=<rgba 0.0052 0.0052 0.0056 1>}.RGB
   B=VectorParameter() {parameterName=IconColor, defaultValue=<rgba 1 1 1 1>}.RGB
   Alpha=@n3.Result
  )
  n4 = ScalarParameter() {defaultValue=0.2, group=Outline, parameterName=Outline}
  n5 = ScalarParameter() {group=Outline, parameterName=OutlineSoftness}
  Max(
   A=@n3.Result
   B=MaterialFunctionCall(Alpha=@n1.R, Min=@n4, Max=Subtract(A=@n4, B=@n5)) {materialFunction=&SmoothStep}.Result
  )
 ]
}
