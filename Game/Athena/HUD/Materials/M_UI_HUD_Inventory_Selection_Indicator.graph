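# Translucent UI material that draws the HUD inventory selection indicator as a glow plus a border.
# An exponential radial gradient blends two glow colors and, scaled by AnimateGlow, sets glow intensity.
# The border mask blends via UseAngle between a rotatable box-distance frame and a hard-edged angled cut from rotated UVs.
# That mask, scaled by AnimateBorder, lerps glow to BorderColor and adds to glow for opacity.
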Material /Game/Athena/HUD/Materials/M_UI_HUD_Inventory_Selection_Indicator
graph {
 properties {blendMode=BLEND_Translucent, materialDomain=MD_UI, pixelDepthOffsetMode=PDOM_Legacy, usedShadingModels=Default Lit}
 nodes [
  n1 = VectorParameter() {parameterName=GlowColorA, defaultValue=<rgba 0 1 0 1>}
  n2 = VectorParameter() {parameterName=GlowColorB, defaultValue=<rgba 1 0 0 1>}
  n3 = TextureCoordinate() {coordinateIndex=4}
  n4 = MaterialFunctionCall(
   UVs=@n3
   Radius=ScalarParameter() {defaultValue=0.5, parameterName=GlowRadius}
   Density=ScalarParameter() {defaultValue=2, parameterName=GlowDensity}
  ) {materialFunction=&RadialGradientExponential}
  n5 = Multiply(A=@n4.RadialGradientExponential, B=ScalarParameter() {parameterName=AnimateGlow})
  n6 = MaterialFunctionCall(UVs=@n3, "Rotation Angle (0-1)"=ScalarParameter() {parameterName=BorderRotation}) {materialFunction=&CustomRotator}
  n7 = Abs(None=Divide(A=ConstantBiasScale(None=@n6.Rotated Values) {bias=-0.5, scale=2}) {constB=1})
  n8 = OneMinus(None=ScalarParameter() {defaultValue=0.125, parameterName=BorderSize})
  n9 = Saturate(None=Divide(A=Subtract(A=@n7, B=@n8) {constB=0.25}, B=Subtract(B=@n8)))
  n10 = Multiply(
   A=Saturate(None=MaterialFunctionCall(Vector 2=@n9) {materialFunction=/Engine/Functions/Engine_MaterialFunctions02/Utility/VectorLength}.V2 Length)
   B=ScalarParameter() {defaultValue=8, parameterName=BorderEdgeHardness}
  )
  n11 = MaterialFunctionCall(UVs=TextureCoordinate() {coordinateIndex=4}, "Rotation Angle (0-1)"=ScalarParameter() {defaultValue=0.0063, parameterName=AngleRotation}) {materialFunction=&CustomRotator}
  n12 = Multiply(A=ScalarParameter() {defaultValue=0.75, parameterName=Offset}) {constB=-1}
  n13 = Multiply(A=ComponentMask(None=Add(A=@n11.Rotated Values, B=@n12)) {g=true}, B=ScalarParameter() {defaultValue=32, parameterName=AngleEdgeHardness})
  n14 = Multiply(
   A=LinearInterpolate(A=Saturate(None=@n10), B=Saturate(None=@n13), Alpha=ScalarParameter() {parameterName=UseAngle})
   B=ScalarParameter() {defaultValue=1, parameterName=AnimateBorder}
  )
  LinearInterpolate(
   A=Multiply(A=LinearInterpolate(A=@n1.RGB, B=@n2.RGB, Alpha=@n4.RadialGradientExponential), B=@n5)
   B=VectorParameter() {parameterName=BorderColor, defaultValue=<rgba 0.5 0.5 0.5 1>}.RGB
   Alpha=@n14
  )
  n15 = Multiply(A=Saturate(None=@n14), B=ScalarParameter() {defaultValue=1, parameterName=FinalBorderAlpha})
  Saturate(None=Add(A=@n5, B=Saturate(None=@n15)))
 ]
}
