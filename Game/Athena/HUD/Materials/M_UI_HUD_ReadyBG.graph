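# M_UI_HUD_ReadyBG is a UI-domain material that, by its name, draws the background of a HUD ready element.
# It overlay-blends a 0-to-1 exponential radial gradient onto the BGColor parameter to form a soft highlight.
# The HighlightPos vector drives the gradient: R supplies both center coordinates, keeping it on the diagonal, and B sets both radius and density.
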Material /Game/Athena/HUD/Materials/M_UI_HUD_ReadyBG
graph {
 properties {materialDomain=MD_UI, pixelDepthOffsetMode=PDOM_Legacy, usedShadingModels=Default Lit}
 nodes [
  n1 = VectorParameter() {parameterName=HighlightPos, defaultValue=<rgba 0.5 1 1 1>}
  n2 = AppendVector(A=@n1.R, B=@n1.R)
  MaterialFunctionCall(
   Base=VectorParameter() {parameterName=BGColor, defaultValue=<rgba 0.0908 0.5333 1 1>}.RGB
   Blend=LinearInterpolate(
    B=Constant() {r=1}
    Alpha=MaterialFunctionCall(CenterPosition=@n2, Radius=@n1.B, Density=@n1.B) {materialFunction=&RadialGradientExponential}.RadialGradientExponential
   )
  ) {materialFunction=&Blend_Overlay}
 ]
}
