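# Translucent UI master material for a HUD health bar, with parameterized progress, fill and backing colors, alphas and an edge marker.
# UV math scaled by hardness and saturated builds the Progress Fill mask and an angled MarkerAlpha from rotated UVs, shared through named reroutes.
# A static switch picks a two- or three-color gradient, blended with marker color and an animatable intensity.
# SimpleProgress and HasBacking switches choose how gradient, marker and backing color combine.
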Material /Game/Athena/HUD/Materials/M_UI_HealthBar_Master
graph {
 properties {blendMode=BLEND_Translucent, materialDomain=MD_UI, pixelDepthOffsetMode=PDOM_Legacy, usedShadingModels=Default Lit}
 nodes [
  LinearInterpolate(
   A=ScalarParameter() {defaultValue=0.5, group=Bar, parameterName=Progress_Backing_Alpha}
   B=ScalarParameter() {defaultValue=1, group=Bar, parameterName=Progress_Fill_Alpha}
   Alpha=NamedRerouteUsage()
  )
  n1 = MaterialFunctionCall() {materialFunction=&GetUserInterfaceUV}
  n2 = ScalarParameter() {defaultValue=-0.0125, group=Marker, parameterName=MarkerAngle_Rotation}
  n3 = Subtract(
   A=OneMinus(None=MaterialFunctionCall(UVs=@n1.Normalized UV, "Rotation Angle (0-1)"=@n2) {materialFunction=&CustomRotator}.Rotated Values)
   B=Add(A=Constant() {r=1}, B=ScalarParameter() {defaultValue=0.0625, group=Marker, parameterName=MarkerAngle_Offset})
  )
  n4 = ScalarParameter() {defaultValue=1, group=Bar, parameterName=Progress}
  n5 = Multiply(A=ComponentMask(None=Add(A=@n3, B=@n4)) {r=true}, B=ScalarParameter() {defaultValue=80, group=Marker, parameterName=Angle_Edge_Hardness})
  NamedRerouteDeclaration(None=OneMinus(None=Saturate(None=@n5))) {name=MarkerAlpha, nodeColor=<rgba 1 0.1641 0 1>}
  n6 = TextureCoordinate()
  n7 = ScalarParameter() {parameterName=OffsetBlendFadeStart}
  n8 = LinearInterpolate(A=ScalarParameter() {parameterName=LiftBlendFadeStart}, Alpha=ComponentMask(None=Add(A=@n6, B=@n7)) {r=true})
  n9 = ScalarParameter() {group=Marker, parameterName=BlendMarkerColor}
  NamedRerouteDeclaration(None=Saturate(None=DotProduct(A=@n8, B=@n9))) {name=Marker Blend, nodeColor=<rgba 0 0.1562 1 1>}
  n10 = MaterialFunctionCall() {materialFunction=&GetUserInterfaceUV}
  NamedRerouteDeclaration(None=LinearInterpolate(
    A=VectorParameter() {group=Color, parameterName=Progress_Color_Start, defaultValue=<rgba 1 0 0 1>}.RGB
    B=VectorParameter() {group=Color, parameterName=Progress_Color_Stop, defaultValue=<rgba 0 0 1 1>}.RGB
    Alpha=ComponentMask(None=@n10.Normalized UV) {r=true}
   )) {name=2ColorGradient, nodeColor=<rgba 0.4766 1 0 1>}
  n11 = ComponentMask(None=MaterialFunctionCall() {materialFunction=&GetUserInterfaceUV}.Normalized UV) {r=true}
  n12 = ScalarParameter() {defaultValue=1, parameterName=3_Color_Gradient_Multiplier}
  n13 = Power(Base=Saturate(None=Multiply(A=@n11, B=@n12)), Exp=ScalarParameter() {defaultValue=1, parameterName=3_Color_Gradient_ChangeRate})
  NamedRerouteDeclaration(None=MaterialFunctionCall(
    A=VectorParameter() {group=Color, parameterName=Progress_Color_Start, defaultValue=<rgba 1 0 0 1>}.RGB
    B=VectorParameter() {group=Color, parameterName=Progress_Color_Middle, defaultValue=<rgba 0 1 0 1>}.RGB
    C=VectorParameter() {group=Color, parameterName=Progress_Color_Stop, defaultValue=<rgba 0 0 1 1>}.RGB
    Alpha=Saturate(None=@n13)
   ) {materialFunction=&3ColorBlend}.Result) {name=3ColorGradient, nodeColor=<rgba 1 0 0.7969 1>}
  n14 = OneMinus(None=MaterialFunctionCall() {materialFunction=&GetUserInterfaceUV}.Normalized UV)
  n15 = ComponentMask(None=Add(A=Subtract(A=@n14), B=ScalarParameter() {defaultValue=1, group=Bar, parameterName=Progress})) {r=true}
  n16 = ScalarParameter() {defaultValue=128, group=Bar, parameterName=Progress_Edge_Hardness}
  NamedRerouteDeclaration(None=Saturate(None=Multiply(A=@n15, B=@n16))) {name=Progress Fill, nodeColor=<rgba 1 0.5625 0 1>}
  n17 = LinearInterpolate(
   A=VectorParameter() {group=Color, parameterName=Backing_Color, defaultValue=<rgba 0 0 0 1>}.RGB
   B=VectorParameter() {group=Color, parameterName=Backing_Color_Stop, defaultValue=<rgba 0 0 0 1>}.RGB
   Alpha=ScalarParameter() {parameterName=AnimateBackingColor}
  )
  n18 = StaticSwitchParameter(True=NamedRerouteUsage(), False=NamedRerouteUsage()) {parameterName=3_Color_Gradient}
  n19 = NamedRerouteUsage()
  n20 = VectorParameter() {group=Color, parameterName=Marker_Color, defaultValue=<rgba 1 1 0 1>}
  n21 = NamedRerouteUsage()
  n22 = LinearInterpolate(
   A=@n18
   B=LinearInterpolate(A=LinearInterpolate(A=@n18, B=@n20.RGB, Alpha=@n21), B=@n20.RGB, Alpha=ScalarParameter() {parameterName=AnimateMarkerIntensity})
   Alpha=NamedRerouteUsage()
  )
  StaticSwitchParameter(
   True=LinearInterpolate(A=@n17, B=@n18, Alpha=NamedRerouteUsage())
   False=StaticSwitchParameter(True=LinearInterpolate(A=@n17, B=@n22, Alpha=@n19), False=@n22) {parameterName=HasBacking}
  ) {parameterName=SimpleProgress}
 ]
}
