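# Translucent UI master material for a segmented health bar, with fill driven by a Progress parameter.
# Rotated UVs feed an absolute sine and SmoothStep to cut angled segments, multiplied by separate backing and fill alphas.
# A hardened progress mask blends backing color into a two- or three-color gradient, with an animatable marker highlight offset from the fill edge.
# Static switches select simple progress, backing and gradient variants.
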Material /Game/Athena/HUD/Materials/M_UI_HealthBar_Segments_Master
graph {
 properties {blendMode=BLEND_Translucent, materialDomain=MD_UI, pixelDepthOffsetMode=PDOM_Legacy, usedShadingModels=Default Lit}
 nodes [
  MaterialFunctionCall() {materialFunction=&MF_UI_LinearGradient}
  n1 = ScalarParameter() {defaultValue=0.25, parameterName=SegmentAngle}
  n2 = ScalarParameter() {group=Bar, parameterName=FillAngle}
  n3 = Add(
   A=MaterialFunctionCall(UVs=TextureCoordinate(), "Rotation Angle (0-1)"=Add(A=@n1, B=@n2)) {materialFunction=&CustomRotator}.Rotated Values
   B=AppendVector(A=Constant(), B=ScalarParameter() {defaultValue=0.004, parameterName=SegmentsOffset})
  ) {constB=0}
  n4 = ScalarParameter() {defaultValue=4, parameterName=Segments}
  n5 = AppendVector(A=Constant() {desc=stretch vertical to eliminate artifacts, r=0.01}, B=@n4)
  n6 = Abs(None=Sine(None=Multiply(A=@n3, B=@n5)) {period=2})
  n7 = Multiply(A=@n4, B=ScalarParameter() {defaultValue=0.01, parameterName=SegmentEdge})
  n8 = Multiply(A=@n4, B=ScalarParameter() {defaultValue=0, parameterName=SegmentEdgeSoftness})
  n9 = MaterialFunctionCall() {materialFunction=&GetUserInterfaceUV}
  n10 = ScalarParameter() {group=Bar, parameterName=FillAngle}
  n11 = MaterialFunctionCall(UVs=@n9.Normalized UV, "Rotation Angle (0-1)"=@n10) {materialFunction=&CustomRotator}
  n12 = Add(A=Subtract(A=OneMinus(None=@n11.Rotated Values)), B=ScalarParameter() {defaultValue=0.5, group=Bar, parameterName=Progress})
  n13 = Saturate(None=Multiply(A=ComponentMask(None=@n12) {r=true}, B=ScalarParameter() {defaultValue=128, group=Bar, parameterName=Progress_Edge_Hardness}))
  Multiply(
   A=MaterialFunctionCall(Alpha=ComponentMask(None=@n6) {g=true}, Min=@n7, Max=Add(A=@n7, B=@n8)) {materialFunction=&SmoothStep}.Result
   B=LinearInterpolate(
    A=ScalarParameter() {defaultValue=0.25, group=Bar, parameterName=Progress_Backing_Alpha}
    B=ScalarParameter() {defaultValue=1, group=Bar, parameterName=Progress_Fill_Alpha}
    Alpha=@n13
   )
  )
  n14 = ComponentMask(None=Add(A=TextureCoordinate(), B=ScalarParameter() {parameterName=OffsetBlendFadeStart})) {r=true}
  OneMinus(None=@n14)
  n15 = LinearInterpolate(
   A=VectorParameter() {group=Color, parameterName=Backing_Color, defaultValue=<rgba 0 0 0 1>}.RGB
   B=VectorParameter() {group=Color, parameterName=Backing_Color_Stop, defaultValue=<rgba 0 0 0 1>}.RGB
   Alpha=ScalarParameter() {parameterName=AnimateBackingColor}
  )
  n16 = VectorParameter() {group=Color, parameterName=Progress_Color_Start, defaultValue=<rgba 1 0 0 1>}
  n17 = VectorParameter() {group=Color, parameterName=Progress_Color_Stop, defaultValue=<rgba 0 0 1 1>}
  n18 = ComponentMask(None=@n11.Rotated Values) {r=true}
  n19 = ScalarParameter() {defaultValue=1, parameterName=3_Color_Gradient_Multiplier}
  n20 = Power(Base=Saturate(None=Multiply(A=@n18, B=@n19)), Exp=ScalarParameter() {defaultValue=1, parameterName=3_Color_Gradient_ChangeRate})
  n21 = StaticSwitchParameter(
   True=MaterialFunctionCall(
    A=@n16.RGB
    B=VectorParameter() {group=Color, parameterName=Progress_Color_Middle, defaultValue=<rgba 0 1 0 1>}.RGB
    C=@n17.RGB
    Alpha=Saturate(None=@n20)
   ) {materialFunction=&3ColorBlend}.Result
   False=LinearInterpolate(A=@n16.RGB, B=@n17.RGB, Alpha=@n18)
  ) {parameterName=3_Color_Gradient}
  n22 = VectorParameter() {group=Color, parameterName=Marker_Color, defaultValue=<rgba 1 1 0 1>}
  n23 = ScalarParameter() {parameterName=LiftBlendFadeStart}
  n24 = Saturate(None=DotProduct(A=LinearInterpolate(A=@n23, Alpha=@n14), B=ScalarParameter() {group=Marker, parameterName=BlendMarkerColor}))
  n25 = Add(A=@n10, B=ScalarParameter() {defaultValue=-0.0125, group=Marker, parameterName=MarkerAngle_Rotation})
  n26 = Subtract(
   A=OneMinus(None=MaterialFunctionCall(UVs=@n9.Normalized UV, "Rotation Angle (0-1)"=@n25) {materialFunction=&CustomRotator}.Rotated Values)
   B=Add(A=Constant() {r=1}, B=ScalarParameter() {defaultValue=0.0625, group=Marker, parameterName=MarkerAngle_Offset})
  )
  n27 = ScalarParameter() {defaultValue=0.5, group=Bar, parameterName=Progress}
  n28 = Multiply(A=ComponentMask(None=Add(A=@n26, B=@n27)) {r=true}, B=ScalarParameter() {defaultValue=80, group=Marker, parameterName=Angle_Edge_Hardness})
  n29 = LinearInterpolate(
   A=@n21
   B=LinearInterpolate(A=LinearInterpolate(A=@n21, B=@n22.RGB, Alpha=@n24), B=@n22.RGB, Alpha=ScalarParameter() {parameterName=AnimateMarkerIntensity})
   Alpha=OneMinus(None=Saturate(None=@n28))
  )
  StaticSwitchParameter(
   True=LinearInterpolate(A=@n15, B=@n21, Alpha=@n13)
   False=StaticSwitchParameter(True=LinearInterpolate(A=@n15, B=@n29, Alpha=@n13), False=@n29) {parameterName=HasBacking}
  ) {parameterName=SimpleProgress}
 ]
}
