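# Translucent UI material that builds an antialiasable hexagon socket shape, either filled or outlined, from a distance field texture for the HUD.
# The hex texture is sampled with center-scaled UVs, and its red channel is thresholded by SmoothStep using edge and softness parameters.
# An Outline static switch multiplies that mask by the inverse of a second threshold, offset by OutlineSize, leaving only a band.
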Material /Game/Athena/HUD/Materials/M_UI_HexSocket
graph {
 properties {blendMode=BLEND_Translucent, materialDomain=MD_UI, pixelDepthOffsetMode=PDOM_Legacy, usedShadingModels=Default Lit}
 nodes [
  Constant() {r=1}
  n1 = TextureSample(UVs=MaterialFunctionCall(Texture Scale=ScalarParameter() {defaultValue=1, parameterName=Scale}) {materialFunction=&ScaleUVsByCenter}.UVs) {texture=/Game/Athena/HUD/LTM/Domination/T_UI_Hex_DF}
  n2 = ScalarParameter() {parameterName=Edge}
  n3 = MaterialFunctionCall(Alpha=@n1.R, Min=@n2, Max=Add(A=@n2, B=ScalarParameter() {parameterName=EdgeSoftness})) {materialFunction=&SmoothStep}
  n4 = Add(A=@n2, B=ScalarParameter() {parameterName=OutlineSize})
  n5 = MaterialFunctionCall(Alpha=@n1.R, Min=@n4, Max=Add(A=@n4, B=ScalarParameter() {parameterName=EdgeSoftness})) {materialFunction=&SmoothStep}
  StaticSwitchParameter(True=Multiply(A=@n3.Result, B=OneMinus(None=@n5.Result)), False=@n3.Result) {parameterName=Outline}
 ]
}
