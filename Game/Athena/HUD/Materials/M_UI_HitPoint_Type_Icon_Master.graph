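# Translucent UI master material for hit point type icons, drawn from one selectable channel of T_UI_HitPoint_Type_Icon.
# Subtract, multiply and saturate steps turn that channel into sharpened fill and wider outline masks, suggesting a distance-field style texture.
# UseOutline blends between plain icon and outlined versions of color and alpha, with tunable edges, colors and opacity.
# Rotation and center-based scale parameters transform the UVs, apparently through the customized UV slot.
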Material /Game/Athena/HUD/Materials/M_UI_HitPoint_Type_Icon_Master
graph {
 properties {blendMode=BLEND_Translucent, materialDomain=MD_UI, numCustomizedUVs=1, pixelDepthOffsetMode=PDOM_Legacy, usedShadingModels=Default Lit}
 nodes [
  n1 = StaticComponentMaskParameter(None=TextureSample(UVs=TextureCoordinate()) {samplerType=SAMPLERTYPE_LinearColor, texture=&/T_UI_HitPoint_Type_Icon}.RGB) {defaultG=true, parameterName=Channel}
  n2 = ScalarParameter() {defaultValue=0.25, parameterName=Subtract_Edge}
  n3 = Saturate(None=Multiply(A=Subtract(A=@n1, B=@n2), B=ScalarParameter() {defaultValue=8, parameterName=Multiply_Edge}))
  n4 = ScalarParameter() {defaultValue=0.125, parameterName=Outline_Subtract_Edge}
  n5 = Saturate(None=Multiply(A=Subtract(A=@n1, B=@n4), B=ScalarParameter() {defaultValue=8, parameterName=Outline_Multiply_Edge}))
  n6 = ScalarParameter() {defaultValue=0.5, parameterName=Outline_Alpha}
  LinearInterpolate(
   A=@n3
   B=LinearInterpolate(A=Multiply(A=@n5, B=@n6), B=ScalarParameter() {defaultValue=1, parameterName=Alpha}, Alpha=@n3)
   Alpha=ScalarParameter() {parameterName=UseOutline}
  )
  n7 = VectorParameter() {parameterName=Icon_Color, defaultValue=<rgba 1 0 0 1>}
  LinearInterpolate(
   A=@n7.RGB
   B=LinearInterpolate(A=VectorParameter() {parameterName=Outline_Color, defaultValue=<rgba 0 1 0 1>}.RGB, B=@n7.RGB, Alpha=@n3)
   Alpha=ScalarParameter() {parameterName=UseOutline}
  )
  MaterialFunctionCall(
   UVs=MaterialFunctionCall(UVs=TextureCoordinate(), "Rotation Angle (0-1)"=ScalarParameter() {parameterName=Rotation}) {materialFunction=&CustomRotator}.Rotated Values
   Texture Scale=ScalarParameter() {defaultValue=1, parameterName=Scale}
  ) {materialFunction=&ScaleUVsByCenter}
 ]
}
