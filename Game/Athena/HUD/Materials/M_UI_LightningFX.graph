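# Translucent UI material that draws animated, tinted lightning over a colored radial glow.
# A FlipBook steps through a lightning sheet, with thresholded channels shaping the colored bolt and bright highlights, blended with the exponential gradient glow.
# Alpha blends that glow with the green channel, masked by a second radial gradient driven by FadeAmount, then scaled by Final_Alpha.
# A Rotator driven by an Angle parameter rotates texture coordinates.
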Material /Game/Athena/HUD/Materials/M_UI_LightningFX
graph {
 properties {blendMode=BLEND_Translucent, materialDomain=MD_UI, numCustomizedUVs=1, pixelDepthOffsetMode=PDOM_Legacy, usedShadingModels=Default Lit}
 nodes [
  n1 = VectorParameter() {parameterName=Lightning Color, defaultValue=<rgba 0.4535 0 0.5521 0>}
  n2 = ScalarParameter() {defaultValue=0.5, parameterName="Gradient  Radius Fade"}
  n3 = ScalarParameter() {defaultValue=2, parameterName=Gradient Density Fade}
  n4 = LinearInterpolate(
   A=OneMinus(None=ScalarParameter() {parameterName=FadeAmount})
   Alpha=OneMinus(None=MaterialFunctionCall(Radius=@n2, Density=@n3) {materialFunction=&RadialGradientExponential}.RadialGradientExponential)
  )
  n5 = TextureSample(UVs=MaterialFunctionCall(
    Number of Rows=ScalarParameter() {defaultValue=1, parameterName=Rows}
    Number of Columns=ScalarParameter() {defaultValue=10, parameterName=Column}
   ) {materialFunction=&FlipBook}.UVs) {texture=&/T-UI-LightningSheet}
  n6 = Multiply(A=Subtract(A=@n5.RGB, B=ScalarParameter() {defaultValue=0.2, parameterName=CenterCore})) {constB=5}
  n7 = Saturate(None=DotProduct(A=@n4, B=Clamp(None=@n6)))
  n8 = Multiply(A=Subtract(A=@n5.R, B=ScalarParameter() {defaultValue=0.75, parameterName=TaperDissipation})) {constB=16}
  n9 = ScalarParameter() {defaultValue=0.5, parameterName=Gradient Radius}
  n10 = ScalarParameter() {defaultValue=2, parameterName=Gradient Density}
  n11 = Multiply(A=LinearInterpolate(
    B=VectorParameter() {parameterName=Gradient Color, defaultValue=<rgba 0.4886 0.05 0.625 0>}.RGB
    Alpha=MaterialFunctionCall(Radius=@n9, Density=@n10) {materialFunction=&RadialGradientExponential}.RadialGradientExponential
   )) {constB=2}
  LinearInterpolate(A=@n11, B=Add(A=LinearInterpolate(A=@n1.RGB, B=@n1.RGB, Alpha=@n7), B=Clamp(None=@n8)))
  n12 = Multiply(A=@n5.G) {constB=1.5}
  n13 = DotProduct(A=Saturate(None=@n4), B=Saturate(None=Multiply(A=@n12) {constB=1.5}))
  Multiply(A=Saturate(None=LinearInterpolate(A=@n11, B=@n13)), B=ScalarParameter() {defaultValue=1, parameterName=Final_Alpha})
  Rotator(Coordinate=TextureCoordinate(), Time=ScalarParameter() {defaultValue=1, parameterName=Angle, sliderMax=1}) {speed=1.5}
 ]
}
