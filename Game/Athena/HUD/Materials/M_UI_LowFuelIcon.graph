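# Translucent UI material for a low fuel icon that fills vertically by PercentFill and emits a repeating pulse.
# Smoothstep thresholds on the texture's red channel cut the icon shape, and a soft vertical edge blends unfilled and filled color and alpha.
# A time-looped, rescaled second sample forms an outline with a partly hollowed interior that fades out in PulseColor.
# The Pulse static switch gates this layer, which merges with icon alpha by Max.
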Material /Game/Athena/HUD/Materials/M_UI_LowFuelIcon
graph {
 properties {blendMode=BLEND_Translucent, materialDomain=MD_UI, pixelDepthOffsetMode=PDOM_Legacy, usedShadingModels=Default Lit}
 nodes [
  n1 = VectorParameter() {parameterName=UnfilledColor, defaultValue=<rgba 0.0052 0.0052 0.0056 0.6>}
  n2 = TextureCoordinate()
  n3 = ScalarParameter() {defaultValue=1, group=Fill, parameterName=PercentFill, sliderMax=1}
  n4 = MaterialFunctionCall(
   Alpha=OneMinus(None=ComponentMask(None=@n2) {g=true})
   Min=@n3
   Max=Subtract(A=@n3, B=ScalarParameter() {defaultValue=0.005, group=Fill, parameterName=PercentEdgeSoftness})
  ) {materialFunction=&SmoothStep}
  n5 = ScalarParameter() {defaultValue=0.75, parameterName=IconScale}
  n6 = ScalarParameter() {defaultValue=0.15, parameterName=Edge}
  n7 = MaterialFunctionCall(
   Alpha=TextureSample(UVs=MaterialFunctionCall(Texture Scale=@n5) {materialFunction=&ScaleUVsByCenter}.UVs) {samplerSource=SSM_Clamp_WorldGroupSettings, texture=&/T_UI_LowFuelIcon}.R
   Min=@n6
   Max=Add(A=@n6, B=ScalarParameter() {defaultValue=0.05, parameterName=EdgeSoftness})
  ) {materialFunction=&SmoothStep}
  LinearInterpolate(
   A=VectorParameter() {parameterName=PulseColor, defaultValue=<rgba 0.5833 0.0189 0 1>}.RGB
   B=LinearInterpolate(A=@n1.RGB, B=VectorParameter() {parameterName=Color, defaultValue=<rgba 1 0.0961 0 1>}.RGB, Alpha=@n4.Result)
   Alpha=@n7.Result
  )
  n8 = ScalarParameter() {defaultValue=1, group=Pulse, parameterName=PulseFinalSize}
  n9 = Fmod(A=Multiply(A=Time(), B=ScalarParameter() {defaultValue=0.7, group=Pulse, parameterName=PulseSpeed}), B=Constant() {r=1})
  n10 = TextureSample(UVs=MaterialFunctionCall(Texture Scale=LinearInterpolate(A=@n5, B=@n8, Alpha=@n9) {constB=3}) {materialFunction=&ScaleUVsByCenter}.UVs) {samplerSource=SSM_Clamp_WorldGroupSettings, texture=&T_UI_LowFuelIcon}
  n11 = ScalarParameter() {defaultValue=0.15, parameterName=Edge}
  n12 = ScalarParameter() {defaultValue=0.05, parameterName=EdgeSoftness}
  n13 = Add(A=@n11, B=@n12)
  n14 = Add(
   A=@n11
   B=LinearInterpolate(
    A=ScalarParameter() {defaultValue=0.1, group=Pulse, parameterName=OutlineSize}
    B=ScalarParameter() {group=Pulse, parameterName=OutlineSizeEnd}
    Alpha=@n9
   )
  )
  n15 = MaterialFunctionCall(Alpha=@n10.R, Min=@n14, Max=Add(A=@n14, B=@n12)) {materialFunction=&SmoothStep}
  n16 = ScalarParameter() {defaultValue=0.75, group=Pulse, parameterName=InnerPulseFill}
  n17 = Clamp(None=Subtract(
    A=MaterialFunctionCall(Alpha=@n10.R, Min=@n11, Max=@n13) {materialFunction=&SmoothStep}.Result
    B=Multiply(A=@n15.Result, B=@n16)
   ) {constB=0})
  n18 = Multiply(A=OneMinus(None=@n9), B=ScalarParameter() {defaultValue=0.75, parameterName=PulseAmount})
  Max(
   A=LinearInterpolate(A=Multiply(A=@n7.Result, B=@n1.A), B=@n7.Result, Alpha=@n4.Result)
   B=StaticSwitchParameter(True=Multiply(A=@n17, B=@n18), False=Constant()) {defaultValue=true, group=Pulse, parameterName=Pulse}
  )
 ]
}
