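# Translucent UI material, named as a message background, that builds a slanted band mask and an optional gradient.
# Two SmoothSteps on rotated UV create soft left and right edges, merged through Max and OneMinus.
# A switchable curve-atlas lookup, borrowing an ItemShop curve, samples rotated normalized UI UVs for the gradient.
# The gradient pivot uses PivotX for both axes, leaving PivotY unused; no material output connections appear.
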Material /Game/Athena/HUD/Materials/M_UI_MessageBG
graph {
 properties {blendMode=BLEND_Translucent, materialDomain=MD_UI, pixelDepthOffsetMode=PDOM_Legacy, usedShadingModels=Default Lit}
 nodes [
  n1 = ComponentMask(None=MaterialFunctionCall(UVs=TextureCoordinate(), "Rotation Angle (0-1)"=ScalarParameter() {defaultValue=-0.026, parameterName=EdgeRotation}) {materialFunction=&CustomRotator}.Rotated Values) {r=true}
  n2 = ScalarParameter() {defaultValue=0.9, parameterName=LeftEdge}
  n3 = Add(A=@n2, B=ScalarParameter() {defaultValue=0.01, parameterName=EdgeSoftness})
  n4 = ScalarParameter() {defaultValue=0.1, parameterName=RightEdge}
  n5 = MaterialFunctionCall(Alpha=@n1, Min=@n4, Max=Add(A=@n4, B=ScalarParameter() {defaultValue=0.01, parameterName=EdgeSoftness})) {materialFunction=&SmoothStep}
  OneMinus(None=Max(
    A=MaterialFunctionCall(Alpha=@n1, Min=@n2, Max=@n3) {materialFunction=&SmoothStep}.Result
    B=OneMinus(None=@n5.Result)
   ))
  n6 = MaterialFunctionCall() {materialFunction=&GetUserInterfaceUV}
  n7 = VectorParameter() {parameterName=GradientRotation, channelNames {a="", b=PivotY, g=PivotX, r=Rotation}, defaultValue=<rgba 0 0.5 0.5 1>}
  n8 = AppendVector(A=@n7.PivotX, B=@n7.PivotX)
  n9 = CurveAtlasRowParameter(CurveTime=ComponentMask(None=MaterialFunctionCall(UVs=@n6.Normalized UV, Rotation Center=@n8, "Rotation Angle (0-1)"=@n7.Rotation) {materialFunction=&CustomRotator}.Rotated Values) {r=true}) {
   atlas=/Game/UI/Frontend/ItemShop/Art/ItemShopColorAtlas
   curve=/Game/UI/Frontend/ItemShop/Art/ItemShop_RiftTourFlare_Curve
   defaultValue=1
   group=Gradient
   parameterName=MaskCurve
  }
  StaticSwitchParameter(True=@n9, False=Constant() {r=1}) {parameterName=UseGradientCurve}
  TextureCoordinate()
 ]
}
