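# Translucent UI material that draws a layered message background with slanted side edges, by its name for the HUD.
# Color1 blends via UseGradient into a three-color horizontal gradient, split at GradientMidpoint by If masks.
# Color2 and Color3 layer over it using soft SmoothStep bands on offset UV1 V.
# Opacity blends the three layer alphas, ignoring gradient alphas, then trims both sides with SmoothSteps on rotated UV0 U.
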Material /Game/Athena/HUD/Materials/M_UI_MessageBG_Layered
graph {
 properties {blendMode=BLEND_Translucent, materialDomain=MD_UI, pixelDepthOffsetMode=PDOM_Legacy, usedShadingModels=Default Lit}
 nodes [
  n1 = VectorParameter() {parameterName=Color1, defaultValue=<rgba 0.1651 0.0802 0.2582 0.6>}
  n2 = VectorParameter() {parameterName=GraientColor2, defaultValue=<rgba 0 0.008 0.5586 0.6>}
  n3 = VectorParameter() {parameterName=GraientColor1, defaultValue=<rgba 0.1875 0.008 0.453 0.6>}
  n4 = ComponentMask(None=TextureCoordinate()) {r=true}
  n5 = ScalarParameter() {defaultValue=0.5, parameterName=GradientMidpoint}
  n6 = Divide(A=Subtract(A=@n4, B=@n5), B=OneMinus(None=@n5))
  n7 = ScalarParameter() {parameterName=0}
  n8 = ScalarParameter() {defaultValue=1, parameterName=1}
  n9 = VectorParameter() {parameterName=GraientColor3, defaultValue=<rgba 0.0508 0.316 0.957 0.6>}
  n10 = Divide(A=@n4, B=@n5)
  n11 = ScalarParameter() {parameterName=0}
  n12 = ScalarParameter() {defaultValue=1, parameterName=1}
  n13 = Add(
   A=Multiply(A=LinearInterpolate(A=@n2.RGB, B=@n3.RGB, Alpha=@n6), B=If(A=@n5, B=@n4, "A > B"=@n7, "A == B"=@n7, "A < B"=@n8))
   B=Multiply(A=LinearInterpolate(A=@n9.RGB, B=@n2.RGB, Alpha=@n10), B=If(A=@n4, B=@n5, "A > B"=@n11, "A == B"=@n11, "A < B"=@n12))
  )
  n14 = ScalarParameter() {defaultValue=1, parameterName=UseGradient}
  n15 = VectorParameter() {parameterName=Color2, defaultValue=<rgba 0 0.15 1 1>}
  n16 = TextureCoordinate() {coordinateIndex=1}
  n17 = Add(A=@n16, B=Multiply(A=ScalarParameter() {defaultValue=0.63, group=Offset, parameterName=Top Off Set2}) {constB=-1})
  n18 = ScalarParameter() {defaultValue=0.1, desc=0.02, group=Fill, parameterName=FillEdge}
  n19 = Add(A=@n18, B=Multiply(A=@n18, B=ScalarParameter() {defaultValue=0.01, desc=0.1, parameterName=EdgeSoftness}) {constB=1.01})
  n20 = Add(A=MaterialFunctionCall(Alpha=ComponentMask(None=@n17) {g=true}, Min=@n18, Max=@n19) {materialFunction=&SmoothStep}.Result) {constB=0}
  n21 = VectorParameter() {parameterName=Color3, defaultValue=<rgba 0 0.4287 1 1>}
  n22 = Add(A=@n16, B=Multiply(A=ScalarParameter() {defaultValue=0.7, group=Offset, parameterName=Top Off Set3}) {constB=-1})
  n23 = Add(A=MaterialFunctionCall(Alpha=ComponentMask(None=@n22) {g=true}, Min=@n18, Max=@n19) {materialFunction=&SmoothStep}.Result) {constB=0}
  LinearInterpolate(A=LinearInterpolate(A=LinearInterpolate(A=@n1.RGB, B=@n13, Alpha=@n14), B=@n15.RGB, Alpha=@n20), B=@n21.RGB, Alpha=@n23)
  n24 = ComponentMask(None=MaterialFunctionCall(UVs=TextureCoordinate(), "Rotation Angle (0-1)"=ScalarParameter() {defaultValue=-0.026, parameterName=EdgeRotation}) {materialFunction=&CustomRotator}.Rotated Values) {r=true}
  n25 = ScalarParameter() {defaultValue=0.9, parameterName=LeftEdge}
  n26 = MaterialFunctionCall(Alpha=@n24, Min=@n25, Max=Add(A=@n25, B=ScalarParameter() {defaultValue=0.01, parameterName=EdgeSoftness})) {materialFunction=&SmoothStep}
  n27 = ScalarParameter() {defaultValue=0.1, parameterName=RightEdge}
  n28 = ScalarParameter() {defaultValue=0.01, parameterName=EdgeSoftness}
  n29 = OneMinus(None=MaterialFunctionCall(Alpha=@n24, Min=@n27, Max=Add(A=@n27, B=@n28)) {materialFunction=&SmoothStep}.Result)
  Multiply(A=LinearInterpolate(A=LinearInterpolate(A=@n1.A, B=@n15.A, Alpha=@n20), B=@n21.A, Alpha=@n23), B=OneMinus(None=Max(A=@n26.Result, B=@n29)))
 ]
}
