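# Translucent UI material for a HUD ping bubble: a ringed, filled circle with a configurable icon and offset shadow.
# Sphere masks build the ring and fill, with a static switch for radial-gradient fill; the tintable icon composites on top, optionally cropped to the circle.
# Scalar blends switch to a solid disc or squad indicator texture.
# Customized UVs carry bubble and icon scale and rotation, plus icon offset.
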Material /Game/Athena/HUD/Materials/M_UI_PingBubble
graph {
 properties {blendMode=BLEND_Translucent, materialDomain=MD_UI, numCustomizedUVs=3, pixelDepthOffsetMode=PDOM_Legacy, usedShadingModels=Default Lit}
 nodes [
  n1 = VectorParameter() {parameterName=ShadowColor, defaultValue=<rgba 0.0052 0.0052 0.0056 0.5>}
  n2 = TextureCoordinate() {coordinateIndex=2}
  n3 = TextureSampleParameter2D(UVs=@n2) {
   parameterName=DisplayedIcon
   samplerSource=SSM_Clamp_WorldGroupSettings
   texture=/Game/UI/Foundation/Textures/Icons/Replay/T-Icon-Replay-Exclamation-L
  }
  n4 = VectorParameter() {parameterName=IconShadowOffset, defaultValue=<rgba -0.1 -0.05 0 1>}
  n5 = TextureSampleParameter2D(UVs=Add(A=@n2, B=AppendVector(A=@n4.R, B=@n4.R))) {
   parameterName=DisplayedIcon
   samplerSource=SSM_Clamp_WorldGroupSettings
   texture=&T-Icon-Replay-Exclamation-L
  }
  n6 = Multiply(A=Abs(None=Subtract(A=@n2) {constB=0.5})) {constB=2}
  n7 = Clamp(None=Multiply(A=Subtract(A=@n6) {constB=0.99}) {constB=100})
  n8 = Add(A=ComponentMask(None=@n7) {r=true}, B=ComponentMask(None=@n7) {g=true})
  n9 = Multiply(A=Max(A=Multiply(A=@n5.A, B=@n1.A), B=@n3.A), B=OneMinus(None=Clamp(None=@n8)))
  n10 = Multiply(A=LinearInterpolate(A=@n1.RGB, B=@n3.RGB, Alpha=@n3.RGB), B=@n9)
  n11 = VectorParameter() {parameterName=IconColor, defaultValue=<rgba 1 1 1 1>}
  n12 = ScalarParameter() {parameterName=IconColorStrength}
  Add(A=@n10, B=Multiply(A=Multiply(A=@n11.RGB, B=@n9), B=@n12))
  n13 = ScalarParameter() {defaultValue=1, parameterName=IconColorIntensity}
  n14 = LinearInterpolate(
   A=Multiply(A=@n11.RGB, B=@n9)
   B=LinearInterpolate(A=Multiply(A=@n13, B=@n10), B=@n10, Alpha=@n12)
   Alpha=ScalarParameter() {parameterName=UseIconColor}
  )
  n15 = ScalarParameter() {parameterName=SolidFull}
  n16 = Multiply(A=@n9, B=LinearInterpolate(A=@n11.A, Alpha=@n15) {constB=0})
  n17 = VectorParameter() {parameterName=FillColorInner, defaultValue=<rgba 0 0.7181 1 1>}
  n18 = VectorParameter() {parameterName=RingColor, defaultValue=<rgba 1 1 1 1>}
  n19 = StaticSwitchParameter(True=@n17.RGB, False=@n18.RGB) {parameterName=UseRadialGradFill}
  n20 = TextureCoordinate() {coordinateIndex=1}
  n21 = SphereMask(
   A=@n20
   B=Constant2Vector() {g=0.5, r=0.5}
   Radius=ScalarParameter() {defaultValue=0.45, parameterName=InnerRingRadius}
   Hardness=ScalarParameter() {defaultValue=0.96, parameterName=C1Hardness}
  )
  n22 = Multiply(
   A=SphereMask(
    A=@n20
    B=Constant2Vector() {g=0.5, r=0.5}
    Radius=ScalarParameter() {defaultValue=0.5, parameterName=OuterRingRadius}
    Hardness=ScalarParameter() {defaultValue=0.96, parameterName=C1Hardness}
   )
   B=OneMinus(None=@n21)
  )
  n23 = VectorParameter() {parameterName=FillColor, defaultValue=<rgba 0.0211 0.5197 0.7465 1>}
  n24 = VectorParameter() {group=RadialGradient, parameterName=RadialGradCenter, defaultValue=<rgba 0.5 0.5 0 1>}
  n25 = ScalarParameter() {defaultValue=0.4, group=RadialGradient, parameterName=RadGradRadius}
  n26 = ScalarParameter() {defaultValue=1.5, group=RadialGradient, parameterName=RadialGradDensity}
  n27 = Multiply(
   A=@n21
   B=LinearInterpolate(
    A=@n23.RGB
    B=@n17.RGB
    Alpha=MaterialFunctionCall(UVs=@n20, CenterPosition=@n24.RGB, Radius=@n25, Density=@n26) {materialFunction=&RadialGradientExponential}.RadialGradientExponential
   )
  )
  n28 = Constant() {r=0.08}
  n29 = Multiply(A=@n17.R, B=@n28)
  n30 = Multiply(A=@n17.G, B=@n28)
  n31 = Multiply(A=@n17.B, B=@n28)
  n32 = Multiply(
   A=@n21
   B=StaticSwitchParameter(
    True=MaterialFunctionCall(X=@n29, Y=@n30, Z=@n31) {materialFunction=&MakeFloat3}.Result
    False=@n23.RGB
   ) {parameterName=UseRadialGradFill}
  )
  n33 = Multiply(A=OneMinus(None=@n16), B=Add(A=Multiply(A=@n19, B=@n22), B=StaticSwitchParameter(True=@n27, False=@n32) {parameterName=UseRadialGradFill}))
  n34 = TextureSample(UVs=TextureCoordinate()) {texture=/Game/Athena/HUD/Art/squadArt/T_UI_SquadGenericIndicator_b}
  n35 = ScalarParameter() {parameterName=IndicatorShape}
  LinearInterpolate(A=LinearInterpolate(A=Add(A=@n14, B=@n33), B=@n19, Alpha=@n15), B=Multiply(A=@n34.RGB, B=@n19), Alpha=@n35)
  n36 = Constant() {r=0.9}
  n37 = Add(A=Multiply(A=@n18.A, B=@n22), B=Multiply(A=@n21, B=StaticSwitchParameter(True=@n36, False=@n23.A) {parameterName=UseRadialGradFill}))
  n38 = Add(A=@n37, B=@n16)
  n39 = Saturate(None=@n37)
  LinearInterpolate(A=LinearInterpolate(A=@n38, B=Multiply(A=@n38, B=@n39), Alpha=ScalarParameter() {parameterName=bCropIconToCircle}), B=@n34.A, Alpha=@n35)
  n40 = VectorParameter() {group=BubbleShape, parameterName=BubbleScale, defaultValue=<rgba 0.8183 0.9329 0 1>}
  MaterialFunctionCall(
   UVs=MaterialFunctionCall(UVs=TextureCoordinate(), "Rotation Angle (0-1)"=ScalarParameter() {defaultValue=0.6373, group=BubbleShape, parameterName=BubbleRot}) {materialFunction=&CustomRotator}.Rotated Values
   Texture Scale=MaterialFunctionCall(X=@n40.R, Y=@n40.R) {materialFunction=&MakeFloat2}.Result
  ) {materialFunction=&ScaleUVsByCenter}
  n41 = MaterialFunctionCall(Float2=TextureCoordinate()) {materialFunction=&BreakOutFloat2Components}
  n42 = Add(A=@n41.R, B=ScalarParameter() {parameterName=IconUVOffset_X})
  n43 = Add(A=@n41.G, B=ScalarParameter() {parameterName=IconUVOffset_Y})
  n44 = VectorParameter() {parameterName=IconScale, defaultValue=<rgba 0.5 0.5 0 1>}
  MaterialFunctionCall(
   UVs=MaterialFunctionCall(
    UVs=MaterialFunctionCall(X=@n42, Y=@n43) {materialFunction=&MakeFloat2}.Result
    "Rotation Angle (0-1)"=ScalarParameter() {parameterName=IconRotator}
   ) {materialFunction=&CustomRotator}.Rotated Values
   Texture Scale=MaterialFunctionCall(X=@n44.R, Y=@n44.R) {materialFunction=&MakeFloat2}.Result
  ) {materialFunction=&ScaleUVsByCenter}
 ]
}
