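# UI material that draws a ping diamond with a parameterized outline color, fill color and icon.
# Two MF_UI_Box shapes on UVs rotated an eighth turn form the diamond: their difference is the Outline mask, the inner box the Fill.
# A center-scaled, offsettable icon is blended over the tinted masks using its own RGB as the lerp alpha.
# Icon alpha adds to the mask coverage, with an Opacity parameter scaling part of it for translucency.
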Material /Game/Athena/HUD/Materials/M_UI_PingDiamond
graph {
 properties {blendMode=BLEND_Translucent, materialDomain=MD_UI, pixelDepthOffsetMode=PDOM_Legacy, usedShadingModels=Default Lit}
 nodes [
  n1 = ScalarParameter() {defaultValue=0.89, parameterName=IconScale}
  n2 = ScalarParameter() {parameterName=IconOffsetX}
  n3 = ScalarParameter() {parameterName=IconOffsetY}
  n4 = TextureSampleParameter2D(UVs=Add(
    A=MaterialFunctionCall(Texture Scale=@n1) {materialFunction=&ScaleUVsByCenter}.UVs
    B=AppendVector(A=@n2, B=@n3)
   ) {constB=0}) {
   parameterName=Icon
   samplerSource=SSM_Clamp_WorldGroupSettings
   texture=/Game/Athena/UI/InGame/Creative/PropertyEditor/Icon_Picker/Textures/T_UI_IconPicker_EMPTY
  }
  n5 = NamedRerouteUsage()
  n6 = ScalarParameter() {defaultValue=0.6, parameterName=Opacity}
  Add(A=@n4.A, B=Add(A=NamedRerouteUsage(), B=Multiply(A=@n5, B=@n6)))
  n7 = NamedRerouteUsage()
  n8 = VectorParameter() {parameterName=OutlineColor, defaultValue=<rgba 0.1346 1 0.557 1>}
  n9 = NamedRerouteUsage()
  n10 = VectorParameter() {parameterName=FillColor, defaultValue=<rgba 0.0176 0.4452 1 1>}
  LinearInterpolate(A=Add(A=Multiply(A=@n7, B=@n8.RGB), B=Multiply(A=@n9, B=@n10.RGB)), B=@n4.RGB, Alpha=@n4.RGB) {constAlpha=0}
  n11 = Constant() {r=0.6}
  n12 = Constant()
  n13 = MaterialFunctionCall(UVs=TextureCoordinate(), "Rotation Angle (0-1)"=Constant() {r=0.125}) {materialFunction=&CustomRotator}
  n14 = MaterialFunctionCall(UVs=@n13.Rotated Values, Size=ScalarParameter() {defaultValue=0.5, parameterName=OutlineThickness}, CornerRoundness=@n12) {materialFunction=&MF_UI_Box}
  NamedRerouteDeclaration(None=Subtract(
    A=MaterialFunctionCall(UVs=@n13.Rotated Values, Size=@n11, CornerRoundness=@n12) {materialFunction=&MF_UI_Box}.Fill
    B=@n14.Fill
   )) {name=Outline, nodeColor=<rgba 0 0 0 1>}
  NamedRerouteDeclaration(None=@n14.Fill) {name=Fill, nodeColor=<rgba 0 0 0 1>}
 ]
}
