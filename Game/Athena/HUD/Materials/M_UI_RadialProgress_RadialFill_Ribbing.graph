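# Translucent UI material for a radial progress indicator with an animated segmented "ribbing" fill and a countdown option.
# A radial mask driven by Progress, optionally inverted by a static switch, separates the filled arc from the background color.
# Within the arc, time-rotated radial segments blend a segment color over the shape color, with controls for visibility and animation display.
# A scaled circle SDF supplies fill and outline, which are subtracted to form the shape mask.
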Material /Game/Athena/HUD/Materials/M_UI_RadialProgress_RadialFill_Ribbing
graph {
 properties {blendMode=BLEND_Translucent, materialDomain=MD_UI, pixelDepthOffsetMode=PDOM_Legacy, usedShadingModels=Default Lit}
 nodes [
  n1 = VectorParameter() {parameterName=Shape Color, defaultValue=<rgba 1 0.44 0.44 1>}
  n2 = Time()
  n3 = ScalarParameter() {defaultValue=4, parameterName=Rotation speed}
  n4 = MaterialFunctionCall!(
   UV=Rotator(Time=Multiply(A=@n2, B=@n3) {constB=4})
   Segment Count=ScalarParameter() {defaultValue=10, parameterName=Segment Counts}
   Segment Spacing=ScalarParameter() {defaultValue=0.5, parameterName=Segments Spacing}
   Fixed Spacing
  ) {materialFunction=/FortUILibrary/MaterialFunctions/Basics/MF_UI_RadialSegments}
  n5 = LinearInterpolate(
   A=@n1.RGB
   B=VectorParameter() {parameterName=Segments Fill Color, defaultValue=<rgba 0.2954 1 0.19 1>}.RGB
   Alpha=Multiply(A=Ceil(None=@n4.Segmented Triangles Gradient), B=ScalarParameter() {defaultValue=1, parameterName=Segments Visibility})
  )
  n6 = ScalarParameter() {defaultValue=1, parameterName=DisplaySegmentAnim}
  n7 = TextureCoordinate()
  n8 = ScalarParameter() {defaultValue=0.9568, parameterName=Progress, sortPriority=4}
  n9 = StaticSwitchParameter(True=OneMinus(None=@n8), False=@n8) {parameterName=IsCountdown}
  n10 = Ceil(None=MaterialFunctionCall!(
    UVs=MaterialFunctionCall(UVs=@n7) {materialFunction=&CustomRotator}.Rotated Values
    EdgeSoftness
    Progress=@n9
   ) {materialFunction=/Game/UI/Foundation/Toasts/Materials/MF_RadialMask}.Mask)
  n11 = Subtract(A=OneMinus(None=@n10), B=OneMinus(None=@n9))
  Add(
   A=Multiply(A=LinearInterpolate(A=@n1.RGB, B=@n5, Alpha=@n6), B=@n10)
   B=Multiply(A=VectorParameter() {parameterName=BGColor, defaultValue=<rgba 0.0312 0.0312 0.0312 1>}.RGB, B=Ceil(None=@n11))
  )
  n12 = ScalarParameter() {defaultValue=1, parameterName=Scale}
  n13 = MaterialFunctionCall(
   UVs=MaterialFunctionCall(Texture Scale=@n12) {materialFunction=&ScaleUVsByCenter}.UVs
   Size=@n12
   OutlineThickness=ScalarParameter() {defaultValue=0.2, parameterName=FillWidth}
  ) {materialFunction=&MF_UI_Circle}
  n14 = NamedRerouteDeclaration(None=@n13.Fill) {name=Fill, nodeColor=<rgba 0.9301 0 1 1>}
  n15 = Ceil(None=@n13.SDFOutline)
  NamedRerouteDeclaration(None=Multiply(A=Subtract(A=@n14, B=@n15), B=ScalarParameter() {defaultValue=1, parameterName=Mask})) {name=Mask, nodeColor=<rgba 0.9301 0 1 1>}
  NamedRerouteUsage()
 ]
}
