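# M_UI_RebootChip_Icon is a translucent UI-domain material for a Reboot Chip icon in the Athena HUD.
# It multiplies the RGB of the T_UI_RebootChip_Icon texture by the FillColor vector parameter, which defaults to white and allows the icon to be tinted.
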Material /Game/Athena/HUD/Materials/M_UI_RebootChip_Icon
graph {
 properties {blendMode=BLEND_Translucent, materialDomain=MD_UI, pixelDepthOffsetMode=PDOM_Legacy, usedShadingModels=Default Lit}
 nodes [Multiply(
   A=TextureSample() {texture=&/T_UI_RebootChip_Icon}.RGB
   B=VectorParameter() {parameterName=FillColor, defaultValue=<rgba 1 1 1 1>}.RGB
  )]
}
