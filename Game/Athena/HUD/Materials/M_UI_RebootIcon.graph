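# M_UI_RebootIcon is a translucent UI material for the reboot icon in the Athena HUD.
# It samples the red channel of the T_UI_RebootIcon texture and runs it through SmoothStep between Edge and Edge plus EdgeSoftness.
# This thresholding gives the icon shape an adjustable cutoff and softness, alongside a Color parameter that defaults to white.
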Material /Game/Athena/HUD/Materials/M_UI_RebootIcon
graph {
 properties {blendMode=BLEND_Translucent, materialDomain=MD_UI, pixelDepthOffsetMode=PDOM_Legacy, usedShadingModels=Default Lit}
 nodes [
  n1 = ScalarParameter() {defaultValue=0.9, parameterName=Edge}
  MaterialFunctionCall(
   Alpha=TextureSample() {samplerSource=SSM_Clamp_WorldGroupSettings, texture=&/T_UI_RebootIcon}.R
   Min=@n1
   Max=Add(A=@n1, B=ScalarParameter() {parameterName=EdgeSoftness})
  ) {materialFunction=&SmoothStep}
  VectorParameter() {parameterName=Color, defaultValue=<rgba 1 1 1 1>}
 ]
}
