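# Additive UI material, named as a retainer effect for a self-destruct HUD element, that builds a stylized mask from a texture's alpha.
# Scaled UVs and their rotated copy multiply into a gradient, sharpened by Edge Falloff and multiplied by that alpha.
# The result is masked by the brighter of an exponential radial gradient and a screen-aligned, tiled dot matrix, which DotMatrixFill blends toward solid.
# Several leftover nodes feed nothing, and no material output connection is shown.
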Material /Game/Athena/HUD/Materials/M_UI_SelfDestructRetainer
graph {
 properties {blendMode=BLEND_Additive, materialDomain=MD_UI, pixelDepthOffsetMode=PDOM_Legacy, usedShadingModels=Default Lit}
 nodes [
  Constant2Vector() {g=128, r=64}
  Constant2Vector() {g=-0.5}
  n1 = AppendVector(A=ScalarParameter() {defaultValue=2, parameterName=UVScale}, B=Constant() {r=1})
  n2 = TextureCoordinate()
  MaterialFunctionCall!(Texture Scale=@n1, UVs=@n2, Offset XY Texturesize Z=Constant3Vector() {constant=<rgba 0 0.5 256 0>}) {materialFunction=/Engine/Functions/Engine_MaterialFunctions01/Texturing/ScaleUVsAroundPoint}
  n3 = Multiply(A=@n2, B=@n1)
  n4 = ComponentMask(None=@n3) {r=true}
  n5 = AppendVector(A=ScalarParameter() {defaultValue=2, parameterName=angle1}, B=ScalarParameter() {parameterName=angle2})
  n6 = Constant() {r=-0.25}
  n7 = Multiply(
   A=@n4
   B=ComponentMask(None=MaterialFunctionCall(UVs=@n3, Rotation Center=@n5, "Rotation Angle (0-1)"=@n6) {materialFunction=&CustomRotator}.Rotated Values) {g=true}
  )
  n8 = ScalarParameter() {defaultValue=5, parameterName=Edge Falloff}
  n9 = ScalarParameter() {defaultValue=0.5, parameterName=Radius}
  n10 = ScalarParameter() {defaultValue=2, parameterName=Density}
  n11 = MaterialFunctionCall() {materialFunction=&GetUserInterfaceUV}
  n12 = TextureSample(UVs=Multiply(
    A=MaterialFunctionCall(Texture Dimensions=@n11.Pixel Size) {materialFunction=&ScreenAlignedUVs}."X 100%, Y 100%"
    B=ScalarParameter() {defaultValue=300, parameterName=DotRepeat}
   )) {texture=&/T_UI_RadialGradient}
  n13 = Constant() {r=1}
  n14 = ScalarParameter() {parameterName=DotMatrixFill}
  Multiply(
   A=Multiply(A=Power(Base=@n7, Exp=@n8), B=TextureSampleParameter2D() {parameterName=Texture}.A)
   B=Max(
    A=MaterialFunctionCall(Radius=@n9, Density=@n10) {materialFunction=&RadialGradientExponential}.RadialGradientExponential
    B=LinearInterpolate(A=@n12.R, B=@n13, Alpha=@n14)
   )
  )
  OneMinus(None=@n4)
 ]
}
