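# Translucent UI material, named for the skydive marker in the HUD.
# It divides the LineHeight parameter by the widget's vertical pixel size to form a Y-only UV scale.
# ScaleUVsAroundPoint takes that scale and a point whose vertical position comes from the YOffset parameter.
# The graph shows no texture sample and no connection from this UV math to the material outputs.
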Material /Game/Athena/HUD/Materials/M_UI_SkydiveMarker
graph {
 properties {blendMode=BLEND_Translucent, materialDomain=MD_UI, pixelDepthOffsetMode=PDOM_Legacy, usedShadingModels=Default Lit}
 nodes [
  Constant() {r=1}
  n1 = ScalarParameter() {defaultValue=2, parameterName=LineHeight}
  n2 = MaterialFunctionCall(Float2=MaterialFunctionCall() {materialFunction=&GetUserInterfaceUV}.Pixel Size) {materialFunction=&BreakOutFloat2Components}
  n3 = Constant() {r=1}
  n4 = ScalarParameter() {parameterName=YOffset}
  MaterialFunctionCall!(
   Texture Scale=MaterialFunctionCall(X=Constant() {r=1}, Y=Divide(A=@n1, B=@n2.G)) {materialFunction=&MakeFloat2}.Result
   UVs
   Offset XY Texturesize Z=AppendVector(A=AppendVector(A=@n3, B=@n4), B=@n3)
  ) {materialFunction=/Engine/Functions/Engine_MaterialFunctions01/Texturing/ScaleUVsAroundPoint}
 ]
}
