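# Translucent UI material that draws a skydive HUD icon, blendable to a glider icon by GliderSwap, with three pulsing arrows.
# Icon UVs scale by a pixel-size ratio around an offset point, and the function's 0-1 mask clips the icon.
# Arrow UVs chain cumulative diagonal offsets from those UVs.
# Staggered time-driven sine pulses and per-arrow fades modulate arrow alphas, which combine with the icon through Max.
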Material /Game/Athena/HUD/Materials/M_UI_SkydivePlayer
graph {
 properties {blendMode=BLEND_Translucent, materialDomain=MD_UI, pixelDepthOffsetMode=PDOM_Legacy, usedShadingModels=Default Lit}
 nodes [
  Constant() {r=1}
  n1 = ScalarParameter() {parameterName=XOffset}
  n2 = ScalarParameter() {parameterName=YOffset}
  n3 = MaterialFunctionCall!(
   Texture Scale=Divide(
    A=ScalarParameter() {defaultValue=16, parameterName=PixelSize}
    B=MaterialFunctionCall() {materialFunction=&GetUserInterfaceUV}.Pixel Size
   )
   UVs
   Offset XY Texturesize Z=AppendVector(A=AppendVector(A=@n1, B=@n2), B=Constant() {r=1})
  ) {materialFunction=/Engine/Functions/Engine_MaterialFunctions01/Texturing/ScaleUVsAroundPoint}
  n4 = VectorParameter() {parameterName=ArrowOffset1, defaultValue=<rgba 0 0 0 1>}
  n5 = Add(A=@n3.UVs, B=AppendVector(A=@n4.R, B=@n4.R))
  n6 = Multiply(A=Time() {bOverride_Period=true, period=1}, B=ScalarParameter() {defaultValue=1, parameterName=PulseSpeed})
  n7 = Multiply(
   A=Multiply(
    A=MaterialFunctionCall(Value=@n6) {materialFunction=&LinearSine}."Rounded Linear Sine "
    B=ScalarParameter() {defaultValue=1, parameterName=Arrow1Fade}
   )
   B=TextureSample(UVs=@n5) {samplerSource=SSM_Clamp_WorldGroupSettings, texture=/Game/Athena/HUD/Art/VehicleNextSeatArrow}.A
  )
  n8 = Multiply(
   A=@n3.0-1 mask
   B=LinearInterpolate(
    A=TextureSample(UVs=@n3.UVs) {texture=/Game/Athena/HUD/Art/BacchusHudIcons/Skydive_Normal}.A
    B=TextureSample(UVs=@n3.UVs) {texture=/Game/UI/Foundation/Textures/Icons/Navigation/t_icon_nav_mobile_audioviz_glider_64}.A
    Alpha=ScalarParameter() {parameterName=GliderSwap}
   )
  )
  n9 = VectorParameter() {parameterName=ArrowOffset2, defaultValue=<rgba 0 0 0 1>}
  n10 = Add(A=@n5, B=AppendVector(A=@n9.R, B=@n9.R))
  n11 = TextureSample(UVs=@n10) {samplerSource=SSM_Clamp_WorldGroupSettings, texture=&VehicleNextSeatArrow}
  n12 = Add(A=@n6, B=ScalarParameter() {defaultValue=0.2, parameterName=PulseOffset})
  n13 = Multiply(
   A=MaterialFunctionCall(Value=@n12) {materialFunction=&LinearSine}."Rounded Linear Sine "
   B=ScalarParameter() {defaultValue=1, parameterName=Arrow2Fade}
  )
  n14 = VectorParameter() {parameterName=ArrowOffset3, defaultValue=<rgba 0 0 0 1>}
  n15 = Add(A=@n10, B=AppendVector(A=@n14.R, B=@n14.R))
  n16 = MaterialFunctionCall(Value=Add(A=@n12, B=ScalarParameter() {defaultValue=0.2, parameterName=PulseOffset})) {materialFunction=&LinearSine}
  n17 = ScalarParameter() {defaultValue=1, parameterName=Arrow3Fade}
  Max(
   A=Max(A=Max(A=@n7, B=@n8), B=Multiply(A=@n11.A, B=@n13))
   B=Multiply(
    A=TextureSample(UVs=@n15) {samplerSource=SSM_Clamp_WorldGroupSettings, texture=&VehicleNextSeatArrow}.A
    B=Multiply(A=@n16."Rounded Linear Sine ", B=@n17)
   )
  )
 ]
}
