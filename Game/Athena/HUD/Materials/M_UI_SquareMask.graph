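# M_UI_SquareMask is a translucent UI-domain material in the Athena HUD folder that computes a mask from two textures.
# It multiplies the alpha of the Target texture parameter by the squared RGB of a fixed T_WhiteMask sample.
# Target also defaults to T_WhiteMask, and the graph shows no connection from the final product to a material output.
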Material /Game/Athena/HUD/Materials/M_UI_SquareMask
graph {
 properties {blendMode=BLEND_Translucent, materialDomain=MD_UI, pixelDepthOffsetMode=PDOM_Legacy, usedShadingModels=Default Lit}
 nodes [
  n1 = TextureSample() {samplerType=SAMPLERTYPE_Masks, texture=&T_WhiteMask}
  Multiply(
   A=TextureSampleParameter2D() {parameterName=Target, samplerType=SAMPLERTYPE_Masks, texture=&T_WhiteMask}.A
   B=Multiply(A=@n1.RGB, B=@n1.RGB)
  )
 ]
}
