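# Translucent UI material for the animated lightning effect on the Storm Surge HUD icon.
# A time-driven FlipBook with parameterized rows and columns steps through frames of a lightning texture.
# Thresholded texture values blend the tint color toward a hot core, while a thresholded red channel adds a brightness term tuned by TaperDissipation.
# The green channel is scaled in a separate blend branch.
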Material /Game/Athena/HUD/Materials/M_UI_StormSurgeIconFX
graph {
 properties {blendMode=BLEND_Translucent, materialDomain=MD_UI, pixelDepthOffsetMode=PDOM_Legacy, usedShadingModels=Default Lit}
 nodes [
  n1 = Time()
  n2 = ScalarParameter() {defaultValue=2.5, parameterName=Time}
  n3 = TextureSample(UVs=MaterialFunctionCall(
    "Animation  Phase (0-1)"=Multiply(A=@n1, B=@n2) {constA=2, constB=2}
    Number of Rows=ScalarParameter() {defaultValue=1, parameterName=Rows}
    Number of Columns=ScalarParameter() {defaultValue=3, parameterName=Column}
   ) {materialFunction=&FlipBook}.UVs) {texture=&/T-UI-StormSurgeLightning}
  LinearInterpolate(B=Multiply(A=@n3.G) {constB=2})
  n4 = VectorParameter() {parameterName=Lightning Color, defaultValue=<rgba 1 0.4828 0.9627 0>}
  n5 = ScalarParameter() {defaultValue=0.5, parameterName=CenterCore}
  n6 = Clamp(None=Multiply(A=Subtract(A=@n3.RGB, B=@n5)) {constB=5})
  n7 = Multiply(A=Subtract(A=@n3.R, B=ScalarParameter() {defaultValue=0.6, parameterName=TaperDissipation})) {constB=16}
  LinearInterpolate(B=Add(A=LinearInterpolate(A=@n4.RGB, Alpha=@n6), B=Clamp(None=@n7)))
  Multiply() {constB=1.5}
 ]
}
