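# Translucent UI material for the Storm Surge glow in the Athena HUD.
# A RadialGradientExponential function builds the glow shape from GradRadius and GradDensity parameters.
# One lerp blends the GradientColor parameter (hot pink by default) with that gradient at a fixed alpha.
# A second lerp takes a boosted copy of the gradient, and the graph shows no links from either lerp to material outputs.
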Material /Game/Athena/HUD/Materials/M_UI_StormSurge_Glow
graph {
 properties {blendMode=BLEND_Translucent, materialDomain=MD_UI, pixelDepthOffsetMode=PDOM_Legacy, usedShadingModels=Default Lit}
 nodes [
  n1 = MaterialFunctionCall(
   Radius=ScalarParameter() {defaultValue=0.5, parameterName=GradRadius}
   Density=ScalarParameter() {defaultValue=0.9, parameterName=GradDensity}
  ) {materialFunction=&RadialGradientExponential}
  LinearInterpolate(A=VectorParameter() {parameterName=GradientColor, defaultValue=<rgba 0.9531 0.0366 0.6434 0>}.RGB, B=@n1.RadialGradientExponential)
  LinearInterpolate(B=Multiply(A=@n1.RadialGradientExponential) {constB=3})
 ]
}
