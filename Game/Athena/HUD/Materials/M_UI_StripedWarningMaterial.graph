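# UI material for a striped warning bar, by default orange and dark red, with a horizontal progress fill.
# Rotated, tiled UVs pan through a vertical gradient, and a SmoothStep with adjustable edge and softness turns it into two-color stripes.
# A second SmoothStep on normalized UI X, driven by Progress, blends the stripes into a near-black color past the fill edge.
# Timer and screen-aligned dot pattern branches exist but do not feed that blend.
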Material /Game/Athena/HUD/Materials/M_UI_StripedWarningMaterial
graph {
 properties {materialDomain=MD_UI, pixelDepthOffsetMode=PDOM_Legacy, usedShadingModels=Default Lit}
 nodes [
  n1 = Time()
  n2 = ScalarParameter() {parameterName=TimerStart}
  Clamp(None=Divide(A=Subtract(A=@n1, B=@n2), B=ScalarParameter() {parameterName=TimerDuration}))
  Constant2Vector()
  n3 = MaterialFunctionCall(UVs=TextureCoordinate(), "Rotation Angle (0-1)"=ScalarParameter() {parameterName=PatternAngle}) {materialFunction=&CustomRotator}
  n4 = VectorParameter() {parameterName=PatternRepeat, defaultValue=<rgba 1 1 0 1>}
  n5 = AppendVector(A=@n4.R, B=@n4.R)
  n6 = Constant()
  n7 = ScalarParameter() {parameterName=PanSpeed}
  n8 = TextureSample(UVs=Panner(Coordinate=Multiply(A=@n3.Rotated Values, B=@n5), Speed=AppendVector(A=@n6, B=@n7))) {texture=&/T_UI_VerticalGradient}
  n9 = ScalarParameter() {defaultValue=0.5, parameterName=Edge}
  n10 = Add(A=@n9, B=ScalarParameter() {defaultValue=0.01, parameterName=EdgeSoftness})
  n11 = MaterialFunctionCall() {materialFunction=&GetUserInterfaceUV}
  n12 = LinearInterpolate(Alpha=ScalarParameter() {defaultValue=1, parameterName=Progress, sliderMax=1}) {constA=-0.1}
  n13 = ScalarParameter() {defaultValue=0.01, parameterName=EdgeSoftness}
  LinearInterpolate(
   A=LinearInterpolate(
    A=VectorParameter() {parameterName=Color1, defaultValue=<rgba 1 0.6445 0 1>}.RGB
    B=VectorParameter() {parameterName=Color2, defaultValue=<rgba 0.0513 0.003 0.0021 1>}.RGB
    Alpha=MaterialFunctionCall(Alpha=@n8.R, Min=@n9, Max=@n10) {materialFunction=&SmoothStep}.Result
   )
   B=Constant3Vector() {constant=<rgba 0.0104 0.0002 0 0>}
   Alpha=MaterialFunctionCall(Alpha=ComponentMask(None=@n11.Normalized UV) {r=true}, Min=@n12, Max=Add(A=@n12, B=@n13)) {materialFunction=&SmoothStep}.Result
  ) {constAlpha=0}
  MaterialFunctionCall() {materialFunction=&GetUserInterfaceUV}
  n14 = MaterialFunctionCall() {materialFunction=&ScreenAlignedUVs}
  n15 = ScalarParameter() {defaultValue=100, parameterName=DotRepeat}
  n16 = ScalarParameter() {parameterName=DotEdge}
  MaterialFunctionCall(
   Alpha=TextureSample(UVs=Multiply(A=@n14."X 100%, Y 100%", B=@n15)) {texture=&/T_UI_RadialGradient}.R
   Min=@n16
   Max=Add(A=@n16, B=ScalarParameter() {defaultValue=0.2, parameterName=DotEdgeSoftness})
  ) {materialFunction=&SmoothStep}
  TextureCoordinate()
 ]
}
