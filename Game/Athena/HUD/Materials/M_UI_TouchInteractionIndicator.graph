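# Translucent UI material that draws the HUD touch interaction indicator: a tinted circular badge with an icon and a radial progress timer.
# Sphere masks build the ring, glow and inner shading, with a DDX-derived term to anti-alias edges.
# An angle gradient through SmoothStep sweeps the Progress ring.
# The scaled icon is resampled with an offset for its drop shadow, CropIcon blends in masking, and a GlowOnly static switch swaps out the composite.
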Material /Game/Athena/HUD/Materials/M_UI_TouchInteractionIndicator
graph {
 properties {blendMode=BLEND_Translucent, materialDomain=MD_UI, pixelDepthOffsetMode=PDOM_Legacy, usedShadingModels=Default Lit}
 nodes [
  n1 = DDX(Value=NamedRerouteDeclaration(None=TextureCoordinate()) {name=TexCoord, nodeColor=<rgba 0.0312 1 0 1>})
  n2 = ScalarParameter() {defaultValue=2, parameterName=AA Multiplier}
  NamedRerouteDeclaration(None=ComponentMask(None=Multiply(A=@n1, B=@n2) {constB=4}) {r=true}) {name=FixAA, nodeColor=<rgba 1 0 0.3516 1>}
  n3 = ScalarParameter() {defaultValue=0.5, parameterName=OuterRingRadius}
  n4 = ScalarParameter() {defaultValue=0.0395, parameterName=InnerRingRadius}
  n5 = ScalarParameter() {defaultValue=0.98, parameterName=EdgeSoftness}
  n6 = NamedRerouteUsage()
  NamedRerouteDeclaration(None=SphereMask(A=NamedRerouteUsage(), B=Constant2Vector() {g=0.5, r=0.5}, Radius=Subtract(A=@n3, B=@n4), Hardness=Subtract(A=@n5, B=@n6))) {name=InnerCircle, nodeColor=<rgba 1 0.7031 0 1>}
  n7 = Subtract(A=NamedRerouteUsage(), B=NamedRerouteUsage())
  n8 = Constant() {r=1}
  n9 = Clamp(None=ScalarParameter() {parameterName=Progress})
  n10 = Constant()
  n11 = NamedRerouteUsage()
  n12 = MaterialFunctionCall(
   Alpha=OneMinus(None=MaterialFunctionCall!(UVs=@n11, Rotation) {materialFunction=/Game/UI/Foundation/Materials/MF_UI_AngleGradient_DEPRECATED}.Result)
   Min=@n9
   Max=Add(A=@n9, B=ScalarParameter() {defaultValue=0.01, parameterName=PercentEdgeSoftness})
  ) {materialFunction=&SmoothStep}
  n13 = If(A=@n9, B=@n8, "A > B"=@n8, "A == B"=@n8, "A < B"=If(A=@n9, B=@n10, "A > B"=OneMinus(None=@n12.Result), "A == B"=@n10, "A < B"=@n10))
  NamedRerouteDeclaration(None=Multiply(A=ScalarParameter() {parameterName=TimerOpacity}, B=Multiply(A=@n7, B=@n13))) {name=Progress, nodeColor=<rgba 0 0.5781 1 1>}
  n14 = NamedRerouteUsage()
  n15 = Constant2Vector() {g=0.5, r=0.5}
  n16 = ScalarParameter() {defaultValue=0.5, parameterName=OuterRingRadius}
  n17 = ScalarParameter() {defaultValue=0.98, parameterName=EdgeSoftness}
  n18 = NamedRerouteUsage()
  NamedRerouteDeclaration(None=SphereMask(A=@n14, B=@n15, Radius=@n16, Hardness=Subtract(A=@n17, B=@n18) {constB=0})) {name=OutCircle, nodeColor=<rgba 0.8984 1 0 1>}
  n19 = ScalarParameter() {defaultValue=0.9, parameterName=OuterGlowAmount}
  n20 = NamedRerouteUsage()
  NamedRerouteDeclaration(None=SphereMask(A=@n14, B=@n15, Radius=@n16, Hardness=Subtract(A=@n19, B=@n20) {constB=0})) {name=InnerMask, nodeColor=<rgba 0.7813 0 1 1>}
  n21 = VectorParameter() {parameterName=Tint, defaultValue=<rgba 0 0.8238 1 1>}
  n22 = ScalarParameter() {defaultValue=0.8, parameterName=InnerDarken}
  n23 = ScalarParameter() {defaultValue=1, parameterName=ColorPower}
  NamedRerouteDeclaration(None=LinearInterpolate(A=Power(Base=@n21.RGB, Exp=@n23), B=Power(Base=@n21.RGB, Exp=@n22), Alpha=NamedRerouteUsage())) {name=InnerTint, nodeColor=<rgba 1 0.5391 0 1>}
  NamedRerouteUsage()
  n24 = NamedRerouteUsage()
  n25 = NamedRerouteUsage()
  n26 = Constant2Vector() {g=0.7, r=0.5}
  n27 = ScalarParameter() {defaultValue=0.7, parameterName=InnerShadowSize}
  n28 = Subtract(A=ScalarParameter() {defaultValue=0.4, parameterName=InnerShadow}, B=NamedRerouteUsage()) {constB=0}
  n29 = Multiply(A=ScalarParameter() {defaultValue=0.9, parameterName=FillAmount}, B=OneMinus(None=SphereMask(A=@n25, B=@n26, Radius=@n27, Hardness=@n28)))
  n30 = Multiply(A=ScalarParameter() {defaultValue=0.7, parameterName=AnimatedIconScale}, B=ScalarParameter() {defaultValue=1, parameterName=IconScale})
  n31 = MaterialFunctionCall(UVs=Add(A=NamedRerouteUsage(), B=Constant2Vector() {g=-0.04, r=-0.04}), Texture Scale=@n30) {materialFunction=&ScaleUVsByCenter}
  n32 = TextureSampleParameter2D(UVs=@n31.UVs) {parameterName=DisplayedIcon, samplerSource=SSM_Clamp_WorldGroupSettings, texture=/Game/Athena/HUD/Art/BacchusHudIcons/Use_Item_Normal}
  n33 = Multiply(A=@n32.A, B=@n31.0-1 mask)
  n34 = ScalarParameter() {defaultValue=1, parameterName=ShadowOpacity}
  n35 = MaterialFunctionCall(UVs=NamedRerouteUsage(), Texture Scale=@n30) {materialFunction=&ScaleUVsByCenter}
  n36 = TextureSampleParameter2D(UVs=@n35.UVs) {parameterName=DisplayedIcon, samplerSource=SSM_Clamp_WorldGroupSettings, texture=&Use_Item_Normal}
  n37 = Multiply(A=@n36.A, B=@n35.0-1 mask)
  n38 = Max(A=Multiply(A=Multiply(A=@n33, B=@n34), B=NamedRerouteUsage()), B=@n37)
  n39 = Max(A=Subtract(A=NamedRerouteUsage(), B=Multiply(A=@n24, B=@n29)), B=@n38)
  n40 = Multiply(A=NamedRerouteUsage(), B=@n39)
  n41 = ScalarParameter() {parameterName=CropIcon}
  Reroute(None=StaticSwitchParameter(True=NamedRerouteUsage(), False=LinearInterpolate(A=@n39, B=@n40, Alpha=@n41) {constAlpha=1}) {parameterName=GlowOnly})
  n42 = NamedRerouteUsage()
  n43 = Add(
   A=Multiply(
    A=VectorParameter() {parameterName=Tint, defaultValue=<rgba 0 0.8238 1 1>}.RGB
    B=ScalarParameter() {defaultValue=0.35, parameterName=OutlineTint}
   )
   B=NamedRerouteUsage()
  )
  n44 = OneMinus(None=NamedRerouteUsage())
  n45 = Constant3Vector() {constant=<rgba 0 0 0.0056 1>}
  n46 = Power(Base=@n36.RGB, Exp=ScalarParameter() {defaultValue=1, parameterName=IconLevels})
  n47 = Multiply(A=@n38, B=NamedRerouteUsage())
  n48 = ScalarParameter() {parameterName=CropIcon}
  StaticSwitchParameter(
   True=NamedRerouteUsage()
   False=LinearInterpolate(
    A=LinearInterpolate(A=@n42, B=@n43, Alpha=@n44)
    B=LinearInterpolate(A=@n45, B=@n46, Alpha=@n37)
    Alpha=LinearInterpolate(A=@n38, B=@n47, Alpha=@n48) {constAlpha=1}
   )
  ) {parameterName=GlowOnly}
 ]
}
