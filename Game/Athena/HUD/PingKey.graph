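# PingKey is a HUD element widget that shows the key binding for placing a map marker in the world, beside a location ping icon.
# A bottom-right aligned row in an overlay holds a keybind widget bound to PlaceMapMarkerInWorld and the squad location indicator image.
# A tutorial highlight fills the same overlay, using a square outline widget with an adjusted render transform.
# The event graph is empty, so the blueprint adds no logic of its own.
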WidgetBlueprint /Game/Athena/HUD/PingKey
graphs [EventGraph {links [], nodes []}]
variables []
widgets {
 info {inheritedWidgetCount=1, parentClass=&FortHUDElementWidget, rootWidgetClass=&Overlay, widgetCount=5}
 widgets [
  Overlay_1 {widgetClassPath=&Overlay, widget=Overlay_1}
  PingRow {
   widgetClassPath=&HorizontalBox
   parent=Overlay_1
   widget=PingRow
   slot {class=OverlaySlot, properties {horizontalAlignment=HAlign_Right, verticalAlignment=VAlign_Bottom, padding=<ltrb 4 2 4 2>}}
  }
  KeybindWidget {
   widgetClassPath=/FortUILibrary/Input/KeybindWidget
   parent=PingRow
   widget=KeybindWidget
   properties {boundAction=PlaceMapMarkerInWorld}
   slot {class=HorizontalBoxSlot}
  }
  PingIcon {
   widgetClassPath=&Image
   bIsVariable=true
   parent=PingRow
   widget=PingIcon
   properties {brush {
     bIsDynamicallyLoaded=false
     drawAs=Image
     imageType=NoImage
     mirroring=NoMirror
     resourceName=None
     resourceObject=/Game/Athena/HUD/Art/squadArt/T_UI_SquadGenericIndicator_Location_b
     tiling=NoTile
     imageSize=<32 32>
     margin=<ltrb 0 0 0 0>
     outlineSettings {
      bUseBrushTransparency=false
      roundingType=HalfHeightRadius
      width=0
      color {colorUseRule=UseColor_Specified, specifiedColor=<rgba 0 0 0 0>}
      cornerRadii=<0 0 0 0>
     }
     tintColor {colorUseRule=UseColor_Specified, specifiedColor=<rgba 1 1 1 1>}
     uVRegion {bIsValid=false, max=<0 0>, min=<0 0>}
    }}
   slot {class=HorizontalBoxSlot, properties {verticalAlignment=VAlign_Center, padding=<ltrb 4 0 0 0>}}
  }
  TutorialHighlight {
   widgetClassPath=&TutorialHighlight
   bIsVariable=true
   parent=Overlay_1
   widget=TutorialHighlight
   properties {
    customHighlightWidget=/Game/Athena/HUD/Tutorial/Tutorial_Outline_Square.Tutorial_Outline_Square_C
    renderTransform {angle=0, scale=<1.3 0.75>, shear=<0 0>, translation=<4 25>}
   }
   slot {class=OverlaySlot, properties {horizontalAlignment=HAlign_Fill, verticalAlignment=VAlign_Fill}}
  }
 ]
}
