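# Creative HUD global options panel, listing Movement, Object Placement and Tools categories in a scroll box above a tooltip area.
# Activation hides HUD elements through a mobile or non-mobile gameplay tag set, and deactivation restores them.
# Custom navigation functions move between category lists, targeting the first or last displayed entry on gamepad, otherwise the list itself.
# The map action and tab overrides all return false.
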WidgetBlueprint /Game/Athena/HUD/WBP_CreativeHUDGlobalOptions_Widget
graphs [
 EventGraph {nodes [
   Event On Activated()
   > n1 = Cast To FortPlayerController(Object=GetOwningPlayer())
   > SetHUDElementVisibilityWithContext!(
    execute
    self:Fort Player Controller Object Reference=@n1
    HUDElementTags:"Gameplay Tag Container Structure (by ref)"=Select(
     Option 0:"Gameplay Tag Container Structure (by ref)"=Get HideGameplayTag() -> (HideGameplayTag:Gameplay Tag Container Structure)
     Option 1:"Gameplay Tag Container Structure (by ref)"=Get HideGameplayTag_Mobile() -> (HideGameplayTag_Mobile:Gameplay Tag Container Structure)
     Index=IsMobileGame()
    ) -> (ReturnValue:"Gameplay Tag Container Structure (by ref)")
    bHideElements:bool=true
    ReasonName:Name=None
    LogContext:String
   ) -> (then)
   Event On Deactivated()
   > n2 = Cast To FortPlayerController(Object=GetOwningPlayer())
   > SetHUDElementVisibilityWithContext(
    self=@n2
    HUDElementTags=Select(Option 0=Get HideGameplayTag(), Option 1=Get HideGameplayTag_Mobile(), Index=IsMobileGame())
    bHideElements=false
    ReasonName=None
   )
   n3 = Event PreConstruct()
   > Branch(Condition=@n3.IsDesignTime)
   > "SetText (Text)"(self=Get Text_ToolTipText() -> (Text_ToolTipText:Common Text Object Reference), InText="INVTEXT(\"{TooltipText}\")")
  ]}
 NavigateDownToMoveToolListView {nodes [
   n1 = NavigateDownToMoveToolListView!() -> (then, Navigation:EUINavigation Enum) {is=title K2Node_FunctionEntry}
   > n2 = Switch on EUINavigation!(execute, Selection:EUINavigation Enum=@n1) -> (Left:Exec, Right:Exec, Up:Exec, Down:Exec, Next:Exec, Previous:Exec, Invalid:Exec) {is=title K2Node_SwitchEnum}
   n3 = Branch(
    execute=@n2.Down
    Condition=IsUsingGamepad!(self:Fort Global UIContext Object Reference=Get FortGlobalUIContext() -> (ReturnValue:Fort Global UIContext Object Reference) {is=title K2Node_GetSubsystem}) -> (ReturnValue:bool)
   )
   n4 = GetOptionsList!(self:Fort Creative Options Category Object Reference=Get Category_MoveTool() -> (Category_MoveTool:Fort Creative Options Category Object Reference)) -> (ReturnValue:Common List View Object Reference)
   > Return Node!(
    execute
    ReturnValue:Widget Object Reference="Get (a copy)"(Array:Array of User Widget Object References=GetDisplayedEntryWidgets!(self:List View Base Object Reference=@n4) -> (ReturnValue:Array of User Widget Object References)) -> (Output:User Widget Object Reference)
   )
   Return Node(execute=@n3.else, ReturnValue=GetOptionsList(self=Get Category_MoveTool()))
  ]}
 NavigateDownToToolsetListView {nodes [
   n1 = NavigateDownToToolsetListView!() -> (then, Navigation:EUINavigation Enum) {is=title K2Node_FunctionEntry}
   > n2 = Switch on EUINavigation(Selection=@n1)
   n3 = Branch(execute=@n2.Down, Condition=IsUsingGamepad(self=Get FortGlobalUIContext()))
   n4 = GetOptionsList(self=Get Category_Toolset() -> (Category_Toolset:Glolbal Option Category List View Object Reference))
   > Return Node(ReturnValue="Get (a copy)"(Array=GetDisplayedEntryWidgets(self=@n4)))
   Return Node(execute=@n3.else, ReturnValue=GetOptionsList(self=Get Category_Toolset()))
  ]}
 NavigateUpToMoveToolListView {nodes [
   n1 = NavigateUpToMoveToolListView!() -> (then, Navigation:EUINavigation Enum) {is=title K2Node_FunctionEntry}
   > n2 = Switch on EUINavigation(Selection=@n1)
   n3 = Branch(execute=@n2.Up, Condition=IsUsingGamepad(self=Get FortGlobalUIContext()))
   n4 = GetDisplayedEntryWidgets(self=GetOptionsList(self=Get Category_MoveTool()))
   n5 = Length(TargetArray:Array of User Widget Object References=@n4)
   > Return Node(ReturnValue="Get (a copy)"(Array=@n4, Dimension 1=int - int(A=@n5, B=1) {is=title K2Node_CallFunction}))
   Return Node(execute=@n3.else, ReturnValue=GetOptionsList(self=Get Category_MoveTool()))
  ]}
 NavigateUpToPlayerListView {nodes [
   n1 = NavigateUpToPlayerListView!() -> (then, Navigation:EUINavigation Enum) {is=title K2Node_FunctionEntry}
   > n2 = Switch on EUINavigation(Selection=@n1)
   n3 = Branch(execute=@n2.Up, Condition=IsUsingGamepad(self=Get FortGlobalUIContext()))
   n4 = GetDisplayedEntryWidgets(self=GetOptionsList(self=Get Category_Player() -> (Category_Player:Fort Creative Options Category Object Reference)))
   n5 = Length(TargetArray=@n4)
   > Return Node(ReturnValue="Get (a copy)"(Array=@n4, Dimension 1=int - int(A=@n5, B=1)))
   Return Node(execute=@n3.else, ReturnValue=GetOptionsList(self=Get Category_Player()))
  ]}
 OnBlockMapActionByKey {nodes [
   OnBlockMapActionByKey!() -> (then, InKey:"Key Structure (by ref)") {is=title K2Node_FunctionEntry}
   > Return Node!(execute, bOutMapActionBlocked:bool=false, ReturnValue:bool=false)
  ]}
 OnHandleMapActionKey {nodes [
   OnHandleMapActionKey!() -> (then, InKey:"Key Structure (by ref)", bKeyDown:bool) {is=title K2Node_FunctionEntry}
   > Return Node!(execute, bOutMapActionHandled:bool=false, ReturnValue=false)
  ]}
 OnHandleMapBackAction {nodes [
   OnHandleMapBackAction() {is=title K2Node_FunctionEntry}
   > Return Node!(execute, bOutBackActionHandled:bool=false, ReturnValue=false)
  ]}
 OnShouldCreateMapScreenTab {nodes [
   OnShouldCreateMapScreenTab!() -> (then, WorldContextObject:Object Reference) {is=title K2Node_FunctionEntry}
   > Return Node!(execute, bOutShouldCreateMapScreenTab:bool=false, ReturnValue=false)
  ]}
 OnShouldCreateTab {nodes [
   OnShouldCreateTab!() -> (then, WorldContextObject:Object Reference) {is=title K2Node_FunctionEntry}
   > Return Node!(execute, bOutShouldCreateTab:bool=false, ReturnValue=false)
  ]}
 OnShouldDisplayBackActionInActionBar {nodes [
   OnShouldDisplayBackActionInActionBar() {is=title K2Node_FunctionEntry}
   > Return Node!(execute, bOutShouldDisplayBackActionInActionBar:bool=false, ReturnValue=false)
  ]}
 OnShouldShowTabOnMapScreenForMinigame {nodes [
   OnShouldShowTabOnMapScreenForMinigame!() -> (then, Minigame:Fort Minigame Object Reference) {is=title K2Node_FunctionEntry}
   > Return Node!(execute, bOutShouldShowTabOnMapScreenForMinigame:bool=false, ReturnValue=false)
  ]}
]
variables [
 HideGameplayTag {type=GameplayTagContainer, default {gameplayTags [{tagName=HUD}]}}
 HideGameplayTag_Mobile {type=GameplayTagContainer, default {gameplayTags [{tagName=HUD.Bacchus}]}}
]
widgets {
 info {inheritedWidgetCount=4, parentClass=/Script/FortniteUI.FortCreativeGlobalOptions, rootWidgetClass=&SafeZone, widgetCount=19}
 widgets [
  SafeZone_1 {widgetClassPath=&SafeZone, widget=SafeZone_1}
  Overlay_1 {widgetClassPath=&Overlay, parent=SafeZone_1, widget=Overlay_1, slot {class=SafeZoneSlot}}
  CommonCustomNavigation_70 {
   widgetClassPath=/Script/CommonUI.CommonCustomNavigation
   parent=Overlay_1
   widget=CommonCustomNavigation_70
   properties {background {
     bIsDynamicallyLoaded=false
     drawAs=NoDrawType
     imageType=NoImage
     mirroring=NoMirror
     resourceName=None
     resourceObject=null
     tiling=NoTile
     imageSize=<32 32>
     margin=<ltrb 0 0 0 0>
     outlineSettings {
      bUseBrushTransparency=false
      roundingType=HalfHeightRadius
      width=0
      color {colorUseRule=UseColor_Specified, specifiedColor=<rgba 0 0 0 0>}
      cornerRadii=<0 0 0 0>
     }
     tintColor {colorUseRule=UseColor_Specified, specifiedColor=<rgba 1 1 1 1>}
     uVRegion {bIsValid=false, max=<0 0>, min=<0 0>}
    }}
   slot {class=OverlaySlot, properties {padding=<ltrb 16 16 0 0>}}
  }
  SizeBox_Shell {
   widgetClassPath=&SizeBox
   parent=CommonCustomNavigation_70
   widget=SizeBox_Shell
   properties {bOverride_WidthOverride=true, minDesiredHeight=572, widthOverride=674}
   slot {class=BorderSlot, properties {padding=<ltrb 0 0 0 0>}}
  }
  Overlay_23 {widgetClassPath=&Overlay, parent=SizeBox_Shell, widget=Overlay_23, slot {class=SizeBoxSlot}}
  WBP_UIKit_Backplate_Glass {
   widgetClassPath=&WBP_UIKit_Backplate
   parent=Overlay_23
   widget=WBP_UIKit_Backplate_Glass
   properties {brightness=Dark}
   slot {class=OverlaySlot, properties {horizontalAlignment=HAlign_Fill, verticalAlignment=VAlign_Fill}}
  }
  VerticalBox_1 {
   widgetClassPath=&VerticalBox
   parent=Overlay_23
   widget=VerticalBox_1
   slot {class=OverlaySlot, properties {horizontalAlignment=HAlign_Fill, verticalAlignment=VAlign_Fill}}
  }
  Overlay_184 {widgetClassPath=&Overlay, parent=VerticalBox_1, widget=Overlay_184, slot {class=VerticalBoxSlot}}
  VerticalBox_Body {
   widgetClassPath=&VerticalBox
   parent=VerticalBox_1
   widget=VerticalBox_Body
   slot {class=VerticalBoxSlot, properties {size {sizeRule=Fill, value=1}}}
  }
  ScrollBox_0 {
   widgetClassPath=&ScrollBox
   parent=VerticalBox_Body
   widget=ScrollBox_0
   properties {navigationDestination=Center, scrollWhenFocusChanges=AnimatedScroll}
   slot {class=VerticalBoxSlot, properties {padding=<ltrb 12 12 12 12>, size {sizeRule=Fill, value=1}}}
  }
  VerticalBox_2 {
   widgetClassPath=&VerticalBox
   parent=ScrollBox_0
   widget=VerticalBox_2
   properties {clipping=OnDemand}
   slot {class=ScrollBoxSlot}
  }
  Category_Player {
   widgetClassPath=&/GlolbalOptionCategoryListView
   bInherited=true
   bIsVariable=true
   parent=VerticalBox_2
   widget=Category_Player
   properties {
    navigation=&self:WidgetTree.Category_Player.WidgetNavigation_0
    optionCategory {categoryDisplayName=Movement, categoryTag=player}
   }
   slot {class=VerticalBoxSlot}
  }
  Category_MoveTool {
   widgetClassPath=&GlolbalOptionCategoryListView
   bInherited=true
   bIsVariable=true
   parent=VerticalBox_2
   widget=Category_MoveTool
   properties {
    navigation=&self:WidgetTree.Category_MoveTool.WidgetNavigation_0
    optionCategory {categoryDisplayName=Object Placement, categoryTag=movetool}
   }
   slot {class=VerticalBoxSlot, properties {padding=<ltrb 0 12 0 0>}}
  }
  Category_Toolset {
   widgetClassPath=&GlolbalOptionCategoryListView
   bIsVariable=true
   parent=VerticalBox_2
   widget=Category_Toolset
   properties {
    navigation=&self:WidgetTree.Category_Toolset.WidgetNavigation_0
    optionCategory {categoryDisplayName=Tools, categoryTag=toolset}
    padding=<ltrb 0 0 0 4>
   }
   slot {class=VerticalBoxSlot, properties {padding=<ltrb 0 24 0 0>}}
  }
  SizeBox_0 {
   widgetClassPath=&SizeBox
   parent=VerticalBox_Body
   widget=SizeBox_0
   properties {bOverride_HeightOverride=true, heightOverride=156}
   slot {class=VerticalBoxSlot}
  }
  Overlay_2 {widgetClassPath=&Overlay, parent=SizeBox_0, widget=Overlay_2, slot {class=SizeBoxSlot}}
  DescInfo_Glass {
   widgetClassPath=&WBP_UIKit_Backplate
   parent=Overlay_2
   widget=DescInfo_Glass
   properties {renderOpacity=0.2}
   slot {class=OverlaySlot, properties {horizontalAlignment=HAlign_Fill, verticalAlignment=VAlign_Fill}}
  }
  Text_ToolTipText {
   widgetClassPath=&CommonTextBlock
   bInherited=true
   bIsVariable=true
   parent=Overlay_2
   widget=Text_ToolTipText
   properties {
    autoWrapText=true
    scrollStyle=&TextScrollStyle-Base_C
    colorAndOpacity {colorUseRule=UseColor_Specified, specifiedColor=<rgba 1 1 1 0.8>}
    font {
     bForceMonospaced=false
     bMaterialIsStencil=false
     fontMaterial=/Game/Athena/UI/Challenges/Art/M_UI_Challenge_DynamicColor_Text_WhiteFill
     fontObject=&HeadingNow
     letterSpacing=0
     monospacedWidth=1
     size=15
     skewAmount=0
     typefaceFontName=74 Regular
     outlineSettings {
      bApplyOutlineToDropShadows=false
      bMiteredCorners=false
      bSeparateFillAlpha=false
      outlineMaterial=null
      outlineSize=0
      outlineColor=<rgba 0 0 0 1>
     }
    }
    shadowOffset=<0 0>
   }
   slot {class=OverlaySlot, properties {horizontalAlignment=HAlign_Fill, verticalAlignment=VAlign_Center, padding=<ltrb 24 24 24 24>}}
  }
  Button_Cancel {widgetClassPath=&CommonButtonBase, bInherited=true, widget=null}
 ]
}
