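# Translucent UI material, named for the Creative start panel, that builds a dark blue gradient backdrop with a slanted edge.
# A corner-centered exponential radial gradient blends the Color1 and Color2 parameters and also blends AlphaBottom to AlphaTop.
# A slightly rotated, offset UV coordinate, inverted and steeply multiplied then clamped, forms a crisp angled edge mask scaled by that alpha.
# Three constant colors and a second, opposite-corner gradient appear unused.
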Material /Game/Athena/HUD/Art/Materials/M_UI_CreativeStart_Panel
graph {
 properties {blendMode=BLEND_Translucent, materialDomain=MD_UI, pixelDepthOffsetMode=PDOM_Legacy, usedShadingModels=Default Lit}
 nodes [
  Constant3Vector() {constant=<rgba 0 0.3333 1 0>}
  Constant3Vector() {constant=<rgba 0 0.1042 0.625 0>}
  Constant3Vector() {constant=<rgba 0.0011 0.0045 0.0425 0>}
  n1 = Constant() {r=1.41}
  n2 = Constant() {r=1.666}
  n3 = MaterialFunctionCall(UVs=TextureCoordinate(), CenterPosition=Constant2Vector(), Radius=@n1, Density=@n2) {materialFunction=&RadialGradientExponential}
  LinearInterpolate(
   A=VectorParameter() {parameterName=Color1, defaultValue=<rgba 0 0.0042 0.0353 1>}.RGB
   B=VectorParameter() {parameterName=Color2, defaultValue=<rgba 0 0.0062 0.075 1>}.RGB
   Alpha=@n3.RadialGradientExponential
  )
  n4 = Rotator(Coordinate=TextureCoordinate(), Time=Constant() {r=-0.0063}) {speed=6.2832}
  n5 = Constant() {r=0.0213}
  n6 = OneMinus(None=ComponentMask(None=Add(A=@n4, B=@n5)) {r=true})
  LinearInterpolate(
   B=Clamp(None=Multiply(A=@n6) {constB=384})
   Alpha=LinearInterpolate(
    A=ScalarParameter() {defaultValue=0.75, parameterName=AlphaBottom}
    B=ScalarParameter() {defaultValue=1, parameterName=AlphaTop}
    Alpha=@n3.RadialGradientExponential
   )
  )
  MaterialFunctionCall(UVs=TextureCoordinate(), CenterPosition=Constant2Vector() {g=1, r=1}, Radius=@n1, Density=@n2) {materialFunction=&RadialGradientExponential}
 ]
}
