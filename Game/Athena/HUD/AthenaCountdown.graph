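# HUD countdown widget for Athena matches, showing remaining seconds and timed final-phase warnings with animation and sound.
# Countdown updates initialize the display, switch from normal to final tick effects at a threshold, and start a final loop whose pitch and volume follow curves of interpolated time.
# Warning events animate two text lines in, then out after a delay.
# On construct, playlist gameplay tags look up a layout offset.
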WidgetBlueprint /Game/Athena/HUD/AthenaCountdown
animations [
 NormalTick {
  binding_count=5
  display_label=""
  length_seconds=0.8
  object=&self:NormalTick
  sequence {
   bindings [
    Image_RedBG {
     id=#1
     possessed_class=Image
     tracks [{
       type=MovieSceneFloatTrack
       display_name=Render Opacity
       sections [{type=MovieSceneFloatSection, channels [None {type=MovieSceneScriptingFloatChannel, keys [[3,0],[11,0.0952],[50.001,0]]}]}]
      }]
    }
    VerticalBox_FinalPhase {
     id=#2
     possessed_class=VerticalBox
     tracks [{
       type=MovieSceneFloatTrack
       display_name=Render Opacity
       sections [{type=MovieSceneFloatSection, channels [None {type=MovieSceneScriptingFloatChannel, const [0,1]}]}]
      }]
    }
    TextBlock_Countdown {
     id=#3
     possessed_class=TextBlock
     tracks [
      {
       type=MovieScene2DTransformTrack
       sections [{
         type=MovieScene2DTransformSection
         channels [
          Translation.X {type=MovieSceneScriptingFloatChannel}
          Translation.Y {type=MovieSceneScriptingFloatChannel}
          Angle {type=MovieSceneScriptingFloatChannel}
          Scale.X {type=MovieSceneScriptingFloatChannel, keys [[3,0.5009],[4,0.4009],[14,0.5261],[21,0.5009]]}
          Scale.Y {type=MovieSceneScriptingFloatChannel, keys [[3,0.5009],[4,0.4009],[14,0.5261],[21,0.5009]]}
          Shear.X {type=MovieSceneScriptingFloatChannel}
          Shear.Y {type=MovieSceneScriptingFloatChannel}
         ]
        }]
      }
      {
       type=MovieSceneFloatTrack
       display_name=Render Opacity
       sections [{type=MovieSceneFloatSection, channels [None {type=MovieSceneScriptingFloatChannel, const [1,2]}]}]
      }
     ]
    }
    OverlayCountdownNumber {
     id=#4
     possessed_class=Overlay
     tracks [{
       type=MovieSceneFloatTrack
       display_name=Render Opacity
       sections [{type=MovieSceneFloatSection, channels [None {type=MovieSceneScriptingFloatChannel, const [1,1]}]}]
      }]
    }
    Image_YellowBG {
     id=#5
     possessed_class=Image
     tracks [{
       type=MovieSceneFloatTrack
       display_name=Render Opacity
       sections [{type=MovieSceneFloatSection, channels [None {type=MovieSceneScriptingFloatChannel, keys [[3,0],[7,0.1184],[50.001,0]]}]}]
      }]
    }
   ]
   display_rate {numerator=60}
  }
 }
 FinalTick {
  binding_count=6
  display_label=""
  length_seconds=0.952
  object=&self:FinalTick
  sequence {bindings [
    Image_YellowBG {
     id=#6
     possessed_class=Image
     tracks [{
       type=MovieSceneFloatTrack
       display_name=Render Opacity
       sections [{
         type=MovieSceneFloatSection
         end_frame=19
         start_frame=0
         channels [None {type=MovieSceneScriptingFloatChannel, keys [[0,0.1624],[4,0.9232],[13,0]]}]
        }]
      }]
    }
    Image_Shade {
     id=#7
     possessed_class=Image
     tracks [{
       type=MovieSceneFloatTrack
       display_name=Render Opacity
       sections [{type=MovieSceneFloatSection, end_frame=19, start_frame=0, channels [None {type=MovieSceneScriptingFloatChannel, const [0,1]}]}]
      }]
    }
    OverlayCountdownNumber {
     id=#8
     possessed_class=Overlay
     tracks [{
       type=MovieSceneFloatTrack
       display_name=Render Opacity
       sections [{type=MovieSceneFloatSection, end_frame=19, start_frame=0, channels [None {type=MovieSceneScriptingFloatChannel, const [1,1]}]}]
      }]
    }
    TextBlock_Countdown {
     id=#9
     possessed_class=TextBlock
     tracks [
      {
       type=MovieScene2DTransformTrack
       sections [{
         type=MovieScene2DTransformSection
         end_frame=19
         start_frame=0
         channels [
          Translation.X {type=MovieSceneScriptingFloatChannel}
          Translation.Y {type=MovieSceneScriptingFloatChannel}
          Angle {type=MovieSceneScriptingFloatChannel}
          Scale.X {type=MovieSceneScriptingFloatChannel, keys [[0,0.667],[2,1.0376],[19,0.3]]}
          Scale.Y {type=MovieSceneScriptingFloatChannel, keys [[0,0.667],[2,1.0376],[19,0.3]]}
          Shear.X {type=MovieSceneScriptingFloatChannel}
          Shear.Y {type=MovieSceneScriptingFloatChannel}
         ]
        }]
      }
      {
       type=MovieSceneFloatTrack
       display_name=Render Opacity
       sections [{type=MovieSceneFloatSection, end_frame=19, start_frame=0, channels [None {type=MovieSceneScriptingFloatChannel, keys [[1,1],[19.0003,0]]}]}]
      }
     ]
    }
    Image_RedBG {
     id=#10
     possessed_class=Image
     tracks [{
       type=MovieSceneFloatTrack
       display_name=Render Opacity
       sections [{type=MovieSceneFloatSection, end_frame=19, start_frame=0, channels [None {type=MovieSceneScriptingFloatChannel, keys [[0,1],[19.0003,0]]}]}]
      }]
    }
    VerticalBox_FinalPhase {
     id=#11
     possessed_class=VerticalBox
     tracks [{
       type=MovieSceneFloatTrack
       display_name=Render Opacity
       sections [{type=MovieSceneFloatSection, end_frame=19, start_frame=0, channels [None {type=MovieSceneScriptingFloatChannel, const [0,1]}]}]
      }]
    }
   ]}
 }
 FinalCountdown {
  binding_count=3
  display_label=""
  length_seconds=5.5283
  object=&self:FinalCountdown
  sequence {
   bindings [
    CommonTextBlock_0 {
     id=#12
     possessed_class=CommonTextBlock
     tracks [
      {
       type=MovieScene2DTransformTrack
       sections [{
         type=MovieScene2DTransformSection
         end_frame=331
         start_frame=0
         channels [
          Translation.X {type=MovieSceneScriptingFloatChannel, const [0,1]}
          Translation.Y {type=MovieSceneScriptingFloatChannel, const [0,1]}
          Angle {type=MovieSceneScriptingFloatChannel, const [0,1]}
          Scale.X {type=MovieSceneScriptingFloatChannel, keys [[0,1],[9,0.3],[12,0.3445],[18,0.3123],[27,0.3289],[300.001,0.3],[306,0.3105],[315,0]]}
          Scale.Y {type=MovieSceneScriptingFloatChannel, keys [[0,1],[9,0.3],[12,0.3445],[18,0.3123],[27,0.3289],[300.001,0.3],[306,0.3105],[315,0]]}
          Shear.X {type=MovieSceneScriptingFloatChannel, const [0,1]}
          Shear.Y {type=MovieSceneScriptingFloatChannel, const [0,1]}
         ]
        }]
      }
      {
       type=MovieSceneFloatTrack
       display_name=Render Opacity
       sections [{
         type=MovieSceneFloatSection
         end_frame=331
         start_frame=0
         channels [None {type=MovieSceneScriptingFloatChannel, keys [[0,0],[4.252,1],[300.001,1],[322.701,0]]}]
        }]
      }
     ]
    }
    VerticalBox_FinalPhase {
     id=#13
     possessed_class=VerticalBox
     tracks [{
       type=MovieSceneFloatTrack
       display_name=Render Opacity
       sections [{type=MovieSceneFloatSection, end_frame=331, start_frame=0, channels [None {type=MovieSceneScriptingFloatChannel, const [0,1]}]}]
      }]
    }
    OverlayCountdownNumber {
     id=#14
     possessed_class=Overlay
     tracks [{
       type=MovieSceneFloatTrack
       display_name=Render Opacity
       sections [{type=MovieSceneFloatSection, end_frame=331, start_frame=0, channels [None {type=MovieSceneScriptingFloatChannel, const [1,1]}]}]
      }]
    }
   ]
   display_rate {numerator=60}
  }
 }
 FinalPhaseStart {
  binding_count=4
  display_label=""
  length_seconds=0.4584
  object=&self:FinalPhaseStart
  sequence {
   bindings [
    VerticalBox_FinalPhase {
     id=#15
     possessed_class=VerticalBox
     tracks [{
       type=MovieSceneFloatTrack
       display_name=Render Opacity
       sections [{type=MovieSceneFloatSection, end_frame=55, start_frame=0, channels [None {type=MovieSceneScriptingFloatChannel, keys [[0,0],[52.178,1]]}]}]
      }]
    }
    WarningTextBoxSecondLine {
     id=#16
     possessed_class=CommonTextBlock
     tracks [{
       type=MovieScene2DTransformTrack
       sections [{
         type=MovieScene2DTransformSection
         end_frame=55
         start_frame=0
         channels [
          Translation.X {type=MovieSceneScriptingFloatChannel, const [0,1]}
          Translation.Y {type=MovieSceneScriptingFloatChannel, const [0,1]}
          Angle {type=MovieSceneScriptingFloatChannel}
          Scale.X {type=MovieSceneScriptingFloatChannel, keys [[7,2],[25,0.8534],[34,1.024],[49.998,1]]}
          Scale.Y {type=MovieSceneScriptingFloatChannel, keys [[7,2],[25,0.8512],[25,0.8534],[34,1.024],[49.998,1]]}
          Shear.X {type=MovieSceneScriptingFloatChannel}
          Shear.Y {type=MovieSceneScriptingFloatChannel}
         ]
        }]
      }]
    }
    WarningTextBoxFirstLine {
     id=#17
     possessed_class=CommonTextBlock
     tracks [{
       type=MovieScene2DTransformTrack
       sections [{
         type=MovieScene2DTransformSection
         end_frame=55
         start_frame=0
         channels [
          Translation.X {type=MovieSceneScriptingFloatChannel, const [0,1]}
          Translation.Y {type=MovieSceneScriptingFloatChannel, const [0,1]}
          Angle {type=MovieSceneScriptingFloatChannel}
          Scale.X {type=MovieSceneScriptingFloatChannel, keys [[0.002,2],[18,0.665],[27,1.0729],[43,1]]}
          Scale.Y {type=MovieSceneScriptingFloatChannel, keys [[0.002,2],[18,0.665],[27,1.0729],[43,1]]}
          Shear.X {type=MovieSceneScriptingFloatChannel}
          Shear.Y {type=MovieSceneScriptingFloatChannel}
         ]
        }]
      }]
    }
    OverlayCountdownNumber {
     id=#18
     possessed_class=Overlay
     tracks [{
       type=MovieSceneFloatTrack
       display_name=Render Opacity
       sections [{type=MovieSceneFloatSection, end_frame=55, start_frame=0, channels [None {type=MovieSceneScriptingFloatChannel, const [0,1]}]}]
      }]
    }
   ]
   display_rate {numerator=120}
  }
 }
 FinalPhaseOutro {
  binding_count=4
  display_label=""
  length_seconds=0.25
  object=&self:FinalPhaseOutro
  sequence {
   bindings [
    OverlayCountdownNumber {
     id=#19
     possessed_class=Overlay
     tracks [{
       type=MovieSceneFloatTrack
       display_name=Render Opacity
       sections [{type=MovieSceneFloatSection, end_frame=30, start_frame=0, channels [None {type=MovieSceneScriptingFloatChannel, const [0,1]}]}]
      }]
    }
    VerticalBox_FinalPhase {
     id=#20
     possessed_class=VerticalBox
     tracks [{
       type=MovieSceneFloatTrack
       display_name=Render Opacity
       sections [{type=MovieSceneFloatSection, end_frame=30, start_frame=0, channels [None {type=MovieSceneScriptingFloatChannel, keys [[0,1],[27.48,0]]}]}]
      }]
    }
    WarningTextBoxSecondLine {
     id=#21
     possessed_class=CommonTextBlock
     tracks [{
       type=MovieScene2DTransformTrack
       sections [{
         type=MovieScene2DTransformSection
         end_frame=30
         start_frame=0
         channels [
          Translation.X {type=MovieSceneScriptingFloatChannel, keys [[0,0],[29.998,-90]]}
          Translation.Y {type=MovieSceneScriptingFloatChannel}
          Angle {type=MovieSceneScriptingFloatChannel}
          Scale.X {type=MovieSceneScriptingFloatChannel}
          Scale.Y {type=MovieSceneScriptingFloatChannel}
          Shear.X {type=MovieSceneScriptingFloatChannel}
          Shear.Y {type=MovieSceneScriptingFloatChannel}
         ]
        }]
      }]
    }
    WarningTextBoxFirstLine {
     id=#22
     possessed_class=CommonTextBlock
     tracks [{
       type=MovieScene2DTransformTrack
       sections [{
         type=MovieScene2DTransformSection
         end_frame=30
         start_frame=0
         channels [
          Translation.X {type=MovieSceneScriptingFloatChannel, keys [[0,0],[29.998,90]]}
          Translation.Y {type=MovieSceneScriptingFloatChannel}
          Angle {type=MovieSceneScriptingFloatChannel}
          Scale.X {type=MovieSceneScriptingFloatChannel}
          Scale.Y {type=MovieSceneScriptingFloatChannel}
          Shear.X {type=MovieSceneScriptingFloatChannel}
          Shear.Y {type=MovieSceneScriptingFloatChannel}
         ]
        }]
      }]
    }
   ]
   display_rate {numerator=120}
  }
 }
]
graphs [
 CheckTagsForWidgetOffset {nodes [
   n1 = CheckTagsForWidgetOffset!() -> (then, GameplayTags:Array of Gameplay Tag Structures) {is=title K2Node_FunctionEntry}
   > n2 = For Each Loop with Break!(Exec:Exec, Array:Array of Gameplay Tag Structures=@n1, Break:Exec=@n4.then) -> (LoopBody:Exec, Array Element:Gameplay Tag Structure, Array Index:int, Completed:Exec)
   n3 = Find(
    TargetMap:Map of Gameplay Tag Structures to Vector 2D Structures=Get PlaylistOffsetMap() -> (PlaylistOffsetMap:Map of Gameplay Tag Structures to Vector 2D Structures)
    Key:"Gameplay Tag Structure (by ref)"=@n2
   ) -> (Value:Vector 2D Structure)
   > Branch(Condition=@n3.ReturnValue)
   > n4 = SetRenderTranslation(self=Get CountdownVerticalBox() -> (CountdownVerticalBox:Vertical Box Object Reference), Translation=@n3)
  ]}
 EventGraph {nodes [
   n1 = Set Countdown Loop Audio Enabled!() -> (OutputDelegate, then, Enabled:bool=~) {is=title K2Node_CustomEvent}
   > n2 = Set CountdownLoopAudioEnabled(CountdownLoopAudioEnabled:bool=@n1.Enabled) -> (Output_Get:bool)
   > n3 = Branch(Condition=@n2)
   > n4 = Get!(execute:Exec) -> (then:Exec, else:Exec, TickLoopAudioComp:Audio Component Object Reference)
   n5 = SpawnSound2D!(
    execute=@n4.else
    Sound:Sound Base Object Reference=Get SoundOnTickLoop() -> (SoundOnTickLoop:Sound Base Object Reference)
    VolumeMultiplier:float=1
    PitchMultiplier:float=1
    StartTime:float=0
    ConcurrencySettings:Sound Concurrency Object Reference
    bPersistAcrossLevelTransition:bool=false
    bAutoDestroy:bool=true
   ) -> (then, ReturnValue:Audio Component Object Reference)
   > Set TickLoopAudioComp(TickLoopAudioComp:Audio Component Object Reference=@n5) -> (Output_Get:Audio Component Object Reference)
   n6 = Get(execute=@n3.else)
   > FadeOut!(execute, self:Audio Component Object Reference=@n6, FadeOutDuration:float=1, FadeVolumeLevel:float=0, FadeCurve:EAudioFaderCurve Enum=Linear) -> (then)
   n7 = InitializeCountdown!() -> (OutputDelegate, then, TimeRemaining:int, DisplayText:Text) {is=title K2Node_CustomEvent}
   > n8 = Branch(Condition=Get bCountdownStarted() -> (bCountdownStarted:bool))
   Set bCountdownStarted(execute=@n8.else, bCountdownStarted:bool=true) -> (Output_Get:bool)
   > SetVisibility(InVisibility=Visible)
   > "SetText (Text)"(self=Get TextBlock_Countdown() -> (TextBlock_Countdown:Text Object Reference), InText=@n7.DisplayText)
   > n9 = Branch(Condition=GetShouldDisplayCountdownTimerText!(self:Fort Game State Athena Object Reference=GetGameStateAthena!() -> (ReturnValue:Fort Game State Athena Object Reference)) -> (ReturnValue:bool))
   > n10 = PlayAnimation(InAnimation=Get FinalCountdown() -> (FinalCountdown:Widget Animation Object Reference))
   PlaySound2D(execute=[@n10.then, @n9.else], Sound=Get SoundOnFinal60Countdown() -> (SoundOnFinal60Countdown:Sound Base Object Reference))
   > n11 = SpawnSound2D(
    Sound=Get SoundOnFinal60Countdown_Music() -> (SoundOnFinal60Countdown_Music:Sound Base Object Reference)
    VolumeMultiplier=1
    PitchMultiplier=1
    StartTime=0
    bPersistAcrossLevelTransition=false
    bAutoDestroy=true
   )
   > Set Final 60 Countdown Music Instance(Final 60 Countdown Music Instance:Audio Component Object Reference=@n11) -> (Output_Get:Audio Component Object Reference)
   Event Construct()
   > "SetText (Text)"(self=Get TextBlock_Countdown())
   > SetVisibility(InVisibility=Collapsed)
   > n12 = PlaylistDataReadyAsync!(execute:Exec) -> (
    then:Exec
    OnReady:Exec
    GameState:Fort Game State Athena Object Reference
    Playlist:Fort Playlist Object Reference
    PlaylistContextTags:"Gameplay Tag Container Structure (by ref)"
   ) {is=title K2Node_AsyncAction}
   n13 = Handle Offset Translation on Gameplay Tags!(
    Exec:Exec=@n12.OnReady
    PlaylistGameplayTags:Array of Gameplay Tag Structures=BreakGameplayTagContainer!(GameplayTagContainer:"Gameplay Tag Container Structure (by ref)"=@n12.PlaylistContextTags) -> (GameplayTags:Array of Gameplay Tag Structures)
    PlaylistOffsetMap:Map of Gameplay Tag Structures to Vector 2D Structures=Get PlaylistOffsetMap()
   ) -> (OnOffsetFound:Exec, OnCompleted:Exec, TranslationOffset:Vector 2D Structure, FoundTag:Gameplay Tag Structure) {is=title K2Node_MacroInstance}
   > SetRenderTranslation(self=Get CountdownVerticalBox(), Translation=@n13)
   n14 = Event OnDisplayWarningMessage!() -> (OutputDelegate, then, FirstLineText:"Text (by ref)", SecondLineText:"Text (by ref)") {is=title K2Node_Event}
   > SetVisibility(InVisibility=Visible)
   > "SetText (Text)"(self=Get WarningTextBoxFirstLine() -> (WarningTextBoxFirstLine:Common Text Object Reference), InText=@n14.FirstLineText)
   > "SetText (Text)"(self=Get WarningTextBoxSecondLine() -> (WarningTextBoxSecondLine:Common Text Object Reference), InText=@n14.SecondLineText)
   > PlaySound2D(Sound=Get SoundOnFinalCountdownWarning_Music() -> (SoundOnFinalCountdownWarning_Music:Sound Base Object Reference))
   > PlayAnimation(InAnimation=Get FinalPhaseStart() -> (FinalPhaseStart:Widget Animation Object Reference))
   > Delay(Duration=5)
   > PlayAnimation(InAnimation=Get FinalPhaseOutro() -> (FinalPhaseOutro:Widget Animation Object Reference))
   > SetVisibility(InVisibility=Collapsed)
   n15 = Event OnCountdownFinished() {is=title K2Node_Event}
   SetVisibility(execute=[@n15.then, @n29.then], InVisibility=Collapsed)
   > Set Countdown Loop Audio Enabled!(execute, self, Enabled:bool=false) -> (then) {is=title K2Node_CallFunction}
   > Set bCountdownStarted(bCountdownStarted=false)
   n16 = Event Tick()
   > Parent: Tick!(execute, MyGeometry:Geometry Structure=@n16.MyGeometry, InDeltaTime:float=@n16.InDeltaTime) -> (then) {is=title K2Node_CallParentFunction}
   n17 = Get TickLoopAudioComp() -> (TickLoopAudioComp:Audio Component Object Reference)
   > Branch(Condition=AND Boolean(A=Get CountdownLoopAudioEnabled() -> (CountdownLoopAudioEnabled:bool), B=IsValid(Object=@n17)))
   n18 = int - int(
    A=Get FinalCountdownStartValue() -> (FinalCountdownStartValue:int=0)
    B="Min (Integer)"(A=Get CurrentTimeRemaining() -> (CurrentTimeRemaining:int=0), B=Get FinalCountdownStartValue() -> (FinalCountdownStartValue=0)) -> (ReturnValue=0)
   ) {is=title K2Node_CallFunction}
   > n19 = Set InterpolatedTimeRemaining(InterpolatedTimeRemaining:double=FInterpTo(
     Current=Get InterpolatedTimeRemaining() -> (InterpolatedTimeRemaining:double)
     Target="To Float (Integer)"(InInt=@n18)
     DeltaTime=GetWorldDeltaSeconds!() -> (ReturnValue:double)
     InterpSpeed=5
    )) -> (Output_Get:double)
   > n20 = Sequence()
   n21 = MapRangeClamped(Value=@n19, InRangeA=0, InRangeB=10)
   n22 = Get TickLoopAudioComp()
   > SetPitchMultiplier!(
    execute
    self:Audio Component Object Reference=@n22
    NewPitchMultiplier:float=GetFloatValue!(self:Curve Float Object Reference=Get CurveCountdownLoopPitch() -> (CurveCountdownLoopPitch:Curve Float Object Reference), InTime:float=@n21) -> (ReturnValue:float)
   ) -> (then)
   SetVolumeMultiplier!(
    execute=@n20.then_1
    self:Audio Component Object Reference=@n22
    NewVolumeMultiplier:float=GetFloatValue(self=Get CurveCountdownLoopVolume() -> (CurveCountdownLoopVolume:Curve Float Object Reference), InTime=@n21)
   ) -> (then)
   n23 = Event OnCountdownUpdate!() -> (OutputDelegate, then, TimeRemaining:int=0, CountdownUpdateText:"Text (by ref)") {is=title K2Node_Event}
   > n24 = Branch(Condition=Get bCountdownStarted())
   n25 = Set CurrentTimeRemaining(execute=[@n28.then_1, @n24.then], CurrentTimeRemaining:int=@n23.TimeRemaining) -> (Output_Get:int=0)
   > "SetText (Text)"(self=Get TextBlock_Countdown(), InText="To Text (Integer)"(Value=@n25))
   n26 = "integer > integer"(A=@n25, B=Get FinalCountdownStartValue() -> (FinalCountdownStartValue=0)) {is=title K2Node_CallFunction}
   > PlayAnimation(InAnimation=Select(
     Option 0:Widget Animation Object Reference=Get FinalTick() -> (FinalTick:Widget Animation Object Reference)
     Option 1:Widget Animation Object Reference=Get NormalTick() -> (NormalTick:Widget Animation Object Reference)
     Index=@n26
    ) -> (ReturnValue:Widget Animation Object Reference))
   > n27 = Sequence()
   K2Node_Composite_0 = Should we play a tick sound!() -> (ReturnValue:bool=~) {is=title K2Node_Composite}
   > Branch(Condition=@K2Node_Composite_0)
   > PlaySound2D(Sound=Select(
     Option 0:Sound Base Object Reference=Get SoundOnTickFinal() -> (SoundOnTickFinal:Sound Base Object Reference)
     Option 1:Sound Base Object Reference=Get SoundOnTickNormal() -> (SoundOnTickNormal:Sound Base Object Reference)
     Index=@n26
    ) -> (ReturnValue:Sound Base Object Reference))
   Branch(
    execute=@n27.then_1
    Condition="Equal (Integer)"(A=Get CurrentTimeRemaining() -> (CurrentTimeRemaining=0), B=Get FinalCountdownStartValue() -> (FinalCountdownStartValue=0)) {is=title K2Node_CallFunction}
   )
   > Set Countdown Loop Audio Enabled(Enabled=true)
   > PlaySound2D(Sound=Get SoundOnFinal10Countdown() -> (SoundOnFinal10Countdown:Sound Base Object Reference))
   > FadeOut(
    self=Get Final 60 Countdown Music Instance() -> (Final 60 Countdown Music Instance:Audio Component Object Reference)
    FadeOutDuration=Get Final 60 Countdown Music Fadeout() -> (Final 60 Countdown Music Fadeout:double)
    FadeVolumeLevel=0
    FadeCurve=Linear
   )
   > PlaySound2D(Sound=Get SoundOnFinal10Countdown_Music() -> (SoundOnFinal10Countdown_Music:Sound Base Object Reference))
   n28 = Sequence(execute=@n24.else)
   > InitializeCountdown!(execute, self, TimeRemaining:int=@n23.TimeRemaining, DisplayText:Text=@n23.CountdownUpdateText) -> (then) {is=title K2Node_CallFunction}
   n29 = Event OnCountdownStopped() {is=title K2Node_Event}
   Event PreConstruct()
  ]}
 Should we play a tick sound {
  path=EventGraph.K2Node_Composite_0.Should we play a tick sound
  nodes [
   Inputs!()
   n1 = Get CurrentTimeRemaining() -> (CurrentTimeRemaining=0)
   n2 = Get FinalCountdownStartValue() -> (FinalCountdownStartValue=0)
   n3 = "Equal (Integer)"(A="% (Integer)"(A=Get CurrentTimeRemaining() -> (CurrentTimeRemaining=0), B=10) -> (ReturnValue=0), B=0)
   n4 = "integer <= integer"!(A:int=Get CurrentTimeRemaining() -> (CurrentTimeRemaining=0), B:int=30) -> (ReturnValue:bool) {is=title K2Node_CallFunction}
   Outputs!(ReturnValue:bool=AND Boolean(A=OR Boolean(A=@n4, B=@n3), B="Not Equal (Integer)"!(A:int=@n1, B:int=@n2) -> (ReturnValue:bool) {is=title K2Node_CallFunction}))
  ]
 }
]
variables [
 bCountdownStarted {type=boolean}
 CountdownLoopAudioEnabled {type=boolean}
 CurrentTimeRemaining {type=integer}
 CurveCountdownLoopPitch {type=CurveFloat, default=/Game/Athena/Sounds/LTMs/Score/C_CountdownLTM_TickLoop_Pitch}
 CurveCountdownLoopVolume {type=CurveFloat, default=/Game/Athena/Sounds/LTMs/Score/C_CountdownLTM_TickLoop_Volume}
 Final 60 Countdown Music Fadeout {type=number}
 Final 60 Countdown Music Instance {type=AudioComponent, default=null}
 FinalCountdownStartValue {type=integer, default=10}
 InterpolatedTimeRemaining {type=number}
 PlaylistOffsetMap {default {
   "(TagName=\"Athena.Playlist.Argon\")"=<0 120>
   "(TagName=\"Athena.Playlist.CobaltRespawn\")"=<0 70>
   "(TagName=\"Athena.Playlist.Score\")"=<0 70>
   "(TagName=\"Athena.Playlist.UraniumP\")"=<0 120>
  }}
 RelevantPlaylistTags {type=GameplayTagContainer, default {gameplayTags []}}
 SoundOnFinal10Countdown {type=SoundBase, default=/Game/Athena/Sounds/LTMs/Score/CountdownLTM_Final10Start_Cue}
 SoundOnFinal10Countdown_Music {type=SoundBase, default=null}
 SoundOnFinal60Countdown {type=SoundBase, default=/Game/Athena/Sounds/LTMs/Score/CountdownLTM_Final60Start_Cue}
 SoundOnFinal60Countdown_Music {type=SoundBase, default=null}
 SoundOnFinalCountdownWarning_Music {type=SoundBase, default=/Game/Athena/Sounds/LTMs/Score/ScoreLTM_FinalCountdown_Warning_Cue}
 SoundOnTickFinal {type=SoundBase, default=/Game/Athena/Sounds/LTMs/Score/CountdownLTM_FinalTick_Cue}
 SoundOnTickLoop {type=SoundBase, default=/Game/Athena/Sounds/LTMs/Score/CountdownLTM_TickLoop_Cue}
 SoundOnTickNormal {type=SoundBase, default=/Game/Athena/Sounds/LTMs/Score/CountdownLTM_NormalTick_Cue}
 TickLoopAudioComp {type=AudioComponent, default=null}
 VerticalOffset {type=Vector2D, default=<0 0>}
]
widgets {
 info {inheritedWidgetCount=1, parentClass=/Script/FortniteUI.AthenaCountdownWidgetBase, rootWidgetClass=&SizeBox, widgetCount=14}
 widgets [
  SizeBox_0 {
   widgetClassPath=&SizeBox
   widget=SizeBox_0
   properties {bOverride_MinDesiredHeight=true, bOverride_MinDesiredWidth=true, minDesiredHeight=100, minDesiredWidth=600}
  }
  CountdownVerticalBox {
   widgetClassPath=&VerticalBox
   bIsVariable=true
   parent=SizeBox_0
   widget=CountdownVerticalBox
   slot {class=SizeBoxSlot, properties {horizontalAlignment=HAlign_Center, verticalAlignment=VAlign_Top}}
  }
  OverlayCountdownNumber {
   widgetClassPath=&Overlay
   parent=CountdownVerticalBox
   widget=OverlayCountdownNumber
   slot {class=VerticalBoxSlot, properties {horizontalAlignment=HAlign_Center}}
  }
  Overlay_1 {
   widgetClassPath=&Overlay
   parent=OverlayCountdownNumber
   widget=Overlay_1
   slot {class=OverlaySlot, properties {horizontalAlignment=HAlign_Center, verticalAlignment=VAlign_Center, padding=<ltrb -200 -200 -200 -200>}}
  }
  Image_Shade {
   widgetClassPath=&Image
   parent=Overlay_1
   widget=Image_Shade
   properties {
    brush {
     bIsDynamicallyLoaded=false
     drawAs=Image
     imageType=NoImage
     mirroring=NoMirror
     resourceName=None
     resourceObject=/Game/Athena/HUD/Materials/M_CountdownBG_Alpha_Inst
     tiling=NoTile
     imageSize=<482.0438 213.1235>
     margin=<ltrb 0 0 0 0>
     outlineSettings {
      bUseBrushTransparency=false
      roundingType=HalfHeightRadius
      width=0
      color {colorUseRule=UseColor_Specified, specifiedColor=<rgba 0 0 0 0>}
      cornerRadii=<0 0 0 0>
     }
     tintColor {colorUseRule=UseColor_Specified, specifiedColor=<rgba 0.0156 0.0063 0.0063 1>}
     uVRegion {bIsValid=false, max=<0 0>, min=<0 0>}
    }
    colorAndOpacity=<rgba 1 1 1 0.2286>
   }
   slot {class=OverlaySlot, properties {horizontalAlignment=HAlign_Center, verticalAlignment=VAlign_Center}}
  }
  Image_RedBG {
   widgetClassPath=&Image
   parent=Overlay_1
   widget=Image_RedBG
   properties {
    brush {
     bIsDynamicallyLoaded=false
     drawAs=Image
     imageType=NoImage
     mirroring=NoMirror
     resourceName=None
     resourceObject=/Game/Athena/HUD/Materials/M_CountdownBG_Base
     tiling=NoTile
     imageSize=<514.2557 300>
     margin=<ltrb 0 0 0 0>
     outlineSettings {
      bUseBrushTransparency=false
      roundingType=HalfHeightRadius
      width=0
      color {colorUseRule=UseColor_Specified, specifiedColor=<rgba 0 0 0 0>}
      cornerRadii=<0 0 0 0>
     }
     tintColor {colorUseRule=UseColor_Specified, specifiedColor=<rgba 1 1 1 1>}
     uVRegion {bIsValid=false, max=<0 0>, min=<0 0>}
    }
    colorAndOpacity=<rgba 1 0 0 1>
   }
   slot {class=OverlaySlot}
  }
  Image_YellowBG {
   widgetClassPath=&Image
   parent=Overlay_1
   widget=Image_YellowBG
   properties {
    brush {
     bIsDynamicallyLoaded=false
     drawAs=Image
     imageType=NoImage
     mirroring=NoMirror
     resourceName=None
     resourceObject=/Game/Athena/HUD/Materials/M_CountdownBG_Core
     tiling=NoTile
     imageSize=<514.2557 300>
     margin=<ltrb 0 0 0 0>
     outlineSettings {
      bUseBrushTransparency=false
      roundingType=HalfHeightRadius
      width=0
      color {colorUseRule=UseColor_Specified, specifiedColor=<rgba 0 0 0 0>}
      cornerRadii=<0 0 0 0>
     }
     tintColor {colorUseRule=UseColor_Specified, specifiedColor=<rgba 1 1 1 1>}
     uVRegion {bIsValid=false, max=<0 0>, min=<0 0>}
    }
    colorAndOpacity=<rgba 1 0.2789 0 1>
   }
   slot {class=OverlaySlot}
  }
  TextBlock_Countdown {
   widgetClassPath=&TextBlock
   bIsVariable=true
   parent=OverlayCountdownNumber
   widget=TextBlock_Countdown
   properties {
    justification=Center
    text=60
    font {
     bForceMonospaced=false
     bMaterialIsStencil=false
     fontMaterial=null
     fontObject=&Burbank
     letterSpacing=0
     monospacedWidth=1
     size=120
     skewAmount=0
     typefaceFontName=BurbankRegular-Black
     outlineSettings {
      bApplyOutlineToDropShadows=true
      bMiteredCorners=false
      bSeparateFillAlpha=false
      outlineMaterial=null
      outlineSize=3
      outlineColor=<rgba 0.3385 0.0164 0 1>
     }
    }
    renderTransform {angle=0, scale=<0.5 0.5>, shear=<-4 0>, translation=<0 0>}
    shadowColorAndOpacity=<rgba 0.3385 0.0164 0 1>
    shadowOffset=<2.3891 6.0637>
   }
   slot {class=OverlaySlot, properties {horizontalAlignment=HAlign_Center, verticalAlignment=VAlign_Center, padding=<ltrb -100 -100 -100 -100>}}
  }
  Overlay_2 {
   widgetClassPath=&Overlay
   parent=CountdownVerticalBox
   widget=Overlay_2
   slot {class=VerticalBoxSlot, properties {padding=<ltrb -300 -30 -300 -200>}}
  }
  CommonTextBlock_0 {
   widgetClassPath=&CommonTextBlock
   parent=Overlay_2
   widget=CommonTextBlock_0
   properties {
    renderOpacity=0
    text=FINAL COUNTDOWN
    visibility=Collapsed
    font {
     bForceMonospaced=false
     bMaterialIsStencil=false
     fontMaterial=null
     fontObject=&Burbank
     letterSpacing=0
     monospacedWidth=1
     size=176
     skewAmount=0
     typefaceFontName=Burbank-Black
     outlineSettings {
      bApplyOutlineToDropShadows=false
      bMiteredCorners=false
      bSeparateFillAlpha=false
      outlineMaterial=null
      outlineSize=3
      outlineColor=<rgba 0.3385 0.0164 0 1>
     }
    }
    renderTransform {angle=0, scale=<0.3 0.3>, shear=<0 0>, translation=<0 0>}
    renderTransformPivot=<0.5 0.2>
    shadowColorAndOpacity=<rgba 0.0052 0.0052 0.0056 0.781>
    shadowOffset=<3 3>
   }
   slot {class=OverlaySlot}
  }
  VerticalBox_FinalPhase {
   widgetClassPath=&VerticalBox
   parent=Overlay_2
   widget=VerticalBox_FinalPhase
   properties {renderOpacity=0, renderTransform {angle=0, scale=<1 1>, shear=<0 0>, translation=<0 -50>}}
   slot {class=OverlaySlot, properties {horizontalAlignment=HAlign_Center}}
  }
  WarningTextBoxFirstLine {
   widgetClassPath=&CommonTextBlock
   bIsVariable=true
   parent=VerticalBox_FinalPhase
   widget=WarningTextBoxFirstLine
   properties {
    justification=Center
    text=FINAL PHASE BEGINS
    font {
     bForceMonospaced=false
     bMaterialIsStencil=false
     fontMaterial=null
     fontObject=&Burbank
     letterSpacing=0
     monospacedWidth=1
     size=30
     skewAmount=0
     typefaceFontName=Burbank-Black
     outlineSettings {
      bApplyOutlineToDropShadows=false
      bMiteredCorners=false
      bSeparateFillAlpha=false
      outlineMaterial=null
      outlineSize=1
      outlineColor=<rgba 0.3385 0.0164 0 1>
     }
    }
    renderTransformPivot=<0.5 1>
    shadowColorAndOpacity=<rgba 0.0052 0.0052 0.0056 0.781>
    shadowOffset=<3 3>
   }
   slot {class=VerticalBoxSlot, properties {horizontalAlignment=HAlign_Center, verticalAlignment=VAlign_Top, padding=<ltrb 0 0 0 -20>}}
  }
  WarningTextBoxSecondLine {
   widgetClassPath=&CommonTextBlock
   bIsVariable=true
   parent=VerticalBox_FinalPhase
   widget=WarningTextBoxSecondLine
   properties {
    justification=Center
    text=5 MINUTES
    colorAndOpacity {colorUseRule=UseColor_Specified, specifiedColor=<rgba 1 0.9411 0.7643 1>}
    font {
     bForceMonospaced=false
     bMaterialIsStencil=false
     fontMaterial=/Game/Athena/HUD/Materials/M_Countdown_Inst
     fontObject=&Burbank
     letterSpacing=0
     monospacedWidth=1
     size=40
     skewAmount=0
     typefaceFontName=Burbank-Black
     outlineSettings {
      bApplyOutlineToDropShadows=false
      bMiteredCorners=false
      bSeparateFillAlpha=false
      outlineMaterial=null
      outlineSize=1
      outlineColor=<rgba 0.3385 0.0164 0 1>
     }
    }
    renderTransform {angle=0, scale=<1 1>, shear=<-3 0>, translation=<0 0>}
    renderTransformPivot=<0.5 0>
    shadowColorAndOpacity=<rgba 0.0052 0.0052 0.0056 0.781>
    shadowOffset=<3 3>
   }
   slot {class=VerticalBoxSlot, properties {horizontalAlignment=HAlign_Center, verticalAlignment=VAlign_Top}}
  }
  TutorialHighlight {widgetClassPath=&TutorialHighlight}
 ]
}
