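# Athena HUD widget that shows an equipped item's durability as an icon beside a text readout.
# Conversion functions pick the normal or low-durability icon material against a threshold, and collapse visibility for negative or infinite durability.
# Another hides the icon when the item definition carries the HideDurabilityIcon gameplay tag, defaulting to visible for invalid items.
# The event graph is empty, so all logic lives in these functions.
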WidgetBlueprint /Game/Athena/HUD/AthenaEquippedItemDurabilityWidget
graphs [
 DurabilityToMaterial {nodes [
   n1 = DurabilityToMaterial!() -> (then, InDurability:double=~) {is=title K2Node_FunctionEntry}
   n2 = Get LowDurabilityThreshold() -> (LowDurabilityThreshold:double)
   > Return Node!(
    execute
    OutMaterialInstance:Material Instance Object Reference=SelectMaterialInstance!(
     A:Material Instance Object Reference=/Game/Athena/HUD/Art/DurabilityIcon/MI_UI_DurabilityComparison_Icon_Normal
     B:Material Instance Object Reference=/Game/Athena/HUD/Art/DurabilityIcon/MI_UI_LowDurability
     bSelectA:bool="float > float"(A=@n1, B=@n2)
    ) -> (ReturnValue:Material Instance Object Reference)
   )
  ]}
 DurabilityToVisibility {nodes [
   n1 = DurabilityToVisibility!() -> (then, InDurability:double=~, bOwnerHasInfiniteDurability:bool=~) {is=title K2Node_FunctionEntry}
   n2 = "float < float"(A=@n1)
   > Return Node!(
    execute
    OutVisibility:ESlateVisibility Enum=Select(Option 0=HitTestInvisible, Option 1=Collapsed, Index=OR Boolean(A=@n2, B=@n1.bOwnerHasInfiniteDurability))
   )
  ]}
 EventGraph {links [], nodes []}
 RepresentedItemToIconVisibility {nodes [
   n1 = RepresentedItemToIconVisibility!() -> (then, RepresentedItem:Fort World Item Object Reference) {is=title K2Node_FunctionEntry}
   > n2 = Is Valid(InputObject=@n1)
   n3 = GetItemDefinition!(self:Fort Item Object Reference=@n1) -> (ReturnValue:Fort Item Definition Object Reference)
   > n4 = Is Valid(InputObject=@n3)
   Return Node(execute=[@n4.Is Not Valid, @n2.Is Not Valid], OutVisibility=Visible)
   n5 = GetOwnedGameplayTags!(TagContainerInterface:Gameplay Tag Asset Interface Interface=@n3) -> (ReturnValue:Gameplay Tag Container Structure)
   > Return Node(OutVisibility=Select(
     Option 0=Visible
     Option 1=Collapsed
     Index=HasTag!(
      TagContainer:"Gameplay Tag Container Structure (by ref)"=@n5
      Tag:Gameplay Tag Structure=(TagName="Athena.HUD.EquippedItem.HideDurabilityIcon")
      bExactMatch:bool=false
     ) -> (ReturnValue:bool)
    ))
  ]}
]
variables [LowDurabilityThreshold {type=number, default=0.25}]
widgets {
 info {inheritedWidgetCount=0, parentClass=&UserWidget, rootWidgetClass=&Overlay, widgetCount=4}
 widgets [
  Overlay_20 {widgetClassPath=&Overlay, widget=Overlay_20, properties {visibility=HitTestInvisible}}
  HorizontalBox_0 {
   widgetClassPath=&HorizontalBox
   parent=Overlay_20
   widget=HorizontalBox_0
   slot {class=OverlaySlot, properties {horizontalAlignment=HAlign_Fill, verticalAlignment=VAlign_Fill}}
  }
  Icon {
   widgetClassPath=&Image
   bIsVariable=true
   parent=HorizontalBox_0
   widget=Icon
   properties {brush {
     bIsDynamicallyLoaded=false
     drawAs=Image
     imageType=FullColor
     mirroring=NoMirror
     resourceName=None
     resourceObject=&MI_UI_DurabilityComparison_Icon_Normal
     tiling=NoTile
     imageSize=<32 32>
     margin=<ltrb 0 0 0 0>
     outlineSettings {
      bUseBrushTransparency=false
      roundingType=HalfHeightRadius
      width=0
      color {colorUseRule=UseColor_Specified, specifiedColor=<rgba 0 0 0 0>}
      cornerRadii=<0 0 0 0>
     }
     tintColor {colorUseRule=UseColor_Specified, specifiedColor=<rgba 1 1 1 1>}
     uVRegion {bIsValid=false, max=<0 0>, min=<0 0>}
    }}
   slot {class=HorizontalBoxSlot, properties {horizontalAlignment=HAlign_Left, verticalAlignment=VAlign_Center, padding=<ltrb 0 0 8 0>}}
  }
  Percentaige {
   widgetClassPath=&CommonTextBlock
   bIsVariable=true
   parent=HorizontalBox_0
   widget=Percentaige
   properties {
    autoWrapText=true
    justification=Right
    text=20
    textTransformPolicy=ToUpper
    font {
     bForceMonospaced=false
     bMaterialIsStencil=false
     fontMaterial=null
     fontObject=&Burbank
     letterSpacing=0
     monospacedWidth=1
     size=22.5
     skewAmount=0
     typefaceFontName=BurbankRegular-Black
     outlineSettings {
      bApplyOutlineToDropShadows=true
      bMiteredCorners=false
      bSeparateFillAlpha=true
      outlineMaterial=&M-FontShadow
      outlineSize=2
      outlineColor=<rgba 0 0 0 0.5>
     }
    }
    shadowOffset=<0 0>
    shapedTextOptions {bOverride_TextFlowDirection=true, bOverride_TextShapingMethod=false, textFlowDirection=LeftToRight, textShapingMethod=Auto}
   }
   slot {class=HorizontalBoxSlot, properties {horizontalAlignment=HAlign_Right, verticalAlignment=VAlign_Center}}
  }
 ]
}
