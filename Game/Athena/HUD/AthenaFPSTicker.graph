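# AthenaFPSTicker is a Widget Blueprint for the Athena HUD that shows an FPS readout with min and max, plus a thermal warning.
# It stacks two right-aligned CommonTextBlocks, FPS and ThermalWarning, in a VerticalBox, with enlarged font scaling on mobile.
# Both text blocks bind to the native AthenaFPSBase parent as inherited widgets.
# The Blueprint supplies only layout and styling, as its event graph is empty.
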WidgetBlueprint /Game/Athena/HUD/AthenaFPSTicker
graphs [EventGraph {links [], nodes []}]
variables []
widgets {
 info {inheritedWidgetCount=2, parentClass=/Script/FortniteUI.AthenaFPSBase, rootWidgetClass=&VerticalBox, widgetCount=3}
 widgets [
  VerticalBox_106 {widgetClassPath=&VerticalBox, widget=VerticalBox_106}
  FPS {
   widgetClassPath=&CommonTextBlock
   bInherited=true
   bIsVariable=true
   parent=VerticalBox_106
   widget=FPS
   properties {
    justification=Right
    minDesiredWidth=200
    mobileFontSizeMultiplier=1.8
    text="120 FPS [ MIN,  MAX ]"
    font {
     bForceMonospaced=false
     bMaterialIsStencil=false
     fontMaterial=null
     fontObject=&NotoSans
     letterSpacing=0
     monospacedWidth=1
     size=10
     skewAmount=0
     typefaceFontName=NotoSans-Bold
     outlineSettings {
      bApplyOutlineToDropShadows=false
      bMiteredCorners=false
      bSeparateFillAlpha=false
      outlineMaterial=null
      outlineSize=0
      outlineColor=<rgba 0 0 0 1>
     }
    }
    shadowOffset=<0 0>
   }
   slot {class=VerticalBoxSlot}
  }
  ThermalWarning {
   widgetClassPath=&CommonTextBlock
   bInherited=true
   bIsVariable=true
   parent=VerticalBox_106
   widget=ThermalWarning
   properties {
    justification=Right
    minDesiredWidth=200
    mobileFontSizeMultiplier=1.8
    text=Device is too hot to run at desired FPS
    font {
     bForceMonospaced=false
     bMaterialIsStencil=false
     fontMaterial=null
     fontObject=&NotoSans
     letterSpacing=0
     monospacedWidth=1
     size=10
     skewAmount=0
     typefaceFontName=NotoSans-Bold
     outlineSettings {
      bApplyOutlineToDropShadows=false
      bMiteredCorners=false
      bSeparateFillAlpha=false
      outlineMaterial=null
      outlineSize=0
      outlineColor=<rgba 0 0 0 1>
     }
    }
    shadowOffset=<0 0>
   }
   slot {class=VerticalBoxSlot}
  }
 ]
}
