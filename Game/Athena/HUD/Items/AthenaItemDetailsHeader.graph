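# Battle Royale HUD item details header, built on AthenaItemDetailsHeaderBase, that shows item name, rarity, category, ammo, durability and cost.
# A widget switcher holds three layouts: full rarity header, short description and short item description.
# Construct asynchronously creates the third layout's cost widget when IsSpatialGameplayExperience passes; item change events call UpdateItemForWidgets.
# SetItemCost picks a cost component class by Creative/Playground playlist and forwards it to all three cost widgets.
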WidgetBlueprint /Game/Athena/HUD/Items/AthenaItemDetailsHeader
graphs [
 EventGraph {nodes [
   Event Construct()
   > Branch(Condition=IsSpatialGameplayExperience!(self:Fort Game State Athena Object Reference=GetGameStateAthena!() -> (ReturnValue:Fort Game State Athena Object Reference)) -> (ReturnValue:bool))
   > n1 = CreateWidgetAsync!(
    execute:Exec
    UserWidgetSoftClass:User Widget Soft Class Reference=/Game/UI/Creative/ItemCostWidget.ItemCostWidget_C
    OwningPlayer:Player Controller Object Reference
    bSuspendInputUntilComplete:bool=false
   ) -> (then:Exec, OnComplete:Exec, OnFailure:Exec, UserWidget:User Widget Object Reference) {is=title K2Node_AsyncAction}
   n2 = Cast To ItemCostWidget!(execute=@n1.OnComplete, Object=@n1) -> (then, CastFailed, AsItem Cost Widget:Item Cost Widget Object Reference) {is=title K2Node_DynamicCast}
   > n3 = Set ItemCostWidgetShortItem(ItemCostWidgetShortItem:Item Cost Widget Base Object Reference=@n2) -> (Output_Get:Item Cost Widget Base Object Reference)
   > AddChild(self=Get ItemCostContainer() -> (ItemCostContainer:Size Box Object Reference), Content=@n3)
   n4 = Event HandleItemToDetailChangedBP() {is=title K2Node_Event}
   UpdateItemForWidgets(execute=[@n4.then, @n5.then, @n6.then], self:Athena Item Details Header Base Object Reference)
   n5 = Event HandleDifferentItemToDetailSet() {is=title K2Node_Event}
   n6 = Event HandleDifferentItemToCompareSet() {is=title K2Node_Event}
  ]}
 SetItemCost {nodes [
   n1 = SetItemCost!() -> (then, Actor:Actor Object Reference) {is=title K2Node_FunctionEntry}
   > Set ItemCostClass(ItemCostClass:Item Cost Component Class Reference=Select(
     Option 0:Item Cost Component Class Reference=/Game/Creative/Devices/ItemSpawnerProp/ItemCostComponent.ItemCostComponent_C
     Option 1:Item Cost Component Class Reference=/Game/Creative/Devices/ItemSpawnerProp/CreativePickupItemCostComponent.CreativePickupItemCostComponent_C
     Index=IsCreativeOrPlaygroundModePlaylist!() -> (ReturnValue:bool)
    ) -> (ReturnValue:Item Cost Component Class Reference)) -> (Output_Get:Item Cost Component Class Reference)
   > n2 = Cast To ItemCostWidget(Object=Get ItemCostWidget() -> (ItemCostWidget:Item Cost Widget Base Object Reference))
   > n3 = SetItemCost!(
    execute
    self:Item Cost Widget Object Reference=@n2
    Actor:Actor Object Reference=@n1
    SpecifiedItemCostComponent:Item Cost Component Class Reference=Get ItemCostClass() -> (ItemCostClass:Item Cost Component Class Reference)
   ) -> (then) {is=title K2Node_CallFunction}
   n4 = Cast To ItemCostWidget(execute=[@n3.then, @n2.CastFailed], Object=Get ItemCostWidgetShort() -> (ItemCostWidgetShort:Item Cost Widget Base Object Reference))
   > n5 = SetItemCost(self=@n4, Actor=@n1, SpecifiedItemCostComponent=Get ItemCostClass())
   n6 = Cast To ItemCostWidget(execute=[@n5.then, @n4.CastFailed], Object=Get ItemCostWidgetShortItem() -> (ItemCostWidgetShortItem:Item Cost Widget Base Object Reference))
   > SetItemCost(self=@n6, Actor=@n1, SpecifiedItemCostComponent=Get ItemCostClass())
  ]}
]
variables []
widgets {
 info {inheritedWidgetCount=32, parentClass=/Script/FortniteUI.AthenaItemDetailsHeaderBase, rootWidgetClass=&VerticalBox, widgetCount=58}
 widgets [
  VerticalBox_1 {widgetClassPath=&VerticalBox, widget=VerticalBox_1}
  CommonWidgetSwitcher_Stack {
   widgetClassPath=&CommonWidgetSwitcherLegacy
   bInherited=true
   bIsVariable=true
   parent=VerticalBox_1
   widget=CommonWidgetSwitcher_Stack
   properties {activeWidgetIndex=1, transitionDuration=0}
   slot {class=VerticalBoxSlot}
  }
  RarityBorder {
   widgetClassPath=&CommonBorder
   bInherited=true
   bIsVariable=true
   parent=CommonWidgetSwitcher_Stack
   widget=RarityBorder
   properties {style=/FortUILibrary/Styles/Border/Border-ItemInfo-Header.Border-ItemInfo-Header_C, padding=<ltrb 0 0 0 0>}
   slot {class=WidgetSwitcherSlot}
  }
  DetailsContainerSlotWidget {
   widgetClassPath=&NamedSlot
   bInherited=true
   bIsVariable=true
   parent=RarityBorder
   widget=DetailsContainerSlotWidget
   properties {bExposeOnInstanceOnly=true}
   slot {class=BorderSlot, properties {padding=<ltrb 0 0 0 0>}}
  }
  Overlay_0 {widgetClassPath=&Overlay, parent=DetailsContainerSlotWidget, widget=Overlay_0, slot {class=PanelSlot}}
  VerticalBox_2 {
   widgetClassPath=&VerticalBox
   parent=Overlay_0
   widget=VerticalBox_2
   slot {class=OverlaySlot, properties {horizontalAlignment=HAlign_Fill, verticalAlignment=VAlign_Fill}}
  }
  HorizontalBox_4 {
   widgetClassPath=&HorizontalBox
   parent=VerticalBox_2
   widget=HorizontalBox_4
   slot {class=VerticalBoxSlot, properties {padding=<ltrb 16 8 16 8>}}
  }
  VerticalBox_0 {
   widgetClassPath=&VerticalBox
   parent=HorizontalBox_4
   widget=VerticalBox_0
   properties {clipping=OnDemand}
   slot {class=HorizontalBoxSlot, properties {size {sizeRule=Fill, value=1}}}
  }
  ItemDetailsHeaderItemDisplayText {
   widgetClassPath=/Game/UI/Foundation/Items/Details/ItemDetailsHeaderItemDisplayText
   parent=VerticalBox_0
   widget=ItemDetailsHeaderItemDisplayText
   slot {class=VerticalBoxSlot, properties {padding=<ltrb 0 -2 0 -2>}}
  }
  TagListBox {
   widgetClassPath=&HorizontalBox
   bInherited=true
   bIsVariable=true
   parent=VerticalBox_0
   widget=TagListBox
   slot {class=VerticalBoxSlot}
  }
  CategoryIndicator {
   widgetClassPath=/Script/FortniteUI.FortItemCategoryIndicator
   parent=TagListBox
   widget=CategoryIndicator
   properties {
    iconBrushSize=Medium
    labelTextStyle=&TextStyle-Base_C
    ordinalOfCategoryToRepresent=Secondary
    visibility=SelfHitTestInvisible
   }
   slot {class=HorizontalBoxSlot, properties {horizontalAlignment=HAlign_Center, verticalAlignment=VAlign_Center}}
  }
  IconSpacer {
   widgetClassPath=&Spacer
   bInherited=true
   bIsVariable=true
   parent=TagListBox
   widget=IconSpacer
   properties {visibility=Collapsed, size=<4 1>}
   slot {class=HorizontalBoxSlot}
  }
  LevelBar {
   widgetClassPath=&CommonBorder
   bInherited=true
   bIsVariable=true
   parent=VerticalBox_2
   widget=LevelBar
   properties {
    style=/FortUILibrary/Styles/Border/Border-ItemInfoHeader_BrightBar.Border-ItemInfoHeader_BrightBar_C
    background {
     bIsDynamicallyLoaded=false
     drawAs=NoDrawType
     imageType=NoImage
     mirroring=NoMirror
     resourceName=None
     resourceObject=/FortUILibrary/Styles/Border/Materials/M-Border-ItemInfoHeader_BrightBar
     tiling=NoTile
     imageSize=<32 32>
     margin=<ltrb 0 0 0 0>
     outlineSettings {
      bUseBrushTransparency=false
      roundingType=HalfHeightRadius
      width=0
      color {colorUseRule=UseColor_Specified, specifiedColor=<rgba 0 0 0 0>}
      cornerRadii=<0 0 0 0>
     }
     tintColor {colorUseRule=UseColor_Specified, specifiedColor=<rgba 1 1 1 1>}
     uVRegion {bIsValid=false, max=<0 0>, min=<0 0>}
    }
    padding=<ltrb 0 0 0 0>
   }
   slot {class=VerticalBoxSlot}
  }
  VerticalBox_3 {widgetClassPath=&VerticalBox, parent=LevelBar, widget=VerticalBox_3, slot {class=BorderSlot, properties {padding=<ltrb 0 0 0 0>}}}
  SizeBox_0 {
   widgetClassPath=&SizeBox
   parent=VerticalBox_3
   widget=SizeBox_0
   properties {bOverride_HeightOverride=true, heightOverride=40}
   slot {class=VerticalBoxSlot, properties {padding=<ltrb 16 0 16 0>}}
  }
  Overlay_1 {widgetClassPath=&Overlay, parent=SizeBox_0, widget=Overlay_1, slot {class=SizeBoxSlot}}
  RatingSwitcher {
   widgetClassPath=&WidgetSwitcher
   bInherited=true
   bIsVariable=true
   parent=Overlay_1
   widget=RatingSwitcher
   slot {class=OverlaySlot, properties {verticalAlignment=VAlign_Center}}
  }
  TierIndicator {
   widgetClassPath=/Script/FortniteUI.FortTierIndicator
   bInherited=true
   bIsVariable=true
   parent=Overlay_1
   widget=TierIndicator
   properties {brushSize=Small}
   slot {class=OverlaySlot, properties {horizontalAlignment=HAlign_Center, verticalAlignment=VAlign_Center, padding=<ltrb 125 0 125 0>}}
  }
  SlotsCount {
   widgetClassPath=&CommonTextBlock
   bInherited=true
   bIsVariable=true
   parent=Overlay_1
   widget=SlotsCount
   properties {
    scrollStyle=/FortUILibrary/Styles/Typography/Foundation/TextScrollStyle-NoFade.TextScrollStyle-NoFade_C
    style=/FortUILibrary/Styles/Typography/Foundation/TextStyle-Base-S-B.TextStyle-Base-S-B_C
    text=2 SLOTS
    visibility=Hidden
    font {
     bForceMonospaced=false
     bMaterialIsStencil=false
     fontMaterial=null
     fontObject=&NotoSans
     letterSpacing=0
     monospacedWidth=1
     size=18
     skewAmount=0
     typefaceFontName=NotoSans-Bold
     outlineSettings {
      bApplyOutlineToDropShadows=false
      bMiteredCorners=false
      bSeparateFillAlpha=false
      outlineMaterial=null
      outlineSize=0
      outlineColor=<rgba 0 0 0 1>
     }
    }
    shadowOffset=<0 0>
   }
   slot {class=OverlaySlot, properties {horizontalAlignment=HAlign_Right, verticalAlignment=VAlign_Fill, padding=<ltrb 8 0 0 0>}}
  }
  TertiaryRow {
   widgetClassPath=&HorizontalBox
   bInherited=true
   bIsVariable=true
   parent=Overlay_1
   widget=TertiaryRow
   properties {visibility=Collapsed}
   slot {class=OverlaySlot, properties {verticalAlignment=VAlign_Center}}
  }
  CategoryIndicatorTertiary {
   widgetClassPath=&FortItemCategoryIndicator
   bInherited=true
   bIsVariable=true
   parent=TertiaryRow
   widget=CategoryIndicatorTertiary
   properties {
    iconBrushSize=Medium
    labelTextStyle=/FortUILibrary/Styles/Typography/Foundation/TextStyle-Base-XS-B-70pc.TextStyle-Base-XS-B-70pc_C
    ordinalOfCategoryToRepresent=Tertiary
   }
   slot {class=HorizontalBoxSlot}
  }
  AmmoLoadedText {
   widgetClassPath=&CommonTextBlock
   bInherited=true
   bIsVariable=true
   parent=TertiaryRow
   widget=AmmoLoadedText
   properties {
    scrollStyle=&TextScrollStyle-NoFade_C
    style=&TextStyle-Base-S-B_C
    text=30
    font {
     bForceMonospaced=false
     bMaterialIsStencil=false
     fontMaterial=null
     fontObject=&NotoSans
     letterSpacing=0
     monospacedWidth=1
     size=18
     skewAmount=0
     typefaceFontName=NotoSans-Bold
     outlineSettings {
      bApplyOutlineToDropShadows=false
      bMiteredCorners=false
      bSeparateFillAlpha=false
      outlineMaterial=null
      outlineSize=0
      outlineColor=<rgba 0 0 0 1>
     }
    }
    shadowOffset=<0 0>
   }
   slot {class=HorizontalBoxSlot, properties {padding=<ltrb 8 0 0 0>}}
  }
  ItemCostWidget {
   widgetClassPath=/Game/UI/Creative/ItemCostWidget
   bInherited=true
   bIsVariable=true
   parent=VerticalBox_3
   widget=ItemCostWidget
   slot {class=VerticalBoxSlot}
  }
  ItemDetailsStackCounter {
   widgetClassPath=/Game/UI/Foundation/Items/ItemDetailsStackCounter
   parent=Overlay_0
   widget=ItemDetailsStackCounter
   properties {visibility=Hidden}
   slot {class=OverlaySlot, properties {horizontalAlignment=HAlign_Right, padding=<ltrb 0 0 10 0>}}
  }
  Overlay_ShortDescription {
   widgetClassPath=&Overlay
   parent=CommonWidgetSwitcher_Stack
   widget=Overlay_ShortDescription
   slot {class=WidgetSwitcherSlot, properties {padding=<ltrb 0 -2 0 -2>}}
  }
  VerticalBox_6 {
   widgetClassPath=&VerticalBox
   parent=Overlay_ShortDescription
   widget=VerticalBox_6
   slot {class=OverlaySlot, properties {horizontalAlignment=HAlign_Fill, verticalAlignment=VAlign_Center, padding=<ltrb 8 8 8 8>}}
  }
  HorizontalBox_0 {widgetClassPath=&HorizontalBox, parent=VerticalBox_6, widget=HorizontalBox_0, slot {class=VerticalBoxSlot}}
  Image_MinimalIconPreview {
   widgetClassPath=&FortLazyImage
   bInherited=true
   bIsVariable=true
   parent=HorizontalBox_0
   widget=Image_MinimalIconPreview
   properties {
    brush {
     bIsDynamicallyLoaded=false
     drawAs=Image
     imageType=NoImage
     mirroring=NoMirror
     resourceName=None
     resourceObject=/Game/UI/Foundation/Textures/Icons/Weapons/Items/T-Icon-Weapons-SK-Grenade-Launcher
     tiling=NoTile
     imageSize=<56 56>
     margin=<ltrb 0 0 0 0>
     outlineSettings {
      bUseBrushTransparency=false
      roundingType=HalfHeightRadius
      width=0
      color {colorUseRule=UseColor_Specified, specifiedColor=<rgba 0 0 0 0>}
      cornerRadii=<0 0 0 0>
     }
     tintColor {colorUseRule=UseColor_Specified, specifiedColor=<rgba 1 1 1 1>}
     uVRegion {bIsValid=false, max=<0 0>, min=<0 0>}
    }
    loadingBackgroundBrush {
     bIsDynamicallyLoaded=false
     drawAs=NoDrawType
     imageType=NoImage
     mirroring=NoMirror
     resourceName=None
     resourceObject=null
     tiling=NoTile
     imageSize=<48 48>
     margin=<ltrb 0 0 0 0>
     outlineSettings {
      bUseBrushTransparency=false
      roundingType=HalfHeightRadius
      width=0
      color {colorUseRule=UseColor_Specified, specifiedColor=<rgba 0 0 0 0>}
      cornerRadii=<0 0 0 0>
     }
     tintColor {colorUseRule=UseColor_Specified, specifiedColor=<rgba 1 1 1 1>}
     uVRegion {bIsValid=false, max=<0 0>, min=<0 0>}
    }
   }
   slot {class=HorizontalBoxSlot, properties {horizontalAlignment=HAlign_Center, verticalAlignment=VAlign_Center}}
  }
  VerticalBox_5 {
   widgetClassPath=&VerticalBox
   parent=HorizontalBox_0
   widget=VerticalBox_5
   slot {class=HorizontalBoxSlot, properties {horizontalAlignment=HAlign_Center, verticalAlignment=VAlign_Center, padding=<ltrb 4 0 10 0>}}
  }
  HorizontalBox_2 {widgetClassPath=&HorizontalBox, parent=VerticalBox_5, widget=HorizontalBox_2, slot {class=VerticalBoxSlot}}
  Text_ItemNameShort {
   widgetClassPath=&CommonTextBlock
   bInherited=true
   bIsVariable=true
   parent=HorizontalBox_2
   widget=Text_ItemNameShort
   properties {
    lineHeightPercentage=0.9
    mobileFontSizeMultiplier=1.3
    text="{ItemName}"
    textTransformPolicy=ToUpper
    visibility=Collapsed
    font {
     bForceMonospaced=false
     bMaterialIsStencil=false
     fontMaterial=null
     fontObject=&Burbank
     letterSpacing=0
     monospacedWidth=1
     size=16
     skewAmount=0
     typefaceFontName=BurbankRegular-Black
     outlineSettings {
      bApplyOutlineToDropShadows=false
      bMiteredCorners=false
      bSeparateFillAlpha=true
      outlineMaterial=&M-FontShadow
      outlineSize=2
      outlineColor=<rgba 0 0 0 0.5>
     }
    }
    renderTransform {angle=0, scale=<1 1>, shear=<-12 0>, translation=<0 0>}
    shadowOffset=<0 0>
   }
   slot {class=HorizontalBoxSlot, properties {verticalAlignment=VAlign_Center, padding=<ltrb 8 0 0 0>}}
  }
  Text_ItemNameShort_Amount {
   widgetClassPath=&CommonTextBlock
   bInherited=true
   bIsVariable=true
   parent=HorizontalBox_2
   widget=Text_ItemNameShort_Amount
   properties {
    mobileFontSizeMultiplier=1.7
    text=x1
    font {
     bForceMonospaced=false
     bMaterialIsStencil=false
     fontMaterial=null
     fontObject=&BurbankSmall
     letterSpacing=0
     monospacedWidth=1
     size=16
     skewAmount=0
     typefaceFontName=Black
     outlineSettings {
      bApplyOutlineToDropShadows=false
      bMiteredCorners=false
      bSeparateFillAlpha=true
      outlineMaterial=&M-FontShadow
      outlineSize=3
      outlineColor=<rgba 0 0 0 0.5>
     }
    }
    renderTransform {angle=0, scale=<1 1>, shear=<-12 0>, translation=<0 0>}
    shadowOffset=<0 0>
   }
   slot {class=HorizontalBoxSlot, properties {verticalAlignment=VAlign_Center, padding=<ltrb 4 0 10 0>}}
  }
  CraftingIngredientBox {
   widgetClassPath=&HorizontalBox
   bInherited=true
   bIsVariable=true
   parent=VerticalBox_5
   widget=CraftingIngredientBox
   properties {visibility=Collapsed}
   slot {class=VerticalBoxSlot, properties {padding=<ltrb 8 0 0 0>}}
  }
  CraftingIcon {
   widgetClassPath=&Image
   parent=CraftingIngredientBox
   widget=CraftingIcon
   properties {
    visibility=SelfHitTestInvisible
    brush {
     bIsDynamicallyLoaded=false
     drawAs=Image
     imageType=NoImage
     mirroring=NoMirror
     resourceName=None
     resourceObject=/Game/Athena/HUD/Inventory/Art/M_UI_CraftingIndicator_Icon_White
     tiling=NoTile
     imageSize=<24 24>
     margin=<ltrb 0 0 0 0>
     outlineSettings {
      bUseBrushTransparency=false
      roundingType=HalfHeightRadius
      width=0
      color {colorUseRule=UseColor_Specified, specifiedColor=<rgba 0 0 0 0>}
      cornerRadii=<0 0 0 0>
     }
     tintColor {colorUseRule=UseColor_Specified, specifiedColor=<rgba 1 1 1 1>}
     uVRegion {bIsValid=false, max=<0 0>, min=<0 0>}
    }
   }
   slot {class=HorizontalBoxSlot, properties {horizontalAlignment=HAlign_Center, verticalAlignment=VAlign_Center}}
  }
  CraftingLabel {
   widgetClassPath=&CommonTextBlock
   parent=CraftingIngredientBox
   widget=CraftingLabel
   properties {
    lineHeightPercentage=0.9
    text=Crafting Ingredient
    textTransformPolicy=ToUpper
    font {
     bForceMonospaced=false
     bMaterialIsStencil=false
     fontMaterial=/Game/Athena/HUD/Inventory/Art/M_UI_ItemDetails_CraftingText_FontColor
     fontObject=&BurbankSmall
     letterSpacing=0
     monospacedWidth=1
     size=13
     skewAmount=0
     typefaceFontName=Black
     outlineSettings {
      bApplyOutlineToDropShadows=false
      bMiteredCorners=false
      bSeparateFillAlpha=false
      outlineMaterial=&M-FontShadow
      outlineSize=2
      outlineColor=<rgba 0.0625 0.0167 0.0002 0.5>
     }
    }
    margin=<ltrb 0 -1 0 0>
    renderTransform {angle=0, scale=<1 1>, shear=<-12 0>, translation=<0 0>}
    shadowOffset=<0 0>
   }
   slot {class=HorizontalBoxSlot, properties {verticalAlignment=VAlign_Center, padding=<ltrb 6 0 0 0>}}
  }
  ItemCostWidgetShort {
   widgetClassPath=&ItemCostWidget
   bInherited=true
   bIsVariable=true
   parent=VerticalBox_6
   widget=ItemCostWidgetShort
   slot {class=VerticalBoxSlot}
  }
  Overlay_ShortItemDescription {widgetClassPath=&Overlay, parent=CommonWidgetSwitcher_Stack, widget=Overlay_ShortItemDescription, slot {class=WidgetSwitcherSlot}}
  VerticalBox_4 {
   widgetClassPath=&VerticalBox
   parent=Overlay_ShortItemDescription
   widget=VerticalBox_4
   slot {class=OverlaySlot, properties {horizontalAlignment=HAlign_Fill, verticalAlignment=VAlign_Fill, padding=<ltrb 4 4 8 4>}}
  }
  Text_WeaponName {
   widgetClassPath=&CommonTextBlock
   bInherited=true
   bIsVariable=true
   parent=VerticalBox_4
   widget=Text_WeaponName
   properties {
    lineHeightPercentage=0.9
    text=Weapon Name
    textTransformPolicy=ToUpper
    font {
     bForceMonospaced=false
     bMaterialIsStencil=false
     fontMaterial=null
     fontObject=&Burbank
     letterSpacing=0
     monospacedWidth=1
     size=20
     skewAmount=0.2
     typefaceFontName=BurbankRegular-Black
     outlineSettings {
      bApplyOutlineToDropShadows=true
      bMiteredCorners=false
      bSeparateFillAlpha=true
      outlineMaterial=&M-FontShadow
      outlineSize=3
      outlineColor=<rgba 0 0 0 0.5>
     }
    }
    margin=<ltrb 0 -2 0 0>
    shadowOffset=<0 0>
   }
   slot {class=VerticalBoxSlot}
  }
  HorizontalBox_1 {
   widgetClassPath=&HorizontalBox
   parent=VerticalBox_4
   widget=HorizontalBox_1
   properties {flowDirectionPreference=Culture}
   slot {class=VerticalBoxSlot}
  }
  Overlay_2 {
   widgetClassPath=&Overlay
   parent=HorizontalBox_1
   widget=Overlay_2
   slot {class=HorizontalBoxSlot, properties {padding=<ltrb 2 0 0 0>}}
  }
  Image_RarityGlow {
   widgetClassPath=&Image
   bInherited=true
   bIsVariable=true
   parent=Overlay_2
   widget=Image_RarityGlow
   properties {
    visibility=SelfHitTestInvisible
    brush {
     bIsDynamicallyLoaded=false
     drawAs=Image
     imageType=NoImage
     mirroring=NoMirror
     resourceName=None
     resourceObject=/Game/Athena/HUD/LTM/Wax/Materials/M_UI_TextGlow_Add
     tiling=NoTile
     imageSize=<32 18>
     margin=<ltrb 0 0 0 0>
     outlineSettings {
      bUseBrushTransparency=false
      roundingType=HalfHeightRadius
      width=0
      color {colorUseRule=UseColor_Specified, specifiedColor=<rgba 0 0 0 0>}
      cornerRadii=<0 0 0 0>
     }
     tintColor {colorUseRule=UseColor_Specified, specifiedColor=<rgba 0.521 0.0781 0 1>}
     uVRegion {bIsValid=false, max=<0 0>, min=<0 0>}
    }
    renderTransform {angle=0, scale=<1.125 1.75>, shear=<0 0>, translation=<0 0>}
   }
   slot {class=OverlaySlot, properties {horizontalAlignment=HAlign_Fill, verticalAlignment=VAlign_Center}}
  }
  Text_WeaponRarityValue {
   widgetClassPath=&CommonTextBlock
   bInherited=true
   bIsVariable=true
   parent=Overlay_2
   widget=Text_WeaponRarityValue
   properties {
    lineHeightPercentage=0.925
    text="{Rarity}"
    textTransformPolicy=ToUpper
    font {
     bForceMonospaced=false
     bMaterialIsStencil=false
     fontMaterial=null
     fontObject=&Burbank
     letterSpacing=0
     monospacedWidth=1
     size=18
     skewAmount=0
     typefaceFontName=Burbank-Black
     outlineSettings {
      bApplyOutlineToDropShadows=false
      bMiteredCorners=false
      bSeparateFillAlpha=false
      outlineMaterial=null
      outlineSize=0
      outlineColor=<rgba 0 0 0 1>
     }
    }
    margin=<ltrb 0 -2 0 0>
    shadowOffset=<0 0>
   }
   slot {class=OverlaySlot, properties {verticalAlignment=VAlign_Center}}
  }
  HorizontalBox_Tertiary2 {
   widgetClassPath=&HorizontalBox
   bInherited=true
   bIsVariable=true
   parent=HorizontalBox_1
   widget=HorizontalBox_Tertiary2
   slot {class=HorizontalBoxSlot, properties {padding=<ltrb 8 0 0 0>}}
  }
  SizeBox_1 {
   widgetClassPath=&SizeBox
   parent=HorizontalBox_Tertiary2
   widget=SizeBox_1
   properties {bOverride_HeightOverride=true, bOverride_WidthOverride=true, heightOverride=24, widthOverride=24}
   slot {class=HorizontalBoxSlot, properties {horizontalAlignment=HAlign_Center, verticalAlignment=VAlign_Center}}
  }
  ScaleBox_0 {
   widgetClassPath=&ScaleBox
   parent=SizeBox_1
   widget=ScaleBox_0
   slot {class=SizeBoxSlot, properties {horizontalAlignment=HAlign_Center, verticalAlignment=VAlign_Center}}
  }
  CategoryIndicatorTertiary_2 {
   widgetClassPath=&FortItemCategoryIndicator
   bInherited=true
   bIsVariable=true
   parent=ScaleBox_0
   widget=CategoryIndicatorTertiary_2
   properties {
    iconBrushSize=Medium
    labelTextStyle=&TextStyle-Base-XS-B-70pc_C
    ordinalOfCategoryToRepresent=Tertiary
   }
   slot {class=ScaleBoxSlot}
  }
  AmmoLoadedText_2 {
   widgetClassPath=&CommonTextBlock
   bInherited=true
   bIsVariable=true
   parent=HorizontalBox_Tertiary2
   widget=AmmoLoadedText_2
   properties {
    scrollStyle=&TextScrollStyle-NoFade_C
    text=42
    font {
     bForceMonospaced=false
     bMaterialIsStencil=false
     fontMaterial=null
     fontObject=&Burbank
     letterSpacing=0
     monospacedWidth=1
     size=18
     skewAmount=0
     typefaceFontName=Burbank-PlayerPower
     outlineSettings {
      bApplyOutlineToDropShadows=false
      bMiteredCorners=false
      bSeparateFillAlpha=false
      outlineMaterial=null
      outlineSize=0
      outlineColor=<rgba 0 0 0 1>
     }
    }
    shadowOffset=<0 0>
   }
   slot {class=HorizontalBoxSlot, properties {verticalAlignment=VAlign_Center}}
  }
  Image_RechargeableIcon {
   widgetClassPath=&FortLazyImage
   bInherited=true
   bIsVariable=true
   parent=HorizontalBox_1
   widget=Image_RechargeableIcon
   properties {
    bIsVolatile=true
    visibility=Collapsed
    brush {
     bIsDynamicallyLoaded=false
     drawAs=Image
     imageType=NoImage
     mirroring=NoMirror
     resourceName=None
     resourceObject=/Game/UI/Foundation/Textures/Icons/Ammo/T-Icon-RechargeAmmo-64
     tiling=NoTile
     imageSize=<22 22>
     margin=<ltrb 0 0 0 0>
     outlineSettings {
      bUseBrushTransparency=false
      roundingType=HalfHeightRadius
      width=0
      color {colorUseRule=UseColor_Specified, specifiedColor=<rgba 0 0 0 0>}
      cornerRadii=<0 0 0 0>
     }
     tintColor {colorUseRule=UseColor_Specified, specifiedColor=<rgba 1 1 1 1>}
     uVRegion {bIsValid=false, max=<0 0>, min=<0 0>}
    }
    loadingBackgroundBrush {
     bIsDynamicallyLoaded=false
     drawAs=NoDrawType
     imageType=NoImage
     mirroring=NoMirror
     resourceName=None
     resourceObject=null
     tiling=NoTile
     imageSize=<48 48>
     margin=<ltrb 0 0 0 0>
     outlineSettings {
      bUseBrushTransparency=false
      roundingType=HalfHeightRadius
      width=0
      color {colorUseRule=UseColor_Specified, specifiedColor=<rgba 0 0 0 0>}
      cornerRadii=<0 0 0 0>
     }
     tintColor {colorUseRule=UseColor_Specified, specifiedColor=<rgba 1 1 1 1>}
     uVRegion {bIsValid=false, max=<0 0>, min=<0 0>}
    }
   }
   slot {class=HorizontalBoxSlot, properties {horizontalAlignment=HAlign_Left, verticalAlignment=VAlign_Center, padding=<ltrb 4 0 4 0>}}
  }
  Spacer_RightToLeftLocFix {
   widgetClassPath=&Spacer
   parent=HorizontalBox_1
   widget=Spacer_RightToLeftLocFix
   slot {class=HorizontalBoxSlot, properties {size {sizeRule=Fill, value=1}}}
  }
  DurabilityStatus {
   widgetClassPath=/Game/Athena/HUD/AthenaItemDetailsDurability
   bInherited=true
   bIsVariable=true
   parent=VerticalBox_4
   widget=DurabilityStatus
   slot {class=VerticalBoxSlot}
  }
  ItemCostContainer {widgetClassPath=&SizeBox, bIsVariable=true, parent=VerticalBox_4, widget=ItemCostContainer, slot {class=VerticalBoxSlot}}
  ItemCostWidgetShortItem {widgetClassPath=/Script/FortniteUI.ItemCostWidgetBase, bInherited=true, widget=null}
  ScrollBox {widgetClassPath=&ScrollBox, bInherited=true, widget=null}
  HighRarityBorder {widgetClassPath=&CommonBorder, bInherited=true, widget=null}
  ItemToDetailLoadGuard {widgetClassPath=/Script/CommonUI.CommonLoadGuard, bInherited=true, widget=null}
  ItemToCompareLoadGuard {widgetClassPath=&CommonLoadGuard, bInherited=true, widget=null}
  Border_Details {widgetClassPath=&CommonBorder, bInherited=true, widget=null}
 ]
}
