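# Widget Blueprint for the Athena HUD pickup item display, built on the native AthenaHUDPickupItemWidgetBase class from FortniteUI.
# Its only widget is the root ItemDetailsHeader, an AthenaItemDetailsHeader instance exposed as a variable and marked as inherited.
# The event graph and variable list are empty, so this asset adds no Blueprint logic of its own.
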WidgetBlueprint /Game/Athena/HUD/AthenaHUD-PickupItemWidget
graphs [EventGraph {links [], nodes []}]
variables []
widgets {
 info {
  inheritedWidgetCount=1
  parentClass=/Script/FortniteUI.AthenaHUDPickupItemWidgetBase
  rootWidgetClass=/Game/Athena/HUD/Items/AthenaItemDetailsHeader.AthenaItemDetailsHeader_C
  widgetCount=1
 }
 widgets [ItemDetailsHeader {widgetClassPath=/Game/Athena/HUD/Items/AthenaItemDetailsHeader, bInherited=true, bIsVariable=true, widget=ItemDetailsHeader}]
}
