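# HUD warning widget built on CommonUserWidget that displays an alert icon and an uppercase interaction failure message in red on a dark border.
# PreConstruct applies the IconTexture and FailureMessage variables to the image and text block, with defaults of an alert icon and "Interaction Error".
# SetFailureMessage and SetIcon pass new values through Delta Gate macros that update the stored variable before refreshing the matching child widget.
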WidgetBlueprint /Game/Athena/HUD/WBP_InteractionWarning
graphs [
 EventGraph {nodes [
   Event PreConstruct()
   > SetBrushFromSoftTexture!(
    execute
    self:Image Object Reference=Get Image_Icon() -> (Image_Icon:Image Object Reference)
    SoftTexture:Texture 2D Soft Object Reference=Get IconTexture() -> (IconTexture:Texture 2D Soft Object Reference)
    bMatchSize:bool=false
   ) -> (then)
   > "SetText (Text)"(self=Get Text_FailureMessage() -> (Text_FailureMessage:Common Text Object Reference), InText=Get FailureMessage() -> (FailureMessage:Text))
  ]}
 SetFailureMessage {nodes [
   n1 = SetFailureMessage!() -> (then, Text:Text) {is=title K2Node_FunctionEntry}
   > n2 = Delta Gate Wildcard!(execute:Exec, Variable:"Text (by ref)"=Get FailureMessage(), New Value:Text=@n1) -> (Success:Exec, Updated Variable:"Text (by ref)", Previous Variable:Text) {is=title K2Node_MacroInstance}
   > "SetText (Text)"(self=Get Text_FailureMessage(), InText=@n2)
  ]}
 SetIcon {nodes [
   n1 = SetIcon!() -> (then, NewIcon:Texture 2D Soft Object Reference) {is=title K2Node_FunctionEntry}
   > n2 = Delta Gate Wildcard(Variable:"Texture 2D Soft Object Reference (by ref)"=Get IconTexture(), New Value:Texture 2D Soft Object Reference=@n1) -> (Updated Variable:"Texture 2D Soft Object Reference (by ref)", Previous Variable:Texture 2D Soft Object Reference)
   > SetBrushFromSoftTexture(self=Get Image_Icon(), SoftTexture=@n2, bMatchSize=false)
  ]}
]
variables [
 FailureMessage {type=string, default=Interaction Error}
 IconTexture {type=Texture2D, default=/Game/UI/Foundation/Textures/Icons/HUD/T-Icon-Alert2-32}
]
widgets {
 info {inheritedWidgetCount=0, parentClass=&CommonUserWidget, rootWidgetClass=&CommonBorder, widgetCount=4}
 widgets [
  CommonBorder_0 {
   widgetClassPath=&CommonBorder
   widget=CommonBorder_0
   properties {
    horizontalAlignment=HAlign_Center
    background {
     bIsDynamicallyLoaded=false
     drawAs=Image
     imageType=NoImage
     mirroring=NoMirror
     resourceName=None
     resourceObject=null
     tiling=NoTile
     imageSize=<32 32>
     margin=<ltrb 0 0 0 0>
     outlineSettings {
      bUseBrushTransparency=false
      roundingType=HalfHeightRadius
      width=0
      color {colorUseRule=UseColor_Specified, specifiedColor=<rgba 0 0 0 0>}
      cornerRadii=<0 0 0 0>
     }
     tintColor {colorUseRule=UseColor_Specified, specifiedColor=<rgba 1 1 1 1>}
     uVRegion {bIsValid=false, max=<0 0>, min=<0 0>}
    }
    brushColor=<rgba 0.0252 0.0123 0.013 1>
    contentColorAndOpacity=<rgba 1 0.0578 0.0666 1>
    padding=<ltrb 8 4 8 4>
   }
  }
  HorizontalBox_3 {
   widgetClassPath=&HorizontalBox
   parent=CommonBorder_0
   widget=HorizontalBox_3
   slot {class=BorderSlot, properties {horizontalAlignment=HAlign_Center, padding=<ltrb 8 4 8 4>}}
  }
  Image_Icon {
   widgetClassPath=&Image
   bIsVariable=true
   parent=HorizontalBox_3
   widget=Image_Icon
   properties {
    visibility=SelfHitTestInvisible
    brush {
     bIsDynamicallyLoaded=false
     drawAs=Image
     imageType=FullColor
     mirroring=NoMirror
     resourceName=None
     resourceObject=&T-Icon-Alert2-32
     tiling=NoTile
     imageSize=<32 32>
     margin=<ltrb 0 0 0 0>
     outlineSettings {
      bUseBrushTransparency=false
      roundingType=HalfHeightRadius
      width=0
      color {colorUseRule=UseColor_Specified, specifiedColor=<rgba 0 0 0 0>}
      cornerRadii=<0 0 0 0>
     }
     tintColor {colorUseRule=UseColor_Specified, specifiedColor=<rgba 1 1 1 1>}
     uVRegion {bIsValid=false, max=<0 0>, min=<0 0>}
    }
   }
   slot {class=HorizontalBoxSlot, properties {horizontalAlignment=HAlign_Center, verticalAlignment=VAlign_Center, padding=<ltrb 4 0 8 0>}}
  }
  Text_FailureMessage {
   widgetClassPath=&CommonTextBlock
   bIsVariable=true
   parent=HorizontalBox_3
   widget=Text_FailureMessage
   properties {
    text=Interaction Error
    textTransformPolicy=ToUpper
    wrapTextAt=500
    font {
     bForceMonospaced=false
     bMaterialIsStencil=false
     fontMaterial=null
     fontObject=&BurbankSmall
     letterSpacing=0
     monospacedWidth=1
     size=13
     skewAmount=0
     typefaceFontName=Black
     outlineSettings {
      bApplyOutlineToDropShadows=true
      bMiteredCorners=false
      bSeparateFillAlpha=true
      outlineMaterial=&M-FontShadow
      outlineSize=2
      outlineColor=<rgba 0.085 0.0146 0.0177 0.5>
     }
    }
    renderTransform {angle=0, scale=<1 1>, shear=<-12 0>, translation=<0 0>}
    shadowOffset=<0 0>
   }
   slot {class=HorizontalBoxSlot, properties {horizontalAlignment=HAlign_Center, verticalAlignment=VAlign_Center, size {sizeRule=Fill, value=1}}}
  }
 ]
}
