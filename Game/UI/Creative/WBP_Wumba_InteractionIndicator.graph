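# Interaction indicator for the Wumba weapon upgrade station, showing current weapon + resource cost = upgraded weapon, plus a warning.
# AttemptWumbaInteraction caches context, station and weapon definition, returning whether both station and weapon are valid.
# HandleWumbaInteraction looks up a vertical (slot 0) or horizontal (slot 1) upgrade, then fills or collapses the cost row.
# Non-upgradable results show context sub-text, except failed horizontal upgrades, which also collapse the warning.
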WidgetBlueprint /Game/UI/Creative/WBP_Wumba_InteractionIndicator
graphs [
 AttemptSetItemToRepresent {nodes [
   n1 = AttemptSetItemToRepresent!() -> (then, Widget:Fort Item Widget NUI Object Reference, ItemToRepresent:Fort Item Object Reference) {is=title K2Node_FunctionEntry}
   n2 = Get ItemToRepresent!(self:Fort Item Widget NUI Object Reference=@n1) -> (ItemToRepresent:Fort Item Weak object)
   > n3 = Is Valid(InputObject=@n2)
   > n4 = Branch(Condition="Not Equal (Object)"!(
     A:Object Reference=GetItemDefinition!(self:Fort Item Object Reference=@n2) -> (ReturnValue:Fort Item Definition Object Reference)
     B:Object Reference=GetItemDefinition(self=@n1.ItemToRepresent)
    ) -> (ReturnValue:bool))
   SetItemToRepresent!(
    execute=[@n4.then, @n3.Is Not Valid]
    self:Fort Item Widget NUI Object Reference=Get Widget() -> (Widget:Fort Item Widget NUI Object Reference)
    ItemToRepresent:Fort Item Object Reference=Get ItemToRepresent!() -> (ItemToRepresent:Fort Item Object Reference)
   ) -> (then)
  ]}
 AttemptWumbaInteraction {nodes [
   n1 = AttemptWumbaInteraction!() -> (then, Current Interaction:Fort Interact Context Info Object Reference, Indicated Actor:Actor Object Reference) {is=title K2Node_FunctionEntry}
   > Set Current Interaction(Current Interaction:Fort Interact Context Info Object Reference=@n1) -> (Output_Get:Fort Interact Context Info Object Reference)
   > n2 = Cast To B_Athena_Wumba!(execute, Object=@n1.Indicated Actor) -> (then, CastFailed, AsB Athena Wumba:B Athena Wumba Object Reference) {is=title K2Node_DynamicCast}
   > Set WumbaItemRef(WumbaItemRef:B Athena Wumba Object Reference=@n2) -> (Output_Get:B Athena Wumba Object Reference)
   n3 = Get CurrentWeaponDefinition!(self:Building Item Weapon Upgrade Actor Object Reference=@n2) -> (CurrentWeaponDefinition:Fort Weapon Item Definition Object Reference)
   > n4 = Is Valid(InputObject=@n3)
   > Set CurrentWeaponDef(CurrentWeaponDef:Fort Weapon Item Definition Object Reference=@n3) -> (Output_Get:Fort Weapon Item Definition Object Reference)
   > Return Node!(execute, CanAttemptInteraction:bool=true)
   Return Node(execute=[@n4.Is Not Valid, @n2.CastFailed], CanAttemptInteraction=false)
  ]}
 EventGraph {links [], nodes []}
 ForceUpdateWidgetVisibility {nodes [
   n1 = ForceUpdateWidgetVisibility!() -> (then, InVisibility:ESlateVisibility Enum=Visible) {is=title K2Node_FunctionEntry}
   > SetVisibility(
    self=[
     Get VBox_WumbaCost() -> (VBox_WumbaCost:Horizontal Box Object Reference)
     Get WBP_WumbaError() -> (WBP_WumbaError:WBP Interaction Warning Object Reference)
    ]
    InVisibility=@n1
   )
  ]}
 HandleWumbaInteraction {nodes [
   n1 = HandleWumbaInteraction!() -> (then, UpgradeDirection:int=0) {is=title K2Node_FunctionEntry}
   > n2 = Set Wumba Interaction Slot(Wumba Interaction Slot:int=@n1) -> (Output_Get:int=0)
   > SetupInteractionError()
   > n3 = Switch on Int!(execute:Exec, Selection:int=@n2) -> (Default:Exec, 0:Exec, 1:Exec)
   n4 = Switch on EFortWeaponUpgradeInteractionResult!(
    execute=@n3.1
    Selection:EFortWeaponUpgradeInteractionResult Enum=Get UpgradeInteractionResult!(self:Building Item Weapon Upgrade Actor Object Reference=Get WumbaItemRef() -> (WumbaItemRef:B Athena Wumba Object Reference)) -> (UpgradeInteractionResult:EFortWeaponUpgradeInteractionResult Enum=Upgradable)
   ) -> (Upgradable:Exec, NotEnoughResources:Exec, CannotUpgrade:Exec, CannotInteract:Exec) {is=title K2Node_SwitchEnum}
   n5 = GetUpgradedWeaponItemFromTable!(
    ItemToUpgrade:Fort Weapon Item Definition Object Reference=Get CurrentWeaponDef() -> (CurrentWeaponDef:Fort Weapon Item Definition Object Reference)
    UpgradeDirection:EFortWeaponUpgradeDirection Enum=Horizontal
   ) -> (ReturnValue:Fort Weapon Item Definition Object Reference)
   n6 = Is Valid(exec=[@n4.NotEnoughResources, @n4.Upgradable], InputObject=@n5)
   ForceUpdateWidgetVisibility!(execute=[@n6.Is Not Valid, @n4.CannotInteract, @n4.CannotUpgrade], self, InVisibility:ESlateVisibility Enum=Collapsed) -> (then) {is=title K2Node_CallFunction}
   > Return Node!(execute, Upgrade Interaction Result:EFortWeaponUpgradeInteractionResult Enum=CannotInteract)
   UpdateItemCostWidget!(execute=@n6.Is Valid, self, ItemDefinition:Item Definition Base Object Reference=@n5) -> (then)
   > Return Node(Upgrade Interaction Result=Get UpgradeInteractionResult(self=Get WumbaItemRef()) -> (UpgradeInteractionResult=Upgradable))
   n7 = Switch on EFortWeaponUpgradeInteractionResult(execute=@n3.0, Selection=Get UpgradeInteractionResult(self=Get WumbaItemRef()) -> (UpgradeInteractionResult=Upgradable))
   n8 = GetUpgradedWeaponItemFromTable(ItemToUpgrade=Get CurrentWeaponDef(), UpgradeDirection=Vertical)
   n9 = Is Valid(exec=[@n7.NotEnoughResources, @n7.Upgradable], InputObject=@n8)
   SetVisibility(execute=[@n9.Is Not Valid, @n7.CannotInteract, @n7.CannotUpgrade], self=Get VBox_WumbaCost(), InVisibility=Collapsed)
   n10 = Get UpgradeInteractionResult(self=Get WumbaItemRef()) -> (UpgradeInteractionResult=Upgradable)
   > Return Node(Upgrade Interaction Result=@n10)
   UpdateItemCostWidget(execute=@n9.Is Valid, ItemDefinition=@n8)
   > Return Node(Upgrade Interaction Result=@n10)
  ]}
 SetupInteractionError {nodes [
   SetupInteractionError() {is=title K2Node_FunctionEntry}
   n1 = Get WumbaItemRef()
   n2 = Get WBP_WumbaError()
   > SetVisibility(
    self=@n2
    InVisibility=Select!(
     Upgradable:ESlateVisibility Enum=Collapsed
     NotEnoughResources:ESlateVisibility Enum=HitTestInvisible
     CannotUpgrade:ESlateVisibility Enum=HitTestInvisible
     CannotInteract:ESlateVisibility Enum=HitTestInvisible
     Index:EFortWeaponUpgradeInteractionResult Enum=Get UpgradeInteractionResult(self=@n1) -> (UpgradeInteractionResult=Upgradable)
    ) -> (ReturnValue)
   )
   > SetFailureMessage!(
    execute
    self:WBP Interaction Warning Object Reference=@n2
    Text:Text=Get ContextSubText!(self:Fort Interact Context Info Object Reference=Get Current Interaction() -> (Current Interaction:Fort Interact Context Info Object Reference)) -> (ContextSubText:Text)
   ) -> (then)
  ]}
 UpdateItemCostWidget {nodes [
   n1 = UpdateItemCostWidget!() -> (then, ItemDefinition:Item Definition Base Object Reference) {is=title K2Node_FunctionEntry}
   > SetItemCost!(
    execute
    self:Item Cost Widget Object Reference=Get WBP_WumbaResources() -> (WBP_WumbaResources:Item Cost Widget Object Reference)
    Actor:Actor Object Reference=Get WumbaItemRef()
    SpecifiedItemCostComponent:Item Cost Component Class Reference=/Game/Athena/Items/Gameplay/Wumba/Wumba_ItemCostComponent.Wumba_ItemCostComponent_C
   ) -> (then)
   > n2 = GetRarityData!(
    execute
    Rarity:EFortRarity Enum=GetRarity!(ItemDefinition:Item Definition Base Object Reference=@n1) -> (ReturnValue:EFortRarity Enum=Common)
   ) -> (then, ReturnValue:Fort Rarity Item Data Structure)
   > SetRarityData!(execute, self:Item Cost Widget Base Object Reference=Get WBP_WumbaResources(), RarityData:"Fort Rarity Item Data Structure (by ref)"=@n2) -> (then)
   > AttemptSetItemToRepresent!(
    execute
    self
    Widget:Fort Item Widget NUI Object Reference=Get ItemCard_WumbaUpgradeTo() -> (ItemCard_WumbaUpgradeTo:Fort Multi Size Item Card Object Reference)
    ItemToRepresent:Fort Item Object Reference=CreateTemporaryItemInstance!(ItemDefinition:Item Definition Base Object Reference=@n1, Count:int=1, Level:int=-1) -> (ReturnValue:Fort Item Object Reference)
   ) -> (then) {is=title K2Node_CallFunction}
   > AttemptSetItemToRepresent(
    Widget=Get ItemCard_WumbaUpgradeFrom() -> (ItemCard_WumbaUpgradeFrom:Fort Multi Size Item Card Object Reference)
    ItemToRepresent=CreateTemporaryItemInstance(ItemDefinition=Get CurrentWeaponDef(), Count=1, Level=-1)
   )
   > SetVisibility(self=Get VBox_WumbaCost(), InVisibility=HitTestInvisible)
  ]}
]
variables [
 Current Interaction {type=FortInteractContextInfo, default=null}
 CurrentWeaponDef {type=FortWeaponItemDefinition, default=null}
 Wumba Interaction Slot {type=integer}
 WumbaItemRef {type=B_Athena_Wumba_C, default=null}
]
widgets {
 info {inheritedWidgetCount=0, parentClass=&UserWidget, rootWidgetClass=&VerticalBox, widgetCount=8}
 widgets [
  HBox_Root {widgetClassPath=&VerticalBox, bIsVariable=true, widget=HBox_Root}
  VBox_WumbaCost {
   widgetClassPath=&HorizontalBox
   bIsVariable=true
   parent=HBox_Root
   widget=VBox_WumbaCost
   properties {visibility=Collapsed}
   slot {class=VerticalBoxSlot, properties {horizontalAlignment=HAlign_Center}}
  }
  ItemCard_WumbaUpgradeFrom {
   widgetClassPath=/Script/FortniteUI.FortMultiSizeItemCard
   bIsVariable=true
   parent=VBox_WumbaCost
   widget=ItemCard_WumbaUpgradeFrom
   properties {bPItemCardSize=S}
   slot {class=HorizontalBoxSlot}
  }
  Text_Plus {
   widgetClassPath=&CommonTextBlock
   parent=VBox_WumbaCost
   widget=Text_Plus
   properties {
    style=/FortUILibrary/Styles/Typography/Foundation/TextStyle_BurbankSmall_L_White.TextStyle_BurbankSmall_L_White_C
    text=+
    font {
     bForceMonospaced=false
     bMaterialIsStencil=false
     fontMaterial=null
     fontObject=&BurbankSmall
     letterSpacing=0
     monospacedWidth=1
     size=36
     skewAmount=0
     typefaceFontName=Black
     outlineSettings {
      bApplyOutlineToDropShadows=false
      bMiteredCorners=false
      bSeparateFillAlpha=false
      outlineMaterial=null
      outlineSize=0
      outlineColor=<rgba 0 0 0 1>
     }
    }
    shadowOffset=<0 0>
   }
   slot {class=HorizontalBoxSlot, properties {horizontalAlignment=HAlign_Center, verticalAlignment=VAlign_Center, padding=<ltrb 8 0 0 0>}}
  }
  WBP_WumbaResources {
   widgetClassPath=&/ItemCostWidget
   bIsVariable=true
   parent=VBox_WumbaCost
   widget=WBP_WumbaResources
   properties {bUseBackgroundForAffordabilityFeedback=true}
   slot {class=HorizontalBoxSlot, properties {padding=<ltrb -10 0 -10 0>}}
  }
  Text_Equal {
   widgetClassPath=&CommonTextBlock
   parent=VBox_WumbaCost
   widget=Text_Equal
   properties {
    style=&TextStyle_BurbankSmall_L_White_C
    text="="
    font {
     bForceMonospaced=false
     bMaterialIsStencil=false
     fontMaterial=null
     fontObject=&BurbankSmall
     letterSpacing=0
     monospacedWidth=1
     size=36
     skewAmount=0
     typefaceFontName=Black
     outlineSettings {
      bApplyOutlineToDropShadows=false
      bMiteredCorners=false
      bSeparateFillAlpha=false
      outlineMaterial=null
      outlineSize=0
      outlineColor=<rgba 0 0 0 1>
     }
    }
    shadowOffset=<0 0>
   }
   slot {class=HorizontalBoxSlot, properties {horizontalAlignment=HAlign_Center, verticalAlignment=VAlign_Center, padding=<ltrb 0 0 8 0>}}
  }
  ItemCard_WumbaUpgradeTo {
   widgetClassPath=&FortMultiSizeItemCard
   bIsVariable=true
   parent=VBox_WumbaCost
   widget=ItemCard_WumbaUpgradeTo
   properties {bPItemCardSize=S}
   slot {class=HorizontalBoxSlot}
  }
  WBP_WumbaError {
   widgetClassPath=/Game/Athena/HUD/WBP_InteractionWarning
   bIsVariable=true
   parent=HBox_Root
   widget=WBP_WumbaError
   slot {class=VerticalBoxSlot}
  }
 ]
}
