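# Athena HUD indicator widget that shows keybind prompts, descriptions and failure warnings for the player's current interaction target.
# HUD context events drive Handle Interaction Changed, which fills both contexts with basic, pickup, override, Wumba upgrade or matchmaking portal displays.
# A bound multi-interact component updates interactor counts and hold progress.
# Stateful Gameplay Event Router events inject async-loaded extra widgets per targeted actor, and touch handlers support tap-to-use.
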WidgetBlueprint /Game/Athena/HUD/AthenaInteractionIndicator
graphs [
 BindMultiInteractUpdate {nodes [
   BindMultiInteractUpdate() {is=title K2Node_FunctionEntry}
   n1 = Get CurrentInteraction() -> (CurrentInteraction:Fort Interact Context Info Object Reference)
   > n2 = Set NewMultiInteractComponent(NewMultiInteractComponent:Fort Actor Component Multi Interact Object Reference=GetComponentByClass!(
     self:Actor Object Reference=Get ReceivingActor!(self:Fort Interact Context Info Object Reference=@n1) -> (ReceivingActor:Actor Weak object)
     ComponentClass:Actor Component Class Reference=/Script/FortniteGame.FortActorComponent_MultiInteract
    ) -> (ReturnValue:Fort Actor Component Multi Interact Object Reference)) -> (Output_Get:Fort Actor Component Multi Interact Object Reference)
   > Branch(Condition="Not Equal (Object)"(
     A:Object Reference=@n2
     B:Object Reference=Get BoundMultiInteractComponent() -> (BoundMultiInteractComponent:Fort Actor Component Multi Interact Object Reference)
    ) -> (ReturnValue:bool=~) {is=title K2Node_PromotableOperator})
   > Unbind Multi Interact Update()
   n3 = Get CurrentInteraction()
   n4 = Get NewMultiInteractComponent() -> (NewMultiInteractComponent:Fort Actor Component Multi Interact Object Reference)
   > Branch(Condition="integer >= integer"(A=Get Interact Actor Count!(
      CurrentInteraction:Fort Interact Context Info Object Reference=@n3
      MultiInteractComponent:Fort Actor Component Multi Interact Object Reference=@n4
      AppliesToInteraction:EInteractionBeingAttempted Enum=FirstInteraction
     ) -> (InteractionCount:int=0) {is=title K2Node_CallFunction}))
   > n5 = Get!(execute:Exec) -> (then:Exec, else:Exec, NewMultiInteractComponent:Fort Actor Component Multi Interact Object Reference)
   > n6 = Set BoundMultiInteractComponent(BoundMultiInteractComponent:Fort Actor Component Multi Interact Object Reference=@n5) -> (Output_Get:Fort Actor Component Multi Interact Object Reference)
   > Bind Event to Actor Count Changed Delegate(self:Fort Actor Component Multi Interact Object Reference=@n6, Delegate=Create Event()) {is=title K2Node_AddDelegate}
   > n7 = Cast To MultiInteractActorComponent!(execute, Object=Get BoundMultiInteractComponent()) -> (then, CastFailed, AsMulti Interact Actor Component:Multi Interact Actor Component Object Reference) {is=title K2Node_DynamicCast}
   > SetRequiredDuration!(
    execute
    self:Multi Interact Actor Component Object Reference=@n7
    NewRequiredDuration:double=Get RequiredDuration!(self:Fort Interact Context Info Object Reference=Get CurrentInteraction()) -> (RequiredDuration:float)
   ) -> (then)
  ]}
 ClearGameplayEventContext {nodes [
   ClearGameplayEventContext() {is=title K2Node_FunctionEntry}
   > n1 = Sequence()
   > Branch(Condition=IsValid(Object=Get Targeted Actor() -> (Targeted Actor:Actor Object Reference)))
   > ClearEventStateByContext!(
    execute
    self:Gameplay Event Router Component Object Reference=GetGlobalEventRouter()
    EventType:Script Struct Object Reference=/Game/Athena/HUD/InteractionData/AthenaInteraction_AddExtraDataSlot_Event
    Context:Object Reference=Get Targeted Actor()
   ) -> (then)
   > Set Targeted Actor(Targeted Actor:Actor Object Reference) -> (Output_Get:Actor Object Reference)
   n2 = Get ExtraDataEvent() -> (ExtraDataEvent:Async Action Start Listening To Stateful Event Object Reference)
   Branch(execute=@n1.then_1, Condition=IsValid(Object=@n2))
   > Cancel(self:Cancellable Async Action Object Reference=@n2)
   > Set ExtraDataEvent(ExtraDataEvent:Async Action Start Listening To Stateful Event Object Reference) -> (Output_Get:Async Action Start Listening To Stateful Event Object Reference)
  ]}
 EventGraph {nodes [
   OnInputMethodChanged_Event_0!() -> (OutputDelegate, then, bNewInputType:ECommonInputType Enum) {is=title K2Node_CustomEvent}
   > UpdateForCurrentInput()
   LoadWumbaWidgetClass() {is=title K2Node_CustomEvent}
   > n1 = Async Load Class Asset!(execute:Exec, AssetClass:Object Soft Class Reference=Get WumbaWidgetSoftAsset() -> (WumbaWidgetSoftAsset:Widget Soft Class Reference)) -> (then:Exec, Completed:Exec, Class:Object Class Reference) {is=title K2Node_LoadAssetClass}
   n2 = Cast To UserWidget Class!(execute:Exec=@n1.Completed, Class:Object Class Reference=@n1) -> (then:Exec, CastFailed:Exec, AsUser Widget:User Widget Class Reference) {is=title K2Node_ClassDynamicCast}
   > Set WumbaWidgetClass(WumbaWidgetClass:User Widget Class Reference=@n2) -> (Output_Get:User Widget Class Reference)
   > SetupWumbaInteraction()
   n3 = EventUpdateTargeting!() -> (OutputDelegate, then, bIsTargeting:bool=~) {is=title K2Node_CustomEvent}
   > SetRenderOpacity(
    self=[Get PrimaryBorder() -> (PrimaryBorder:Border Object Reference), Get SecondaryBorder() -> (SecondaryBorder:Border Object Reference)]
    InOpacity=Select(Option 0:double=1, Option 1:double=0.25, Index=@n3.bIsTargeting) -> (ReturnValue:double)
   )
   ListeningExtraData() {is=title K2Node_CustomEvent}
   > n4 = StartListeningToStatefulEvent!(
    execute:Exec
    Target:Gameplay Event Router Component Object Reference=GetGlobalEventRouter()
    EventType:Script Struct Object Reference=&AthenaInteraction_AddExtraDataSlot_Event
    Context:Object Reference=Get Targeted Actor()
    EventBubblingRule:EEventBubblingRule Enum
   ) -> (
    then:Exec
    AsyncTaskProxy:Async Action Start Listening To Stateful Event Object Reference
    OnEventReceived:Exec
    OnSavedEventStateExists:Exec
    OnEventStateCleared:Exec
    Payload_WidgetData_3_#1:User Widget Soft Class Reference
    Payload_InteractionType_7_#2:EInteractionBeingAttempted Enum
   ) {is=title K2Node_StartListeningToStatefulEventAsyncAction}
   > Set ExtraDataEvent(ExtraDataEvent=@n4)
   Branch(
    execute=[@n4.OnEventReceived, @n4.OnSavedEventStateExists]
    Condition="Not Equal (SoftClassReference)"(
     A:"Object Soft Class Reference (by ref)"=@n4.Payload_WidgetData_3_#1
     B:"Object Soft Class Reference (by ref)"=Get ExtraDataSlotWidget() -> (ExtraDataSlotWidget:User Widget Soft Class Reference)
    ) -> (ReturnValue:bool=~) {is=title K2Node_PromotableOperator}
   )
   > n5 = Sequence()
   > Set ExtraDataSlotWidget(ExtraDataSlotWidget:User Widget Soft Class Reference=@n4.Payload_WidgetData_3_#1) -> (Output_Get:User Widget Soft Class Reference)
   > Set Payload Interaction Type(Payload Interaction Type:EInteractionBeingAttempted Enum=@n4.Payload_InteractionType_7_#2) -> (Output_Get:EInteractionBeingAttempted Enum=FirstInteraction)
   > ClearChildren(self=[
     Get Slot_ExtraAppend_FirstIntection() -> (Slot_ExtraAppend_FirstIntection:Overlay Object Reference)
     Get Slot_ExtraAppend_SecondInteraction() -> (Slot_ExtraAppend_SecondInteraction:Overlay Object Reference)
    ])
   AddWidgetExtraData(execute=@n5.then_1)
   > ClearEventStateByContext(
    self=GetGlobalEventRouter()
    EventType=&AthenaInteraction_AddExtraDataSlot_Event
    Context=Get Targeted Actor()
   )
   AddWidgetExtraData() {is=title K2Node_CustomEvent}
   > n6 = Get!(execute) -> (then, else, ExtraDataSlotWidget:User Widget Soft Class Reference)
   > n7 = Async Load Class Asset(AssetClass=@n6)
   n8 = Cast To UserWidget Class(execute=@n7.Completed, Class=@n7)
   > n9 = Create Widget!(execute:Exec, Class:User Widget Class Reference=@n8, OwningPlayer:Player Controller Object Reference) -> (then:Exec, ReturnValue:User Widget Object Reference)
   n10 = Select!(
    FirstInteraction:Overlay Object Reference=Get Slot_ExtraAppend_FirstIntection()
    SecondInteraction:Overlay Object Reference=Get Slot_ExtraAppend_SecondInteraction()
    AllInteraction:Overlay Object Reference
    Index:EInteractionBeingAttempted Enum=Get Payload Interaction Type() -> (Payload Interaction Type:EInteractionBeingAttempted Enum=FirstInteraction)
   ) -> (ReturnValue:Overlay Object Reference)
   > Is Valid(InputObject=@n10)
   > n11 = AddChildToOverlay!(execute, self:Overlay Object Reference=@n10, Content:Widget Object Reference=@n9) -> (then, ReturnValue:Overlay Slot Object Reference)
   > SetHorizontalAlignment(self=@n11, InHorizontalAlignment=HAlign_Fill)
   > SetVerticalAlignment(self=@n11, InVerticalAlignment=VAlign_Fill)
   n12 = Event OnTargetActorChanged!() -> (OutputDelegate, then, TargetedActor:Actor Object Reference) {is=title K2Node_Event}
   > n13 = Branch(Condition="Not Equal (Object)"(A=@n12.TargetedActor, B=Get Targeted Actor()) -> (ReturnValue=~))
   > ClearGameplayEventContext() {is=title K2Node_CallFunction}
   > Set Targeted Actor(Targeted Actor=@n12.TargetedActor)
   > Set ExtraDataSlotWidget()
   > ClearChildren(self=Get Slot_ExtraAppend_SecondInteraction())
   > ListeningExtraData() {is=title K2Node_CallFunction}
   ClearGameplayEventContext(execute=@n13.else)
   n14 = Event HandleValidInteractResponse!() -> (OutputDelegate, then, InteractContextInfo:Fort Interact Context Info Object Reference) {is=title K2Node_Event}
   > Handle Interaction Changed!(execute, self, CurrentInteraction:Fort Interact Context Info Object Reference=@n14.InteractContextInfo) -> (then)
   Event Destruct()
   n15 = Get FortHUDContext!() -> (ReturnValue:Fort HUDContext Object Reference) {is=title K2Node_GetBlueprintContextEx}
   > n16 = Is Valid(InputObject=@n15)
   n17 = Create Event()
   > Unbind Event from On Player Can Interact(self:Fort HUDContext Object Reference=@n15, Delegate=@n17) {is=title K2Node_RemoveDelegate}
   n18 = Create Event()
   n19 = Get FortHUDContext()
   > Unbind Event from On Player Targeting Changed(self:Fort HUDContext Object Reference=@n19, Delegate=@n18) {is=title K2Node_RemoveDelegate}
   n20 = Create Event()
   n21 = Get FortHUDContext()
   > n22 = Unbind Event from On Interact Updated(self:Fort HUDContext Object Reference=@n21, Delegate=@n20) {is=title K2Node_RemoveDelegate}
   n23 = Get FortGlobalUIContext!() -> (ReturnValue:Fort Global UIContext Object Reference) {is=title K2Node_GetBlueprintContextEx}
   n24 = Is Valid(exec=[@n22.then, @n16.Is Not Valid], InputObject=@n23)
   n25 = Create Event()
   > n26 = Unbind Event from On Keybinds Changed(self:Fort Global UIContext Object Reference=@n23, Delegate=@n25) {is=title K2Node_RemoveDelegate}
   Unbind Multi Interact Update(execute=[@n26.then, @n24.Is Not Valid])
   > ClearChildren(self=Get Slot_ExtraAppend_SecondInteraction())
   > StopListeningToStatefulEvent(self:Async Action Start Listening To Stateful Event Object Reference=Get ExtraDataEvent())
   Event Construct()
   > Bind Event to On Player Can Interact(self:Fort HUDContext Object Reference=@n15, Delegate=@n17) {is=title K2Node_AddDelegate}
   > Bind Event to On Player Targeting Changed(self:Fort HUDContext Object Reference=@n19, Delegate=@n18) {is=title K2Node_AddDelegate}
   > Bind Event to On Interact Updated(self:Fort HUDContext Object Reference=@n21, Delegate=@n20) {is=title K2Node_AddDelegate}
   > Assign On Keybinds Changed!(execute:Exec, self:Fort Global UIContext Object Reference=@n23, Delegate:"Delegate (by ref)"=@n25) -> (then:Exec) {is=title K2Node_AssignDelegate}
   > RefreshInteractionFailureText()
   n27 = Get UseKeybindWidgets() -> (UseKeybindWidgets:Array of Keybind Widget Object References)
   > Add(
    TargetArray:Array of Keybind Widget Object References=@n27
    NewItem:"Keybind Widget Object Reference (by ref)"=Get KeybindWidget() -> (KeybindWidget:Keybind Widget Object Reference)
   ) {is=title K2Node_CallArrayFunction}
   > Add(TargetArray=@n27, NewItem=Get Keybind_PrimaryPortal() -> (Keybind_PrimaryPortal:Keybind Widget Object Reference))
   n28 = Get AltInteractionKeybindWidgets() -> (AltInteractionKeybindWidgets:Array of Keybind Widget Object References)
   > Add(TargetArray=@n28, NewItem=Get SecondInteractionKeybind() -> (SecondInteractionKeybind:Keybind Widget Object Reference))
   > Add(TargetArray=@n28, NewItem=Get Keybind_SecondaryPortal() -> (Keybind_SecondaryPortal:Keybind Widget Object Reference))
   > UpdateKeybinds()
   > Assign On Input Method Changed!(execute:Exec, self:Common Input Object Reference=Get CommonInputSubsystem(), Delegate:"Delegate (by ref)"=Create Event()) -> (then:Exec) {is=title K2Node_AssignDelegate}
   > UpdateForCurrentInput()
  ]}
 Get Interact Actor Count {nodes [
   n1 = Get Interact Actor Count!() -> (
    then
    CurrentInteraction:Fort Interact Context Info Object Reference
    MultiInteractComponent:Fort Actor Component Multi Interact Object Reference
    AppliesToInteraction:EInteractionBeingAttempted Enum=FirstInteraction
   ) {is=title K2Node_FunctionEntry}
   > n2 = Is Valid(InputObject=@n1.MultiInteractComponent)
   Return Node!(execute=@n2.Is Not Valid, InteractionCount:int=-1)
   > n3 = Branch(Condition=IsMultipleInteractionEnabled!(self:Fort Actor Component Multi Interact Object Reference=@n1.MultiInteractComponent) -> (ReturnValue:bool))
   > n4 = Branch(Condition="Equal (Enum)"(
     A:EInteractionBeingAttempted Enum=Get AppliesToInteraction!(self:Fort Actor Component Multi Interact Object Reference=@n1.MultiInteractComponent) -> (AppliesToInteraction:EInteractionBeingAttempted Enum=FirstInteraction)
     B:EInteractionBeingAttempted Enum=@n1.AppliesToInteraction
    ))
   n5 = Get RequiredDuration(self=@n1)
   n6 = Get SecondRequiredDuration!(self:Fort Interact Context Info Object Reference=@n1) -> (SecondRequiredDuration:float)
   > n7 = Branch(Condition="float > float"(A=Select(FirstInteraction:float=@n5, SecondInteraction:float=@n6, AllInteraction:float=0, Index=@n1.AppliesToInteraction) -> (ReturnValue:float)))
   > Return Node(InteractionCount=GetInteractingActorCount!(
     self:Fort Actor Component Multi Interact Object Reference=@n1.MultiInteractComponent
     bIncludePlayers:bool=true
     bIncludeNonPlayers:bool=true
    ) -> (ReturnValue:int=0))
   Return Node(execute=@n7.else, InteractionCount=-1)
   Return Node(execute=@n4.else, InteractionCount=-1)
   Return Node(execute=@n3.else, InteractionCount=-1)
  ]}
 Handle Interaction Changed {nodes [
   n1 = Handle Interaction Changed!() -> (then, CurrentInteraction:Fort Interact Context Info Object Reference) {is=title K2Node_FunctionEntry}
   > n2 = Set CurrentInteraction(CurrentInteraction:Fort Interact Context Info Object Reference=@n1) -> (Output_Get:Fort Interact Context Info Object Reference)
   > n3 = Is Valid(InputObject=@n2)
   n4 = Get bHasCustomInteractionWidget!(self:Fort Interact Context Info Object Reference=Get CurrentInteraction()) -> (bHasCustomInteractionWidget:bool)
   n5 = IsValid(Object=Get OptionalHUDDisplayWidget!(self:Fort Interact Context Info Object Reference=Get CurrentInteraction()) -> (OptionalHUDDisplayWidget:User Widget Object Reference))
   > n6 = Set ValidInteraction(ValidInteraction:bool=AND Boolean(A=NOT Boolean(A=@n4), B=NOT Boolean(A=@n5))) -> (Output_Get:bool)
   SetVisibility(
    execute=[@n6.then, @n3.Is Not Valid]
    InVisibility=Select!(
     Option 0:ESlateVisibility Enum=Collapsed
     Option 1:ESlateVisibility Enum=SelfHitTestInvisible
     Index:bool=Get ValidInteraction() -> (ValidInteraction:bool)
    ) -> (ReturnValue:ESlateVisibility Enum)
   )
   > SetVisibility(
    self=[
     Get WBP_InteractionWarning() -> (WBP_InteractionWarning:WBP Interaction Warning Object Reference)
     Get HUD-PickupItemWidget() -> (HUD-PickupItemWidget:Athena HUD-Pickup Item Widget Object Reference)
     Get Border_SubDescription() -> (Border_SubDescription:Border Object Reference)
     Get Overlay_Wumba_Upgrade() -> (Overlay_Wumba_Upgrade:Overlay Object Reference)
     Get Overlay_Wumba_Sidegrade() -> (Overlay_Wumba_Sidegrade:Overlay Object Reference)
     Get SecondInteractionBox() -> (SecondInteractionBox:Horizontal Box Object Reference)
     Get BasicInteractionWidget() -> (BasicInteractionWidget:Basic Interaction Widget Object Reference)
    ]
    InVisibility=Collapsed
   )
   > SetVisibility(
    self=[
     Get InteractionKeybindBox() -> (InteractionKeybindBox:Vertical Box Object Reference)
     Get Interaction() -> (Interaction:Horizontal Box Object Reference)
    ]
    InVisibility=SelfHitTestInvisible
   )
   > SetContentColorAndOpacity!(
    execute
    self:Border Object Reference=Get InteractionKeybindColor() -> (InteractionKeybindColor:Common Border Object Reference)
    InContentColorAndOpacity:Linear Color Structure=LinearColor_White!() -> (ReturnValue:Linear Color Structure)
   ) -> (then)
   > SetActiveWidgetIndex(
    self:Common Visibility Switcher Object Reference=[
     Get Switcher_PrimaryContext() -> (Switcher_PrimaryContext:Common Visibility Switcher Object Reference)
     Get Switcher_SecondaryContext() -> (Switcher_SecondaryContext:Common Visibility Switcher Object Reference)
    ]
    Index=0
   )
   > n7 = Branch(Condition=Get ValidInteraction())
   n8 = Get CurrentInteraction()
   > SetOverrideTextPositionMobile!(
    execute
    self:Fort Actor Indicator Widget Object Reference
    OverrideTextPositionMobile:bool=Get bOverrideInteractionTextPositionMobile!(self:Fort Interact Context Info Object Reference=@n8) -> (bOverrideInteractionTextPositionMobile:bool)
   ) -> (then)
   > SetIndicatedActor!(execute, self:Fort Actor Indicator Widget Object Reference, NewIndicatorActor:Actor Object Reference=Get ReceivingActor(self=@n8)) -> (then)
   > Set IndicatedActorComponent(IndicatedActorComponent:Primitive Component Weak object=Get InteractComponent!(self:Fort Interact Context Info Object Reference=@n8) -> (InteractComponent:Primitive Component Weak object)) -> (Output_Get:Primitive Component Weak object)
   > HandleFirstContext()
   > HandleSecondContext()
   > UpdateWidgetVisibility()
   > Update Keybind Custom Progress!(execute, self, SecondsToInterpolate:double=2) -> (then)
   SetPickup!(execute=@n7.else, self:Athena HUDPickup Item Widget Base Object Reference=Get HUD-PickupItemWidget(), NewPickup:Fort Pickup Object Reference) -> (then)
   > RefreshInteractionFailureText()
   > DisableCustomProgress(self:Fort Keybind Widget Object Reference=Get KeybindWidget())
   > Unbind Multi Interact Update()
  ]}
 HandleFirstContext {nodes [
   HandleFirstContext() {is=title K2Node_FunctionEntry}
   > SetVisibility(self=Get VerticalBox_PortalMatchMaking() -> (VerticalBox_PortalMatchMaking:Vertical Box Object Reference), InVisibility=Collapsed)
   > SetVisibility(self=[Get SecondaryBorder(), Get Switcher_PrimaryContext(), Get PrimaryBorder()], InVisibility=SelfHitTestInvisible)
   > n1 = Branch(Condition=HasValidContextOverride!(self:Fort Interact Context Info Object Reference=Get CurrentInteraction()) -> (ReturnValue:bool))
   > ShowContextOverride()
   RefreshInteractionFailureText(execute=@n1.else)
   > n2 = Cast To FortPickup!(execute, Object=GetIndicatedActor!(self:Fort Actor Indicator Interface Interface) -> (ReturnValue:Actor Object Reference)) -> (then, CastFailed, AsFort Pickup:Fort Pickup Object Reference) {is=title K2Node_DynamicCast}
   > n3 = Branch(Condition=PickedUp!(self:Fort Pickup Object Reference=@n2) -> (ReturnValue:bool))
   > SetVisibility(InVisibility=Collapsed)
   n4 = Branch(execute=@n3.else, Condition=Get bUsePickupWidget!(self:Fort Pickup Object Reference=@n2) -> (bUsePickupWidget:bool))
   > ShowPickupWidget!(execute, self, Pickup:Fort Pickup Object Reference=@n2) -> (then)
   ShowBasicInteractionWidget(execute=@n4.else)
   ShowBasicInteractionWidget(execute=@n2.CastFailed)
   n5 = GetIndicatedActor()
   > n6 = HasMatchingGameplayTag!(execute:Exec, self:Object Reference=@n5, TagToCheck:Gameplay Tag Structure=(TagName="Building.Type.ItemContainer.Wumba")) -> (then:Exec, ReturnValue:bool) {is=title K2Node_Message}
   > n7 = Branch(Condition=@n6)
   > LoadWumbaWidgetClass() {is=title K2Node_CallFunction}
   n8 = HasMatchingGameplayTag(execute=@n7.else, self=@n5, TagToCheck=(TagName="Building.Type.Prop.CreativeMatchmakingPortal"))
   > Branch(Condition=@n8)
   > n9 = AthenaInteractDetails!(execute:Exec, self:Object Reference=@n5) -> (then:Exec, LocalPlayerLookingAt:bool, IsInPlayMode:bool) {is=title K2Node_Message}
   > Branch(Condition=AND Boolean(A=NOT Boolean(A=@n9), B=CanInitiatePrivateMatchmaking!() -> (ReturnValue:bool)))
   > ShowMatchmakingPortal!(execute, self, Portal:"Actor Object Reference (by ref)"=@n5) -> (then)
  ]}
 HandleKeybindsChanged {nodes [
   HandleKeybindsChanged() {is=title K2Node_FunctionEntry}
   > Handle Interaction Changed(CurrentInteraction=Get CurrentInteraction()) {is=title K2Node_CallFunction}
  ]}
 HandleSecondContext {nodes [
   HandleSecondContext() {is=title K2Node_FunctionEntry}
   > ShowSecondInteractionWidget()
   n1 = HasValidSecondContextOverride!(self:Fort Interact Context Info Object Reference=Get CurrentInteraction()) -> (ReturnValue:bool)
   > SetVisibility(
    self=Get BasicInteractionWidgetSecondary() -> (BasicInteractionWidgetSecondary:Basic Interaction Widget Object Reference)
    InVisibility=Select(Option 0=HitTestInvisible, Option 1=Collapsed, Index=@n1)
   )
   > Branch(Condition=@n1)
   n2 = Get CurrentInteraction()
   > ActivateContextOverride!(
    execute
    self:Fort Interaction Indicator Widget Object Reference
    ContextOverrideWidget:"Fort Interact Context Info Widget Class Reference (by ref)"=Get SecondContextOverrideWidget!(self:Fort Interact Context Info Object Reference=@n2) -> (SecondContextOverrideWidget:Fort Interact Context Info Widget Class Reference)
    ContextInfo:Fort Interact Context Info Object Reference=@n2
    bPrimary:bool=false
   ) -> (then, ReturnValue:Fort Interact Context Info Widget Object Reference)
  ]}
 IsInteractionWidgetSuppressed {nodes [
   IsInteractionWidgetSuppressed() {is=title K2Node_FunctionEntry}
   > n1 = Is Valid(InputObject=Get CurrentInteraction())
   Return Node!(execute=@n1.Is Not Valid, IsSuppressed:bool=false)
   n2 = IsUsingTouch(PlayerController=GetOwningPlayer())
   n3 = IsSimpleInteractionWidgetVisible!() -> (Visible:bool)
   n4 = Get CurrentInteraction()
   n5 = Get bSuppressSimpleInteractionWidgetForTouch!(self:Fort Interact Context Info Object Reference=@n4) -> (bSuppressSimpleInteractionWidgetForTouch:bool)
   > Return Node(IsSuppressed=OR Boolean(
     A=Get bSuppressInteractionWidget!(self:Fort Interact Context Info Object Reference=@n4) -> (bSuppressInteractionWidget:bool)
     B=AND Boolean!(A=@n5, B=@n3, C:bool=@n2) -> (ReturnValue)
    ))
  ]}
 IsSimpleInteractionWidgetVisible {nodes [
   IsSimpleInteractionWidgetVisible() {is=title K2Node_FunctionEntry}
   n1 = OR Boolean!(
    A=IsVisible(self=Get HUD-PickupItemWidget())
    B=IsVisible(self=Get Overlay_Wumba_Upgrade())
    C:bool=IsVisible(self=Get Overlay_Wumba_Sidegrade())
   ) -> (ReturnValue)
   n2 = Get BasicInteractionWidget()
   > Return Node!(execute, Visible:bool=AND Boolean!(A=NOT Boolean(A=@n1), B=IsVisible(self=@n2)) -> (ReturnValue))
  ]}
 OnMultiInteractCountChanged {nodes [
   n1 = OnMultiInteractCountChanged!() -> (then, InteractingActorCount:int, MultiInteractMultiplier:float=~, InteractingDuration:float=~) {is=title K2Node_FunctionEntry}
   > UpdateMultiInteractionCount!(execute, self:Basic Interaction Widget Base Object Reference=Get BasicInteractionWidget(), NewInteractionCount:int=@n1) -> (then)
   > UpdateProgressMaterial!(execute, self, InteractionCount:int=@n1) -> (then)
   > Update Keybind Custom Progress(SecondsToInterpolate=1.5)
  ]}
 OnTouchEnded {nodes [
   n1 = OnTouchEnded!() -> (then, MyGeometry:Geometry Structure, InTouchEvent:"Pointer Event Structure (by ref)") {is=title K2Node_FunctionEntry}
   > n2 = Branch(Condition=HasMouseCaptureByUser!(
     self:Widget Object Reference
     UserIndex:int=Get User Index!(Input:"Pointer Event Structure (by ref)"=@n1.InTouchEvent) -> (ReturnValue:int=0) {is=title K2Node_CallFunction}
     PointerIndex:int=Get Pointer Index!(Input:"Pointer Event Structure (by ref)"=@n1.InTouchEvent) -> (ReturnValue:int=0) {is=title K2Node_CallFunction}
    ) -> (ReturnValue:bool))
   > TryBestUseForActor!(
    execute
    self:Athena HUDContext Object Reference=Get AthenaHUDContext!() -> (ReturnValue:Athena HUDContext Object Reference) {is=title K2Node_GetBlueprintContextEx}
    ActorToUse:Actor Object Reference=Get ReceivingActor(self=Get CurrentInteraction())
   ) -> (then)
   > Return Node!(
    execute
    ReturnValue:Event Reply Structure=ReleaseMouseCapture!(Reply:"Event Reply Structure (by ref)"=Unhandled()) -> (ReturnValue:Event Reply Structure)
   )
   Return Node(execute=@n2.else, ReturnValue=Unhandled())
  ]}
 OnTouchMoved {nodes [
   n1 = OnTouchMoved!() -> (then, MyGeometry:Geometry Structure, InTouchEvent:"Pointer Event Structure (by ref)") {is=title K2Node_FunctionEntry}
   n2 = Get User Index(Input=@n1.InTouchEvent) -> (ReturnValue=0)
   n3 = Get Pointer Index(Input=@n1.InTouchEvent) -> (ReturnValue=0)
   > n4 = Branch(Condition=NOT Boolean(A=HasMouseCaptureByUser(UserIndex=@n2, PointerIndex=@n3)))
   > Return Node(ReturnValue=Unhandled())
   n5 = Distance2D!(
    V1:Vector 2D Structure=Get Screen Space Position!(Input:"Pointer Event Structure (by ref)"=Get InTouchEvent() -> (InTouchEvent:"Pointer Event Structure (by ref)")) -> (ReturnValue:Vector 2D Structure) {is=title K2Node_CallFunction}
    V2:Vector 2D Structure=Get Touch Start Position() -> (Touch Start Position:Vector 2D Structure)
   ) -> (ReturnValue:double)
   n6 = Get Max Tap Distance() -> (Max Tap Distance:double)
   n7 = float - float(A=GetTimeSeconds!() -> (ReturnValue:double), B=Get Touch Start Time() -> (Touch Start Time:double)) {is=title K2Node_CallFunction}
   n8 = Get Max Tap Time() -> (Max Tap Time:double)
   n9 = Branch(
    execute=@n4.else
    Condition=AND Boolean(A="float < float"!(A:double=@n7, B:double=@n8) -> (ReturnValue:bool) {is=title K2Node_CallFunction}, B="float < float"(A=@n5, B=@n6))
   )
   > Return Node(ReturnValue=Handled())
   Return Node(execute=@n9.else, ReturnValue=ReleaseMouseCapture(Reply=Handled()))
  ]}
 OnTouchStarted {nodes [
   n1 = OnTouchStarted!() -> (then, MyGeometry:Geometry Structure, InTouchEvent:"Pointer Event Structure (by ref)") {is=title K2Node_FunctionEntry}
   n2 = Get AthenaHUDContext()
   > n3 = Branch(Condition="Not Equal (Enum)"!(
     A:ETouchInteractMode Enum=GetTouchInteractMode!(self:Athena HUDContext Object Reference=@n2) -> (ReturnValue:ETouchInteractMode Enum=Off)
     B:ETouchInteractMode Enum=Off
    ) -> (ReturnValue:bool))
   > Set Touch Start Position(Touch Start Position:Vector 2D Structure=Get Screen Space Position(Input=@n1.InTouchEvent)) -> (Output_Get:Vector 2D Structure)
   > Set Touch Start Time(Touch Start Time:double=GetTimeSeconds()) -> (Output_Get:double)
   > Return Node(ReturnValue=CaptureMouse!(Reply:"Event Reply Structure (by ref)"=Handled(), CapturingWidget:Widget Object Reference=Self-Reference()) -> (ReturnValue:Event Reply Structure))
   Return Node(execute=@n3.else, ReturnValue=Unhandled())
  ]}
 RefreshInteractionFailureText {nodes [
   RefreshInteractionFailureText() {is=title K2Node_FunctionEntry}
   n1 = Get CurrentInteraction()
   > n2 = Is Valid(InputObject=@n1)
   > n3 = Branch(Condition=TextIsEmpty(InText=Get ErrorText!(self:Fort Interact Context Info Object Reference=@n1) -> (ErrorText:Text)))
   SetVisibility(execute=[@n3.then, @n2.Is Not Valid], self=Get WBP_InteractionWarning(), InVisibility=Collapsed)
   SetVisibility(execute=@n3.else, self=Get WBP_InteractionWarning(), InVisibility=SelfHitTestInvisible)
   > SetFailureMessage!(execute, self:WBP Interaction Warning Object Reference=Get WBP_InteractionWarning(), Text:Text=Get ErrorText(self=Get CurrentInteraction())) -> (then)
   n4 = Get ErrorIconTag!(self:Fort Interact Context Info Object Reference=Get CurrentInteraction()) -> (ErrorIconTag:Gameplay Tag Structure)
   > SetIcon!(
    execute
    self:WBP Interaction Warning Object Reference=Get WBP_InteractionWarning()
    NewIcon:Texture 2D Soft Object Reference=Select(
     Option 0:Texture 2D Soft Object Reference=/Game/UI/Foundation/Textures/Icons/HUD/T-Icon-Alert2-32
     Option 1:Texture 2D Soft Object Reference=/Game/UI/Foundation/Textures/Icons/HUD/T-Icon-Backpack-64
     Index="Equal (GameplayTag)"(A=@n4, B=(TagName="HUD.Athena.ErrorIcon.Interaction.BackpackFull"))
    ) -> (ReturnValue:Texture 2D Soft Object Reference)
   ) -> (then)
  ]}
 SetKeybindCustomProgress {nodes [
   SetKeybindCustomProgress!() -> (then, UseCustomProgress:bool=~, SecondsToInterpolate:double=~, ChangedTarget:bool=~, IsSecondary:bool=~) {is=title K2Node_FunctionEntry}
   > n1 = Cast To KeybindWidget!(
    execute
    Object=Select(
     Option 0:Object Reference=Get KeybindWidget()
     Option 1:Object Reference=Get SecondInteractionKeybind()
     Index=Get IsSecondary() -> (IsSecondary:bool)
    ) -> (ReturnValue:Object Reference)
   ) -> (then, CastFailed, AsKeybind Widget:Keybind Widget Object Reference) {is=title K2Node_DynamicCast}
   n2 = Get BoundMultiInteractComponent()
   > n3 = Is Valid(InputObject=@n2)
   > n4 = Branch(Condition=Get ChangedTarget() -> (ChangedTarget:bool))
   DisableCustomProgress(execute=[@n6.CastFailed, @n5.else, @n4.then, @n3.Is Not Valid], self=@n1)
   n5 = Branch(execute=@n4.else, Condition=Get UseCustomProgress() -> (UseCustomProgress:bool))
   > n6 = Cast To MultiInteractActorComponent(Object=@n2)
   n7 = Get CurrentInteraction()
   n8 = Get CurrentInteraction()
   > EnableCustomProgress!(
    execute
    self:Fort Keybind Widget Object Reference=@n1
    RequiredDuration:float=SelectFloat(A=Get SecondRequiredDuration(self=@n8), B=Get RequiredDuration(self=@n7), bPickA=Get IsSecondary())
    InteractingMultiplier:float=GetMultiplier!(self:Fort Actor Component Multi Interact Object Reference=@n6) -> (ReturnValue:float)
    InteractDuration:float=GetInteractionDuration!(self:Fort Actor Component Multi Interact Object Reference=@n6) -> (ReturnValue:float)
    SecondsToInterpolateValues:float=Get SecondsToInterpolate() -> (SecondsToInterpolate:double)
   ) -> (then)
  ]}
 SetupWumbaInteraction {nodes [
   SetupWumbaInteraction() {is=title K2Node_FunctionEntry}
   > SetVisibility(
    self=[
     Get OuterBorder!(self:Basic Interaction Widget Base Object Reference=Get BasicInteractionWidget()) -> (OuterBorder:Border Object Reference)
     Get Border_SubDescription()
    ]
    InVisibility=Collapsed
   )
   > n1 = Is Valid(InputObject=Get Upgrade_WumbaWidget() -> (Upgrade_WumbaWidget:User Widget Object Reference))
   n2 = Create Widget(execute=@n1.Is Not Valid, Class=Get WumbaWidgetClass() -> (WumbaWidgetClass:User Widget Class Reference))
   > n3 = Set Upgrade_WumbaWidget(Upgrade_WumbaWidget:User Widget Object Reference=@n2) -> (Output_Get:User Widget Object Reference)
   > n4 = AddChildToOverlay(self=Get Overlay_Wumba_Upgrade(), Content=@n3)
   n5 = AttemptWumbaInteraction!(
    execute:Exec=[@n4.then, @n1.Is Valid]
    self:Object Reference=@n3
    Current Interaction:Fort Interact Context Info Object Reference=Get CurrentInteraction()
    Indicated Actor:Actor Object Reference=GetIndicatedActor()
   ) -> (then:Exec, CanAttemptInteraction:bool) {is=title K2Node_Message}
   > n6 = Branch(Condition=@n5)
   > SetVisibility(self=Get Overlay_Wumba_Upgrade(), InVisibility=SelfHitTestInvisible)
   > n7 = HandleWumbaInteraction!(execute:Exec, self:Object Reference=Get Upgrade_WumbaWidget(), UpgradeDirection:int=0) -> (then:Exec, Upgrade Interaction Result:EFortWeaponUpgradeInteractionResult Enum=Upgradable) {is=title K2Node_Message}
   n8 = LinearColor_Gray!() -> (ReturnValue:Linear Color Structure)
   > SetContentColorAndOpacity(
    self=Get InteractionKeybindColor()
    InContentColorAndOpacity=Select!(
     Upgradable:Linear Color Structure=LinearColor_White()
     NotEnoughResources:Linear Color Structure=@n8
     CannotUpgrade:Linear Color Structure=@n8
     CannotInteract:Linear Color Structure=@n8
     Index:EFortWeaponUpgradeInteractionResult Enum=@n7
    ) -> (ReturnValue:Linear Color Structure)
   )
   > SetVisibility(self=Get Overlay_Wumba_Sidegrade(), InVisibility=SelfHitTestInvisible)
   > n9 = Is Valid(InputObject=Get Sidegrade_WumbaWidget() -> (Sidegrade_WumbaWidget:User Widget Object Reference))
   n10 = Create Widget(execute=@n9.Is Not Valid, Class=Get WumbaWidgetClass())
   > n11 = Set Sidegrade_WumbaWidget(Sidegrade_WumbaWidget:User Widget Object Reference=@n10) -> (Output_Get:User Widget Object Reference)
   > n12 = AddChildToOverlay(self=Get Overlay_Wumba_Sidegrade(), Content=@n11)
   n13 = AttemptWumbaInteraction(execute=[@n12.then, @n9.Is Valid], self=@n11, Current Interaction=Get CurrentInteraction(), Indicated Actor=GetIndicatedActor())
   > n14 = Branch(Condition=@n13)
   > n15 = HandleWumbaInteraction(self=Get Sidegrade_WumbaWidget(), UpgradeDirection=1) -> (Upgrade Interaction Result=Upgradable)
   n16 = LinearColor_Gray()
   > SetContentColorAndOpacity(
    self=Get SecondInteractionKeybindColor() -> (SecondInteractionKeybindColor:Border Object Reference)
    InContentColorAndOpacity=Select(Upgradable=LinearColor_White(), NotEnoughResources=@n16, CannotUpgrade=@n16, CannotInteract=@n16, Index=@n15)
   )
   > SetVisibility(
    self=[Get Overlay_Wumba_Sidegrade(), Get SecondInteractionBox()]
    InVisibility=Select(
     Upgradable:ESlateVisibility Enum=SelfHitTestInvisible
     NotEnoughResources:ESlateVisibility Enum=SelfHitTestInvisible
     CannotUpgrade:ESlateVisibility Enum=Collapsed
     CannotInteract:ESlateVisibility Enum=Collapsed
     Index=@n15
    ) -> (ReturnValue:ESlateVisibility Enum)
   )
   SetVisibility(execute=@n14.else, self=[Get Overlay_Wumba_Sidegrade(), Get SecondInteractionBox()], InVisibility=Collapsed)
   n17 = LinearColor_Gray()
   SetContentColorAndOpacity(
    execute=@n6.else
    self=Get InteractionKeybindColor()
    InContentColorAndOpacity=Select(
     Upgradable:Linear Color Structure=LinearColor_White()
     NotEnoughResources:Linear Color Structure=@n17
     CannotUpgrade:Linear Color Structure=@n17
     CannotInteract:Linear Color Structure=@n17
     Index=Upgradable
    ) -> (ReturnValue:Linear Color Structure)
   )
   > SetVisibility(self=[Get Overlay_Wumba_Upgrade(), Get Overlay_Wumba_Sidegrade()], InVisibility=Collapsed)
  ]}
 ShouldShowCreativeModeName {nodes [
   ShouldShowCreativeModeName() {is=title K2Node_FunctionEntry}
   > n1 = Cast To FortPlayerPawnAthena(Object=GetPlayerPawn!(PlayerIndex:int=0) -> (ReturnValue:Pawn Object Reference))
   > n2 = Cast To FortPlayerControllerAthena(Object=GetController!(self:Pawn Object Reference=@n1) -> (ReturnValue:Controller Object Reference))
   > n3 = Branch(Condition=IsCreativeModeEnabled!(self:Fort Player Controller Gameplay Object Reference=@n2) -> (ReturnValue:bool))
   > n4 = Branch(Condition=HasStarted!(self:Fort Minigame Object Reference=GetMinigame!(ContextActor:Actor Object Reference=@n2) -> (ReturnValue:Fort Minigame Object Reference)) -> (ReturnValue:bool))
   Return Node!(execute=[@n8.then, @n6.CastFailed, @n4.then, @n3.else, @n2.CastFailed, @n1.CastFailed], Show:bool=false, CreativeNameText:Text)
   n5 = GetIndicatedActor()
   n6 = Cast To FortActorLabelOverrideInterface!(execute=@n4.else, Object=@n5) -> (then, CastFailed, AsFort Actor Label Override Interface:Fort Actor Label Override Interface Interface) {is=title K2Node_DynamicCast}
   n7 = GetDisplayLabel!(self:Fort Actor Label Override Interface Interface=@n6) -> (ReturnValue:Text)
   > n8 = Branch(Condition="Equal Exactly (Text)"!(A:"Text (by ref)"=@n7, B:"Text (by ref)"=GetEmptyText!() -> (ReturnValue:Text)) -> (ReturnValue:bool))
   Return Node(
    execute=@n8.else
    Show=true
    CreativeNameText=Select!(
     Option 0:Text=@n7
     Option 1:Text=Format Text!(Format="LOCTEXT(\"{Device} (Locked)\")", Device:Text=@n7) -> (Result)
     Index:bool=IsActorLocked!(ActorToCheck:Actor Object Reference=@n5) -> (ReturnValue:bool)
    ) -> (ReturnValue:Text)
   )
  ]}
 ShowBasicInteractionWidget {nodes [
   ShowBasicInteractionWidget() {is=title K2Node_FunctionEntry}
   n1 = Get CurrentInteraction()
   > n2 = Is Valid(InputObject=@n1)
   > BindMultiInteractUpdate() {is=title K2Node_CallFunction}
   > n3 = Branch(Condition=Get bShowFirstInteraction!(self:Fort Interact Context Info Object Reference=@n1) -> (bShowFirstInteraction:bool))
   > UpdateKeybinds()
   > UpdatePrimaryContext()
   > SetVisibility(self=Get BasicInteractionWidget(), InVisibility=SelfHitTestInvisible)
   > n4 = ShouldShowCreativeModeName!(execute, self) -> (then, Show:bool, CreativeNameText:Text) {is=title K2Node_CallFunction}
   n5 = Get CurrentInteraction()
   n6 = Get ContextText!(self:Fort Interact Context Info Object Reference=Get CurrentInteraction()) -> (ContextText:Text)
   > UpdateDescription!(
    execute
    self:Basic Interaction Widget Base Object Reference=Get BasicInteractionWidget()
    Description:Text=Select(
     Option 0=@n6
     Option 1=Format Text!(Format="LOCTEXT(\"{context}\\r\\n{name}\")", context:Text=@n6, name:Text=@n4.CreativeNameText) -> (Result)
     Index=@n4
    )
    DescriptionColor:Linear Color Structure=Get ContextTextColor!(self:Fort Interact Context Info Object Reference=@n5) -> (ContextTextColor:Linear Color Structure)
    SubDescription:Text
    SubDescriptionColor:Linear Color Structure
    InteractionCount:int=Get Interact Actor Count(CurrentInteraction=@n5, MultiInteractComponent=Get BoundMultiInteractComponent(), AppliesToInteraction=FirstInteraction) -> (InteractionCount=0) {is=title K2Node_CallFunction}
   ) -> (then)
   > SetVisibility(self=Get TextUnderKeybindText() -> (TextUnderKeybindText:Common Text Object Reference), InVisibility=Collapsed)
   > n7 = Branch(Condition=Get bDisplayTextOnly!(self:Fort Interact Context Info Object Reference=Get CurrentInteraction()) -> (bDisplayTextOnly:bool))
   n8 = Get UseKeybindWidgets()
   > SetForcedHoldKeybindStatus!(execute, self:Fort Keybind Widget Object Reference=@n8, InForcedHoldKeybindStatus:EFortKeybindForcedHoldStatus Enum=NoForcedHold) -> (then)
   > SetVisibility(self=@n8, InVisibility=Collapsed)
   > SetShowProgressCountDown!(execute, self:Fort Keybind Widget Object Reference=@n8, bShow:bool=false) -> (then)
   n9 = Get RequiredDuration(self=Get CurrentInteraction())
   SetForcedHoldKeybindStatus(
    execute=@n7.else
    self=@n8
    InForcedHoldKeybindStatus=Select(
     Option 0:EFortKeybindForcedHoldStatus Enum=NoForcedHold
     Option 1:EFortKeybindForcedHoldStatus Enum=ForcedHold
     Index="float > float"(A=@n9, B=0) {is=title K2Node_CallFunction}
    ) -> (ReturnValue:EFortKeybindForcedHoldStatus Enum)
   )
   n10 = Get RequiredDuration(self=Get CurrentInteraction())
   n11 = Get CurrentInteraction()
   > SetShowProgressCountDown(
    self=@n8
    bShow=AND Boolean(A="float > float"(A=@n10, B=0), B=Get bShowCountDown!(self:Fort Interact Context Info Object Reference=@n11) -> (bShowCountDown:bool))
   )
   SetVisibility(execute=[@n3.else, @n2.Is Not Valid], self=[Get Switcher_PrimaryContext(), Get InteractionKeybindBox()], InVisibility=Collapsed)
  ]}
 ShowContextOverride {nodes [
   ShowContextOverride() {is=title K2Node_FunctionEntry}
   > n1 = Branch(Condition=Get bShowFirstInteraction(self=Get CurrentInteraction()))
   > UpdatePrimaryContext()
   n2 = Get RequiredDuration(self=Get CurrentInteraction())
   > SetForcedHoldKeybindStatus(self=Get KeybindWidget(), InForcedHoldKeybindStatus=Select(Option 0=NoForcedHold, Option 1=ForcedHold, Index="float > float"(A=@n2, B=0)))
   n3 = Get CurrentInteraction()
   > n4 = ActivateContextOverride(
    ContextOverrideWidget=Get ContextOverrideWidget!(self:Fort Interact Context Info Object Reference=@n3) -> (ContextOverrideWidget:Fort Interact Context Info Widget Class Reference)
    ContextInfo=@n3
    bPrimary=true
   )
   > n5 = GetInteractText!(execute, self:Fort Interact Context Info Widget Object Reference=@n4) -> (then, InteractText:Text, Result:bool)
   n6 = Get TextUnderKeybindText()
   > "SetText (Text)"(self=@n6, InText=@n5)
   > SetVisibility(
    self=@n6
    InVisibility=Select(Option 0:ESlateVisibility Enum=Collapsed, Option 1:ESlateVisibility Enum=SelfHitTestInvisible, Index=@n5.Result) -> (ReturnValue:ESlateVisibility Enum)
   )
   > Branch(Condition=GetShowErrorsUsingDefaultWidget!(self:Fort Interact Context Info Widget Object Reference=@n4) -> (ReturnValue:bool))
   > RefreshInteractionFailureText() {is=title K2Node_CallFunction}
   SetVisibility(execute=@n1.else, self=[Get Switcher_PrimaryContext(), Get InteractionKeybindBox()], InVisibility=Collapsed)
  ]}
 ShowMatchmakingPortal {nodes [
   n1 = ShowMatchmakingPortal!() -> (then, Portal:"Actor Object Reference (by ref)") {is=title K2Node_FunctionEntry}
   > n2 = AthenaInteractDetails(self=@n1)
   > Branch(Condition=@n2.IsInPlayMode)
   > SetVisibility(self=Get VerticalBox_PortalMatchMaking(), InVisibility=SelfHitTestInvisible)
   > SetVisibility(self=[Get SecondaryBorder(), Get PrimaryBorder()], InVisibility=Collapsed)
   n3 = Get CurrentInteraction()
   > "SetText (Text)"(self=Get Text_PrimaryPortal() -> (Text_PrimaryPortal:Common Text Object Reference), InText=Get ContextText(self=@n3))
   > "SetText (Text)"(
    self=Get Text_SecondaryPortal() -> (Text_SecondaryPortal:Common Text Object Reference)
    InText=Get SecondContextText!(self:Fort Interact Context Info Object Reference=@n3) -> (SecondContextText:Text)
   )
   > n4 = Cast To FortPlayerPawn(Object=GetOwningPlayerPawn!(self:User Widget Object Reference) -> (ReturnValue:Pawn Object Reference))
   > n5 = MatchmakingPortalHasSecondInteraction!(
    execute:Exec
    self:Object Reference=Get Portal() -> (Portal:"Actor Object Reference (by ref)")
    InteractingPawn:Fort Player Pawn Object Reference=@n4
   ) -> (then:Exec, Return:bool) {is=title K2Node_Message}
   > SetVisibility(
    self=Get HorizontalBox_SecondaryPortalMatchmaking_1() -> (HorizontalBox_SecondaryPortalMatchmaking_1:Horizontal Box Object Reference)
    InVisibility=Select(Option 0=Collapsed, Option 1=HitTestInvisible, Index=@n5)
   )
   SetVisibility(execute=@n4.CastFailed, InVisibility=Collapsed)
  ]}
 ShowPickupWidget {nodes [
   n1 = ShowPickupWidget!() -> (then, Pickup:Fort Pickup Object Reference) {is=title K2Node_FunctionEntry}
   > UpdateKeybinds()
   > UpdatePrimaryContext()
   n2 = Get HUD-PickupItemWidget()
   > SetVisibility(self=@n2, InVisibility=SelfHitTestInvisible)
   > SetPickup(self=@n2, NewPickup=@n1)
   > n3 = GetInteractErrorText!(execute, self:Fort Pickup Object Reference=@n1) -> (then, OutInteractErrorText:Text, OutInteractErrorIconTag:Gameplay Tag Structure, ReturnValue:bool)
   n4 = Get KeybindWidget()
   > SetVisibility(self=@n4, InVisibility=Select(Option 0=SelfHitTestInvisible, Option 1=Collapsed, Index=@n3.ReturnValue))
   n5 = Get RequiredDuration(self=Get CurrentInteraction())
   > SetForcedHoldKeybindStatus(
    self=@n4
    InForcedHoldKeybindStatus=Select(
     Option 0:EFortKeybindForcedHoldStatus Enum=NoForcedHold
     Option 1:EFortKeybindForcedHoldStatus Enum=ForcedHold
     Index="float > float"(A=@n5, B=0)
    ) -> (ReturnValue:EFortKeybindForcedHoldStatus Enum)
   )
   n6 = Get RequiredDuration(self=Get CurrentInteraction())
   n7 = Get CurrentInteraction()
   > SetShowProgressCountDown(self=@n4, bShow=AND Boolean(A="float > float"(A=@n6, B=0), B=Get bShowCountDown(self=@n7)))
   > n8 = GetInteractText!(execute, self:Fort Pickup Object Reference=@n1) -> (then, InteractText, ReturnValue:bool)
   n9 = Get TextUnderKeybindText()
   > "SetText (Text)"(self=@n9, InText=@n8)
   > SetVisibility(
    self=@n9
    InVisibility=Select(Option 0:ESlateVisibility Enum=SelfHitTestInvisible, Option 1:ESlateVisibility Enum=Collapsed, Index=@n8.ReturnValue) -> (ReturnValue:ESlateVisibility Enum)
   )
   > UpdateWidgetVisibility()
  ]}
 ShowSecondInteractionWidget {nodes [
   ShowSecondInteractionWidget() {is=title K2Node_FunctionEntry}
   > n1 = Branch(Condition=Get bShowSecondInteraction!(self:Fort Interact Context Info Object Reference=Get CurrentInteraction()) -> (bShowSecondInteraction:bool))
   > UpdateKeybinds()
   > SetVisibility(
    self=[
     Get SecondaryBoxDividerLine() -> (SecondaryBoxDividerLine:Image Object Reference)
     Get Spacer_SecondaryAction() -> (Spacer_SecondaryAction:Spacer Object Reference)
     Get SecondInteractionBox()
    ]
    InVisibility=HitTestInvisible
   )
   n2 = Get CurrentInteraction()
   > UpdateDescription(
    self=Get BasicInteractionWidgetSecondary()
    Description=Get SecondContextText(self=@n2)
    DescriptionColor=Get ContextTextColor(self=@n2)
    InteractionCount=Get Interact Actor Count(CurrentInteraction=@n2, MultiInteractComponent=Get BoundMultiInteractComponent(), AppliesToInteraction=SecondInteraction) -> (InteractionCount=0)
   )
   n3 = Get CurrentInteraction()
   > SetVisibility(
    self=Get SecondInteractionKeybindColor()
    InVisibility=Select(
     Option 0=Collapsed
     Option 1=HitTestInvisible
     Index=Get bShowSecondInteractionKeybind!(self:Fort Interact Context Info Object Reference=@n3) -> (bShowSecondInteractionKeybind:bool)
    )
   )
   n4 = Get CurrentInteraction()
   > SetContentColorAndOpacity(
    self=Get SecondInteractionKeybindColor()
    InContentColorAndOpacity=SelectColor(
     A=Get SecondContextTextColor!(self:Fort Interact Context Info Object Reference=@n4) -> (SecondContextTextColor:Linear Color Structure)
     B=LinearColor_Gray()
     bPickA=Get bIsSecondInteractionActive!(self:Fort Interact Context Info Object Reference=@n4) -> (bIsSecondInteractionActive:bool)
    )
   )
   n5 = Get SecondRequiredDuration(self=Get CurrentInteraction())
   > SetForcedHoldKeybindStatus(
    self=Get AltInteractionKeybindWidgets()
    InForcedHoldKeybindStatus=Select(
     Option 0:EFortKeybindForcedHoldStatus Enum=NoForcedHold
     Option 1:EFortKeybindForcedHoldStatus Enum=ForcedHold
     Index="float > float"(A=@n5, B=0)
    ) -> (ReturnValue:EFortKeybindForcedHoldStatus Enum)
   )
   n6 = Get SecondRequiredDuration(self=Get CurrentInteraction())
   n7 = Get CurrentInteraction()
   > SetShowProgressCountDown(self=Get AltInteractionKeybindWidgets(), bShow=AND Boolean(A="float > float"(A=@n6, B=0), B=Get bShowCountDown(self=@n7)))
   n8 = Get DynamicInteractionOptions!(self:Fort Interact Context Info Object Reference=Get CurrentInteraction()) -> (DynamicInteractionOptions:Gameplay Tag Container Structure)
   > SetActiveWidgetIndex(
    self=Get Switcher_SecondaryDivider() -> (Switcher_SecondaryDivider:Common Visibility Switcher Object Reference)
    Index=Select(
     Option 0:int=0
     Option 1:int=1
     Index=HasTag!(
      TagContainer:"Gameplay Tag Container Structure (by ref)"=@n8
      Tag:Gameplay Tag Structure=(TagName="HUD.Athena.InteractionOptions.SeparateSecondaryAction")
      bExactMatch:bool=true
     ) -> (ReturnValue:bool)
    ) -> (ReturnValue:int)
   )
   > n9 = Sequence!(execute) -> (then_0, then_1, then_2:Exec)
   n10 = Get CurrentInteraction()
   > n11 = Branch(Condition=TextIsEmpty(InText=Get SecondContextErrorText!(self:Fort Interact Context Info Object Reference=@n10) -> (SecondContextErrorText:Text)))
   n12 = Get WBP_SecondaryInteractionWarning() -> (WBP_SecondaryInteractionWarning:WBP Interaction Warning Object Reference)
   > SetVisibility(self=@n12, InVisibility=Collapsed)
   SetVisibility(execute=@n11.else, self=@n12, InVisibility=SelfHitTestInvisible)
   > SetFailureMessage(self=Get WBP_SecondaryInteractionWarning(), Text=Get SecondContextErrorText(self=Get CurrentInteraction()))
   n13 = HasMatchingGameplayTag(execute=@n9.then_1, self=GetIndicatedActor(), TagToCheck=(TagName="Building.Type.ItemContainer.Wumba"))
   > n14 = Branch(Condition=@n13)
   n15 = Get Border_SubDescription_Secondary() -> (Border_SubDescription_Secondary:Border Object Reference)
   SetVisibility(execute=[@n14.then, @n17.then], self=@n15, InVisibility=Collapsed)
   n16 = Get CurrentInteraction()
   n17 = Branch(
    execute=@n14.else
    Condition=TextIsEmpty(InText=Get SecondContextSubText!(self:Fort Interact Context Info Object Reference=@n16) -> (SecondContextSubText:Text))
   )
   SetVisibility(execute=@n17.else, self=@n15, InVisibility=SelfHitTestInvisible)
   > SetText(
    self=Get TextSubDescription_Secondary() -> (TextSubDescription_Secondary:Common Rich Text Block Object Reference)
    InText=Get SecondContextSubText(self=Get CurrentInteraction())
   )
   n18 = Get CurrentInteraction()
   > SetVisibility(
    self=Get SubDescriptionAlertImage_Secondary() -> (SubDescriptionAlertImage_Secondary:Image Object Reference)
    InVisibility=Select(
     Option 0:ESlateVisibility Enum=Collapsed
     Option 1:ESlateVisibility Enum=HitTestInvisible
     Index=Get bDisplaySecondSubInteractIcon!(self:Fort Interact Context Info Object Reference=@n18) -> (bDisplaySecondSubInteractIcon:bool)
    ) -> (ReturnValue:ESlateVisibility Enum)
   )
   n19 = Get SecondContextSubTextImage!(self:Fort Interact Context Info Object Reference=Get CurrentInteraction()) -> (SecondContextSubTextImage:Texture 2D Soft Object Reference)
   > SetBrushFromSoftTexture!(
    execute
    self:Image Object Reference=Get SubDescriptionAlertImage_Secondary()
    SoftTexture:Texture 2D Soft Object Reference=Select(
     Option 0:Texture 2D Soft Object Reference=@n19
     Option 1:Texture 2D Soft Object Reference=Get DefaultAlertImage() -> (DefaultAlertImage:Texture 2D Soft Object Reference)
     Index=IsAssetNull!(Asset:Soft Object Reference=@n19) -> (ReturnValue:bool)
    ) -> (ReturnValue:Texture 2D Soft Object Reference)
    bMatchSize:bool=false
   ) -> (then)
   n20 = Get DynamicInteractionOptions(self=Get CurrentInteraction())
   > SetBrushTintColor!(
    execute
    self:Image Object Reference=Get SubDescriptionAlertImage_Secondary()
    TintColor:Slate Color Structure=Select(
     Option 0:Slate Color Structure=Get DefaultAlertImageTint() -> (DefaultAlertImageTint:Slate Color Structure)
     Option 1:Slate Color Structure=Get DisabledAlertImageTint() -> (DisabledAlertImageTint:Slate Color Structure)
     Index=HasTag(TagContainer=@n20, Tag=(TagName="HUD.Athena.InteractionOptions.DisableInteractKey"), bExactMatch=true)
    ) -> (ReturnValue:Slate Color Structure)
   ) -> (then)
   n21 = Get DynamicInteractionOptions(self=Get CurrentInteraction())
   n22 = Get CurrentInteraction()
   n23 = NOT Boolean(A=TextIsEmpty(InText=Get SecondContextErrorText(self=@n22)))
   n24 = NOT Boolean(A=TextIsEmpty(InText=Get SecondContextSubText(self=@n22)))
   n25 = Branch(
    execute=@n9.then_2
    Condition=AND Boolean(
     A=HasTag(TagContainer=@n21, Tag=(TagName="HUD.Athena.InteractionOptions.OnErrorDimContext"), bExactMatch=false)
     B=OR Boolean!(A=@n23, B=@n24) -> (ReturnValue)
    )
   )
   n26 = Get Border_SecondaryDimmedColor() -> (Border_SecondaryDimmedColor:Border Object Reference)
   > Branch(Condition="Not Equal (LinearColor)"(
     A:Linear Color Structure=Get ContentColorAndOpacity!(self:Border Object Reference=@n26) -> (ContentColorAndOpacity:Linear Color Structure)
     B:Linear Color Structure=Get ColorDimmed() -> (ColorDimmed:Linear Color Structure)
    ) -> (ReturnValue:bool=~) {is=title K2Node_PromotableOperator})
   > SetContentColorAndOpacity(self=Get Border_SecondaryDimmedColor(), InContentColorAndOpacity=Get ColorDimmed())
   Branch(
    execute=@n25.else
    Condition="Not Equal (LinearColor)"(A=Get ContentColorAndOpacity(self=@n26), B=Get ColorDefault() -> (ColorDefault:Linear Color Structure)) -> (ReturnValue=~)
   )
   > SetContentColorAndOpacity(self=Get Border_SecondaryDimmedColor(), InContentColorAndOpacity=Get ColorDefault())
   SetVisibility(
    execute=@n1.else
    self=[
     Get SecondaryBoxDividerLine()
     Get Spacer_SecondaryAction()
     Get WBP_SecondaryInteractionWarning()
     Get Border_SubDescription_Secondary()
     Get SecondInteractionBox()
    ]
    InVisibility=Collapsed
   )
  ]}
 Unbind Multi Interact Update {nodes [
   Unbind Multi Interact Update() {is=title K2Node_FunctionEntry}
   > UpdateProgressMaterial(InteractionCount=0)
   > n1 = Get!(execute) -> (then, else, BoundMultiInteractComponent:Fort Actor Component Multi Interact Object Reference)
   > Unbind Event from Actor Count Changed Delegate(self:Fort Actor Component Multi Interact Object Reference=@n1, Delegate=Create Event()) {is=title K2Node_RemoveDelegate}
   > Set BoundMultiInteractComponent()
  ]}
 Update Keybind Custom Progress {nodes [
   n1 = Update Keybind Custom Progress!() -> (then, SecondsToInterpolate:double=~) {is=title K2Node_FunctionEntry}
   > Branch(Condition=IsValid(Object=Get CurrentInteraction()))
   > n2 = Branch(Condition=IsValid(Object=Get LastInteraction() -> (LastInteraction:Fort Interact Context Info Object Reference)))
   n3 = Get CurrentInteraction()
   n4 = Get LastInteraction()
   > n5 = Set SameTarget(SameTarget:bool="Equal (Object)"(A=Get ReceivingActor(self=@n3), B=Get ReceivingActor(self=@n4))) -> (Output_Get:bool)
   Set LastInteraction(execute=[@n10.then, @n5.then], LastInteraction:Fort Interact Context Info Object Reference=Get CurrentInteraction()) -> (Output_Get:Fort Interact Context Info Object Reference)
   n6 = Get bUseMultiInteractComponentProgress!(self:Fort Interact Context Info Object Reference=Get CurrentInteraction()) -> (bUseMultiInteractComponentProgress:bool)
   n7 = Get SameTarget() -> (SameTarget:bool)
   n8 = InvertBool!(bValue:bool=@n7) -> (ReturnValue:bool)
   n9 = SelectFloat(A=@n1, B=0, bPickA=@n7)
   > SetKeybindCustomProgress!(execute, self, UseCustomProgress:bool=@n6, SecondsToInterpolate:double=@n9, ChangedTarget:bool=@n8, IsSecondary:bool=false) -> (then) {is=title K2Node_CallFunction}
   > SetKeybindCustomProgress(UseCustomProgress=@n6, SecondsToInterpolate=@n9, ChangedTarget=@n8, IsSecondary=true)
   n10 = Set SameTarget(execute=@n2.else, SameTarget=false)
  ]}
 UpdateForCurrentInput {nodes [
   UpdateForCurrentInput() {is=title K2Node_FunctionEntry}
   > Set bIsUsingTouch(bIsUsingTouch:bool=IsUsingTouch(PlayerController=GetOwningPlayer())) -> (Output_Get:bool)
   > n1 = Sequence()
   > SetVisibility(
    self=Get InteractionKeybindBox()
    InVisibility=Select(
     Option 0:ESlateVisibility Enum=SelfHitTestInvisible
     Option 1:ESlateVisibility Enum=Collapsed
     Index=Get bIsUsingTouch() -> (bIsUsingTouch:bool)
    ) -> (ReturnValue:ESlateVisibility Enum)
   )
   Set SecondSIBV(execute=@n1.then_1, SecondSIBV:ESlateVisibility Enum=Collapsed) -> (Output_Get:ESlateVisibility Enum=Visible)
   > n2 = Get!(execute) -> (then, else, CurrentInteraction:Fort Interact Context Info Object Reference)
   n3 = NOT Boolean(A=Get bIsUsingTouch())
   n4 = Get bShowSecondInteraction(self=@n2)
   > Set SecondSIBV(SecondSIBV=Select(Option 0=Collapsed, Option 1=SelfHitTestInvisible, Index=AND Boolean(A=@n4, B=@n3))) -> (Output_Get=Visible)
   SetVisibility(execute=@n1.then_2, self=Get SecondInteractionBox(), InVisibility=Get SecondSIBV() -> (SecondSIBV:ESlateVisibility Enum=Visible))
  ]}
 UpdateKeybinds {nodes [
   UpdateKeybinds() {is=title K2Node_FunctionEntry}
   n1 = Get KeybindWidget()
   > SetBoundAction!(
    execute
    self:Fort Keybind Widget Object Reference=Get UseKeybindWidgets()
    NewBoundAction:Name=GetUseActionName!(
     self:Fort Global UIContext Object Reference=Get FortGlobalUIContext()
     OverrideInputType:ECommonInputType Enum=Get InputTypeOverride!(self:Fort Keybind Widget Object Reference=@n1) -> (InputTypeOverride:ECommonInputType Enum)
    ) -> (ReturnValue:Name=None)
   ) -> (then)
   n2 = Get SecondInteractionKeybind()
   > SetBoundAction(
    self=Get AltInteractionKeybindWidgets()
    NewBoundAction=GetAltInteractionActionName!(
     self:Fort Global UIContext Object Reference=Get FortGlobalUIContext()
     OverrideInputType:ECommonInputType Enum=Get InputTypeOverride(self=@n2)
    ) -> (ReturnValue:Name=None)
   )
  ]}
 UpdatePrimaryContext {nodes [
   UpdatePrimaryContext() {is=title K2Node_FunctionEntry}
   > UpdateSubDescription()
   n1 = Get DynamicInteractionOptions(self=Get CurrentInteraction())
   n2 = HasTag(TagContainer=@n1, Tag=(TagName="HUD.Athena.InteractionOptions.DisableInteractKey"), bExactMatch=true)
   > SetVisibility(
    self=Get InteractionKeybindBox()
    InVisibility=Select(
     Option 0=Select(Option 0=HitTestInvisible, Option 1=Hidden, Index=@n2)
     Option 1=Collapsed
     Index=HasTag(TagContainer=@n1, Tag=(TagName="HUD.Athena.InteractionOptions.CollapseInteractKey"), bExactMatch=true)
    )
   )
   > n3 = Sequence!(execute) -> (then_0, then_1)
   n4 = Get DynamicInteractionOptions(self=Get CurrentInteraction())
   n5 = HasTag(
    TagContainer=Get DynamicInteractionOptions(self=Get CurrentInteraction())
    Tag=(TagName="HUD.Athena.InteractionOptions.OnErrorDimContext")
    bExactMatch=false
   )
   n6 = Get CurrentInteraction()
   n7 = NOT Boolean(A=TextIsEmpty(InText=Get ErrorText(self=@n6)))
   > n8 = Branch(Condition=OR Boolean(A=AND Boolean(A=@n5, B=@n7), B=HasTag(TagContainer=@n4, Tag=(TagName="HUD.Athena.InteractionOptions.DimPrimaryContext"), bExactMatch=false)))
   n9 = Get Border_PrimaryDimmedColor() -> (Border_PrimaryDimmedColor:Border Object Reference)
   > Branch(Condition="Not Equal (LinearColor)"(A=Get ContentColorAndOpacity(self=@n9), B=Get ColorDimmed()) -> (ReturnValue=~))
   > SetContentColorAndOpacity(self=Get Border_PrimaryDimmedColor(), InContentColorAndOpacity=Get ColorDimmed())
   n10 = Get Border_PrimaryDimmedColor()
   Branch(execute=@n8.else, Condition="Not Equal (LinearColor)"(A=Get ContentColorAndOpacity(self=@n10), B=Get ColorDefault()) -> (ReturnValue=~))
   > SetContentColorAndOpacity(self=Get Border_PrimaryDimmedColor(), InContentColorAndOpacity=Get ColorDefault())
   n11 = Get CurrentInteraction()
   n12 = Branch(
    execute=@n3.then_1
    Condition=HasTag(TagContainer=Get DynamicInteractionOptions(self=@n11), Tag=(TagName="HUD.Athena.InteractionOptions.HideBackground"), bExactMatch=false)
   )
   n13 = Get PrimaryBorder()
   > Branch(Condition="Not Equal (Enum)"(
     A:ESlateBrushDrawType Enum=Get!(self:Border Object Reference=@n13) -> (
      Background_TintColor:Slate Color Structure
      Background_DrawAs:ESlateBrushDrawType Enum
      Background_Tiling:ESlateBrushTileType Enum
      Background_Mirroring:ESlateBrushMirrorType Enum
      Background_ImageSize:Vector 2D Structure
      Background_Margin:Margin Structure
      Background_ResourceObject:Object Reference
      Background_OutlineSettings:Slate Brush Outline Settings Structure
     ).Background_DrawAs
     B:ESlateBrushDrawType Enum=NoDrawType
    ))
   n14 = Get(self=Get PrimaryBorder())
   > SetBrush(
    self:Border Object Reference=Get PrimaryBorder()
    InBrush=Make Slate Brush!(
     TintColor:Slate Color Structure=@n14
     DrawAs:ESlateBrushDrawType Enum=NoDrawType
     Tiling:ESlateBrushTileType Enum=@n14.Background_Tiling
     Mirroring:ESlateBrushMirrorType Enum=@n14.Background_Mirroring
     ImageSize:Vector 2D Structure=@n14.Background_ImageSize
     Margin:Margin Structure=@n14.Background_Margin
     ResourceObject:Object Reference=@n14.Background_ResourceObject
     OutlineSettings:Slate Brush Outline Settings Structure=@n14.Background_OutlineSettings
    ) -> (SlateBrush:Slate Brush Structure) {is=title K2Node_MakeStruct}
   )
   n15 = Get PrimaryBorder()
   Branch(execute=@n12.else, Condition="Not Equal (Enum)"(A=Get(self=@n15).Background_DrawAs, B=Box))
   n16 = Get(self=Get PrimaryBorder())
   > SetBrush(
    self=Get PrimaryBorder()
    InBrush=Make Slate Brush(
     TintColor=@n16
     DrawAs=Box
     Tiling=@n16.Background_Tiling
     Mirroring=@n16.Background_Mirroring
     ImageSize=@n16.Background_ImageSize
     Margin=@n16.Background_Margin
     ResourceObject=@n16.Background_ResourceObject
     OutlineSettings=@n16.Background_OutlineSettings
    )
   )
  ]}
 UpdateProgressMaterial {nodes [
   n1 = UpdateProgressMaterial!() -> (then, InteractionCount:int) {is=title K2Node_FunctionEntry}
   > SetScalarParameterValue(
    self=Get ProgressPercentageMID!(self:Fort Keybind Widget Object Reference=Get KeybindWidget()) -> (ProgressPercentageMID:Material Instance Dynamic Object Reference)
    ParameterName=HasMultipleInteracts
    Value=Select(Option 0:double=0, Option 1:double=1, Index="integer > integer"(A=@n1, B=1)) -> (ReturnValue:double)
   )
  ]}
 UpdateSubDescription {nodes [
   UpdateSubDescription() {is=title K2Node_FunctionEntry}
   n1 = Get CurrentInteraction()
   > n2 = Branch(Condition=TextIsEmpty(InText=Get ContextSubText!(self:Fort Interact Context Info Object Reference=@n1) -> (ContextSubText:Text)))
   n3 = Get Border_SubDescription()
   > SetVisibility(self=@n3, InVisibility=Collapsed)
   SetVisibility(execute=@n2.else, self=@n3, InVisibility=SelfHitTestInvisible)
   n4 = Get CurrentInteraction()
   > SetContentColorAndOpacity(
    self=Get Border_SubDescription()
    InContentColorAndOpacity=Get ContextSubTextColor!(self:Fort Interact Context Info Object Reference=@n4) -> (ContextSubTextColor:Linear Color Structure)
   )
   > Set IsDisableInteractKeyActive(IsDisableInteractKeyActive:bool=HasTag(TagContainer=Get DynamicInteractionOptions(self=@n4), Tag=(TagName="HUD.Athena.InteractionOptions.DisableInteractKey"), bExactMatch=true)) -> (Output_Get:bool)
   > SetBrushColor(
    self=Get Border_SubDescription()
    InBrushColor=SelectColor(
     A=Get DisabledSubBackgroundColor() -> (DisabledSubBackgroundColor:Linear Color Structure)
     B=Get DefaultSubBackgroundColor() -> (DefaultSubBackgroundColor:Linear Color Structure)
     bPickA=Get IsDisableInteractKeyActive() -> (IsDisableInteractKeyActive:bool)
    )
   )
   > SetText(
    self=Get TextSubDescription() -> (TextSubDescription:Common Rich Text Block Object Reference)
    InText=Get ContextSubText(self=Get CurrentInteraction())
   )
   n5 = Get CurrentInteraction()
   > SetVisibility(
    self=Get SubDescriptionAlertImage() -> (SubDescriptionAlertImage:Image Object Reference)
    InVisibility=Select(
     Option 0:ESlateVisibility Enum=Collapsed
     Option 1:ESlateVisibility Enum=HitTestInvisible
     Index=Get bDisplaySubInteractIcon!(self:Fort Interact Context Info Object Reference=@n5) -> (bDisplaySubInteractIcon:bool)
    ) -> (ReturnValue:ESlateVisibility Enum)
   )
   n6 = Get ContextSubTextImage!(self:Fort Interact Context Info Object Reference=Get CurrentInteraction()) -> (ContextSubTextImage:Texture 2D Soft Object Reference)
   > SetBrushFromSoftTexture(
    self=Get SubDescriptionAlertImage()
    SoftTexture=Select(
     Option 0:Texture 2D Soft Object Reference=@n6
     Option 1:Texture 2D Soft Object Reference=Get DefaultAlertImage()
     Index=IsAssetNull(Asset=@n6)
    ) -> (ReturnValue:Texture 2D Soft Object Reference)
    bMatchSize=false
   )
   > SetBrushTintColor(
    self=Get SubDescriptionAlertImage()
    TintColor=Select(
     Option 0:Slate Color Structure=Get DefaultAlertImageTint()
     Option 1:Slate Color Structure=Get DisabledAlertImageTint()
     Index=Get IsDisableInteractKeyActive()
    ) -> (ReturnValue:Slate Color Structure)
   )
  ]}
 UpdateWidgetVisibility {nodes [
   UpdateWidgetVisibility() {is=title K2Node_FunctionEntry}
   > n1 = Branch(Condition=IsUsingTouch(PlayerController=GetOwningPlayer()))
   > n2 = UpdateWidgetVisibilityTouch!(execute, self) -> (then, bFailed:bool)
   > n3 = Branch(Condition=@n2)
   SetVisibility(
    execute=[@n1.else, @n3.then]
    InVisibility=Select(
     Option 0:ESlateVisibility Enum=SelfHitTestInvisible
     Option 1:ESlateVisibility Enum=Collapsed
     Index=IsInteractionWidgetSuppressed!() -> (IsSuppressed:bool) {is=title K2Node_CallFunction}
    ) -> (ReturnValue:ESlateVisibility Enum)
   )
  ]}
 UpdateWidgetVisibilityTouch {nodes [
   UpdateWidgetVisibilityTouch() {is=title K2Node_FunctionEntry}
   > n1 = Cast To FortPlayerPawnAthena(Object=GetPlayerPawn(PlayerIndex=0))
   > n2 = Cast To FortPlayerControllerAthena(Object=GetController(self=@n1))
   n3 = NOT Boolean(A=IsInteractionWidgetSuppressed())
   n4 = CanPerformNativeAction!(
    self:Fort Player Controller Object Reference=@n2
    ActionTag:"Gameplay Tag Structure (by ref)"=MakeLiteralGameplayTag!(Value:Gameplay Tag Structure=(TagName="Gameplay.Action.Player.InteractAny")) -> (ReturnValue:Gameplay Tag Structure)
   ) -> (ReturnValue:bool)
   n5 = IsValid(Object=GetAbilitySystemComponent!(Actor:Actor Object Reference=@n1) -> (ReturnValue:Ability System Component Object Reference))
   > SetVisibility(InVisibility=Select(Option 0=Collapsed, Option 1=SelfHitTestInvisible, Index=AND Boolean!(A=@n5, B=@n4, C:bool=@n3) -> (ReturnValue)))
   Return Node!(execute=[@n2.CastFailed, @n1.CastFailed], bFailed:bool=true)
  ]}
]
variables [
 AltInteractionKeybindWidgets {type="array<KeybindWidget_C>", default []}
 BoundMultiInteractComponent {type=FortActorComponent_MultiInteract, default=null}
 ColorDefault {type=LinearColor, default=<rgba 1 1 1 1>}
 ColorDimmed {type=LinearColor, default=<rgba 0.3 0.3 0.3 1>}
 CurrentInteraction {type=FortInteractContextInfo, default=null}
 DefaultAlertImage {type=Texture2D, default=&T-Icon-Alert2-32}
 ExtraDataEvent {type=AsyncAction_StartListeningToStatefulEvent, default=null}
 ExtraDataSlotWidget {type=Class, default=null}
 InvalidInteractionTags {
  type=GameplayTagContainer
  default {gameplayTags [{tagName=HUD.Athena.InteractionOptions.DisableInteractKey}, {tagName=HUD.Athena.InteractionOptions.CollapseInteractKey}]}
 }
 LastInteraction {type=FortInteractContextInfo, default=null}
 Max Tap Distance {type=number, default=20}
 Max Tap Time {type=number, default=0.25}
 Payload Interaction Type {type=EInteractionBeingAttempted, default=FirstInteraction}
 Sidegrade_WumbaWidget {type=UserWidget, default=null}
 Targeted Actor {type=Actor, default=null}
 Touch Start Position {type=Vector2D, default=<0 0>}
 Touch Start Time {type=number}
 Upgrade_WumbaWidget {type=UserWidget, default=null}
 UseKeybindWidgets {type="array<KeybindWidget_C>", default []}
 WumbaWidgetClass {type=Class, default=null}
 WumbaWidgetSoftAsset {type=Class, default=/Game/UI/Creative/WBP_Wumba_InteractionIndicator.WBP_Wumba_InteractionIndicator_C}
]
widgets {
 info {
  inheritedWidgetCount=4
  parentClass=/Script/FortniteUI.FortInteractionIndicatorWidget
  rootWidgetClass=&VerticalBox
  widgetCount=63
 }
 widgets [
  VerticalBox_0 {widgetClassPath=&VerticalBox, widget=VerticalBox_0}
  PrimaryBorder {
   widgetClassPath=&Border
   bIsVariable=true
   parent=VerticalBox_0
   widget=PrimaryBorder
   properties {
    background {
     bIsDynamicallyLoaded=false
     drawAs=Box
     imageType=NoImage
     mirroring=NoMirror
     resourceName=None
     resourceObject=/Game/Athena/HUD/Inventory/Art/M_UI_HUD_Rounded_Interaction_Backing
     tiling=NoTile
     imageSize=<8 8>
     margin=<ltrb 0.5 0.5 0.5 0.5>
     outlineSettings {
      bUseBrushTransparency=false
      roundingType=HalfHeightRadius
      width=0
      color {colorUseRule=UseColor_Specified, specifiedColor=<rgba 0 0 0 0>}
      cornerRadii=<0 0 0 0>
     }
     tintColor {colorUseRule=UseColor_Specified, specifiedColor=<rgba 1 1 1 1>}
     uVRegion {bIsValid=false, max=<0 0>, min=<0 0>}
    }
    padding=<ltrb 0 0 0 0>
   }
   slot {class=VerticalBoxSlot}
  }
  VerticalBox {
   widgetClassPath=&VerticalBox
   parent=PrimaryBorder
   widget=VerticalBox
   slot {class=BorderSlot, properties {padding=<ltrb 0 0 0 0>}}
  }
  Border_PrimaryDimmedColor {
   widgetClassPath=&Border
   bIsVariable=true
   parent=VerticalBox
   widget=Border_PrimaryDimmedColor
   properties {
    visibility=SelfHitTestInvisible
    background {
     bIsDynamicallyLoaded=false
     drawAs=NoDrawType
     imageType=NoImage
     mirroring=NoMirror
     resourceName=None
     resourceObject=null
     tiling=NoTile
     imageSize=<32 32>
     margin=<ltrb 0 0 0 0>
     outlineSettings {
      bUseBrushTransparency=false
      roundingType=HalfHeightRadius
      width=0
      color {colorUseRule=UseColor_Specified, specifiedColor=<rgba 0 0 0 0>}
      cornerRadii=<0 0 0 0>
     }
     tintColor {colorUseRule=UseColor_Specified, specifiedColor=<rgba 1 1 1 1>}
     uVRegion {bIsValid=false, max=<0 0>, min=<0 0>}
    }
    padding=<ltrb 8 0 16 0>
   }
   slot {class=VerticalBoxSlot}
  }
  Interaction {
   widgetClassPath=&HorizontalBox
   bIsVariable=true
   parent=Border_PrimaryDimmedColor
   widget=Interaction
   slot {class=BorderSlot, properties {padding=<ltrb 8 0 16 0>}}
  }
  InteractionKeybindColor {
   widgetClassPath=&CommonBorder
   bIsVariable=true
   parent=Interaction
   widget=InteractionKeybindColor
   properties {padding=<ltrb 8 8 12 8>}
   slot {class=HorizontalBoxSlot, properties {horizontalAlignment=HAlign_Left, verticalAlignment=VAlign_Center}}
  }
  InteractionKeybindBox {
   widgetClassPath=&VerticalBox
   bIsVariable=true
   parent=InteractionKeybindColor
   widget=InteractionKeybindBox
   slot {class=BorderSlot, properties {padding=<ltrb 8 8 12 8>}}
  }
  KeybindWidget {
   widgetClassPath=/FortUILibrary/Input/KeybindWidget
   bIsVariable=true
   parent=InteractionKeybindBox
   widget=KeybindWidget
   slot {class=VerticalBoxSlot}
  }
  TextUnderKeybindText {
   widgetClassPath=&CommonTextBlock
   bIsVariable=true
   parent=InteractionKeybindBox
   widget=TextUnderKeybindText
   properties {
    justification=Center
    mobileFontSizeMultiplier=1.3
    text="{Action}"
    textTransformPolicy=ToUpper
    font {
     bForceMonospaced=false
     bMaterialIsStencil=false
     fontMaterial=null
     fontObject=&BurbankSmall
     letterSpacing=0
     monospacedWidth=1
     size=13
     skewAmount=0
     typefaceFontName=Black
     outlineSettings {
      bApplyOutlineToDropShadows=false
      bMiteredCorners=false
      bSeparateFillAlpha=true
      outlineMaterial=&M-FontShadow
      outlineSize=2
      outlineColor=<rgba 0 0 0 0.5>
     }
    }
    shadowOffset=<0 0>
   }
   slot {class=VerticalBoxSlot, properties {horizontalAlignment=HAlign_Left, verticalAlignment=VAlign_Top, padding=<ltrb 0 8 0 0>}}
  }
  HUD-PickupItemWidget {
   widgetClassPath=&/AthenaHUD-PickupItemWidget
   bIsVariable=true
   parent=Interaction
   widget=HUD-PickupItemWidget
   properties {visibility=Collapsed}
   slot {class=HorizontalBoxSlot, properties {horizontalAlignment=HAlign_Left, verticalAlignment=VAlign_Center, padding=<ltrb 0 0 8 0>}}
  }
  Switcher_PrimaryContext {
   widgetClassPath=&CommonVisibilitySwitcher
   bInherited=true
   bIsVariable=true
   parent=Interaction
   widget=Switcher_PrimaryContext
   slot {class=HorizontalBoxSlot, properties {padding=<ltrb 0 4 0 0>, size {sizeRule=Fill, value=1}}}
  }
  BasicInteractionWidget {
   widgetClassPath=/Game/UI/InGame/HUD/BasicInteractionWidget
   bIsVariable=true
   parent=Switcher_PrimaryContext
   widget=BasicInteractionWidget
   properties {visibility=Collapsed}
   slot {class=CommonVisibilitySwitcherSlot}
  }
  WBP_InteractionWarning {
   widgetClassPath=&/WBP_InteractionWarning
   bIsVariable=true
   parent=VerticalBox
   widget=WBP_InteractionWarning
   slot {class=VerticalBoxSlot}
  }
  Border_SubDescription {
   widgetClassPath=&Border
   bIsVariable=true
   parent=VerticalBox
   widget=Border_SubDescription
   properties {
    horizontalAlignment=HAlign_Center
    verticalAlignment=VAlign_Center
    visibility=Collapsed
    brushColor=<rgba 0.0273 0.0273 0.0273 0.8>
    padding=<ltrb 8 4 8 4>
   }
   slot {class=VerticalBoxSlot}
  }
  SubTextBox {
   widgetClassPath=&HorizontalBox
   bIsVariable=true
   parent=Border_SubDescription
   widget=SubTextBox
   properties {visibility=HitTestInvisible}
   slot {class=BorderSlot, properties {horizontalAlignment=HAlign_Center, verticalAlignment=VAlign_Center, padding=<ltrb 8 4 8 4>}}
  }
  SubDescriptionAlertImage {
   widgetClassPath=&Image
   bIsVariable=true
   parent=SubTextBox
   widget=SubDescriptionAlertImage
   properties {
    visibility=HitTestInvisible
    brush {
     bIsDynamicallyLoaded=false
     drawAs=Image
     imageType=NoImage
     mirroring=NoMirror
     resourceName=None
     resourceObject=&T-Icon-Alert2-32
     tiling=NoTile
     imageSize=<32 32>
     margin=<ltrb 0 0 0 0>
     outlineSettings {
      bUseBrushTransparency=false
      roundingType=HalfHeightRadius
      width=0
      color {colorUseRule=UseColor_Specified, specifiedColor=<rgba 0 0 0 0>}
      cornerRadii=<0 0 0 0>
     }
     tintColor {colorUseRule=UseColor_Specified, specifiedColor=<rgba 1 1 1 1>}
     uVRegion {bIsValid=false, max=<0 0>, min=<0 0>}
    }
   }
   slot {class=HorizontalBoxSlot, properties {verticalAlignment=VAlign_Center, padding=<ltrb 4 0 8 0>}}
  }
  TextSubDescription {
   widgetClassPath=&CommonRichTextBlock
   bIsVariable=true
   parent=SubTextBox
   widget=TextSubDescription
   properties {
    bOverrideDefaultStyle=true
    defaultTextStyleOverrideClass=/FortUILibrary/Styles/Typography/Foundation/TextStyle-BurbankSmall-XS-White.TextStyle-BurbankSmall-XS-White_C
    mobileTextBlockScale=1.4
    text=Sub Description Sub Description Sub Description
    textStyleSet=/Game/Athena/HUD/TextStyles/AthenaSubInteractionStyles
    textTransformPolicy=ToUpper
    defaultTextStyleOverride {
     overflowPolicy=Clip
     transformPolicy=None
     colorAndOpacity {colorUseRule=UseColor_Specified, specifiedColor=<rgba 1 1 1 1>}
     font {
      bForceMonospaced=false
      bMaterialIsStencil=false
      fontMaterial=null
      fontObject=&NotoSans
      letterSpacing=0
      monospacedWidth=1
      size=24
      skewAmount=0
      typefaceFontName=NotoSans-Regular
      outlineSettings {
       bApplyOutlineToDropShadows=false
       bMiteredCorners=false
       bSeparateFillAlpha=false
       outlineMaterial=null
       outlineSize=0
       outlineColor=<rgba 0 0 0 1>
      }
     }
     highlightColor {colorUseRule=UseColor_Specified, specifiedColor=<rgba 0 0 0 1>}
     highlightShape {
      bIsDynamicallyLoaded=false
      drawAs=Image
      imageType=NoImage
      mirroring=NoMirror
      resourceName=None
      resourceObject=null
      tiling=NoTile
      imageSize=<32 32>
      margin=<ltrb 0 0 0 0>
      outlineSettings {
       bUseBrushTransparency=false
       roundingType=HalfHeightRadius
       width=0
       color {colorUseRule=UseColor_Specified, specifiedColor=<rgba 0 0 0 0>}
       cornerRadii=<0 0 0 0>
      }
      tintColor {colorUseRule=UseColor_Specified, specifiedColor=<rgba 1 1 1 1>}
      uVRegion {bIsValid=false, max=<0 0>, min=<0 0>}
     }
     selectedBackgroundColor {colorUseRule=UseColor_Foreground, specifiedColor=<rgba 1 0 1 1>}
     shadowColorAndOpacity=<rgba 0 0 0 1>
     shadowOffset=<0 0>
     strikeBrush {
      bIsDynamicallyLoaded=false
      drawAs=Image
      imageType=NoImage
      mirroring=NoMirror
      resourceName=None
      resourceObject=null
      tiling=NoTile
      imageSize=<32 32>
      margin=<ltrb 0 0 0 0>
      outlineSettings {
       bUseBrushTransparency=false
       roundingType=HalfHeightRadius
       width=0
       color {colorUseRule=UseColor_Specified, specifiedColor=<rgba 0 0 0 0>}
       cornerRadii=<0 0 0 0>
      }
      tintColor {colorUseRule=UseColor_Specified, specifiedColor=<rgba 1 1 1 1>}
      uVRegion {bIsValid=false, max=<0 0>, min=<0 0>}
     }
     underlineBrush {
      bIsDynamicallyLoaded=false
      drawAs=Image
      imageType=NoImage
      mirroring=NoMirror
      resourceName=None
      resourceObject=null
      tiling=NoTile
      imageSize=<32 32>
      margin=<ltrb 0 0 0 0>
      outlineSettings {
       bUseBrushTransparency=false
       roundingType=HalfHeightRadius
       width=0
       color {colorUseRule=UseColor_Specified, specifiedColor=<rgba 0 0 0 0>}
       cornerRadii=<0 0 0 0>
      }
      tintColor {colorUseRule=UseColor_Specified, specifiedColor=<rgba 1 1 1 1>}
      uVRegion {bIsValid=false, max=<0 0>, min=<0 0>}
     }
    }
   }
   slot {class=HorizontalBoxSlot, properties {horizontalAlignment=HAlign_Left, verticalAlignment=VAlign_Center}}
  }
  Overlay_Wumba_Upgrade {
   widgetClassPath=&Overlay
   bIsVariable=true
   parent=VerticalBox
   widget=Overlay_Wumba_Upgrade
   properties {bOverride_Cursor=true}
   slot {class=VerticalBoxSlot, properties {verticalAlignment=VAlign_Center}}
  }
  Slot_ExtraAppend_FirstIntection {
   widgetClassPath=&Overlay
   bIsVariable=true
   parent=VerticalBox
   widget=Slot_ExtraAppend_FirstIntection
   properties {bOverride_Cursor=true}
   slot {class=VerticalBoxSlot, properties {verticalAlignment=VAlign_Top}}
  }
  Switcher_SecondaryDivider {
   widgetClassPath=&CommonVisibilitySwitcher
   bIsVariable=true
   parent=VerticalBox_0
   widget=Switcher_SecondaryDivider
   slot {class=VerticalBoxSlot, properties {verticalAlignment=VAlign_Top}}
  }
  SecondaryBoxDividerLine {
   widgetClassPath=&Image
   bIsVariable=true
   parent=Switcher_SecondaryDivider
   widget=SecondaryBoxDividerLine
   properties {
    visibility=Collapsed
    brush {
     bIsDynamicallyLoaded=false
     drawAs=Image
     imageType=NoImage
     mirroring=NoMirror
     resourceName=None
     resourceObject=null
     tiling=NoTile
     imageSize=<32 2>
     margin=<ltrb 0 0 0 0>
     outlineSettings {
      bUseBrushTransparency=false
      roundingType=HalfHeightRadius
      width=0
      color {colorUseRule=UseColor_Specified, specifiedColor=<rgba 0 0 0 0>}
      cornerRadii=<0 0 0 0>
     }
     tintColor {colorUseRule=UseColor_Specified, specifiedColor=<rgba 1 1 1 1>}
     uVRegion {bIsValid=false, max=<0 0>, min=<0 0>}
    }
    colorAndOpacity=<rgba 1 1 1 0.0625>
   }
   slot {class=CommonVisibilitySwitcherSlot, properties {padding=<ltrb 0 0 0 -2>}}
  }
  Spacer_SecondaryAction {
   widgetClassPath=&Spacer
   bIsVariable=true
   parent=Switcher_SecondaryDivider
   widget=Spacer_SecondaryAction
   properties {size=<0 25>}
   slot {class=CommonVisibilitySwitcherSlot}
  }
  SecondaryBorder {
   widgetClassPath=&Border
   bInherited=true
   bIsVariable=true
   parent=VerticalBox_0
   widget=SecondaryBorder
   properties {
    background {
     bIsDynamicallyLoaded=false
     drawAs=Box
     imageType=FullColor
     mirroring=NoMirror
     resourceName=None
     resourceObject=&M_UI_HUD_Rounded_Interaction_Backing
     tiling=NoTile
     imageSize=<8 8>
     margin=<ltrb 0.5 0.5 0.5 0.5>
     outlineSettings {
      bUseBrushTransparency=false
      roundingType=HalfHeightRadius
      width=0
      color {colorUseRule=UseColor_Specified, specifiedColor=<rgba 0 0 0 0>}
      cornerRadii=<0 0 0 0>
     }
     tintColor {colorUseRule=UseColor_Specified, specifiedColor=<rgba 1 1 1 1>}
     uVRegion {bIsValid=false, max=<0 0>, min=<0 0>}
    }
    padding=<ltrb 0 0 0 0>
   }
   slot {class=VerticalBoxSlot}
  }
  VerticalBox_110 {
   widgetClassPath=&VerticalBox
   parent=SecondaryBorder
   widget=VerticalBox_110
   slot {class=BorderSlot, properties {padding=<ltrb 0 0 0 0>}}
  }
  Border_SecondaryDimmedColor {
   widgetClassPath=&Border
   bIsVariable=true
   parent=VerticalBox_110
   widget=Border_SecondaryDimmedColor
   properties {
    background {
     bIsDynamicallyLoaded=false
     drawAs=NoDrawType
     imageType=NoImage
     mirroring=NoMirror
     resourceName=None
     resourceObject=null
     tiling=NoTile
     imageSize=<32 32>
     margin=<ltrb 0 0 0 0>
     outlineSettings {
      bUseBrushTransparency=false
      roundingType=HalfHeightRadius
      width=0
      color {colorUseRule=UseColor_Specified, specifiedColor=<rgba 0 0 0 0>}
      cornerRadii=<0 0 0 0>
     }
     tintColor {colorUseRule=UseColor_Specified, specifiedColor=<rgba 1 1 1 1>}
     uVRegion {bIsValid=false, max=<0 0>, min=<0 0>}
    }
    padding=<ltrb 0 0 0 0>
   }
   slot {class=VerticalBoxSlot}
  }
  VerticalBox_1 {
   widgetClassPath=&VerticalBox
   parent=Border_SecondaryDimmedColor
   widget=VerticalBox_1
   slot {class=BorderSlot, properties {padding=<ltrb 0 0 0 0>}}
  }
  SecondInteractionBox {
   widgetClassPath=&HorizontalBox
   bIsVariable=true
   parent=VerticalBox_1
   widget=SecondInteractionBox
   slot {class=VerticalBoxSlot, properties {padding=<ltrb 8 0 16 0>}}
  }
  SecondInteractionKeybindColor {
   widgetClassPath=&Border
   bIsVariable=true
   parent=SecondInteractionBox
   widget=SecondInteractionKeybindColor
   properties {
    verticalAlignment=VAlign_Center
    visibility=SelfHitTestInvisible
    background {
     bIsDynamicallyLoaded=false
     drawAs=NoDrawType
     imageType=NoImage
     mirroring=NoMirror
     resourceName=None
     resourceObject=null
     tiling=NoTile
     imageSize=<32 32>
     margin=<ltrb 0 0 0 0>
     outlineSettings {
      bUseBrushTransparency=false
      roundingType=HalfHeightRadius
      width=0
      color {colorUseRule=UseColor_Specified, specifiedColor=<rgba 0 0 0 0>}
      cornerRadii=<0 0 0 0>
     }
     tintColor {colorUseRule=UseColor_Specified, specifiedColor=<rgba 1 1 1 1>}
     uVRegion {bIsValid=false, max=<0 0>, min=<0 0>}
    }
    padding=<ltrb 8 8 12 8>
   }
   slot {class=HorizontalBoxSlot}
  }
  VerticalBox_2 {
   widgetClassPath=&VerticalBox
   parent=SecondInteractionKeybindColor
   widget=VerticalBox_2
   properties {bOverride_Cursor=true}
   slot {class=BorderSlot, properties {verticalAlignment=VAlign_Center, padding=<ltrb 8 8 12 8>}}
  }
  SecondInteractionKeybind {
   widgetClassPath=&KeybindWidget
   bIsVariable=true
   parent=VerticalBox_2
   widget=SecondInteractionKeybind
   slot {class=VerticalBoxSlot, properties {horizontalAlignment=HAlign_Center, verticalAlignment=VAlign_Center, padding=<ltrb 0 0 4 0>}}
  }
  TextSecondInteraction {
   widgetClassPath=&CommonTextBlock
   bIsVariable=true
   parent=VerticalBox_2
   widget=TextSecondInteraction
   properties {
    text="{Action}"
    textTransformPolicy=ToUpper
    visibility=Collapsed
    font {
     bForceMonospaced=false
     bMaterialIsStencil=false
     fontMaterial=null
     fontObject=&BurbankSmall
     letterSpacing=0
     monospacedWidth=1
     size=15
     skewAmount=0
     typefaceFontName=Black
     outlineSettings {
      bApplyOutlineToDropShadows=false
      bMiteredCorners=false
      bSeparateFillAlpha=true
      outlineMaterial=&M-FontShadow
      outlineSize=2
      outlineColor=<rgba 0 0 0 0.5>
     }
    }
    shadowOffset=<0 0>
   }
   slot {class=VerticalBoxSlot, properties {horizontalAlignment=HAlign_Left, verticalAlignment=VAlign_Top, padding=<ltrb 0 8 0 0>}}
  }
  Switcher_SecondaryContext {
   widgetClassPath=&CommonVisibilitySwitcher
   bInherited=true
   bIsVariable=true
   parent=SecondInteractionBox
   widget=Switcher_SecondaryContext
   slot {class=HorizontalBoxSlot, properties {verticalAlignment=VAlign_Center, size {sizeRule=Fill, value=1}}}
  }
  BasicInteractionWidgetSecondary {
   widgetClassPath=&BasicInteractionWidget
   bIsVariable=true
   parent=Switcher_SecondaryContext
   widget=BasicInteractionWidgetSecondary
   properties {visibility=Collapsed}
   slot {class=CommonVisibilitySwitcherSlot, properties {padding=<ltrb 0 4 0 0>}}
  }
  Border_SubDescription_Secondary {
   widgetClassPath=&Border
   bIsVariable=true
   parent=VerticalBox_1
   widget=Border_SubDescription_Secondary
   properties {
    horizontalAlignment=HAlign_Center
    verticalAlignment=VAlign_Center
    visibility=Collapsed
    brushColor=<rgba 0.0273 0.0273 0.0273 0.8>
    padding=<ltrb 8 4 8 4>
   }
   slot {class=VerticalBoxSlot}
  }
  SubTextBox_Secondary {
   widgetClassPath=&HorizontalBox
   parent=Border_SubDescription_Secondary
   widget=SubTextBox_Secondary
   slot {class=BorderSlot, properties {horizontalAlignment=HAlign_Center, verticalAlignment=VAlign_Center, padding=<ltrb 8 4 8 4>}}
  }
  SubDescriptionAlertImage_Secondary {
   widgetClassPath=&Image
   bIsVariable=true
   parent=SubTextBox_Secondary
   widget=SubDescriptionAlertImage_Secondary
   properties {
    visibility=SelfHitTestInvisible
    brush {
     bIsDynamicallyLoaded=false
     drawAs=Image
     imageType=NoImage
     mirroring=NoMirror
     resourceName=None
     resourceObject=&T-Icon-Alert2-32
     tiling=NoTile
     imageSize=<32 32>
     margin=<ltrb 0 0 0 0>
     outlineSettings {
      bUseBrushTransparency=false
      roundingType=HalfHeightRadius
      width=0
      color {colorUseRule=UseColor_Specified, specifiedColor=<rgba 0 0 0 0>}
      cornerRadii=<0 0 0 0>
     }
     tintColor {colorUseRule=UseColor_Specified, specifiedColor=<rgba 1 1 1 1>}
     uVRegion {bIsValid=false, max=<0 0>, min=<0 0>}
    }
   }
   slot {class=HorizontalBoxSlot, properties {verticalAlignment=VAlign_Center, padding=<ltrb 4 0 8 0>}}
  }
  TextSubDescription_Secondary {
   widgetClassPath=&CommonRichTextBlock
   bIsVariable=true
   parent=SubTextBox_Secondary
   widget=TextSubDescription_Secondary
   properties {
    bOverrideDefaultStyle=true
    defaultTextStyleOverrideClass=&TextStyle-BurbankSmall-XS-White_C
    text="{sub description}"
    textStyleSet=&AthenaSubInteractionStyles
    textTransformPolicy=ToUpper
    defaultTextStyleOverride {
     overflowPolicy=Clip
     transformPolicy=None
     colorAndOpacity {colorUseRule=UseColor_Specified, specifiedColor=<rgba 1 1 1 1>}
     font {
      bForceMonospaced=false
      bMaterialIsStencil=false
      fontMaterial=null
      fontObject=&NotoSans
      letterSpacing=0
      monospacedWidth=1
      size=24
      skewAmount=0
      typefaceFontName=NotoSans-Regular
      outlineSettings {
       bApplyOutlineToDropShadows=false
       bMiteredCorners=false
       bSeparateFillAlpha=false
       outlineMaterial=null
       outlineSize=0
       outlineColor=<rgba 0 0 0 1>
      }
     }
     highlightColor {colorUseRule=UseColor_Specified, specifiedColor=<rgba 0 0 0 1>}
     highlightShape {
      bIsDynamicallyLoaded=false
      drawAs=Image
      imageType=NoImage
      mirroring=NoMirror
      resourceName=None
      resourceObject=null
      tiling=NoTile
      imageSize=<32 32>
      margin=<ltrb 0 0 0 0>
      outlineSettings {
       bUseBrushTransparency=false
       roundingType=HalfHeightRadius
       width=0
       color {colorUseRule=UseColor_Specified, specifiedColor=<rgba 0 0 0 0>}
       cornerRadii=<0 0 0 0>
      }
      tintColor {colorUseRule=UseColor_Specified, specifiedColor=<rgba 1 1 1 1>}
      uVRegion {bIsValid=false, max=<0 0>, min=<0 0>}
     }
     selectedBackgroundColor {colorUseRule=UseColor_Foreground, specifiedColor=<rgba 1 0 1 1>}
     shadowColorAndOpacity=<rgba 0 0 0 1>
     shadowOffset=<0 0>
     strikeBrush {
      bIsDynamicallyLoaded=false
      drawAs=Image
      imageType=NoImage
      mirroring=NoMirror
      resourceName=None
      resourceObject=null
      tiling=NoTile
      imageSize=<32 32>
      margin=<ltrb 0 0 0 0>
      outlineSettings {
       bUseBrushTransparency=false
       roundingType=HalfHeightRadius
       width=0
       color {colorUseRule=UseColor_Specified, specifiedColor=<rgba 0 0 0 0>}
       cornerRadii=<0 0 0 0>
      }
      tintColor {colorUseRule=UseColor_Specified, specifiedColor=<rgba 1 1 1 1>}
      uVRegion {bIsValid=false, max=<0 0>, min=<0 0>}
     }
     underlineBrush {
      bIsDynamicallyLoaded=false
      drawAs=Image
      imageType=NoImage
      mirroring=NoMirror
      resourceName=None
      resourceObject=null
      tiling=NoTile
      imageSize=<32 32>
      margin=<ltrb 0 0 0 0>
      outlineSettings {
       bUseBrushTransparency=false
       roundingType=HalfHeightRadius
       width=0
       color {colorUseRule=UseColor_Specified, specifiedColor=<rgba 0 0 0 0>}
       cornerRadii=<0 0 0 0>
      }
      tintColor {colorUseRule=UseColor_Specified, specifiedColor=<rgba 1 1 1 1>}
      uVRegion {bIsValid=false, max=<0 0>, min=<0 0>}
     }
    }
   }
   slot {class=HorizontalBoxSlot, properties {horizontalAlignment=HAlign_Left, verticalAlignment=VAlign_Center}}
  }
  WBP_SecondaryInteractionWarning {
   widgetClassPath=&WBP_InteractionWarning
   bIsVariable=true
   parent=VerticalBox_110
   widget=WBP_SecondaryInteractionWarning
   slot {class=VerticalBoxSlot}
  }
  Overlay_Wumba_Sidegrade {
   widgetClassPath=&Overlay
   bIsVariable=true
   parent=VerticalBox_110
   widget=Overlay_Wumba_Sidegrade
   properties {bOverride_Cursor=true}
   slot {class=VerticalBoxSlot, properties {verticalAlignment=VAlign_Center}}
  }
  Slot_ExtraAppend_SecondInteraction {
   widgetClassPath=&Overlay
   bIsVariable=true
   parent=VerticalBox_110
   widget=Slot_ExtraAppend_SecondInteraction
   properties {bOverride_Cursor=true}
   slot {class=VerticalBoxSlot, properties {verticalAlignment=VAlign_Top}}
  }
  VerticalBox_PortalMatchMaking {
   widgetClassPath=&VerticalBox
   bIsVariable=true
   parent=VerticalBox_0
   widget=VerticalBox_PortalMatchMaking
   slot {class=VerticalBoxSlot}
  }
  HorizontalBox_PortalMatchmaking {
   widgetClassPath=&HorizontalBox
   bIsVariable=true
   parent=VerticalBox_PortalMatchMaking
   widget=HorizontalBox_PortalMatchmaking
   slot {class=VerticalBoxSlot}
  }
  Overlay_LeftSidebar {
   widgetClassPath=&Overlay
   parent=HorizontalBox_PortalMatchmaking
   widget=Overlay_LeftSidebar
   properties {clipping=ClipToBounds}
   slot {class=HorizontalBoxSlot}
  }
  Image_117 {
   widgetClassPath=&Image
   bIsVariable=true
   parent=Overlay_LeftSidebar
   widget=Image_117
   properties {
    brush {
     bIsDynamicallyLoaded=false
     drawAs=Image
     imageType=NoImage
     mirroring=NoMirror
     resourceName=None
     resourceObject=/Game/UI/Creative/Textures/T_Icon_HUD_Playoption_button
     tiling=NoTile
     imageSize=<50 50>
     margin=<ltrb 0 0 0 0>
     outlineSettings {
      bUseBrushTransparency=false
      roundingType=HalfHeightRadius
      width=0
      color {colorUseRule=UseColor_Specified, specifiedColor=<rgba 0 0 0 0>}
      cornerRadii=<0 0 0 0>
     }
     tintColor {colorUseRule=UseColor_Specified, specifiedColor=<rgba 1 1 1 1>}
     uVRegion {bIsValid=false, max=<0 0>, min=<0 0>}
    }
    colorAndOpacity=<rgba 0 0 0 0.5>
    renderTransform {angle=0, scale=<1 1>, shear=<0 0>, translation=<25 0>}
   }
   slot {class=OverlaySlot, properties {verticalAlignment=VAlign_Fill, padding=<ltrb -25 0 0 0>}}
  }
  Border_83 {
   widgetClassPath=&Border
   parent=HorizontalBox_PortalMatchmaking
   widget=Border_83
   properties {brushColor=<rgba 0 0 0 0.5>, padding=<ltrb 0 0 0 0>}
   slot {class=HorizontalBoxSlot, properties {size {sizeRule=Fill, value=1}}}
  }
  Border_PrimaryPortalKeybind {
   widgetClassPath=&Border
   bIsVariable=true
   parent=Border_83
   widget=Border_PrimaryPortalKeybind
   properties {
    verticalAlignment=VAlign_Center
    visibility=SelfHitTestInvisible
    background {
     bIsDynamicallyLoaded=false
     drawAs=NoDrawType
     imageType=NoImage
     mirroring=NoMirror
     resourceName=None
     resourceObject=null
     tiling=NoTile
     imageSize=<32 32>
     margin=<ltrb 0 0 0 0>
     outlineSettings {
      bUseBrushTransparency=false
      roundingType=HalfHeightRadius
      width=0
      color {colorUseRule=UseColor_Specified, specifiedColor=<rgba 0 0 0 0>}
      cornerRadii=<0 0 0 0>
     }
     tintColor {colorUseRule=UseColor_Specified, specifiedColor=<rgba 1 1 1 1>}
     uVRegion {bIsValid=false, max=<0 0>, min=<0 0>}
    }
    padding=<ltrb 0 0 0 0>
   }
   slot {class=BorderSlot, properties {padding=<ltrb 0 0 0 0>}}
  }
  HorizontalBox_4 {
   widgetClassPath=&HorizontalBox
   parent=Border_PrimaryPortalKeybind
   widget=HorizontalBox_4
   slot {class=BorderSlot, properties {verticalAlignment=VAlign_Center, padding=<ltrb 0 0 0 0>}}
  }
  Keybind_PrimaryPortal {
   widgetClassPath=&KeybindWidget
   bIsVariable=true
   parent=HorizontalBox_4
   widget=Keybind_PrimaryPortal
   slot {class=HorizontalBoxSlot, properties {horizontalAlignment=HAlign_Left, verticalAlignment=VAlign_Center, padding=<ltrb 0 0 4 0>}}
  }
  Text_PrimaryPortal {
   widgetClassPath=&CommonTextBlock
   bIsVariable=true
   parent=HorizontalBox_4
   widget=Text_PrimaryPortal
   properties {
    text="{Action}"
    textTransformPolicy=ToUpper
    font {
     bForceMonospaced=false
     bMaterialIsStencil=false
     fontMaterial=null
     fontObject=&BurbankSmall
     letterSpacing=0
     monospacedWidth=1
     size=15
     skewAmount=0
     typefaceFontName=Black
     outlineSettings {
      bApplyOutlineToDropShadows=false
      bMiteredCorners=false
      bSeparateFillAlpha=true
      outlineMaterial=&M-FontShadow
      outlineSize=2
      outlineColor=<rgba 0 0 0 0.5>
     }
    }
    shadowOffset=<0 0>
   }
   slot {class=HorizontalBoxSlot, properties {verticalAlignment=VAlign_Center, padding=<ltrb 8 0 0 0>}}
  }
  Overlay_RightSidebar {
   widgetClassPath=&Overlay
   parent=HorizontalBox_PortalMatchmaking
   widget=Overlay_RightSidebar
   properties {clipping=ClipToBounds}
   slot {class=HorizontalBoxSlot}
  }
  Image_169 {
   widgetClassPath=&Image
   bIsVariable=true
   parent=Overlay_RightSidebar
   widget=Image_169
   properties {
    brush {
     bIsDynamicallyLoaded=false
     drawAs=Image
     imageType=NoImage
     mirroring=NoMirror
     resourceName=None
     resourceObject=&T_Icon_HUD_Playoption_button
     tiling=NoTile
     imageSize=<50 50>
     margin=<ltrb 0 0 0 0>
     outlineSettings {
      bUseBrushTransparency=false
      roundingType=HalfHeightRadius
      width=0
      color {colorUseRule=UseColor_Specified, specifiedColor=<rgba 0 0 0 0>}
      cornerRadii=<0 0 0 0>
     }
     tintColor {colorUseRule=UseColor_Specified, specifiedColor=<rgba 1 1 1 1>}
     uVRegion {bIsValid=false, max=<0 0>, min=<0 0>}
    }
    colorAndOpacity=<rgba 0 0 0 0.5>
    renderTransform {angle=0, scale=<1 1>, shear=<0 0>, translation=<25 0>}
   }
   slot {class=OverlaySlot, properties {verticalAlignment=VAlign_Fill, padding=<ltrb -50 0 25 0>}}
  }
  Spacer_70 {
   widgetClassPath=&Spacer
   parent=VerticalBox_PortalMatchMaking
   widget=Spacer_70
   properties {size=<1 10>}
   slot {class=VerticalBoxSlot}
  }
  HorizontalBox_SecondaryPortalMatchmaking_1 {
   widgetClassPath=&HorizontalBox
   bIsVariable=true
   parent=VerticalBox_PortalMatchMaking
   widget=HorizontalBox_SecondaryPortalMatchmaking_1
   slot {class=VerticalBoxSlot}
  }
  Overlay_LeftSidebar_1 {
   widgetClassPath=&Overlay
   parent=HorizontalBox_SecondaryPortalMatchmaking_1
   widget=Overlay_LeftSidebar_1
   properties {clipping=ClipToBounds}
   slot {class=HorizontalBoxSlot}
  }
  Image {
   widgetClassPath=&Image
   bIsVariable=true
   parent=Overlay_LeftSidebar_1
   widget=Image
   properties {
    brush {
     bIsDynamicallyLoaded=false
     drawAs=Image
     imageType=NoImage
     mirroring=NoMirror
     resourceName=None
     resourceObject=&T_Icon_HUD_Playoption_button
     tiling=NoTile
     imageSize=<50 50>
     margin=<ltrb 0 0 0 0>
     outlineSettings {
      bUseBrushTransparency=false
      roundingType=HalfHeightRadius
      width=0
      color {colorUseRule=UseColor_Specified, specifiedColor=<rgba 0 0 0 0>}
      cornerRadii=<0 0 0 0>
     }
     tintColor {colorUseRule=UseColor_Specified, specifiedColor=<rgba 1 1 1 1>}
     uVRegion {bIsValid=false, max=<0 0>, min=<0 0>}
    }
    colorAndOpacity=<rgba 0 0 0 0.5>
    renderTransform {angle=0, scale=<1 1>, shear=<0 0>, translation=<25 0>}
   }
   slot {class=OverlaySlot, properties {verticalAlignment=VAlign_Fill, padding=<ltrb -25 0 0 0>}}
  }
  Border {
   widgetClassPath=&Border
   parent=HorizontalBox_SecondaryPortalMatchmaking_1
   widget=Border
   properties {brushColor=<rgba 0 0 0 0.5>, padding=<ltrb 0 0 0 0>}
   slot {class=HorizontalBoxSlot, properties {size {sizeRule=Fill, value=1}}}
  }
  Border_SecondaryPortalKeybind {
   widgetClassPath=&Border
   bIsVariable=true
   parent=Border
   widget=Border_SecondaryPortalKeybind
   properties {
    verticalAlignment=VAlign_Center
    visibility=SelfHitTestInvisible
    background {
     bIsDynamicallyLoaded=false
     drawAs=NoDrawType
     imageType=NoImage
     mirroring=NoMirror
     resourceName=None
     resourceObject=null
     tiling=NoTile
     imageSize=<32 32>
     margin=<ltrb 0 0 0 0>
     outlineSettings {
      bUseBrushTransparency=false
      roundingType=HalfHeightRadius
      width=0
      color {colorUseRule=UseColor_Specified, specifiedColor=<rgba 0 0 0 0>}
      cornerRadii=<0 0 0 0>
     }
     tintColor {colorUseRule=UseColor_Specified, specifiedColor=<rgba 1 1 1 1>}
     uVRegion {bIsValid=false, max=<0 0>, min=<0 0>}
    }
    padding=<ltrb 0 0 0 0>
   }
   slot {class=BorderSlot, properties {padding=<ltrb 0 0 0 0>}}
  }
  HorizontalBox_6 {
   widgetClassPath=&HorizontalBox
   parent=Border_SecondaryPortalKeybind
   widget=HorizontalBox_6
   slot {class=BorderSlot, properties {verticalAlignment=VAlign_Center, padding=<ltrb 0 0 0 0>}}
  }
  Keybind_SecondaryPortal {
   widgetClassPath=&KeybindWidget
   bIsVariable=true
   parent=HorizontalBox_6
   widget=Keybind_SecondaryPortal
   slot {class=HorizontalBoxSlot, properties {horizontalAlignment=HAlign_Left, verticalAlignment=VAlign_Center, padding=<ltrb 0 0 4 0>}}
  }
  Text_SecondaryPortal {
   widgetClassPath=&CommonTextBlock
   bIsVariable=true
   parent=HorizontalBox_6
   widget=Text_SecondaryPortal
   properties {
    text="{Action}"
    textTransformPolicy=ToUpper
    font {
     bForceMonospaced=false
     bMaterialIsStencil=false
     fontMaterial=null
     fontObject=&BurbankSmall
     letterSpacing=0
     monospacedWidth=1
     size=15
     skewAmount=0
     typefaceFontName=Black
     outlineSettings {
      bApplyOutlineToDropShadows=false
      bMiteredCorners=false
      bSeparateFillAlpha=true
      outlineMaterial=&M-FontShadow
      outlineSize=2
      outlineColor=<rgba 0 0 0 0.5>
     }
    }
    shadowOffset=<0 0>
   }
   slot {class=HorizontalBoxSlot, properties {verticalAlignment=VAlign_Center, padding=<ltrb 8 0 0 0>}}
  }
  Overlay_RightSidebar_1 {
   widgetClassPath=&Overlay
   parent=HorizontalBox_SecondaryPortalMatchmaking_1
   widget=Overlay_RightSidebar_1
   properties {clipping=ClipToBounds}
   slot {class=HorizontalBoxSlot}
  }
  Image_1 {
   widgetClassPath=&Image
   bIsVariable=true
   parent=Overlay_RightSidebar_1
   widget=Image_1
   properties {
    brush {
     bIsDynamicallyLoaded=false
     drawAs=Image
     imageType=NoImage
     mirroring=NoMirror
     resourceName=None
     resourceObject=&T_Icon_HUD_Playoption_button
     tiling=NoTile
     imageSize=<50 50>
     margin=<ltrb 0 0 0 0>
     outlineSettings {
      bUseBrushTransparency=false
      roundingType=HalfHeightRadius
      width=0
      color {colorUseRule=UseColor_Specified, specifiedColor=<rgba 0 0 0 0>}
      cornerRadii=<0 0 0 0>
     }
     tintColor {colorUseRule=UseColor_Specified, specifiedColor=<rgba 1 1 1 1>}
     uVRegion {bIsValid=false, max=<0 0>, min=<0 0>}
    }
    colorAndOpacity=<rgba 0 0 0 0.5>
    renderTransform {angle=0, scale=<1 1>, shear=<0 0>, translation=<25 0>}
   }
   slot {class=OverlaySlot, properties {verticalAlignment=VAlign_Fill, padding=<ltrb -50 0 25 0>}}
  }
  TutorialHighlight {widgetClassPath=&TutorialHighlight}
 ]
}
