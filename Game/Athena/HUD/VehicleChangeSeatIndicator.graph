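# Actor indicator HUD widget that labels the target seat during a vehicle seat switch.
# On a target-index change with a positive seat cooldown, it attaches to the vehicle, sets seat text, seat/wheel icon and tail flip, then plays a scale-in intro.
# Tick refreshes the focused seat socket while the cooldown runs.
# Entering as driver or passenger hides it, and a cooldown-length timer serves as failsafe.
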WidgetBlueprint /Game/Athena/HUD/VehicleChangeSeatIndicator
animations [Intro {
  binding_count=1
  display_label=""
  length_seconds=0.1667
  object=&self:Intro
  sequence {
   bindings [VerticalBox_0 {
     id=#1
     possessed_class=VerticalBox
     tracks [{
       type=MovieScene2DTransformTrack
       sections [{
         type=MovieScene2DTransformSection
         end_frame=20
         start_frame=0
         channels [
          Translation.X {type=MovieSceneScriptingFloatChannel}
          Translation.Y {type=MovieSceneScriptingFloatChannel}
          Angle {type=MovieSceneScriptingFloatChannel}
          Scale.X {type=MovieSceneScriptingFloatChannel, keys [[0,0],[9,0.998],[14,0.9601],[19.996,1]]}
          Scale.Y {type=MovieSceneScriptingFloatChannel, keys [[0,0],[9,0.998],[14,0.9601],[19.996,1]]}
          Shear.X {type=MovieSceneScriptingFloatChannel}
          Shear.Y {type=MovieSceneScriptingFloatChannel}
         ]
        }]
      }]
    }]
   display_rate {numerator=120}
  }
 }]
graphs [EventGraph {nodes [
   HideWidget() {is=title K2Node_CustomEvent}
   > SetVisibility(InVisibility=Collapsed)
   > Clear and Invalidate Timer by Handle(Handle=Get FailsafeHideTimer() -> (FailsafeHideTimer:Timer Handle Structure))
   OnEnterVehicleDriver_Bind() {is=title K2Node_CustomEvent}
   > HideWidget() {is=title K2Node_CallFunction}
   OnEnterVehiclePassenger_Bind() {is=title K2Node_CustomEvent}
   > HideWidget()
   n1 = OnVehicleSeatTransitionTargetIndexChange_Bind!() -> (OutputDelegate, then, SeatIndex:int) {is=title K2Node_CustomEvent}
   > Set CurrentSeatIndex(CurrentSeatIndex:int=@n1.SeatIndex) -> (Output_Get:int=0)
   > n2 = Cast To PlayerPawn_Athena(Object=Get Controlled Pawn!(self:Controller Object Reference=GetOwningPlayer()) -> (ReturnValue:Pawn Object Reference) {is=title K2Node_CallFunction})
   n3 = GetVehicleInterface!(self:Fort Player Pawn Object Reference=@n2) -> (ReturnValue:Fort Vehicle Interface Interface)
   > Is Valid(InputObject="To Object (Interface)"!(Interface:"Interface Interface (by ref)"=@n3) -> (ReturnValue:Object Reference))
   n4 = GetVehicleSeatIndex!(self:Fort Player Pawn Object Reference=@n2) -> (ReturnValue:int=0)
   > Branch(Condition="float > float"(A=GetSeatSwitchCooldown!(self:Fort Vehicle Interface Interface=@n3, SeatIndex:int=@n4) -> (ReturnValue:float), B=0) {is=title K2Node_CallFunction})
   > n5 = GetVehicleActor!(execute, self:Fort Vehicle Interface Interface=@n3) -> (then, ReturnValue:Actor Object Reference)
   > SetIndicatedActor!(execute, self:Fort Actor Indicator Widget Object Reference, NewIndicatorActor:Actor Object Reference=@n5) -> (then)
   > SetCurrentFocusedSocketLocation!(execute, self:Fort Vehicle Interface Interface=@n3, SeatIndex:int=Get CurrentSeatIndex() -> (CurrentSeatIndex:int=0)) -> (then)
   > UpdateDescription!(
    execute
    self:Basic Interaction Widget Base Object Reference=Get BasicInteractionWidget() -> (BasicInteractionWidget:Basic Interaction Widget Object Reference)
    Description:Text=GetVehicleSeatChoiceDisplayText!(self:Fort Vehicle Interface Interface=@n3, SeatIndex:int=Get CurrentSeatIndex() -> (CurrentSeatIndex=0)) -> (ReturnValue:Text)
    DescriptionColor:Linear Color Structure=Get Description Color() -> (Description Color:Linear Color Structure)
    SubDescription:Text
    SubDescriptionColor:Linear Color Structure
    InteractionCount:int=-1
   ) -> (then)
   n6 = Get CurrentSeatIndex() -> (CurrentSeatIndex=0)
   > SetBrushFromTexture!(
    execute
    self:Image Object Reference=Get Image_Icon() -> (Image_Icon:Image Object Reference)
    Texture:Texture 2D Object Reference=Select(
     Option 0:Texture 2D Object Reference=/Game/Athena/HUD/Materials/T_SeatIcon
     Option 1:Texture 2D Object Reference=/Game/Athena/HUD/Materials/T_WheelIcon
     Index="Equal (Integer)"(A=@n6, B=0) {is=title K2Node_CallFunction}
    ) -> (ReturnValue:Texture 2D Object Reference)
    bMatchSize:bool=false
   ) -> (then)
   n7 = Get CurrentSeatIndex() -> (CurrentSeatIndex=0)
   n8 = "Equal (Integer)"(A=@n7, B=3)
   n9 = "Equal (Integer)"(A=@n7, B=0)
   > SetRenderScale!(
    execute
    self:Widget Object Reference=Get I_SquiglyTail() -> (I_SquiglyTail:Image Object Reference)
    Scale:Vector 2D Structure=Select(Option 0:Vector 2D Structure=(X=-1,Y=0.75), Option 1:Vector 2D Structure=(X=1,Y=0.75), Index=OR Boolean(A=@n9, B=@n8)) -> (ReturnValue:Vector 2D Structure)
   ) -> (then)
   > SetVisibility(InVisibility=SelfHitTestInvisible)
   > PlayAnimation(InAnimation=Get Intro())
   > Clear and Invalidate Timer by Handle(Handle=Get FailsafeHideTimer())
   > n10 = Set Timer by Event(Delegate=Create Event(), Time=GetSeatSwitchCooldown(self=@n3, SeatIndex=GetVehicleSeatIndex(self=@n2) -> (ReturnValue=0)), bLooping=false)
   > Set FailsafeHideTimer(FailsafeHideTimer:Timer Handle Structure=@n10) -> (Output_Get:Timer Handle Structure)
   n11 = BreakVector2D(InVec=GetDesiredSize!(self:Widget Object Reference=Get BasicInteractionWidget()) -> (ReturnValue:Vector 2D Structure))
   > Set ScreenSpaceRelativeOffset(ScreenSpaceRelativeOffset:Vector 2D Structure=MakeVector2D(X=float * float(A=@n11, B=-0.5), Y=-100)) -> (Output_Get:Vector 2D Structure)
   n12 = Event Tick()
   > Parent: Tick!(execute, MyGeometry:Geometry Structure=@n12.MyGeometry, InDeltaTime:float=@n12.InDeltaTime) -> (then) {is=title K2Node_CallParentFunction}
   > n13 = Cast To PlayerPawn_Athena(Object=Get Controlled Pawn(self=GetOwningPlayer()))
   n14 = GetVehicleInterface(self=@n13)
   > Is Valid(InputObject="To Object (Interface)"(Interface=@n14))
   n15 = GetVehicleSeatIndex(self=@n13) -> (ReturnValue=0)
   > Branch(Condition="float > float"(A=GetSeatSwitchCooldown(self=@n14, SeatIndex=@n15), B=0))
   > SetCurrentFocusedSocketLocation(self=@n14, SeatIndex=Get CurrentSeatIndex() -> (CurrentSeatIndex=0))
   Event Destruct()
   > n16 = Cast To FortPlayerControllerZone!(execute, Object=GetOwningPlayer()) -> (then, CastFailed, AsFort Player Controller Zone:Fort Player Controller Zone Object Reference) {is=title K2Node_DynamicCast}
   n17 = Create Event()
   > Unbind Event from On Vehicle Seat Transition Target Index Change(self:Fort Player Controller Zone Object Reference=@n16, Delegate=@n17) {is=title K2Node_RemoveDelegate}
   n18 = Get FortGlobalUIContext!() -> (ReturnValue:Fort Global UIContext Object Reference) {is=title K2Node_GetBlueprintContextEx}
   > Is Valid(InputObject=@n18)
   n19 = Create Event()
   > Unbind Event from On Enter Vehicle Driver(self:Fort Global UIContext Object Reference=@n18, Delegate=@n19) {is=title K2Node_RemoveDelegate}
   n20 = Create Event()
   > Unbind Event from On Enter Vehicle Passenger(self:Fort Global UIContext Object Reference=@n18, Delegate=@n20) {is=title K2Node_RemoveDelegate}
   Event Construct()
   > n21 = Cast To FortPlayerControllerZone(Object=GetOwningPlayer())
   > Bind Event to On Vehicle Seat Transition Target Index Change(self:Fort Player Controller Zone Object Reference=@n21, Delegate=@n17) {is=title K2Node_AddDelegate}
   > Bind Event to On Enter Vehicle Driver(self:Fort Global UIContext Object Reference=@n18, Delegate=@n19) {is=title K2Node_AddDelegate}
   > Bind Event to On Enter Vehicle Passenger(self:Fort Global UIContext Object Reference=@n18, Delegate=@n20) {is=title K2Node_AddDelegate}
   n22 = Get BasicInteractionWidget()
   > Set Justification!(
    execute
    Justification:ETextJustify Enum=Center
    self:Text Layout Widget Object Reference=Get RichText_Description!(self:Basic Interaction Widget Base Object Reference=@n22) -> (RichText_Description:Common Rich Text Block Object Reference)
   ) -> (then, Output_Get:ETextJustify Enum=Left)
   Get TextDescription!(self:Basic Interaction Widget Object Reference=@n22) -> (TextDescription:Common Text Object Reference)
  ]}]
variables [
 CurrentSeatIndex {type=integer}
 Description Color {type=LinearColor, default=<rgba 1 1 1 1>}
 FailsafeHideTimer {type=TimerHandle}
]
widgets {
 info {inheritedWidgetCount=1, parentClass=/Script/FortniteUI.FortActorIndicatorWidget, rootWidgetClass=&Overlay, widgetCount=10}
 widgets [
  Overlay_2 {widgetClassPath=&Overlay, widget=Overlay_2}
  VerticalBox_0 {
   widgetClassPath=&VerticalBox
   parent=Overlay_2
   widget=VerticalBox_0
   properties {renderTransformPivot=<0.5 1.2>}
   slot {class=OverlaySlot, properties {horizontalAlignment=HAlign_Fill, verticalAlignment=VAlign_Fill}}
  }
  Overlay_0 {widgetClassPath=&Overlay, parent=VerticalBox_0, widget=Overlay_0, slot {class=VerticalBoxSlot}}
  VerticalBox_1 {
   widgetClassPath=&VerticalBox
   parent=Overlay_0
   widget=VerticalBox_1
   slot {class=OverlaySlot, properties {horizontalAlignment=HAlign_Fill, verticalAlignment=VAlign_Fill}}
  }
  Image_0 {
   widgetClassPath=&Image
   bIsVariable=true
   parent=VerticalBox_1
   widget=Image_0
   properties {
    brush {
     bIsDynamicallyLoaded=false
     drawAs=Image
     imageType=NoImage
     mirroring=NoMirror
     resourceName=None
     resourceObject=/FortUILibrary/Foundation/Buttons/Materials/M_SkewButtonShape_DF1_Inst
     tiling=NoTile
     imageSize=<32 32>
     margin=<ltrb 0 0 0 0>
     outlineSettings {
      bUseBrushTransparency=false
      roundingType=HalfHeightRadius
      width=0
      color {colorUseRule=UseColor_Specified, specifiedColor=<rgba 0 0 0 0>}
      cornerRadii=<0 0 0 0>
     }
     tintColor {colorUseRule=UseColor_Specified, specifiedColor=<rgba 0.0052 0.0052 0.0056 1>}
     uVRegion {bIsValid=false, max=<0 0>, min=<0 0>}
    }
    renderTransform {angle=0, scale=<1 1>, shear=<-2 -0.25>, translation=<0 0>}
   }
   slot {class=VerticalBoxSlot, properties {size {sizeRule=Fill, value=1}}}
  }
  I_SquiglyTail {
   widgetClassPath=&Image
   bIsVariable=true
   parent=VerticalBox_1
   widget=I_SquiglyTail
   properties {
    visibility=SelfHitTestInvisible
    brush {
     bIsDynamicallyLoaded=false
     drawAs=Image
     imageType=NoImage
     mirroring=NoMirror
     resourceName=None
     resourceObject=/Game/Athena/UI/Frontend/Parties/Art/M_UI_LightningBanner_Normal_Tail_White
     tiling=NoTile
     imageSize=<64 64>
     margin=<ltrb 0 0 0 0>
     outlineSettings {
      bUseBrushTransparency=false
      roundingType=HalfHeightRadius
      width=0
      color {colorUseRule=UseColor_Specified, specifiedColor=<rgba 0 0 0 0>}
      cornerRadii=<0 0 0 0>
     }
     tintColor {colorUseRule=UseColor_Specified, specifiedColor=<rgba 0.0052 0.0052 0.0056 1>}
     uVRegion {bIsValid=false, max=<0 0>, min=<0 0>}
    }
    colorAndOpacity=<rgba 1 1 1 0.9>
    renderTransform {angle=4, scale=<-1 0.75>, shear=<0 0>, translation=<0 -15>}
    renderTransformPivot=<0.5 0>
   }
   slot {class=VerticalBoxSlot, properties {horizontalAlignment=HAlign_Center, verticalAlignment=VAlign_Bottom, padding=<ltrb 0 0 0 -64>}}
  }
  HorizontalBox_0 {
   widgetClassPath=&HorizontalBox
   parent=Overlay_0
   widget=HorizontalBox_0
   slot {class=OverlaySlot, properties {padding=<ltrb 20 10 20 10>}}
  }
  Image_Icon {
   widgetClassPath=&Image
   bIsVariable=true
   parent=HorizontalBox_0
   widget=Image_Icon
   properties {
    brush {
     bIsDynamicallyLoaded=false
     drawAs=Image
     imageType=NoImage
     mirroring=NoMirror
     resourceName=None
     resourceObject=&T_SeatIcon
     tiling=NoTile
     imageSize=<40 40>
     margin=<ltrb 0 0 0 0>
     outlineSettings {
      bUseBrushTransparency=false
      roundingType=HalfHeightRadius
      width=0
      color {colorUseRule=UseColor_Specified, specifiedColor=<rgba 0 0 0 0>}
      cornerRadii=<0 0 0 0>
     }
     tintColor {colorUseRule=UseColor_Specified, specifiedColor=<rgba 1 1 1 1>}
     uVRegion {bIsValid=false, max=<0 0>, min=<0 0>}
    }
    renderTransform {angle=0, scale=<1 1>, shear=<0 0>, translation=<-6 0>}
   }
   slot {class=HorizontalBoxSlot, properties {horizontalAlignment=HAlign_Left, verticalAlignment=VAlign_Top}}
  }
  BasicInteractionWidget {
   widgetClassPath=/Game/UI/InGame/HUD/BasicInteractionWidget
   bIsVariable=true
   parent=HorizontalBox_0
   widget=BasicInteractionWidget
   slot {class=HorizontalBoxSlot}
  }
  TutorialHighlight {widgetClassPath=&TutorialHighlight}
 ]
}
