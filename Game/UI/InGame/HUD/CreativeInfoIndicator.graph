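# HUD actor indicator widget that shows the display name of the actor focused by the Creative move tool.
# On construct it collapses and polls on a looping timer for the owning FortPawn, then binds to weapon equips.
# Equipping a move tool, if CreativeMoveTool.EnableWidgets is on, binds focus changes that set the indicated actor, visibility and name.
# For building static mesh actors, the indicator offsets to the local bounds center, otherwise zero.
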WidgetBlueprint /Game/UI/InGame/HUD/CreativeInfoIndicator
graphs [EventGraph {nodes [
   n1 = OnWeaponEquippedDelegate_Bind!() -> (OutputDelegate, then, NewWeapon:Fort Weapon Object Reference, PrevWeapon:Fort Weapon Object Reference) {is=title K2Node_CustomEvent}
   > n2 = Cast To FortCreativeMoveTool!(execute, Object=@n1.NewWeapon) -> (then, CastFailed, AsFort Creative Move Tool:Fort Creative Move Tool Object Reference) {is=title K2Node_DynamicCast}
   > Branch(Condition="float > float"(A=GetNumericCVar!(CVar:String=CreativeMoveTool.EnableWidgets) -> (OutValue:float, ReturnValue:bool)))
   > Bind Event to On Move Tool Focus Changed(self:Fort Creative Move Tool Object Reference=@n2, Delegate=Create Event()) {is=title K2Node_AddDelegate}
   n3 = OnIndicatedActorChanged!() -> (OutputDelegate, then, IndicatedActor:Actor Object Reference) {is=title K2Node_CustomEvent}
   > SetIndicatedActor!(execute, self:Fort Actor Indicator Widget Object Reference, NewIndicatorActor:Actor Object Reference=@n3.IndicatedActor) -> (then)
   n4 = IsValid(Object=@n3.IndicatedActor)
   > SetVisibility(InVisibility=Select(Option 0=Collapsed, Option 1=Visible, Index=@n4))
   > Branch(Condition=@n4)
   n5 = GetIndicatedActor!(self:Fort Actor Indicator Interface Interface) -> (ReturnValue:Actor Object Reference)
   > "SetText (Text)"(self=Get NameText() -> (NameText:Text Object Reference), InText="To Text (String)"(InString=GetDisplayName(Object=@n5)))
   > n6 = Cast To BuildingSMActor!(execute, Object=@n5) -> (then, CastFailed, AsBuilding Static Mesh:Building Static Mesh Object Reference) {is=title K2Node_DynamicCast}
   n7 = GetLocalBounds!(self:Static Mesh Component Object Reference=GetBuildingMeshComponent!(self:Building Static Mesh Object Reference=@n6) -> (ReturnValue:Static Mesh Component Object Reference)) -> (Min:Vector, Max:Vector)
   > Set RelativeLocation(RelativeLocation:Vector=vector / float!(
     A:Vector=vector + vector(A:Vector=@n7, B:Vector=@n7.Max) -> (ReturnValue:Vector) {is=title K2Node_CommutativeAssociativeBinaryOperator}
     B:double=2
    ) -> (ReturnValue:Vector)) -> (Output_Get:Vector)
   Set RelativeLocation(execute=@n6.CastFailed, RelativeLocation="0, 0, 0")
   WaitForPawn() {is=title K2Node_CustomEvent}
   > n8 = Cast To FortPawn!(
    execute
    Object=Get Controlled Pawn!(self:Controller Object Reference=GetOwningPlayer()) -> (ReturnValue:Pawn Object Reference) {is=title K2Node_CallFunction}
   ) -> (then, CastFailed, AsFort Pawn:Fort Pawn Object Reference) {is=title K2Node_DynamicCast}
   > Bind Event to On Weapon Equipped Delegate(self:Fort Pawn Object Reference=@n8, Delegate=Create Event()) {is=title K2Node_AddDelegate}
   > Clear and Invalidate Timer by Handle(Handle=Get WaitForPawnTimerHandle() -> (WaitForPawnTimerHandle:Timer Handle Structure))
   SetWaitForPawnTimer() {is=title K2Node_CustomEvent}
   > n9 = Set Timer by Event(Delegate=Create Event(), Time=0.25, bLooping=true)
   > Set WaitForPawnTimerHandle(WaitForPawnTimerHandle:Timer Handle Structure=@n9) -> (Output_Get:Timer Handle Structure)
   Event Construct()
   > Parent: Construct()
   > Sequence!(execute) -> (then_0, then_1, then_2:Exec)
   > SetVisibility(InVisibility=Collapsed)
   > SetWaitForPawnTimer() {is=title K2Node_CallFunction}
  ]}]
variables [WaitForPawnTimerHandle {type=TimerHandle}]
widgets {
 info {
  inheritedWidgetCount=1
  parentClass=/Script/FortniteUI.FortActorIndicatorWidget
  rootWidgetClass=&CommonBorder
  widgetCount=6
 }
 widgets [
  CommonBorder_3 {widgetClassPath=&CommonBorder, widget=CommonBorder_3}
  VerticalBox_0 {widgetClassPath=&VerticalBox, parent=CommonBorder_3, widget=VerticalBox_0, slot {class=BorderSlot}}
  SwitcherInteractions {
   widgetClassPath=&CommonWidgetSwitcherLegacy
   bIsVariable=true
   parent=VerticalBox_0
   widget=SwitcherInteractions
   properties {transitionDuration=0}
   slot {class=VerticalBoxSlot}
  }
  HorizontalBox_0 {
   widgetClassPath=&HorizontalBox
   parent=SwitcherInteractions
   widget=HorizontalBox_0
   slot {class=WidgetSwitcherSlot, properties {horizontalAlignment=HAlign_Center, verticalAlignment=VAlign_Center}}
  }
  NameText {
   widgetClassPath=&TextBlock
   bIsVariable=true
   parent=HorizontalBox_0
   widget=NameText
   properties {
    text=Text Block
    font {
     bForceMonospaced=false
     bMaterialIsStencil=false
     fontMaterial=null
     fontObject=&NotoSans
     letterSpacing=0
     monospacedWidth=1
     size=24
     skewAmount=0
     typefaceFontName=NotoSans-Bold
     outlineSettings {
      bApplyOutlineToDropShadows=false
      bMiteredCorners=false
      bSeparateFillAlpha=false
      outlineMaterial=null
      outlineSize=0
      outlineColor=<rgba 0 0 0 1>
     }
    }
   }
   slot {class=HorizontalBoxSlot}
  }
  TutorialHighlight {widgetClassPath=&TutorialHighlight}
 ]
}
