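# HUD actor indicator widget that marks a point of interest with a tinted icon, pulsing glow, rings and optional label.
# InitPointOfInterest binds the actor, sets text and image, enables screen-space clamping with an arrow, tints the layers and loops two animations.
# RemovePointOfInterest hides and resets the widget only when the given actor matches the indicated one.
# It starts collapsed and exposes only its bottom canvas container to the parent FortActorIndicatorWidget.
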WidgetBlueprint /Game/UI/InGame/HUD/InterestIndicatorWidget
animations [
 NewAnimation_1 {
  binding_count=1
  display_label=""
  length_seconds=1
  object=&self:NewAnimation_1
  sequence {bindings [backgroundpulse {
     id=#1
     possessed_class=Image
     tracks [{
       type=MovieSceneColorTrack
       display_name=Color and Opacity
       sections [{
         type=MovieSceneColorSection
         end_frame=20
         start_frame=0
         channels [
          Color.R {type=MovieSceneScriptingFloatChannel, const [1,1]}
          Color.G {type=MovieSceneScriptingFloatChannel, const [0.6963,1]}
          Color.B {type=MovieSceneScriptingFloatChannel}
          Color.A {type=MovieSceneScriptingFloatChannel, keys [[0,0.125],[10,0.25],[20,0.125]]}
         ]
        }]
      }]
    }]}
 }
 NewAnimation_2 {
  binding_count=3
  display_label=""
  length_seconds=2
  object=&self:NewAnimation_2
  sequence {bindings [
    center {
     id=#2
     possessed_class=Image
     tracks [{
       type=MovieScene2DTransformTrack
       sections [{
         type=MovieScene2DTransformSection
         end_frame=40
         start_frame=0
         channels [
          Translation.X {type=MovieSceneScriptingFloatChannel}
          Translation.Y {type=MovieSceneScriptingFloatChannel}
          Angle {type=MovieSceneScriptingFloatChannel}
          Scale.X {type=MovieSceneScriptingFloatChannel, keys [[0,1],[20,0.825],[40,1]]}
          Scale.Y {type=MovieSceneScriptingFloatChannel, keys [[0,1],[20,0.8],[40,1]]}
          Shear.X {type=MovieSceneScriptingFloatChannel}
          Shear.Y {type=MovieSceneScriptingFloatChannel}
         ]
        }]
      }]
    }
    smallspinring {
     id=#3
     possessed_class=Image
     tracks [
      {
       type=MovieSceneColorTrack
       display_name=Color and Opacity
       sections [{
         type=MovieSceneColorSection
         end_frame=40
         start_frame=0
         channels [
          Color.R {type=MovieSceneScriptingFloatChannel}
          Color.G {type=MovieSceneScriptingFloatChannel}
          Color.B {type=MovieSceneScriptingFloatChannel}
          Color.A {type=MovieSceneScriptingFloatChannel, keys [[0,0.4],[20,0.75],[40,0.4]]}
         ]
        }]
      }
      {
       type=MovieScene2DTransformTrack
       sections [{
         type=MovieScene2DTransformSection
         end_frame=40
         start_frame=0
         channels [
          Translation.X {type=MovieSceneScriptingFloatChannel}
          Translation.Y {type=MovieSceneScriptingFloatChannel}
          Angle {type=MovieSceneScriptingFloatChannel, keys [[0,150],[20,0],[40,-180]]}
          Scale.X {type=MovieSceneScriptingFloatChannel}
          Scale.Y {type=MovieSceneScriptingFloatChannel}
          Shear.X {type=MovieSceneScriptingFloatChannel}
          Shear.Y {type=MovieSceneScriptingFloatChannel}
         ]
        }]
      }
     ]
    }
    largespinring {
     id=#4
     possessed_class=Image
     tracks [{
       type=MovieSceneColorTrack
       display_name=Color and Opacity
       sections [{
         type=MovieSceneColorSection
         end_frame=40
         start_frame=0
         channels [
          Color.R {type=MovieSceneScriptingFloatChannel}
          Color.G {type=MovieSceneScriptingFloatChannel}
          Color.B {type=MovieSceneScriptingFloatChannel}
          Color.A {type=MovieSceneScriptingFloatChannel, keys [[0,0.75],[20,0.4],[40,0.75]]}
         ]
        }]
      }]
    }
   ]}
 }
]
graphs [
 EventGraph {nodes [
   n1 = SetScreenSpacePosition!() -> (OutputDelegate, then, bEnabled:bool=~) {is=title K2Node_CustomEvent}
   > Set bUseScreenSpacePosition(bUseScreenSpacePosition:bool=@n1.bEnabled) -> (Output_Get:bool)
   Event Construct()
   > SetVisibility(InVisibility=Collapsed)
   > PlaySound2D()
   Event Destruct()
   > PlaySound2D()
   Event Tick()
  ]}
 GetCanvasPanelContainer {nodes [
   n1 = GetCanvasPanelContainer!() -> (then, ContainerEnum:EFortActorIndicatorContainerWidget Enum=Top) {is=title K2Node_FunctionEntry}
   > n2 = Switch on EFortActorIndicatorContainerWidget!(execute, Selection:EFortActorIndicatorContainerWidget Enum=@n1) -> (Top:Exec, Middle:Exec, Bottom:Exec) {is=title K2Node_SwitchEnum}
   Return Node!(
    execute=@n2.Bottom
    ReturnValue:Canvas Panel Object Reference=Get BottomCanvasPanelContainer() -> (BottomCanvasPanelContainer:Canvas Panel Object Reference)
   )
   Return Node(execute=[@n2.Middle, @n2.Top])
  ]}
 InitPointOfInterest {nodes [
   n1 = InitPointOfInterest!() -> (
    then
    PointOfInterest:Actor Object Reference
    DisplayText:Text
    DisplayImage:Texture 2D Object Reference
    ImageColor:Linear Color Structure=(R=1,G=0.6963,B=0,A=1)
   ) {is=title K2Node_FunctionEntry}
   > SetIndicatedActor!(execute, self:Fort Actor Indicator Widget Object Reference, NewIndicatorActor:Actor Object Reference=@n1) -> (then)
   > n2 = Branch(Condition="Not Equal (Object)"!(A:Object Reference=@n1, B:Object Reference) -> (ReturnValue:bool))
   > SetDisplayText!(execute, self, Description:Text=@n1.DisplayText) -> (then)
   > SetDisplayImage!(execute, self, Icon:Texture 2D Object Reference=@n1.DisplayImage) -> (then)
   > SetVisibility(InVisibility=Visible)
   > Set bShowClampToScreenArrow(bShowClampToScreenArrow:bool=true) -> (Output_Get:bool)
   > Set bClampOnScreen(bClampOnScreen:bool=true) -> (Output_Get:bool)
   > Set bUseScreenSpacePosition(bUseScreenSpacePosition=true)
   > SetColorAndOpacity(
    self:Image Object Reference=[Get center() -> (center:Image Object Reference), Get largespinring() -> (largespinring:Image Object Reference)]
    InColorAndOpacity:Linear Color Structure=Get ImageColor() -> (ImageColor:Linear Color Structure)
   )
   n3 = BreakColor(InColor=Get ImageColor())
   > SetColorAndOpacity(
    self=Get smallspinring() -> (smallspinring:Image Object Reference)
    InColorAndOpacity=Make Linear Color!(R:float=@n3, G:float=@n3.G, B:float=@n3.B, A:float=0.6) -> (LinearColor:Linear Color Structure) {is=title K2Node_MakeStruct}
   )
   n4 = BreakColor(InColor=Get ImageColor())
   > SetColorAndOpacity(self=Get backgroundpulse() -> (backgroundpulse:Image Object Reference), InColorAndOpacity=Make Linear Color(R=@n4, G=@n4.G, B=@n4.B, A=0.25))
   > PlayAnimation(InAnimation=Get NewAnimation_1() -> (NewAnimation_1:Widget Animation Object Reference), NumLoopsToPlay=0)
   > PlayAnimation(InAnimation=Get NewAnimation_2() -> (NewAnimation_2:Widget Animation Object Reference), NumLoopsToPlay=0)
   SetVisibility(execute=@n2.else, InVisibility=Collapsed)
   Set ScreenSpaceRelativeOffset(ScreenSpaceRelativeOffset:Vector 2D Structure=(X=0,Y=200)) -> (Output_Get:Vector 2D Structure)
  ]}
 RemovePointOfInterest {nodes [
   n1 = RemovePointOfInterest!() -> (then, PointOfInterest:Actor Object Reference) {is=title K2Node_FunctionEntry}
   > n2 = Branch(Condition="Equal (Object)"(A=@n1, B=GetIndicatedActor!(self:Fort Actor Indicator Interface Interface) -> (ReturnValue:Actor Object Reference)) {is=title K2Node_CallFunction})
   > SetIndicatedActor()
   > SetVisibility(InVisibility=Collapsed)
   > Set bShowClampToScreenArrow(bShowClampToScreenArrow=false)
   > Set bClampOnScreen(bClampOnScreen=false)
   > Set bUseScreenSpacePosition(bUseScreenSpacePosition=false)
   > Return Node!(execute, Removed:bool=true)
   Return Node(execute=@n2.else, Removed=false)
  ]}
 SetDisplayImage {nodes [
   n1 = SetDisplayImage!() -> (then, Icon:Texture 2D Object Reference) {is=title K2Node_FunctionEntry}
   n2 = Get center()
   > SetBrushFromTexture!(execute, self:Image Object Reference=@n2, Texture:Texture 2D Object Reference=@n1, bMatchSize:bool=true) -> (then)
   > SetVisibility(self=@n2, InVisibility=Select(Option 0=Visible, Option 1=Collapsed, Index="Equal (Object)"(A=@n1)))
   > SetBrushFromTexture(self=Get largespinring(), Texture=/Game/Missions/Primary/OB_FarmsteadFort/ArtAssets/Textures/NavigationMarkerThick, bMatchSize=true)
   > SetBrushFromTexture(self=Get smallspinring(), Texture=&NavigationMarkerThick, bMatchSize=true)
   > SetBrushFromTexture(self=Get backgroundpulse(), Texture=/Game/UI/Frontend/Store/Textures/T_RadGradient, bMatchSize=true)
  ]}
 SetDisplayText {nodes [
   n1 = SetDisplayText!() -> (then, Description:Text) {is=title K2Node_FunctionEntry}
   n2 = Get DecriptionText() -> (DecriptionText:Common Text Object Reference)
   > "SetText (Text)"(self=@n2, InText=@n1)
   > SetVisibility(self=@n2, InVisibility=Select(Option 0=Visible, Option 1=Collapsed, Index=TextIsEmpty(InText=@n1)))
  ]}
]
variables []
widgets {
 info {inheritedWidgetCount=1, parentClass=/Script/FortniteUI.FortActorIndicatorWidget, rootWidgetClass=&CanvasPanel, widgetCount=9}
 widgets [
  Panel {widgetClassPath=&CanvasPanel, widget=Panel}
  center {
   widgetClassPath=&Image
   bIsVariable=true
   parent=Panel
   widget=center
   properties {
    brush {
     bIsDynamicallyLoaded=false
     drawAs=Image
     imageType=NoImage
     mirroring=NoMirror
     resourceName=None
     resourceObject=/Game/Missions/Primary/Icons/T_PrimaryIcon
     tiling=NoTile
     imageSize=<512 512>
     margin=<ltrb 0 0 0 0>
     outlineSettings {
      bUseBrushTransparency=false
      roundingType=HalfHeightRadius
      width=0
      color {colorUseRule=UseColor_Specified, specifiedColor=<rgba 0 0 0 0>}
      cornerRadii=<0 0 0 0>
     }
     tintColor {colorUseRule=UseColor_Specified, specifiedColor=<rgba 1 1 1 1>}
     uVRegion {bIsValid=false, max=<0 0>, min=<0 0>}
    }
    colorAndOpacity=<rgba 1 0.6963 0 1>
   }
   slot {
    class=CanvasPanelSlot
    properties {zOrder=10, layoutData {alignment=<0 0>, anchors {maximum=<0.5 0>, minimum=<0.5 0>}, offsets=<ltrb -27 13.5 54 54>}}
   }
  }
  VerticalBox_0 {
   widgetClassPath=&VerticalBox
   parent=Panel
   widget=VerticalBox_0
   slot {class=CanvasPanelSlot, properties {layoutData {alignment=<0.5 0>, anchors {maximum=<0.5 0>, minimum=<0.5 0>}, offsets=<ltrb 0 80 120 30>}}}
  }
  DecriptionText {
   widgetClassPath=&CommonTextBlock
   bIsVariable=true
   parent=VerticalBox_0
   widget=DecriptionText
   properties {
    justification=Center
    style=/FortUILibrary/Styles/Typography/Foundation/TextStyle-Base-S-B-S.TextStyle-Base-S-B-S_C
    text=TEXT
    font {
     bForceMonospaced=false
     bMaterialIsStencil=false
     fontMaterial=null
     fontObject=&NotoSans
     letterSpacing=0
     monospacedWidth=1
     size=18
     skewAmount=0
     typefaceFontName=NotoSans-Bold
     outlineSettings {
      bApplyOutlineToDropShadows=true
      bMiteredCorners=false
      bSeparateFillAlpha=true
      outlineMaterial=&M-FontShadow
      outlineSize=3
      outlineColor=<rgba 0 0 0 0.5>
     }
    }
    shadowOffset=<0 0>
   }
   slot {class=VerticalBoxSlot}
  }
  BottomCanvasPanelContainer {
   widgetClassPath=&CanvasPanel
   bIsVariable=true
   parent=VerticalBox_0
   widget=BottomCanvasPanelContainer
   properties {visibility=Collapsed}
   slot {class=VerticalBoxSlot, properties {horizontalAlignment=HAlign_Center, verticalAlignment=VAlign_Top}}
  }
  largespinring {
   widgetClassPath=&Image
   bIsVariable=true
   parent=Panel
   widget=largespinring
   properties {
    brush {
     bIsDynamicallyLoaded=false
     drawAs=Image
     imageType=NoImage
     mirroring=NoMirror
     resourceName=None
     resourceObject=&NavigationMarkerThick
     tiling=NoTile
     imageSize=<512 512>
     margin=<ltrb 0 0 0 0>
     outlineSettings {
      bUseBrushTransparency=false
      roundingType=HalfHeightRadius
      width=0
      color {colorUseRule=UseColor_Specified, specifiedColor=<rgba 0 0 0 0>}
      cornerRadii=<0 0 0 0>
     }
     tintColor {colorUseRule=UseColor_Specified, specifiedColor=<rgba 1 1 1 1>}
     uVRegion {bIsValid=false, max=<0 0>, min=<0 0>}
    }
    colorAndOpacity=<rgba 1 0.7011 0 1>
   }
   slot {
    class=CanvasPanelSlot
    properties {zOrder=1, layoutData {alignment=<0 0>, anchors {maximum=<0.5 0>, minimum=<0.5 0>}, offsets=<ltrb -40 0.75 80 80>}}
   }
  }
  smallspinring {
   widgetClassPath=&Image
   bIsVariable=true
   parent=Panel
   widget=smallspinring
   properties {
    brush {
     bIsDynamicallyLoaded=false
     drawAs=Image
     imageType=NoImage
     mirroring=NoMirror
     resourceName=None
     resourceObject=&NavigationMarkerThick
     tiling=NoTile
     imageSize=<512 512>
     margin=<ltrb 0 0 0 0>
     outlineSettings {
      bUseBrushTransparency=false
      roundingType=HalfHeightRadius
      width=0
      color {colorUseRule=UseColor_Specified, specifiedColor=<rgba 0 0 0 0>}
      cornerRadii=<0 0 0 0>
     }
     tintColor {colorUseRule=UseColor_Specified, specifiedColor=<rgba 1 1 1 1>}
     uVRegion {bIsValid=false, max=<0 0>, min=<0 0>}
    }
    colorAndOpacity=<rgba 1 0.6963 0 0.6>
    renderTransform {angle=90, scale=<1 1>, shear=<0 0>, translation=<0 0>}
   }
   slot {class=CanvasPanelSlot, properties {layoutData {alignment=<0 0>, anchors {maximum=<0.5 0>, minimum=<0.5 0>}, offsets=<ltrb -33 7.5 66 66>}}}
  }
  backgroundpulse {
   widgetClassPath=&Image
   bIsVariable=true
   parent=Panel
   widget=backgroundpulse
   properties {
    brush {
     bIsDynamicallyLoaded=false
     drawAs=Image
     imageType=NoImage
     mirroring=NoMirror
     resourceName=None
     resourceObject=&T_RadGradient
     tiling=NoTile
     imageSize=<256 256>
     margin=<ltrb 0 0 0 0>
     outlineSettings {
      bUseBrushTransparency=false
      roundingType=HalfHeightRadius
      width=0
      color {colorUseRule=UseColor_Specified, specifiedColor=<rgba 0 0 0 0>}
      cornerRadii=<0 0 0 0>
     }
     tintColor {colorUseRule=UseColor_Specified, specifiedColor=<rgba 1 1 1 1>}
     uVRegion {bIsValid=false, max=<0 0>, min=<0 0>}
    }
    colorAndOpacity=<rgba 1 0.6963 0 0.25>
    renderTransform {angle=0, scale=<1.4 1.4>, shear=<0 0>, translation=<0 0>}
   }
   slot {
    class=CanvasPanelSlot
    properties {zOrder=1025, layoutData {alignment=<0 0>, anchors {maximum=<0.5 0>, minimum=<0.5 0>}, offsets=<ltrb -40 0 80 80>}}
   }
  }
  TutorialHighlight {widgetClassPath=&TutorialHighlight}
 ]
}
