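# HUD indicator widget that shows a vehicle's health bar, with an extension slot for dynamic UI.
# Vehicle, health and force-hide events refresh health and visibility, hiding it when invalid, destroyed, force-hidden or already shown on the occupant's HUD.
# A change-gated visibility function runs a looping timer only while visible.
# That timer scales the widget by squared player distance, showing numbers only at full scale.
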WidgetBlueprint /Game/UI/InGame/HUD/VehicleInfoIndicator
graphs [
 EventGraph {nodes [
   n1 = OnPlayerVehicleStateChanged_Event!() -> (OutputDelegate, then, PlayerPawn:Fort Player Pawn Object Reference, NewVehicle:Actor Object Reference, OldVehicle:Actor Object Reference) {is=title K2Node_CustomEvent}
   > n2 = Cast To FortVehicleInterface!(execute, Object=@n1.NewVehicle) -> (then, CastFailed, AsFort Vehicle Interface:Fort Vehicle Interface Interface) {is=title K2Node_DynamicCast}
   > n3 = Branch(Condition=ShowVehicleHealthBarOnPlayerHUD!(self:Fort Vehicle Interface Interface=@n2) -> (ReturnValue:bool))
   SetIndicatorVisibility!(execute=[@n3.then, @n4.then], self, IsVisible:bool=false) -> (then)
   n4 = Branch(execute=@n3.else, Condition=GetForceHidden!(self:Fort Vehicle Info Indicator Widget Object Reference) -> (ReturnValue:bool))
   UpdateHealth() {is=title K2Node_CustomEvent}
   > n5 = Cast To FortVehicleInterface(Object=Get CurrentVehicle() -> (CurrentVehicle:Actor Weak object))
   n6 = GetHealth!(self:Fort Vehicle Interface Interface=@n5) -> (ReturnValue:float)
   > Update Current HP Value!(
    execute
    self:Common Health Bar Object Reference=Get CommonHealthBar() -> (CommonHealthBar:Common Health Bar Object Reference)
    Current:double=Select(
     Option 0:double=@n6
     Option 1:double=0
     Index="float < float"!(A:double=@n6, B:double=1) -> (ReturnValue:bool) {is=title K2Node_CallFunction}
    ) -> (ReturnValue:double)
   ) -> (then)
   Event OnInitialized()
   > Assign On Player Vehicle State Changed!(
    execute:Exec
    self:Fort HUDContext Object Reference=Get FortHUDContext!() -> (ReturnValue:Fort HUDContext Object Reference) {is=title K2Node_GetBlueprintContextEx}
    Delegate:"Delegate (by ref)"=Create Event()
   ) -> (then:Exec) {is=title K2Node_AssignDelegate}
   > Set FullScaleDistanceSqr(FullScaleDistanceSqr:double=Square!(A:double=Get FullScaleDistance() -> (FullScaleDistance:double)) -> (ReturnValue:double)) -> (Output_Get:double)
   > Set MaxDistanceSqr(MaxDistanceSqr:double=Square(A=Get MaxDistance() -> (MaxDistance:float))) -> (Output_Get:double)
   Event OnForceHideChanged!() -> (OutputDelegate, then, bForceHide:bool) {is=title K2Node_Event}
   > UpdateHealth() {is=title K2Node_CallFunction}
   Event OnVehicleChanged() {is=title K2Node_Event}
   > UpdateHealth()
   > n7 = Branch(Condition=IsValid(Object=Get CurrentVehicle()))
   > n8 = Cast To FortVehicleInterface(Object=Get CurrentVehicle())
   > n9 = Set MaxHealth(MaxHealth:double=GetMaxHealth!(self:Fort Vehicle Interface Interface=@n8) -> (ReturnValue:float)) -> (Output_Get:double)
   > n10 = Branch(Condition="float > float"(A=@n9, B=0) {is=title K2Node_CallFunction})
   n11 = GetHealth(self=@n8)
   > n12 = Branch(Condition=NOT Boolean(A="float < float"(A=@n11, B=1)))
   > n13 = Cast To FortVehicleInterface(Object=Get CurrentVehicle())
   > n14 = Branch(Condition=ShowVehicleHealthBarOnPlayerHUD(self=@n13))
   n15 = GetControlledPawn!(self:Fort Vehicle Info Indicator Widget Object Reference) -> (ReturnValue:Pawn Object Reference)
   > n16 = Cast To FortVehicleInterface(Object=@n15)
   > n17 = GetVehicleActor!(execute, self:Fort Vehicle Interface Interface=@n16) -> (then, ReturnValue:Actor Object Reference)
   > n18 = Branch(Condition="Not Equal (Object)"!(A:Object Reference=@n17, B:Object Reference=Get CurrentVehicle()) -> (ReturnValue:bool))
   n19 = Branch(execute=[@n21.then, @n18.then, @n14.else], Condition=NOT Boolean(A=GetForceHidden()))
   > SetIndicatorVisibility(IsVisible=true)
   > Update Max HP Value!(execute, self:Common Health Bar Object Reference=Get CommonHealthBar(), Max:double=Get MaxHealth() -> (MaxHealth:double)) -> (then)
   > UpdateWidgetScale()
   SetIndicatorVisibility(execute=[@n19.else, @n21.else, @n20.CastFailed, @n12.else, @n10.else, @n8.CastFailed, @n7.else], IsVisible=false)
   n20 = Cast To FortPlayerPawn(execute=[@n18.else, @n16.CastFailed], Object=@n15)
   > n21 = Branch(Condition="Not Equal (Object)"(A=GetVehicleActor!(self:Fort Player Pawn Object Reference=@n20) -> (ReturnValue), B=Get CurrentVehicle()))
   Event Destruct()
   > Parent: Destruct()
   > SetIndicatorVisibility(IsVisible=false)
   Event OnVehicleHealthChanged() {is=title K2Node_Event}
   > Parent: OnVehicleHealthChanged() {is=title K2Node_CallParentFunction}
   > UpdateHealth()
  ]}
 RemoveContentFromWidgetHost {nodes [
   RemoveContentFromWidgetHost!() -> (
    then
    WidgetHostTag:"Gameplay Tag Structure (by ref)"
    ContentWidget:Widget Object Reference
    RemovalType:EWidgetSlottingRemovalType Enum=Unparent
   ) {is=title K2Node_FunctionEntry}
   > Return Node!(execute, ReturnValue:bool=false)
  ]}
 SetIndicatorVisibility {nodes [
   n1 = SetIndicatorVisibility!() -> (then, IsVisible:bool=~) {is=title K2Node_FunctionEntry}
   > n2 = Delta Gate Wildcard!(execute:Exec, Variable:"Boolean (by ref)"=Get IsHealthBarVisible() -> (IsHealthBarVisible:bool), New Value:bool=@n1) -> (Success:Exec, Updated Variable:"Boolean (by ref)", Previous Variable:bool=~) {is=title K2Node_MacroInstance}
   > SetVisibility(
    self=[Get ExtensionSlot() -> (ExtensionSlot:Widget Slotting Host Overlay Object Reference), Get CommonHealthBar()]
    InVisibility=Select(Option 0:ESlateVisibility Enum=Hidden, Option 1:ESlateVisibility Enum=HitTestInvisible, Index=@n2) -> (ReturnValue:ESlateVisibility Enum)
   )
   > n3 = Branch(Condition=Get IsHealthBarVisible())
   > n4 = Set Timer by Event(Delegate=Create Event(), Time=Get DistanceUpdateInterval() -> (DistanceUpdateInterval:float), bLooping=true, bMaxOncePerFrame=true)
   > Set UpdateWidgetScaleTimerHandle(UpdateWidgetScaleTimerHandle:Timer Handle Structure=@n4) -> (Output_Get:Timer Handle Structure)
   Clear and Invalidate Timer by Handle(execute=@n3.else, Handle=Get UpdateWidgetScaleTimerHandle() -> (UpdateWidgetScaleTimerHandle:Timer Handle Structure))
  ]}
 UpdateWidgetScale {nodes [
   UpdateWidgetScale() {is=title K2Node_FunctionEntry}
   > n1 = Set CalculatedScale(CalculatedScale:double=MapRangeClamped(
     Value=Get CachedDistanceToPlayerSq() -> (CachedDistanceToPlayerSq:float)
     InRangeA=Get FullScaleDistanceSqr() -> (FullScaleDistanceSqr:double)
     InRangeB=Get MaxDistanceSqr() -> (MaxDistanceSqr:double)
     OutRangeA=1
     OutRangeB=Get MinimumScaleBeforeModifier() -> (MinimumScaleBeforeModifier:float)
    )) -> (Output_Get:double)
   > n2 = Branch(Condition="Nearly Equal (Float)"(A=@n1, B=Get SetScale() -> (SetScale:double), ErrorTolerance=0.001))
   n3 = Set SetScale(execute=@n2.else, SetScale:double=Get CalculatedScale() -> (CalculatedScale:double)) -> (Output_Get:double)
   n4 = MakeVector2D(X=@n3, Y=@n3)
   n5 = Get ScaleModifier() -> (ScaleModifier:double)
   n6 = "float > float"(A=@n3, B=0.999) {is=title K2Node_PromotableOperator}
   > SetRenderScale!(
    execute
    self:Widget Object Reference
    Scale:Vector 2D Structure=Select(
     Option 0:Vector 2D Structure=vector2d * vector2d(A:Vector 2D Structure=@n4, B=@n5) -> (ReturnValue:Vector 2D Structure) {is=title K2Node_PromotableOperator}
     Option 1:Vector 2D Structure=@n4
     Index=@n6
    ) -> (ReturnValue:Vector 2D Structure)
   ) -> (then)
   > SetShowNumbers!(execute, self:Common Health Bar Object Reference=Get CommonHealthBar(), DoShow:bool=@n6) -> (then)
  ]}
]
variables [
 DistanceUpdateInterval {type=number, default=0.1}
 FullScaleDistance {type=number, default=2000}
 FullScaleDistanceSqr {type=number}
 IsHealthBarVisible {type=boolean}
 MaxDistanceSqr {type=number}
 MaxHealth {type=number}
 MinimumScaleBeforeModifier {type=number, default=0.5}
 ScaleModifier {type=number, default=0.5}
 SetScale {type=number}
 UpdateWidgetScaleTimerHandle {type=TimerHandle}
]
widgets {
 info {inheritedWidgetCount=1, parentClass=/Script/FortniteUI.FortVehicleInfoIndicatorWidget, rootWidgetClass=&Overlay, widgetCount=4}
 widgets [
  Overlay_0 {widgetClassPath=&Overlay, widget=Overlay_0}
  ExtensionSlot {
   widgetClassPath=/Script/WidgetSlotting.WidgetSlottingHostOverlay
   bIsVariable=true
   parent=Overlay_0
   widget=ExtensionSlot
   properties {hostTag {tagName=DynamicUI.HUD.Fort.Athena.NextToVehicleHealthInfo.Slot}}
   slot {class=OverlaySlot, properties {verticalAlignment=VAlign_Center}}
  }
  CommonHealthBar {
   widgetClassPath=&/CommonHealthBar
   bIsVariable=true
   parent=Overlay_0
   widget=CommonHealthBar
   properties {visibility=Collapsed}
   slot {class=OverlaySlot}
  }
  TutorialHighlight {widgetClassPath=&TutorialHighlight}
 ]
}
