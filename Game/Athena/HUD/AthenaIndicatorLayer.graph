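# Athena HUD layer that fills an actor canvas with in-world indicators for points of interest, vehicles, interactions and, for non-spectators, building info.
# Point-of-interest events from the HUD context drive indicators recycled between available and active arrays.
# Creative playlists add async-loaded info indicators once playlist data is ready.
# An extension slot can replace the interaction indicator, and the layer registers as widget host for auto-aim reticle content.
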WidgetBlueprint /Game/Athena/HUD/AthenaIndicatorLayer
graphs [
 AddInteractionInfoWidget {nodes [
   n1 = AddInteractionInfoWidget!() -> (then, IndicatorWidget:Fort Actor Indicator Widget Object Reference) {is=title K2Node_FunctionEntry}
   > n2 = Get!(execute:Exec) -> (then:Exec, else:Exec, InteractionIndicatorWidget:Fort Actor Indicator Widget Object Reference)
   > n3 = RemoveFromParent(self=@n2)
   n4 = Set InteractionIndicatorWidget(execute=[@n3.then, @n2.else], InteractionIndicatorWidget:Fort Actor Indicator Widget Object Reference=@n1) -> (Output_Get:Fort Actor Indicator Widget Object Reference)
   > n5 = AddActorIndicator!(
    execute
    self:Fort Actor Canvas Object Reference=Get GeneralIndicators() -> (GeneralIndicators:Fort Actor Canvas Object Reference)
    Indicator:Fort Actor Indicator Widget Object Reference=@n4
   ) -> (then, ReturnValue:Fort Actor Canvas Slot Object Reference)
   > SetCanAutoRemove!(execute, self:Fort Actor Canvas Slot Object Reference=@n5, bAllowAutoRemove:bool=false) -> (then)
   > SetVerticalAlignment(self:Fort Actor Canvas Slot Object Reference=@n5, InVerticalAlignment=VAlign_Top)
   > SetHorizontalAlignment(self:Fort Actor Canvas Slot Object Reference=@n5, InHorizontalAlignment=HAlign_Left)
  ]}
 BP_ExtensionWidgetAdded {nodes [
   n1 = Handle Extension Widget Added!() -> (then, Widget:User Widget Object Reference, SlotType:EUIExtensionSlot Enum=Invalid) {is=title K2Node_FunctionEntry}
   > n2 = Cast To FortActorIndicatorWidget!(execute, Object=@n1) -> (then, CastFailed, AsFort Actor Indicator Widget:Fort Actor Indicator Widget Object Reference) {is=title K2Node_DynamicCast}
   > Branch(Condition="Equal (Enum)"(A:EUIExtensionSlot Enum=@n1.SlotType, B:EUIExtensionSlot Enum=AthenaInteractionIndicator))
   > AddInteractionInfoWidget!(execute, self, IndicatorWidget:Fort Actor Indicator Widget Object Reference=@n2) -> (then) {is=title K2Node_CallFunction}
  ]}
 CreateInterestIndicatorWidget {nodes [
   CreateInterestIndicatorWidget() {is=title K2Node_FunctionEntry}
   > n1 = Create Widget!(
    execute:Exec
    Class:User Widget Class Reference=/Game/UI/InGame/HUD/InterestIndicatorWidget.InterestIndicatorWidget_C
    OwningPlayer:Player Controller Object Reference=GetOwningPlayer()
   ) -> (then:Exec, ReturnValue:Interest Indicator Widget Object Reference)
   > Add Unique(
    TargetArray:Array of Interest Indicator Widget Object References=Get AvailableInterestIndicators() -> (AvailableInterestIndicators:Array of Interest Indicator Widget Object References)
    NewItem:"Interest Indicator Widget Object Reference (by ref)"=@n1
   )
   > n2 = AddActorIndicator(self=Get GeneralIndicators(), Indicator=@n1)
   > SetCanAutoRemove(self=@n2, bAllowAutoRemove=false)
  ]}
 EventGraph {nodes [
   n1 = Event Add Content To Host Widget!() -> (
    OutputDelegate
    then
    WidgetHostTag:"Gameplay Tag Structure (by ref)"
    ContentWidget:Widget Object Reference
    Padding:"Margin Structure (by ref)"
    bFill:bool
   ) {is=title K2Node_Event}
   > Branch(Condition=MatchesTag!(
     TagOne:Gameplay Tag Structure=@n1.WidgetHostTag
     TagTwo:Gameplay Tag Structure=Get AutoAimReticleExtensionSlot() -> (AutoAimReticleExtensionSlot:Gameplay Tag Structure)
     bExactMatch:bool=true
    ) -> (ReturnValue:bool))
   > n2 = Cast To FortActorIndicatorWidget(Object=@n1.ContentWidget)
   > n3 = Set InteractionIndicatorWidget(InteractionIndicatorWidget=@n2)
   > n4 = AddActorIndicator(self=Get GeneralIndicators(), Indicator=@n3)
   > SetCanAutoRemove(self=@n4, bAllowAutoRemove=false)
   > SetVerticalAlignment(self=@n4, InVerticalAlignment=VAlign_Center)
   > SetHorizontalAlignment(self=@n4, InHorizontalAlignment=HAlign_Center)
   Event Construct()
   > RegisterWidgetHost!(
    execute
    WidgetHost:Widget Object Reference
    WidgetHostTag:"Gameplay Tag Structure (by ref)"=Get AutoAimReticleExtensionSlot()
    IsPlayerSpecific:bool=true
   ) -> (then)
   n5 = GetOwningPlayer()
   > n6 = Cast To FortPlayerControllerSpectating!(execute, Object=@n5) -> (then, CastFailed, AsFort Player Controller Spectating:Fort Player Controller Spectating Object Reference) {is=title K2Node_DynamicCast}
   n7 = Create Event()
   n8 = Get FortHUDContext!() -> (ReturnValue:Fort HUDContext Object Reference) {is=title K2Node_GetBlueprintContextEx}
   Bind Event to On Point Of Interest Added(execute=[@n6.then, @n17.then], self:Fort HUDContext Object Reference=@n8, Delegate=@n7) {is=title K2Node_AddDelegate}
   n9 = Create Event()
   > Bind Event to On Point Of Interest Removed(self:Fort HUDContext Object Reference=@n8, Delegate=@n9) {is=title K2Node_AddDelegate}
   > CreateInterestIndicatorWidget() {is=title K2Node_CallFunction}
   > n10 = Create Widget(Class=/Game/UI/InGame/HUD/VehicleInfoIndicator.VehicleInfoIndicator_C, OwningPlayer=GetOwningPlayer()) -> (ReturnValue:Vehicle Info Indicator Object Reference)
   > n11 = AddActorIndicator(self=Get GeneralIndicators(), Indicator=@n10)
   > SetCanAutoRemove(self=@n11, bAllowAutoRemove=false)
   > Setup Creative Indicators!(execute:Exec) -> (then:Exec) {is=title K2Node_MacroInstance}
   > Branch(Condition=Get ShowInteractionWidget() -> (ShowInteractionWidget:bool))
   > n12 = Create Widget(Class=&/AthenaInteractionIndicator.AthenaInteractionIndicator_C, OwningPlayer=GetOwningPlayer()) -> (ReturnValue:Athena Interaction Indicator Object Reference)
   > AddInteractionInfoWidget(IndicatorWidget=@n12)
   > n13 = Create Widget(Class=&/VehicleChangeSeatIndicator.VehicleChangeSeatIndicator_C, OwningPlayer=GetOwningPlayer()) -> (ReturnValue:Vehicle Change Seat Indicator Object Reference)
   > n14 = AddActorIndicator(self=Get GeneralIndicators(), Indicator=@n13)
   > SetCanAutoRemove(self=@n14, bAllowAutoRemove=false)
   > SetVerticalAlignment(self=@n14, InVerticalAlignment=VAlign_Top)
   > SetHorizontalAlignment(self=@n14, InHorizontalAlignment=HAlign_Left)
   n15 = Create Widget(execute=@n6.CastFailed, Class=/Game/UI/InGame/HUD/BuildingInfoIndicator_Athena.BuildingInfoIndicator_Athena_C, OwningPlayer=@n5) -> (ReturnValue:Building Info Indicator Athena Object Reference)
   > n16 = AddActorIndicator(self=Get GeneralIndicators(), Indicator=@n15)
   > n17 = SetCanAutoRemove(self=@n16, bAllowAutoRemove=false)
   Event Destruct()
   > UnregisterWidgetHost!(execute, WidgetHost:Widget Object Reference, WidgetHostTag:"Gameplay Tag Structure (by ref)"=Get AutoAimReticleExtensionSlot()) -> (then)
   > Is Valid(InputObject=@n8)
   > Unbind Event from On Point Of Interest Added(self:Fort HUDContext Object Reference=@n8, Delegate=@n7) {is=title K2Node_RemoveDelegate}
   > Unbind Event from On Point Of Interest Removed(self:Fort HUDContext Object Reference=@n8, Delegate=@n9) {is=title K2Node_RemoveDelegate}
  ]}
 HandleOnPointOfInterestAdded {nodes [
   n1 = HandleOnPointOfInterestAdded!() -> (
    then
    PointOfInterest:Actor Object Reference
    DisplayText:Text
    DisplayImage:Texture 2D Object Reference
    SpecifyColor:bool=~
    ImageColor:Linear Color Structure
   ) {is=title K2Node_FunctionEntry}
   n2 = Get AvailableInterestIndicators()
   > n3 = Branch(Condition="integer > integer"(A=Length(TargetArray:Array of Interest Indicator Widget Object References=@n2), B=0) {is=title K2Node_CallFunction})
   > n4 = CreateInterestIndicatorWidget()
   n5 = "Get (a copy)"(Array:Array of Interest Indicator Widget Object References=Get AvailableInterestIndicators()) -> (Output:Interest Indicator Widget Object Reference)
   InitPointOfInterest!(
    execute=[@n4.then, @n3.else]
    self:Interest Indicator Widget Object Reference=@n5
    PointOfInterest:Actor Object Reference=@n1
    DisplayText:Text=@n1.DisplayText
    DisplayImage:Texture 2D Object Reference=@n1.DisplayImage
    ImageColor:Linear Color Structure=Select(Option 0:Linear Color Structure=(R=1,G=0.6963,B=0,A=1), Option 1:Linear Color Structure=@n1.ImageColor, Index=@n1.SpecifyColor) -> (ReturnValue:Linear Color Structure)
   ) -> (then)
   > Add Unique(TargetArray=Get ActiveInterestIndicators() -> (ActiveInterestIndicators:Array of Interest Indicator Widget Object References), NewItem=@n5)
   > Remove Index!(execute, TargetArray:Array of Interest Indicator Widget Object References=Get AvailableInterestIndicators(), IndexToRemove:int=0) -> (then)
  ]}
 HandleOnPointOfInterestRemoved {nodes [
   n1 = HandleOnPointOfInterestRemoved!() -> (then, PointOfInterest:Actor Object Reference) {is=title K2Node_FunctionEntry}
   > n2 = For Each Loop with Break!(Exec:Exec, Array:Array of Interest Indicator Widget Object References=Get ActiveInterestIndicators(), Break:Exec=@n7.then) -> (LoopBody:Exec, Array Element:Interest Indicator Widget Object Reference, Array Index:int, Completed:Exec)
   Branch(execute=@n2.Completed, Condition=Get Removed() -> (Removed:bool))
   n3 = Get CurrentPointOfInterestWidget() -> (CurrentPointOfInterestWidget:Interest Indicator Widget Object Reference)
   > Remove Item!(
    execute
    TargetArray:Array of Interest Indicator Widget Object References=Get ActiveInterestIndicators()
    Item:"Interest Indicator Widget Object Reference (by ref)"=@n3
   ) -> (then, ReturnValue:bool)
   > Add Unique(TargetArray=Get AvailableInterestIndicators(), NewItem=@n3)
   n4 = Set CurrentPointOfInterestWidget(execute=@n2.LoopBody, CurrentPointOfInterestWidget:Interest Indicator Widget Object Reference=@n2) -> (Output_Get:Interest Indicator Widget Object Reference)
   > n5 = RemovePointOfInterest!(execute, self:Interest Indicator Widget Object Reference=@n4, PointOfInterest:Actor Object Reference=@n1) -> (then, Removed:bool)
   > n6 = Set Removed(Removed:bool=@n5) -> (Output_Get:bool)
   > n7 = Branch(Condition=@n6)
  ]}
 RemoveContentFromWidgetHost {nodes [
   n1 = Remove Content From Host Widget!() -> (
    then
    WidgetHostTag:"Gameplay Tag Structure (by ref)"
    ContentWidget:Widget Object Reference
    RemovalType:EWidgetSlottingRemovalType Enum=Unparent
   ) {is=title K2Node_FunctionEntry}
   > Branch(Condition=MatchesTag(TagOne=@n1, TagTwo=Get AutoAimReticleExtensionSlot(), bExactMatch=true))
   > n2 = Switch on EWidgetSlottingRemovalType!(execute, Selection:EWidgetSlottingRemovalType Enum=@n1.RemovalType) -> (Unparent:Exec, Collapse:Exec) {is=title K2Node_SwitchEnum}
   SetVisibility(execute=@n2.Collapse, self=@n1.ContentWidget, InVisibility=Collapsed)
   n3 = RemoveChild(execute=@n2.Unparent, self=Get GeneralIndicators(), Content=@n1.ContentWidget)
   > Return Node!(execute, ReturnValue:bool=@n3)
   Return Node(ReturnValue=false)
  ]}
 SetupCreativeIndicators {nodes [
   Inputs()
   > n1 = PlaylistDataReadyAsync!(execute:Exec) -> (
    then:Exec
    OnReady:Exec
    GameState:Fort Game State Athena Object Reference
    Playlist:Fort Playlist Object Reference
    PlaylistContextTags:"Gameplay Tag Container Structure (by ref)"
   ) {is=title K2Node_AsyncAction}
   > Outputs!(then:Exec)
   Branch(
    execute=@n1.OnReady
    Condition=OR Boolean(
     A=IsCreativeModePlaylist!(self:Fort Game State Athena Object Reference=@n1) -> (ReturnValue:bool)
     B=IsCreativeLTMPlaylist!(self:Fort Game State Athena Object Reference=@n1) -> (ReturnValue:bool)
    )
   )
   > n2 = Sequence()
   > n3 = Async Load Class Asset!(
    execute:Exec
    AssetClass:Object Soft Class Reference=Get SoftCreativeInfoIndicator() -> (SoftCreativeInfoIndicator:Fort Actor Indicator Widget Soft Class Reference)
   ) -> (then:Exec, Completed:Exec, Class:Object Class Reference) {is=title K2Node_LoadAssetClass}
   n4 = Cast To FortActorIndicatorWidget Class!(execute:Exec=@n3.Completed, Class:Object Class Reference=@n3) -> (then:Exec, CastFailed:Exec, AsFort Actor Indicator Widget:Fort Actor Indicator Widget Class Reference) {is=title K2Node_ClassDynamicCast}
   > n5 = Create Widget(Class=@n4, OwningPlayer=GetOwningPlayer()) -> (ReturnValue:Fort Actor Indicator Widget Object Reference)
   > n6 = AddActorIndicator(self=Get GeneralIndicators(), Indicator=@n5)
   > SetCanAutoRemove(self=@n6, bAllowAutoRemove=false)
   n7 = Async Load Class Asset(
    execute=@n2.then_1
    AssetClass=Get SoftCreativePlayerInfoIndicator() -> (SoftCreativePlayerInfoIndicator:Creative Player Info Indicator Soft Class Reference)
   )
   n8 = Cast To CreativePlayerInfoIndicator Class!(execute:Exec=@n7.Completed, Class:Object Class Reference=@n7) -> (then:Exec, CastFailed:Exec, AsCreative Player Info Indicator:Creative Player Info Indicator Class Reference) {is=title K2Node_ClassDynamicCast}
   > n9 = Create Widget(Class=@n8, OwningPlayer=GetOwningPlayer()) -> (ReturnValue:Creative Player Info Indicator Object Reference)
   > n10 = AddActorIndicator(self=Get GeneralIndicators(), Indicator=@n9)
   > SetCanAutoRemove(self=@n10, bAllowAutoRemove=false)
  ]}
]
variables [
 ActiveInterestIndicators {type="array<InterestIndicatorWidget_C>", default []}
 AutoAimReticleExtensionSlot {type=GameplayTag, default {tagName=DynamicUI.HUD.Fort.Athena.AutoAim.Reticle.Slot}}
 AvailableInterestIndicators {type="array<InterestIndicatorWidget_C>", default []}
 InteractionIndicatorWidget {type=FortActorIndicatorWidget, default=null}
 ShowInteractionWidget {type=boolean, default=true}
 SoftCreativeInfoIndicator {type=Class, default=/Game/UI/InGame/HUD/CreativeInfoIndicator.CreativeInfoIndicator_C}
 SoftCreativePlayerInfoIndicator {type=Class, default=/Game/UI/Creative/CreativePlayerInfoIndicator.CreativePlayerInfoIndicator_C}
]
widgets {
 info {inheritedWidgetCount=3, parentClass=/Script/FortniteUI.AthenaIndicatorLayerBase, rootWidgetClass=&GridPanel, widgetCount=4}
 widgets [
  Layers {widgetClassPath=&GridPanel, widget=Layers, properties {columnFill [1], rowFill [1]}}
  GeneralIndicators {
   widgetClassPath=/Script/FortniteUI.FortActorCanvas
   bInherited=true
   bIsVariable=true
   parent=Layers
   widget=GeneralIndicators
   properties {
    bDrawElementsInOrder=true
    visibility=SelfHitTestInvisible
    arrowIconBrush {
     bIsDynamicallyLoaded=false
     drawAs=Image
     imageType=FullColor
     mirroring=NoMirror
     resourceName=None
     resourceObject=/Game/UI/CodeReferenced/T_UI_DirectionArrow
     tiling=NoTile
     imageSize=<32 32>
     margin=<ltrb 0 0 0 0>
     outlineSettings {
      bUseBrushTransparency=false
      roundingType=HalfHeightRadius
      width=0
      color {colorUseRule=UseColor_Specified, specifiedColor=<rgba 0 0 0 0>}
      cornerRadii=<0 0 0 0>
     }
     tintColor {colorUseRule=UseColor_Specified, specifiedColor=<rgba 1 1 1 1>}
     uVRegion {bIsValid=false, max=<0 0>, min=<0 0>}
    }
    defaultHUDElementTags {gameplayTags [{tagName=HUD.InWorld.ActorCanvas}]}
    elevationArrowIconBrush {
     bIsDynamicallyLoaded=false
     drawAs=Image
     imageType=FullColor
     mirroring=NoMirror
     resourceName=None
     resourceObject=/Game/UI/CodeReferenced/T_UI_ElevationArrow
     tiling=NoTile
     imageSize=<32 32>
     margin=<ltrb 0 0 0 0>
     outlineSettings {
      bUseBrushTransparency=false
      roundingType=HalfHeightRadius
      width=0
      color {colorUseRule=UseColor_Specified, specifiedColor=<rgba 0 0 0 0>}
      cornerRadii=<0 0 0 0>
     }
     tintColor {colorUseRule=UseColor_Specified, specifiedColor=<rgba 1 1 1 1>}
     uVRegion {bIsValid=false, max=<0 0>, min=<0 0>}
    }
   }
   slot {class=GridSlot, properties {columnSpan=1, rowSpan=1}}
  }
  NameIndicators {
   widgetClassPath=&FortActorCanvas
   bInherited=true
   bIsVariable=true
   parent=Layers
   widget=NameIndicators
   properties {
    visibility=SelfHitTestInvisible
    arrowIconBrush {
     bIsDynamicallyLoaded=false
     drawAs=Image
     imageType=FullColor
     mirroring=NoMirror
     resourceName=None
     resourceObject=&T_UI_DirectionArrow
     tiling=NoTile
     imageSize=<32 32>
     margin=<ltrb 0 0 0 0>
     outlineSettings {
      bUseBrushTransparency=false
      roundingType=HalfHeightRadius
      width=0
      color {colorUseRule=UseColor_Specified, specifiedColor=<rgba 0 0 0 0>}
      cornerRadii=<0 0 0 0>
     }
     tintColor {colorUseRule=UseColor_Specified, specifiedColor=<rgba 1 1 1 1>}
     uVRegion {bIsValid=false, max=<0 0>, min=<0 0>}
    }
    defaultHUDElementTags {gameplayTags [{tagName=HUD.InWorld.TeamMemberActorCanvas}]}
    elevationArrowIconBrush {
     bIsDynamicallyLoaded=false
     drawAs=Image
     imageType=FullColor
     mirroring=NoMirror
     resourceName=None
     resourceObject=&T_UI_ElevationArrow
     tiling=NoTile
     imageSize=<32 32>
     margin=<ltrb 0 0 0 0>
     outlineSettings {
      bUseBrushTransparency=false
      roundingType=HalfHeightRadius
      width=0
      color {colorUseRule=UseColor_Specified, specifiedColor=<rgba 0 0 0 0>}
      cornerRadii=<0 0 0 0>
     }
     tintColor {colorUseRule=UseColor_Specified, specifiedColor=<rgba 1 1 1 1>}
     uVRegion {bIsValid=false, max=<0 0>, min=<0 0>}
    }
   }
   slot {class=GridSlot, properties {columnSpan=1, layer=1, rowSpan=1}}
  }
  TutorialHighlight {widgetClassPath=&TutorialHighlight}
 ]
}
