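# Translucent UI material that shows a Target texture, defaulting to a map legend reboot image, within a circular mask with an optional outer ring.
# Anti-aliased SDF circles build the base and inset masks, and an outlined larger circle forms the gapped ring.
# ShowOuterRing lerps both color and alpha toward the ringed look, where OuterRingColor tints the ring and OuterOutlineAlpha scales its opacity.
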Material /Game/Athena/HUD/Materials/M_UI_CircularMask
graph {
 properties {blendMode=BLEND_Translucent, materialDomain=MD_UI, pixelDepthOffsetMode=PDOM_Legacy, usedShadingModels=Default Lit}
 nodes [
  n1 = VectorParameter() {parameterName=OuterRingColor}
  n2 = NamedRerouteUsage()
  LinearInterpolate(
   A=NamedRerouteUsage()
   B=LinearInterpolate(A=NamedRerouteUsage(), B=Multiply(A=@n1.RGB, B=@n2), Alpha=@n2)
   Alpha=ScalarParameter() {parameterName=ShowOuterRing}
  ) {constAlpha=0}
  n3 = Constant() {r=1}
  n4 = NamedRerouteDeclaration(None=MaterialFunctionCall(DistanceField=MaterialFunctionCall!(UVs, Radius=@n3) {materialFunction=/FortUILibrary/MaterialFunctions/Shapes/SDFs/MF_SDF_Circle}.SDF) {materialFunction=&MF_UI_AntiAliasing}.Inner) {name=TextureAlpha, nodeColor=<rgba 0 0 0 1>}
  n5 = ScalarParameter() {defaultValue=0.85, parameterName=Radius}
  n6 = ScalarParameter() {defaultValue=0.03, parameterName=InnerMaskRadiusOffset}
  n7 = MaterialFunctionCall(DistanceField=MaterialFunctionCall(Radius=Subtract(A=@n5, B=@n6) {constB=0.03}) {materialFunction=&MF_SDF_Circle}.SDF) {materialFunction=&MF_UI_AntiAliasing}
  n8 = MaterialFunctionCall(Radius=Add(A=ScalarParameter() {defaultValue=0.85, parameterName=Radius}, B=ScalarParameter() {defaultValue=0.048, parameterName=OuterRingGap}) {constB=0.047}) {materialFunction=&MF_SDF_Circle}
  n9 = NamedRerouteDeclaration(None=MaterialFunctionCall(DistanceField=MaterialFunctionCall!(DistanceField=@n8.SDF, Threshold, Thickness, Alignment, Smoothing) {materialFunction=/FortUILibrary/MaterialFunctions/Shapes/Modifiers/MF_UI_ShapeOutline}.SDFOutline) {materialFunction=&MF_UI_AntiAliasing}.Inner) {name=CardinalOuterRing, nodeColor=<rgba 0 0 0 1>}
  n10 = ScalarParameter() {defaultValue=0.35, parameterName=OuterOutlineAlpha}
  LinearInterpolate(
   A=NamedRerouteUsage()
   B=Add(A=Multiply(A=@n4, B=@n7.Inner), B=Multiply(A=@n9, B=@n10) {constB=0.65})
   Alpha=ScalarParameter() {parameterName=ShowOuterRing}
  )
  NamedRerouteDeclaration(None=TextureSampleParameter2D() {parameterName=Target, texture=&/T_UI_MapLegend_Reboot}.RGB) {name=TextureColor, nodeColor=<rgba 0 0 0 1>}
 ]
}
