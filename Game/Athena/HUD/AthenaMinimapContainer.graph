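# HUD minimap container widget that hosts the Athena minimap and toggles the fullscreen map on click.
# Map controller device payloads from the playspace event router override shape, transparency, backing color and rotation mode.
# Shape changes from the device or HUD context swap the RetainerBox mask material.
# When the HUD's circular setting and cardinal display are on, a fast looping timer rotates the compass grid and counter-rotates its letters to stay upright.
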WidgetBlueprint /Game/Athena/HUD/AthenaMinimapContainer
graphs [EventGraph {nodes [
   "On Clicked (Button_Map)"() {is=title K2Node_ComponentBoundEvent}
   > ToggleFullscreenMap(self:Fort HUDContext Object Reference=Get FortHUDContext!() -> (ReturnValue:Fort HUDContext Object Reference) {is=title K2Node_GetBlueprintContextEx})
   SetupForMapDeviceOverrides() {is=title K2Node_CustomEvent}
   > Set DefaultBackingColor(DefaultBackingColor:Linear Color Structure=Get BrushColor!(self:Border Object Reference=Get Border() -> (Border:Border Object Reference)) -> (BrushColor:Linear Color Structure)) -> (Output_Get:Linear Color Structure)
   > n1 = GetEventRouter!(execute, Scope:Object Class Reference=/Script/FortniteGame.FortPlayspace, ContextActor:Actor Object Reference=GetOwningPlayer()) -> (then, ReturnValue:Gameplay Event Router Component Object Reference)
   > Is Valid(InputObject=@n1)
   > n2 = StartListeningToEvent!(
    execute:Exec
    Target:Gameplay Event Router Component Object Reference=@n1
    EventType:Script Struct Object Reference=/Game/Athena/HUD/MiniMap/MapController/GER_MapControllerDevice_ActivePayload
    Context:Object Reference
    EventBubblingRule:EEventBubblingRule Enum
   ) -> (
    then:Exec
    AsyncTaskProxy:Async Action Start Listening To Event Object Reference
    OnEventReceived:Exec
    Payload:GER Map Controller Device Active Payload Structure
   )
   n3 = Set MapControllerSettings!(execute=@n2.OnEventReceived, MapControllerSettings:GER Map Controller Device Active Payload Structure=@n2.Payload) -> (
    then
    Output_Get_CustomMapActive_1_#1:bool=~
    Output_Get_HideMapInfoPanel_4_#2:bool=~
    Output_Get_TransparentMinimap_12_#3:bool=~
    Output_Get_MapShape_10_#4:EMapController_MapShape Enum
    Output_Get_MapRotationMode_16_#5:EMapControllerDevice_RotationMode Enum
   )
   > Set MapDeviceShapeSetting(MapDeviceShapeSetting:EMapController_MapShape Enum=Select(
     Option 0:EMapController_MapShape Enum=NewEnumerator0
     Option 1:EMapController_MapShape Enum=@n3.Output_Get_MapShape_10_#4
     Index=@n3
    ) -> (ReturnValue:EMapController_MapShape Enum)) -> (Output_Get:EMapController_MapShape Enum=NewEnumerator0)
   > Set UseTransparentMap(UseTransparentMap:bool=AND Boolean(A=@n3, B=@n3.Output_Get_TransparentMinimap_12_#3)) -> (Output_Get:bool)
   > OnMinimapShapeChanged!(execute, self, bMinimapCircularShapeEnabled:bool=false) -> (then)
   n4 = Get MapControllerSettings() -> (MapControllerSettings:GER Map Controller Device Active Payload Structure)
   > SetBrushColor(
    self=Get Border()
    InBrushColor=SelectColor(
     A=Get CustomMapBackingColor() -> (CustomMapBackingColor:Linear Color Structure)
     B=Get DefaultBackingColor() -> (DefaultBackingColor:Linear Color Structure)
     bPickA=Break GER Map Controller Device Active Payload!(GER_MapControllerDevice_ActivePayload:"GER Map Controller Device Active Payload Structure (by ref)"=@n4) -> (CustomMapActive_1_#1:bool=~) {is=title K2Node_BreakStruct}
    )
   )
   n5 = Break GER Map Controller Device Active Payload!(GER_MapControllerDevice_ActivePayload=Get MapControllerSettings()) -> (
    CustomMapActive_1_#1=~
    MapRotationMode_16_#5:EMapControllerDevice_RotationMode Enum
   )
   > OverrideMinimapRotationMode!(
    execute
    self:Athena Minimap Object Reference=Get Minimap() -> (Minimap:Athena Minimap Object Reference)
    bOverrideActive:bool=AND Boolean(
     A=@n5
     B="Not Equal (Enum)"!(
      A:EMapControllerDevice_RotationMode Enum=@n5.MapRotationMode_16_#5
      B:EMapControllerDevice_RotationMode Enum=NewEnumerator0
     ) -> (ReturnValue:bool)
    )
    bRotationEnabled:bool="Equal (Enum)"(
     A:EMapControllerDevice_RotationMode Enum=@n5.MapRotationMode_16_#5
     B:EMapControllerDevice_RotationMode Enum=NewEnumerator2
    )
   ) -> (then)
   OnMinimapShapeChanged!() -> (OutputDelegate, then, bMinimapCircularShapeEnabled:bool=~) {is=title K2Node_CustomEvent}
   n6 = IsCircularMinimapShapeSettingEnabled!(self:Fort HUDContext Object Reference=Get FortHUDContext()) -> (ReturnValue:bool)
   n7 = Get CircularMask() -> (CircularMask:Material Interface Object Reference)
   n8 = Get SquareMask() -> (SquareMask:Material Interface Object Reference)
   > SetEffectMaterial!(
    execute
    self:Retainer Box Object Reference=Get RetainerBox() -> (RetainerBox:Retainer Box Object Reference)
    EffectMaterial:Material Interface Object Reference=Select!(
     NewEnumerator0:Material Interface Object Reference=Select!(Option 0:Material Interface Object Reference=@n8, Option 1:Material Interface Object Reference=@n7, Index:bool=@n6) -> (ReturnValue)
     NewEnumerator1:Material Interface Object Reference=Get SquareMask()
     NewEnumerator2:Material Interface Object Reference=Get CircularMask()
     Index:EMapController_MapShape Enum=Get MapDeviceShapeSetting() -> (MapDeviceShapeSetting:EMapController_MapShape Enum=NewEnumerator0)
    ) -> (ReturnValue:Material Interface Object Reference)
   ) -> (then)
   > Set RenderOpacity!(
    execute
    RenderOpacity:float=Select(
     Option 0:float=1
     Option 1:float=Get TransparentMinimapOpacity() -> (TransparentMinimapOpacity:double)
     Index=Get UseTransparentMap() -> (UseTransparentMap:bool)
    ) -> (ReturnValue:float)
    self:Widget Object Reference=Get RetainerBox()
   ) -> (then, Output_Get:float)
   n9 = Get FortHUDContext()
   > n10 = Branch(Condition=AND Boolean(A=IsCircularMinimapShapeSettingEnabled(self=@n9), B=Get ShouldShowCardinalDirections() -> (ShouldShowCardinalDirections:bool)))
   n11 = Get Grid_CircleMap_CardinalDirections() -> (Grid_CircleMap_CardinalDirections:Grid Panel Object Reference)
   > SetVisibility(self=@n11, InVisibility=HitTestInvisible)
   > Set FortPlayerController(FortPlayerController:Fort Player Controller Object Reference=GetFortPlayerControllerFromActor!(Actor:Actor Object Reference=GetOwningPlayer()) -> (ReturnValue:Fort Player Controller Object Reference)) -> (Output_Get:Fort Player Controller Object Reference)
   > SetEffectMaterial(self=Get RetainerBox(), EffectMaterial=Get CircularMaskWithRing() -> (CircularMaskWithRing:Material Interface Object Reference))
   > n12 = Set RenderOpacity(RenderOpacity=Get TransparentMinimapOpacity(), self=Get Grid_CircleMap_CardinalDirections())
   n13 = OnApplyRotation() {is=title K2Node_CustomEvent}
   n14 = Set Timer by Event(execute=@n12.then, Delegate=@n13, Time=0.01, bLooping=true, bMaxOncePerFrame=true)
   SetVisibility(execute=@n10.else, self=@n11, InVisibility=Collapsed)
   > Clear and Invalidate Timer by Handle(Handle=@n14)
   Event Construct()
   > Bind Event to On Minimap Shape Changed(self:Fort HUDContext Object Reference=Get FortHUDContext(), Delegate=Create Event()) {is=title K2Node_AddDelegate}
   > SetupForMapDeviceOverrides() {is=title K2Node_CallFunction}
   > Is Valid(InputObject=Get RetainerBox())
   > OnMinimapShapeChanged(bMinimapCircularShapeEnabled=IsCircularMinimapShapeSettingEnabled(self=Get FortHUDContext())) {is=title K2Node_CallFunction}
   n15 = GetOwningPlayer()
   > Branch(Condition=AND Boolean(A=IsMobileGame(), B=IsUsingTouch(PlayerController=@n15)))
   > SetVisibility(self=Get MinimapSlottingOverlay() -> (MinimapSlottingOverlay:Widget Slotting Host Overlay Object Reference), InVisibility=Collapsed)
   Event Destruct()
   > Unbind Event from On Minimap Shape Changed(self:Fort HUDContext Object Reference=Get FortHUDContext(), Delegate=Create Event()) {is=title K2Node_RemoveDelegate}
   n16 = GetPlayerRotation!(
    self:Athena Minimap Object Reference=Get Minimap()
    FortPlayerController:Fort Player Controller Object Reference=Get FortPlayerController() -> (FortPlayerController:Fort Player Controller Object Reference)
   ) -> (ReturnValue:float)
   SetRenderTransformAngle!(
    execute=@n17.then
    self:Widget Object Reference=[
     Get Text_West() -> (Text_West:Text Block Object Reference)
     Get Text_South() -> (Text_South:Text Block Object Reference)
     Get Text_East() -> (Text_East:Text Block Object Reference)
     Get Text_North() -> (Text_North:Text Block Object Reference)
    ]
    Angle:float=@n16
   ) -> (then)
   n17 = SetRenderTransformAngle(execute=@n13.then, self=Get Grid_CircleMap_CardinalDirections(), Angle=float * float(A=@n16, B=-1) {is=title K2Node_PromotableOperator})
  ]}]
variables [
 CircularMask {type=MaterialInterface, default=/Game/Athena/HUD/Materials/M_UI_CircularMask}
 CircularMaskWithRing {type=MaterialInterface, default=/Game/Athena/HUD/Materials/MI_UI_CircularMask}
 CustomMapBackingColor {type=LinearColor, default=<rgba 0 0 0 0.5>}
 DefaultBackingColor {type=LinearColor, default=<rgba 0 0.1384 0.3763 1>}
 FortPlayerController {type=FortPlayerController, default=null}
 Key {type=Key, default=Semicolon}
 MapControllerSettings {
  type=GER_MapControllerDevice_ActivePayload
  default {customMapActive=false, hideMapInfoPanel=false, mapRotationMode=Default, mapShape=Default, transparentMinimap=false}
 }
 MapDeviceShapeSetting {type=EMapController_MapShape, default=Default}
 ShouldShowCardinalDirections {type=boolean, default=true}
 SquareMask {type=MaterialInterface, default=null}
 TransparentMinimapOpacity {type=number, default=0.92}
 UseTransparentMap {type=boolean}
]
widgets {
 info {inheritedWidgetCount=1, parentClass=&FortHUDElementWidget, rootWidgetClass=&Overlay, widgetCount=15}
 widgets [
  Overlay_0 {widgetClassPath=&Overlay, widget=Overlay_0}
  MinimapSlottingOverlay {
   widgetClassPath=/Script/WidgetSlotting.WidgetSlottingHostOverlay
   bIsVariable=true
   parent=Overlay_0
   widget=MinimapSlottingOverlay
   properties {hostTag {tagName=Athena.HUD.MinimapOverlaySlot}}
   slot {class=OverlaySlot, properties {verticalAlignment=VAlign_Fill}}
  }
  RetainerBox {
   widgetClassPath=/Script/UMG.RetainerBox
   bIsVariable=true
   parent=Overlay_0
   widget=RetainerBox
   properties {effectMaterial=&M_UI_CircularMask, phaseCount=2, textureParameter=Target}
   slot {class=OverlaySlot}
  }
  Border {
   widgetClassPath=&Border
   bIsVariable=true
   parent=RetainerBox
   widget=Border
   properties {visibility=SelfHitTestInvisible, brushColor=<rgba 0 0.1384 0.3763 1>, padding=<ltrb 0 0 0 0>}
   slot {class=PanelSlot}
  }
  Overlay_Map {
   widgetClassPath=&Overlay
   bIsVariable=true
   parent=Border
   widget=Overlay_Map
   slot {class=BorderSlot, properties {padding=<ltrb 0 0 0 0>}}
  }
  Button_Map {
   widgetClassPath=&Button
   parent=Overlay_Map
   widget=Button_Map
   properties {
    isFocusable=false
    backgroundColor=<rgba 1 1 1 0>
    colorAndOpacity=<rgba 1 1 1 0>
    widgetStyle {
     clickedSlateSound {resourceObject=null}
     disabled {
      bIsDynamicallyLoaded=false
      drawAs=NoDrawType
      imageType=NoImage
      mirroring=NoMirror
      resourceName=None
      resourceObject=null
      tiling=NoTile
      imageSize=<0 0>
      margin=<ltrb 0 0 0 0>
      outlineSettings {
       bUseBrushTransparency=false
       roundingType=HalfHeightRadius
       width=0
       color {colorUseRule=UseColor_Specified, specifiedColor=<rgba 0 0 0 0>}
       cornerRadii=<0 0 0 0>
      }
      tintColor {colorUseRule=UseColor_Specified, specifiedColor=<rgba 1 1 1 1>}
      uVRegion {bIsValid=false, max=<0 0>, min=<0 0>}
     }
     disabledForeground {colorUseRule=UseColor_Specified, specifiedColor=<rgba 0.5271 0.5271 0.5271 1>}
     hovered {
      bIsDynamicallyLoaded=false
      drawAs=RoundedBox
      imageType=NoImage
      mirroring=NoMirror
      resourceName=None
      resourceObject=null
      tiling=NoTile
      imageSize=<32 32>
      margin=<ltrb 0 0 0 0>
      outlineSettings {
       bUseBrushTransparency=true
       roundingType=FixedRadius
       width=1
       color {colorUseRule=UseColor_Specified, specifiedColor=<rgba 0.7243 0.7243 0.7243 1>}
       cornerRadii=<4 4 4 4>
      }
      tintColor {colorUseRule=UseColor_Specified, specifiedColor=<rgba 1 1 1 0>}
      uVRegion {bIsValid=false, max=<0 0>, min=<0 0>}
     }
     hoveredForeground {colorUseRule=UseColor_Specified, specifiedColor=<rgba 0 1 1 1>}
     hoveredSlateSound {resourceObject=null}
     normal {
      bIsDynamicallyLoaded=false
      drawAs=RoundedBox
      imageType=NoImage
      mirroring=NoMirror
      resourceName=None
      resourceObject=null
      tiling=NoTile
      imageSize=<32 32>
      margin=<ltrb 0 0 0 0>
      outlineSettings {
       bUseBrushTransparency=true
       roundingType=FixedRadius
       width=1
       color {colorUseRule=UseColor_Specified, specifiedColor=<rgba 0.6951 0.6951 0.6951 1>}
       cornerRadii=<4 4 4 4>
      }
      tintColor {colorUseRule=UseColor_Specified, specifiedColor=<rgba 1 1 1 0>}
      uVRegion {bIsValid=false, max=<0 0>, min=<0 0>}
     }
     normalForeground {colorUseRule=UseColor_Specified, specifiedColor=<rgba 0 1 1 1>}
     normalPadding=<ltrb 12 1.5 12 1.5>
     pressed {
      bIsDynamicallyLoaded=false
      drawAs=RoundedBox
      imageType=NoImage
      mirroring=NoMirror
      resourceName=None
      resourceObject=null
      tiling=NoTile
      imageSize=<32 32>
      margin=<ltrb 0 0 0 0>
      outlineSettings {
       bUseBrushTransparency=true
       roundingType=FixedRadius
       width=1
       color {colorUseRule=UseColor_Specified, specifiedColor=<rgba 0.7243 0.7243 0.7243 1>}
       cornerRadii=<4 4 4 4>
      }
      tintColor {colorUseRule=UseColor_Specified, specifiedColor=<rgba 1 1 1 0>}
      uVRegion {bIsValid=false, max=<0 0>, min=<0 0>}
     }
     pressedForeground {colorUseRule=UseColor_Specified, specifiedColor=<rgba 0 1 1 1>}
     pressedPadding=<ltrb 12 1.5 12 1.5>
     pressedSlateSound {resourceObject=null}
    }
   }
   slot {class=OverlaySlot, properties {horizontalAlignment=HAlign_Fill, verticalAlignment=VAlign_Fill}}
  }
  Minimap {
   widgetClassPath=/Script/FortniteUI.AthenaMinimap
   bIsVariable=true
   parent=Overlay_Map
   widget=Minimap
   properties {
    visibility=SelfHitTestInvisible
    desiredSize=<300 300>
    elevationArrowIconBrush {
     bIsDynamicallyLoaded=false
     drawAs=Image
     imageType=FullColor
     mirroring=NoMirror
     resourceName=None
     resourceObject=/Game/UI/CodeReferenced/T_UI_ElevationArrow
     tiling=NoTile
     imageSize=<32 32>
     margin=<ltrb 0 0 0 0>
     outlineSettings {
      bUseBrushTransparency=false
      roundingType=HalfHeightRadius
      width=0
      color {colorUseRule=UseColor_Specified, specifiedColor=<rgba 0 0 0 0>}
      cornerRadii=<0 0 0 0>
     }
     tintColor {colorUseRule=UseColor_Specified, specifiedColor=<rgba 1 1 1 1>}
     uVRegion {bIsValid=false, max=<0 0>, min=<0 0>}
    }
    playerIconBrush {
     bIsDynamicallyLoaded=false
     drawAs=Image
     imageType=FullColor
     mirroring=NoMirror
     resourceName=None
     resourceObject=/Game/UI/CodeReferenced/UI_Radar_AthenaPlayerArrow
     tiling=NoTile
     imageSize=<20 20>
     margin=<ltrb 0 0 0 0>
     outlineSettings {
      bUseBrushTransparency=false
      roundingType=HalfHeightRadius
      width=0
      color {colorUseRule=UseColor_Specified, specifiedColor=<rgba 0 0 0 0>}
      cornerRadii=<0 0 0 0>
     }
     tintColor {colorUseRule=UseColor_Specified, specifiedColor=<rgba 1 1 1 1>}
     uVRegion {bIsValid=false, max=<0 0>, min=<0 0>}
    }
    squadPinIconBrush {
     bIsDynamicallyLoaded=false
     drawAs=Image
     imageType=FullColor
     mirroring=NoMirror
     resourceName=None
     resourceObject=/Game/UI/CodeReferenced/UI_Squad_Pin
     tiling=NoTile
     imageSize=<35 35>
     margin=<ltrb 0 0 0 0>
     outlineSettings {
      bUseBrushTransparency=false
      roundingType=HalfHeightRadius
      width=0
      color {colorUseRule=UseColor_Specified, specifiedColor=<rgba 0 0 0 0>}
      cornerRadii=<0 0 0 0>
     }
     tintColor {colorUseRule=UseColor_Specified, specifiedColor=<rgba 1 1 1 1>}
     uVRegion {bIsValid=false, max=<0 0>, min=<0 0>}
    }
   }
   slot {class=OverlaySlot}
  }
  TutorialHighlight {
   widgetClassPath=&TutorialHighlight
   bIsVariable=true
   parent=Overlay_Map
   widget=TutorialHighlight
   properties {customHighlightWidget=/Game/Athena/HUD/Tutorial/Tutorial_Outline_Square.Tutorial_Outline_Square_C, visibility=SelfHitTestInvisible}
   slot {class=OverlaySlot, properties {horizontalAlignment=HAlign_Fill, verticalAlignment=VAlign_Fill, padding=<ltrb -6 -6 -6 -6>}}
  }
  Grid_CircleMap_CardinalDirections {
   widgetClassPath=&GridPanel
   bIsVariable=true
   parent=Overlay_0
   widget=Grid_CircleMap_CardinalDirections
   properties {columnFill [0.1,0,0.1], rowFill [0.1,0,0.1]}
   slot {class=OverlaySlot, properties {horizontalAlignment=HAlign_Fill, verticalAlignment=VAlign_Fill}}
  }
  Border_0 {
   widgetClassPath=&Border
   parent=Grid_CircleMap_CardinalDirections
   widget=Border_0
   properties {
    background {
     bIsDynamicallyLoaded=false
     drawAs=Image
     imageType=FullColor
     mirroring=NoMirror
     resourceName=None
     resourceObject=/Game/Athena/HUD/Materials/MI_UI_CardinalNorth
     tiling=NoTile
     imageSize=<32 32>
     margin=<ltrb 0 0 0 0>
     outlineSettings {
      bUseBrushTransparency=false
      roundingType=HalfHeightRadius
      width=0
      color {colorUseRule=UseColor_Specified, specifiedColor=<rgba 0 0 0 0>}
      cornerRadii=<0 0 0 0>
     }
     tintColor {colorUseRule=UseColor_Specified, specifiedColor=<rgba 1 1 1 1>}
     uVRegion {bIsValid=false, max=<0 0>, min=<0 0>}
    }
    brushColor=<rgba 0 0 0 1>
    padding=<ltrb 9 3 8 10>
   }
   slot {class=GridSlot, properties {column=1, horizontalAlignment=HAlign_Center, verticalAlignment=VAlign_Center}}
  }
  Text_North {
   widgetClassPath=&UEFN_TextBlock
   bIsVariable=true
   parent=Border_0
   widget=Text_North
   properties {
    justification=Center
    text=N
    font {
     bForceMonospaced=false
     bMaterialIsStencil=false
     fontMaterial=null
     fontObject=&BurbankSmall
     letterSpacing=0
     monospacedWidth=1
     size=11.25
     skewAmount=0
     typefaceFontName=Black
     outlineSettings {
      bApplyOutlineToDropShadows=false
      bMiteredCorners=false
      bSeparateFillAlpha=false
      outlineMaterial=null
      outlineSize=0
      outlineColor=<rgba 0 0 0 1>
     }
    }
   }
   slot {class=BorderSlot, properties {padding=<ltrb 9 3 8 10>}}
  }
  Text_East {
   widgetClassPath=&UEFN_TextBlock
   bIsVariable=true
   parent=Grid_CircleMap_CardinalDirections
   widget=Text_East
   properties {
    text=E
    font {
     bForceMonospaced=false
     bMaterialIsStencil=false
     fontMaterial=null
     fontObject=&BurbankSmall
     letterSpacing=0
     monospacedWidth=1
     size=11.25
     skewAmount=0
     typefaceFontName=Black
     outlineSettings {
      bApplyOutlineToDropShadows=false
      bMiteredCorners=false
      bSeparateFillAlpha=false
      outlineMaterial=null
      outlineSize=0
      outlineColor=<rgba 0 0 0 1>
     }
    }
   }
   slot {class=GridSlot, properties {column=2, horizontalAlignment=HAlign_Center, row=1, verticalAlignment=VAlign_Center}}
  }
  Spacer_166 {
   widgetClassPath=&Spacer
   parent=Grid_CircleMap_CardinalDirections
   widget=Spacer_166
   properties {size=<242 220>}
   slot {class=GridSlot, properties {column=1, row=1, verticalAlignment=VAlign_Top}}
  }
  Text_South {
   widgetClassPath=&UEFN_TextBlock
   bIsVariable=true
   parent=Grid_CircleMap_CardinalDirections
   widget=Text_South
   properties {
    text=S
    font {
     bForceMonospaced=false
     bMaterialIsStencil=false
     fontMaterial=null
     fontObject=&BurbankSmall
     letterSpacing=0
     monospacedWidth=1
     size=11.25
     skewAmount=0
     typefaceFontName=Black
     outlineSettings {
      bApplyOutlineToDropShadows=false
      bMiteredCorners=false
      bSeparateFillAlpha=false
      outlineMaterial=null
      outlineSize=0
      outlineColor=<rgba 0 0 0 1>
     }
    }
   }
   slot {class=GridSlot, properties {column=1, horizontalAlignment=HAlign_Center, row=2, verticalAlignment=VAlign_Center, nudge=<0 5>}}
  }
  Text_West {
   widgetClassPath=&UEFN_TextBlock
   bIsVariable=true
   parent=Grid_CircleMap_CardinalDirections
   widget=Text_West
   properties {
    text=W
    font {
     bForceMonospaced=false
     bMaterialIsStencil=false
     fontMaterial=null
     fontObject=&BurbankSmall
     letterSpacing=0
     monospacedWidth=1
     size=11.25
     skewAmount=0
     typefaceFontName=Black
     outlineSettings {
      bApplyOutlineToDropShadows=false
      bMiteredCorners=false
      bSeparateFillAlpha=false
      outlineMaterial=null
      outlineSize=0
      outlineColor=<rgba 0 0 0 1>
     }
    }
   }
   slot {class=GridSlot, properties {horizontalAlignment=HAlign_Center, row=1, verticalAlignment=VAlign_Center}}
  }
 ]
}
