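# HUD widget on AthenaRespawnBase that shows whether respawn is available, plus a remaining-lives counter.
# ToggleRespawnState, driven by OnUpdateRespawnState and PreConstruct, switches background color, check icon and RESPAWN/NO RESPAWN! text style, and collapses the lives box when respawn is off.
# OnUpdateLives interpolates the counter if the box already shows, otherwise sets it directly, then reveals the box only while the check is visible.
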WidgetBlueprint /Game/Athena/HUD/AthenaRespawnIndicator
graphs [
 EventGraph {nodes [
   n1 = ToggleRespawnState!() -> (OutputDelegate, then, canRespawn:bool=~) {is=title K2Node_CustomEvent}
   > SetBrushColor(
    self=Get Background() -> (Background:Border Object Reference)
    InBrushColor=Select(
     Option 0:Linear Color Structure=(R=0.0075,G=0.0033,B=0.0027,A=0.61)
     Option 1:Linear Color Structure=(R=0.0052,G=0.0052,B=0.0056,A=0.61)
     Index=@n1.canRespawn
    ) -> (ReturnValue:Linear Color Structure)
   )
   > SetVisibility(
    self=Get Image_Check() -> (Image_Check:Image Object Reference)
    InVisibility=Select(Option 0:ESlateVisibility Enum=Hidden, Option 1:ESlateVisibility Enum=Visible, Index=@n1.canRespawn) -> (ReturnValue:ESlateVisibility Enum)
   )
   > SetColorAndOpacity(
    self=Get CommonTextBlock_StatusText() -> (CommonTextBlock_StatusText:Common Text Object Reference)
    InColorAndOpacity=Select(
     Option 0:Slate Color Structure=Make Slate Color(SpecifiedColor=(R=0.3073,G=0.0139,B=0,A=1))
     Option 1:Slate Color Structure=Make Slate Color(SpecifiedColor=(R=1,G=1,B=1,A=1))
     Index=@n1.canRespawn
    ) -> (ReturnValue:Slate Color Structure)
   )
   n2 = Get CommonTextBlock_StatusText()
   > "SetText (Text)"(
    self=@n2
    InText=Select(Option 0:Text="LOCTEXT(\"NO RESPAWN!\")", Option 1:Text="LOCTEXT(\"RESPAWN\")", Index=@n1.canRespawn) -> (ReturnValue:Text)
   )
   > SetRenderShear!(
    execute
    self:Widget Object Reference=@n2
    Shear:Vector 2D Structure=Select(Option 0:Vector 2D Structure=(X=-4,Y=0), Option 1:Vector 2D Structure, Index=@n1.canRespawn) -> (ReturnValue:Vector 2D Structure)
   ) -> (then)
   > n3 = Branch(Condition=@n1.canRespawn)
   SetVisibility(execute=@n3.else, self=Get LivesIndicatorBox() -> (LivesIndicatorBox:Horizontal Box Object Reference), InVisibility=Collapsed)
   n4 = SetVisibilty!() -> (OutputDelegate, then, Visible:bool=~) {is=title K2Node_CustomEvent}
   > SetVisibility(InVisibility=Select(Option 0:ESlateVisibility Enum=Collapsed, Option 1:ESlateVisibility Enum=HitTestInvisible, Index=@n4.Visible) -> (ReturnValue:ESlateVisibility Enum))
   n5 = Event OnUpdateRespawnState!() -> (OutputDelegate, then, bEnabled:bool) {is=title K2Node_Event}
   > ToggleRespawnState!(execute, self, canRespawn:bool=@n5.bEnabled) -> (then) {is=title K2Node_CallFunction}
   n6 = Event OnUpdateLives!() -> (OutputDelegate, then, Lives:int=0) {is=title K2Node_Event}
   > n7 = Branch(Condition=Should Interpolate Lives Counter!() -> (ReturnValue:bool=~) {is=title K2Node_MacroInstance})
   n8 = "To Float (Integer)"(InInt=@n6.Lives)
   n9 = Get PlayerLivesText() -> (PlayerLivesText:Common Numeric Text Block Object Reference)
   > n10 = InterpolateToValue!(
    execute
    self:Common Numeric Text Block Object Reference=@n9
    TargetValue:float=@n8
    MaximumInterpolationDuration:float=2
    MinimumChangeRate:float=1
    OutroOffset:float=0
   ) -> (then)
   Branch(execute=[@n10.then, @n11.then], Condition=IsVisible(self=Get Image_Check()))
   > SetVisibility(self=Get LivesIndicatorBox(), InVisibility=HitTestInvisible)
   n11 = SetCurrentValue!(execute=@n7.else, self:Common Numeric Text Block Object Reference=@n9, NewValue:float=@n8) -> (then)
   Event PreConstruct()
   > ToggleRespawnState(canRespawn=Get canRespawn() -> (canRespawn:bool))
  ]}
 Should Interpolate Lives Counter {nodes [
   Inputs!()
   n1 = Get LivesIndicatorBox()
   Outputs!(ReturnValue:bool="Equal (Enum)"(
     A:ESlateVisibility Enum=GetVisibility!(self:Widget Object Reference=@n1) -> (ReturnValue:ESlateVisibility Enum=Visible)
     B:ESlateVisibility Enum=HitTestInvisible
    ))
  ]}
]
variables [canRespawn {type=boolean, default=true}]
widgets {
 info {inheritedWidgetCount=1, parentClass=/Script/FortniteUI.AthenaRespawnBase, rootWidgetClass=&Border, widgetCount=8}
 widgets [
  Background {
   widgetClassPath=&Border
   bIsVariable=true
   widget=Background
   properties {horizontalAlignment=HAlign_Center, brushColor=<rgba 0.0052 0.0052 0.0056 0.61>, padding=<ltrb 0 0 0 0>}
  }
  HorizontalBox_0 {
   widgetClassPath=&HorizontalBox
   parent=Background
   widget=HorizontalBox_0
   slot {class=BorderSlot, properties {horizontalAlignment=HAlign_Center, padding=<ltrb 0 0 0 0>}}
  }
  Image_Check {
   widgetClassPath=&Image
   bIsVariable=true
   parent=HorizontalBox_0
   widget=Image_Check
   properties {
    brush {
     bIsDynamicallyLoaded=false
     drawAs=Image
     imageType=NoImage
     mirroring=NoMirror
     resourceName=None
     resourceObject=/Game/UI/Foundation/Textures/Icons/Checks/T-Icon-Check-64
     tiling=NoTile
     imageSize=<25 25>
     margin=<ltrb 0 0 0 0>
     outlineSettings {
      bUseBrushTransparency=false
      roundingType=HalfHeightRadius
      width=0
      color {colorUseRule=UseColor_Specified, specifiedColor=<rgba 0 0 0 0>}
      cornerRadii=<0 0 0 0>
     }
     tintColor {colorUseRule=UseColor_Specified, specifiedColor=<rgba 1 1 1 1>}
     uVRegion {bIsValid=false, max=<0 0>, min=<0 0>}
    }
    colorAndOpacity=<rgba 0 1 0.006 1>
   }
   slot {class=HorizontalBoxSlot, properties {horizontalAlignment=HAlign_Left, verticalAlignment=VAlign_Center}}
  }
  CommonTextBlock_StatusText {
   widgetClassPath=&CommonTextBlock
   bIsVariable=true
   parent=HorizontalBox_0
   widget=CommonTextBlock_StatusText
   properties {
    justification=Center
    text=RESPAWN
    font {
     bForceMonospaced=false
     bMaterialIsStencil=false
     fontMaterial=null
     fontObject=&NotoSans
     letterSpacing=0
     monospacedWidth=1
     size=18
     skewAmount=0
     typefaceFontName=NotoSans-Bold
     outlineSettings {
      bApplyOutlineToDropShadows=false
      bMiteredCorners=false
      bSeparateFillAlpha=false
      outlineMaterial=null
      outlineSize=1
      outlineColor=<rgba 0.0052 0.0052 0.0056 0.2>
     }
    }
   }
   slot {class=HorizontalBoxSlot, properties {verticalAlignment=VAlign_Center}}
  }
  LivesIndicatorBox {
   widgetClassPath=&HorizontalBox
   bIsVariable=true
   parent=HorizontalBox_0
   widget=LivesIndicatorBox
   properties {visibility=Collapsed}
   slot {class=HorizontalBoxSlot, properties {verticalAlignment=VAlign_Center}}
  }
  CommonTextSeparator {
   widgetClassPath=&CommonTextBlock
   parent=LivesIndicatorBox
   widget=CommonTextSeparator
   properties {
    justification=Center
    text=:
    font {
     bForceMonospaced=false
     bMaterialIsStencil=false
     fontMaterial=null
     fontObject=&NotoSans
     letterSpacing=0
     monospacedWidth=1
     size=18
     skewAmount=0
     typefaceFontName=NotoSans-Bold
     outlineSettings {
      bApplyOutlineToDropShadows=false
      bMiteredCorners=false
      bSeparateFillAlpha=false
      outlineMaterial=null
      outlineSize=1
      outlineColor=<rgba 0.0052 0.0052 0.0056 0.2>
     }
    }
   }
   slot {class=HorizontalBoxSlot, properties {verticalAlignment=VAlign_Center}}
  }
  PlayerLivesText {
   widgetClassPath=/Script/CommonUI.CommonNumericTextBlock
   bIsVariable=true
   parent=LivesIndicatorBox
   widget=PlayerLivesText
   properties {
    interpolationUpdateInterval=0.1
    text=Text Block
    font {
     bForceMonospaced=false
     bMaterialIsStencil=false
     fontMaterial=null
     fontObject=&NotoSans
     letterSpacing=0
     monospacedWidth=1
     size=18
     skewAmount=0
     typefaceFontName=NotoSans-Bold
     outlineSettings {
      bApplyOutlineToDropShadows=false
      bMiteredCorners=false
      bSeparateFillAlpha=false
      outlineMaterial=null
      outlineSize=1
      outlineColor=<rgba 0.0052 0.0052 0.0056 0.2>
     }
    }
    shadowOffset=<0 0>
   }
   slot {class=HorizontalBoxSlot, properties {verticalAlignment=VAlign_Center}}
  }
  TutorialHighlight {widgetClassPath=&TutorialHighlight}
 ]
}
