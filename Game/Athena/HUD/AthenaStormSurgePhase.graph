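# HUD widget that shows the Storm Surge warning in Athena, with storm icon, alert title and optional subtitle over glow and lightning materials.
# On first display it broadcasts start, plays a staggered scale pop-in animation and picks a damage-active or warning sound.
# Every display call updates the texts and sets a timer.
# StopDisplayAfterTime or OnStopDisplaying clears the displaying flag, broadcasts stop and marks content inactive.
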WidgetBlueprint /Game/Athena/HUD/AthenaStormSurgePhase
animations [StormDMGAlert {
  binding_count=4
  display_label=""
  length_seconds=2.5
  object=&self:StormDMGAlert
  sequence {bindings [
    HelpText {
     id=#1
     possessed_class=CommonTextBlock
     tracks [{
       type=MovieScene2DTransformTrack
       sections [{
         type=MovieScene2DTransformSection
         end_frame=50
         start_frame=0
         channels [
          Translation.X {type=MovieSceneScriptingFloatChannel, const [0,1]}
          Translation.Y {type=MovieSceneScriptingFloatChannel, const [0,1]}
          Angle {type=MovieSceneScriptingFloatChannel, const [0,1]}
          Scale.X {type=MovieSceneScriptingFloatChannel, keys [[0,0],[13,0],[19,1],[22,0.9],[25,1]]}
          Scale.Y {type=MovieSceneScriptingFloatChannel, keys [[0,0],[13,0],[19,1],[22,0.9],[25,1]]}
          Shear.X {type=MovieSceneScriptingFloatChannel, const [0,1]}
          Shear.Y {type=MovieSceneScriptingFloatChannel, const [0,1]}
         ]
        }]
      }]
    }
    AlertText {
     id=#2
     possessed_class=CommonTextBlock
     tracks [{
       type=MovieScene2DTransformTrack
       sections [{
         type=MovieScene2DTransformSection
         end_frame=50
         start_frame=0
         channels [
          Translation.X {type=MovieSceneScriptingFloatChannel, const [0,1]}
          Translation.Y {type=MovieSceneScriptingFloatChannel, const [0,1]}
          Angle {type=MovieSceneScriptingFloatChannel, const [0,1]}
          Scale.X {type=MovieSceneScriptingFloatChannel, keys [[9,0],[13,1],[16,0.9],[19,1]]}
          Scale.Y {type=MovieSceneScriptingFloatChannel, keys [[9,0],[13,1],[16,0.9],[19,1]]}
          Shear.X {type=MovieSceneScriptingFloatChannel, const [0,1]}
          Shear.Y {type=MovieSceneScriptingFloatChannel, const [0,1]}
         ]
        }]
      }]
    }
    StormIcon {
     id=#3
     possessed_class=Image
     tracks [{
       type=MovieScene2DTransformTrack
       sections [{
         type=MovieScene2DTransformSection
         end_frame=49
         start_frame=0
         channels [
          Translation.X {type=MovieSceneScriptingFloatChannel, const [0,1]}
          Translation.Y {type=MovieSceneScriptingFloatChannel, const [0,1]}
          Angle {type=MovieSceneScriptingFloatChannel, const [0,1]}
          Scale.X {type=MovieSceneScriptingFloatChannel, keys [[0,0],[5,1],[7,0.9],[9,1]]}
          Scale.Y {type=MovieSceneScriptingFloatChannel, keys [[0,0],[5,1],[7,0.9],[9,1]]}
          Shear.X {type=MovieSceneScriptingFloatChannel, const [0,1]}
          Shear.Y {type=MovieSceneScriptingFloatChannel, const [0,1]}
         ]
        }]
      }]
    }
    LightningBack {
     id=#4
     possessed_class=Image
     tracks [{
       type=MovieScene2DTransformTrack
       sections [{
         type=MovieScene2DTransformSection
         end_frame=49
         start_frame=0
         channels [
          Translation.X {type=MovieSceneScriptingFloatChannel, const [0,1]}
          Translation.Y {type=MovieSceneScriptingFloatChannel, const [0,1]}
          Angle {type=MovieSceneScriptingFloatChannel, const [0,1]}
          Scale.X {type=MovieSceneScriptingFloatChannel, keys [[9,0],[16,0],[21,1]]}
          Scale.Y {type=MovieSceneScriptingFloatChannel, const [0.9,1]}
          Shear.X {type=MovieSceneScriptingFloatChannel, const [0,1]}
          Shear.Y {type=MovieSceneScriptingFloatChannel, const [0,1]}
         ]
        }]
      }]
    }
   ]}
 }]
graphs [
 Anim_Load {nodes [
   Anim_Load() {is=title K2Node_FunctionEntry}
   > PlayAnimation(InAnimation=Get StormDMGAlert() -> (StormDMGAlert:Widget Animation Object Reference))
  ]}
 EventGraph {nodes [
   n1 = StopDisplayAfterTime() {is=title K2Node_CustomEvent}
   Set bIsDisplaying(execute=[@n1.then, @n11.then], bIsDisplaying:bool=false) -> (Output_Get:bool)
   > Call Stop Storm Surge Display!(execute:Exec, self:Athena Storm Surge Phase Object Reference) -> (then:Exec) {is=title K2Node_CallDelegate}
   > SetHasActiveContent!(execute, self:Fort Prioritized Widget Interface, bNewActive:bool=false) -> (then)
   n2 = Event OnStartDisplaying!() -> (OutputDelegate, then, TitleText:"Text (by ref)", SubTitleText:"Text (by ref)", bDisplaySubtitle:bool, bPlayDamageActiveSound:bool) {is=title K2Node_Event}
   > n3 = Branch(Condition=Get bIsDisplaying() -> (bIsDisplaying:bool))
   "SetText (Text)"(execute=[@n3.then, @n9.then, @n10.then], self=Get AlertText() -> (AlertText:Common Text Object Reference), InText=@n2.TitleText)
   > n4 = Branch(Condition=@n2.bDisplaySubtitle)
   n5 = Get HelpText() -> (HelpText:Common Text Object Reference)
   > n6 = SetVisibility(self=@n5, InVisibility=SelfHitTestInvisible)
   "SetText (Text)"(execute=[@n7.then, @n6.then], self=@n5, InText=@n2.SubTitleText)
   > Set Timer by Event(Delegate=Create Event(), Time=10, bLooping=false)
   n7 = SetVisibility(execute=@n4.else, self=@n5, InVisibility=Hidden)
   Set bIsDisplaying(execute=@n3.else, bIsDisplaying=true)
   > Call Start Storm Surge Display!(execute:Exec, self:Athena Storm Surge Phase Object Reference) -> (then:Exec) {is=title K2Node_CallDelegate}
   > Anim_Load() {is=title K2Node_CallFunction}
   > n8 = Branch(Condition=@n2.bPlayDamageActiveSound)
   > n9 = PlaySound2D(Sound=/Game/Athena/Sounds/LTMs/StormCap/StormCap_UIAlert_Final_Cue)
   n10 = PlaySound2D(execute=@n8.else, Sound=/Game/Athena/Sounds/LTMs/StormCap/StormCap_UIAlert_Warning_Cue)
   n11 = Event OnStopDisplaying() {is=title K2Node_Event}
   Event PreConstruct()
   Event Construct()
   Event Tick()
  ]}
 StartStormSurgeDisplay {nodes [StartStormSurgeDisplay() {is=title K2Node_FunctionEntry}]}
 StopStormSurgeDisplay {nodes [StopStormSurgeDisplay() {is=title K2Node_FunctionEntry}]}
]
variables [
 bIsDisplaying {type=boolean}
 StartStormSurgeDisplay {}
 StopStormSurgeDisplay {}
]
widgets {
 info {inheritedWidgetCount=1, parentClass=/Script/FortniteUI.AthenaStormSurgeWarningBase, rootWidgetClass=&SizeBox, widgetCount=11}
 widgets [
  SizeBox_0 {
   widgetClassPath=&SizeBox
   widget=SizeBox_0
   properties {bOverride_MinDesiredHeight=true, bOverride_MinDesiredWidth=true, minDesiredHeight=175, minDesiredWidth=600}
  }
  VerticalBox_2 {
   widgetClassPath=&VerticalBox
   parent=SizeBox_0
   widget=VerticalBox_2
   slot {class=SizeBoxSlot, properties {horizontalAlignment=HAlign_Center, verticalAlignment=VAlign_Top}}
  }
  Overlay_0 {
   widgetClassPath=&Overlay
   parent=VerticalBox_2
   widget=Overlay_0
   slot {class=VerticalBoxSlot, properties {horizontalAlignment=HAlign_Center, verticalAlignment=VAlign_Center, padding=<ltrb 0 0 0 -10>}}
  }
  Glow {
   widgetClassPath=&Image
   bIsVariable=true
   parent=Overlay_0
   widget=Glow
   properties {
    brush {
     bIsDynamicallyLoaded=false
     drawAs=Image
     imageType=NoImage
     mirroring=NoMirror
     resourceName=None
     resourceObject=/Game/Athena/HUD/LTM/Wax/Materials/M_UI_TextGlow_Add_Gradient_StormSurge
     tiling=NoTile
     imageSize=<32 32>
     margin=<ltrb 0 0 0 0>
     outlineSettings {
      bUseBrushTransparency=false
      roundingType=HalfHeightRadius
      width=0
      color {colorUseRule=UseColor_Specified, specifiedColor=<rgba 0 0 0 0>}
      cornerRadii=<0 0 0 0>
     }
     tintColor {colorUseRule=UseColor_Specified, specifiedColor=<rgba 1 1 1 1>}
     uVRegion {bIsValid=false, max=<0 0>, min=<0 0>}
    }
    renderTransform {angle=0, scale=<1.375 1.125>, shear=<0 0>, translation=<0 8>}
   }
   slot {class=OverlaySlot, properties {horizontalAlignment=HAlign_Fill, verticalAlignment=VAlign_Fill}}
  }
  StormIcon {
   widgetClassPath=&Image
   bIsVariable=true
   parent=Overlay_0
   widget=StormIcon
   properties {brush {
     bIsDynamicallyLoaded=false
     drawAs=Image
     imageType=NoImage
     mirroring=NoMirror
     resourceName=None
     resourceObject=/Game/Athena/HUD/Materials/T-UI-Icon-StormDMG
     tiling=NoTile
     imageSize=<80 80>
     margin=<ltrb 0 0 0 0>
     outlineSettings {
      bUseBrushTransparency=false
      roundingType=HalfHeightRadius
      width=0
      color {colorUseRule=UseColor_Specified, specifiedColor=<rgba 0 0 0 0>}
      cornerRadii=<0 0 0 0>
     }
     tintColor {colorUseRule=UseColor_Specified, specifiedColor=<rgba 1 1 1 1>}
     uVRegion {bIsValid=false, max=<0 0>, min=<0 0>}
    }}
   slot {class=OverlaySlot, properties {horizontalAlignment=HAlign_Center, verticalAlignment=VAlign_Center}}
  }
  Overlay_3 {
   widgetClassPath=&Overlay
   parent=VerticalBox_2
   widget=Overlay_3
   slot {class=VerticalBoxSlot, properties {horizontalAlignment=HAlign_Center, verticalAlignment=VAlign_Center}}
  }
  LightningBack {
   widgetClassPath=&Image
   bIsVariable=true
   parent=Overlay_3
   widget=LightningBack
   properties {
    brush {
     bIsDynamicallyLoaded=false
     drawAs=Image
     imageType=NoImage
     mirroring=NoMirror
     resourceName=None
     resourceObject=/Game/Athena/HUD/Materials/M_UI_LightningFX_Inst
     tiling=NoTile
     imageSize=<450 64>
     margin=<ltrb 0 0 0 0>
     outlineSettings {
      bUseBrushTransparency=false
      roundingType=HalfHeightRadius
      width=0
      color {colorUseRule=UseColor_Specified, specifiedColor=<rgba 0 0 0 0>}
      cornerRadii=<0 0 0 0>
     }
     tintColor {colorUseRule=UseColor_Specified, specifiedColor=<rgba 1 1 1 1>}
     uVRegion {bIsValid=false, max=<0 0>, min=<0 0>}
    }
    renderTransform {angle=0, scale=<1 1.125>, shear=<0 0>, translation=<0 0>}
   }
   slot {class=OverlaySlot, properties {horizontalAlignment=HAlign_Center, verticalAlignment=VAlign_Center, padding=<ltrb 4 2 4 2>}}
  }
  Border_0 {
   widgetClassPath=&Border
   parent=Overlay_3
   widget=Border_0
   properties {
    background {
     bIsDynamicallyLoaded=false
     drawAs=Box
     imageType=NoImage
     mirroring=NoMirror
     resourceName=None
     resourceObject=&M_UI_TextGlow_Add_Gradient_StormSurge
     tiling=NoTile
     imageSize=<64 64>
     margin=<ltrb 0.5 0.5 0.5 0.5>
     outlineSettings {
      bUseBrushTransparency=false
      roundingType=HalfHeightRadius
      width=0
      color {colorUseRule=UseColor_Specified, specifiedColor=<rgba 0 0 0 0>}
      cornerRadii=<0 0 0 0>
     }
     tintColor {colorUseRule=UseColor_Specified, specifiedColor=<rgba 1 1 1 1>}
     uVRegion {bIsValid=false, max=<0 0>, min=<0 0>}
    }
    brushColor=<rgba 1 1 1 0.0625>
    padding=<ltrb 24 12 24 12>
   }
   slot {class=OverlaySlot, properties {horizontalAlignment=HAlign_Center, verticalAlignment=VAlign_Center}}
  }
  AlertText {
   widgetClassPath=&CommonTextBlock
   bIsVariable=true
   parent=Border_0
   widget=AlertText
   properties {
    text="STORM SURGE WARNING!"
    font {
     bForceMonospaced=false
     bMaterialIsStencil=false
     fontMaterial=null
     fontObject=&Burbank
     letterSpacing=0
     monospacedWidth=1
     size=20
     skewAmount=0
     typefaceFontName=BurbankRegular-Black
     outlineSettings {
      bApplyOutlineToDropShadows=false
      bMiteredCorners=false
      bSeparateFillAlpha=false
      outlineMaterial=null
      outlineSize=3
      outlineColor=<rgba 0.0232 0 0.0307 0.25>
     }
    }
    shadowColorAndOpacity=<rgba 0.0232 0 0.0307 0.8>
    shadowOffset=<0 0>
   }
   slot {class=BorderSlot, properties {padding=<ltrb 24 12 24 12>}}
  }
  HelpText {
   widgetClassPath=&CommonTextBlock
   bIsVariable=true
   parent=VerticalBox_2
   widget=HelpText
   properties {
    text="Damage enemies to avoid elimination!"
    textTransformPolicy=ToUpper
    colorAndOpacity {colorUseRule=UseColor_Specified, specifiedColor=<rgba 1 0.5 1 1>}
    font {
     bForceMonospaced=false
     bMaterialIsStencil=false
     fontMaterial=null
     fontObject=&BurbankSmall
     letterSpacing=0
     monospacedWidth=1
     size=15
     skewAmount=0
     typefaceFontName=Black
     outlineSettings {
      bApplyOutlineToDropShadows=false
      bMiteredCorners=false
      bSeparateFillAlpha=false
      outlineMaterial=null
      outlineSize=2
      outlineColor=<rgba 0.0232 0 0.0307 0.125>
     }
    }
    renderTransform {angle=0, scale=<1 1>, shear=<-12 0>, translation=<0 0>}
    shadowOffset=<0 0>
   }
   slot {class=VerticalBoxSlot, properties {horizontalAlignment=HAlign_Center, verticalAlignment=VAlign_Bottom}}
  }
  TutorialHighlight {widgetClassPath=&TutorialHighlight}
 ]
}
