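# Skewed tab button widget on AthenaSkewedTabButton, with an uppercase title, slanted material background and an initially collapsed "!" badge.
# Hover events animate title scale and font-material Hover only when unselected; Unhovered plays in reverse and omits the scale overshoot.
# PlaySelectedAnimation runs Selected forward or backward, driving background opacity, horizontal scale and the Select parameter on background and title materials.
# SetTextColor writes RGB1 on the title's dynamic font material.
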WidgetBlueprint /Game/Athena/HUD/BP_SkewedTabButton
animations [
 Selected {
  binding_count=2
  length_seconds=0.15
  object=&self:Selected
  sequence {bindings [
    Background {
     id=#1
     possessed_class=Image
     tracks [
      {
       type=MovieSceneFloatTrack
       display_name=Render Opacity
       sections [{type=MovieSceneFloatSection, channels [None {type=MovieSceneScriptingFloatChannel, keys [[0,0],[3,1]]}]}]
      }
      {
       type=MovieScene2DTransformTrack
       sections [{
         type=MovieScene2DTransformSection
         channels [
          Translation.X {type=MovieSceneScriptingFloatChannel}
          Translation.Y {type=MovieSceneScriptingFloatChannel}
          Angle {type=MovieSceneScriptingFloatChannel}
          Scale.X {type=MovieSceneScriptingFloatChannel, keys [[0,0.75],[2,1.05],[3,1]]}
          Scale.Y {type=MovieSceneScriptingFloatChannel, const [1,2]}
          Shear.X {type=MovieSceneScriptingFloatChannel}
          Shear.Y {type=MovieSceneScriptingFloatChannel}
         ]
        }]
      }
      {
       type=MovieSceneWidgetMaterialTrack
       sections [{type=MovieSceneParameterSection, channels [Select {type=MovieSceneScriptingFloatChannel, keys [[0,0],[3,1]]}]}]
      }
     ]
    }
    Text_Title {
     id=#2
     possessed_class=CommonTextBlock
     tracks [{
       type=MovieSceneWidgetMaterialTrack
       display_name=Font.FontMaterial
       sections [{type=MovieSceneParameterSection, channels [Select {type=MovieSceneScriptingFloatChannel, keys [[0,0],[3,1]]}]}]
      }]
    }
   ]}
 }
 Hovered {
  binding_count=1
  length_seconds=0.15
  object=&self:Hovered
  sequence {bindings [Text_Title {
     id=#3
     display_name=CommonTextBlock_78
     possessed_class=Image
     tracks [
      {
       type=MovieSceneWidgetMaterialTrack
       display_name=Font.FontMaterial
       sections [{type=MovieSceneParameterSection, channels [Hover {type=MovieSceneScriptingFloatChannel, keys [[0,0],[3,1]]}]}]
      }
      {
       type=MovieScene2DTransformTrack
       sections [{
         type=MovieScene2DTransformSection
         channels [
          Translation.X {type=MovieSceneScriptingFloatChannel}
          Translation.Y {type=MovieSceneScriptingFloatChannel}
          Angle {type=MovieSceneScriptingFloatChannel}
          Scale.X {type=MovieSceneScriptingFloatChannel, keys [[0,0.9],[2,1.05],[3,1]]}
          Scale.Y {type=MovieSceneScriptingFloatChannel, keys [[0,0.9],[2,1.05],[3,1]]}
          Shear.X {type=MovieSceneScriptingFloatChannel}
          Shear.Y {type=MovieSceneScriptingFloatChannel}
         ]
        }]
      }
     ]
    }]}
 }
 Unhovered {
  binding_count=1
  length_seconds=0.15
  object=&self:Unhovered
  sequence {bindings [Text_Title {
     id=#3
     display_name=CommonTextBlock_78
     possessed_class=Image
     tracks [
      {
       type=MovieSceneWidgetMaterialTrack
       display_name=Font.FontMaterial
       sections [{type=MovieSceneParameterSection, channels [Hover {type=MovieSceneScriptingFloatChannel, keys [[0,0],[3,1]]}]}]
      }
      {
       type=MovieScene2DTransformTrack
       sections [{
         type=MovieScene2DTransformSection
         channels [
          Translation.X {type=MovieSceneScriptingFloatChannel}
          Translation.Y {type=MovieSceneScriptingFloatChannel}
          Angle {type=MovieSceneScriptingFloatChannel}
          Scale.X {type=MovieSceneScriptingFloatChannel, keys [[0,0.9],[3,1]]}
          Scale.Y {type=MovieSceneScriptingFloatChannel, keys [[0,0.9],[3,1]]}
          Shear.X {type=MovieSceneScriptingFloatChannel}
          Shear.Y {type=MovieSceneScriptingFloatChannel}
         ]
        }]
      }
     ]
    }]}
 }
]
graphs [
 EventGraph {nodes [
   Event On Hovered()
   n1 = GetSelected!(self:Common Button Base Object Reference) -> (ReturnValue:bool)
   > n2 = Branch(Condition=@n1)
   PlayAnimationForward(execute=@n2.else, InAnimation=Get Hovered() -> (Hovered:Widget Animation Object Reference))
   Event On Unhovered()
   > n3 = Branch(Condition=@n1)
   PlayAnimationReverse(execute=@n3.else, InAnimation=Get Unhovered() -> (Unhovered:Widget Animation Object Reference))
   n4 = Event PlaySelectedAnimation!() -> (OutputDelegate, then, bForward:bool) {is=title K2Node_Event}
   > n5 = Branch(Condition=@n4.bForward)
   n6 = Get Selected() -> (Selected:Widget Animation Object Reference)
   > PlayAnimationForward(InAnimation=@n6)
   PlayAnimationReverse(execute=@n5.else, InAnimation=@n6)
  ]}
 SetTextColor {nodes [
   n1 = SetTextColor!() -> (then, Color:Linear Color Structure) {is=title K2Node_FunctionEntry}
   > n2 = GetDynamicFontMaterial!(execute, self:Text Object Reference=Get Text_Title() -> (Text_Title:Common Text Object Reference)) -> (then, ReturnValue:Material Instance Dynamic Object Reference)
   > SetVectorParameterValue(self=@n2, ParameterName=RGB1, Value=@n1)
  ]}
]
variables []
widgets {
 info {inheritedWidgetCount=5, parentClass=/Script/FortniteUI.AthenaSkewedTabButton, rootWidgetClass=&SizeBox, widgetCount=10}
 widgets [
  SizeBox_0 {widgetClassPath=&SizeBox, widget=SizeBox_0, properties {bOverride_HeightOverride=true, heightOverride=70}}
  Overlay_0 {widgetClassPath=&Overlay, parent=SizeBox_0, widget=Overlay_0, slot {class=SizeBoxSlot}}
  Background {
   widgetClassPath=&Image
   bInherited=true
   bIsVariable=true
   parent=Overlay_0
   widget=Background
   properties {
    renderOpacity=0
    visibility=SelfHitTestInvisible
    brush {
     bIsDynamicallyLoaded=false
     drawAs=Box
     imageType=NoImage
     mirroring=NoMirror
     resourceName=None
     resourceObject=/Game/Athena/Collections/Art/M_Ui__SkewedTab_Inst
     tiling=NoTile
     imageSize=<64 64>
     margin=<ltrb 0.25 0 0.25 0>
     outlineSettings {
      bUseBrushTransparency=false
      roundingType=HalfHeightRadius
      width=0
      color {colorUseRule=UseColor_Specified, specifiedColor=<rgba 0 0 0 0>}
      cornerRadii=<0 0 0 0>
     }
     tintColor {colorUseRule=UseColor_Specified, specifiedColor=<rgba 1 1 1 1>}
     uVRegion {bIsValid=false, max=<0 0>, min=<0 0>}
    }
    renderTransform {angle=0, scale=<1 1>, shear=<-12 0>, translation=<0 0>}
    renderTransformPivot=<0.5 1>
   }
   slot {class=OverlaySlot, properties {horizontalAlignment=HAlign_Fill, verticalAlignment=VAlign_Fill, padding=<ltrb -16 0 -6 0>}}
  }
  HorizontalBox_0 {
   widgetClassPath=&HorizontalBox
   parent=Overlay_0
   widget=HorizontalBox_0
   slot {class=OverlaySlot, properties {verticalAlignment=VAlign_Center, padding=<ltrb 23 0 20 0>}}
  }
  Text_Title {
   widgetClassPath=&CommonTextBlock
   bInherited=true
   bIsVariable=true
   parent=HorizontalBox_0
   widget=Text_Title
   properties {
    mobileFontSizeMultiplier=1.4
    text="Quests && Punchards"
    textTransformPolicy=ToUpper
    font {
     bForceMonospaced=false
     bMaterialIsStencil=false
     fontMaterial=/Game/Athena/Collections/Art/M_Ui_MapScreenTabs_TextColor_Inst
     fontObject=&BurbankSmall
     letterSpacing=0
     monospacedWidth=1
     size=20
     skewAmount=0
     typefaceFontName=Black
     outlineSettings {
      bApplyOutlineToDropShadows=false
      bMiteredCorners=false
      bSeparateFillAlpha=false
      outlineMaterial=null
      outlineSize=0
      outlineColor=<rgba 0 0 0 1>
     }
    }
    renderTransform {angle=0, scale=<0.9 0.9>, shear=<-12 0>, translation=<0 0>}
    shadowOffset=<0 0>
   }
   slot {class=HorizontalBoxSlot}
  }
  Widget_Bang {
   widgetClassPath=&SizeBox
   bInherited=true
   bIsVariable=true
   parent=HorizontalBox_0
   widget=Widget_Bang
   properties {visibility=Collapsed, renderTransform {angle=0, scale=<1 1>, shear=<0 0>, translation=<4 -2>}}
   slot {class=HorizontalBoxSlot, properties {horizontalAlignment=HAlign_Center, verticalAlignment=VAlign_Center}}
  }
  Border_NewBang {
   widgetClassPath=&CommonBorder
   parent=Widget_Bang
   widget=Border_NewBang
   properties {
    horizontalAlignment=HAlign_Center
    background {
     bIsDynamicallyLoaded=false
     drawAs=Box
     imageType=NoImage
     mirroring=NoMirror
     resourceName=None
     resourceObject=/Game/Athena/UI/ReplayBrowser/M_Ui_Skewed_Vfor9slices_Bang
     tiling=NoTile
     imageSize=<32 32>
     margin=<ltrb 0.5 0 0.5 0>
     outlineSettings {
      bUseBrushTransparency=false
      roundingType=HalfHeightRadius
      width=0
      color {colorUseRule=UseColor_Specified, specifiedColor=<rgba 0 0 0 0>}
      cornerRadii=<0 0 0 0>
     }
     tintColor {colorUseRule=UseColor_Specified, specifiedColor=<rgba 0.9301 1 0.0097 1>}
     uVRegion {bIsValid=false, max=<0 0>, min=<0 0>}
    }
    padding=<ltrb 8 6 8 6>
   }
   slot {class=SizeBoxSlot, properties {horizontalAlignment=HAlign_Right, verticalAlignment=VAlign_Center}}
  }
  CommonTextBlock_209 {
   widgetClassPath=&CommonTextBlock
   parent=Border_NewBang
   widget=CommonTextBlock_209
   properties {
    lineHeightPercentage=0.7429
    text="!"
    colorAndOpacity {colorUseRule=UseColor_Specified, specifiedColor=<rgba 0 0 0 1>}
    font {
     bForceMonospaced=false
     bMaterialIsStencil=false
     fontMaterial=null
     fontObject=&Burbank
     letterSpacing=0
     monospacedWidth=1
     size=18
     skewAmount=0
     typefaceFontName=BurbankRegular-Black
     outlineSettings {
      bApplyOutlineToDropShadows=false
      bMiteredCorners=false
      bSeparateFillAlpha=false
      outlineMaterial=null
      outlineSize=0
      outlineColor=<rgba 0.9301 1 0.0097 1>
     }
    }
    margin=<ltrb 0 -6 0 0>
    renderTransform {angle=0, scale=<1 1>, shear=<-12 0>, translation=<0 0>}
   }
   slot {class=BorderSlot, properties {horizontalAlignment=HAlign_Center, padding=<ltrb 8 6 8 6>}}
  }
  LazyImage_Icon {widgetClassPath=/Script/CommonUI.CommonLazyImage, bInherited=true, widget=null}
  InputActionWidget {widgetClassPath=&CommonActionWidget, bInherited=true, widget=null}
 ]
}
