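# Creative HUD widget showing an "Awards XP" tag with an XP icon for content whose XP calibration phase is LiveXp.
# On Construct it listens for PlayerEnteredSpatialActor events, reads the phase from that actor's FortLevelSaveComponent, and hides itself unless LiveXp.
# SetVisualState sizes icon and spacers by isExpanded, which also controls the description in SetStatus.
# SetStatus ignores its phase input, leaving Calibrating and NoXp strings unused; design-time PreConstruct previews LiveXp.
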WidgetBlueprint /Game/Athena/HUD/Creative/Cr_XP_State
graphs [
 EventGraph {nodes [
   n1 = Event PreConstruct()
   > Branch(Condition=@n1.IsDesignTime)
   > SetVisualState!(execute, self, isVisible:bool=true) -> (then)
   > SetStatus!(execute, self, InCalibrationStatus:EDynamicXpCalibrationPhase Enum=LiveXp) -> (then)
   Event Construct()
   > n2 = StartListeningToEventBackwardCompatible!(
    execute:Exec
    Channel:Event Message Tag Structure=(TagName="EventMessage.Component.Playspace.PlayerEnteredSpatialActor")
    Target:Gameplay Event Router Component Object Reference=GetGlobalEventRouter()
    EventType:Script Struct Object Reference=/Script/FortniteGame.PlayerEnterSpatialActorContext
    Context:Object Reference
    EventBubblingRule:EEventBubblingRule Enum=Down
   ) -> (
    then:Exec
    AsyncTaskProxy:Async Action Start Listening To Event Object Reference
    OnEventReceived:Exec
    Payload:Player Enter Spatial Actor Context Structure
   ) {is=title K2Node_StartListeningToEventAsyncAction}
   n3 = GetComponentByClass!(
    self:Actor Object Reference=Break Player Enter Spatial Actor Context!(PlayerEnterSpatialActorContext:"Player Enter Spatial Actor Context Structure (by ref)"=@n2.Payload) -> (SpatialGameplayInterfaceActor:Actor Object Reference, EnteringPlayerState:Player State Object Reference) {is=title K2Node_BreakStruct}
    ComponentClass:Actor Component Class Reference=/Script/FortniteGame.FortLevelSaveComponent
   ) -> (ReturnValue:Fort Level Save Component Object Reference)
   Is Valid(exec=@n2.OnEventReceived, InputObject=@n3)
   n4 = Get Loaded Link Data XpCalibrationPhase!(self:Fort Level Save Component Object Reference=@n3) -> (CalibrationPhase:EDynamicXpCalibrationPhase Enum=None) {is=title K2Node_CallFunction}
   > SetVisualState(isVisible="Equal (Enum)"(A:EDynamicXpCalibrationPhase Enum=@n4, B:EDynamicXpCalibrationPhase Enum=LiveXp))
   > SetStatus(InCalibrationStatus=@n4)
  ]}
 SetStatus {nodes [
   SetStatus!() -> (then, InCalibrationStatus:EDynamicXpCalibrationPhase Enum=None) {is=title K2Node_FunctionEntry}
   > "SetText (Text)"(self=Get Text_Label() -> (Text_Label:Common Text Object Reference), InText=Get Status_Awarded() -> (Status_Awarded:Text))
   > n1 = Branch(Condition=NOT Boolean(A=Get isExpanded() -> (isExpanded:bool)))
   n2 = Get Text_Description() -> (Text_Description:Common Text Object Reference)
   > SetVisibility(self=@n2, InVisibility=Collapsed)
   SetVisibility(execute=@n1.else, self=@n2, InVisibility=SelfHitTestInvisible)
   > "SetText (Text)"(self=@n2, InText=Get Description_Awarded() -> (Description_Awarded:Text))
  ]}
 SetVisualState {nodes [
   n1 = SetVisualState!() -> (then, isVisible:bool=~) {is=title K2Node_FunctionEntry}
   > SetVisibility(InVisibility=Select(Option 0=Collapsed, Option 1=SelfHitTestInvisible, Index=@n1))
   > SetDesiredSizeOverride!(
    execute
    self:Image Object Reference=Get Image_XpState() -> (Image_XpState:Image Object Reference)
    DesiredSize:Vector 2D Structure=Select(
     Option 0:Vector 2D Structure=Get SmallIconSize() -> (SmallIconSize:Vector 2D Structure)
     Option 1:Vector 2D Structure=Get LargeIconSize() -> (LargeIconSize:Vector 2D Structure)
     Index=Get isExpanded()
    ) -> (ReturnValue:Vector 2D Structure)
   ) -> (then)
   n2 = BreakVector2D(InVec=Select(Option 0=Get SmallIconSize(), Option 1=Get LargeIconSize(), Index=Get isExpanded()))
   > SetSize!(
    execute
    self:Spacer Object Reference=[Get Spacer_Icon() -> (Spacer_Icon:Spacer Object Reference), Get Spacer_Text() -> (Spacer_Text:Spacer Object Reference)]
    InSize:Vector 2D Structure=MakeVector2D(X=float * float(A=@n2, B=0.5), Y=6)
   ) -> (then)
   n3 = Get isExpanded()
   > SetPadding!(
    execute
    self:Overlay Slot Object Reference=SlotAsOverlaySlot(Widget=Get Image_XpState())
    InPadding:Margin Structure=Make Margin(Left=0, Top=Select(Option 0:double=-12, Option 1:double=-24, Index=@n3) -> (ReturnValue:double), Right=0, Bottom=0)
   ) -> (then)
  ]}
]
variables [
 Description_Awarded {type=string, default=This content contains earnable XP}
 Description_Calibrating {type=string, default=Full XP will be awarded when calibration is complete.}
 Description_NoXp {type=string, default=XP cannot be gained from this content}
 isExpanded {type=boolean, default=true}
 LargeIconSize {type=Vector2D, default=<120 120>}
 SmallIconSize {type=Vector2D, default=<64 64>}
 Status_Awarded {type=string, default=Awards XP}
 Status_Calibrating {type=string, default=Calibrating}
 Status_NoXp {type=string, default=XP Unavailable}
]
widgets {
 info {inheritedWidgetCount=0, parentClass=&UserWidget, rootWidgetClass=&Overlay, widgetCount=12}
 widgets [
  Overlay_31 {widgetClassPath=&Overlay, widget=Overlay_31}
  SizeBox_0 {
   widgetClassPath=&SizeBox
   parent=Overlay_31
   widget=SizeBox_0
   properties {bOverride_MaxDesiredWidth=true, maxDesiredWidth=1000, widthOverride=640}
   slot {class=OverlaySlot}
  }
  HorizontalBox_2 {widgetClassPath=&HorizontalBox, parent=SizeBox_0, widget=HorizontalBox_2, slot {class=SizeBoxSlot}}
  Spacer_Icon {
   widgetClassPath=&Spacer
   bIsVariable=true
   parent=HorizontalBox_2
   widget=Spacer_Icon
   properties {size=<76.7767 6>}
   slot {class=HorizontalBoxSlot, properties {horizontalAlignment=HAlign_Left, verticalAlignment=VAlign_Center}}
  }
  TagContainer {
   widgetClassPath=&CommonBorder
   bIsVariable=true
   parent=HorizontalBox_2
   widget=TagContainer
   properties {
    verticalAlignment=VAlign_Center
    background {
     bIsDynamicallyLoaded=false
     drawAs=Box
     imageType=NoImage
     mirroring=NoMirror
     resourceName=None
     resourceObject=&/MI_UI_Creative_XP_Awarding_Tag
     tiling=NoTile
     imageSize=<32 32>
     margin=<ltrb 0.5 0 0.5 0>
     outlineSettings {
      bUseBrushTransparency=false
      roundingType=HalfHeightRadius
      width=0
      color {colorUseRule=UseColor_Specified, specifiedColor=<rgba 0 0 0 0>}
      cornerRadii=<0 0 0 0>
     }
     tintColor {colorUseRule=UseColor_Specified, specifiedColor=<rgba 1 1 1 1>}
     uVRegion {bIsValid=false, max=<0 0>, min=<0 0>}
    }
    padding=<ltrb -4 10 18 10>
   }
   slot {class=HorizontalBoxSlot, properties {verticalAlignment=VAlign_Center}}
  }
  HorizontalBox_0 {
   widgetClassPath=&HorizontalBox
   parent=TagContainer
   widget=HorizontalBox_0
   slot {class=BorderSlot, properties {verticalAlignment=VAlign_Center, padding=<ltrb -4 10 18 10>}}
  }
  Spacer_Text {
   widgetClassPath=&Spacer
   bIsVariable=true
   parent=HorizontalBox_0
   widget=Spacer_Text
   properties {size=<61.4305 16.2656>}
   slot {class=HorizontalBoxSlot, properties {horizontalAlignment=HAlign_Left, verticalAlignment=VAlign_Center}}
  }
  Image_CalibrationIcon {
   widgetClassPath=&Image
   bIsVariable=true
   parent=HorizontalBox_0
   widget=Image_CalibrationIcon
   properties {
    visibility=Collapsed
    brush {
     bIsDynamicallyLoaded=false
     drawAs=Image
     imageType=NoImage
     mirroring=NoMirror
     resourceName=None
     resourceObject=/Game/UI/SocialPanel/Art/Textures/T_UI_Icon_LocalUserSettings
     tiling=NoTile
     imageSize=<20 20>
     margin=<ltrb 0 0 0 0>
     outlineSettings {
      bUseBrushTransparency=false
      roundingType=HalfHeightRadius
      width=0
      color {colorUseRule=UseColor_Specified, specifiedColor=<rgba 0 0 0 0>}
      cornerRadii=<0 0 0 0>
     }
     tintColor {colorUseRule=UseColor_Specified, specifiedColor=<rgba 1 1 1 1>}
     uVRegion {bIsValid=false, max=<0 0>, min=<0 0>}
    }
   }
   slot {class=HorizontalBoxSlot, properties {horizontalAlignment=HAlign_Left, verticalAlignment=VAlign_Top}}
  }
  VerticalBox_0 {
   widgetClassPath=&VerticalBox
   parent=HorizontalBox_0
   widget=VerticalBox_0
   slot {class=HorizontalBoxSlot, properties {verticalAlignment=VAlign_Center, padding=<ltrb 4 2 4 2>, size {sizeRule=Fill, value=1}}}
  }
  Text_Label {
   widgetClassPath=&CommonTextBlock
   bIsVariable=true
   parent=VerticalBox_0
   widget=Text_Label
   properties {
    mobileFontSizeMultiplier=1.2
    text="[Status]"
    textTransformPolicy=ToUpper
    visibility=HitTestInvisible
    font {
     bForceMonospaced=false
     bMaterialIsStencil=false
     fontMaterial=null
     fontObject=&BurbankSmall
     letterSpacing=0
     monospacedWidth=1
     size=15
     skewAmount=0
     typefaceFontName=Black
     outlineSettings {
      bApplyOutlineToDropShadows=false
      bMiteredCorners=false
      bSeparateFillAlpha=false
      outlineMaterial=null
      outlineSize=0
      outlineColor=<rgba 0 0 0 1>
     }
    }
    margin=<ltrb 0 -4.5 0 -5>
    renderTransform {angle=0, scale=<1 1>, shear=<-12 0>, translation=<0 0>}
    shadowOffset=<0 0>
   }
   slot {class=VerticalBoxSlot}
  }
  Text_Description {
   widgetClassPath=&CommonTextBlock
   bIsVariable=true
   parent=VerticalBox_0
   widget=Text_Description
   properties {
    autoWrapText=true
    mobileFontSizeMultiplier=1.2
    text=Full XP will be awarded when calibration is complete.
    textTransformPolicy=ToUpper
    visibility=HitTestInvisible
    colorAndOpacity {colorUseRule=UseColor_Specified, specifiedColor=<rgba 0.2159 0.6584 1 1>}
    font {
     bForceMonospaced=false
     bMaterialIsStencil=false
     fontMaterial=null
     fontObject=&BurbankSmall
     letterSpacing=0
     monospacedWidth=1
     size=15
     skewAmount=0
     typefaceFontName=Black
     outlineSettings {
      bApplyOutlineToDropShadows=false
      bMiteredCorners=false
      bSeparateFillAlpha=false
      outlineMaterial=null
      outlineSize=0
      outlineColor=<rgba 0 0 0 1>
     }
    }
    margin=<ltrb 0 -4.5 0 -5>
    renderTransform {angle=0, scale=<1 1>, shear=<-12 0>, translation=<0 0>}
    shadowOffset=<0 0>
   }
   slot {class=VerticalBoxSlot, properties {padding=<ltrb 0 12 0 4>}}
  }
  Image_XpState {
   widgetClassPath=&Image
   bIsVariable=true
   parent=Overlay_31
   widget=Image_XpState
   properties {
    visibility=HitTestInvisible
    brush {
     bIsDynamicallyLoaded=false
     drawAs=Image
     imageType=FullColor
     mirroring=NoMirror
     resourceName=None
     resourceObject=/Game/UI/Foundation/Textures/Icons/Items/UniversalXP/T_UI_UniversalXP_Icon_Standard_64
     tiling=NoTile
     imageSize=<56 56>
     margin=<ltrb 0 0 0 0>
     outlineSettings {
      bUseBrushTransparency=false
      roundingType=HalfHeightRadius
      width=0
      color {colorUseRule=UseColor_Specified, specifiedColor=<rgba 0 0 0 0>}
      cornerRadii=<0 0 0 0>
     }
     tintColor {colorUseRule=UseColor_Specified, specifiedColor=<rgba 1 1 1 1>}
     uVRegion {bIsValid=false, max=<0 0>, min=<0 0>}
    }
   }
   slot {class=OverlaySlot, properties {padding=<ltrb 0 -24 0 0>}}
  }
 ]
}
