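# Widget Blueprint for the Creative island info panel in the HUD, showing name, creator, bullet details and start-game and support-creator buttons.
# Enabling panel inputs shows Support Creator unless disabled by runtime options; it opens the affiliate panel asynchronously.
# Global event-router listeners track the local player entering or exiting minigame volumes, binding state changes that refresh the start button and details toggle.
# Widget animations drive the intro and details show/hide.
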WidgetBlueprint /Game/Athena/HUD/CreativeMiniGameInfoWidget
animations [
 AnimShowDetails {
  binding_count=7
  length_seconds=0.4
  object=&self:AnimShowDetails
  sequence {
   bindings [
    BulletBox1 {
     id=#1
     possessed_class=HorizontalBox
     tracks [
      {
       type=MovieSceneFloatTrack
       display_name=Render Opacity
       sections [{type=MovieSceneFloatSection, end_frame=23, start_frame=0, channels [None {type=MovieSceneScriptingFloatChannel, keys [[3,0],[18,1]]}]}]
      }
      {
       type=MovieScene2DTransformTrack
       sections [{
         type=MovieScene2DTransformSection
         end_frame=23
         start_frame=0
         channels [
          Translation.X {type=MovieSceneScriptingFloatChannel, keys [[3,-128],[18,0]]}
          Translation.Y {type=MovieSceneScriptingFloatChannel}
          Angle {type=MovieSceneScriptingFloatChannel}
          Scale.X {type=MovieSceneScriptingFloatChannel}
          Scale.Y {type=MovieSceneScriptingFloatChannel}
          Shear.X {type=MovieSceneScriptingFloatChannel}
          Shear.Y {type=MovieSceneScriptingFloatChannel}
         ]
        }]
      }
     ]
    }
    GameDetailsContent {
     id=#2
     possessed_class=Overlay
     tracks [{
       type=MovieSceneFloatTrack
       display_name=Render Opacity
       sections [{type=MovieSceneFloatSection, end_frame=24, start_frame=0, channels [None {type=MovieSceneScriptingFloatChannel, const [1,1]}]}]
      }]
    }
    BackingHeader {
     id=#3
     possessed_class=Image
     tracks [
      {
       type=MovieSceneWidgetMaterialTrack
       sections [{
         type=MovieSceneParameterSection
         end_frame=24
         start_frame=0
         channels [CenterOffset {type=MovieSceneScriptingFloatChannel, keys [[6,0.625],[21,-0.1333]]}]
        }]
      }
      {
       type=MovieSceneFloatTrack
       display_name=Render Opacity
       sections [{type=MovieSceneFloatSection, end_frame=24, start_frame=0, channels [None {type=MovieSceneScriptingFloatChannel, keys [[6,0],[21,1]]}]}]
      }
     ]
    }
    BackingFooter {
     id=#4
     possessed_class=Image
     tracks [
      {
       type=MovieSceneFloatTrack
       display_name=Render Opacity
       sections [{type=MovieSceneFloatSection, end_frame=24, start_frame=0, channels [None {type=MovieSceneScriptingFloatChannel, keys [[15,0],[24,0.5]]}]}]
      }
      {
       type=MovieScene2DTransformTrack
       sections [{
         type=MovieScene2DTransformSection
         end_frame=24
         start_frame=0
         channels [
          Translation.X {type=MovieSceneScriptingFloatChannel}
          Translation.Y {type=MovieSceneScriptingFloatChannel}
          Angle {type=MovieSceneScriptingFloatChannel}
          Scale.X {type=MovieSceneScriptingFloatChannel}
          Scale.Y {type=MovieSceneScriptingFloatChannel}
          Shear.X {type=MovieSceneScriptingFloatChannel}
          Shear.Y {type=MovieSceneScriptingFloatChannel}
         ]
        }]
      }
     ]
    }
    DetailsBacking {
     id=#5
     possessed_class=Image
     tracks [
      {
       type=MovieSceneWidgetMaterialTrack
       sections [{type=MovieSceneParameterSection, end_frame=23, start_frame=0, channels [Scale {type=MovieSceneScriptingFloatChannel, keys [[0,0],[18,1]]}]}]
      }
      {
       type=MovieSceneFloatTrack
       display_name=Render Opacity
       sections [{type=MovieSceneFloatSection, end_frame=23, start_frame=0, channels [None {type=MovieSceneScriptingFloatChannel, keys [[0,0],[18,1]]}]}]
      }
     ]
    }
    BulletBox3 {
     id=#6
     possessed_class=HorizontalBox
     tracks [
      {
       type=MovieSceneFloatTrack
       display_name=Render Opacity
       sections [{type=MovieSceneFloatSection, end_frame=24, start_frame=0, channels [None {type=MovieSceneScriptingFloatChannel, keys [[9,0],[24,1]]}]}]
      }
      {
       type=MovieScene2DTransformTrack
       sections [{
         type=MovieScene2DTransformSection
         end_frame=24
         start_frame=0
         channels [
          Translation.X {type=MovieSceneScriptingFloatChannel, keys [[9,-128],[24,0]]}
          Translation.Y {type=MovieSceneScriptingFloatChannel}
          Angle {type=MovieSceneScriptingFloatChannel}
          Scale.X {type=MovieSceneScriptingFloatChannel}
          Scale.Y {type=MovieSceneScriptingFloatChannel}
          Shear.X {type=MovieSceneScriptingFloatChannel}
          Shear.Y {type=MovieSceneScriptingFloatChannel}
         ]
        }]
      }
     ]
    }
    BulletBox2 {
     id=#7
     possessed_class=HorizontalBox
     tracks [
      {
       type=MovieSceneFloatTrack
       display_name=Render Opacity
       sections [{type=MovieSceneFloatSection, end_frame=23, start_frame=0, channels [None {type=MovieSceneScriptingFloatChannel, keys [[6,0],[21,1]]}]}]
      }
      {
       type=MovieScene2DTransformTrack
       sections [{
         type=MovieScene2DTransformSection
         end_frame=23
         start_frame=0
         channels [
          Translation.X {type=MovieSceneScriptingFloatChannel, keys [[6,-128],[21,0]]}
          Translation.Y {type=MovieSceneScriptingFloatChannel}
          Angle {type=MovieSceneScriptingFloatChannel}
          Scale.X {type=MovieSceneScriptingFloatChannel}
          Scale.Y {type=MovieSceneScriptingFloatChannel}
          Shear.X {type=MovieSceneScriptingFloatChannel}
          Shear.Y {type=MovieSceneScriptingFloatChannel}
         ]
        }]
      }
     ]
    }
   ]
   display_rate {numerator=60}
  }
 }
 AnimHideDetails {
  binding_count=5
  length_seconds=0.4
  object=&self:AnimHideDetails
  sequence {
   bindings [
    BulletBox3 {
     id=#8
     possessed_class=HorizontalBox
     tracks [{
       type=MovieSceneFloatTrack
       display_name=Render Opacity
       sections [{type=MovieSceneFloatSection, end_frame=23, start_frame=0, channels [None {type=MovieSceneScriptingFloatChannel, keys [[0,1],[12,0]]}]}]
      }]
    }
    BulletBox1 {
     id=#9
     possessed_class=HorizontalBox
     tracks [{
       type=MovieSceneFloatTrack
       display_name=Render Opacity
       sections [{type=MovieSceneFloatSection, end_frame=23, start_frame=0, channels [None {type=MovieSceneScriptingFloatChannel, keys [[6,1],[18,0]]}]}]
      }]
    }
    BackingHeader {
     id=#10
     possessed_class=Image
     tracks [
      {
       type=MovieSceneWidgetMaterialTrack
       sections [{
         type=MovieSceneParameterSection
         end_frame=23
         start_frame=0
         channels [CenterOffset {type=MovieSceneScriptingFloatChannel, keys [[0,-0.1333],[15,-0.5633]]}]
        }]
      }
      {
       type=MovieSceneFloatTrack
       display_name=Render Opacity
       sections [{type=MovieSceneFloatSection, end_frame=24, start_frame=0, channels [None {type=MovieSceneScriptingFloatChannel, keys [[0,1],[17.999,0]]}]}]
      }
     ]
    }
    BulletBox2 {
     id=#11
     possessed_class=HorizontalBox
     tracks [{
       type=MovieSceneFloatTrack
       display_name=Render Opacity
       sections [{type=MovieSceneFloatSection, end_frame=23, start_frame=0, channels [None {type=MovieSceneScriptingFloatChannel, keys [[3,1],[15,0]]}]}]
      }]
    }
    GameDetailsContent {
     id=#12
     possessed_class=Overlay
     tracks [{
       type=MovieSceneFloatTrack
       display_name=Render Opacity
       sections [{type=MovieSceneFloatSection, end_frame=24, start_frame=0, channels [None {type=MovieSceneScriptingFloatChannel, keys [[6,1],[24,0]]}]}]
      }]
    }
   ]
   display_rate {numerator=60}
  }
 }
 Intro {
  binding_count=6
  length_seconds=0.5
  object=&self:Intro
  sequence {bindings [
    CreativeLogoWatermark {
     id=#13
     possessed_class=Image
     tracks [
      {
       type=MovieScene2DTransformTrack
       sections [{
         type=MovieScene2DTransformSection
         end_frame=10
         start_frame=0
         channels [
          Translation.X {type=MovieSceneScriptingFloatChannel}
          Translation.Y {type=MovieSceneScriptingFloatChannel}
          Angle {type=MovieSceneScriptingFloatChannel}
          Scale.X {type=MovieSceneScriptingFloatChannel, keys [[3,3],[7,0.75],[8,1.05],[10,1]]}
          Scale.Y {type=MovieSceneScriptingFloatChannel, keys [[3,3],[7,0.75],[8,1.05],[10,1]]}
          Shear.X {type=MovieSceneScriptingFloatChannel}
          Shear.Y {type=MovieSceneScriptingFloatChannel}
         ]
        }]
      }
      {
       type=MovieSceneFloatTrack
       display_name=Render Opacity
       sections [{type=MovieSceneFloatSection, end_frame=10, start_frame=0, channels [None {type=MovieSceneScriptingFloatChannel, keys [[3,0],[5,1]]}]}]
      }
     ]
    }
    HorizontalBox_3 {
     id=#14
     possessed_class=HorizontalBox
     tracks [
      {
       type=MovieScene2DTransformTrack
       sections [{
         type=MovieScene2DTransformSection
         end_frame=10
         start_frame=0
         channels [
          Translation.X {type=MovieSceneScriptingFloatChannel, keys [[3,-128],[7,0]]}
          Translation.Y {type=MovieSceneScriptingFloatChannel}
          Angle {type=MovieSceneScriptingFloatChannel}
          Scale.X {type=MovieSceneScriptingFloatChannel}
          Scale.Y {type=MovieSceneScriptingFloatChannel}
          Shear.X {type=MovieSceneScriptingFloatChannel}
          Shear.Y {type=MovieSceneScriptingFloatChannel}
         ]
        }]
      }
      {
       type=MovieSceneFloatTrack
       display_name=Render Opacity
       sections [{type=MovieSceneFloatSection, end_frame=10, start_frame=0, channels [None {type=MovieSceneScriptingFloatChannel, keys [[3,0],[7,1]]}]}]
      }
     ]
    }
    Text_Name {
     id=#15
     possessed_class=CommonTextBlock
     tracks [{
       type=MovieSceneFloatTrack
       display_name=Render Opacity
       sections [{type=MovieSceneFloatSection, end_frame=10, start_frame=0, channels [None {type=MovieSceneScriptingFloatChannel, keys [[2,0],[10,1]]}]}]
      }]
    }
    CreatorBox {
     id=#16
     possessed_class=HorizontalBox
     tracks [
      {
       type=MovieScene2DTransformTrack
       sections [{
         type=MovieScene2DTransformSection
         end_frame=10
         start_frame=0
         channels [
          Translation.X {type=MovieSceneScriptingFloatChannel, keys [[2,-128],[6,0]]}
          Translation.Y {type=MovieSceneScriptingFloatChannel}
          Angle {type=MovieSceneScriptingFloatChannel}
          Scale.X {type=MovieSceneScriptingFloatChannel}
          Scale.Y {type=MovieSceneScriptingFloatChannel}
          Shear.X {type=MovieSceneScriptingFloatChannel, const [-8,1]}
          Shear.Y {type=MovieSceneScriptingFloatChannel, const [0,1]}
         ]
        }]
      }
      {
       type=MovieSceneFloatTrack
       display_name=Render Opacity
       sections [{type=MovieSceneFloatSection, end_frame=10, start_frame=0, channels [None {type=MovieSceneScriptingFloatChannel, keys [[2,0],[6,1]]}]}]
      }
     ]
    }
    StartGameButton {
     id=#17
     possessed_class=IconTextButton_C
     tracks [
      {
       type=MovieScene2DTransformTrack
       sections [{
         type=MovieScene2DTransformSection
         end_frame=10
         start_frame=0
         channels [
          Translation.X {type=MovieSceneScriptingFloatChannel}
          Translation.Y {type=MovieSceneScriptingFloatChannel, keys [[4,128],[9,0]]}
          Angle {type=MovieSceneScriptingFloatChannel}
          Scale.X {type=MovieSceneScriptingFloatChannel, keys [[4,0.75],[8,1.05],[10,1]]}
          Scale.Y {type=MovieSceneScriptingFloatChannel}
          Shear.X {type=MovieSceneScriptingFloatChannel}
          Shear.Y {type=MovieSceneScriptingFloatChannel}
         ]
        }]
      }
      {
       type=MovieSceneFloatTrack
       display_name=Render Opacity
       sections [{type=MovieSceneFloatSection, end_frame=10, start_frame=0, channels [None {type=MovieSceneScriptingFloatChannel, keys [[4,0],[9,1]]}]}]
      }
     ]
    }
    Toggle_Details {
     id=#18
     possessed_class=SizeBox
     tracks [{
       type=MovieSceneFloatTrack
       display_name=Render Opacity
       sections [{type=MovieSceneFloatSection, end_frame=10, start_frame=0, channels [None {type=MovieSceneScriptingFloatChannel, keys [[0,0],[4,0],[10,1]]}]}]
      }]
    }
   ]}
 }
]
graphs [
 EnablePanelInputs {nodes [
   n1 = EnablePanelInputs!() -> (then, InputEnabled:bool=~) {is=title K2Node_FunctionEntry}
   > n2 = Set bInputEnabled(bInputEnabled:bool=@n1) -> (Output_Get:bool)
   > n3 = Branch(Condition=@n2)
   > PushPanelInputComponent(self:Fort Creative Mini Game Info Object Reference)
   > UpdateStartButtonVisibility()
   n4 = GetRuntimeOptions()
   n5 = NOT Boolean(A=OR Boolean(
     A=Get bDisableAffiliateFeature!(self:Fort Runtime Options Object Reference=@n4) -> (bDisableAffiliateFeature:bool)
     B=Get bDisableCreativeSACFeature!(self:Fort Runtime Options Object Reference=@n4) -> (bDisableCreativeSACFeature:bool)
    ))
   > Branch(Condition=@n5)
   n6 = Create Event()
   > Bind Event to On Support Creator(self:Creative Mini Game Info Widget Object Reference, Delegate=@n6) {is=title K2Node_AddDelegate}
   > SetVisibility(self=Get SupportAffliateButton() -> (SupportAffliateButton:Input Reflector Button Object Reference), InVisibility=Visible)
   > SetVisibility(self=Get SACBox() -> (SACBox:Vertical Box Object Reference), InVisibility=Visible)
   n7 = Create Event()
   > Bind Event to On Toggle Details(self:Fort Creative Mini Game Info Object Reference, Delegate=@n7) {is=title K2Node_AddDelegate}
   PopPanelInputComponent(execute=@n3.else, self:Fort Creative Mini Game Info Object Reference)
   > UpdateStartButtonVisibility()
   > Branch(Condition=@n5)
   > Unbind Event from On Support Creator(self:Creative Mini Game Info Widget Object Reference, Delegate=@n6) {is=title K2Node_RemoveDelegate}
   > Unbind Event from On Toggle Details(self:Fort Creative Mini Game Info Object Reference, Delegate=@n7) {is=title K2Node_RemoveDelegate}
  ]}
 EventGraph {nodes [
   "On Button Clicked (SupportAffliateButton)"!() -> (OutputDelegate, then, Button:Common Button Legacy Object Reference) {is=title K2Node_ComponentBoundEvent}
   > SupportCreator()
   "On Button Clicked (StartGameButton)"!() -> (OutputDelegate, then, Button:Common Button Legacy Object Reference) {is=title K2Node_ComponentBoundEvent}
   > n1 = Cast To FortGameStateAthena!(execute, Object=GetGameState!() -> (ReturnValue:Game State Base Object Reference)) -> (then, CastFailed, AsFort Game State Athena:Fort Game State Athena Object Reference) {is=title K2Node_DynamicCast}
   > n2 = Branch(Condition=IsCreativeLTMPlaylist!(self:Fort Game State Athena Object Reference=@n1) -> (ReturnValue:bool))
   n3 = Get FortHUDContext!() -> (ReturnValue:Fort HUDContext Object Reference) {is=title K2Node_GetBlueprintContextEx}
   Branch(
    execute=[@n2.else, @n1.CastFailed]
    Condition=NOT Boolean(A=IsInCursorMode!(self:Fort HUDContext Object Reference=@n3) -> (ReturnValue:bool))
   )
   > n4 = Get Player Controller Athena!(execute:Exec) -> (then:Exec, CastFailed:Exec, AsFortPlayerController:Fort Player Controller Athena Object Reference) {is=title K2Node_MacroInstance}
   > ServerStartMinigame(self:Fort Player Controller Athena Object Reference=@n4)
   > Call On Starting Game!(execute:Exec, self:Creative Mini Game Info Widget Object Reference) -> (then:Exec) {is=title K2Node_CallDelegate}
   n5 = "On Button Clicked (HideDetailsButton)"!() -> (OutputDelegate, then, Button:Common Button Legacy Object Reference) {is=title K2Node_ComponentBoundEvent}
   ToggleInfo(execute=[@n5.then, @n6.then])
   n6 = "On Button Clicked (ShowDetailsButton)"!() -> (OutputDelegate, then, Button:Common Button Legacy Object Reference) {is=title K2Node_ComponentBoundEvent}
   SupportCreator() {is=title K2Node_CustomEvent}
   n7 = Get Text_CreatorName() -> (Text_CreatorName:Common Text Object Reference)
   n8 = Contains Item(
    TargetArray:Array of Strings=Get BannedNames() -> (BannedNames:Array of Strings)
    ItemToFind:"String (by ref)"="To String (Text)"(InText="GetText (Text)"!(self:Text Object Reference=@n7) -> (ReturnValue:Text))
   )
   n9 = IsInCursorMode(self=Get FortHUDContext())
   n10 = Get bDisplayingAffiliatePanel() -> (bDisplayingAffiliatePanel:bool)
   > Branch(Condition=AND Boolean!(A=NOT Boolean(A=@n9), B=NOT Boolean(A=@n10), C:bool=NOT Boolean(A=@n8)) -> (ReturnValue))
   > n11 = CreateWidgetAsync!(
    execute:Exec
    UserWidgetSoftClass:User Widget Soft Class Reference=/Game/Athena/UI/InGame/Creative/CreativeAffiliatePanel.CreativeAffiliatePanel_C
    OwningPlayer:Player Controller Object Reference
    bSuspendInputUntilComplete:bool=true
   ) -> (then:Exec, OnComplete:Exec, OnFailure:Exec, UserWidget:User Widget Object Reference) {is=title K2Node_AsyncAction}
   > PopPanelInputComponent()
   n12 = Cast To CreativeAffiliatePanel!(execute=@n11.OnComplete, Object=@n11) -> (then, CastFailed, AsCreative Affiliate Panel:Creative Affiliate Panel Object Reference) {is=title K2Node_DynamicCast}
   > n13 = Set AffiliatePanel(AffiliatePanel:Creative Affiliate Panel Object Reference=@n12) -> (Output_Get:Creative Affiliate Panel Object Reference)
   > InitFromObject!(execute, self:Creative Affiliate Panel Object Reference=@n13, InitObject:Object Reference=Self-Reference()) -> (then)
   > SetAffiliateName!(
    execute
    self:Creative Affiliate Panel Object Reference=@n13
    AffiliateName:Text="GetText (Text)"(self=Get Text_SupportCode() -> (Text_SupportCode:Common Text Object Reference))
    Can Edit:bool=true
   ) -> (then)
   > SetSourceName!(execute, self:Creative Affiliate Panel Object Reference=@n13, NewSourceName:String=Get Source Name() -> (Source Name:String)) -> (then)
   > SetVisibility(InVisibility=Collapsed)
   > Bind Event to On Panel Close(self:Creative Affiliate Panel Object Reference=@n13, Delegate=Create Event()) {is=title K2Node_AddDelegate}
   > Set bDisplayingAffiliatePanel(bDisplayingAffiliatePanel:bool=true) -> (Output_Get:bool)
   AffiliatePanelClosed() {is=title K2Node_CustomEvent}
   > Set bDisplayingAffiliatePanel(bDisplayingAffiliatePanel=false)
   > PushPanelInputComponent()
   > SetVisibility(InVisibility=SelfHitTestInvisible)
   CollapseDetailsAfterAnim() {is=title K2Node_CustomEvent}
   > SetVisibility(self=Get GameDetailsContent() -> (GameDetailsContent:Overlay Object Reference), InVisibility=Hidden)
   > SetVisibility(self=Get XpState_Compact_Overlay() -> (XpState_Compact_Overlay:Overlay Object Reference), InVisibility=HitTestInvisible)
   BindVolumeEvents() {is=title K2Node_CustomEvent}
   > n14 = StartListeningToEventBackwardCompatible!(
    execute:Exec
    Channel:Event Message Tag Structure=(TagName="EventMessage.Component.Playspace.PlayerEnteredSpatialActor.LegacyMinigame")
    Target:Gameplay Event Router Component Object Reference=GetGlobalEventRouter()
    EventType:Script Struct Object Reference=/Script/FortniteGame.PlayerEnterSpatialActorContextLegacyMinigame
    Context:Object Reference
    EventBubblingRule:EEventBubblingRule Enum=Down
   ) -> (
    then:Exec
    AsyncTaskProxy:Async Action Start Listening To Event Object Reference
    OnEventReceived:Exec
    Payload:Player Enter Spatial Actor Context Legacy Minigame Structure
   ) {is=title K2Node_StartListeningToEventAsyncAction}
   > Set PlayerEnteredHandle(PlayerEnteredHandle:Async Action Start Listening To Event Object Reference=@n14) -> (Output_Get:Async Action Start Listening To Event Object Reference)
   > n15 = StartListeningToEventBackwardCompatible(
    Channel=(TagName="EventMessage.Component.Playspace.PlayerExitedSpatialActor.LegacyMinigame")
    Target=GetGlobalEventRouter()
    EventType=/Script/FortniteGame.PlayerExitSpatialActorContextLegacyMinigame
    EventBubblingRule=Down
   ) -> (Payload:Player Exit Spatial Actor Context Legacy Minigame Structure)
   > Set ExitedEventHandle(ExitedEventHandle:Async Action Start Listening To Event Object Reference=@n15) -> (Output_Get:Async Action Start Listening To Event Object Reference)
   n16 = Break Player Exit Spatial Actor Context Legacy Minigame!(PlayerExitSpatialActorContextLegacyMinigame:"Player Exit Spatial Actor Context Legacy Minigame Structure (by ref)"=@n15.Payload) -> (SpatialGameplayInterfaceActor:Actor Object Reference, ExitingPlayerState:Player State Object Reference) {is=title K2Node_BreakStruct}
   OnExitVolume!(
    execute=@n15.OnEventReceived
    self
    ExitingPlayerState:Player State Object Reference=@n16.ExitingPlayerState
    ExitedMinigame:Fort Minigame Object Reference=GetMinigame!(ContextActor:Actor Object Reference=@n16) -> (ReturnValue:Fort Minigame Object Reference)
   ) -> (then)
   n17 = Break Player Enter Spatial Actor Context Legacy Minigame!(PlayerEnterSpatialActorContextLegacyMinigame:"Player Enter Spatial Actor Context Legacy Minigame Structure (by ref)"=@n14.Payload) -> (SpatialGameplayInterfaceActor:Actor Object Reference, EnteringPlayerState:Player State Object Reference) {is=title K2Node_BreakStruct}
   OnEnterVolume!(
    execute=@n14.OnEventReceived
    self
    EnteringPlayerState:Player State Object Reference=@n17.EnteringPlayerState
    EnteredMinigame:Fort Minigame Object Reference=GetMinigame(ContextActor=@n17)
   ) -> (then)
   Event OnTextSet() {is=title K2Node_Event}
   n18 = Get Text_SupportCode()
   > Hide if Text Empty!(
    execute
    self
    TextToCheck:Text Object Reference=@n18
    ElementToHide:Widget Object Reference=Get SaC_Button() -> (SaC_Button:Size Box Object Reference)
   ) -> (then)
   > Hide if Text Empty(TextToCheck=@n18, ElementToHide=Get SupportCodeBox() -> (SupportCodeBox:Horizontal Box Object Reference))
   > Hide if Text Empty(TextToCheck=Get Text_CreatorName(), ElementToHide=Get CreatorBox() -> (CreatorBox:Horizontal Box Object Reference))
   > Hide if Text Empty(
    TextToCheck=Get Text_Bullet1() -> (Text_Bullet1:Common Text Object Reference)
    ElementToHide=Get BulletBox1() -> (BulletBox1:Horizontal Box Object Reference)
   )
   > Hide if Text Empty(
    TextToCheck=Get Text_Bullet2() -> (Text_Bullet2:Common Text Object Reference)
    ElementToHide=Get BulletBox2() -> (BulletBox2:Horizontal Box Object Reference)
   )
   > Hide if Text Empty(
    TextToCheck=Get Text_Bullet3() -> (Text_Bullet3:Common Text Object Reference)
    ElementToHide=Get BulletBox3() -> (BulletBox3:Horizontal Box Object Reference)
   )
   n19 = "GetText (Text)"(self=Get Text_CreatorName())
   n20 = Contains Item(TargetArray=Get BannedNames(), ItemToFind="To String (Text)"(InText=@n19))
   n21 = "Equal, Case Insensitive (Text)"!(A:"Text (by ref)"=@n19, B:"Text (by ref)"="GetText (Text)"(self=Get Text_SupportCode())) -> (ReturnValue:bool)
   > SetVisibility(self=Get SupportCodeBox(), InVisibility=Select(Option 0=SelfHitTestInvisible, Option 1=Collapsed, Index=OR Boolean(A=@n20, B=@n21)))
   > SetVisibility(
    self=Get Toggle_Details() -> (Toggle_Details:Size Box Object Reference)
    InVisibility=Select(
     Option 0=Collapsed
     Option 1=SelfHitTestInvisible
     Index=AreDetailsToggleable!(self:Fort Creative Mini Game Info Object Reference) -> (ReturnValue:bool)
    )
   )
   Event Construct()
   > InitProgressDynamicMaterial()
   n22 = Get SupportAffliateButton()
   > SetMouseKeyboardStyle!(
    execute
    self:Input Reflector Button Object Reference=@n22
    ControllerInputStyle:Common Button Style Class Reference=/FortUILibrary/Foundation/Buttons/ButtonStyles/Athena/Athena_ButtonStyle_AngledGreenMenuButton.Athena_ButtonStyle_AngledGreenMenuButton_C
   ) -> (then)
   > SetControllerStyle!(
    execute
    self:Input Reflector Button Object Reference=@n22
    ControllerInputStyle:Common Button Style Class Reference=&Athena_ButtonStyle_AngledGreenMenuButton_C
   ) -> (then)
   > SetVisibility(self=[Get SACBox(), Get StartGameButton() -> (StartGameButton:Icon Text Button Object Reference), @n22], InVisibility=Collapsed)
   > SetRenderScale!(
    execute
    self:Widget Object Reference=Get InputActionWidget!(self:Common Button Base Object Reference=@n22) -> (InputActionWidget:Common Action Widget Object Reference)
    Scale:Vector 2D Structure=(X=0.8,Y=0.8)
   ) -> (then)
   > UpdateStartButtonVisibility()
   > SetActiveWidgetIndex(self=Get ShowHideDetailsTextSwitcher() -> (ShowHideDetailsTextSwitcher:Common Widget Switcher Legacy Object Reference), Index=0)
   > SetVisibility(self=Get GameDetailsContent(), InVisibility=Hidden)
   > BindVolumeEvents() {is=title K2Node_CallFunction}
   n23 = SlotAsVerticalBoxSlot(Widget=Get Overlayheader() -> (Overlayheader:Overlay Object Reference))
   n24 = GetIslandDescriptionLocationData!() -> (Margin:Margin Structure, Alignment:EHorizontalAlignment Enum=HAlign_Fill)
   > SetPadding!(execute, self:Vertical Box Slot Object Reference=@n23, InPadding:Margin Structure=@n24) -> (then)
   > SetHorizontalAlignment(self:Vertical Box Slot Object Reference=@n23, InHorizontalAlignment=@n24.Alignment)
   Event OnMiniGameStartedFromInputComponent() {is=title K2Node_Event}
   > PlaySound2D(Sound=/Game/Sounds/Fort_UI_Sounds/FrontEnd/SC_UI_MatchMaking_PlayButton_Pressed_Creative_01)
   > Call On Starting Game()
   Event Destruct()
   n25 = Get PlayerEnteredHandle() -> (PlayerEnteredHandle:Async Action Start Listening To Event Object Reference)
   > n26 = Is Valid(InputObject=@n25)
   n27 = Get ExitedEventHandle() -> (ExitedEventHandle:Async Action Start Listening To Event Object Reference)
   Is Valid(exec=[@n28.then, @n26.Is Not Valid], InputObject=@n27)
   > StopListeningToEvent(self:Async Action Start Listening To Event Object Reference=@n27)
   n28 = StopListeningToEvent(execute=@n26.Is Valid, self=@n25)
   Event PreConstruct()
  ]}
 GetPlayerControllerAthena {nodes [
   Inputs()
   > n1 = Cast To FortPlayerControllerAthena(Object=GetOwningPlayer())
   > Outputs!(then:Exec, CastFailed:Exec=@n1.CastFailed, AsFortPlayerController:Fort Player Controller Athena Object Reference=@n1)
  ]}
 GetStartButtomProgressMaterial {nodes [
   GetStartButtomProgressMaterial() {is=title K2Node_FunctionEntry}
   > Return Node!(
    execute
    Output_Get:Material Instance Dynamic Object Reference=GetDynamicMaterial!(Brush:"Slate Brush Structure (by ref)"=Get ProgressBarSlateBrushCopy() -> (ProgressBarSlateBrushCopy:Slate Brush Structure)) -> (ReturnValue)
   )
  ]}
 Hide if Text Empty {nodes [
   n1 = Hide if Text Empty!() -> (then, TextToCheck:Text Object Reference, ElementToHide:Widget Object Reference) {is=title K2Node_FunctionEntry}
   n2 = "GetText (Text)"(self=@n1)
   > SetVisibility(self=@n1.ElementToHide, InVisibility=Select(Option 0=SelfHitTestInvisible, Option 1=Collapsed, Index=TextIsEmpty(InText=@n2)))
  ]}
 InitProgressDynamicMaterial {nodes [
   InitProgressDynamicMaterial() {is=title K2Node_FunctionEntry}
   n1 = Get StartGameButton()
   n2 = Break Slate Brush!(SlateBrush:"Slate Brush Structure (by ref)"=Get ProgressMaterialBrush!(self:Common Action Widget Object Reference=Get InputActionWidget(self=@n1)) -> (ProgressMaterialBrush:Slate Brush Structure)) -> (
    TintColor:Slate Color Structure
    DrawAs:ESlateBrushDrawType Enum
    Tiling:ESlateBrushTileType Enum
    Mirroring:ESlateBrushMirrorType Enum
    ImageSize:Vector 2D Structure
    Margin:Margin Structure
    ResourceObject:Object Reference
    OutlineSettings:Slate Brush Outline Settings Structure
   ) {is=title K2Node_BreakStruct}
   > n3 = Set ProgressBarSlateBrushCopy(ProgressBarSlateBrushCopy:Slate Brush Structure=Make Slate Brush!(
     TintColor:Slate Color Structure=Make Slate Color(SpecifiedColor=(R=1,G=1,B=1,A=1))
     DrawAs:ESlateBrushDrawType Enum=@n2.DrawAs
     Tiling:ESlateBrushTileType Enum=@n2.Tiling
     Mirroring:ESlateBrushMirrorType Enum=@n2.Mirroring
     ImageSize:Vector 2D Structure=@n2.ImageSize
     Margin:Margin Structure=@n2.Margin
     ResourceObject:Object Reference=@n2.ResourceObject
     OutlineSettings:Slate Brush Outline Settings Structure=(CornerRadii=(X=0,Y=0,Z=0,W=0),Color=(SpecifiedColor=(R=0,G=0,B=0,A=0),ColorUseRule=UseColor_Specified),Width=0,RoundingType=HalfHeightRadius,bUseBrushTransparency=False)
    ) -> (SlateBrush:Slate Brush Structure) {is=title K2Node_MakeStruct}) -> (Output_Get:Slate Brush Structure)
   > SetVectorParameterValue(self=GetDynamicMaterial(Brush=@n3), ParameterName=Color, Value=(R=1,G=1,B=1,A=1))
  ]}
 OnEnterVolume {nodes [
   n1 = OnEnterVolume!() -> (then, EnteringPlayerState:Player State Object Reference, EnteredMinigame:Fort Minigame Object Reference) {is=title K2Node_FunctionEntry}
   > OnVolumeChanged!(
    execute
    self
    ChangedPlayerState:Player State Object Reference=@n1
    ChangedMinigame:Fort Minigame Object Reference=@n1.EnteredMinigame
    bIsEntering:bool=true
   ) -> (then)
   > ShowDetailsPanel!(execute, self, ShouldShow:bool=false, ShouldAnimate:bool=false) -> (then)
  ]}
 OnExitVolume {nodes [
   n1 = OnExitVolume!() -> (then, ExitingPlayerState:Player State Object Reference, ExitedMinigame:Fort Minigame Object Reference) {is=title K2Node_FunctionEntry}
   > OnVolumeChanged(ChangedPlayerState=@n1, ChangedMinigame=@n1.ExitedMinigame, bIsEntering=false)
  ]}
 OnMinigameStateChange {nodes [
   n1 = OnMinigameStateChange!() -> (then, ChangedMinigame:Fort Minigame Object Reference, MinigameState:EFortMinigameState Enum=PreGame) {is=title K2Node_FunctionEntry}
   > UpdateStartButtonVisibility()
   > SetVisibility(
    self=Get ShowHideDetailsTextSwitcher()
    InVisibility=Select(Option 0=Collapsed, Option 1=Visible, Index="Equal (Enum)"(A:EFortMinigameState Enum=@n1.MinigameState, B:EFortMinigameState Enum=PreGame))
   )
  ]}
 OnStartingGame {nodes [OnStartingGame() {is=title K2Node_FunctionEntry}]}
 OnVolumeChanged {nodes [
   OnVolumeChanged!() -> (then, ChangedPlayerState:Player State Object Reference, ChangedMinigame:Fort Minigame Object Reference, bIsEntering:bool=~) {is=title K2Node_FunctionEntry}
   n1 = GetOwningPlayer()
   > Is Valid(InputObject=@n1)
   > n2 = Get!(execute:Exec, self:Controller Object Reference=@n1) -> (then:Exec, else:Exec, PlayerState:Player State Object Reference)
   > Branch(Condition="Equal (Object)"(A=@n2, B=Get ChangedPlayerState() -> (ChangedPlayerState:Player State Object Reference)) {is=title K2Node_CallFunction})
   > n3 = Get!(execute) -> (then, else, ChangedMinigame:Fort Minigame Object Reference)
   > n4 = Set EnteredMinigame(EnteredMinigame:Fort Minigame Object Reference=@n3) -> (Output_Get:Fort Minigame Object Reference)
   > Is Valid(InputObject=@n4)
   > n5 = Branch(Condition=Get bIsEntering() -> (bIsEntering:bool))
   > Bind Event to State Changed Delegate(self:Fort Minigame Object Reference=Get EnteredMinigame() -> (EnteredMinigame:Fort Minigame Object Reference), Delegate=Create Event()) {is=title K2Node_AddDelegate}
   > Bind Event to Server Time Replicated Delegate(self:Fort Minigame Object Reference=Get EnteredMinigame(), Delegate=Create Event()) {is=title K2Node_AddDelegate}
   > OnMinigameStateChange!(
    execute
    self
    ChangedMinigame:Fort Minigame Object Reference=Get EnteredMinigame()
    MinigameState:EFortMinigameState Enum=Get CurrentState!(self:Fort Minigame Object Reference=Get EnteredMinigame()) -> (CurrentState:EFortMinigameState Enum=PreGame)
   ) -> (then) {is=title K2Node_CallFunction}
   Unbind Event from State Changed Delegate(execute=@n5.else, self:Fort Minigame Object Reference=Get EnteredMinigame(), Delegate=Create Event()) {is=title K2Node_RemoveDelegate}
   > Unbind Event from Server Time Replicated Delegate(self:Fort Minigame Object Reference=Get EnteredMinigame(), Delegate=Create Event()) {is=title K2Node_RemoveDelegate}
  ]}
 PlayIntro {nodes [
   PlayIntro() {is=title K2Node_FunctionEntry}
   > PlayAnimation(InAnimation=Get Intro())
  ]}
 ShowDetailsPanel {nodes [
   n1 = ShowDetailsPanel!() -> (then, ShouldShow:bool=~, ShouldAnimate:bool=true) {is=title K2Node_FunctionEntry}
   > n2 = Branch(Condition=@n1)
   > n3 = Branch(Condition=IsAnimationPlaying(InAnimation=Get AnimHideDetails() -> (AnimHideDetails:Widget Animation Object Reference)))
   SetActiveWidgetIndex(execute=@n3.else, self=Get ShowHideDetailsTextSwitcher(), Index=1)
   > SetVisibility(self=Get GameDetailsContent(), InVisibility=SelfHitTestInvisible)
   > PlayAnimation(InAnimation=Get AnimShowDetails() -> (AnimShowDetails:Widget Animation Object Reference))
   > SetVisibility(self=Get XpState_Compact_Overlay(), InVisibility=Collapsed)
   n4 = Branch(execute=@n2.else, Condition=IsAnimationPlaying(InAnimation=Get AnimShowDetails()))
   SetActiveWidgetIndex(execute=@n4.else, self=Get ShowHideDetailsTextSwitcher(), Index=0)
   > n5 = Branch(Condition=@n1.ShouldAnimate)
   > PlayAnimation(InAnimation=Get AnimHideDetails())
   > BindToAnimationFinished(Animation=Get AnimHideDetails(), Delegate=Create Event())
   CollapseDetailsAfterAnim(execute=@n5.else) {is=title K2Node_CallFunction}
  ]}
 ToggleInfo {nodes [
   ToggleInfo() {is=title K2Node_FunctionEntry}
   > Branch(Condition=AreDetailsToggleable())
   n1 = GetActiveWidgetIndex!(self:Widget Switcher Object Reference=Get ShowHideDetailsTextSwitcher()) -> (ReturnValue:int=0)
   n2 = GetVisibility!(self:Widget Object Reference=Get GameDetailsContent()) -> (ReturnValue:ESlateVisibility Enum=Visible)
   > ShowDetailsPanel(
    ShouldShow=AND Boolean!(
     A="Equal (Integer)"(A=@n1, B=0) {is=title K2Node_CallFunction}
     B="Equal (Enum)"(A:ESlateVisibility Enum=@n2, B:ESlateVisibility Enum=Hidden)
    ) -> (ReturnValue)
    ShouldAnimate=true
   ) {is=title K2Node_CallFunction}
  ]}
 UpdateStartButtonVisibility {nodes [
   UpdateStartButtonVisibility() {is=title K2Node_FunctionEntry}
   > Branch(Condition=Get bInputEnabled() -> (bInputEnabled:bool))
   > n1 = Branch(Condition=IsStartGameEnabled!(self:Fort Creative Mini Game Info Object Reference) -> (ReturnValue:bool))
   n2 = Get StartGameButton()
   > SetIsInteractionEnabled!(execute, self:Common Button Base Object Reference=@n2, bInIsInteractionEnabled:bool=true) -> (then)
   > SetVisibility(self=@n2, InVisibility=Visible)
   n3 = Get panelSpacer() -> (panelSpacer:Spacer Object Reference)
   > n4 = SetVisibility(self=@n3, InVisibility=Collapsed)
   ForceRefresh!(execute=[@n5.then, @n4.then], self:Fort Creative Mini Game Info Object Reference, bSetupInputComponent:bool=true) -> (then, ReturnValue:bool)
   SetIsInteractionEnabled(execute=@n1.else, self=@n2, bInIsInteractionEnabled=false)
   > SetVisibility(self=@n2, InVisibility=Collapsed)
   > n5 = SetVisibility(self=@n3, InVisibility=Collapsed)
  ]}
]
variables [
 AffiliatePanel {type=CreativeAffiliatePanel_C, default=null}
 BannedNames {type="array<string>", default [Epic Games, EpicGames]}
 bDisplayingAffiliatePanel {type=boolean}
 bInputEnabled {type=boolean}
 ExitedEventHandle {type=AsyncAction_StartListeningToEvent, default=null}
 OnStartingGame {}
 PlayerEnteredHandle {type=AsyncAction_StartListeningToEvent, default=null}
 ProgressBarSlateBrushCopy {
  type=SlateBrush
  default {
   bIsDynamicallyLoaded=false
   drawAs=Image
   imageType=NoImage
   mirroring=NoMirror
   resourceName=None
   resourceObject=null
   tiling=NoTile
   imageSize=<32 32>
   margin=<ltrb 0 0 0 0>
   outlineSettings {
    bUseBrushTransparency=false
    roundingType=HalfHeightRadius
    width=0
    color {colorUseRule=UseColor_Specified, specifiedColor=<rgba 0 0 0 0>}
    cornerRadii=<0 0 0 1>
   }
   tintColor {colorUseRule=UseColor_Specified, specifiedColor=<rgba 1 1 1 1>}
   uVRegion {bIsValid=false, max=<0 0>, min=<0 0>}
  }
 }
 Source Name {type=string, default=CreativePregameLobby}
]
widgets {
 info {inheritedWidgetCount=7, parentClass=/Script/FortniteUI.FortCreativeMiniGameInfo, rootWidgetClass=&SizeBox, widgetCount=51}
 widgets [
  SizeBox_0 {
   widgetClassPath=&SizeBox
   widget=SizeBox_0
   properties {bOverride_HeightOverride=true, bOverride_WidthOverride=true, heightOverride=1080, maxDesiredWidth=491.4909, widthOverride=1920}
  }
  Overlay_2 {widgetClassPath=&Overlay, parent=SizeBox_0, widget=Overlay_2, slot {class=SizeBoxSlot}}
  VerticalBox_0 {
   widgetClassPath=&VerticalBox
   parent=Overlay_2
   widget=VerticalBox_0
   slot {class=OverlaySlot, properties {horizontalAlignment=HAlign_Fill, verticalAlignment=VAlign_Fill}}
  }
  Overlayheader {
   widgetClassPath=&Overlay
   bIsVariable=true
   parent=VerticalBox_0
   widget=Overlayheader
   slot {
    class=VerticalBoxSlot
    properties {horizontalAlignment=HAlign_Left, verticalAlignment=VAlign_Top, padding=<ltrb 80 0 0 0>, size {sizeRule=Fill, value=1}}
   }
  }
  CreativeLogoWatermark {
   widgetClassPath=&Image
   bIsVariable=true
   parent=Overlayheader
   widget=CreativeLogoWatermark
   properties {
    visibility=SelfHitTestInvisible
    brush {
     bIsDynamicallyLoaded=false
     drawAs=Image
     imageType=NoImage
     mirroring=NoMirror
     resourceName=None
     resourceObject=/Game/Athena/HUD/Art/Materials/M_UI_CreativeLogo
     tiling=NoTile
     imageSize=<108 108>
     margin=<ltrb 0 0 0 0>
     outlineSettings {
      bUseBrushTransparency=false
      roundingType=HalfHeightRadius
      width=0
      color {colorUseRule=UseColor_Specified, specifiedColor=<rgba 0 0 0 0>}
      cornerRadii=<0 0 0 0>
     }
     tintColor {colorUseRule=UseColor_Specified, specifiedColor=<rgba 1 1 1 1>}
     uVRegion {bIsValid=false, max=<0 0>, min=<0 0>}
    }
    renderTransform {angle=-15, scale=<1 1>, shear=<0 0>, translation=<-52 -40>}
   }
   slot {class=OverlaySlot}
  }
  Header {
   widgetClassPath=&VerticalBox
   parent=Overlayheader
   widget=Header
   slot {class=OverlaySlot, properties {horizontalAlignment=HAlign_Fill, padding=<ltrb 0 10 0 0>}}
  }
  CreatorBox {
   widgetClassPath=&HorizontalBox
   bIsVariable=true
   parent=Header
   widget=CreatorBox
   properties {renderTransform {angle=0, scale=<1 1>, shear=<-8 0>, translation=<0 0>}}
   slot {class=VerticalBoxSlot, properties {padding=<ltrb 0 -12 0 0>}}
  }
  CommonTextBlock_1 {
   widgetClassPath=&CommonTextBlock
   parent=CreatorBox
   widget=CommonTextBlock_1
   properties {
    text=Created By:
    textTransformPolicy=ToUpper
    colorAndOpacity {colorUseRule=UseColor_Specified, specifiedColor=<rgba 1 1 1 0.8>}
    font {
     bForceMonospaced=false
     bMaterialIsStencil=false
     fontMaterial=null
     fontObject=&BurbankSmall
     letterSpacing=0
     monospacedWidth=1
     size=16
     skewAmount=0
     typefaceFontName=Black
     outlineSettings {
      bApplyOutlineToDropShadows=false
      bMiteredCorners=false
      bSeparateFillAlpha=false
      outlineMaterial=null
      outlineSize=1
      outlineColor=<rgba 0 0 0 0.1>
     }
    }
    shadowOffset=<0 0>
   }
   slot {class=HorizontalBoxSlot, properties {verticalAlignment=VAlign_Center}}
  }
  Text_CreatorName {
   widgetClassPath=&CommonTextBlock
   bInherited=true
   parent=CreatorBox
   widget=Text_CreatorName
   properties {
    text=CreatorName
    colorAndOpacity {colorUseRule=UseColor_Specified, specifiedColor=<rgba 0 1 1 1>}
    font {
     bForceMonospaced=false
     bMaterialIsStencil=false
     fontMaterial=null
     fontObject=&BurbankSmall
     letterSpacing=0
     monospacedWidth=1
     size=20
     skewAmount=0
     typefaceFontName=Black
     outlineSettings {
      bApplyOutlineToDropShadows=false
      bMiteredCorners=false
      bSeparateFillAlpha=false
      outlineMaterial=null
      outlineSize=1
      outlineColor=<rgba 0 0 0 0.1>
     }
    }
    renderTransform {angle=0, scale=<1 1>, shear=<0 0>, translation=<0 -2>}
    shadowOffset=<0 0>
   }
   slot {class=HorizontalBoxSlot, properties {verticalAlignment=VAlign_Center, padding=<ltrb 6 0 0 0>}}
  }
  Text_Name {
   widgetClassPath=&CommonTextBlock
   bInherited=true
   parent=Header
   widget=Text_Name
   properties {
    text=Lorem ipsum dolor
    textTransformPolicy=ToUpper
    wrapTextAt=640
    font {
     bForceMonospaced=false
     bMaterialIsStencil=false
     fontMaterial=null
     fontObject=&Burbank
     letterSpacing=0
     monospacedWidth=1
     size=48
     skewAmount=0
     typefaceFontName=Burbank-Black
     outlineSettings {
      bApplyOutlineToDropShadows=false
      bMiteredCorners=false
      bSeparateFillAlpha=false
      outlineMaterial=null
      outlineSize=1
      outlineColor=<rgba 0 0 0 0.1>
     }
    }
    renderTransform {angle=0, scale=<1 1>, shear=<-8 0>, translation=<0 0>}
    shadowOffset=<0 0>
   }
   slot {class=VerticalBoxSlot, properties {padding=<ltrb 0 -12 0 0>}}
  }
  HorizontalBox_3 {
   widgetClassPath=&HorizontalBox
   parent=Header
   widget=HorizontalBox_3
   slot {class=VerticalBoxSlot, properties {horizontalAlignment=HAlign_Left, verticalAlignment=VAlign_Top, padding=<ltrb -4 0 0 0>}}
  }
  SaC_Button {
   widgetClassPath=&SizeBox
   bIsVariable=true
   parent=HorizontalBox_3
   widget=SaC_Button
   properties {heightOverride=56}
   slot {class=HorizontalBoxSlot, properties {verticalAlignment=VAlign_Center}}
  }
  SupportAffliateButton {
   widgetClassPath=/FortUILibrary/Input/InputReflectorButton
   bIsVariable=true
   parent=SaC_Button
   widget=SupportAffliateButton
   properties {
    buttonText=SUPPORT A CREATOR
    contentAlignment=HAlign_Left
    onButtonBaseClicked=(&self:WidgetTree.SupportAffliateButton.HandleOnButtonClicked)
    onButtonBaseDoubleClicked=(&self:WidgetTree.SupportAffliateButton.HandleOnButtonDoubleClicked)
    onButtonBaseHovered=(&self:WidgetTree.SupportAffliateButton.HandleOnButtonHovered)
    onButtonBaseUnhovered=(&self:WidgetTree.SupportAffliateButton.HandleOnButtonUnhovered)
    onSelectedChangedBase=(&self:WidgetTree.SupportAffliateButton.HandleOnSelectedChanged)
    style=&Athena_ButtonStyle_AngledGreenMenuButton_C
    triggeringInputAction {dataTable=/FortUILibrary/Input/CreativeInputActionDataTable, rowName=IslandPanelSupportCreator}
   }
   slot {class=SizeBoxSlot, properties {verticalAlignment=VAlign_Center}}
  }
  SupportCodeBox {
   widgetClassPath=&HorizontalBox
   bIsVariable=true
   parent=HorizontalBox_3
   widget=SupportCodeBox
   slot {class=HorizontalBoxSlot, properties {horizontalAlignment=HAlign_Left, verticalAlignment=VAlign_Center, padding=<ltrb 4 0 0 0>}}
  }
  Image_609 {
   widgetClassPath=&Image
   bIsVariable=true
   parent=SupportCodeBox
   widget=Image_609
   properties {
    brush {
     bIsDynamicallyLoaded=false
     drawAs=Image
     imageType=NoImage
     mirroring=NoMirror
     resourceName=None
     resourceObject=/Game/Items/Traps/Blueprints/Toys/UI_Art_Assets/M_UI_ActionDividerLine
     tiling=NoTile
     imageSize=<16 32>
     margin=<ltrb 0 0 0 0>
     outlineSettings {
      bUseBrushTransparency=false
      roundingType=HalfHeightRadius
      width=0
      color {colorUseRule=UseColor_Specified, specifiedColor=<rgba 0 0 0 0>}
      cornerRadii=<0 0 0 0>
     }
     tintColor {colorUseRule=UseColor_Specified, specifiedColor=<rgba 0.063 0.624 0.1384 1>}
     uVRegion {bIsValid=false, max=<0 0>, min=<0 0>}
    }
    renderTransform {angle=-3, scale=<1 1>, shear=<0 0>, translation=<0 0>}
   }
   slot {class=HorizontalBoxSlot, properties {padding=<ltrb 0 -8 0 -12>}}
  }
  SACBox {
   widgetClassPath=&VerticalBox
   bIsVariable=true
   parent=SupportCodeBox
   widget=SACBox
   slot {class=HorizontalBoxSlot, properties {padding=<ltrb 8 0 0 0>}}
  }
  CommonTextBlock_0 {
   widgetClassPath=&CommonTextBlock
   parent=SACBox
   widget=CommonTextBlock_0
   properties {
    text=Use Code:
    textTransformPolicy=ToUpper
    colorAndOpacity {colorUseRule=UseColor_Specified, specifiedColor=<rgba 1 1 1 0.8>}
    font {
     bForceMonospaced=false
     bMaterialIsStencil=false
     fontMaterial=null
     fontObject=&BurbankSmall
     letterSpacing=0
     monospacedWidth=1
     size=16
     skewAmount=0
     typefaceFontName=Black
     outlineSettings {
      bApplyOutlineToDropShadows=false
      bMiteredCorners=false
      bSeparateFillAlpha=false
      outlineMaterial=null
      outlineSize=0
      outlineColor=<rgba 0 0 0 1>
     }
    }
    renderTransform {angle=0, scale=<1 1>, shear=<-8 0>, translation=<0 0>}
    shadowOffset=<0 0>
   }
   slot {class=VerticalBoxSlot, properties {horizontalAlignment=HAlign_Left}}
  }
  Text_SupportCode {
   widgetClassPath=&CommonTextBlock
   bInherited=true
   parent=SACBox
   widget=Text_SupportCode
   properties {
    text=SupportCode
    wrapTextAt=500
    colorAndOpacity {colorUseRule=UseColor_Specified, specifiedColor=<rgba 0 1 1 1>}
    font {
     bForceMonospaced=false
     bMaterialIsStencil=false
     fontMaterial=null
     fontObject=&BurbankSmall
     letterSpacing=0
     monospacedWidth=1
     size=20
     skewAmount=0
     typefaceFontName=Black
     outlineSettings {
      bApplyOutlineToDropShadows=false
      bMiteredCorners=false
      bSeparateFillAlpha=false
      outlineMaterial=null
      outlineSize=0
      outlineColor=<rgba 0 0 0 1>
     }
    }
    renderTransform {angle=0, scale=<1 1>, shear=<-8 0>, translation=<0 0>}
    shadowOffset=<0 0>
   }
   slot {class=VerticalBoxSlot, properties {verticalAlignment=VAlign_Top}}
  }
  panelSpacer {
   widgetClassPath=&Spacer
   bIsVariable=true
   parent=VerticalBox_0
   widget=panelSpacer
   properties {visibility=HitTestInvisible}
   slot {class=VerticalBoxSlot, properties {size {sizeRule=Fill, value=1}}}
  }
  StartGameButton {
   widgetClassPath=&IconTextButton
   bIsVariable=true
   parent=VerticalBox_0
   widget=StartGameButton
   properties {
    buttonText=START GAME
    contentAlignment=HAlign_Center
    minHeight=104
    minWidth=512
    onButtonBaseClicked=(&self:WidgetTree.StartGameButton.HandleOnButtonClicked)
    onButtonBaseDoubleClicked=(&self:WidgetTree.StartGameButton.HandleOnButtonDoubleClicked)
    onButtonBaseHovered=(&self:WidgetTree.StartGameButton.HandleOnButtonHovered)
    onButtonBaseUnhovered=(&self:WidgetTree.StartGameButton.HandleOnButtonUnhovered)
    onSelectedChangedBase=(&self:WidgetTree.StartGameButton.HandleOnSelectedChanged)
    style=/FortUILibrary/Foundation/Buttons/ButtonStyles/ButtonStyle_Feature_L_Skew.ButtonStyle_Feature_L_Skew_C
    pressedSlateSoundOverride {resourceObject=&SC_UI_MatchMaking_PlayButton_Pressed_Creative_01}
    triggeringInputAction {dataTable=&CreativeInputActionDataTable, rowName=IslandPanelStartGame}
   }
   slot {class=VerticalBoxSlot, properties {horizontalAlignment=HAlign_Center, verticalAlignment=VAlign_Bottom}}
  }
  GameDetailsOverlay {
   widgetClassPath=&Overlay
   bIsVariable=true
   parent=Overlay_2
   widget=GameDetailsOverlay
   slot {class=OverlaySlot, properties {horizontalAlignment=HAlign_Fill, verticalAlignment=VAlign_Center}}
  }
  GameDetailsContent {
   widgetClassPath=&Overlay
   bIsVariable=true
   parent=GameDetailsOverlay
   widget=GameDetailsContent
   properties {visibility=Hidden}
   slot {class=OverlaySlot, properties {horizontalAlignment=HAlign_Fill, verticalAlignment=VAlign_Fill}}
  }
  Backing {
   widgetClassPath=&SizeBox
   parent=GameDetailsContent
   widget=Backing
   properties {bOverride_HeightOverride=true, heightOverride=320, visibility=HitTestInvisible}
   slot {class=OverlaySlot, properties {horizontalAlignment=HAlign_Fill, verticalAlignment=VAlign_Fill}}
  }
  DetailsBacking {
   widgetClassPath=&Image
   bIsVariable=true
   parent=Backing
   widget=DetailsBacking
   properties {
    visibility=HitTestInvisible
    brush {
     bIsDynamicallyLoaded=false
     drawAs=Image
     imageType=NoImage
     mirroring=NoMirror
     resourceName=None
     resourceObject=/Game/Athena/HUD/Art/Materials/M_UI_ModalRadialExpand_GameDetails
     tiling=NoTile
     imageSize=<32 32>
     margin=<ltrb 0 0 0 0>
     outlineSettings {
      bUseBrushTransparency=false
      roundingType=HalfHeightRadius
      width=0
      color {colorUseRule=UseColor_Specified, specifiedColor=<rgba 0 0 0 0>}
      cornerRadii=<0 0 0 0>
     }
     tintColor {colorUseRule=UseColor_Specified, specifiedColor=<rgba 1 1 1 1>}
     uVRegion {bIsValid=false, max=<0 0>, min=<0 0>}
    }
   }
   slot {class=SizeBoxSlot}
  }
  BackingHeader {
   widgetClassPath=&Image
   bIsVariable=true
   parent=GameDetailsContent
   widget=BackingHeader
   properties {brush {
     bIsDynamicallyLoaded=false
     drawAs=Image
     imageType=NoImage
     mirroring=NoMirror
     resourceName=None
     resourceObject=/Game/UI/Foundation/Materials/M_UI_Modal_Lightning_Header
     tiling=NoTile
     imageSize=<32 64>
     margin=<ltrb 0 0 0 0>
     outlineSettings {
      bUseBrushTransparency=false
      roundingType=HalfHeightRadius
      width=0
      color {colorUseRule=UseColor_Specified, specifiedColor=<rgba 0 0 0 0>}
      cornerRadii=<0 0 0 0>
     }
     tintColor {colorUseRule=UseColor_Specified, specifiedColor=<rgba 1 1 1 1>}
     uVRegion {bIsValid=false, max=<0 0>, min=<0 0>}
    }}
   slot {class=OverlaySlot, properties {horizontalAlignment=HAlign_Fill}}
  }
  BackingFooter {
   widgetClassPath=&Image
   bIsVariable=true
   parent=GameDetailsContent
   widget=BackingFooter
   properties {
    renderOpacity=0.5
    visibility=HitTestInvisible
    brush {
     bIsDynamicallyLoaded=false
     drawAs=Image
     imageType=NoImage
     mirroring=NoMirror
     resourceName=None
     resourceObject=&M_UI_Modal_Lightning_Header
     tiling=NoTile
     imageSize=<32 72>
     margin=<ltrb 0 0 0 0>
     outlineSettings {
      bUseBrushTransparency=false
      roundingType=HalfHeightRadius
      width=0
      color {colorUseRule=UseColor_Specified, specifiedColor=<rgba 0 0 0 0>}
      cornerRadii=<0 0 0 0>
     }
     tintColor {colorUseRule=UseColor_Specified, specifiedColor=<rgba 1 1 1 1>}
     uVRegion {bIsValid=false, max=<0 0>, min=<0 0>}
    }
    colorAndOpacity=<rgba 1 1 1 0>
    renderTransform {angle=0, scale=<-1 -1>, shear=<0 0>, translation=<0 0>}
   }
   slot {class=OverlaySlot, properties {horizontalAlignment=HAlign_Fill, verticalAlignment=VAlign_Bottom}}
  }
  DetailsBox {
   widgetClassPath=&HorizontalBox
   bIsVariable=true
   parent=GameDetailsContent
   widget=DetailsBox
   properties {renderTransform {angle=0, scale=<1 1>, shear=<0 0>, translation=<0 12>}}
   slot {class=OverlaySlot, properties {horizontalAlignment=HAlign_Fill, verticalAlignment=VAlign_Center, padding=<ltrb 128 0 0 0>}}
  }
  DetailsSizeBox1 {
   widgetClassPath=&SizeBox
   parent=DetailsBox
   widget=DetailsSizeBox1
   properties {bOverride_WidthOverride=true, widthOverride=512}
   slot {class=HorizontalBoxSlot, properties {horizontalAlignment=HAlign_Left, verticalAlignment=VAlign_Center}}
  }
  BulletBox1 {widgetClassPath=&HorizontalBox, bIsVariable=true, parent=DetailsSizeBox1, widget=BulletBox1, slot {class=SizeBoxSlot}}
  Image_3 {
   widgetClassPath=&Image
   bIsVariable=true
   parent=BulletBox1
   widget=Image_3
   properties {
    brush {
     bIsDynamicallyLoaded=false
     drawAs=Image
     imageType=NoImage
     mirroring=NoMirror
     resourceName=None
     resourceObject=/Game/Items/Traps/Blueprints/Toys/UI_Art_Assets/M_UI_ActionDividerLine_Modal
     tiling=NoTile
     imageSize=<24 32>
     margin=<ltrb 0 0 0 0>
     outlineSettings {
      bUseBrushTransparency=false
      roundingType=HalfHeightRadius
      width=0
      color {colorUseRule=UseColor_Specified, specifiedColor=<rgba 0 0 0 0>}
      cornerRadii=<0 0 0 0>
     }
     tintColor {colorUseRule=UseColor_Specified, specifiedColor=<rgba 1 1 1 1>}
     uVRegion {bIsValid=false, max=<0 0>, min=<0 0>}
    }
    colorAndOpacity=<rgba 1 1 1 0.75>
    renderTransformPivot=<0.5 1>
   }
   slot {class=HorizontalBoxSlot, properties {padding=<ltrb 0 -16 0 -22>}}
  }
  Text_Bullet1 {
   widgetClassPath=&CommonTextBlock
   bInherited=true
   bIsVariable=true
   parent=BulletBox1
   widget=Text_Bullet1
   properties {
    autoWrapText=true
    text="Lorem ipsum dolor sit amet, consectetuer adipiscing elit. Aenean commodo ligula eget dolor. Aenean massa. Cum sociis natoque penatibus et magnis disss"
    font {
     bForceMonospaced=false
     bMaterialIsStencil=false
     fontMaterial=/Game/Athena/HUD/Art/Materials/M_UI_CreativeStartPanel_DescriptionColor
     fontObject=&BurbankSmall
     letterSpacing=0
     monospacedWidth=1
     size=16
     skewAmount=0
     typefaceFontName=Bold
     outlineSettings {
      bApplyOutlineToDropShadows=false
      bMiteredCorners=false
      bSeparateFillAlpha=false
      outlineMaterial=null
      outlineSize=0
      outlineColor=<rgba 0 0 0 1>
     }
    }
    renderTransform {angle=0, scale=<1 1>, shear=<-8 0>, translation=<0 0>}
    shadowOffset=<0 0>
   }
   slot {class=HorizontalBoxSlot, properties {verticalAlignment=VAlign_Center, padding=<ltrb 8 0 0 0>, size {sizeRule=Fill, value=1}}}
  }
  Spacer_116 {widgetClassPath=&Spacer, parent=DetailsBox, widget=Spacer_116, properties {size=<64 1>}, slot {class=HorizontalBoxSlot}}
  DetailsSizeBox2 {
   widgetClassPath=&SizeBox
   parent=DetailsBox
   widget=DetailsSizeBox2
   properties {bOverride_WidthOverride=true, widthOverride=512}
   slot {class=HorizontalBoxSlot, properties {horizontalAlignment=HAlign_Left}}
  }
  BulletBox2 {widgetClassPath=&HorizontalBox, bIsVariable=true, parent=DetailsSizeBox2, widget=BulletBox2, slot {class=SizeBoxSlot}}
  Image_294 {
   widgetClassPath=&Image
   bIsVariable=true
   parent=BulletBox2
   widget=Image_294
   properties {
    brush {
     bIsDynamicallyLoaded=false
     drawAs=Image
     imageType=NoImage
     mirroring=NoMirror
     resourceName=None
     resourceObject=&M_UI_ActionDividerLine_Modal
     tiling=NoTile
     imageSize=<24 32>
     margin=<ltrb 0 0 0 0>
     outlineSettings {
      bUseBrushTransparency=false
      roundingType=HalfHeightRadius
      width=0
      color {colorUseRule=UseColor_Specified, specifiedColor=<rgba 0 0 0 0>}
      cornerRadii=<0 0 0 0>
     }
     tintColor {colorUseRule=UseColor_Specified, specifiedColor=<rgba 1 1 1 1>}
     uVRegion {bIsValid=false, max=<0 0>, min=<0 0>}
    }
    colorAndOpacity=<rgba 1 1 1 0.75>
    renderTransformPivot=<0.5 1>
   }
   slot {class=HorizontalBoxSlot, properties {padding=<ltrb 0 -16 0 -22>}}
  }
  Text_Bullet2 {
   widgetClassPath=&CommonTextBlock
   bInherited=true
   bIsVariable=true
   parent=BulletBox2
   widget=Text_Bullet2
   properties {
    autoWrapText=true
    text="Lorem ipsum dolor sit amet, consectetuer adipiscing elit. Aenean commodo ligula eget dolor. Aenean massa. Cum sociis natoque penatibus et magnis disss"
    font {
     bForceMonospaced=false
     bMaterialIsStencil=false
     fontMaterial=&M_UI_CreativeStartPanel_DescriptionColor
     fontObject=&BurbankSmall
     letterSpacing=0
     monospacedWidth=1
     size=16
     skewAmount=0
     typefaceFontName=Bold
     outlineSettings {
      bApplyOutlineToDropShadows=false
      bMiteredCorners=false
      bSeparateFillAlpha=false
      outlineMaterial=null
      outlineSize=0
      outlineColor=<rgba 0 0 0 1>
     }
    }
    renderTransform {angle=0, scale=<1 1>, shear=<-8 0>, translation=<0 0>}
    shadowOffset=<0 0>
   }
   slot {class=HorizontalBoxSlot, properties {verticalAlignment=VAlign_Center, padding=<ltrb 8 0 0 0>, size {sizeRule=Fill, value=1}}}
  }
  Spacer_245 {widgetClassPath=&Spacer, parent=DetailsBox, widget=Spacer_245, properties {size=<64 1>}, slot {class=HorizontalBoxSlot}}
  DetailsSizeBox3 {
   widgetClassPath=&SizeBox
   parent=DetailsBox
   widget=DetailsSizeBox3
   properties {bOverride_WidthOverride=true, widthOverride=512}
   slot {class=HorizontalBoxSlot, properties {horizontalAlignment=HAlign_Left}}
  }
  BulletBox3 {widgetClassPath=&HorizontalBox, bIsVariable=true, parent=DetailsSizeBox3, widget=BulletBox3, slot {class=SizeBoxSlot}}
  Image_295 {
   widgetClassPath=&Image
   bIsVariable=true
   parent=BulletBox3
   widget=Image_295
   properties {
    brush {
     bIsDynamicallyLoaded=false
     drawAs=Image
     imageType=NoImage
     mirroring=NoMirror
     resourceName=None
     resourceObject=&M_UI_ActionDividerLine_Modal
     tiling=NoTile
     imageSize=<24 32>
     margin=<ltrb 0 0 0 0>
     outlineSettings {
      bUseBrushTransparency=false
      roundingType=HalfHeightRadius
      width=0
      color {colorUseRule=UseColor_Specified, specifiedColor=<rgba 0 0 0 0>}
      cornerRadii=<0 0 0 0>
     }
     tintColor {colorUseRule=UseColor_Specified, specifiedColor=<rgba 1 1 1 1>}
     uVRegion {bIsValid=false, max=<0 0>, min=<0 0>}
    }
    colorAndOpacity=<rgba 1 1 1 0.75>
    renderTransformPivot=<0.5 1>
   }
   slot {class=HorizontalBoxSlot, properties {padding=<ltrb 0 -16 0 -22>}}
  }
  Text_Bullet3 {
   widgetClassPath=&CommonTextBlock
   bInherited=true
   bIsVariable=true
   parent=BulletBox3
   widget=Text_Bullet3
   properties {
    autoWrapText=true
    text="Lorem ipsum dolor sit amet, consectetuer adipiscing elit. Aenean commodo ligula eget dolor. Aenean massa. Cum sociis natoque penatibus et magnis disss"
    font {
     bForceMonospaced=false
     bMaterialIsStencil=false
     fontMaterial=&M_UI_CreativeStartPanel_DescriptionColor
     fontObject=&BurbankSmall
     letterSpacing=0
     monospacedWidth=1
     size=16
     skewAmount=0
     typefaceFontName=Bold
     outlineSettings {
      bApplyOutlineToDropShadows=false
      bMiteredCorners=false
      bSeparateFillAlpha=false
      outlineMaterial=null
      outlineSize=0
      outlineColor=<rgba 0 0 0 1>
     }
    }
    renderTransform {angle=0, scale=<1 1>, shear=<-8 0>, translation=<0 0>}
    shadowOffset=<0 0>
   }
   slot {class=HorizontalBoxSlot, properties {verticalAlignment=VAlign_Center, padding=<ltrb 8 0 0 0>, size {sizeRule=Fill, value=1}}}
  }
  XpState_Expanded {
   widgetClassPath=/Game/Athena/HUD/Creative/Cr_XP_State
   bIsVariable=true
   parent=GameDetailsContent
   widget=XpState_Expanded
   properties {visibility=Collapsed}
   slot {class=OverlaySlot, properties {horizontalAlignment=HAlign_Center, verticalAlignment=VAlign_Bottom, padding=<ltrb 0 0 0 -64>}}
  }
  Toggle_Details {
   widgetClassPath=&SizeBox
   bIsVariable=true
   parent=GameDetailsOverlay
   widget=Toggle_Details
   properties {heightOverride=56, visibility=Hidden}
   slot {class=OverlaySlot, properties {padding=<ltrb 80 16 0 0>}}
  }
  HorizontalBox_6 {
   widgetClassPath=&HorizontalBox
   parent=Toggle_Details
   widget=HorizontalBox_6
   slot {class=SizeBoxSlot, properties {verticalAlignment=VAlign_Center}}
  }
  ShowHideDetailsTextSwitcher {
   widgetClassPath=&CommonWidgetSwitcherLegacy
   bIsVariable=true
   parent=HorizontalBox_6
   widget=ShowHideDetailsTextSwitcher
   properties {activeWidgetIndex=1, transitionDuration=0.1}
   slot {class=HorizontalBoxSlot, properties {horizontalAlignment=HAlign_Left, verticalAlignment=VAlign_Center}}
  }
  ShowDetailsButton {
   widgetClassPath=&InputReflectorButton
   bIsVariable=true
   parent=ShowHideDetailsTextSwitcher
   widget=ShowDetailsButton
   properties {
    buttonText=SHOW DETAILS
    onButtonBaseClicked=(&self:WidgetTree.ShowDetailsButton.HandleOnButtonClicked)
    onButtonBaseDoubleClicked=(&self:WidgetTree.ShowDetailsButton.HandleOnButtonDoubleClicked)
    onButtonBaseHovered=(&self:WidgetTree.ShowDetailsButton.HandleOnButtonHovered)
    onButtonBaseUnhovered=(&self:WidgetTree.ShowDetailsButton.HandleOnButtonUnhovered)
    onSelectedChangedBase=(&self:WidgetTree.ShowDetailsButton.HandleOnSelectedChanged)
    style=&Athena_ButtonStyle_AngledGreenMenuButton_C
    triggeringInputAction {dataTable=&CreativeInputActionDataTable, rowName=ToggleMinigameDetails}
   }
   slot {class=WidgetSwitcherSlot}
  }
  HideDetailsButton {
   widgetClassPath=&InputReflectorButton
   bIsVariable=true
   parent=ShowHideDetailsTextSwitcher
   widget=HideDetailsButton
   properties {
    buttonText=HIDE DETAILS
    onButtonBaseClicked=(&self:WidgetTree.HideDetailsButton.HandleOnButtonClicked)
    onButtonBaseDoubleClicked=(&self:WidgetTree.HideDetailsButton.HandleOnButtonDoubleClicked)
    onButtonBaseHovered=(&self:WidgetTree.HideDetailsButton.HandleOnButtonHovered)
    onButtonBaseUnhovered=(&self:WidgetTree.HideDetailsButton.HandleOnButtonUnhovered)
    onSelectedChangedBase=(&self:WidgetTree.HideDetailsButton.HandleOnSelectedChanged)
    style=&Athena_ButtonStyle_AngledGreenMenuButton_C
    triggeringInputAction {dataTable=&CreativeInputActionDataTable, rowName=ToggleMinigameDetails}
   }
   slot {class=WidgetSwitcherSlot}
  }
  XpState_Compact_Overlay {
   widgetClassPath=&Overlay
   bIsVariable=true
   parent=GameDetailsOverlay
   widget=XpState_Compact_Overlay
   slot {class=OverlaySlot, properties {padding=<ltrb 80 100 0 0>}}
  }
  XpState_Compact {
   widgetClassPath=&Cr_XP_State
   bIsVariable=true
   parent=XpState_Compact_Overlay
   widget=XpState_Compact
   properties {isExpanded=false, visibility=Collapsed}
   slot {class=OverlaySlot}
  }
  CommonActionWidget_216 {
   widgetClassPath=&CommonActionWidget
   bIsVariable=true
   parent=Overlay_2
   widget=CommonActionWidget_216
   slot {class=OverlaySlot}
  }
  EntryBox_MiniGameTags {widgetClassPath=&DynamicEntryBox, bInherited=true, widget=null}
 ]
}
