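# Creative-mode quickbar slot widget for the HUD, a child of AthenaQuickBarSlot that adds no widgets of its own.
# On construct it sets its native cell to show selection when empty and binds to the Creative HUD context's active-quickbar change to refresh the item.
# A separate function refreshes only when a given slot index matches its own.
# Touch start broadcasts OnSlotTouched before passing to the parent handler.
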WidgetBlueprint /Game/Athena/HUD/CreativeQuickbarSlot
graphs [
 EventGraph {nodes [
   n1 = Event PreConstruct()
   > Parent: PreConstruct(IsDesignTime=@n1.IsDesignTime)
   Event Construct()
   > Parent: Construct()
   > Set bShowSelectedWhenEmpty!(
    execute
    bShowSelectedWhenEmpty:bool=true
    self:Athena Quick Bar Native Cell Object Reference=Get NativeCell() -> (NativeCell:Athena Quick Bar Native Cell Object Reference)
   ) -> (then, Output_Get:bool)
   > Bind Event to On Creative Quickbar Active Changed(
    self:Creative HUDLContext Object Reference=Get CreativeHUDLContext!() -> (ReturnValue:Creative HUDLContext Object Reference) {is=title K2Node_GetBlueprintContextEx}
    Delegate=Create Event()
   ) {is=title K2Node_AddDelegate}
   Event OnRefreshItem() {is=title K2Node_Event}
   > Parent: OnRefreshItem() {is=title K2Node_CallParentFunction}
   Event Tick()
  ]}
 OnCreativeQuickbarActiveChanged {nodes [
   OnCreativeQuickbarActiveChanged!() -> (then, bIsActive:bool=~) {is=title K2Node_FunctionEntry}
   > RefreshItem(self:Fort Quick Bar Slot Base Object Reference)
  ]}
 OnSlotTouched {nodes [OnSlotTouched() {is=title K2Node_FunctionEntry}]}
 OnTouchStarted {nodes [
   n1 = OnTouchStarted!() -> (then, MyGeometry:Geometry Structure, InTouchEvent:"Pointer Event Structure (by ref)") {is=title K2Node_FunctionEntry}
   > Call On Slot Touched!(execute:Exec, self:Creative Quickbar Slot Object Reference) -> (then:Exec) {is=title K2Node_CallDelegate}
   > n2 = Parent: OnTouchStarted!(execute, MyGeometry:Geometry Structure=@n1, InTouchEvent:"Pointer Event Structure (by ref)"=@n1.InTouchEvent) -> (then, ReturnValue:Event Reply Structure) {is=title K2Node_CallParentFunction}
   > Return Node!(execute, ReturnValue:Event Reply Structure=@n2)
  ]}
 RefreshItemIfSlotIndexMatches {nodes [
   n1 = RefreshItemIfSlotIndexMatches!() -> (then, SlotIndex:int, HasContent:bool=~) {is=title K2Node_FunctionEntry}
   > Branch(Condition="Equal (Integer)"(A=@n1, B=Get SlotIndex() -> (SlotIndex:int=0)) {is=title K2Node_CallFunction})
   > RefreshItem()
  ]}
]
variables [
 IsPhoneEnabled {type=boolean}
 OnSlotTouched {}
]
widgets {
 info {inheritedWidgetCount=0, parentClass=&/AthenaQuickBarSlot.AthenaQuickBarSlot_C, rootWidgetClass=null, widgetCount=0}
 widgets []
}
