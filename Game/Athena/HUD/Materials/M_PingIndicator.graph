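# M_PingIndicator is a translucent UI material for a HUD ping indicator, showing either a squad location icon or a procedural dot.
# Centered UVs drive two nested sphere masks, forming a disc filled by the Color parameter with a black outer rim.
# The Shape scalar lerps both color and alpha from the icon texture to that circle.
# Color tints only the circle; the icon keeps its texture colors.
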Material /Game/Athena/HUD/Materials/M_PingIndicator
graph {
 properties {blendMode=BLEND_Translucent, materialDomain=MD_UI, pixelDepthOffsetMode=PDOM_Legacy, usedShadingModels=Default Lit}
 nodes [
  n1 = Multiply(A=Subtract(A=TextureCoordinate()) {constB=0.5}) {constB=2}
  n2 = SphereMask(A=@n1, B=Constant2Vector(), Radius=Constant() {r=0.7})
  n3 = SphereMask(A=@n1, B=Constant2Vector(), Radius=Constant() {r=0.8})
  n4 = Multiply(A=@n3, B=OneMinus(None=@n2))
  Add(A=@n2, B=@n4)
  n5 = TextureSample() {texture=/Game/Athena/HUD/Art/squadArt/T_UI_SquadGenericIndicator_Location_b}
  n6 = ScalarParameter() {parameterName=Shape}
  LinearInterpolate(A=@n5.A, B=Clamp(None=@n3), Alpha=@n6)
  n7 = Constant3Vector()
  n8 = VectorParameter() {parameterName=Color, defaultValue=<rgba 1 1 1 1>}
  LinearInterpolate(A=@n5.RGB, B=Add(A=Multiply(A=@n7, B=@n4), B=Multiply(A=@n2, B=@n8.RGB)), Alpha=@n6)
 ]
}
