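# Additive UI material that draws a looping, tinted diamond pulse, named as a location ping reminder for the HUD.
# A Time-and-PulseSpeed sawtooth drives both the centered scale of offset UVs, from zero to the set scale, and a fade from StartingOpacity to zero.
# The diamond texture samples UV channel 1 with customized UVs enabled, its RGB tinted by Color and its alpha multiplied by the fade and Intensity.
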Material /Game/Athena/HUD/Materials/M_UI_LocationPingReminder
graph {
 properties {blendMode=BLEND_Additive, materialDomain=MD_UI, numCustomizedUVs=2, pixelDepthOffsetMode=PDOM_Legacy, usedShadingModels=Default Lit}
 nodes [
  n1 = ScalarParameter() {parameterName=PulseUVOffset_X}
  n2 = MaterialFunctionCall(Float2=TextureCoordinate()) {materialFunction=&BreakOutFloat2Components}
  n3 = ScalarParameter() {parameterName=PulseUVOffset_Y}
  n4 = ScalarParameter() {defaultValue=1, parameterName=PulseUVScale_X}
  n5 = ScalarParameter() {defaultValue=1, parameterName=PulseUVScale_Y}
  n6 = Fmod(A=Multiply(A=Time(), B=ScalarParameter() {defaultValue=0.7, parameterName=PulseSpeed}), B=Constant() {r=1})
  MaterialFunctionCall(
   UVs=AppendVector(A=Add(A=@n1, B=@n2.R), B=Add(A=@n2.G, B=@n3))
   Texture Scale=LinearInterpolate(A=Constant2Vector(), B=AppendVector(A=@n4, B=@n5), Alpha=@n6)
  ) {materialFunction=&ScaleUVsByCenter}
  n7 = TextureSample(UVs=TextureCoordinate() {coordinateIndex=1}) {samplerSource=SSM_Clamp_WorldGroupSettings, texture=/Game/Athena/HUD/Art/T_UI_Indicator_Diamond}
  Multiply(A=VectorParameter() {parameterName=Color, defaultValue=<rgba 1 1 1 1>}.RGB, B=@n7.RGB)
  n8 = ScalarParameter() {defaultValue=1, parameterName=StartingOpacity}
  n9 = Constant()
  Multiply(
   A=ScalarParameter() {defaultValue=2, parameterName=Intensity}
   B=Multiply(A=LinearInterpolate(A=@n8, B=@n9, Alpha=@n6) {constA=1, constB=0}, B=@n7.A)
  )
 ]
}
