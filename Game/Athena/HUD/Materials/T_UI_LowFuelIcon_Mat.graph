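# Translucent UI material that draws the low-fuel HUD icon with a soft edge and a shadow.
# The red channel of the icon texture feeds two SmoothStep thresholds, each with edge and softness parameters, like a distance-field cutout.
# The main mask drives a Lerp at default inputs, shading the icon white over black.
# The shadow mask is scaled down, added to the main mask and saturated into a combined alpha.
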Material /Game/Athena/HUD/Materials/T_UI_LowFuelIcon_Mat
graph {
 properties {blendMode=BLEND_Translucent, materialDomain=MD_UI, pixelDepthOffsetMode=PDOM_Legacy, usedShadingModels=Default Lit}
 nodes [
  Constant() {r=1}
  n1 = TextureSample() {texture=&/T_UI_LowFuelIcon}
  n2 = ScalarParameter() {defaultValue=0.25, parameterName=Edge}
  n3 = MaterialFunctionCall(Alpha=@n1.R, Min=@n2, Max=Add(A=@n2, B=ScalarParameter() {defaultValue=0.05, parameterName=EdgeSoftness})) {materialFunction=&SmoothStep}
  LinearInterpolate(Alpha=@n3.Result)
  n4 = ScalarParameter() {defaultValue=0.1, parameterName=Edge_Shadow}
  n5 = MaterialFunctionCall(Alpha=@n1.R, Min=@n4, Max=Add(A=@n4, B=ScalarParameter() {defaultValue=0.005, parameterName=EdgeSoftness_Shadow})) {materialFunction=&SmoothStep}
  Saturate(None=Add(A=@n3.Result, B=Multiply(A=@n5.Result) {constB=0.25}))
 ]
}
