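# HUD widget for a pinned quest, showing a numbered pin, optional badge and progress text as count or percentage.
# Setup formats rich text, sets the pin material's BackingTypeSelector from the index, async-loads the badge as texture or material, and adjusts mobile sizing.
# After a short delay it plays Intro.
# Completion hides the badge, plays the "Quest Complete" Outro, then releases the pinned quest after an extra delay.
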WidgetBlueprint /Game/Athena/HUD/WBP_PlayerPinnedQuestInfo
animations [
 Intro {
  binding_count=7
  length_seconds=0.25
  object=&self:Intro
  sequence {bindings [
    Image_Complete_BG {
     id=#1
     possessed_class=FortLazyImage
     tracks [{
       type=MovieSceneFloatTrack
       display_name=Render Opacity
       sections [{type=MovieSceneFloatSection, end_frame=0, start_frame=0, channels [None {type=MovieSceneScriptingFloatChannel, const [0,1]}]}]
      }]
    }
    Image_CQ {
     id=#2
     possessed_class=Image
     tracks [
      {
       type=MovieSceneFloatTrack
       display_name=Render Opacity
       sections [{type=MovieSceneFloatSection, channels [None {type=MovieSceneScriptingFloatChannel, keys [[0,0],[1,0.4133],[5,1]]}]}]
      }
      {
       type=MovieScene2DTransformTrack
       sections [{
         type=MovieScene2DTransformSection
         channels [
          Translation.X {type=MovieSceneScriptingFloatChannel, const [0,2]}
          Translation.Y {type=MovieSceneScriptingFloatChannel, keys [[1,18],[5,0]]}
          Angle {type=MovieSceneScriptingFloatChannel}
          Scale.X {type=MovieSceneScriptingFloatChannel, keys [[2.333,0.5],[5,1]]}
          Scale.Y {type=MovieSceneScriptingFloatChannel, keys [[2.333,0.5],[5,1]]}
          Shear.X {type=MovieSceneScriptingFloatChannel}
          Shear.Y {type=MovieSceneScriptingFloatChannel}
         ]
        }]
      }
     ]
    }
    LazyImage_PinColor {
     id=#3
     possessed_class=FortLazyImage
     tracks [
      {
       type=MovieScene2DTransformTrack
       sections [{
         type=MovieScene2DTransformSection
         channels [
          Translation.X {type=MovieSceneScriptingFloatChannel, const [0,2]}
          Translation.Y {type=MovieSceneScriptingFloatChannel, keys [[1,18],[5,0]]}
          Angle {type=MovieSceneScriptingFloatChannel}
          Scale.X {type=MovieSceneScriptingFloatChannel, keys [[2.333,0.5],[5,1]]}
          Scale.Y {type=MovieSceneScriptingFloatChannel, keys [[2.333,0.5],[5,1]]}
          Shear.X {type=MovieSceneScriptingFloatChannel}
          Shear.Y {type=MovieSceneScriptingFloatChannel}
         ]
        }]
      }
      {
       type=MovieSceneFloatTrack
       display_name=Render Opacity
       sections [{type=MovieSceneFloatSection, channels [None {type=MovieSceneScriptingFloatChannel, keys [[0,0],[1,0.4133],[5,1]]}]}]
      }
     ]
    }
    Text_PinnedQuestDesc_RichText {
     id=#4
     possessed_class=CommonRichTextBlock
     tracks [
      {
       type=MovieScene2DTransformTrack
       sections [{
         type=MovieScene2DTransformSection
         channels [
          Translation.X {type=MovieSceneScriptingFloatChannel, keys [[0,-8],[4,0]]}
          Translation.Y {type=MovieSceneScriptingFloatChannel, const [0,2]}
          Angle {type=MovieSceneScriptingFloatChannel}
          Scale.X {type=MovieSceneScriptingFloatChannel, const [1,2]}
          Scale.Y {type=MovieSceneScriptingFloatChannel, const [1,2]}
          Shear.X {type=MovieSceneScriptingFloatChannel}
          Shear.Y {type=MovieSceneScriptingFloatChannel}
         ]
        }]
      }
      {
       type=MovieSceneFloatTrack
       display_name=Render Opacity
       sections [{type=MovieSceneFloatSection, end_frame=4, start_frame=0, channels [None {type=MovieSceneScriptingFloatChannel, keys [[0,0.017],[4,0.85]]}]}]
      }
     ]
    }
    Text_Complete {
     id=#5
     possessed_class=UEFN_TextBlock_C
     tracks [{
       type=MovieSceneFloatTrack
       display_name=Render Opacity
       sections [{type=MovieSceneFloatSection, end_frame=0, start_frame=0, channels [None {type=MovieSceneScriptingFloatChannel, const [0,1]}]}]
      }]
    }
    Image_AdditiveHorizontalBeam {
     id=#6
     possessed_class=Image
     tracks [{
       type=MovieSceneFloatTrack
       display_name=Render Opacity
       sections [{type=MovieSceneFloatSection, end_frame=0, start_frame=0, channels [None {type=MovieSceneScriptingFloatChannel, const [0,1]}]}]
      }]
    }
    Image_Check {
     id=#7
     possessed_class=Image
     tracks [{
       type=MovieSceneFloatTrack
       display_name=Render Opacity
       sections [{type=MovieSceneFloatSection, end_frame=0, start_frame=0, channels [None {type=MovieSceneScriptingFloatChannel, const [0,1]}]}]
      }]
    }
   ]}
 }
 Intro_Mobile {
  binding_count=7
  length_seconds=0.25
  object=&self:Intro_Mobile
  sequence {bindings [
    Image_Complete_BG {
     id=#1
     possessed_class=FortLazyImage
     tracks [{
       type=MovieSceneFloatTrack
       display_name=Render Opacity
       sections [{type=MovieSceneFloatSection, end_frame=0, start_frame=0, channels [None {type=MovieSceneScriptingFloatChannel, const [0,1]}]}]
      }]
    }
    Image_CQ {
     id=#8
     possessed_class=Image
     tracks [
      {
       type=MovieSceneFloatTrack
       display_name=Render Opacity
       sections [{type=MovieSceneFloatSection, channels [None {type=MovieSceneScriptingFloatChannel, keys [[0,0],[1,0.4133],[5,1]]}]}]
      }
      {
       type=MovieScene2DTransformTrack
       sections [{
         type=MovieScene2DTransformSection
         channels [
          Translation.X {type=MovieSceneScriptingFloatChannel, const [0,2]}
          Translation.Y {type=MovieSceneScriptingFloatChannel, keys [[1,18],[5,0]]}
          Angle {type=MovieSceneScriptingFloatChannel}
          Scale.X {type=MovieSceneScriptingFloatChannel, keys [[2.333,0.5],[5,1]]}
          Scale.Y {type=MovieSceneScriptingFloatChannel, keys [[2.333,0.5],[5,1]]}
          Shear.X {type=MovieSceneScriptingFloatChannel}
          Shear.Y {type=MovieSceneScriptingFloatChannel}
         ]
        }]
      }
     ]
    }
    LazyImage_PinColor {
     id=#3
     possessed_class=FortLazyImage
     tracks [
      {
       type=MovieScene2DTransformTrack
       sections [{
         type=MovieScene2DTransformSection
         channels [
          Translation.X {type=MovieSceneScriptingFloatChannel, const [0,2]}
          Translation.Y {type=MovieSceneScriptingFloatChannel, keys [[1,18],[5,0]]}
          Angle {type=MovieSceneScriptingFloatChannel}
          Scale.X {type=MovieSceneScriptingFloatChannel, keys [[2.333,0.5],[5,1]]}
          Scale.Y {type=MovieSceneScriptingFloatChannel, keys [[2.333,0.5],[5,1]]}
          Shear.X {type=MovieSceneScriptingFloatChannel}
          Shear.Y {type=MovieSceneScriptingFloatChannel}
         ]
        }]
      }
      {
       type=MovieSceneFloatTrack
       display_name=Render Opacity
       sections [{type=MovieSceneFloatSection, channels [None {type=MovieSceneScriptingFloatChannel, keys [[0,0],[1,0.4133],[5,1]]}]}]
      }
     ]
    }
    Text_PinnedQuestDesc_RichText {
     id=#4
     possessed_class=CommonRichTextBlock
     tracks [
      {
       type=MovieScene2DTransformTrack
       sections [{
         type=MovieScene2DTransformSection
         channels [
          Translation.X {type=MovieSceneScriptingFloatChannel, keys [[0,-8],[4,0]]}
          Translation.Y {type=MovieSceneScriptingFloatChannel, const [0,2]}
          Angle {type=MovieSceneScriptingFloatChannel}
          Scale.X {type=MovieSceneScriptingFloatChannel, const [1,2]}
          Scale.Y {type=MovieSceneScriptingFloatChannel, const [1,2]}
          Shear.X {type=MovieSceneScriptingFloatChannel}
          Shear.Y {type=MovieSceneScriptingFloatChannel}
         ]
        }]
      }
      {
       type=MovieSceneFloatTrack
       display_name=Render Opacity
       sections [{type=MovieSceneFloatSection, end_frame=4, start_frame=0, channels [None {type=MovieSceneScriptingFloatChannel, keys [[0,0.017],[4,1]]}]}]
      }
     ]
    }
    Text_Complete {
     id=#5
     possessed_class=UEFN_TextBlock_C
     tracks [{
       type=MovieSceneFloatTrack
       display_name=Render Opacity
       sections [{type=MovieSceneFloatSection, end_frame=0, start_frame=0, channels [None {type=MovieSceneScriptingFloatChannel, const [0,1]}]}]
      }]
    }
    Image_AdditiveHorizontalBeam {
     id=#6
     possessed_class=Image
     tracks [{
       type=MovieSceneFloatTrack
       display_name=Render Opacity
       sections [{type=MovieSceneFloatSection, end_frame=0, start_frame=0, channels [None {type=MovieSceneScriptingFloatChannel, const [0,1]}]}]
      }]
    }
    Image_Check {
     id=#7
     possessed_class=Image
     tracks [{
       type=MovieSceneFloatTrack
       display_name=Render Opacity
       sections [{type=MovieSceneFloatSection, end_frame=0, start_frame=0, channels [None {type=MovieSceneScriptingFloatChannel, const [0,1]}]}]
      }]
    }
   ]}
 }
 Outro {
  binding_count=5
  length_seconds=2.4017
  object=&self:Outro
  sequence {bindings [
    LazyImage_PinColor {
     id=#9
     possessed_class=FortLazyImage
     tracks [
      {
       type=MovieSceneFloatTrack
       display_name=Render Opacity
       sections [{
         type=MovieSceneFloatSection
         end_frame=9
         start_frame=-9
         channels [None {type=MovieSceneScriptingFloatChannel, keys [[-8.0003,1],[-2.1717,1],[4.1427,0.8],[7.5427,0]]}]
        }]
      }
      {
       type=MovieScene2DTransformTrack
       sections [{
         type=MovieScene2DTransformSection
         end_frame=6
         start_frame=-9
         channels [
          Translation.X {type=MovieSceneScriptingFloatChannel, const [0,1]}
          Translation.Y {type=MovieSceneScriptingFloatChannel, const [0,1]}
          Angle {type=MovieSceneScriptingFloatChannel, const [0,1]}
          Scale.X {type=MovieSceneScriptingFloatChannel, keys [[-8.0003,1],[-6.0573,1],[-3.6287,1.1],[5,1],[6.5713,1]]}
          Scale.Y {type=MovieSceneScriptingFloatChannel, keys [[-8.0003,1],[-6.0573,1],[-3.6287,1.1],[5,1],[6.5713,1]]}
          Shear.X {type=MovieSceneScriptingFloatChannel, const [0,1]}
          Shear.Y {type=MovieSceneScriptingFloatChannel, const [0,1]}
         ]
        }]
      }
     ]
    }
    Text_PinnedQuestDesc_RichText {
     id=#10
     possessed_class=CommonRichTextBlock
     tracks [{
       type=MovieSceneFloatTrack
       display_name=Render Opacity
       sections [{
         type=MovieSceneFloatSection
         end_frame=9
         start_frame=-9
         channels [None {type=MovieSceneScriptingFloatChannel, keys [[-8.0003,1],[3.333,1],[7.897,0]]}]
        }]
      }]
    }
    Text_Complete {
     id=#11
     possessed_class=UEFN_TextBlock_C
     tracks [
      {
       type=MovieScene2DTransformTrack
       sections [{
         type=MovieScene2DTransformSection
         channels [
          Translation.X {type=MovieSceneScriptingFloatChannel, keys [[4,-100],[22.598,0],[40.0337,0]]}
          Angle {type=MovieSceneScriptingFloatChannel}
          Scale.X {type=MovieSceneScriptingFloatChannel, const [1,2]}
          Scale.Y {type=MovieSceneScriptingFloatChannel, const [1,2]}
          Shear.X {type=MovieSceneScriptingFloatChannel}
          Shear.Y {type=MovieSceneScriptingFloatChannel}
         ]
        }]
      }
      {
       type=MovieSceneFloatTrack
       display_name=Render Opacity
       sections [{
         type=MovieSceneFloatSection
         end_frame=40
         start_frame=4
         channels [None {type=MovieSceneScriptingFloatChannel, keys [[9,0],[12.0003,0],[17.0003,0.4578],[22.598,1],[30.7347,1],[33.059,0.8],[40.0337,0]]}]
        }]
      }
     ]
    }
    Image_Check {
     id=#12
     possessed_class=Image
     tracks [
      {
       type=MovieScene2DTransformTrack
       sections [{
         type=MovieScene2DTransformSection
         end_frame=24
         start_frame=9
         channels [
          Translation.X {type=MovieSceneScriptingFloatChannel}
          Translation.Y {type=MovieSceneScriptingFloatChannel}
          Angle {type=MovieSceneScriptingFloatChannel}
          Scale.X {type=MovieSceneScriptingFloatChannel, keys [[9.3677,1],[24.184,0.6]]}
          Scale.Y {type=MovieSceneScriptingFloatChannel, keys [[9.3677,1],[24.184,0.6]]}
          Shear.X {type=MovieSceneScriptingFloatChannel}
          Shear.Y {type=MovieSceneScriptingFloatChannel}
         ]
        }]
      }
      {
       type=MovieSceneFloatTrack
       display_name=Render Opacity
       sections [{
         type=MovieSceneFloatSection
         end_frame=39
         start_frame=3
         channels [None {type=MovieSceneScriptingFloatChannel, keys [[3.018,0],[9.3677,0],[24.184,1],[32.6503,1],[39,0]]}]
        }]
      }
     ]
    }
    Image_AdditiveHorizontalBeam {
     id=#13
     possessed_class=Image
     tracks [
      {
       type=MovieSceneFloatTrack
       display_name=Render Opacity
       sections [{
         type=MovieSceneFloatSection
         end_frame=40
         start_frame=0
         channels [None {type=MovieSceneScriptingFloatChannel, keys [[0,0],[22.8763,1],[29,1],[40.0333,0]]}]
        }]
      }
      {
       type=MovieScene2DTransformTrack
       sections [{
         type=MovieScene2DTransformSection
         end_frame=17
         start_frame=0
         channels [
          Translation.X {type=MovieSceneScriptingFloatChannel}
          Translation.Y {type=MovieSceneScriptingFloatChannel}
          Angle {type=MovieSceneScriptingFloatChannel}
          Scale.X {type=MovieSceneScriptingFloatChannel, keys [[0,0.4],[17.157,2]]}
          Scale.Y {type=MovieSceneScriptingFloatChannel, const [1,1]}
          Shear.X {type=MovieSceneScriptingFloatChannel}
          Shear.Y {type=MovieSceneScriptingFloatChannel}
         ]
        }]
      }
     ]
    }
   ]}
 }
]
graphs [
 DisplayPinSelect {nodes [
   n1 = DisplayPinSelect!() -> (then, Value:float=~) {is=title K2Node_FunctionEntry}
   > n2 = GetDynamicMaterial(self=Get LazyImage_PinColor() -> (LazyImage_PinColor:Fort Lazy Image Object Reference))
   > SetScalarParameterValue(self=@n2, ParameterName=BackingTypeSelector, Value=@n1)
  ]}
 EventGraph {nodes [
   n1 = SetQuestDescriptionText!() -> (OutputDelegate, then, Achieved:int, Required:int, Description:Text) {is=title K2Node_CustomEvent}
   n2 = "To Float (Integer)"(InInt=@n1.Achieved)
   n3 = "To Float (Integer)"(InInt=@n1.Required)
   n4 = Floor(A=float * float(A=float / float(A=@n2, B=@n3), B=100) {is=title K2Node_PromotableOperator}) {is=title K2Node_CallFunction}
   n5 = MakeLiteralText(Value="To Text (Integer)"(Value=@n1.Achieved))
   > SetText(
    self=Get Text_PinnedQuestDesc_RichText() -> (Text_PinnedQuestDesc_RichText:Common Rich Text Block Object Reference)
    InText=Select!(
     Default:"Text (by ref)"=Format Text!(
      Format="INVTEXT(\"<Progress>{Achieved}/{Required}</>  {Description}\")"
      Achieved:Text=@n5
      Required:int=@n1.Required
      Description:Text=@n1.Description
     ) -> (Result)
     Percentage:"Text (by ref)"=Format Text!(Format="INVTEXT(\"<Progress>{percent}% </>  {Description}\")", percent:int=@n4, Description=@n1.Description) -> (Result)
     Index:EQuestProgressStyle Enum=GetProgressStyle!(self:Athena Multi Pinned Quest Info Object Reference) -> (ReturnValue:EQuestProgressStyle Enum=Default)
    ) -> (ReturnValue:"Text (by ref)")
   )
   n6 = Event BP_OnSetupPinnedQuestInfo!() -> (
    OutputDelegate
    then
    PinIndex:int=0
    Achieved:int=0
    Required:int=0
    Description:"Text (by ref)"
    CustomBadge:"Soft Object Reference (by ref)"
   ) {is=title K2Node_Event}
   > SetQuestDescriptionText!(execute, self, Achieved:int=@n6.Achieved, Required:int=@n6.Required, Description:Text=@n6.Description) -> (then) {is=title K2Node_CallFunction}
   > SetIndexEntryOnPinnedQuest!(execute, self, IndexToPinnedQuest:int=@n6.PinIndex) -> (then)
   > n7 = Sequence()
   > SetVisibility(
    self=Get Image_CQ() -> (Image_CQ:Image Object Reference)
    InVisibility=Select!(
     Option 0:ESlateVisibility Enum=Collapsed
     Option 1:ESlateVisibility Enum=SelfHitTestInvisible
     Index:bool=IsValidSoftObjectReference(SoftObjectReference=@n6.CustomBadge)
    ) -> (ReturnValue:ESlateVisibility Enum)
   )
   > n8 = Async Load Asset(Asset=@n6.CustomBadge)
   n9 = Cast To Texture2D(execute=@n8.Completed, Object=@n8)
   > SetBrushFromTexture!(execute, self:Image Object Reference=Get Image_CQ(), Texture:Texture 2D Object Reference=@n9, bMatchSize:bool=false) -> (then)
   n10 = Cast To MaterialInterface!(execute=@n9.CastFailed, Object=@n8) -> (then, CastFailed, AsMaterial Interface:Material Interface Object Reference) {is=title K2Node_DynamicCast}
   > SetBrushFromMaterial(self=Get Image_CQ(), Material=@n10)
   SetMobileTextWrap(execute=@n7.then_1)
   > Delay(Duration=0.3)
   > PlayAnimation(InAnimation=Get Intro())
   > Branch(Condition=IsUsingCompactUI())
   > SetRenderOpacity(self=Get Text_PinnedQuestDesc_RichText(), InOpacity=1)
   Event BP_OnQuestCompleted() {is=title K2Node_Event}
   > SetVisibility(self=Get Image_CQ(), InVisibility=Collapsed)
   n11 = Get Outro() -> (Outro:Widget Animation Object Reference)
   > PlayAnimation(InAnimation=@n11)
   > Delay(Duration=float + float(A=GetEndTime(self=@n11), B=Get CompleteExtraDelay() -> (CompleteExtraDelay:double)))
   > ReleasePinnedQuestDelay(self:Athena Multi Pinned Quest Info Object Reference)
   n12 = Event BP_OnQuestUpdated!() -> (OutputDelegate, then, TotalObjectives:int=0, CompletedObjectives:int=0, Description:"Text (by ref)") {is=title K2Node_Event}
   > SetQuestDescriptionText(Achieved=@n12.CompletedObjectives, Required=@n12.TotalObjectives, Description=@n12.Description)
   n13 = Event BP_OnSetRewardValue!() -> (OutputDelegate, then, RewardValue:int=0, RewardType:EFortEarnedSubRewardType Enum=XP) {is=title K2Node_Event}
   PrintString!(
    execute
    InString=Append(A=MakeLiteralString(Value="EARNED XP "), B="To String (Integer)"(InInt=@n13.RewardValue))
    bPrintToScreen
    bPrintToLog
    TextColor=(R=0.0943,G=1,B=0.0458,A=1)
    Duration=5
    Key
   ) -> (then)
   Event BP_OnFinishRewardAnim() {is=title K2Node_Event}
  ]}
 GetPinnedQuestTextColor {nodes [
   GetPinnedQuestTextColor() {is=title K2Node_FunctionEntry}
   > Return Node!(execute, TextColor:String=B2DDF3)
  ]}
 SetIndexEntryOnPinnedQuest {nodes [
   n1 = SetIndexEntryOnPinnedQuest!() -> (then, IndexToPinnedQuest:int) {is=title K2Node_FunctionEntry}
   > n2 = Branch(Condition="Not Equal (Integer)"(A=@n1, B=-1))
   > n3 = Increment Int(Value=Get IndexToPinnedQuest() -> (IndexToPinnedQuest:int=0))
   > DisplayPinSelect!(execute, self, Value:float="To Float (Integer)"(InInt=@n3)) -> (then) {is=title K2Node_CallFunction}
   DisplayPinSelect(execute=@n2.else, Value=Get StartValueParam() -> (StartValueParam:double))
  ]}
 SetMobileTextWrap {nodes [
   SetMobileTextWrap() {is=title K2Node_FunctionEntry}
   > Branch(Condition=IsMobileGame())
   > n1 = Switch on ECommonInputType!(execute, Selection:ECommonInputType Enum=GetCurrentInputType(self=Get CommonInputSubsystem())) -> (MouseAndKeyboard:Exec, Gamepad:Exec, Touch:Exec, Count:Exec) {is=title K2Node_SwitchEnum}
   SetMaxDesiredHeight!(
    execute=@n1.Gamepad
    self:Size Box Object Reference=Get SizeBox_All() -> (SizeBox_All:Size Box Object Reference)
    InMaxDesiredHeight:float=70
   ) -> (then)
   SetMaxDesiredHeight(execute=@n1.Touch, self=Get SizeBox_All(), InMaxDesiredHeight=70)
   > SetMaxDesiredWidth!(execute, self:Size Box Object Reference=Get SizeBox_All(), InMaxDesiredWidth:float=325) -> (then)
  ]}
]
variables [
 CompleteExtraDelay {type=number, default=3}
 IsCQ {type=boolean}
 StartValueParam {type=number, default=4}
]
widgets {
 info {inheritedWidgetCount=4, parentClass=/Script/FortniteUI.AthenaMultiPinnedQuestInfo, rootWidgetClass=&Overlay, widgetCount=18}
 widgets [
  Overlay_3 {widgetClassPath=&Overlay, widget=Overlay_3}
  Image_AdditiveHorizontalBeam {
   widgetClassPath=&Image
   bIsVariable=true
   parent=Overlay_3
   widget=Image_AdditiveHorizontalBeam
   properties {
    renderOpacity=0
    visibility=SelfHitTestInvisible
    brush {
     bIsDynamicallyLoaded=false
     drawAs=Image
     imageType=FullColor
     mirroring=NoMirror
     resourceName=None
     resourceObject=/Game/Athena/HUD/MiniMap/Art/M_UI_PinnedQuest_Rift
     tiling=NoTile
     imageSize=<500 80>
     margin=<ltrb 0 0 0 0>
     outlineSettings {
      bUseBrushTransparency=false
      roundingType=HalfHeightRadius
      width=0
      color {colorUseRule=UseColor_Specified, specifiedColor=<rgba 0 0 0 0>}
      cornerRadii=<0 0 0 0>
     }
     tintColor {colorUseRule=UseColor_Specified, specifiedColor=<rgba 1 1 1 0.5>}
     uVRegion {bIsValid=false, max=<0 0>, min=<0 0>}
    }
    renderTransform {angle=0, scale=<1.1192 1>, shear=<0 0>, translation=<-120 0>}
    renderTransformPivot=<0 0.5>
   }
   slot {class=OverlaySlot, properties {verticalAlignment=VAlign_Center}}
  }
  SizeBox_All {
   widgetClassPath=&SizeBox
   bIsVariable=true
   parent=Overlay_3
   widget=SizeBox_All
   properties {bOverride_MaxDesiredHeight=true, bOverride_MaxDesiredWidth=true, clipping=ClipToBounds, maxDesiredHeight=68, maxDesiredWidth=600}
   slot {class=OverlaySlot, properties {verticalAlignment=VAlign_Center}}
  }
  Overlay_1 {widgetClassPath=&Overlay, parent=SizeBox_All, widget=Overlay_1, slot {class=SizeBoxSlot}}
  Image_Complete_BG {
   widgetClassPath=&FortLazyImage
   bIsVariable=true
   parent=Overlay_1
   widget=Image_Complete_BG
   properties {
    renderOpacity=0
    brush {
     bIsDynamicallyLoaded=false
     drawAs=Image
     imageType=FullColor
     mirroring=NoMirror
     resourceName=None
     resourceObject=/Game/Athena/HUD/MiniMap/Art/MI_UI_PinnedQuest_Complete
     tiling=NoTile
     imageSize=<500 32>
     margin=<ltrb 0 0 0 0>
     outlineSettings {
      bUseBrushTransparency=false
      roundingType=HalfHeightRadius
      width=0
      color {colorUseRule=UseColor_Specified, specifiedColor=<rgba 0 0 0 0>}
      cornerRadii=<0 0 0 0>
     }
     tintColor {colorUseRule=UseColor_Specified, specifiedColor=<rgba 1 1 1 1.3503>}
     uVRegion {bIsValid=false, max=<0 0>, min=<0 0>}
    }
   }
   slot {class=OverlaySlot, properties {horizontalAlignment=HAlign_Fill, verticalAlignment=VAlign_Fill}}
  }
  HorizontalBox_0 {
   widgetClassPath=&HorizontalBox
   parent=Overlay_1
   widget=HorizontalBox_0
   slot {class=OverlaySlot, properties {horizontalAlignment=HAlign_Fill, verticalAlignment=VAlign_Fill}}
  }
  Overlay_IndexPin {
   widgetClassPath=&Overlay
   parent=HorizontalBox_0
   widget=Overlay_IndexPin
   slot {class=HorizontalBoxSlot, properties {horizontalAlignment=HAlign_Center, verticalAlignment=VAlign_Top, padding=<ltrb 8 0 5 0>}}
  }
  Switcher_PinImages {
   widgetClassPath=&CommonVisibilitySwitcher
   bInherited=true
   parent=Overlay_IndexPin
   widget=Switcher_PinImages
   slot {class=OverlaySlot, properties {horizontalAlignment=HAlign_Center, verticalAlignment=VAlign_Center}}
  }
  LazyImage_PinColor {
   widgetClassPath=&FortLazyImage
   bIsVariable=true
   parent=Switcher_PinImages
   widget=LazyImage_PinColor
   properties {
    renderOpacity=0.4133
    brush {
     bIsDynamicallyLoaded=false
     drawAs=Image
     imageType=FullColor
     mirroring=NoMirror
     resourceName=None
     resourceObject=/Game/Athena/HUD/Materials/MI_UI_QuestPin_HUD
     tiling=NoTile
     imageSize=<32 32>
     margin=<ltrb 0 0 0 0>
     outlineSettings {
      bUseBrushTransparency=false
      roundingType=HalfHeightRadius
      width=0
      color {colorUseRule=UseColor_Specified, specifiedColor=<rgba 0 0 0 0>}
      cornerRadii=<0 0 0 0>
     }
     tintColor {colorUseRule=UseColor_Specified, specifiedColor=<rgba 1 1 1 0.7>}
     uVRegion {bIsValid=false, max=<0 0>, min=<0 0>}
    }
   }
   slot {class=CommonVisibilitySwitcherSlot}
  }
  LazyImage_PinCustom {
   widgetClassPath=&FortLazyImage
   bInherited=true
   bIsVariable=true
   parent=Switcher_PinImages
   widget=LazyImage_PinCustom
   slot {class=CommonVisibilitySwitcherSlot}
  }
  Image_Check {
   widgetClassPath=&Image
   bIsVariable=true
   parent=Overlay_IndexPin
   widget=Image_Check
   properties {
    renderOpacity=0
    visibility=Collapsed
    brush {
     bIsDynamicallyLoaded=false
     drawAs=Image
     imageType=FullColor
     mirroring=NoMirror
     resourceName=None
     resourceObject=/Game/Athena/HUD/Art/T_UI_Icon_Check_MP_64_Pinned
     tiling=NoTile
     imageSize=<40 40>
     margin=<ltrb 0 0 0 0>
     outlineSettings {
      bUseBrushTransparency=false
      roundingType=HalfHeightRadius
      width=0
      color {colorUseRule=UseColor_Specified, specifiedColor=<rgba 0 0 0 0>}
      cornerRadii=<0 0 0 0>
     }
     tintColor {colorUseRule=UseColor_Specified, specifiedColor=<rgba 1 1 1 1>}
     uVRegion {bIsValid=false, max=<0 0>, min=<0 0>}
    }
    colorAndOpacity=<rgba 0.0343 1 0.227 1>
   }
   slot {class=OverlaySlot, properties {horizontalAlignment=HAlign_Right, verticalAlignment=VAlign_Center}}
  }
  Image_CQ {
   widgetClassPath=&Image
   bIsVariable=true
   parent=HorizontalBox_0
   widget=Image_CQ
   properties {
    renderOpacity=0
    visibility=SelfHitTestInvisible
    brush {
     bIsDynamicallyLoaded=false
     drawAs=Image
     imageType=FullColor
     mirroring=NoMirror
     resourceName=None
     resourceObject=/Game/Athena/UI/Challenges/Art/T_UI_CommunityQuest_64
     tiling=NoTile
     imageSize=<34 34>
     margin=<ltrb 0 0 0 0>
     outlineSettings {
      bUseBrushTransparency=false
      roundingType=HalfHeightRadius
      width=0
      color {colorUseRule=UseColor_Specified, specifiedColor=<rgba 0 0 0 0>}
      cornerRadii=<0 0 0 0>
     }
     tintColor {colorUseRule=UseColor_Specified, specifiedColor=<rgba 1 1 1 1>}
     uVRegion {bIsValid=false, max=<0 0>, min=<0 0>}
    }
   }
   slot {class=HorizontalBoxSlot, properties {horizontalAlignment=HAlign_Center, verticalAlignment=VAlign_Center}}
  }
  Overlay_2 {
   widgetClassPath=&Overlay
   parent=HorizontalBox_0
   widget=Overlay_2
   properties {bOverride_Cursor=true}
   slot {class=HorizontalBoxSlot, properties {verticalAlignment=VAlign_Center, padding=<ltrb 6 0 5 0>, size {sizeRule=Fill, value=1}}}
  }
  Horizontal_Complete {
   widgetClassPath=&HorizontalBox
   parent=Overlay_2
   widget=Horizontal_Complete
   slot {class=OverlaySlot, properties {verticalAlignment=VAlign_Center}}
  }
  Text_Complete {
   widgetClassPath=&UEFN_TextBlock
   parent=Horizontal_Complete
   widget=Text_Complete
   properties {
    renderOpacity=0
    text=Quest Complete
    visibility=SelfHitTestInvisible
    colorAndOpacity {colorUseRule=UseColor_Specified, specifiedColor=<rgba 1 1 1 0.8>}
    font {
     bForceMonospaced=false
     bMaterialIsStencil=false
     fontMaterial=null
     fontObject=&HeadingNow
     letterSpacing=0
     monospacedWidth=1
     size=18.75
     skewAmount=0
     typefaceFontName=86 Bold Italic
     outlineSettings {
      bApplyOutlineToDropShadows=false
      bMiteredCorners=false
      bSeparateFillAlpha=false
      outlineMaterial=null
      outlineSize=0
      outlineColor=<rgba 0 0 0 1>
     }
    }
    shadowColorAndOpacity=<rgba 0 0 0 0.15>
    shadowOffset=<3 2>
   }
   slot {class=HorizontalBoxSlot, properties {horizontalAlignment=HAlign_Center, verticalAlignment=VAlign_Center}}
  }
  Text_PinnedQuestDesc_RichText {
   widgetClassPath=&CommonRichTextBlock
   bIsVariable=true
   parent=Overlay_2
   widget=Text_PinnedQuestDesc_RichText
   properties {
    autoWrapText=true
    bIsScrollingEnabled=false
    mobileTextBlockScale=1.16
    text="<Progress>0/1</> <Default>Search an Olympus Chest or Underworld Chest testing longer text to see what happens,</>"
    textOverflowPolicy=MultilineEllipsis
    textStyleSet=/Game/Athena/HUD/TextStyles/PinnedQuestTextStyle
    decoratorClasses [/Game/Athena/HUD/QuestIcons/BP_QuestTile_IconDecorator.BP_QuestTile_IconDecorator_C]
    defaultTextStyleOverride {
     overflowPolicy=MultilineEllipsis
     transformPolicy=None
     colorAndOpacity {colorUseRule=UseColor_Specified, specifiedColor=<rgba 1 1 1 1>}
     font {
      bForceMonospaced=false
      bMaterialIsStencil=false
      fontMaterial=null
      fontObject=&HeadingNow
      letterSpacing=0
      monospacedWidth=1
      size=12
      skewAmount=0
      typefaceFontName=74 Regular
      outlineSettings {
       bApplyOutlineToDropShadows=false
       bMiteredCorners=false
       bSeparateFillAlpha=false
       outlineMaterial=null
       outlineSize=1
       outlineColor=<rgba 0 0 0 0.25>
      }
     }
     highlightColor {colorUseRule=UseColor_Specified, specifiedColor=<rgba 0 0 0 1>}
     highlightShape {
      bIsDynamicallyLoaded=false
      drawAs=Image
      imageType=NoImage
      mirroring=NoMirror
      resourceName=None
      resourceObject=null
      tiling=NoTile
      imageSize=<32 32>
      margin=<ltrb 0 0 0 0>
      outlineSettings {
       bUseBrushTransparency=false
       roundingType=HalfHeightRadius
       width=0
       color {colorUseRule=UseColor_Specified, specifiedColor=<rgba 0 0 0 0>}
       cornerRadii=<0 0 0 0>
      }
      tintColor {colorUseRule=UseColor_Specified, specifiedColor=<rgba 1 1 1 1>}
      uVRegion {bIsValid=false, max=<0 0>, min=<0 0>}
     }
     selectedBackgroundColor {colorUseRule=UseColor_Foreground, specifiedColor=<rgba 1 0 1 1>}
     shadowColorAndOpacity=<rgba 0 0 0 1>
     shadowOffset=<0 0>
     strikeBrush {
      bIsDynamicallyLoaded=false
      drawAs=Image
      imageType=NoImage
      mirroring=NoMirror
      resourceName=None
      resourceObject=null
      tiling=NoTile
      imageSize=<32 32>
      margin=<ltrb 0 0 0 0>
      outlineSettings {
       bUseBrushTransparency=false
       roundingType=HalfHeightRadius
       width=0
       color {colorUseRule=UseColor_Specified, specifiedColor=<rgba 0 0 0 0>}
       cornerRadii=<0 0 0 0>
      }
      tintColor {colorUseRule=UseColor_Specified, specifiedColor=<rgba 1 1 1 1>}
      uVRegion {bIsValid=false, max=<0 0>, min=<0 0>}
     }
     underlineBrush {
      bIsDynamicallyLoaded=false
      drawAs=Image
      imageType=NoImage
      mirroring=NoMirror
      resourceName=None
      resourceObject=null
      tiling=NoTile
      imageSize=<32 32>
      margin=<ltrb 0 0 0 0>
      outlineSettings {
       bUseBrushTransparency=false
       roundingType=HalfHeightRadius
       width=0
       color {colorUseRule=UseColor_Specified, specifiedColor=<rgba 0 0 0 0>}
       cornerRadii=<0 0 0 0>
      }
      tintColor {colorUseRule=UseColor_Specified, specifiedColor=<rgba 1 1 1 1>}
      uVRegion {bIsValid=false, max=<0 0>, min=<0 0>}
     }
    }
   }
   slot {class=OverlaySlot, properties {horizontalAlignment=HAlign_Fill, verticalAlignment=VAlign_Fill}}
  }
  Image_PinCustom {widgetClassPath=&Image, bInherited=true, widget=null}
  TutorialHighlight {widgetClassPath=&TutorialHighlight}
 ]
}
