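# Athena HUD widget that lists the player's pinned quests above externally hosted content and shows a "Quests" keybind hint.
# The hint shows only in Default/Arena playlists with the CVar enabled, a valid key and pinned entries, re-evaluated on rebuilds and keybind changes.
# A teammate event switches the list between one and three entries.
# It registers as a tagged widget host and adjusts spacing for compact UI and Mash touch play.
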WidgetBlueprint /Game/Athena/HUD/PinnedContentList
graphs [
 CollapseShowButton {nodes [
   n1 = CollapseShowButton!() -> (then, SetCollapse:bool=~) {is=title K2Node_FunctionEntry}
   > SetVisibility(
    self=Get Horizontal_TrackedQuest_DeepLink() -> (Horizontal_TrackedQuest_DeepLink:Horizontal Box Object Reference)
    InVisibility=Select(Option 0=HitTestInvisible, Option 1=Collapsed, Index=@n1)
   )
  ]}
 EventGraph {nodes [
   n1 = HandleKeybindsChanged() {is=title K2Node_CustomEvent}
   n2 = Get BoundKey!(self:Fort Keybind Widget Object Reference=Get KeybindWidget_QuestMenu() -> (KeybindWidget_QuestMenu:Keybind Widget Object Reference)) -> (BoundKey:Key Structure)
   n3 = GetNumericCVar!(CVar:String=Fort.EnableOpenQuestsInputAction) -> (OutValue:float, ReturnValue:bool)
   n4 = "Equal (Float)"(A=@n3, B=1)
   n5 = Branch(
    execute=[@n1.then, @n7.then]
    Condition=AND Boolean!(
     A=Get IsBR() -> (IsBR:bool)
     B=Select(Option 0:bool=false, Option 1:bool=@n4, Index=@n3.ReturnValue) -> (ReturnValue:bool)
     C:bool=Is Valid Key!(Key:"Key Structure (by ref)"=@n2) -> (ReturnValue:bool)
    ) -> (ReturnValue)
   )
   n6 = Get FortDynamicEntryBox_PinnedQuests() -> (FortDynamicEntryBox_PinnedQuests:Fort Dynamic Entry Box Object Reference)
   > CollapseShowButton!(execute, self, SetCollapse:bool="Equal (Integer)"(A=GetNumEntries!(self:Dynamic Entry Box Base Object Reference=@n6) -> (ReturnValue:int=0))) -> (then) {is=title K2Node_CallFunction}
   CollapseShowButton(execute=@n5.else, SetCollapse=true)
   n7 = Event OnRebuildPinnedQuest() {is=title K2Node_Event}
   Event Construct()
   > RegisterWidgetHost!(
    execute
    WidgetHost:Widget Object Reference
    WidgetHostTag:"Gameplay Tag Structure (by ref)"=Get PinnedContentListTag() -> (PinnedContentListTag:Gameplay Tag Structure)
    IsPlayerSpecific:bool=true
   ) -> (then)
   > Bind Event to On Keybinds Changed(
    self:Fort Global UIContext Object Reference=Get FortGlobalUIContext() -> (ReturnValue:Fort Global UIContext Object Reference) {is=title K2Node_GetSubsystem}
    Delegate=@n1
   ) {is=title K2Node_AddDelegate}
   > n8 = PlaylistDataReadyAsync!(execute:Exec) -> (
    then:Exec
    OnReady:Exec
    GameState:Fort Game State Athena Object Reference
    Playlist:Fort Playlist Object Reference
    PlaylistContextTags:"Gameplay Tag Container Structure (by ref)"
   ) {is=title K2Node_AsyncAction}
   Set IsBR(
    execute=@n8.OnReady
    IsBR:bool=HasAnyTags!(
     TagContainer:"Gameplay Tag Container Structure (by ref)"=@n8.PlaylistContextTags
     OtherContainer:"Gameplay Tag Container Structure (by ref)"=MakeLiteralGameplayTagContainer!(Value:Gameplay Tag Container Structure=(GameplayTags=((TagName="Athena.Playlist.Default"),(TagName="Athena.Playlist.Arena")))) -> (ReturnValue:Gameplay Tag Container Structure)
     bExactMatch:bool=false
    ) -> (ReturnValue:bool)
   ) -> (Output_Get:bool)
   n9 = GetGlobalEventRouter()
   > n10 = StartListeningToEvent!(
    execute:Exec
    Target:Gameplay Event Router Component Object Reference=@n9
    EventType:Script Struct Object Reference=&/TeammateStruct
    Context:Object Reference=GetOwningPlayer()
    EventBubblingRule:EEventBubblingRule Enum
    None:%=@n9
   ) -> (
    then:Exec
    AsyncTaskProxy:Async Action Start Listening To Event Object Reference
    OnEventReceived:Exec
    Payload_IsActive_1_#1:bool=~
   )
   > Set Async Task Listen for Count Limit(Async Task Listen for Count Limit:Async Action Start Listening To Event Object Reference=@n10) -> (Output_Get:Async Action Start Listening To Event Object Reference)
   n11 = GetOwningPlayer()
   > Branch(Condition=AND Boolean(
     A=IsMobileGame()
     B=IsUsingTouch(PlayerController=@n11)
     C=HasTag!(
      TagContainer:"Gameplay Tag Container Structure (by ref)"=@n8.PlaylistContextTags
      Tag:Gameplay Tag Structure=(TagName="Athena.Playlist.Mash")
      bExactMatch:bool=false
     ) -> (ReturnValue:bool)
    ))
   > SetPadding!(
    execute
    self:Vertical Box Slot Object Reference=SlotAsVerticalBoxSlot(Widget=Get Horizontal_TrackedQuest_DeepLink())
    InPadding_Left:float=0
    InPadding_Top:float=20
    InPadding_Right:float=0
    InPadding_Bottom:float=0
   ) -> (then)
   SetPinnedAmmount!(
    execute=@n10.OnEventReceived
    self
    AmmountEntries:int=Select(Option 0:int=3, Option 1:int=1, Index=@n10.Payload_IsActive_1_#1) -> (ReturnValue:int)
   ) -> (then)
   Event Destruct()
   > UnregisterWidgetHost!(execute, WidgetHost:Widget Object Reference, WidgetHostTag:"Gameplay Tag Structure (by ref)"=Get PinnedContentListTag()) -> (then)
   > Unbind Event from On Keybinds Changed(self:Fort Global UIContext Object Reference=Get FortGlobalUIContext(), Delegate=@n1) {is=title K2Node_RemoveDelegate}
   > StopListeningToEvent(self:Async Action Start Listening To Event Object Reference=Get Async Task Listen for Count Limit() -> (Async Task Listen for Count Limit:Async Action Start Listening To Event Object Reference))
   > Set Async Task Listen for Count Limit()
   Event PreConstruct()
   > SetMobileStyle() {is=title K2Node_CallFunction}
   n12 = Event Add Content To Host Widget!() -> (
    OutputDelegate
    then
    WidgetHostTag:"Gameplay Tag Structure (by ref)"
    ContentWidget:Widget Object Reference
    Padding:"Margin Structure (by ref)"
    bFill:bool
   ) {is=title K2Node_Event}
   > Branch(Condition="Equal (GameplayTag)"(A=Get PinnedContentListTag(), B=@n12.WidgetHostTag))
   > n13 = AddChildToVerticalBox(self=Get VerticalBox_PinnedContentList() -> (VerticalBox_PinnedContentList:Vertical Box Object Reference), Content=@n12.ContentWidget)
   > SetHorizontalAlignment(self:Vertical Box Slot Object Reference=@n13, InHorizontalAlignment=HAlign_Fill)
   > SetVerticalAlignment(self:Vertical Box Slot Object Reference=@n13, InVerticalAlignment=VAlign_Fill)
  ]}
 RemoveContentFromWidgetHost {nodes [
   n1 = Remove Content From Host Widget!() -> (
    then
    WidgetHostTag:"Gameplay Tag Structure (by ref)"
    ContentWidget:Widget Object Reference
    RemovalType:EWidgetSlottingRemovalType Enum=Unparent
   ) {is=title K2Node_FunctionEntry}
   > Branch(Condition="Equal (GameplayTag)"(A=@n1, B=Get PinnedContentListTag()))
   > n2 = RemoveChild(self=Get VerticalBox_PinnedContentList(), Content=@n1.ContentWidget)
   > Return Node!(execute, ReturnValue:bool=@n2)
  ]}
 SetMobileStyle {nodes [
   SetMobileStyle() {is=title K2Node_FunctionEntry}
   > Branch(Condition=IsUsingCompactUI())
   > SetEntrySpacing!(
    execute
    self:Dynamic Entry Box Base Object Reference=Get FortDynamicEntryBox_PinnedQuests()
    InEntrySpacing_X:double=0
    InEntrySpacing_Y:double=-15
   ) -> (then)
  ]}
 SetPinnedAmmount {nodes [
   n1 = SetPinnedAmmount!() -> (then, AmmountEntries:int) {is=title K2Node_FunctionEntry}
   > n2 = Branch(Condition="integer > integer"(A=@n1, B=3))
   > n3 = Set NumEntries(NumEntries:int=3) -> (Output_Get:int=0)
   n4 = Set NumEntries(execute=@n2.else, NumEntries=Get AmmountEntries() -> (AmmountEntries:int=0)) -> (Output_Get=0)
   SetMaxQuestEntriesToShow!(execute=[@n3.then, @n4.then], self:Fort Pinned Content List Object Reference, MaxDisplayedEntries:int=@n4) -> (then)
   > RebuildPinnedQuestsList(self:Fort Pinned Content List Object Reference)
  ]}
]
variables [
 Async Task Listen for Count Limit {type=AsyncAction_StartListeningToEvent, default=null}
 IsBR {type=boolean}
 NumEntries {type=integer, default=3}
 PinnedContentListTag {type=GameplayTag, default {tagName=DynamicUI.HUD.Fort.Athena.Tracking.PinnedContentList.Slot}}
]
widgets {
 info {inheritedWidgetCount=3, parentClass=/Script/FortniteUI.FortPinnedContentList, rootWidgetClass=&VerticalBox, widgetCount=7}
 widgets [
  VerticalBox_PinnedContentList {widgetClassPath=&VerticalBox, bInherited=true, bIsVariable=true, widget=VerticalBox_PinnedContentList}
  Horizontal_TrackedQuest_DeepLink {
   widgetClassPath=&HorizontalBox
   bIsVariable=true
   parent=VerticalBox_PinnedContentList
   widget=Horizontal_TrackedQuest_DeepLink
   slot {class=VerticalBoxSlot, properties {horizontalAlignment=HAlign_Left}}
  }
  SizeBox_0 {
   widgetClassPath=&SizeBox
   parent=Horizontal_TrackedQuest_DeepLink
   widget=SizeBox_0
   properties {bOverride_MaxDesiredHeight=true, bOverride_MaxDesiredWidth=true, maxDesiredHeight=35, maxDesiredWidth=35}
   slot {
    class=HorizontalBoxSlot
    properties {horizontalAlignment=HAlign_Left, verticalAlignment=VAlign_Center, padding=<ltrb 10 0 0 0>, size {sizeRule=Fill, value=1}}
   }
  }
  KeybindWidget_QuestMenu {
   widgetClassPath=/FortUILibrary/Input/KeybindWidget
   bIsVariable=true
   parent=SizeBox_0
   widget=KeybindWidget_QuestMenu
   properties {boundAction=OpenQuests, forcedHoldKeybindStatus=ForcedHold, boundActionFallbacks [GamepadOpenQuests]}
   slot {class=SizeBoxSlot}
  }
  Text_Complete {
   widgetClassPath=&UEFN_TextBlock
   parent=Horizontal_TrackedQuest_DeepLink
   widget=Text_Complete
   properties {
    text=" Quests"
    visibility=SelfHitTestInvisible
    colorAndOpacity {colorUseRule=UseColor_Specified, specifiedColor=<rgba 1 1 1 0.9>}
    font {
     bForceMonospaced=false
     bMaterialIsStencil=false
     fontMaterial=null
     fontObject=&HeadingNow
     letterSpacing=0
     monospacedWidth=1
     size=13.5
     skewAmount=0
     typefaceFontName=75 Medium
     outlineSettings {
      bApplyOutlineToDropShadows=false
      bMiteredCorners=false
      bSeparateFillAlpha=false
      outlineMaterial=null
      outlineSize=0
      outlineColor=<rgba 0 0 0 1>
     }
    }
    shadowColorAndOpacity=<rgba 0 0 0 0.15>
    shadowOffset=<3 2>
   }
   slot {class=HorizontalBoxSlot, properties {horizontalAlignment=HAlign_Center, verticalAlignment=VAlign_Center, padding=<ltrb 5 0 0 0>}}
  }
  FortDynamicEntryBox_PinnedQuests {
   widgetClassPath=/Script/FortniteUI.FortDynamicEntryBox
   bInherited=true
   bIsVariable=true
   parent=VerticalBox_PinnedContentList
   widget=FortDynamicEntryBox_PinnedQuests
   properties {
    entryBoxType=Vertical
    entryHorizontalAlignment=HAlign_Left
    entryVerticalAlignment=VAlign_Top
    entryWidgetClass=&/WBP_PlayerPinnedQuestInfo.WBP_PlayerPinnedQuestInfo_C
    maximumEntries=3
    entrySpacing=<0 -23>
   }
   slot {class=VerticalBoxSlot, properties {horizontalAlignment=HAlign_Left, verticalAlignment=VAlign_Top, padding=<ltrb 0 -8 0 0>}}
  }
  TutorialHighlight {widgetClassPath=&TutorialHighlight}
 ]
}
