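# SimpleMaterialProgressBar is a HUD widget Blueprint of FortSimpleMaterialProgressBar that provides a material-driven progress bar.
# Its only widget is the inherited root image, Image_Bar, whose brush has no material or texture assigned here.
# The BoundAnim_FillBar animation drives that image's FillAmount material parameter from empty to full over about one second.
# The event graph is empty, leaving any progress logic to the parent class.
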WidgetBlueprint /Game/Athena/HUD/SimpleMaterialProgressBar
animations [BoundAnim_FillBar {
  binding_count=1
  length_seconds=0.9956
  object=&self:BoundAnim_FillBar
  sequence {bindings [Image_Bar {
     id=#1
     possessed_class=Image
     tracks [{
       type=MovieSceneWidgetMaterialTrack
       sections [{type=MovieSceneParameterSection, channels [FillAmount {type=MovieSceneScriptingFloatChannel, keys [[0,0],[19.9127,1]]}]}]
      }]
    }]}
 }]
graphs [EventGraph {links [], nodes []}]
variables []
widgets {
 info {inheritedWidgetCount=1, parentClass=/Script/FortniteUI.FortSimpleMaterialProgressBar, rootWidgetClass=&Image, widgetCount=1}
 widgets [Image_Bar {
   widgetClassPath=&Image
   bInherited=true
   bIsVariable=true
   widget=Image_Bar
   properties {
    bFlipForRightToLeftFlowDirection=true
    visibility=SelfHitTestInvisible
    brush {
     bIsDynamicallyLoaded=false
     drawAs=Box
     imageType=NoImage
     mirroring=NoMirror
     resourceName=None
     resourceObject=null
     tiling=NoTile
     imageSize=<32 32>
     margin=<ltrb 0.5 0.5 0.5 0.5>
     outlineSettings {
      bUseBrushTransparency=false
      roundingType=HalfHeightRadius
      width=0
      color {colorUseRule=UseColor_Specified, specifiedColor=<rgba 0 0 0 0>}
      cornerRadii=<0 0 0 0>
     }
     tintColor {colorUseRule=UseColor_Specified, specifiedColor=<rgba 1 1 1 1>}
     uVRegion {bIsValid=false, max=<0 0>, min=<0 0>}
    }
   }
  }]
}
