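# Sound cue for the respawn countdown in Athena limited-time modes, classed as non-critical in-game UI audio.
# A single wave player, wired straight to the root, reuses a Score LTM count pop sound at boosted volume.
# A concurrency override allows one instance at a time and stops the oldest, so each new trigger cuts off the previous one.
# Virtualization is disabled.
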SoundCue /Game/Athena/Sounds/LTMs/Respawn/LTM_Respawn_Countdown_Cue
root=WavePlayer_3
nodes [{id=WavePlayer_3, properties {soundWaveAssetPtr=/Game/Athena/Sounds/LTMs/Score/ScoreLTM_InGamePop_CountNormal_05}}]
properties {
 bOverrideConcurrency=true
 maxDistance=2097152
 soundClassObject=&Fort_UI_InGame_NonCritical
 virtualizationMode=Disabled
 volumeMultiplier=2
 concurrencyOverrides {maxCount=1, resolutionRule=StopOldest}
}
