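# Sound cue for a respawn in Athena limited-time modes, routed to the non-critical in-game UI sound class.
# A mixer layers a reused score count-final pop with a lobby teleport wave.
# An Enveloper bends the teleport's pitch upward over time while its volume stays flat.
# A concurrency override caps simultaneous plays by blocking new ones and sets a short retrigger time.
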SoundCue /Game/Athena/Sounds/LTMs/Respawn/LTM_Respawn_Cue
root=Mixer_0
nodes [
 {
  id=Enveloper_0
  children [WavePlayer_1]
  properties {
   pitchCurve {editorCurveData {keys [<at 0 0.8 RCIM_Cubic>, {arriveTangent=1.6573, leaveTangent=1.2179, time=0.6, value=1.2}, <at 2.5 1.9>]}}
   volumeCurve {editorCurveData {keys [<at 0 1>]}}
  }
 }
 {id=Mixer_0, children [WavePlayer_3,Enveloper_0], properties {inputVolume [1,1]}}
 {id=WavePlayer_1, properties {soundWaveAssetPtr=/Game/Sounds/Fort_Gadgets/Teleporter/Lobby_Teleport_01}}
 {id=WavePlayer_3, properties {soundWaveAssetPtr=/Game/Athena/Sounds/LTMs/Score/ScoreLTM_InGamePop_CountFinal_06}}
]
properties {
 bOverrideConcurrency=true
 maxDistance=2097152
 soundClassObject=&Fort_UI_InGame_NonCritical
 volumeMultiplier=0.5
 concurrencyOverrides {maxCount=2, resolutionRule=PreventNew, retriggerTime=0.4}
}
